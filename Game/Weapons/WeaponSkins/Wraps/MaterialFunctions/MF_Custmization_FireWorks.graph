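# Material function that adds animated 4th of July fireworks glow to customization wraps, behind the FireWorks static switch.
# Each RGB channel of a warp texture offsets a time-scrolled gradient lookup at its own speed, masked by the matching mask channel.
# The three layers share a time-cycled color tint and glow strength, then sum into Emissive.
# An optional side-projection mask uses one contrast-adjusted pre-skinned normal axis, chosen by vehicle and mech inputs.
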MaterialFunction /Game/Weapons/WeaponSkins/Wraps/MaterialFunctions/MF_Custmization_FireWorks
graph {nodes [
  n1 = GetMaterialAttributes(None=FunctionInput() {inputName=Material, inputType=FunctionInput_MaterialAttributes}) {attributeGetTypes [#1]}
  n2 = Time()
  n3 = Multiply(A=@n2, B=ScalarParameter() {defaultValue=-0.2, group=SideProjection_FireWorks, parameterName=FireWorks_Speed_RedChannel})
  n4 = FunctionInput(Preview=TextureCoordinate()) {inputName=Uvs, inputType=FunctionInput_Vector2}
  n5 = TextureSampleParameter2D(UVs=@n4) {
   group=SideProjection_FireWorks
   parameterName=FireWorks_TextureWarp
   samplerType=SAMPLERTYPE_LinearColor
   texture=/Game/Weapons/WeaponSkins/Wraps/Textures/4thofJuly/T_wrap_4thofJuly_M1
  }
  n6 = Multiply(A=@n5.R) {constA=0.5, constB=0.5}
  n7 = TextureObjectParameter() {
   group=SideProjection_FireWorks
   parameterName=FireWorks_TextureGrad
   texture=/Game/Weapons/WeaponSkins/Wraps/Textures/4thofJuly/T_wrap_4thofJuly_Grad_BC
  }
  n8 = TextureSampleParameter2D(UVs=@n4) {
   group=SideProjection_FireWorks
   parameterName=FireWorks_TextureMask
   texture=/Game/Weapons/WeaponSkins/Wraps/Textures/4thofJuly/T_wrap_4thofJuly_M2
  }
  n9 = Multiply(
   A=TextureSample(UVs=Add(A=@n3, B=@n6), Tex=@n7) {texture=&T_wrap_4thofJuly_Grad_BC}.RGB
   B=Saturate(None=@n8.R)
  )
  n10 = TextureSampleParameter2D(UVs=Multiply(A=@n2, B=ScalarParameter() {defaultValue=0.5, group=SideProjection_FireWorks, parameterName=FireWorks_Speed_ColorGrad})) {
   group=SideProjection_FireWorks
   parameterName=FireWorks_TextureColorGrad
   texture=/Game/Global/Textures/Default/Masks_and_Gradients/3pxRGBImage
  }
  n11 = Multiply(A=@n10.RGB, B=ScalarParameter() {defaultValue=100, group=SideProjection_FireWorks, parameterName=FireWorks_GlowStrength}) {constB=100}
  n12 = Multiply(A=@n2, B=ScalarParameter() {defaultValue=-0.5, group=SideProjection_FireWorks, parameterName=FireWorks_Speed_GreenChannel})
  n13 = Multiply(A=@n5.G) {constA=0.5, constB=0.5}
  n14 = Multiply(
   A=TextureSample(UVs=Add(A=@n12, B=@n13), Tex=@n7) {texture=&T_wrap_4thofJuly_Grad_BC}.RGB
   B=Saturate(None=@n8.G)
  )
  n15 = ScalarParameter() {defaultValue=-0.35, group=SideProjection_FireWorks, parameterName=FireWorks_Speed_BlueChannel}
  n16 = TextureSample(UVs=Add(A=Multiply(A=@n2, B=@n15), B=Multiply(A=@n5.B) {constB=0.5}), Tex=@n7) {texture=&T_wrap_4thofJuly_Grad_BC}
  n17 = Saturate(None=@n8.B)
  n18 = Add(A=Add(A=Multiply(A=@n9, B=@n11), B=Multiply(A=@n14, B=@n11)), B=Multiply(A=Multiply(A=@n16.RGB, B=@n17), B=@n11))
  n19 = FunctionInput(Preview=VertexInterpolator(VS=PreSkinnedNormal()).PS) {bUsePreviewValueAsDefault=true, inputName=Pre-Skinned Local Normal, sortPriority=42}
  n20 = ScalarParameter() {group=SideProjection_FireWorks, parameterName=FireWorks_TriplanerMask_Contrast}
  n21 = MaterialFunctionCall(In=Abs(None=ComponentMask(None=@n19) {r=true}), Contrast=@n20) {materialFunction=&CheapContrast}
  n22 = ComponentMask(None=@n19) {b=true}
  n23 = ComponentMask(None=MaterialFunctionCall(In=Abs(None=@n22), Contrast=@n20) {materialFunction=&CheapContrast}.Result) {r=true}
  n24 = ComponentMask(None=@n19) {g=true}
  n25 = ComponentMask(None=MaterialFunctionCall(In=Abs(None=@n24), Contrast=@n20) {materialFunction=&CheapContrast}.Result) {r=true}
  n26 = FunctionInput() {bUsePreviewValueAsDefault=true, inputName=Side Project Vehicle, inputType=FunctionInput_Scalar}
  n27 = Constant()
  n28 = LinearInterpolate(
   A=LinearInterpolate(
    A=ComponentMask(None=@n21.Result) {r=true}
    B=LinearInterpolate(A=@n23, B=@n25, Alpha=@n26)
    Alpha=FunctionInput(Preview=@n27) {inputName=isVehicle, inputType=FunctionInput_Scalar}
   )
   B=@n25
   Alpha=FunctionInput() {bUsePreviewValueAsDefault=true, inputName=isMech, inputType=FunctionInput_Scalar}
  )
  n29 = Add(
   A=@n1.EmissiveColor
   B=StaticSwitchParameter(True=LinearInterpolate(B=@n18, Alpha=@n28), False=@n18) {group=SideProjection_FireWorks, parameterName=FireWorks_TriplanerMask}
  )
  FunctionOutput(None=SetMaterialAttributes!(
    MaterialAttributes=@n1.MaterialAttributes
    EmissiveColor=StaticSwitchParameter(True=@n29, False=@n1.EmissiveColor) {group=Customization, parameterName=FireWorks}
   ) {attributeSetTypes [#1]})
 ]}
