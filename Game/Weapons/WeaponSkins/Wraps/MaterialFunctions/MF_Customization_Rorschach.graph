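# Material function for Rorschach weapon-skin wraps that outputs an animated, mirrored inkblot-style mask.
# MF_Projection_Side_Normal builds UVs, with rotation, offset and scale blended by isState between weapon, vehicle, vehicle-side and mech parameters.
# A time-scrolled warp texture distorts mask lookups on normal and U-negated UVs, hard-split at U 0.5, then blended against white by the safe-zone mask.
# An older isVehicle local-position projection chain remains, unconnected to the output.
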MaterialFunction /Game/Weapons/WeaponSkins/Wraps/MaterialFunctions/MF_Customization_Rorschach
graph {nodes [
  n1 = VectorParameter() {desc=only RG is animated, group=Rorschach, parameterName=ROR_WarpUVs_Speed, sortPriority=4, defaultValue=<rgba 0.01 0.01 0 1>}
  n2 = Multiply(A=Time(), B=ComponentMask(None=@n1.RGB) {g=true, r=true})
  n3 = PreSkinnedLocalBounds()
  n4 = Subtract(A=VertexInterpolator(VS=PreSkinnedPosition()).PS, B=@n3.Min)
  n5 = Subtract(A=@n3.Max, B=@n3.Max)
  n6 = FunctionInput(Preview=Constant3Vector()) {inputName=isState, sortPriority=3}
  n7 = MaterialFunctionCall(Color=@n6) {materialFunction=&SplitComponents}
  n8 = ScalarParameter() {group=Rorschach, parameterName=RoR_Rotation_Weapon}
  n9 = MaterialFunctionCall(Color=@n6) {materialFunction=&SplitComponents}
  n10 = LinearInterpolate(
   A=ScalarParameter() {group=Rorschach, parameterName=RoR_Rotation_Vehicle}
   B=ScalarParameter() {group=Rorschach, parameterName=RoR_Rotation_Vehicl_Side}
   Alpha=@n9.G
  )
  n11 = VectorParameter() {group=Rorschach, parameterName=RoR_Offset_Weapon, defaultValue=<rgba 0 0 0 1>}
  n12 = MaterialFunctionCall(Color=@n6) {materialFunction=&SplitComponents}
  n13 = LinearInterpolate(
   A=VectorParameter() {group=Rorschach, parameterName=RoR_Offset_Vehicle, defaultValue=<rgba 0 0 0 1>}.RGB
   B=VectorParameter() {group=Rorschach, parameterName=RoR_Offset_Vehicle_Side, defaultValue=<rgba 0 0 0 1>}.RGB
   Alpha=@n12.G
  )
  n14 = ScalarParameter() {defaultValue=0.5, group=Rorschach, parameterName=RoR_Scale_Weapon}
  n15 = MaterialFunctionCall(Color=@n6) {materialFunction=&SplitComponents}
  n16 = LinearInterpolate(
   A=ScalarParameter() {defaultValue=1, group=Rorschach, parameterName=RoR_Scale_Vehicle}
   B=ScalarParameter() {defaultValue=1, group=Rorschach, parameterName=RoR_Scale_Vehicl_Side}
   Alpha=@n15.G
  )
  n17 = MaterialFunctionCall!(
   For_Safe_Zones=StaticBool() {value=true}
   Normal_Map
   Caculated_Bounds=FunctionInput(Preview=Divide(A=@n4, B=@n5)) {inputName=Calcated bounds, sortPriority=3}
   Pre_Skinned_Normals=FunctionInput(Preview=PreSkinnedNormal()) {bUsePreviewValueAsDefault=true, inputName=Local_Normals}
   Is Vehicle=@n7.R
   Is VehicleSide=@n7.R
   Is Mech=@n7.R
   Vehicle_Rotation
   UV_Rotation Bypass=LinearInterpolate(A=LinearInterpolate(A=@n8, B=@n10, Alpha=@n9.R), B=ScalarParameter() {group=Rorschach, parameterName=RoR_Rotation_Mech}, Alpha=@n9.B)
   Mech_Rotation
   Uvs_Offset=LinearInterpolate(
    A=LinearInterpolate(A=@n11.RGB, B=@n13, Alpha=@n12.R)
    B=VectorParameter() {group=Rorschach, parameterName=RoR_Offset_Mech, defaultValue=<rgba 0 0 0 1>}.RGB
    Alpha=@n12.B
   )
   Offset-after rotation
   Uvs_TileRate=LinearInterpolate(
    A=LinearInterpolate(A=@n14, B=@n16, Alpha=@n15.R)
    B=ScalarParameter() {defaultValue=1, group=Rorschach, parameterName=RoR_Scale_Mech}
    Alpha=@n15.B
   )
  ) {materialFunction=&/MF_Projection_Side_Normal}
  n18 = TextureObjectParameter() {group=Rorschach, parameterName=ROR_UVWarpTexture, texture=&T_BlacksRGB}
  n19 = TextureSample(UVs=Add(A=@n17.UVs, B=@n2), Tex=@n18) {samplerType=SAMPLERTYPE_LinearGrayscale}
  n20 = ScalarParameter() {defaultValue=0.12, group=Rorschach, parameterName=ROR_WarpUVsAmount}
  n21 = TextureObjectParameter() {group=Rorschach, parameterName=ROR_MaskTexture, texture=&T_BlacksRGB}
  n22 = TextureSample(UVs=Add(A=@n17.UVs, B=Multiply(A=@n19.R, B=@n20)), Tex=@n21) {samplerType=SAMPLERTYPE_LinearGrayscale}
  n23 = MaterialFunctionCall(Float2=@n17.UVs) {materialFunction=&BreakOutFloat2Components}
  n24 = AppendVector(A=Multiply(A=@n23.R) {constB=-1}, B=@n23.G)
  n25 = TextureSample(UVs=Add(A=@n2, B=@n24), Tex=@n18) {samplerType=SAMPLERTYPE_LinearGrayscale}
  n26 = TextureSample(UVs=Add(A=Multiply(A=@n20, B=@n25.R), B=@n24), Tex=@n21) {samplerType=SAMPLERTYPE_LinearGrayscale}
  n27 = ComponentMask(None=@n17.UVs) {r=true}
  n28 = Saturate(None=Divide(A=Add(A=@n27) {constB=-0.5}) {constB=0})
  FunctionOutput(None=LinearInterpolate(B=LinearInterpolate(A=@n22.R, B=@n26.R, Alpha=@n28), Alpha=@n17.Safe_Zone_Mask) {constA=1}) {outputName=Mask}
  n29 = VectorParameter() {group=Rorschach, parameterName=RoR_Offset_Pivot_Vehicle, defaultValue=<rgba 0 0 0 1>}
  n30 = VectorParameter() {group=Rorschach, parameterName=RoR_Offset_Pivot_Vehicle_Side, defaultValue=<rgba 0 0 0 1>}
  n31 = MaterialFunctionCall(Color=@n6) {materialFunction=&SplitComponents}
  LinearInterpolate(
   A=LinearInterpolate(
    A=VectorParameter() {group=Rorschach, parameterName=RoR_Offset_Pivot_Weapon, defaultValue=<rgba 0 -0.47 0.1 1>}.RGB
    B=LinearInterpolate(A=@n29.RGB, B=@n30.RGB, Alpha=@n31.G)
    Alpha=@n31.R
   )
   B=VectorParameter() {group=Rorschach, parameterName=RoR_Offset_Pivot_Mech, defaultValue=<rgba 0 0 0 1>}.RGB
   Alpha=@n31.B
  )
  Multiply()
  n32 = VertexInterpolator(VS=PreSkinnedPosition())
  n33 = VectorParameter() {group=Rorschach, parameterName=RoR_Offset_Pivot_Weapon, sortPriority=11, defaultValue=<rgba 0 -0.47 0.1 1>}
  n34 = VectorParameter() {group=Rorschach, parameterName=RoR_VehiclePivotOffset, sortPriority=7, defaultValue=<rgba 0 0 0 1>}
  n35 = FunctionInput() {inputName=isVehicle, inputType=FunctionInput_Scalar, sortPriority=3}
  n36 = LinearInterpolate(A=@n33.RGB, B=@n34.RGB, Alpha=@n35)
  n37 = FunctionInput(Preview=VertexInterpolator(VS=PreSkinnedLocalBounds().Half Extents).PS) {bUsePreviewValueAsDefault=true, inputName=Local Bounds, sortPriority=2}
  n38 = MaterialFunctionCall(Color=@n37) {materialFunction=&SplitComponents}
  n39 = StaticSwitchParameter(True=Max(A=Max(A=@n38.R, B=@n38.R), B=@n38.B), False=@n37) {desc=uses the maximim size to set the texture to power of 2, group=Rorschach, parameterName=ROR_UVs MakeSquare, sortPriority=-5}
  n40 = Divide(A=Add(A=FunctionInput(Preview=@n32.PS) {bUsePreviewValueAsDefault=true, inputName=Local Position}, B=Multiply(A=@n36, B=@n37)), B=@n39)
  n41 = Constant() {r=0.0001}
  Saturate(None=Power(Base=ComponentMask(None=@n40) {r=true}, Exp=@n41))
  n42 = VectorParameter() {group=Rorschach, parameterName=ROR_UVScale_Weapon, sortPriority=12, defaultValue=<rgba 0 2 1 1>}
  n43 = AppendVector(A=@n42.G, B=@n42.G)
  n44 = ScalarParameter() {defaultValue=0.8, group=Rorschach, parameterName=ROR_UVScale_WeaponUniform, sortPriority=13}
  n45 = ComponentMask(None=@n40) {g=true}
  n46 = ComponentMask(None=@n40) {b=true}
  n47 = ComponentMask(None=@n40) {r=true}
  n48 = VectorParameter() {group=Rorschach, parameterName=ROR_UVScale_Vehicle, sortPriority=9, defaultValue=<rgba 1 2 1 1>}
  n49 = AppendVector(A=@n48.G, B=@n48.G)
  n50 = ScalarParameter() {defaultValue=1, group=Rorschach, parameterName=ROR_UVScale_VehicleUniform, sortPriority=10}
  n51 = LinearInterpolate(
   A=Multiply(A=Multiply(A=@n43, B=@n44), B=AppendVector(A=@n45, B=@n46))
   B=Multiply(A=AppendVector(A=@n47, B=@n45), B=Multiply(A=@n49, B=@n50))
   Alpha=@n35
  )
  n52 = MaterialFunctionCall(Float2=@n51) {materialFunction=&BreakOutFloat2Components}
  n53 = Multiply(A=@n52.R) {constB=-1}
  n54 = Saturate(None=Power(Base=ComponentMask(None=@n40) {g=true}, Exp=@n41))
  n55 = MaterialFunctionCall(
   UVs=@n51
   Rotation Center=Constant3Vector() {constant=<rgba 0 0 0 1>}
   "Rotation Angle (0-1)"=ScalarParameter() {defaultValue=-0.25, group=Rorschach, parameterName=ROR_VehicleRotationAngle, sortPriority=8}
  ) {materialFunction=&CustomRotator}
  n56 = MaterialFunctionCall(Float2=@n55.Rotated Values) {materialFunction=&BreakOutFloat2Components}
  n57 = Multiply(A=@n56.R) {constB=-1}
  n58 = LinearInterpolate(
   A=LinearInterpolate(A=@n51, B=AppendVector(A=@n53, B=@n52.G), Alpha=@n54)
   B=LinearInterpolate(A=@n55.Rotated Values, B=AppendVector(A=@n57, B=@n56.G), Alpha=@n54)
   Alpha=@n35
  )
  n59 = VectorParameter() {desc=only RG is animated, group=Rorschach, parameterName=SP UVAnimation, sortPriority=4, defaultValue=<rgba 0.01 0.01 0 1>}
  n60 = Multiply(A=Time(), B=ComponentMask(None=@n59.RGB) {g=true, r=true})
  n61 = TextureSampleParameter2D(UVs=StaticSwitchParameter(True=Add(A=@n58, B=@n60), False=@n58) {defaultValue=true, parameterName=ROR_UVS_Animate, sortPriority=3}) {
   group=Rorschach
   parameterName=ROR_UVWarpTexture
   samplerType=SAMPLERTYPE_LinearGrayscale
   sortPriority=1
   texture=/Game/Weapons/WeaponSkins/Wraps/Textures/Rorschach/T_Wrap_Rorschach_M1
  }
  n62 = LinearInterpolate(
   A=ScalarParameter() {defaultValue=0.12, group=Rorschach, parameterName=ROR_WeaponWarpUVsAmount, sortPriority=5}
   B=ScalarParameter() {defaultValue=0.12, group=Rorschach, parameterName=ROR_VehicleWarpUVsAmount, sortPriority=6}
   Alpha=@n35
  )
  TextureSampleParameter2D(UVs=Add(A=@n58, B=Multiply(A=@n61.R, B=@n62))) {
   group=Rorschach
   parameterName=ROR_MaskTexture
   samplerType=SAMPLERTYPE_LinearGrayscale
   texture=/Game/Weapons/WeaponSkins/Wraps/Textures/Rorschach/T_Wrap_Rorschach_M2
  }
 ]}
