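# Material function that adds animated diffuse, emissive and normal overlays to side-projected weapon wrap skins, behind a master static switch.
# Up to three texture layers pan linearly or by sine, with separate weapon and vehicle speeds, and composite over BaseColor by alpha.
# Optional emissive uses panning masks or warped gradients, while optional normals follow the first layer.
# An optional pre-skinned normal mask limits overlays by facing axis, varying for vehicles, mechs and characters.
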MaterialFunction /Game/Weapons/WeaponSkins/Wraps/MaterialFunctions/MF_Customization_SPAnimatedDiffuse
graph {nodes [
  Constant()
  Constant()
  n1 = GetMaterialAttributes(None=FunctionInput() {inputName=Material, inputType=FunctionInput_MaterialAttributes}) {attributeGetTypes [
    #1
    #2
    #3
    #4
   ]}
  n2 = VectorParameter() {desc=Panning Speed Red and Green, group=Side Projection DiffuseMaskAnimated, parameterName=AnimatedDif_Speed_Weapon, sortPriority=4}
  n3 = VectorParameter() {desc=Panning Speed Red and Green, group=Side Projection DiffuseMaskAnimated, parameterName=AnimatedDif_Speed_Vehicle, sortPriority=5}
  n4 = FunctionInput() {inputName=isVehicle, inputType=FunctionInput_Scalar, sortPriority=2}
  n5 = Multiply(A=Time(), B=LinearInterpolate(A=AppendVector(A=@n2.R, B=@n2.R), B=AppendVector(A=@n3.R, B=@n3.R), Alpha=@n4))
  n6 = ComponentMask(None=@n5) {r=true}
  n7 = ComponentMask(None=@n5) {g=true}
  n8 = AppendVector(A=Add(A=@n6, B=ScalarParameter() {group=Side Projection DiffuseMaskAnimated, parameterName=AnimatedDif_Speed_offset, sortPriority=7}), B=@n7)
  n9 = Multiply(A=Add(A=Sine(None=@n8))) {constB=0.5}
  n10 = StaticSwitchParameter(
   True=LinearInterpolate(
    A=ScalarParameter() {group=Side Projection DiffuseMaskAnimated, parameterName=AnimatedDif_X_Min, sortPriority=8}
    B=ScalarParameter() {defaultValue=1, group=Side Projection DiffuseMaskAnimated, parameterName=AnimatedDif_X_Max, sortPriority=9}
    Alpha=ComponentMask(None=@n9) {r=true}
   )
   False=@n6
  ) {group=Side Projection DiffuseMaskAnimated, parameterName=AnimatedDif_Panning_SineX, sortPriority=6}
  n11 = StaticSwitchParameter(
   True=LinearInterpolate(
    A=ScalarParameter() {group=Side Projection DiffuseMaskAnimated, parameterName=AnimatedDif_Y_Min, sortPriority=11}
    B=ScalarParameter() {defaultValue=1, group=Side Projection DiffuseMaskAnimated, parameterName=AnimatedDif_Y_Max, sortPriority=10}
    Alpha=ComponentMask(None=@n9) {g=true}
   )
   False=@n7
  ) {group=Side Projection DiffuseMaskAnimated, parameterName=AnimatedDif_Panning_SineY, sortPriority=6}
  n12 = FunctionInput(Preview=TextureCoordinate()) {inputName=UVs, inputType=FunctionInput_Vector2, sortPriority=1}
  n13 = Multiply(
   A=ScalarParameter() {defaultValue=1, group=Side Projection DiffuseMaskAnimated, parameterName=AnimatedDif01-TileRate, sortPriority=3}
   B=@n12
  )
  n14 = StaticSwitchParameter(True=Add(A=@n13, B=AppendVector(A=@n10, B=@n11)), False=@n13) {group=Side Projection DiffuseMaskAnimated, parameterName=-AnimatedDif-Animate01-----, sortPriority=6}
  n15 = TextureSampleParameter2D(UVs=@n14) {
   group=Side Projection DiffuseMaskAnimated
   parameterName=AnimatedDif_Texture01
   sortPriority=3
   texture=&T_BlacksRGB
  }
  n16 = ScalarParameter() {group=Side Projection DiffuseMaskAnimated, parameterName=AnimatedDif02_X_Min, sortPriority=28}
  n17 = ScalarParameter() {defaultValue=1, group=Side Projection DiffuseMaskAnimated, parameterName=AnimatedDif02_X_Max, sortPriority=29}
  n18 = VectorParameter() {desc=Panning Speed Red and Green, group=Side Projection DiffuseMaskAnimated, parameterName=AnimatedDif02_Speed_Weapon, sortPriority=24}
  n19 = VectorParameter() {desc=Panning Speed Red and Green, group=Side Projection DiffuseMaskAnimated, parameterName=AnimatedDif02_Speed_Vehicle, sortPriority=25}
  n20 = Multiply(A=Time(), B=LinearInterpolate(A=AppendVector(A=@n18.R, B=@n18.R), B=AppendVector(A=@n19.R, B=@n19.R), Alpha=@n4))
  n21 = ComponentMask(None=@n20) {r=true}
  n22 = ComponentMask(None=@n20) {g=true}
  n23 = AppendVector(
   A=Add(A=@n21, B=ScalarParameter() {group=Side Projection DiffuseMaskAnimated, parameterName=AnimatedDif02_Speed_offset, sortPriority=27})
   B=@n22
  )
  n24 = Multiply(A=Add(A=Sine(None=@n23))) {constB=0.5}
  n25 = ComponentMask(None=@n24) {r=true}
  n26 = ScalarParameter() {group=Side Projection DiffuseMaskAnimated, parameterName=AnimatedDif02_Y_Min, sortPriority=31}
  n27 = ScalarParameter() {defaultValue=1, group=Side Projection DiffuseMaskAnimated, parameterName=AnimatedDif02_Y_Max, sortPriority=30}
  n28 = ComponentMask(None=@n24) {g=true}
  n29 = AppendVector(
   A=StaticSwitchParameter(True=LinearInterpolate(A=@n16, B=@n17, Alpha=@n25), False=@n21) {group=Side Projection DiffuseMaskAnimated, parameterName=AnimatedDif02_Panning_SineX, sortPriority=26}
   B=StaticSwitchParameter(True=LinearInterpolate(A=@n26, B=@n27, Alpha=@n28), False=@n22) {group=Side Projection DiffuseMaskAnimated, parameterName=AnimatedDif02_Panning_SineY, sortPriority=26}
  )
  n30 = Multiply(
   A=ScalarParameter() {defaultValue=1, group=Side Projection DiffuseMaskAnimated, parameterName=AnimatedDif02-TileRate, sortPriority=20}
   B=@n12
  )
  n31 = TextureSampleParameter2D(UVs=StaticSwitchParameter(True=Add(A=@n30, B=@n29), False=@n30) {group=Side Projection DiffuseMaskAnimated, parameterName=-AnimatedDif-Animate02------, sortPriority=21}) {
   group=Side Projection DiffuseMaskAnimated
   parameterName=AnimatedDif_Texture02
   sortPriority=20
   texture=&T_BlacksRGB
  }
  n32 = ScalarParameter() {group=Side Projection DiffuseMaskAnimated, parameterName=AnimatedDif03_X_Min, sortPriority=48}
  n33 = ScalarParameter() {defaultValue=1, group=Side Projection DiffuseMaskAnimated, parameterName=AnimatedDif03_X_Max, sortPriority=49}
  n34 = VectorParameter() {desc=Panning Speed Red and Green, group=Side Projection DiffuseMaskAnimated, parameterName=AnimatedDif03_Speed_Weapon, sortPriority=44}
  n35 = VectorParameter() {desc=Panning Speed Red and Green, group=Side Projection DiffuseMaskAnimated, parameterName=AnimatedDif03_Speed_Vehicle, sortPriority=45}
  n36 = Multiply(A=Time(), B=LinearInterpolate(A=AppendVector(A=@n34.R, B=@n34.R), B=AppendVector(A=@n35.R, B=@n35.R), Alpha=@n4))
  n37 = ComponentMask(None=@n36) {r=true}
  n38 = ComponentMask(None=@n36) {g=true}
  n39 = AppendVector(
   A=Add(A=@n37, B=ScalarParameter() {group=Side Projection DiffuseMaskAnimated, parameterName=AnimatedDif03_Speed_offset, sortPriority=47})
   B=@n38
  )
  n40 = Multiply(A=Add(A=Sine(None=@n39))) {constB=0.5}
  n41 = ComponentMask(None=@n40) {r=true}
  n42 = ScalarParameter() {group=Side Projection DiffuseMaskAnimated, parameterName=AnimatedDif03_Y_Min, sortPriority=51}
  n43 = ScalarParameter() {defaultValue=1, group=Side Projection DiffuseMaskAnimated, parameterName=AnimatedDif03_Y_Max, sortPriority=50}
  n44 = ComponentMask(None=@n40) {g=true}
  n45 = AppendVector(
   A=StaticSwitchParameter(True=LinearInterpolate(A=@n32, B=@n33, Alpha=@n41), False=@n37) {group=Side Projection DiffuseMaskAnimated, parameterName=AnimatedDif03_Panning_SineX, sortPriority=46}
   B=StaticSwitchParameter(True=LinearInterpolate(A=@n42, B=@n43, Alpha=@n44), False=@n38) {group=Side Projection DiffuseMaskAnimated, parameterName=AnimatedDif03_Panning_SineY, sortPriority=46}
  )
  n46 = Multiply(
   A=ScalarParameter() {defaultValue=1, group=Side Projection DiffuseMaskAnimated, parameterName=AnimatedDif03-TileRate, sortPriority=40}
   B=@n12
  )
  n47 = TextureSampleParameter2D(UVs=StaticSwitchParameter(True=Add(A=@n46, B=@n45), False=@n46) {group=Side Projection DiffuseMaskAnimated, parameterName=-AnimatedDif-Animate03-----, sortPriority=41}) {
   group=Side Projection DiffuseMaskAnimated
   parameterName=AnimatedDif_Texture03
   sortPriority=40
   texture=&T_BlacksRGB
  }
  n48 = StaticSwitchParameter(True=LinearInterpolate(A=@n31.RGB, B=@n47.RGB, Alpha=@n47.RGB), False=@n31.RGB) {group=Side Projection DiffuseMaskAnimated, parameterName=Diffuse02 add Diffus03, sortPriority=39}
  n49 = StaticSwitchParameter(True=Saturate(None=Add(A=@n31.A, B=@n47.A)), False=@n31.A) {group=Side Projection DiffuseMaskAnimated, parameterName=Diffuse02 add Diffus03, sortPriority=39}
  n50 = StaticSwitchParameter(True=Add(A=@n15.A, B=@n49), False=@n15.A) {group=Side Projection DiffuseMaskAnimated, parameterName=Diffuse01 add Diffus02, sortPriority=19}
  n51 = TextureObjectParameter() {
   desc="Red and green pan Blue is used to mask "
   group=Side Projection DiffuseMaskAnimated
   parameterName=AnimatedDif_PanningEmissive
   samplerType=SAMPLERTYPE_Masks
   sortPriority=63
   texture=&T_BlackMask
  }
  n52 = TextureSample(UVs=@n12, Tex=@n51) {samplerType=SAMPLERTYPE_LinearColor, texture=/Game/Weapons/WeaponSkins/Wraps/Textures/DriftSummer/TWrap_DriftSummer_M2}
  n53 = OneMinus(None=@n52.B)
  n54 = Multiply(A=@n50, B=LinearInterpolate(A=@n52.B, B=@n53, Alpha=@n4))
  n55 = LinearInterpolate(
   A=@n1.BaseColor
   B=StaticSwitchParameter(True=LinearInterpolate(A=@n15.RGB, B=@n48, Alpha=@n49), False=@n15.RGB) {group=Side Projection DiffuseMaskAnimated, parameterName=Diffuse01 add Diffus02, sortPriority=19}
   Alpha=Saturate(None=StaticSwitchParameter(True=@n54, False=@n50) {group=Side Projection DiffuseMaskAnimated, parameterName=Animated_TriplanerMask_Blue Channel, sortPriority=-20})
  )
  n56 = FunctionInput(Preview=VertexInterpolator(VS=PreSkinnedNormal()).PS) {bUsePreviewValueAsDefault=true, inputName=Pre-Skinned Local Normal, sortPriority=42}
  n57 = ScalarParameter() {group=Side Projection DiffuseMaskAnimated, parameterName=Animated_TriplanerMask_Contrast, sortPriority=-19}
  n58 = MaterialFunctionCall(In=Abs(None=ComponentMask(None=@n56) {r=true}), Contrast=@n57) {materialFunction=&CheapContrast}
  n59 = ComponentMask(None=@n56) {b=true}
  n60 = ComponentMask(None=MaterialFunctionCall(In=Abs(None=@n59), Contrast=@n57) {materialFunction=&CheapContrast}.Result) {r=true}
  n61 = ComponentMask(None=@n56) {g=true}
  n62 = ComponentMask(None=MaterialFunctionCall(In=Abs(None=@n61), Contrast=@n57) {materialFunction=&CheapContrast}.Result) {r=true}
  n63 = FunctionInput() {bUsePreviewValueAsDefault=true, inputName=Side Project Vehicle, inputType=FunctionInput_Scalar}
  n64 = LinearInterpolate(
   A=LinearInterpolate(A=ComponentMask(None=@n58.Result) {r=true}, B=LinearInterpolate(A=@n60, B=@n62, Alpha=@n63), Alpha=@n4)
   B=@n62
   Alpha=FunctionInput() {bUsePreviewValueAsDefault=true, inputName=isMech, inputType=FunctionInput_Scalar}
  )
  n65 = LinearInterpolate(A=@n1.BaseColor, B=@n55, Alpha=@n64)
  n66 = Multiply(A=ScalarParameter() {parameterName=Animated_TriplanarMask_Character}, B=NamedRerouteUsage())
  n67 = Multiply(
   A=Time()
   B=AppendVector(
    A=ScalarParameter() {defaultValue=-0.5, group=Side Projection DiffuseMaskAnimated, parameterName=AnimatedDif_WarpedSpeed, sortPriority=64}
    B=Constant()
   )
  )
  n68 = TextureSampleParameter2D(UVs=@n12) {
   group=Side Projection DiffuseMaskAnimated
   parameterName=AnimatedDif_WarpedTexture
   samplerType=SAMPLERTYPE_Masks
   sortPriority=63
   texture=&T_BlackMask
  }
  n69 = MaterialFunctionCall(Color=Multiply(
    A=@n68.RGB
    B=ScalarParameter() {defaultValue=1, group=Side Projection DiffuseMaskAnimated, parameterName=AnimatedDif_WarpedAmount, sortPriority=65}
   ) {constA=0.8}) {materialFunction=&SplitComponents}
  n70 = TextureObjectParameter() {
   group=Side Projection DiffuseMaskAnimated
   parameterName=AnimatedDif_WarpedGrad
   sortPriority=62
   texture=/Game/Global/Textures/Default/Masks_and_Gradients/gradientPass
  }
  n71 = Multiply(A=@n68.R) {constA=10, constB=550}
  n72 = Multiply(
   A=TextureSample(UVs=Add(A=@n67, B=@n69.R), Tex=@n70) {texture=/Game/Weapons/WeaponSkins/Wraps/Textures/DriftSummer/TWrap_DriftSummer_Grad}.RGB
   B=Multiply(A=Saturate(None=@n71), B=@n68.A)
  )
  n73 = Add(A=@n67) {constB=0.5}
  n74 = Multiply(A=@n68.G) {constA=10, constB=550}
  n75 = Multiply(
   A=TextureSample(UVs=Add(A=@n73, B=@n69.G), Tex=@n70) {texture=&TWrap_DriftSummer_Grad}.R
   B=Multiply(A=Saturate(None=@n74))
  )
  n76 = TextureSample(UVs=Add(A=Add(A=@n67) {constB=0.25}, B=@n69.B), Tex=@n70) {texture=&TWrap_DriftSummer_Grad}
  n77 = Multiply(A=Saturate(None=Multiply(A=@n68.B) {constA=10, constB=550}))
  n78 = Time()
  n79 = AppendVector(
   A=ScalarParameter() {defaultValue=-0.5, group=Side Projection DiffuseMaskAnimated, parameterName=AnimatedDif_PaningRed_X, sortPriority=64}
   B=ScalarParameter() {defaultValue=-0.5, group=Side Projection DiffuseMaskAnimated, parameterName=AnimatedDif_PaningRed_Y, sortPriority=64}
  )
  n80 = TextureSample(UVs=Add(A=@n12, B=Multiply(A=@n78, B=@n79)), Tex=@n51) {samplerType=SAMPLERTYPE_LinearColor, texture=&TWrap_DriftSummer_M2}
  n81 = Time()
  n82 = AppendVector(
   A=ScalarParameter() {defaultValue=-0.5, group=Side Projection DiffuseMaskAnimated, parameterName=AnimatedDif_PaningBlue_X, sortPriority=64}
   B=ScalarParameter() {group=Side Projection DiffuseMaskAnimated, parameterName=AnimatedDif_PaningBlue_Y, sortPriority=64}
  )
  n83 = TextureSample(UVs=Add(A=@n12, B=Multiply(A=@n81, B=@n82)), Tex=@n51) {samplerType=SAMPLERTYPE_LinearColor, texture=&TWrap_DriftSummer_M2}
  n84 = OneMinus(None=FunctionInput() {inputName=Gradent})
  n85 = StaticSwitchParameter(
   True=Add(A=Add(A=@n72, B=@n75), B=Multiply(A=@n76.RGB, B=@n77))
   False=Multiply(
    A=Add(A=@n80.R, B=@n83.G)
    B=StaticSwitchParameter(True=@n84, False=@n52.B) {group=Side Projection DiffuseMaskAnimated, parameterName=AnimatedDif_Emissive BlueChannel/Gradent, sortPriority=63}
   )
  ) {
   desc=Animated a second texture on top of diffuse masked by alpha
   group=Side Projection DiffuseMaskAnimated
   parameterName=AnimatedDif_Emissive Panning/Warped
   sortPriority=61
  }
  n86 = VectorParameter() {group=Side Projection DiffuseMaskAnimated, parameterName=AnimatedDif_EmissiveColor, sortPriority=66, defaultValue=<rgba 1 0.7145 0.0267 1>}
  n87 = StaticSwitchParameter(True=Add(A=@n1.EmissiveColor, B=Multiply(A=@n85, B=@n86.RGB)), False=@n1.EmissiveColor) {
   desc=Animated a second texture on top of diffuse masked by alpha
   group=Side Projection DiffuseMaskAnimated
   parameterName=AnimatedDif_Emissive
   sortPriority=61
  }
  n88 = LinearInterpolate(A=@n1.EmissiveColor, B=@n87, Alpha=@n64)
  n89 = Multiply(A=ScalarParameter() {parameterName=Animated_TriplanarMask_Character}, B=NamedRerouteUsage())
  n90 = Constant3Vector() {constant=<rgba 0 0 1 0>}
  n91 = TextureSampleParameter2D(UVs=@n14) {
   group=Side Projection DiffuseMaskAnimated
   parameterName=AnimatedDif_Normals01
   samplerType=SAMPLERTYPE_Normal
   sortPriority=4
   texture=/Engine/EngineMaterials/FlatNormal
  }
  n92 = LinearInterpolate(A=LinearInterpolate(A=@n90, B=@n91.RGB, Alpha=@n52.B), B=LinearInterpolate(A=@n90, B=@n91.RGB, Alpha=@n53), Alpha=@n4)
  n93 = StaticSwitchParameter(
   True=MaterialFunctionCall(
    BaseNormal=@n1.Normal
    AdditionalNormal=StaticSwitchParameter(True=@n92, False=@n91.RGB) {group=Side Projection DiffuseMaskAnimated, parameterName=Animated_TriplanerMask_Blue Channel, sortPriority=-20}
   ) {materialFunction=&BlendAngleCorrectedNormals}.Result
   False=@n1.Normal
  ) {
   desc=Animated a second texture on top of diffuse masked by alpha
   group=Side Projection DiffuseMaskAnimated
   parameterName=AnimatedDif_Normal
   sortPriority=4
  }
  n94 = LinearInterpolate(A=@n1.Normal, B=@n93, Alpha=@n64)
  n95 = Multiply(A=ScalarParameter() {parameterName=Animated_TriplanarMask_Character}, B=NamedRerouteUsage())
  n96 = SetMaterialAttributes!(
   MaterialAttributes=@n1.MaterialAttributes
   BaseColor=StaticSwitchParameter(True=LinearInterpolate(A=@n65, B=@n55, Alpha=@n66), False=@n55) {group=Side Projection DiffuseMaskAnimated, parameterName=Animated_TriplanerMask, sortPriority=-20}
   Metallic=@n1.MaterialAttributes
   EmissiveColor=StaticSwitchParameter(True=LinearInterpolate(A=@n88, B=@n87, Alpha=@n89), False=@n87) {group=Side Projection DiffuseMaskAnimated, parameterName=Animated_TriplanerMask, sortPriority=-20}
   Normal=StaticSwitchParameter(True=LinearInterpolate(A=@n94, B=@n93, Alpha=@n95), False=@n93) {group=Side Projection DiffuseMaskAnimated, parameterName=Animated_TriplanerMask, sortPriority=-20}
  ) {attributeSetTypes [
    #1
    #2
    #3
    #4
   ]}
  n97 = StaticSwitchParameter(True=@n96, False=@n1.MaterialAttributes) {
   desc=Animated a second texture on top of diffuse masked by alpha
   group=Side Projection
   parameterName=-Diffuse-Mask--Animated----
   sortPriority=20
  }
  FunctionOutput(None=@n97)
  NamedRerouteDeclaration(None=FunctionInput() {bUsePreviewValueAsDefault=true, inputName=isCharacter, inputType=FunctionInput_Scalar}) {name=IsCharacter, nodeColor=<rgba 0.3125 0 1 1>}
 ]}
