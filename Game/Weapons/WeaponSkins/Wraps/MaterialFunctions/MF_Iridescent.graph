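# Material function for weapon skin wraps that layers an optional iridescent finish over an incoming material, toggled by a static switch.
# A Fresnel term using the material's normal samples a gradient texture to drive added base color and metallic.
# The crevasse input and incoming roughness form edge and depth masks that shape color and remap roughness.
# MatLayerBlend_Standard applies the result through the customization mask, optionally multiplied by a second wrap mask.
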MaterialFunction /Game/Weapons/WeaponSkins/Wraps/MaterialFunctions/MF_Iridescent
graph {nodes [
  n1 = FunctionInput() {inputName=Material, inputType=FunctionInput_MaterialAttributes, sortPriority=-200}
  n2 = GetMaterialAttributes(None=@n1) {attributeGetTypes [
    #1
    #2
    #3
    #4
   ]}
  n3 = TextureSampleParameter2D(UVs=Fresnel(
    ExponentIn=ScalarParameter() {defaultValue=0.5, group=Iridescent, parameterName=IR_Exponent}
    BaseReflectFractionIn=ScalarParameter() {defaultValue=0.08, group=Iridescent, parameterName=IR_BaseReflecFraction}
    Normal=Transform(None=@n2.Normal)
   )) {group=Iridescent, parameterName=Iridescent_Grad, texture=/Game/Weapons/WeaponSkins/Wraps/Textures/Iridescence/T_iridescent_gradient}
  n4 = ScalarParameter() {defaultValue=2, group=Iridescent, parameterName=IR_Intensity}
  n5 = Multiply(A=@n3.RGB, B=@n4)
  n6 = FunctionInput(Preview=Constant()) {bUsePreviewValueAsDefault=true, inputName="Customization Crevasse (_CM G)", inputType=FunctionInput_Scalar, sortPriority=21}
  n7 = Power(Base=@n6, Exp=ScalarParameter() {defaultValue=3, group=Iridescent, parameterName=IR_edge_intensity})
  n8 = Clamp(None=LinearInterpolate(Alpha=Clamp(None=@n6) {minDefault=0.5}) {constA=-1})
  n9 = ScalarParameter() {defaultValue=1.3809, group=Iridescent, parameterName=Edge_contrast}
  n10 = Clamp(None=Multiply(A=Power(Base=@n8, Exp=@n9), B=ScalarParameter() {defaultValue=1.3107, group=Iridescent, parameterName=Edge_Stregnth}))
  n11 = LinearInterpolate(A=Add(A=LinearInterpolate(A=@n5, Alpha=@n7), B=@n2.BaseColor), B=@n2.BaseColor, Alpha=@n10)
  n12 = Clamp(None=LinearInterpolate(Alpha=Clamp(None=@n2.Roughness) {maxDefault=0.5}) {constB=2})
  n13 = LinearInterpolate(Alpha=@n12) {constA=0.5}
  n14 = Constant() {r=1}
  n15 = ScalarParameter() {defaultValue=1, group=Iridescent, parameterName=Roughness_pits}
  n16 = ScalarParameter() {defaultValue=0.4, group=Iridescent, parameterName=Roughness_low}
  n17 = ScalarParameter() {defaultValue=1, group=Iridescent, parameterName=Roghness_high}
  n18 = LinearInterpolate(
   A=Clamp(None=LinearInterpolate(A=@n16, B=@n17, Alpha=@n2.Roughness))
   B=ScalarParameter() {group=Iridescent, parameterName=Roughness_Edge}
   Alpha=@n10
  )
  n19 = ScalarParameter() {defaultValue=1.3262, group=Iridescent, parameterName=Depth_Contrast}
  n20 = Clamp(None=Multiply(A=Power(Base=@n12, Exp=@n19), B=ScalarParameter() {defaultValue=1.2614, group=Iridescent, parameterName=Depth_intensity}))
  n21 = SetMaterialAttributes!(
   MaterialAttributes=@n2.MaterialAttributes
   BaseColor=LinearInterpolate(
    A=@n2.BaseColor
    B=Multiply(A=@n11, B=@n13)
    Alpha=FunctionInput(Preview=@n14) {bUsePreviewValueAsDefault=true, inputName=Mask_Secondary, inputType=FunctionInput_Scalar, sortPriority=24}
   )
   Roughness=Clamp(None=LinearInterpolate(A=@n15, B=@n18, Alpha=@n20))
   Normal=@n2.Normal
   Metallic=LinearInterpolate(A=Multiply(A=@n3.A, B=@n4), Alpha=@n7)
  ) {attributeSetTypes [
    #1
    #2
    #3
    #4
   ]}
  n22 = FunctionInput(Preview=@n14) {bUsePreviewValueAsDefault=true, inputName="Customization Mask (_CM B)", inputType=FunctionInput_Scalar, sortPriority=24}
  n23 = Multiply(
   A=@n22
   B=StaticSwitchParameter(True=FunctionInput() {inputName=Uvs_1, inputType=FunctionInput_Scalar}, False=@n22) {group=Customization, parameterName=IR_Wrap_Mask, sortPriority=-500}
  )
  FunctionOutput(None=StaticSwitchParameter(
    True=MaterialFunctionCall!(Base Material=@n1, Top Material=@n21, Alpha=@n23) {materialFunction=/Engine/Functions/MaterialLayerFunctions/MatLayerBlend_Standard}.Blended Material
    False=@n1
   ) {group=Customization, parameterName="Iridescent ", sortPriority=-500})
 ]}
