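# Material function for weapon wraps that builds a directional gradient and projected UVs for weapons, vehicles and mechs.
# An overridable Bounds input, defaulting to bounds-relative pre-skinned position, gets an offset, then feeds both gradient and UVs.
# isVehicle and isMech lerps select offset, falloff, forward axis and projection without static branching.
# The gradient outputs switched (Invert_Gradent) and raw, while UVs add rotation choices, scaling, time-based panning and a warp offset.
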MaterialFunction /Game/Weapons/WeaponSkins/Wraps/MaterialFunctions/MF_Wrap_Gradent_Creation
graph {nodes [
  n1 = PreSkinnedPosition()
  n2 = PreSkinnedLocalBounds()
  n3 = Subtract(
   A=FunctionInput(Preview=VertexInterpolator(VS=@n1).PS) {bUsePreviewValueAsDefault=true, inputName=Pre-Skinned Local Position, sortPriority=1}
   B=@n2.Min
  )
  n4 = Subtract(A=@n2.Max, B=@n2.Max)
  n5 = FunctionInput(Preview=Constant3Vector()) {bUsePreviewValueAsDefault=true, inputName=WeaponOffset, sortPriority=3}
  n6 = Constant()
  n7 = FunctionInput(Preview=@n6) {bUsePreviewValueAsDefault=true, inputName=VehicleOffset, sortPriority=5}
  n8 = FunctionInput(Preview=Constant()) {inputName=isVehicle, inputType=FunctionInput_Scalar, sortPriority=2}
  n9 = Constant()
  n10 = FunctionInput(Preview=@n9) {bUsePreviewValueAsDefault=true, inputName=isMech, inputType=FunctionInput_Scalar, sortPriority=2}
  n11 = Add(
   A=FunctionInput(Preview=Divide(A=@n3, B=@n4)) {bUsePreviewValueAsDefault=true, inputName=Bounds}
   B=LinearInterpolate(
    A=LinearInterpolate(A=@n5, B=@n7, Alpha=@n8)
    B=FunctionInput(Preview=@n6) {bUsePreviewValueAsDefault=true, inputName=MechOffset, sortPriority=7}
    Alpha=@n10
   )
  )
  n12 = Constant() {r=0.5}
  n13 = FunctionInput(Preview=@n12) {bUsePreviewValueAsDefault=true, inputName=WeaponFalloff, inputType=FunctionInput_Scalar, sortPriority=4}
  n14 = FunctionInput(Preview=@n12) {bUsePreviewValueAsDefault=true, inputName=VehicleFalloff, inputType=FunctionInput_Scalar, sortPriority=6}
  n15 = Divide(
   A=@n11
   B=LinearInterpolate(
    A=LinearInterpolate(A=@n13, B=@n14, Alpha=@n8)
    B=FunctionInput(Preview=@n12) {bUsePreviewValueAsDefault=true, inputName=MechFalloff, inputType=FunctionInput_Scalar, sortPriority=8}
    Alpha=@n10
   )
  )
  n16 = LinearInterpolate(
   A=LinearInterpolate(A=Constant3Vector() {constant=<rgba 0 1 0 0>}, B=Constant3Vector() {constant=<rgba 1 0 0 0>}, Alpha=@n8)
   B=Constant3Vector() {constant=<rgba 0 0 1 0>}
   Alpha=@n10
  )
  n17 = MaterialFunctionCall(Float3=Multiply(A=Saturate(None=@n15), B=FunctionInput(Preview=@n16) {bUsePreviewValueAsDefault=true, inputName=ForwardAxis, sortPriority=-5})) {materialFunction=&BreakOutFloat3Components}
  n18 = Max(A=Max(A=@n17.R, B=@n17.R), B=@n17.B)
  n19 = OneMinus(None=@n18)
  n20 = FunctionInput(Preview=StaticBool()) {bUsePreviewValueAsDefault=true, inputName=Invert_Gradent, inputType=FunctionInput_StaticBool, sortPriority=10}
  FunctionOutput(None=Saturate(None=StaticSwitch(True=@n18, False=@n19, Value=@n20))) {outputName=Gradent}
  n21 = MaterialFunctionCall(Float3=@n11) {materialFunction=&BreakOutFloat3Components}
  n22 = AppendVector(A=@n21.G, B=@n21.G)
  n23 = Constant() {r=1}
  n24 = LinearInterpolate(
   A=LinearInterpolate(
    A=AppendVector(A=@n21.G, B=@n21.G)
    B=AppendVector(A=@n21.R, B=@n21.R)
    Alpha=FunctionInput(Preview=@n23) {bUsePreviewValueAsDefault=true, inputName=Rotate_Vehicle, inputType=FunctionInput_Scalar, sortPriority=15}
   )
   B=LinearInterpolate(
    A=AppendVector(A=@n21.B, B=@n21.B)
    B=AppendVector(A=@n21.R, B=@n21.R)
    Alpha=FunctionInput(Preview=@n23) {bUsePreviewValueAsDefault=true, inputName=Rotate_Vehicle_Side, inputType=FunctionInput_Scalar, sortPriority=15}
   )
   Alpha=FunctionInput(Preview=@n9) {bUsePreviewValueAsDefault=true, inputName=Side Project Vehicle, inputType=FunctionInput_Scalar, sortPriority=2}
  )
  n25 = AppendVector(A=@n21.R, B=@n21.R)
  n26 = AppendVector(A=@n21.B, B=@n21.B)
  n27 = FunctionInput(Preview=@n23) {bUsePreviewValueAsDefault=true, inputName=Rotate_Mech, inputType=FunctionInput_Scalar, sortPriority=15}
  n28 = Constant3Vector() {constant=<rgba 1 1 1 0>}
  n29 = Multiply(
   A=LinearInterpolate(A=LinearInterpolate(A=@n22, B=@n24, Alpha=@n8), B=LinearInterpolate(A=@n25, B=@n26, Alpha=@n27), Alpha=@n10)
   B=ComponentMask(None=FunctionInput(Preview=@n28) {bUsePreviewValueAsDefault=true, inputName=UV_Spread_XY, sortPriority=11}) {g=true, r=true}
  )
  n30 = Constant3Vector()
  n31 = Multiply(
   A=Time()
   B=ComponentMask(None=FunctionInput(Preview=@n30) {bUsePreviewValueAsDefault=true, inputName=UV_Animated_XY, sortPriority=12}) {g=true, r=true}
  )
  FunctionOutput(None=Add(
    A=Add(A=@n29, B=@n31)
    B=FunctionInput() {bUsePreviewValueAsDefault=true, inputName=Warp_UVs, inputType=FunctionInput_Scalar, sortPriority=13}
   )) {outputName=Uvs, sortPriority=1}
  FunctionOutput(None=@n18) {outputName=Gradent_B}
 ]}
