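# Material function for weapon skin wraps that optionally adds tinted emissive, driven by a side-projection mask channel, to incoming material attributes.
# Static switches choose the plain mask or a height blend, either multiplicative or HeightLerp.
# Height comes from MF_Wrap_Gradent_Creation, with weapon, vehicle and mech controls, as a gradient or texture sampled with its warped UVs.
# The non-fixed bounds path divides by Max minus Max, a zero extent.
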MaterialFunction /Game/Weapons/WeaponSkins/Wraps/MaterialFunctions/MF_SideProjection_Emissive
graph {nodes [
  n1 = FunctionInput() {inputName=Material, inputType=FunctionInput_MaterialAttributes}
  n2 = GetMaterialAttributes(None=@n1) {attributeGetTypes [#1]}
  n3 = FunctionInput() {inputName=SideProjection01_Mask, sortPriority=2}
  n4 = ChannelMaskParameter(None=@n3) {group=SideProjection_Emissive, parameterName="SideProjection_01 Emissive Mask Channel ", sortPriority=5}
  n5 = FunctionInput() {bUsePreviewValueAsDefault=true, inputName=Pre-Skinned Local Position, sortPriority=1}
  n6 = PreSkinnedLocalBounds()
  n7 = Subtract(A=@n5, B=@n6.Min)
  n8 = Subtract(A=@n6.Max, B=@n6.Max)
  n9 = VectorParameter() {group=SideProjection_Emissive, parameterName=SP_Emissive_Height_UV_Spread_XY, sortPriority=17, defaultValue=<rgba 1 1 1 0>}
  n10 = MaterialFunctionCall!(
   ForwardAxis=FunctionInput() {bUsePreviewValueAsDefault=true, inputName=Forward Axis}
   Bounds=StaticSwitchParameter(True=FunctionInput() {bUsePreviewValueAsDefault=true, inputName=Bounds}, False=Divide(A=@n7, B=@n8)) {group=SideProjection_Emissive, parameterName=Emissive_Object FixedBounds, sortPriority=7}
   Pre-Skinned Local Position=@n5
   isVehicle=FunctionInput() {inputName=isVehicle, inputType=FunctionInput_Scalar, sortPriority=2}
   isMech=FunctionInput() {bUsePreviewValueAsDefault=true, inputName=isMech, inputType=FunctionInput_Scalar, sortPriority=2}
   Side Project Vehicle=FunctionInput() {bUsePreviewValueAsDefault=true, inputName=Side Project Vehicle, inputType=FunctionInput_Scalar}
   WeaponOffset=ScalarParameter() {group=SideProjection_Emissive, parameterName=SP_Emissive_Height_WeaponOffset, sortPriority=9}
   WeaponFalloff=ScalarParameter() {defaultValue=1, group=SideProjection_Emissive, parameterName=SP_Emissive_Height_WeaponFalloff, sortPriority=10}
   VehicleOffset=ScalarParameter() {group=SideProjection_Emissive, parameterName=SP_Emissive_Height_VehicleOffset, sortPriority=11}
   VehicleFalloff=ScalarParameter() {defaultValue=1, group=SideProjection_Emissive, parameterName=SP_Emissive_Height_VehicleFalloff, sortPriority=12}
   MechOffset=ScalarParameter() {group=SideProjection_Emissive, parameterName=SP_Emissive_Height_MechOffset, sortPriority=13}
   MechFalloff=ScalarParameter() {defaultValue=1, group=SideProjection_Emissive, parameterName=SP_Emissive_Height_MechFalloff, sortPriority=14}
   Invert_Gradent=StaticBoolParameter() {defaultValue=true, group=SideProjection_Emissive, parameterName=SP_Emissive_Height_Invert_Gradent, sortPriority=15}
   UV_Spread_XY=@n9.RGB
   UV_Animated_XY=VectorParameter() {group=SideProjection_Emissive, parameterName=SP_Emissive_Height_UV_Animated_XY, sortPriority=18}.RGB
   Warp_UVs=Multiply(
    A=ChannelMaskParameter(None=@n3) {group=SideProjection_Emissive, parameterName="EM_SideProjection_01 WarpUVs Mask Channel ", sortPriority=15}.R
    B=ScalarParameter() {group=SideProjection_Emissive, parameterName=SP_Emissive_WarpUVs, sortPriority=16}
   )
   Rotate_Mech
   Rotate_Vehicle_Side
   Rotate_Vehicle
  ) {materialFunction=&/MF_Wrap_Gradent_Creation}
  n11 = TextureSampleParameter2D(UVs=@n10.Uvs) {
   group=SideProjection_Emissive
   parameterName=Emissive_Height_Texture
   samplerType=SAMPLERTYPE_Masks
   sortPriority=8
   texture=/Game/Global/Textures/Default/Masks_and_Gradients/T_VertGradient_MASK
  }
  n12 = StaticSwitchParameter(
   True=ChannelMaskParameter(None=@n11.RGBA) {group=SideProjection_Emissive, parameterName="Emissive_HeightBlend-Gradient_Channel ", sortPriority=8}.R
   False=@n10.Gradent
  ) {group=SideProjection_Emissive, parameterName="-Emissive-Type-Gradent\\Texture-----------", sortPriority=7}
  n13 = Multiply(A=@n4.R, B=@n12)
  n14 = StaticSwitchParameter(
   True=Multiply(A=@n13, B=Add(A=@n13, B=@n12))
   False=MaterialFunctionCall(
    Transition Phase=@n12
    Height Texture=@n4.R
    Contrast=ScalarParameter() {defaultValue=1, group=SideProjection_Emissive, parameterName=SP_Emissive_Height_Contrast, sortPriority=6}
   ) {materialFunction=&HeightLerp}.Alpha
  ) {group=SideProjection_Emissive, parameterName="-SP-Emissive-Height\\X+X---------", sortPriority=4}
  n15 = Saturate(None=@n4.R)
  n16 = VectorParameter() {group=SideProjection_Emissive, parameterName=SP_Emisive_Color, sortPriority=2, defaultValue=<rgba 0 0 0 1>}
  n17 = Add(
   A=@n2.EmissiveColor
   B=Multiply(
    A=StaticSwitchParameter(True=@n14, False=@n15) {group=SideProjection_Emissive, parameterName="-SP-Emissive-Mask\\Height---------", sortPriority=3}
    B=@n16.RGB
   )
  )
  FunctionOutput(None=StaticSwitchParameter(
    True=SetMaterialAttributes!(MaterialAttributes=@n2.MaterialAttributes, EmissiveColor=@n17) {attributeSetTypes [#1]}
    False=@n1
   ) {group=SideProjection_Emissive, parameterName=-SideProjection-Emissive----------})
 ]}
