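# Material function that computes planar side-projection UVs for wrap skins on weapons, pickaxes, vehicles, mechs, characters and backpacks.
# It swizzles an input vector into two-axis pairs, with static switches swapping axes to rotate most projections, then applies per-type offset and uniform scale parameters.
# Scalar type inputs choose the result through nested lerps, with character and backpack nested under the isMech branch.
# Vehicles add a side-projection variant with its own parameters.
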MaterialFunction /Game/Weapons/WeaponSkins/Wraps/MaterialFunctions/MF_SideProjection_Math
graph {nodes [
  n1 = VectorParameter() {desc="G=X  B=Y A=Uniform", group=Side Projection UVS, parameterName=UVs_Scale_Weapon, sortPriority=1, defaultValue=<rgba 1 1 1 1>}
  n2 = VectorParameter() {desc="G=X  B=Y A=Uniform", group=Side Projection UVS, parameterName=UVs_Scale_Pickaxe, sortPriority=1, defaultValue=<rgba 1 1 1 1>}
  n3 = NamedRerouteUsage()
  n4 = VectorParameter() {desc="G=Y  B=X", group=Side Projection UVS, parameterName=UVs_Offset_Weapon, sortPriority=2, defaultValue=<rgba 0 0 0 1>}
  n5 = ComponentMask(None=@n4.RGB) {g=true, r=true}
  n6 = VectorParameter() {desc="G=Y  B=X", group=Side Projection UVS, parameterName=UVs_Offset_Pickaxe, sortPriority=2, defaultValue=<rgba 0 0 0 1>}
  n7 = ComponentMask(None=@n6.RGB) {g=true, r=true}
  n8 = FunctionInput()
  n9 = ComponentMask(None=@n8) {g=true}
  n10 = ComponentMask(None=@n8) {b=true}
  n11 = AppendVector(A=@n9, B=@n10)
  n12 = StaticSwitchParameter(True=AppendVector(A=@n10, B=@n9), False=@n11) {group=Side Projection UVS, parameterName=Rotate Pickaxe Projection, sortPriority=-4}
  n13 = NamedRerouteUsage()
  n14 = Multiply(
   A=LinearInterpolate(A=AppendVector(A=@n1.R, B=@n1.R), B=AppendVector(A=@n2.R, B=@n2.R), Alpha=@n3)
   B=Add(A=LinearInterpolate(A=@n5, B=@n7, Alpha=@n3), B=LinearInterpolate(A=@n11, B=@n12, Alpha=@n13))
  )
  n15 = ComponentMask(None=@n8) {r=true}
  n16 = OneMinus(None=@n10)
  n17 = FunctionInput() {bUsePreviewValueAsDefault=true, inputName=Side Project Vehicle, inputType=FunctionInput_Scalar}
  n18 = LinearInterpolate(
   A=StaticSwitchParameter(True=AppendVector(A=@n9, B=@n15), False=AppendVector(A=@n15, B=@n9)) {group=Side Projection UVS, parameterName=Rotate Vehicle Projection, sortPriority=-4}
   B=StaticSwitchParameter(True=AppendVector(A=@n16, B=@n15), False=AppendVector(A=@n15, B=@n16)) {group=Side Projection UVS, parameterName=Rotate Vehicle SP Projection, sortPriority=-4}
   Alpha=@n17
  )
  n19 = VectorParameter() {desc="R=Y  G=X", group=Side Projection UVS, parameterName=UVs_Offset_Vehicle, sortPriority=4, defaultValue=<rgba 0 0 0 1>}
  n20 = ComponentMask(None=@n19.RGB) {g=true, r=true}
  n21 = VectorParameter() {desc="R=  G=  A=Uniform", group=Side Projection UVS, parameterName=UVs_Scale_Vehicle, sortPriority=3, defaultValue=<rgba 1 1 1 1>}
  n22 = VectorParameter() {desc="R=Y  G=X", group=Side Projection UVS, parameterName=UVs_Offset_Vehicle_Side, sortPriority=7, defaultValue=<rgba 0 0 0 1>}
  n23 = ComponentMask(None=@n22.RGB) {g=true, r=true}
  n24 = VectorParameter() {desc="R=  G=  A=Uniform", group=Side Projection UVS, parameterName=UVs_Scale_Vehicle_Side, sortPriority=6, defaultValue=<rgba 1 1 1 1>}
  n25 = LinearInterpolate(
   A=Multiply(A=Add(A=@n18, B=@n20), B=AppendVector(A=@n21.R, B=@n21.R))
   B=Multiply(A=Add(A=@n23, B=@n18), B=AppendVector(A=@n24.R, B=@n24.R))
   Alpha=Multiply(A=@n17, B=ScalarParameter() {group=Side Projection UVS, parameterName="UVs_Vehicle_Side 0=off 1=Active", sortPriority=5})
  )
  n26 = FunctionInput() {inputName=isVehicle, inputType=FunctionInput_Scalar}
  n27 = AppendVector(A=@n15, B=@n10)
  n28 = AppendVector(A=@n10, B=@n15)
  n29 = LinearInterpolate(
   A=StaticSwitchParameter(True=@n27, False=@n28) {group=Side Projection UVS, parameterName=Rotate Mech Projection, sortPriority=-4}
   B=LinearInterpolate(
    A=StaticSwitchParameter(True=@n27, False=@n28) {group=Side Projection UVS, parameterName=Rotate Character SP Projection, sortPriority=-4}
    B=StaticSwitchParameter(True=@n27, False=@n28) {group=Side Projection UVS, parameterName=Rotate Backpack SP Projection, sortPriority=-4}
    Alpha=NamedRerouteUsage()
   )
   Alpha=NamedRerouteUsage()
  )
  n30 = VectorParameter() {desc="R=Y  G=X", group=Side Projection UVS, parameterName=UVs_Offset_Mech, sortPriority=9, defaultValue=<rgba 0 0 0 1>}
  n31 = VectorParameter() {desc="R=  G=  A=Uniform", group=Side Projection UVS, parameterName=UVs_Scale_Mech, sortPriority=8, defaultValue=<rgba 1 1 1 1>}
  n32 = Multiply(A=Add(A=@n29, B=ComponentMask(None=@n30.RGB) {g=true, r=true}), B=AppendVector(A=@n31.R, B=@n31.R))
  n33 = VectorParameter() {desc="R=Y  G=X", group=Side Projection UVS, parameterName=UVs_Offset_Character, sortPriority=9, defaultValue=<rgba 0 0 0 1>}
  n34 = ComponentMask(None=@n33.RGB) {g=true, r=true}
  n35 = VectorParameter() {desc="R=Y  G=X", group=Side Projection UVS, parameterName=UVs_Offset_Backpack, sortPriority=9, defaultValue=<rgba 0 0 0 1>}
  n36 = ComponentMask(None=@n35.RGB) {g=true, r=true}
  n37 = NamedRerouteUsage()
  n38 = VectorParameter() {desc="R=  G=  A=Uniform", group=Side Projection UVS, parameterName=UVs_Scale_Character, sortPriority=8, defaultValue=<rgba 1 1 1 1>}
  n39 = VectorParameter() {desc="R=  G=  A=Uniform", group=Side Projection UVS, parameterName=UVs_Scale_Backpack, sortPriority=8, defaultValue=<rgba 1 1 1 1>}
  n40 = Multiply(
   A=Add(A=@n29, B=LinearInterpolate(A=@n34, B=@n36, Alpha=@n37))
   B=LinearInterpolate(A=AppendVector(A=@n38.R, B=@n38.R), B=AppendVector(A=@n39.R, B=@n39.R), Alpha=@n37)
  )
  n41 = FunctionInput() {bUsePreviewValueAsDefault=true, inputName=isCharacter, inputType=FunctionInput_Scalar, sortPriority=100}
  FunctionOutput(None=LinearInterpolate(
    A=LinearInterpolate(A=@n14, B=@n25, Alpha=@n26)
    B=LinearInterpolate(A=@n32, B=@n40, Alpha=@n41)
    Alpha=FunctionInput() {bUsePreviewValueAsDefault=true, inputName=isMech, inputType=FunctionInput_Scalar}
   ))
  NamedRerouteDeclaration(None=FunctionInput() {bUsePreviewValueAsDefault=true, inputName=IsBackpack, inputType=FunctionInput_Scalar, sortPriority=100}) {name=IsBackpack, nodeColor=<rgba 0 0.9253 1 1>}
  NamedRerouteDeclaration(None=FunctionInput() {bUsePreviewValueAsDefault=true, inputName=IsPickaxe, inputType=FunctionInput_Scalar, sortPriority=100}) {name=IsPickaxe, nodeColor=<rgba 0 0.9253 1 1>}
  NamedRerouteDeclaration(None=@n41) {name=ISCharacter, nodeColor=<rgba 0 1 0.6719 1>}
  VectorParameter() {parameterName=Forward Axis, defaultValue=<rgba 0 0 0 1>}
 ]}
