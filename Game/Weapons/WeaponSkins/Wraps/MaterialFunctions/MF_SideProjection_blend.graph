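# Material function for weapon, vehicle and mech wraps that blends a second side-projection layer over the incoming material.
# MF_Wrap_Gradent_Creation builds a gradient or warped mask-texture UVs, which HeightLerp combines with the first projection's mask, or masked_Values supplies the alpha.
# That alpha drives base color, normal, roughness and metallic blends, with static switches for simple values or bypass.
# The non-fixed bounds path divides by bounds Max minus Max, which is zero.
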MaterialFunction /Game/Weapons/WeaponSkins/Wraps/MaterialFunctions/MF_SideProjection_blend
graph {nodes [
  n1 = FunctionInput() {inputName=material, inputType=FunctionInput_MaterialAttributes}
  n2 = GetMaterialAttributes(None=@n1) {attributeGetTypes [
    #1
    #2
    #3
    #4
   ]}
  n3 = VectorParameter() {group=Side Projection, parameterName=SideProejction 02 SimpleColor, defaultValue=<rgba 0 0 0 1>}
  n4 = FunctionInput() {inputName=Uvs, inputType=FunctionInput_Vector2, sortPriority=1}
  n5 = TextureSampleParameter2D(UVs=@n4) {group=Side Projection, parameterName=SideProejction 02 Diffuse, texture=/Game/Global/Textures/Default/Masks_and_Gradients/3pxRGBImage}
  n6 = PreSkinnedLocalBounds()
  n7 = Subtract(A=FunctionInput() {bUsePreviewValueAsDefault=true, inputName=Pre-Skinned Local Position, sortPriority=1}, B=@n6.Min)
  n8 = Subtract(A=@n6.Max, B=@n6.Max)
  n9 = VectorParameter() {group=Side Projection, parameterName=Height_UV_Spread_XY, sortPriority=67, defaultValue=<rgba 1 1 1 0>}
  n10 = FunctionInput() {inputName=SideProjection01_Mask, sortPriority=2}
  n11 = ChannelMaskParameter(None=@n10) {group=Side Projection, maskChannel=Green, parameterName=SideProjection_01 WarpUVs Channel BL, sortPriority=65, defaultValue=<rgba 0 1 0 0>}
  n12 = MaterialFunctionCall!(
   ForwardAxis=FunctionInput() {bUsePreviewValueAsDefault=true, inputName=Forward Axis}
   Bounds=StaticSwitchParameter(True=FunctionInput() {bUsePreviewValueAsDefault=true, inputName=Bounds}, False=Divide(A=@n7, B=@n8)) {group=Side Projection, parameterName=Height_Object FixedBounds, sortPriority=60}
   Pre-Skinned Local Position
   isVehicle=FunctionInput() {inputName=isVehicle, inputType=FunctionInput_Scalar, sortPriority=2}
   isMech=FunctionInput() {bUsePreviewValueAsDefault=true, inputName=isMech, inputType=FunctionInput_Scalar, sortPriority=2}
   Side Project Vehicle=FunctionInput() {bUsePreviewValueAsDefault=true, inputName=Side Project Vehicle, inputType=FunctionInput_Scalar}
   WeaponOffset=ScalarParameter() {group=Side Projection, parameterName=Height_WeaponOffset, sortPriority=61}
   WeaponFalloff=ScalarParameter() {defaultValue=1, group=Side Projection, parameterName=HeightWeaponFalloff, sortPriority=62}
   VehicleOffset=ScalarParameter() {group=Side Projection, parameterName=Height_VehicleOffset, sortPriority=63}
   VehicleFalloff=ScalarParameter() {defaultValue=1, group=Side Projection, parameterName=Height_VehicleFalloff, sortPriority=64}
   MechOffset=ScalarParameter() {group=Side Projection, parameterName=Height_MechOffset, sortPriority=63}
   MechFalloff=ScalarParameter() {defaultValue=1, group=Side Projection, parameterName=Height_MechFalloff, sortPriority=63}
   Invert_Gradent=StaticBoolParameter() {group=Side Projection, parameterName=Height_Invert_Gradent, sortPriority=60}
   UV_Spread_XY=@n9.RGB
   UV_Animated_XY=VectorParameter() {group=Side Projection, parameterName=Height_UV_Animated_XY, sortPriority=68}.RGB
   Warp_UVs=Multiply(A=@n11.R, B=ScalarParameter() {parameterName=WarpUVs Amount BL, sortPriority=66})
   Rotate_Mech
   Rotate_Vehicle_Side
   Rotate_Vehicle
  ) {materialFunction=&/MF_Wrap_Gradent_Creation}
  n13 = TextureSampleParameter2D(UVs=@n12.Uvs) {
   group=Side Projection
   parameterName=Height_Blend_Texture
   samplerType=SAMPLERTYPE_Masks
   sortPriority=61
   texture=/Game/Global/Textures/Default/Masks_and_Gradients/T_VertGradient_MASK
  }
  n14 = ChannelMaskParameter(None=@n13.RGBA) {group=Side Projection, parameterName=HeightBlend-Gradient_Channel, sortPriority=61}
  n15 = ChannelMaskParameter(None=@n10) {group=Side Projection, maskChannel=Green, parameterName=SideProjection_01 Mask Channel BL, sortPriority=60, defaultValue=<rgba 0 1 0 0>}
  n16 = StaticSwitchParameter(
   True=MaterialFunctionCall(
    Transition Phase=StaticSwitchParameter(True=@n14.R, False=@n12.Gradent) {group=Side Projection, parameterName="-Blend-Type-Gradent\\Texture-----------", sortPriority=60}
    Height Texture=@n15.R
    Contrast=ScalarParameter() {defaultValue=1, group=Side Projection, parameterName=Height_Contrast, sortPriority=60}
   ) {materialFunction=&HeightLerp}.Alpha
   False=Saturate(None=FunctionInput() {inputName=masked_Values, inputType=FunctionInput_Scalar, sortPriority=2})
  ) {group=Side Projection, parameterName="-Blend-Type-Gradent\\Height-----------", sortPriority=31}
  n17 = TextureSampleParameter2D(UVs=@n4) {
   group=Side Projection
   parameterName=SideProejction 02 Normal
   samplerType=SAMPLERTYPE_Normal
   sortPriority=41
   texture=/Engine/EngineMaterials/FlatNormal
  }
  n18 = MaterialFunctionCall(BaseNormal=FunctionInput() {inputName=Normals, sortPriority=3}, AdditionalNormal=@n17.RGB) {materialFunction=&BlendAngleCorrectedNormals}
  n19 = StaticSwitchParameter(
   True=ScalarParameter() {defaultValue=0.5, group=Side Projection, parameterName=SideProejction 02 SimpleRoghness, sortPriority=35}
   False=@n5.A
  ) {group=Side Projection, parameterName=Blend-Sideprojection 02 SimpleRoughness, sortPriority=34}
  n20 = TextureSampleParameter2D(UVs=@n4) {
   group=Side Projection
   parameterName=SideProejction 02 Mask
   sortPriority=52
   texture=&3pxRGBImage
  }
  n21 = StaticSwitchParameter(
   True=ScalarParameter() {defaultValue=0.5, group=Side Projection, parameterName=SideProejction 02 SimpleMetallic, sortPriority=52}
   False=@n20.G
  ) {group=Side Projection, parameterName=Blend-Sideprojection 02 SimpleMetallic, sortPriority=51}
  n22 = SetMaterialAttributes!(
   MaterialAttributes=@n2.MaterialAttributes
   BaseColor=LinearInterpolate(
    A=@n2.BaseColor
    B=StaticSwitchParameter(True=@n3.RGB, False=@n5.RGB) {group=Side Projection, parameterName=Blend-Sideprojection 02 SimpleColor, sortPriority=31}
    Alpha=@n16
   )
   Normal=StaticSwitchParameter(True=LinearInterpolate(A=@n2.Normal, B=@n18.Result, Alpha=@n16), False=@n2.Normal) {group=Side Projection, parameterName=Blend-Sideprojection 02 Normals, sortPriority=40}
   Roughness=StaticSwitchParameter(True=LinearInterpolate(A=@n2.Roughness, B=@n19, Alpha=@n16), False=@n2.Roughness) {group=Side Projection, parameterName="Blend-Sideprojection 02 Roughness (alpha)", sortPriority=33}
   Metallic=StaticSwitchParameter(True=LinearInterpolate(A=@n2.Metallic, B=@n21, Alpha=@n16), False=@n2.Metallic) {group=Side Projection, parameterName="Blend-Sideprojection 02 Metalic (Green)", sortPriority=50}
  ) {attributeSetTypes [
    #1
    #2
    #3
    #4
   ]}
  n23 = StaticSwitchParameter(True=@n22, False=@n1) {
   desc="Using gradient to blend between sideprojection 01 to sideprojection 02 "
   group=Side Projection
   parameterName=-Blend-Sideprojection-01-to-SP-02-------
   sortPriority=30
  }
  FunctionOutput(None=@n23) {outputName=output}
 ]}
