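# Material function for weapon skin wraps that blends a side-projection material layer over a base layer using a single-axis triplanar-style mask.
# The mask is contrast-adjusted absolute pre-skinned local normal: X by default, Z or Y for vehicles, Y for mechs.
# Static switches can multiply or replace it with a side-projection mask channel or diffuse alpha.
# With the main switch off, the side layer passes through unblended.
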MaterialFunction /Game/Weapons/WeaponSkins/Wraps/MaterialFunctions/MF_Side_TriBlendMask
graph {nodes [
  n1 = FunctionInput() {inputName=Base, inputType=FunctionInput_MaterialAttributes}
  n2 = GetMaterialAttributes(None=FunctionInput() {inputName=Side, inputType=FunctionInput_MaterialAttributes, sortPriority=1})
  n3 = StaticSwitchParameter(
   True=ChannelMaskParameter(None=FunctionInput() {bUsePreviewValueAsDefault=true, inputName=SideProejction 01 Mask}) {group=Side Projection, parameterName="SDP01 Mask Channel ", sortPriority=-46}.R
   False=FunctionInput() {bUsePreviewValueAsDefault=true, inputName=SideProejction 01 Diffuse Alpha, inputType=FunctionInput_Scalar}
  ) {group=Side Projection, parameterName=Use SideProejction 01 Mask, sortPriority=-47}
  n4 = FunctionInput(Preview=VertexInterpolator(VS=PreSkinnedNormal()).PS) {bUsePreviewValueAsDefault=true, inputName=Pre-Skinned Local Normal, sortPriority=42}
  n5 = ScalarParameter() {defaultValue=1, group=Side Projection, parameterName=SideProjecton_TriplanerMask_Contrast, sortPriority=-49}
  n6 = MaterialFunctionCall(In=Abs(None=ComponentMask(None=@n4) {r=true}), Contrast=@n5) {materialFunction=&CheapContrast}
  n7 = ComponentMask(None=@n4) {b=true}
  n8 = ComponentMask(None=MaterialFunctionCall(In=Abs(None=@n7), Contrast=@n5) {materialFunction=&CheapContrast}.Result) {r=true}
  n9 = ComponentMask(None=@n4) {g=true}
  n10 = ComponentMask(None=MaterialFunctionCall(In=Abs(None=@n9), Contrast=@n5) {materialFunction=&CheapContrast}.Result) {r=true}
  n11 = FunctionInput(Preview=Constant()) {bUsePreviewValueAsDefault=true, inputName=Side Project Vehicle, inputType=FunctionInput_Scalar}
  n12 = LinearInterpolate(
   A=LinearInterpolate(
    A=ComponentMask(None=@n6.Result) {r=true}
    B=LinearInterpolate(A=@n8, B=@n10, Alpha=@n11)
    Alpha=FunctionInput() {inputName=isVehicle, inputType=FunctionInput_Scalar, sortPriority=2}
   )
   B=@n10
   Alpha=FunctionInput() {bUsePreviewValueAsDefault=true, inputName=isMech, inputType=FunctionInput_Scalar}
  )
  n13 = BlendMaterialAttributes(
   A=GetMaterialAttributes(None=@n1).MaterialAttributes
   B=@n2.MaterialAttributes
   Alpha=StaticSwitchParameter(
    True=Multiply(A=@n3, B=@n12)
    False=StaticSwitchParameter(True=@n3, False=@n12) {group=Side Projection, parameterName=SideProejction 01 Diffuse Alpha Only, sortPriority=-48}
   ) {group=Side Projection, parameterName=TriplanerMask Mask by SideProejction 01 Diffuse Alpha, sortPriority=-48}
  ) {vertexAttributeBlendType=UseA}
  FunctionOutput(None=StaticSwitchParameter(True=SetMaterialAttributes!(MaterialAttributes=@n13), False=@n2.MaterialAttributes) {group=Side Projection, parameterName=-SideProjecton-TriplanerMask----, sortPriority=-50})
  GetMaterialAttributes(None=@n1) {attributeGetTypes [
    #1
    #2
    #3
    #4
   ]}
 ]}
