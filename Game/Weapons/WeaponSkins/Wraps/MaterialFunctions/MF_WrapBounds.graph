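# Material function for weapon wraps that converts a local position into bounds-normalized coordinates, with special handling for vehicles and mechs.
# Adjusted pre-skinned local bounds have their Y zeroed by the vehicle input, which the mech input reverts.
# The position minus the minimum is divided by the larger X or Z size, ignoring Y, then offset by a center adjustment and vehicle-weighted Perm_VehicleCenter.
# It also outputs raw bound extents.
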MaterialFunction /Game/Weapons/WeaponSkins/Wraps/MaterialFunctions/MF_WrapBounds
graph {nodes [
  Constant()
  FunctionInput() {bUsePreviewValueAsDefault=true, inputName=Side Project Vehicle, inputType=FunctionInput_Scalar}
  n1 = PreSkinnedLocalBounds()
  n2 = Add(A=@n1.Min, B=FunctionInput() {inputName=Object Bounds Min Adjustment})
  n3 = MaterialFunctionCall(Color=@n2) {materialFunction=&SplitComponents}
  n4 = Constant()
  n5 = AppendVector(A=@n3.R, B=@n4)
  n6 = FunctionInput() {inputName=is Vehicle, inputType=FunctionInput_Scalar}
  n7 = FunctionInput() {bUsePreviewValueAsDefault=true, inputName=isMech, inputType=FunctionInput_Scalar}
  n8 = LinearInterpolate(A=LinearInterpolate(A=@n2, B=AppendVector(A=@n5, B=@n3.B), Alpha=@n6), B=@n2, Alpha=@n7)
  n9 = Subtract(A=FunctionInput() {inputName=Local Position}, B=@n8)
  n10 = Add(A=@n1.Max, B=FunctionInput() {inputName=Object Bounds Max Adjustment})
  n11 = MaterialFunctionCall(Color=@n10) {materialFunction=&SplitComponents}
  n12 = LinearInterpolate(A=@n10, B=AppendVector(A=AppendVector(A=@n11.R, B=@n4), B=@n11.B), Alpha=@n6)
  n13 = MaterialFunctionCall(Color=Subtract(A=LinearInterpolate(A=@n12, B=@n10, Alpha=@n7), B=@n8)) {materialFunction=&SplitComponents}
  n14 = Max(A=Max(A=@n13.R, B=@n13.R), B=@n13.B)
  n15 = LinearInterpolate(B=VectorParameter() {parameterName=Perm_VehicleCenter, defaultValue=<rgba 0 0.5 0 1>}.RGB, Alpha=@n6)
  n16 = FunctionInput() {bUsePreviewValueAsDefault=true, inputName=Object Bounds Center Adjustment}
  FunctionOutput(None=Add(A=Divide(A=@n9, B=@n14), B=Add(A=@n15, B=@n16)))
  FunctionOutput(None=@n1.Extents) {outputName=Bound Extents}
 ]}
