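# Material function for weapon and vehicle wraps that outputs an invertible 0–1 gradient along a ForwardAxis input.
# Per Obj Bounds, defaulting to bounds-relative pre-skinned position, gets offset and falloff blended by isVehicle, then axis masking and power shaping.
# As wired, the default bounds divisor is Max minus Max, and the final component max reads R and B, skipping G.
# The object bounds/pivot switches do not reach the output.
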MaterialFunction /Game/Weapons/WeaponSkins/Wraps/MaterialFunctions/MF_WrapGradient
graph {nodes [
  n1 = PreSkinnedLocalBounds()
  n2 = FunctionInput(Preview=VertexInterpolator(VS=PreSkinnedPosition()).PS) {inputName=Pre Skinned Local Position}
  n3 = Subtract(A=@n1.Max, B=@n1.Max)
  n4 = Divide(A=Subtract(A=@n2, B=@n1.Min), B=@n3)
  n5 = FunctionInput(Preview=Constant()) {inputName=isVehicle, inputType=FunctionInput_Scalar}
  n6 = LinearInterpolate(
   A=ScalarParameter() {group=Primary Triplaner Gradient, parameterName=Wrap Gradient_WeaponOffset, sortPriority=15}
   B=ScalarParameter() {group=Primary Triplaner Gradient, parameterName=Wrap Gradient_VehicleOffset, sortPriority=17}
   Alpha=@n5
  )
  n7 = Divide(
   A=Add(A=FunctionInput(Preview=@n4) {bUsePreviewValueAsDefault=true, inputName=Per Obj Bounds, sortPriority=50}, B=@n6)
   B=LinearInterpolate(
    A=ScalarParameter() {defaultValue=0.5, group=Primary Triplaner Gradient, parameterName=Wrap Gradient_WeaponFalloff, sortPriority=14}
    B=ScalarParameter() {defaultValue=0.5, group=Primary Triplaner Gradient, parameterName=Wrap Gradient_VehicleFalloff, sortPriority=16}
    Alpha=@n5
   )
  )
  n8 = ConstantBiasScale(None=@n7)
  n9 = Multiply(A=Clamp(None=@n8), B=FunctionInput(Preview=Constant3Vector() {constant=<rgba 0 1 0 0>}) {inputName=ForwardAxis})
  n10 = ScalarParameter() {group=Primary Triplaner Gradient, parameterName=Wrap Gradient_ GradientOffset, sortPriority=14}
  n11 = MaterialFunctionCall(Float3=Power(
    Base=Add(A=@n9, B=@n10)
    Exp=ScalarParameter() {defaultValue=0.75, group=Primary Triplaner Gradient, parameterName=Wrap Gradient_ Power, sortPriority=1}
   )) {materialFunction=&BreakOutFloat3Components}
  n12 = Saturate(None=Max(A=Max(A=@n11.R, B=@n11.R), B=@n11.B))
  n13 = OneMinus(None=@n12)
  n14 = ScalarParameter() {group=Primary Triplaner Gradient, parameterName=Wrap Gradient_ invert Gradient}
  FunctionOutput(None=Saturate(None=LinearInterpolate(A=@n12, B=@n13, Alpha=@n14)))
  Multiply(A=@n6, B=@n3) {constB=-1}
  n15 = ScalarParameter() {defaultValue=40, group=Primary Triplaner Gradient, parameterName=Wrap Gradient_ WeaponFalloff, sortPriority=15}
  n16 = Multiply(
   A=ObjectRadius()
   B=ScalarParameter() {defaultValue=0.1, group=Primary Triplaner Gradient, parameterName=Wrap Gradient_VehicleRadius Falloff, sortPriority=16}
  )
  StaticSwitchParameter(True=@n7, False=Divide(A=@n2, B=LinearInterpolate(A=@n15, B=@n16, Alpha=@n5))) {
   desc="use object bound set gradent .. position will be at the front to back of the object "
   group=Primary Triplaner Gradient
   parameterName="Use Objectbounds\\ObjectPivot"
   sortPriority=-20
  }
  n17 = ScalarParameter() {group=Primary Triplaner Gradient, parameterName=Wrap Gradient_ WeaponOffset, sortPriority=14}
  n18 = ScalarParameter() {group=Primary Triplaner Gradient, parameterName=Wrap Gradient_ VehicleOffset, sortPriority=17}
  StaticSwitchParameter(True=@n8, False=Add(A=LinearInterpolate(A=@n17, B=@n18, Alpha=@n5), B=@n8)) {
   desc="use object bound set gradent .. position will be at the front of objects. "
   group=Primary Triplaner Gradient
   parameterName="Use Objectbounds\\ObjectPivot"
   sortPriority=-20
  }
 ]}
