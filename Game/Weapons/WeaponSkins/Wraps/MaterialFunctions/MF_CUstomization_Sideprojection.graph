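# Weapon wrap customization function that layers a side-projected texture material onto weapons, vehicles, mechs, characters and backpacks.
# UVs come from bounds-based side projection math or a legacy per-type calculation, with optional skew, warp and panning.
# Projected textures set material attributes, then pass through blend, emissive, tri-blend mask, iridescent, animated diffuse and fireworks functions.
# The result blends over the input material by customization mask, gated by a static switch.
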MaterialFunction /Game/Weapons/WeaponSkins/Wraps/MaterialFunctions/MF_CUstomization_Sideprojection.MF_Customization_Sideprojection
graph {nodes [
  n1 = FunctionInput() {inputName=Material, inputType=FunctionInput_MaterialAttributes}
  n2 = GetMaterialAttributes(None=@n1) {attributeGetTypes [
    #1
    #2
    #3
    #4
   ]}
  n3 = VectorParameter() {group=Side Projection, parameterName=Simple Diffuse and Roughness, defaultValue=<rgba 0 0 0 0.5>}
  n4 = StaticBoolParameter() {group=Side Projection, parameterName=Use Simple Diffuse and Roughness}
  n5 = FunctionInput(Preview=VertexInterpolator(VS=PreSkinnedPosition()).PS) {bUsePreviewValueAsDefault=true, inputName=Local Position}
  n6 = FunctionInput() {bUsePreviewValueAsDefault=true, inputName=isMech, inputType=FunctionInput_Scalar}
  n7 = FunctionInput(Preview=Constant() {r=1}) {inputName=isVehicle, inputType=FunctionInput_Scalar}
  n8 = MaterialFunctionCall!(
   is Vehicle=@n7
   Object Bounds Max Adjustment=FunctionInput() {bUsePreviewValueAsDefault=true, inputName=Object Bounds Max Adjustment}
   Object Bounds Min Adjustment=FunctionInput() {bUsePreviewValueAsDefault=true, inputName=Object Bounds Min Adjustment}
   Object Bounds Center Adjustment=FunctionInput() {bUsePreviewValueAsDefault=true, inputName=Object Bounds Center Adjustment}
   isMech=@n6
   Side Project Vehicle
   Local Position=@n5
  ) {materialFunction=&/MF_WrapBounds}
  n9 = MaterialFunctionCall(Float2=Divide(A=@n5, B=@n8.Bound Extents)) {materialFunction=&BreakOutFloat2Components}
  n10 = OneMinus(None=LinearInterpolate(A=@n9.R, B=@n9.R, Alpha=@n7))
  n11 = VectorParameter() {desc=Alpha unform Scale, group=Side Projection UVS, parameterName=Scale_Weapon, sortPriority=2, defaultValue=<rgba 1 0.95 -0.5 1>}
  n12 = AppendVector(A=@n11.G, B=@n11.G)
  n13 = FunctionInput(Preview=VertexInterpolator(VS=PreSkinnedLocalBounds().Half Extents).PS) {bUsePreviewValueAsDefault=true, inputName=Local Bounds}
  n14 = MaterialFunctionCall(Color=@n13) {materialFunction=&SplitComponents}
  n15 = StaticSwitchParameter(True=Max(A=Max(A=@n14.R, B=@n14.R), B=@n14.B), False=@n13) {
   desc=uses the maximim size to set the texture to power of 2
   group=Side Projection UVS
   parameterName=SideProjection_UVs MakeSquare
   sortPriority=-5
  }
  n16 = Divide(A=@n5, B=@n15)
  n17 = ComponentMask(None=@n16) {g=true}
  n18 = Time()
  n19 = VectorParameter() {group=Side Projection UVS, parameterName=Pan_Weapon, sortPriority=1, defaultValue=<rgba 0 0 0 1>}
  n20 = AppendVector(A=@n19.G, B=@n19.G)
  n21 = VectorParameter() {group=Side Projection UVS, parameterName=Offset_Weapon, sortPriority=1, defaultValue=<rgba 0 0.15 -0.25 1>}
  n22 = Add(
   A=AppendVector(A=@n17, B=ComponentMask(None=@n16) {b=true})
   B=StaticSwitchParameter(True=Multiply(A=@n18, B=@n20), False=AppendVector(A=@n21.G, B=@n21.G)) {group=Side Projection UVS, parameterName=AnimatedUVS}
  )
  n23 = ComponentMask(None=@n16) {g=true}
  n24 = ComponentMask(None=@n16) {r=true}
  n25 = ComponentMask(None=@n16) {r=true}
  n26 = Time()
  n27 = VectorParameter() {group=Side Projection UVS, parameterName=Pan_Vehicle, sortPriority=3, defaultValue=<rgba 0 0 0 1>}
  n28 = AppendVector(A=@n27.G, B=@n27.G)
  n29 = VectorParameter() {group=Side Projection UVS, parameterName=Offset_Vehicle, sortPriority=3, defaultValue=<rgba 0.5 0.55 0 1>}
  n30 = Add(
   A=StaticSwitchParameter(True=AppendVector(A=@n23, B=@n24), False=AppendVector(A=@n25, B=@n17)) {group=Side Projection UVS, parameterName=Rotate Vehicle Projection, sortPriority=-4}
   B=StaticSwitchParameter(True=Multiply(A=@n26, B=@n28), False=AppendVector(A=@n29.G, B=@n29.G)) {group=Side Projection UVS, parameterName=AnimatedUVS}
  )
  n31 = VectorParameter() {desc=Alpha unform Scale, group=Side Projection UVS, parameterName=Scale_Vehicle, sortPriority=4, defaultValue=<rgba 1 1 1 1>}
  n32 = AppendVector(A=@n31.G, B=@n31.G)
  n33 = LinearInterpolate(A=Multiply(A=@n11.A, B=Multiply(A=@n12, B=@n22)), B=Multiply(A=Multiply(A=@n30, B=@n32), B=@n31.A), Alpha=@n7)
  n34 = AppendVector(A=ComponentMask(None=@n16) {b=true}, B=ComponentMask(None=@n16) {r=true})
  n35 = VectorParameter() {group=Side Projection UVS, parameterName=Pan_Mech, sortPriority=7, defaultValue=<rgba 0 0 0 1>}
  n36 = VectorParameter() {group=Side Projection UVS, parameterName=Offset_Mech, sortPriority=7, defaultValue=<rgba 0.5 0.55 0 1>}
  n37 = StaticSwitchParameter(True=Multiply(A=Time(), B=AppendVector(A=@n35.G, B=@n35.G)), False=AppendVector(A=@n36.G, B=@n36.G)) {group=Side Projection UVS, parameterName=AnimatedUVS}
  n38 = VectorParameter() {desc=Alpha unform Scale, group=Side Projection UVS, parameterName=Scale_Mech, sortPriority=6, defaultValue=<rgba 1 1 1 1>}
  n39 = Multiply(A=Multiply(A=Add(A=@n34, B=@n37), B=AppendVector(A=@n38.G, B=@n38.G)), B=@n38.A)
  n40 = FunctionInput() {bUsePreviewValueAsDefault=true, inputName=Side Project Vehicle, inputType=FunctionInput_Scalar}
  n41 = FunctionInput() {bUsePreviewValueAsDefault=true, inputName=isCharacter, inputType=FunctionInput_Scalar, sortPriority=100}
  n42 = MaterialFunctionCall!(
   isVehicle=@n7
   isMech=@n6
   Side Project Vehicle=@n40
   In=@n8.Result
   IsBackpack=NamedRerouteUsage()
   IsPickaxe=NamedRerouteUsage()
   isCharacter=@n41
  ) {materialFunction=&/MF_SideProjection_Math}
  n43 = Add(
   A=Multiply(A=Abs(None=@n10), B=ScalarParameter() {group=Side Projection, parameterName=Side Projection Skew Factor})
   B=StaticSwitchParameter(True=LinearInterpolate(A=@n33, B=@n39, Alpha=@n6), False=@n42.Result) {group=Side Projection UVS, parameterName=use Legacy side Projection Calculation}
  )
  n44 = TextureSampleParameter2D(UVs=@n42.Result) {
   group=Side Projection UVS
   parameterName=SideProejction WarpUVS Mask
   samplerType=SAMPLERTYPE_Masks
   sortPriority=104
   texture=&T_BlackMask
  }
  n45 = LinearInterpolate(
   A=LinearInterpolate(
    A=ScalarParameter() {defaultValue=1, group=Side Projection UVS, parameterName=Side_WarpUvs_Weapon_TileRate, sortPriority=115}
    B=ScalarParameter() {defaultValue=1, group=Side Projection UVS, parameterName=Side_WarpUvs_Vehicle_TileRate, sortPriority=116}
    Alpha=@n7
   )
   B=ScalarParameter() {defaultValue=1, group=Side Projection UVS, parameterName=Side_WarpUvs_Mech_TileRate, sortPriority=117}
   Alpha=@n6
  )
  n46 = Time()
  n47 = VectorParameter() {group=Side Projection UVS, parameterName=Side_WrapAnimation_Weapon, sortPriority=112, defaultValue=<rgba 0 0 0 1>}
  n48 = VectorParameter() {group=Side Projection UVS, parameterName=Side_WrapAnimation_Vehicle, sortPriority=113, defaultValue=<rgba 0 0 0 1>}
  n49 = VectorParameter() {group=Side Projection UVS, parameterName=Side_WrapAnimation_Mech, sortPriority=114, defaultValue=<rgba 0 0 0 1>}
  n50 = ComponentMask(None=LinearInterpolate(A=LinearInterpolate(A=@n47.RGB, B=@n48.RGB, Alpha=@n7), B=@n49.RGB, Alpha=@n6)) {g=true, r=true}
  n51 = TextureSampleParameter2D(UVs=Add(A=Multiply(A=@n42.Result, B=@n45), B=Multiply(A=@n46, B=@n50))) {
   group=Side Projection UVS
   parameterName=SideProejction WarpUVS
   samplerType=SAMPLERTYPE_Masks
   sortPriority=111
   texture=&T_BlackMask
  }
  n52 = ChannelMaskParameter(None=@n51.RGBA) {group=Side Projection UVS, parameterName=SideWarpUVs Channel, sortPriority=110}
  n53 = StaticSwitch(
   True=Multiply(A=ChannelMaskParameter(None=@n44.RGBA) {group=Side Projection UVS, parameterName=SideMask Channel WarpUVSMask, sortPriority=103}.R, B=@n52.R)
   False=@n52.R
   Value=StaticBoolParameter() {group=Side Projection UVS, parameterName=SideWarpUVs_Mask, sortPriority=102}
  )
  n54 = ScalarParameter() {group=Side Projection UVS, parameterName=SideWarpUvs_Amount, sortPriority=101}
  n55 = StaticSwitch(
   True=Add(A=@n43, B=Multiply(A=@n53, B=@n54))
   False=@n43
   Value=StaticBoolParameter() {group=Side Projection UVS, parameterName=SideWarpUVs, sortPriority=100}
  )
  n56 = ScalarParameter() {defaultValue=1, parameterName=ScaleSideProjectionUV_Pliant_Backpack}
  n57 = VectorParameter() {group=Side Projection UVS, parameterName=UV_Panning_Weapon_01}
  n58 = MaterialFunctionCall(Float4=NamedRerouteUsage()) {materialFunction=&BreakOutFloat4Components}
  n59 = LinearInterpolate(
   A=VectorParameter() {group=Side Projection UVS, parameterName=UV_Panning_Vehicle_01}.RGB
   B=VectorParameter() {group=Side Projection UVS, parameterName=UV_Panning_Vehicle_Side_01}.RGB
   Alpha=@n58.G
  )
  n60 = VectorParameter() {group=Side Projection UVS, parameterName=UV_Panning_Mech_01}
  n61 = VectorParameter() {group=Side Projection UVS, parameterName=UV_Panning_Character_01}
  n62 = LinearInterpolate(
   A=LinearInterpolate(A=LinearInterpolate(A=@n57.RGB, B=@n59, Alpha=@n58.R), B=LinearInterpolate(A=@n60.RGB, B=@n61.RGB, Alpha=@n58.A), Alpha=@n58.B)
   B=VectorParameter() {group=Side Projection UVS, parameterName=UV_Panning_Pliant_Backpack_01}.RGB
   Alpha=NamedRerouteUsage()
  )
  n63 = Add(
   A=LinearInterpolate(A=@n55, B=Multiply(A=@n55, B=@n56), Alpha=NamedRerouteUsage())
   B=Multiply(A=Time(), B=ComponentMask(None=@n62) {g=true, r=true})
  )
  n64 = TextureSampleParameter2D(UVs=@n63) {group=Side Projection, parameterName=SideProejction 01 Diffuse, texture=/Game/Global/Textures/Default/Masks_and_Gradients/3pxRGBImage}
  n65 = Multiply(
   A=@n64.RGB
   B=MaterialFunctionCall!(
    Local_Normals=NamedRerouteUsage()
    Local Position=@n5
    Local Bounds=@n13
    Calcated bounds=@n8.Result
    isState=NamedRerouteUsage()
    isVehicle=@n7
   ) {materialFunction=&/MF_Customization_Rorschach}.Mask
  )
  n66 = GetMaterialAttributes(None=FunctionInput() {inputName=Root_Mat, inputType=FunctionInput_MaterialAttributes}) {attributeGetTypes [
    #1
    #2
    #3
    #4
   ]}
  n67 = LinearInterpolate(
   A=@n66.Normal
   B=@n2.Normal
   Alpha=ScalarParameter() {defaultValue=1, group=Side Projection, parameterName="Use Side Root Normals   0=On", sortPriority=-75}
  )
  n68 = TextureSampleParameter2D(UVs=@n63) {
   group=Side Projection
   parameterName=SideProejction 01 Normal
   samplerType=SAMPLERTYPE_Normal
   sortPriority=3
   texture=/Engine/EngineMaterials/FlatNormal
  }
  n69 = TextureSampleParameter2D(UVs=@n63) {
   group=Side Projection
   parameterName=SideProejction 01 Mask
   samplerType=SAMPLERTYPE_Masks
   sortPriority=4
   texture=&T_BlackMask
  }
  n70 = ScalarParameter() {group=Side Projection, parameterName=SideRoughnessMin, sortPriority=1}
  n71 = ScalarParameter() {defaultValue=1, group=Side Projection, parameterName=SideRoughnessMax, sortPriority=2}
  n72 = ChannelMaskParameter(None=@n69.RGBA) {group=Side Projection, parameterName="Mask Channel = Emissive"}
  n73 = Multiply(
   A=Multiply(A=@n64.RGB, B=@n72.R)
   B=ScalarParameter() {defaultValue=1, group=Side Projection, parameterName=SideProjection_EmissiveIntensity, sortPriority=8}
  )
  n74 = VectorParameter() {group=Side Projection, parameterName=SideProjection_Emissive Color, sortPriority=9, defaultValue=<rgba 0 0 0 1>}
  n75 = Multiply(A=@n72.R, B=@n74.RGB)
  n76 = SetMaterialAttributes!(
   MaterialAttributes=@n2.MaterialAttributes
   BaseColor=StaticSwitch(True=@n3.RGB, False=StaticSwitchParameter(True=@n65, False=@n64.RGB) {group=Customization, parameterName=Rorschach}, Value=@n4)
   Normal=StaticSwitchParameter(
    True=@n67
    False=MaterialFunctionCall(BaseNormal=@n67, AdditionalNormal=@n68.RGB) {materialFunction=&BlendAngleCorrectedNormals}.Result
   ) {group=Side Projection, parameterName=Use Simple Normal}
   Metallic=StaticSwitchParameter(
    True=ChannelMaskParameter(None=@n69.RGBA) {group=Side Projection, maskChannel=Green, parameterName="Mask Channel = Metallic"}.R
    False=ScalarParameter() {group=Side Projection, parameterName=SideProjection_Metalness, sortPriority=6}
   ) {group=Side Projection, parameterName=SideProjection_MetallicMask, sortPriority=5}
   Roughness=StaticSwitch(True=@n3.A, False=LinearInterpolate(A=@n70, B=@n71, Alpha=@n64.A), Value=@n4)
   EmissiveColor=StaticSwitchParameter(
    True=StaticSwitchParameter(True=@n73, False=@n75) {group=Side Projection, parameterName="SideProjection_Combine ", sortPriority=8}
    False=@n2.EmissiveColor
   ) {group=Side Projection, parameterName=Side Projection Use Emissive, sortPriority=7}
   Specular=ScalarParameter() {defaultValue=0.5, group=Side Projection, parameterName=SideSpecular-01, sortPriority=1}
  ) {attributeSetTypes [
    #4
    #1
    #5
    #3
    #2
    #6
   ]}
  n77 = FunctionInput() {bUsePreviewValueAsDefault=true, inputName=Forward Axis}
  n78 = MaterialFunctionCall!(isVehicle=@n7, Pre Skinned Local Position=@n5, ForwardAxis=@n77, Per Obj Bounds=@n8.Result) {desc=Gradient, materialFunction=&/MF_WrapGradient}
  n79 = FunctionInput() {bUsePreviewValueAsDefault=true, inputName=Pre-Skinned Local Normal, sortPriority=42}
  n80 = MaterialFunctionCall!(
   SideProejction 01 Diffuse Alpha=@n64.A
   SideProejction 01 Mask=@n69.RGB
   Side Project Vehicle=@n40
   isMech=@n6
   Base=@n1
   Side=MaterialFunctionCall!(
    Side Project Vehicle=@n40
    Bounds=@n8.Result
    Forward Axis=@n77
    Material=MaterialFunctionCall!(
     Side Project Vehicle=@n40
     Bounds=@n8.Result
     Forward Axis=@n77
     material=@n76
     Uvs=@n43
     Pre-Skinned Local Position=@n5
     SideProjection01_Mask=@n69.RGB
     isMech=@n6
     masked_Values=@n78.Result
     isVehicle=@n7
     Normals=@n67
    ) {materialFunction=&/MF_SideProjection_blend}.output
    Pre-Skinned Local Position=@n5
    isMech=@n6
    SideProjection01_Mask=@n69.RGB
    isVehicle=@n7
   ) {materialFunction=&/MF_SideProjection_Emissive}.Result
   isVehicle=@n7
   Pre-Skinned Local Normal=@n79
  ) {materialFunction=&/MF_Side_TriBlendMask}
  n81 = Constant()
  n82 = Constant() {r=1}
  n83 = NamedRerouteDeclaration(None=@n79) {name=Pr_Skin_Normals, nodeColor=<rgba 0 0.2748 1 1>}
  n84 = MaterialFunctionCall!(
   Material=MaterialFunctionCall!(
    Side Project Vehicle=@n40
    Material=MaterialFunctionCall!(Material=@n80.Result, Uvs_1=@n64.R, "Customization Crevasse (_CM G)"=@n81, Mask_Secondary=@n82, "Customization Mask (_CM B)"=@n82) {materialFunction=&/MF_Iridescent}.Result
    isCharacter=NamedRerouteUsage()
    Gradent=@n78.Result
    isMech=@n6
    UVs=@n43
    isVehicle=@n7
    Pre-Skinned Local Normal=@n83
   ) {materialFunction=&/MF_Customization_SPAnimatedDiffuse}.Result
   Side Project Vehicle=@n40
   isVehicle=@n7
   isMech=@n6
   Uvs=@n43
   Pre-Skinned Local Normal=@n83
  ) {materialFunction=&/MF_Custmization_FireWorks}
  n85 = FunctionInput() {
   bUsePreviewValueAsDefault=true
   desc=secondary
   inputName="Customization Mask 2 (_CM B Top Right)"
   inputType=FunctionInput_Scalar
   sortPriority=25
  }
  n86 = ScalarParameter() {group=Side Projection, parameterName=Side_Remove_Customization Mask, sortPriority=-98}
  n87 = NamedRerouteUsage()
  n88 = Saturate(None=Add(
    A=FunctionInput() {bUsePreviewValueAsDefault=true, inputName="Customization Mask (_CM B)", inputType=FunctionInput_Scalar, sortPriority=24}
    B=LinearInterpolate(A=@n86, Alpha=@n87) {constB=0}
   ))
  n89 = StaticSwitchParameter(
   True=@n85
   False=StaticSwitchParameter(True=Multiply(A=@n88, B=@n64.A), False=@n88) {group=Side Projection, parameterName=Use Side ProjectionAlphaAsMask, sortPriority=-99}
  ) {group=Side Projection, parameterName=Use Side Projection As Secondary, sortPriority=-100}
  FunctionOutput(None=StaticSwitchParameter(True=BlendMaterialAttributes(A=@n2.MaterialAttributes, B=@n84.Result, Alpha=@n89) {vertexAttributeBlendType=UseA}, False=@n1) {group=Customization, parameterName=Side Projection})
  Multiply()
  n90 = FunctionInput() {bUsePreviewValueAsDefault=true, inputName=IsStatsB, inputType=FunctionInput_Vector2, sortPriority=100}
  NamedRerouteDeclaration(None=ComponentMask(None=@n90) {r=true}) {name=IsBackpack, nodeColor=<rgba 0 0 1 1>}
  NamedRerouteDeclaration(None=MaterialFunctionCall(R=@n7, G=@n40, B=@n6, A=@n41) {materialFunction=&AppendMany}.RGBA) {name=IsState, nodeColor=<rgba 0.9219 1 0 1>}
  NamedRerouteDeclaration(None=@n41) {name=IsCharacter, nodeColor=<rgba 1 0.9744 0.046 1>}
  NamedRerouteDeclaration(None=ComponentMask(None=@n90) {g=true}) {name=IsPickaxe, nodeColor=<rgba 0 0 1 1>}
  Reroute(None=@n5)
  n91 = VectorParameter() {desc="G=X  B=Y A=Uniform", group=OLD, parameterName=UVs_Scale_Weapon, sortPriority=1, defaultValue=<rgba 1 1 1 1>}
  n92 = VectorParameter() {desc="G=Y  B=X", group=OLD, parameterName=UVs_Offset_WeaponA, sortPriority=2, defaultValue=<rgba 0 -1 0 1>}
  n93 = ComponentMask(None=@n92.RGB) {g=true, r=true}
  n94 = Constant()
  n95 = PreSkinnedLocalBounds()
  n96 = MaterialFunctionCall(Color=@n95.Min) {materialFunction=&SplitComponents}
  n97 = Constant()
  n98 = LinearInterpolate(A=@n95.Min, B=AppendVector(A=AppendVector(A=@n96.R, B=@n97), B=@n96.B), Alpha=@n7)
  n99 = VectorParameter() {desc="G=Y  B=X", group=OLD, parameterName=UVs_Offset_Weapon, sortPriority=2, defaultValue=<rgba -0.5 -0.5 -0.5 1>}
  n100 = VectorParameter() {desc="R=Y  G=X", group=OLD, parameterName=UVs_Offset_Vehicle, sortPriority=4, defaultValue=<rgba -0.5 -0.5 -0.5 0>}
  n101 = StaticSwitchParameter(
   True=LinearInterpolate(A=Constant3Vector() {constant=<rgba 0 -0.5 -0.5 0>}, B=Constant3Vector() {constant=<rgba -0.5 0 0 1>}, Alpha=@n7)
   False=LinearInterpolate(A=@n99.RGB, B=@n100.RGB, Alpha=@n7)
  ) {group=OLD, parameterName=SideProjection_UVs MakeSquare Position, sortPriority=-5}
  n102 = Subtract(A=@n95.Max, B=@n95.Max)
  n103 = MaterialFunctionCall(Color=@n102) {materialFunction=&SplitComponents}
  n104 = StaticSwitchParameter(True=Max(A=Max(A=@n103.R, B=@n103.R), B=@n103.B), False=@n102) {desc=uses the maximim size to set the texture to power of 2, group=OLD, parameterName=SideProjection_UVs MakeSquare, sortPriority=-5}
  n105 = Divide(A=Add(A=Subtract(A=@n5, B=@n98), B=Multiply(A=@n101, B=@n102)), B=@n104)
  n106 = ComponentMask(None=@n105) {g=true}
  n107 = OneMinus(None=ComponentMask(None=@n105) {b=true})
  n108 = Multiply(
   A=AppendVector(A=@n91.B, B=@n91.B)
   B=Add(
    A=StaticSwitchParameter(True=@n93, False=@n94) {group=OLD, parameterName=SideProjection_UVs MakeSquare Position, sortPriority=-5}
    B=AppendVector(A=@n106, B=@n107)
   )
  )
  n109 = ComponentMask(None=@n105) {r=true}
  n110 = AppendVector(A=OneMinus(None=@n106), B=@n109)
  n111 = AppendVector(A=@n109, B=@n106)
  n112 = VectorParameter() {desc="R=Y  G=X", group=OLD, parameterName=UVs_Offset_VehicleA, sortPriority=4, defaultValue=<rgba 0 0 0 1>}
  n113 = ComponentMask(None=@n112.RGB) {g=true, r=true}
  n114 = Constant()
  n115 = VectorParameter() {desc="R=  G=  A=Uniform", group=OLD, parameterName=UVs_Scale_Vehicle, sortPriority=3, defaultValue=<rgba 1 1 1 1>}
  n116 = Multiply(
   A=Add(
    A=StaticSwitchParameter(True=@n110, False=@n111) {group=OLD, parameterName=Rotate Vehicle Projection, sortPriority=-4}
    B=StaticSwitchParameter(True=@n113, False=@n114) {group=OLD, parameterName=SideProjection_UVs MakeSquare Position, sortPriority=-5}
   )
   B=AppendVector(A=@n115.G, B=@n115.G)
  )
  Reroute(None=LinearInterpolate(A=Multiply(A=@n108, B=@n91.A), B=Multiply(A=@n116, B=@n115.A), Alpha=@n7))
 ]}
