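# Material function for weapon skin wraps that optionally replaces incoming color curve values with samples from a shared curve atlas texture.
# The UV scalar sets U, while Index plus a half-texel offset divided by the curve count parameter selects the atlas row, sampled at an explicit mip level.
# A static bool switches between that sample and the CCinput passthrough, output as RGBA, RGB and alpha.
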MaterialFunction /Game/Weapons/WeaponSkins/Wraps/MaterialFunctions/MF_Global_CC_Tex
graph {nodes [
  n1 = ComponentMask(None=TextureCoordinate()) {r=true}
  n2 = Add(A=FunctionInput(Preview=Constant()) {inputName=Index, inputType=FunctionInput_Scalar}) {constB=0.5}
  n3 = ScalarParameter() {defaultValue=257, desc=How many CColor Curve you have in the Atlas, group=00_ColorCurve, parameterName=ColorCurve Shared_CC_Index_Count}
  n4 = TextureObjectParameter() {
   group=00_ColorCurve
   parameterName=ColorCurve Shared_CC_Tex
   samplerType=SAMPLERTYPE_LinearColor
   texture=/Game/Weapons/WeaponSkins/Wraps/Textures/CruveAtlas_Wraps_01
  }
  n5 = TextureSample!(
   UVs=AppendVector(A=FunctionInput(Preview=@n1) {inputName=UV, inputType=FunctionInput_Scalar}, B=Divide(A=@n2, B=@n3))
   Tex=@n4
   Level
   Apply View MipBias
  ) {mipValueMode=TMVM_MipLevel}
  n6 = FunctionInput() {inputName=CCinput}
  n7 = FunctionInput() {inputName=CCinput_Alpha, inputType=FunctionInput_Scalar}
  n8 = StaticBoolParameter() {desc="Replace all Color Curves with a Texture sample ", group=00_ColorCurve, parameterName="ColorCurveAtlas ---->Global CC Tex PerWrap"}
  FunctionOutput(None=StaticSwitch(True=@n5.RGBA, False=AppendVector(A=@n6, B=@n7), Value=@n8)) {outputName=4v, sortPriority=1}
  FunctionOutput(None=StaticSwitch(True=@n5.RGB, False=@n6, Value=@n8)) {outputName=3V}
  FunctionOutput(None=StaticSwitch(True=@n5.A, False=@n7, Value=@n8)) {outputName=Alpha, sortPriority=5}
 ]}
