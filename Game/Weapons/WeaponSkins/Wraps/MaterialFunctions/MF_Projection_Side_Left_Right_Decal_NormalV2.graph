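# Material function for weapon wrap skins that projects a normal-map decal onto the left and right sides of a mesh.
# Is State channels pick the projection axis, and the pre-skinned normal's sign gives a side mask that mirrors UVs and rotation.
# Bounds-relative UVs get offset, rotation, scale, pivot and optional clamping.
# The sampled normal is reoriented per axis and converted local-to-tangent, output alongside UVs and masks.
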MaterialFunction /Game/Weapons/WeaponSkins/Wraps/MaterialFunctions/MF_Projection_Side_Left_Right_Decal_NormalV2
graph {nodes [
  Constant() {r=5}
  n1 = FunctionInput() {bUsePreviewValueAsDefault=true, inputName=Is State, inputType=FunctionInput_Vector4, sortPriority=5}
  n2 = FunctionInput(Preview=LinearInterpolate(Alpha=MaterialFunctionCall(Float4=@n1) {materialFunction=&BreakOutFloat4Components}.B) {constB=0.25}) {bUsePreviewValueAsDefault=true, description=defualt 0, inputName=Uvs_Rotation, inputType=FunctionInput_Scalar, sortPriority=50}
  n3 = FunctionInput(Preview=VertexInterpolator(VS=PreSkinnedNormal()).PS) {bUsePreviewValueAsDefault=true, inputName=Pre_Skinned_Normals, sortPriority=2}
  n4 = MaterialFunctionCall(Color=@n3) {materialFunction=&SplitComponents}
  n5 = MaterialFunctionCall(Float4=@n1) {materialFunction=&BreakOutFloat4Components}
  n6 = LinearInterpolate(A=@n4.R, B=LinearInterpolate(A=@n4.B, B=@n4.B, Alpha=@n5.G), Alpha=@n5.R)
  n7 = Saturate(None=Ceil(None=LinearInterpolate(A=@n6, B=@n4.G, Alpha=@n5.B)))
  n8 = FunctionInput() {bUsePreviewValueAsDefault=true, description=defualt 0, inputName=Uvs_Rotation_Mirroring, inputType=FunctionInput_Scalar, sortPriority=50}
  n9 = LinearInterpolate(A=LinearInterpolate(A=@n2, B=Multiply(A=@n2) {constB=-1}, Alpha=@n7), B=@n2, Alpha=@n8)
  MaterialExpressionFunctionInput_7 = FunctionInput(Preview=Constant3Vector()) {bUsePreviewValueAsDefault=true, inputName=Pivot}
  n10 = Convert(None=@MaterialExpressionFunctionInput_7) {
   nodeName=Break Float3
   convertInputs [{
     type=Vector3
     expressionInput {
      expression=&self:MaterialExpressionFunctionInput_7
      inputName=None
      mask=0
      maskA=0
      maskB=0
      maskG=0
      maskR=0
      outputIndex=0
     }
    }]
   convertOutputs [{type=Scalar}, {type=Scalar}, {type=Scalar}]
  }
  n11 = AppendVector(A=Multiply(B=@n10) {constA=-1}, B=@n10)
  n12 = ComponentMask(None=@MaterialExpressionFunctionInput_7) {g=true, r=true}
  n13 = StaticBool()
  n14 = Constant3Vector() {constant=<rgba 0 1 0 0>}
  n15 = MaterialFunctionCall(Float4=@n1) {materialFunction=&BreakOutFloat4Components}
  n16 = LinearInterpolate(B=@n14, Alpha=@n15.G)
  n17 = ComponentMask(None=LinearInterpolate(A=LinearInterpolate(A=@n14, B=@n16, Alpha=@n15.R), B=@n14, Alpha=@n15.B)) {g=true, r=true}
  n18 = PreSkinnedLocalBounds()
  n19 = Subtract(A=VertexInterpolator(VS=PreSkinnedPosition()).PS, B=@n18.Min)
  n20 = Subtract(A=@n18.Max, B=@n18.Max)
  n21 = Multiply(
   A=FunctionInput(Preview=Divide(A=@n19, B=@n20)) {bUsePreviewValueAsDefault=true, inputName=Caculated_Bounds, sortPriority=1}
   B=Constant3Vector() {constant=<rgba 1 1 -1 0>}
  )
  n22 = MaterialFunctionCall(Color=@n21) {materialFunction=&SplitComponents}
  n23 = OneMinus(None=@n22.R)
  n24 = AppendVector(A=OneMinus(None=@n22.R), B=@n22.B)
  n25 = LinearInterpolate(A=AppendVector(A=@n22.G, B=@n22.G), B=LinearInterpolate(A=AppendVector(A=@n23, B=@n22.G), B=@n24, Alpha=@n15.G), Alpha=@n15.R)
  n26 = MaterialFunctionCall(Color=@n21) {materialFunction=&SplitComponents}
  n27 = AppendVector(A=@n26.R, B=@n26.R)
  n28 = LinearInterpolate(
   A=AppendVector(A=OneMinus(None=@n26.G), B=@n26.B)
   B=LinearInterpolate(A=AppendVector(A=@n26.R, B=@n26.R), B=@n27, Alpha=@n15.G)
   Alpha=@n15.R
  )
  n29 = LinearInterpolate(
   A=Add(A=@n17, B=LinearInterpolate(A=@n25, B=@n24, Alpha=@n15.B))
   B=Add(A=@n17, B=LinearInterpolate(A=@n28, B=@n27, Alpha=@n15.B))
   Alpha=@n7
  )
  n30 = FunctionInput(Preview=Constant3Vector()) {bUsePreviewValueAsDefault=true, description=defualt 0, inputName=Uvs_Offset, sortPriority=54}
  n31 = Multiply(A=@n30, B=Constant3Vector() {constant=<rgba -1 1 1 0>})
  n32 = Add(A=@n29, B=LinearInterpolate(A=ComponentMask(None=@n30) {g=true, r=true}, B=ComponentMask(None=@n31) {g=true, r=true}, Alpha=@n7)) {constB=0}
  n33 = Constant() {r=0.5}
  n34 = FunctionInput(Preview=Constant()) {bUsePreviewValueAsDefault=true, inputName=Offset RT}
  n35 = FunctionInput(Preview=Constant3Vector() {constant=<rgba 1 1 1 0>}) {bUsePreviewValueAsDefault=true, description=defualt 0, inputName=Uvs_TileRate, sortPriority=52}
  n36 = ComponentMask(None=@n35) {g=true, r=true}
  n37 = MaterialFunctionCall(
   UVs=Add(
    A=MaterialFunctionCall(UVs=@n32, Rotation Center=@n33, "Rotation Angle (0-1)"=@n9) {materialFunction=&CustomRotator}.Rotated Values
    B=ComponentMask(None=@n34) {g=true, r=true}
   )
   Texture Scale=@n36
  ) {materialFunction=&ScaleUVsByCenter}
  n38 = Add(
   A=StaticSwitch(
    True=LinearInterpolate(A=@n11, B=@n12, Alpha=@n7)
    False=@n12
    Value=FunctionInput(Preview=@n13) {bUsePreviewValueAsDefault=true, inputName="Pivot_L\\R", inputType=FunctionInput_StaticBool}
   )
   B=@n37.UVs
  )
  n39 = ComponentMask(None=@n38) {r=true}
  n40 = FunctionInput(Preview=Constant()) {bUsePreviewValueAsDefault=true, inputName=Clamp Padding_Comp, sortPriority=151}
  n41 = Subtract(B=@n40)
  n42 = StaticBool()
  n43 = ComponentMask(None=@n38) {g=true}
  n44 = StaticBool()
  n45 = AppendVector(
   A=StaticSwitch(
    True=Clamp(None=@n39, Min=@n40, Max=@n41)
    False=@n39
    Value=FunctionInput(Preview=@n42) {bUsePreviewValueAsDefault=true, inputName=Clamp X, inputType=FunctionInput_StaticBool, sortPriority=150}
   )
   B=StaticSwitch(
    True=Clamp(None=@n43, Min=@n40, Max=@n41)
    False=@n43
    Value=FunctionInput(Preview=@n44) {bUsePreviewValueAsDefault=true, inputName=Clamp Y, inputType=FunctionInput_StaticBool, sortPriority=150}
   )
  )
  n46 = TextureSample(
   UVs=@n45
   Tex=FunctionInput(Preview=TextureObject() {samplerType=SAMPLERTYPE_Normal, texture=/Engine/Functions/Engine_MaterialFunctions02/ExampleContent/Textures/SphereRenderNormalsMap}) {bUsePreviewValueAsDefault=true, inputName=Normal_Map, inputType=FunctionInput_Texture2D}
  ) {samplerType=SAMPLERTYPE_Normal}
  n47 = MaterialFunctionCall(Color=@n46.RGB) {materialFunction=&SplitComponents}
  n48 = ComponentMask(None=@n3) {r=true}
  n49 = Multiply(A=@n47.R) {constB=-1}
  n50 = AppendVector(A=AppendVector(A=Multiply(A=@n47.B, B=@n48), B=LinearInterpolate(A=@n47.R, B=@n49, Alpha=@n7)), B=Multiply(A=@n47.G) {constB=-1})
  n51 = RotateAboutAxis(
   NormalizedRotationAxis=Constant3Vector() {constant=<rgba 1 0 0 0>}
   RotationAngle=LinearInterpolate(A=@n9, B=Multiply(A=@n9) {constB=-1}, Alpha=@n7)
   PivotPoint=Constant3Vector()
   Position=@n50
  )
  n52 = Multiply(A=@n9) {constB=-1}
  n53 = MaterialFunctionCall(Color=@n46.RGB) {materialFunction=&SplitComponents}
  n54 = Multiply(A=@n53.R) {constB=-1}
  n55 = AppendVector(A=AppendVector(A=LinearInterpolate(A=@n54, B=@n53.R, Alpha=@n7), B=@n53.G), B=Multiply(A=@n53.B, B=ComponentMask(None=@n3) {b=true}))
  n56 = Add(
   A=RotateAboutAxis(
    NormalizedRotationAxis=Constant3Vector() {constant=<rgba 0 0 1 0>}
    RotationAngle=LinearInterpolate(A=@n9, B=@n52, Alpha=@n7)
    PivotPoint=Constant3Vector()
    Position=@n55
   )
   B=@n55
  )
  n57 = Multiply(A=@n9) {constB=-1}
  n58 = MaterialFunctionCall(Color=@n46.RGB) {materialFunction=&SplitComponents}
  n59 = Multiply(A=@n58.R) {constB=-1}
  n60 = ComponentMask(None=@n3) {g=true}
  n61 = AppendVector(A=AppendVector(A=LinearInterpolate(A=@n59, B=@n58.R, Alpha=@n7), B=Multiply(A=@n58.B, B=@n60)), B=Multiply(A=@n58.G) {constB=-1})
  n62 = Add(
   A=RotateAboutAxis(
    NormalizedRotationAxis=Constant3Vector() {constant=<rgba 0 1 0 0>}
    RotationAngle=LinearInterpolate(A=@n9, B=@n57, Alpha=@n7)
    PivotPoint=Constant3Vector()
    Position=@n61
   )
   B=@n61
  )
  n63 = MaterialFunctionCall(Float4=@n1) {materialFunction=&BreakOutFloat4Components}
  n64 = Constant3Vector() {constant=<rgba 0 1 0 0>}
  n65 = LinearInterpolate(A=@n9, B=Multiply(A=@n9) {constB=-1}, Alpha=@n7)
  n66 = Constant3Vector()
  n67 = MaterialFunctionCall(Color=@n46.RGB) {materialFunction=&SplitComponents}
  n68 = Multiply(A=@n67.R) {constB=-1}
  n69 = ComponentMask(None=@n3) {g=true}
  n70 = AppendVector(A=AppendVector(A=LinearInterpolate(A=@n68, B=@n67.R, Alpha=@n7), B=Multiply(A=@n67.B, B=@n69)), B=Multiply(A=@n67.G) {constB=-1})
  n71 = LinearInterpolate(
   A=LinearInterpolate(A=Add(A=@n51, B=@n50), B=LinearInterpolate(A=@n56, B=@n62, Alpha=@n63.G), Alpha=@n63.R)
   B=Add(A=RotateAboutAxis(NormalizedRotationAxis=@n64, RotationAngle=@n65, PivotPoint=@n66, Position=@n70), B=@n70)
   Alpha=@n63.B
  )
  n72 = Constant3Vector() {constant=<rgba 1 0 0 0>}
  n73 = Multiply(A=Constant3Vector() {constant=<rgba 0 0 1 0>}, B=@n3)
  n74 = Multiply(A=Constant3Vector() {constant=<rgba 0 1 0 0>}, B=@n3)
  n75 = LinearInterpolate(
   A=LinearInterpolate(A=Multiply(A=@n72, B=@n3), B=LinearInterpolate(A=@n73, B=@n74, Alpha=@n63.G), Alpha=@n63.R)
   B=Multiply(A=Constant3Vector() {constant=<rgba 0 1 0 0>}, B=@n3)
   Alpha=@n63.B
  )
  n76 = Transform(None=Subtract(A=@n71, B=@n75)) {transformSourceType=TRANSFORMSOURCE_Local, transformType=TRANSFORM_Tangent}
  FunctionOutput(None=Add(A=Constant3Vector() {constant=<rgba 0 0 1 0>}, B=@n76)) {outputName=Normal, sortPriority=2}
  FunctionOutput(None=@n29) {outputName=UVs No Offsets/Tiling, sortPriority=1}
  FunctionOutput(None=@n76) {description=add to object normals, outputName=Raw_Normal, sortPriority=3}
  FunctionOutput(None=@n7) {outputName=Left_Right_Mask}
  FunctionOutput(None=@n37.0-1 mask) {outputName=Clamp_Mask}
  LinearInterpolate(A=@n36, B=ComponentMask(None=@n35) {g=true, r=true}, Alpha=@n7)
  MaterialFunctionCall() {materialFunction=&SplitComponents}
  Normalize(VectorInput=@n75)
  Normalize(VectorInput=@n71)
  TextureSample(UVs=FunctionOutput(None=@n45) {outputName=UVs}) {texture=/Game/Weapons/WeaponSkins/Wraps/Textures/UVCheaker}
 ]}
