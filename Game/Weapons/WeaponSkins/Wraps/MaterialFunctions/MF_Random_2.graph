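# Material function that generates a pseudo-random scalar from a 2D vector input, stored with the weapon skin wrap material functions.
# It uses the classic shader hash: dot the input with a constant vector, take the sine, then multiply by a large RandomSeed input that has a default.
# One output returns the fractional part and Raw Data returns the unwrapped product.
# A Floor of the input and a second constant vector are left unconnected.
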MaterialFunction /Game/Weapons/WeaponSkins/Wraps/MaterialFunctions/MF_Random_2
graph {nodes [
  Constant2Vector() {g=78.233, r=12.9898}
  n1 = FunctionInput(Preview=TextureCoordinate()) {inputName=2 Vector, inputType=FunctionInput_Vector2}
  Floor(None=@n1)
  n2 = Constant2Vector() {g=3.2257, r=1.3322}
  n3 = FunctionInput(Preview=Constant() {r=43758.5469}) {bUsePreviewValueAsDefault=true, inputName=RandomSeed, inputType=FunctionInput_Scalar, sortPriority=4, previewValue=<7 0 0 1>}
  n4 = Multiply(A=Sine(None=DotProduct(A=@n1, B=@n2)), B=@n3) {constB=43758.5469}
  FunctionOutput(None=@n4) {outputName=Raw Data, sortPriority=1}
  FunctionOutput(None=Frac(None=@n4)) {outputName=utput}
 ]}
