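# Material function for weapon skin wraps that corrects normals applied through side-projected UVs.
# It reorients a sampled or supplied normal into per-axis branches using pre-skinned normals, counter-rotates each for UV rotation, and blends them by the is_State mask.
# Optional left/right mirroring uses a mask from the normal's sign.
# It subtracts the axis-aligned base normal and converts local to tangent space, outputting both that raw offset and a version added to tangent up.
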MaterialFunction /Game/Weapons/WeaponSkins/Wraps/MaterialFunctions/MF_UVs_Side_Fix_Normals
graph {nodes [
  n1 = TextureObject() {samplerType=SAMPLERTYPE_Normal, texture=/Engine/Functions/Engine_MaterialFunctions02/ExampleContent/Textures/SphereRenderNormalsMap}
  n2 = StaticSwitch(
   True=FunctionInput() {bUsePreviewValueAsDefault=true, inputName=Normals_Source, sortPriority=1}
   False=TextureSample(
    UVs=FunctionInput() {bUsePreviewValueAsDefault=true, inputName=UVs, inputType=FunctionInput_Vector2, sortPriority=2}
    Tex=FunctionInput(Preview=@n1) {bUsePreviewValueAsDefault=true, inputName=Normal_Map, inputType=FunctionInput_Texture2D}
   ) {samplerType=SAMPLERTYPE_Normal}.RGB
   Value=FunctionInput(Preview=StaticBool()) {bUsePreviewValueAsDefault=true, inputName=Normals_Source_Use, inputType=FunctionInput_StaticBool, sortPriority=1}
  )
  ComponentMask(None=@n2) {g=true, r=true}
  n3 = Constant3Vector() {constant=<rgba 1 0 0 0>}
  n4 = FunctionInput(Preview=Constant()) {bUsePreviewValueAsDefault=true, description=defualt 0, inputName=UV_Rotation, inputType=FunctionInput_Scalar, sortPriority=15}
  n5 = Multiply(A=@n4) {constB=-1}
  n6 = NamedRerouteUsage()
  n7 = FunctionInput(Preview=StaticBool()) {bUsePreviewValueAsDefault=true, inputName="Left\\Right", inputType=FunctionInput_StaticBool, sortPriority=200}
  n8 = FunctionInput(Preview=StaticBool()) {bUsePreviewValueAsDefault=true, inputName=No_Rotation, inputType=FunctionInput_StaticBool, sortPriority=200}
  n9 = StaticSwitch(True=Constant(), False=StaticSwitch(True=LinearInterpolate(A=@n4, B=@n5, Alpha=@n6), False=@n4, Value=@n7), Value=@n8)
  n10 = NamedRerouteUsage()
  n11 = StaticSwitch(True=LinearInterpolate(A=@n9, B=Multiply(A=@n9) {constB=-1}, Alpha=@n10), False=@n9, Value=NamedRerouteUsage())
  n12 = Constant3Vector()
  n13 = MaterialFunctionCall(Color=@n2) {materialFunction=&SplitComponents}
  n14 = FunctionInput(Preview=VertexInterpolator(VS=PreSkinnedNormal()).PS) {bUsePreviewValueAsDefault=true, inputName=Pre_Skinned_Normals, sortPriority=110}
  n15 = ComponentMask(None=@n14) {r=true}
  n16 = LinearInterpolate(A=@n13.R, B=Multiply(A=@n13.R) {constB=-1}, Alpha=@n10)
  n17 = NamedRerouteUsage()
  n18 = AppendVector(A=AppendVector(A=Multiply(A=@n13.B, B=@n15), B=StaticSwitch(True=@n16, False=@n13.R, Value=@n17)), B=Multiply(A=@n13.G) {constB=-1})
  n19 = Multiply(A=@n9) {constB=-1}
  n20 = NamedRerouteUsage()
  n21 = MaterialFunctionCall(Color=@n2) {materialFunction=&SplitComponents}
  n22 = LinearInterpolate(A=Multiply(A=@n21.R) {constB=-1}, B=@n21.R, Alpha=NamedRerouteUsage())
  n23 = NamedRerouteUsage()
  n24 = AppendVector(A=AppendVector(A=StaticSwitch(True=@n22, False=@n21.R, Value=@n23), B=@n21.G), B=Multiply(A=@n13.B, B=ComponentMask(None=@n14) {b=true}))
  n25 = RotateAboutAxis(
   NormalizedRotationAxis=Constant3Vector() {constant=<rgba 0 0 1 0>}
   RotationAngle=StaticSwitch(True=LinearInterpolate(A=@n9, B=@n19, Alpha=@n20), False=@n19, Value=NamedRerouteUsage())
   PivotPoint=Constant3Vector()
   Position=@n24
  )
  n26 = Multiply(A=@n9) {constB=-1}
  n27 = NamedRerouteUsage()
  n28 = MaterialFunctionCall(Color=@n2) {materialFunction=&SplitComponents}
  n29 = LinearInterpolate(A=Multiply(B=@n28.R) {constA=-1}, B=@n28.R, Alpha=@n27)
  n30 = NamedRerouteUsage()
  n31 = ComponentMask(None=@n14) {g=true}
  n32 = AppendVector(A=AppendVector(A=StaticSwitch(True=@n29, False=@n28.R, Value=@n30), B=Multiply(A=@n28.B, B=@n31)), B=Multiply(A=@n28.G) {constB=-1})
  n33 = RotateAboutAxis(
   NormalizedRotationAxis=Constant3Vector() {constant=<rgba 0 1 0 0>}
   RotationAngle=StaticSwitch(True=LinearInterpolate(A=@n9, B=@n26, Alpha=@n27), False=@n26, Value=NamedRerouteUsage())
   PivotPoint=Constant3Vector()
   Position=@n32
  )
  n34 = FunctionInput(Preview=Constant4Vector().RGBA) {bUsePreviewValueAsDefault=true, inputName=is_State, inputType=FunctionInput_Vector4, sortPriority=110}
  n35 = MaterialFunctionCall(Color=ComponentMask(None=NamedRerouteDeclaration(None=@n34) {name=IsState, nodeColor=<rgba 1 0 0.75 1>}) {b=true, g=true, r=true}) {materialFunction=&SplitComponents}
  n36 = LinearInterpolate(
   A=Add(A=RotateAboutAxis(NormalizedRotationAxis=@n3, RotationAngle=@n11, PivotPoint=@n12, Position=@n18), B=@n18)
   B=LinearInterpolate(A=Add(A=@n25, B=@n24), B=Add(A=@n33, B=@n32), Alpha=@n35.G)
   Alpha=@n35.R
  )
  n37 = FunctionInput(Preview=Constant()) {
   bUsePreviewValueAsDefault=true
   desc=set to .25 if useing disable rotation. 1
   inputName=Rotation_Comp
   inputType=FunctionInput_Scalar
   sortPriority=15
  }
  n38 = Add(A=@n37, B=StaticSwitch(True=Constant() {r=0.25}, False=@n4, Value=@n8))
  n39 = NamedRerouteUsage()
  n40 = LinearInterpolate(A=@n38, B=Multiply(A=@n38) {constB=-1}, Alpha=@n39)
  n41 = Multiply(A=@n38) {constB=-1}
  n42 = NamedRerouteUsage()
  n43 = MaterialFunctionCall(Color=@n2) {materialFunction=&SplitComponents}
  n44 = LinearInterpolate(A=Multiply(A=@n43.R) {constB=-1}, B=@n43.R, Alpha=@n39)
  n45 = NamedRerouteUsage()
  n46 = ComponentMask(None=@n14) {g=true}
  n47 = AppendVector(A=AppendVector(A=StaticSwitch(True=@n44, False=@n43.R, Value=@n45), B=Multiply(A=@n43.B, B=@n46)), B=Multiply(A=@n43.G) {constB=-1})
  n48 = Add(
   A=RotateAboutAxis(
    NormalizedRotationAxis=Constant3Vector() {constant=<rgba 0 1 0 0>}
    RotationAngle=StaticSwitch(True=@n40, False=@n41, Value=@n42)
    PivotPoint=Constant3Vector()
    Position=@n47
   )
   B=@n47
  )
  n49 = Constant3Vector() {constant=<rgba 0 0 1 0>}
  n50 = Constant3Vector() {constant=<rgba 0 1 0 0>}
  n51 = LinearInterpolate(
   A=Multiply(A=Constant3Vector() {constant=<rgba 1 0 0 0>}, B=@n14)
   B=LinearInterpolate(A=Multiply(A=@n49, B=@n14), B=Multiply(A=@n50, B=@n14), Alpha=@n35.G)
   Alpha=@n35.R
  )
  n52 = Multiply(A=Constant3Vector() {constant=<rgba 0 1 0 0>}, B=@n14)
  n53 = Transform(None=Subtract(A=LinearInterpolate(A=@n36, B=@n48, Alpha=@n35.B), B=LinearInterpolate(A=@n51, B=@n52, Alpha=@n35.B))) {transformSourceType=TRANSFORMSOURCE_Local, transformType=TRANSFORM_Tangent}
  FunctionOutput(None=Add(A=Constant3Vector() {constant=<rgba 0 0 1 0>}, B=@n53)) {outputName=Normal, sortPriority=2}
  FunctionOutput(None=@n53) {description=add to object normals, outputName=Raw_Normal, sortPriority=3}
  n54 = MaterialFunctionCall(Color=@n14) {materialFunction=&SplitComponents}
  n55 = MaterialFunctionCall(Float4=NamedRerouteUsage()) {materialFunction=&BreakOutFloat4Components}
  n56 = LinearInterpolate(A=LinearInterpolate(A=@n54.R, B=LinearInterpolate(A=@n54.B, B=@n54.B, Alpha=@n55.G), Alpha=@n55.R), B=@n54.G, Alpha=@n55.B)
  NamedRerouteDeclaration(None=Saturate(None=Ceil(None=@n56))) {name="L\\R_Msk", nodeColor=<rgba 1 0.6797 0 1>}
  NamedRerouteDeclaration(None=@n7) {name="L\\R_State", nodeColor=<rgba 1 0 0.2812 1>}
 ]}
