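# Weapon wrap material function that layers a reactive decal effect onto incoming material attributes, limited by customization-mask channels.
# Four parallel emissive layers (A–D) build scrolling or reactively driven gradient masks, procedural or textured, with warping and per-item-type rotation, then combine them with decal masks via multiply or HeightLerp.
# Decals set base color, roughness, specular, metallic, normal and emissive, or blend emissive-only.
# Static switches select features, including disabling decal normals on mobile.
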MaterialFunction /Game/Weapons/WeaponSkins/Wraps/MaterialFunctions/MF_Effect_Decal_Reactive_ref
graph {nodes [
  n1 = NamedRerouteUsage()
  n2 = Multiply(
   A=ChannelMaskParameter(None=NamedRerouteUsage()) {group=Effect_Decal_Reactive_Emissive_D, parameterName=EDRec_Emissive_D_Mask_Channel, sortPriority=5}.R
   B=NamedRerouteUsage()
  )
  n3 = StaticSwitch(
   True=Add(A=@n2, B=NamedRerouteUsage())
   False=@n2
   Value=StaticSwitch(True=NamedRerouteUsage(), False=StaticBool(), Value=NamedRerouteUsage())
  )
  n4 = StaticSwitchParameter(True=NamedRerouteUsage(), False=NamedRerouteUsage()) {group=Effect_Decal_Reactive_Emissive_D, parameterName="EDRec_Emissive_D_Warp_ Msk-->Bounds_Warp_Msk"}
  n5 = Multiply(
   A=ScalarParameter() {group=Effect_Decal_Reactive_Emissive_D, parameterName=EDRec_Emissive_D_Warp_Streagth}
   B=Add(
    A=ChannelMaskParameter(None=@n4) {group=Effect_Decal_Reactive_Emissive_D, parameterName=EDRec_Emissive_D_Warp_Mask_Channel}.R
    B=ScalarParameter() {group=Effect_Decal_Reactive_Emissive_D, parameterName=EDRec_Emissive_D_Warp_Offset}
   )
  )
  n6 = ScalarParameter() {defaultValue=1, group=Effect_Decal_Reactive_Emissive_D_Reactive|Gradent, parameterName=EDRec_Emissive_D_Gradent_Time_End, sortPriority=210}
  n7 = LinearInterpolate(
   A=ScalarParameter() {group=Effect_Decal_Reactive_Emissive_D_Reactive|Gradent, parameterName=EDRec_Emissive_D_Gradent_Time_Start, sortPriority=210}
   B=@n6
   Alpha=NamedRerouteUsage()
  )
  n8 = VectorParameter() {
   group=Effect_Decal_Reactive_Emissive_D_Reactive|Gradent
   parameterName=EDRec_Emissive_D_Gradent_Mask Direction
   sortPriority=210
   channelNames {a="", b="", g=X, r=Y}
   defaultValue=<rgba 1 1 0 1>
  }
  n9 = ComponentMask(None=@n8.RGB) {g=true, r=true}
  n10 = Time()
  n11 = AppendVector(
   A=Constant()
   B=ScalarParameter() {group=Effect_Decal_Reactive_Emissive_D_Reactive|Gradent, parameterName=EDRec_Emissive_D_Gradent_Time_X_Speed, sortPriority=210}
  )
  n12 = VectorParameter() {group=Effect_Decal_Reactive_Emissive_D, parameterName=EDRec_Emissive_D_Gradent_Offset, defaultValue=<rgba 0 0 0 1>}
  n13 = VectorParameter() {group=Effect_Decal_Reactive_Emissive_D, parameterName=EDRec_Emissive_D_Gradent_Speed, defaultValue=<rgba 0 0 0 1>}
  n14 = Multiply(A=Time(), B=@n13.RGB)
  n15 = StaticBoolParameter() {group=Effect_Decal_Reactive_Emissive_D_Reactive, parameterName=EDRec_Emissive_D_Gradent_Reactive---------------------, sortPriority=200}
  n16 = StaticSwitch(
   True=Add(A=Multiply(A=@n7, B=@n9), B=Multiply(A=@n10, B=@n11))
   False=ComponentMask(None=Add(A=@n12.RGB, B=@n14)) {g=true, r=true}
   Value=@n15
  )
  n17 = Add(A=NamedRerouteUsage(), B=@n16)
  n18 = VectorParameter() {
   group=Effect_Decal_Reactive_Emissive_D
   parameterName=EDRec_Emissive_D_Gradent_Rotation
   channelNames {a="Mech\\Character\\Back", b=Vehicle_Side, g=Vehicle, r="Weapon\\Pickaxe"}
  }
  n19 = MaterialFunctionCall(Float4=NamedRerouteUsage()) {materialFunction=&BreakOutFloat4Components}
  n20 = LinearInterpolate(A=@n18."Weapon\\Pickaxe", B=LinearInterpolate(A=@n18.Vehicle, B=@n18.Vehicle, Alpha=@n19.G), Alpha=@n19.R)
  n21 = MaterialFunctionCall(
   UVs=NamedRerouteUsage()
   "Rotation Angle (0-1)"=Add(
    A=ScalarParameter() {group=Effect_Decal_Reactive_Emissive_D, parameterName=EDRec_Emissive_D_Gradent_Rotation}
    B=LinearInterpolate(A=@n20, B=@n18."Mech\\Character\\Back", Alpha=@n19.B)
   )
  ) {materialFunction=&CustomRotator}
  n22 = VectorParameter() {group=Effect_Decal_Reactive_Emissive_D, parameterName=EDRec_Emissive_D_Gradent_Scale, defaultValue=<rgba 1 1 1 1>}
  n23 = Multiply(A=@n22.RGB, B=@n22.RGB)
  n24 = Add(A=Multiply(A=@n21.Rotated Values, B=ComponentMask(None=@n23) {g=true, r=true}), B=@n16)
  n25 = ComponentMask(None=@n24) {g=true, r=true}
  n26 = StaticSwitchParameter(
   True=StaticSwitchParameter(True=Add(A=@n5, B=@n17), False=@n17) {group=Effect_Decal_Reactive_Emissive_D, parameterName=EDRec_Emissive_D_Warp_ Msk}
   False=StaticSwitchParameter(True=Add(A=@n5, B=@n25), False=@n25) {group=Effect_Decal_Reactive_Emissive_D, parameterName=EDRec_Emissive_D_Warp_ Msk}
  ) {group=Effect_Decal_Reactive_Emissive_D, parameterName="EDRec_Emissive_D_Grad UV---->00_UV_Bounds", sortPriority=0}
  n27 = StaticSwitchParameter(True=NamedRerouteUsage(), False=NamedRerouteUsage()) {
   desc=Look To TopSide Shared
   group=Effect_Decal_Reactive_Emissive_D
   parameterName="EDRec_Emissive_D_Gradent__Tex -->TopS_00B_Sharerd_Tex_Msk"
  }
  n28 = StaticSwitchParameter(True=NamedRerouteUsage(), False=@n27) {
   desc=Look To TopSide Shared
   group=Effect_Decal_Reactive_Emissive_D
   parameterName="EDRec_Emissive_D_Gradent__Tex -->TopS_00A_Sharerd_Tex_Msk"
  }
  n29 = TextureSample(
   UVs=LinearInterpolate(
    A=@n26
    B=Saturate(None=@n26)
    Alpha=ScalarParameter() {group=Effect_Decal_Reactive_Emissive_D, parameterName="EDRec_Emissive_D_Gradent_Clamp  1=on"}
   )
   Tex=@n28
  )
  n30 = ComponentMask(None=@n24) {r=true}
  n31 = OneMinus(None=Divide(
    A=Frac(None=@n30)
    B=ScalarParameter() {defaultValue=1, group=Effect_Decal_Reactive_Emissive_D, parameterName=EDRec_Emissive_D_Gradent_Thickness}
   ) {constB=0.2})
  n32 = ScalarParameter() {defaultValue=1, group=Effect_Decal_Reactive_Emissive_D_Reactive, parameterName=EDRec_Emissive_D_Gradent_Falloff, sortPriority=211}
  n33 = Multiply(
   A=StaticSwitchParameter(
    True=ChannelMaskParameter(None=@n29.RGBA) {group=Effect_Decal_Reactive_Emissive_D, parameterName=EDRec_Emissive_D_Gradent_MSk_Channel}.R
    False=Power(Base=@n31, Exp=@n32)
   ) {group=Effect_Decal_Reactive_Emissive_D, parameterName="EDRec_Emissive_D_Gradent ---> use Texture"}
   B=NamedRerouteUsage()
  )
  n34 = Multiply(A=NamedRerouteUsage(), B=@n33)
  n35 = Multiply(A=@n3, B=@n34)
  n36 = Multiply(A=@n35, B=Add(A=@n35, B=@n34))
  n37 = Multiply(A=@n3, B=@n33)
  n38 = ScalarParameter() {group=Effect_Decal_Reactive_Emissive_D, parameterName="EDRec-Emissive-D_Height___X+X---->X", sortPriority=4}
  n39 = ScalarParameter() {defaultValue=1, group=Effect_Decal_Reactive_Emissive_D, parameterName=EDRec-Emissive-D_Height_Contrast, sortPriority=4}
  n40 = MaterialFunctionCall(In=Saturate(None=LinearInterpolate(A=@n36, B=@n37, Alpha=@n38)), Contrast=@n39) {materialFunction=&CheapContrast}
  n41 = ScalarParameter() {group=Effect_Decal_Reactive_Emissive_D, parameterName=EDRec-Emissive-D_Height__invert_Msk, sortPriority=4}
  n42 = LinearInterpolate(A=@n34, B=@n3, Alpha=@n41)
  n43 = LinearInterpolate(A=@n3, B=@n34, Alpha=@n41)
  n44 = ChannelMaskParameter(None=NamedRerouteUsage()) {group=Effect_Decal_Reactive_Emissive_D, parameterName=EDRec_Emissive_D_Mask_Channel, sortPriority=5}
  n45 = NamedRerouteUsage()
  n46 = StaticSwitch(
   True=StaticSwitch(
    True=StaticBoolParameter() {group=Effect_Decal_Reactive_Emissive_D, parameterName=EDRec_Emissive_D_Mask_Add_Counter, sortPriority=5}
    False=StaticBool()
    Value=NamedRerouteUsage()
   )
   False=StaticBool()
   Value=NamedRerouteUsage()
  )
  n47 = StaticSwitchParameter(
   True=StaticSwitch(True=Add(A=@n44.R, B=@n45), False=@n44.R, Value=@n46)
   False=StaticSwitchParameter(
    True=MaterialFunctionCall(Transition Phase=@n42, Height Texture=@n43, Contrast=@n39) {materialFunction=&HeightLerp}.Alpha
    False=StaticSwitchParameter(True=@n40.Result, False=@n33) {group=Effect_Decal_Reactive_Emissive_D, parameterName=EDRec-Emissive-D_Height___X+X, sortPriority=4}
   ) {group=Effect_Decal_Reactive_Emissive_D, parameterName=EDRec-Emissive-D_Height___Lerp, sortPriority=4}
  ) {group=Effect_Decal_Reactive_Emissive_D, parameterName=EDRec-Emissive-D_Only_Mask, sortPriority=4}
  Add(A=@n47, B=@n1)
  n48 = NamedRerouteUsage()
  n49 = StaticSwitchParameter(True=NamedRerouteUsage(), False=NamedRerouteUsage()) {group=Effect_Decal_Reactive_Emissive_A, parameterName="EDRec_Emissive_A_Warp_ Msk-->Bounds_Warp_Msk"}
  n50 = Multiply(
   A=ScalarParameter() {group=Effect_Decal_Reactive_Emissive_A, parameterName=EDRec_Emissive_A_Warp_Streagth}
   B=Add(
    A=ChannelMaskParameter(None=@n49) {group=Effect_Decal_Reactive_Emissive_A, parameterName=EDRec_Emissive_A_Warp_Mask_Channel}.R
    B=ScalarParameter() {group=Effect_Decal_Reactive_Emissive_A, parameterName=EDRec_Emissive_A_Warp_Offset}
   ) {constB=-0.5}
  )
  n51 = ScalarParameter() {defaultValue=1, group=Effect_Decal_Reactive_Emissive_A_Reactive |Gradent, parameterName=EDRec_Emissive_A_Gradent_Time_End, sortPriority=210}
  n52 = LinearInterpolate(
   A=ScalarParameter() {group=Effect_Decal_Reactive_Emissive_A_Reactive |Gradent, parameterName=EDRec_Emissive_A_Gradent_Time_Start, sortPriority=210}
   B=@n51
   Alpha=NamedRerouteUsage()
  )
  n53 = VectorParameter() {
   group=Effect_Decal_Reactive_Emissive_A_Reactive |Gradent
   parameterName=EDRec_Emissive_A_Gradent_Mask Direction
   sortPriority=210
   channelNames {a="", b="", g=X, r=Y}
   defaultValue=<rgba 1 1 0 1>
  }
  n54 = ComponentMask(None=@n53.RGB) {g=true, r=true}
  n55 = Time()
  n56 = ScalarParameter() {group=Effect_Decal_Reactive_Emissive_A_Reactive |Gradent, parameterName=EDRec_Emissive_A_Gradent_Time_X_Speed, sortPriority=210}
  n57 = AppendVector(A=Constant(), B=@n56)
  n58 = VectorParameter() {group=Effect_Decal_Reactive_Emissive_A, parameterName=EDRec_Emissive_A_Gradent_Offset, defaultValue=<rgba 0 0 0 1>}
  n59 = VectorParameter() {group=Effect_Decal_Reactive_Emissive_A, parameterName=EDRec_Emissive_A_Gradent_Speed, defaultValue=<rgba 0 0 0 1>}
  n60 = Multiply(A=Time(), B=@n59.RGB)
  n61 = StaticBoolParameter() {group=Effect_Decal_Reactive_Emissive_A_Reactive, parameterName=EDRec_Emissive_A_Gradent_Reactive---------------------, sortPriority=200}
  n62 = StaticSwitch(
   True=Add(A=Multiply(A=@n52, B=@n54), B=Multiply(A=@n55, B=@n57))
   False=ComponentMask(None=Add(A=@n58.RGB, B=@n60)) {g=true, r=true}
   Value=@n61
  )
  n63 = Add(A=NamedRerouteUsage(), B=@n62)
  n64 = VectorParameter() {
   group=Effect_Decal_Reactive_Emissive_A
   parameterName=EDRec_Emissive_A_Gradent_Rotation
   channelNames {a="Mech\\Character\\Back", b=Vehicle_Side, g=Vehicle, r="Weapon\\Pickaxe"}
  }
  n65 = MaterialFunctionCall(Float4=NamedRerouteUsage()) {materialFunction=&BreakOutFloat4Components}
  n66 = LinearInterpolate(A=@n64."Weapon\\Pickaxe", B=LinearInterpolate(A=@n64.Vehicle, B=@n64.Vehicle, Alpha=@n65.G), Alpha=@n65.R)
  n67 = MaterialFunctionCall(
   UVs=NamedRerouteUsage()
   "Rotation Angle (0-1)"=Add(
    A=ScalarParameter() {group=Effect_Decal_Reactive_Emissive_A, parameterName=EDRec_Emissive_A_Gradent_Rotation}
    B=LinearInterpolate(A=@n66, B=@n64."Mech\\Character\\Back", Alpha=@n65.B)
   )
  ) {materialFunction=&CustomRotator}
  n68 = VectorParameter() {group=Effect_Decal_Reactive_Emissive_A, parameterName=EDRec_Emissive_A_Gradent_Scale, defaultValue=<rgba 1 1 1 1>}
  n69 = Multiply(A=@n68.RGB, B=@n68.RGB)
  n70 = Add(A=Multiply(A=@n67.Rotated Values, B=ComponentMask(None=@n69) {g=true, r=true}), B=@n62)
  n71 = ComponentMask(None=@n70) {g=true, r=true}
  n72 = StaticSwitchParameter(
   True=StaticSwitchParameter(True=Add(A=@n50, B=@n63), False=@n63) {group=Effect_Decal_Reactive_Emissive_A, parameterName=EDRec_Emissive_A_Warp_ Msk}
   False=StaticSwitchParameter(True=Add(A=@n50, B=@n71), False=@n71) {group=Effect_Decal_Reactive_Emissive_A, parameterName=EDRec_Emissive_A_Warp_ Msk}
  ) {group=Effect_Decal_Reactive_Emissive_A, parameterName="EDRec_Emissive_A_Grad UV---->00_UV_Bounds", sortPriority=0}
  n73 = StaticSwitchParameter(True=NamedRerouteUsage(), False=NamedRerouteUsage()) {
   desc=Look To TopSide Shared
   group=Effect_Decal_Reactive_Emissive_A
   parameterName="EDRec_Emissive_A_Gradent__Tex -->TopS_00B_Sharerd_Tex_Msk"
  }
  n74 = StaticSwitchParameter(True=NamedRerouteUsage(), False=@n73) {
   desc=Look To TopSide Shared
   group=Effect_Decal_Reactive_Emissive_A
   parameterName="EDRec_Emissive_A_Gradent__Tex -->TopS_00A_Sharerd_Tex_Msk"
  }
  n75 = TextureSample(
   UVs=LinearInterpolate(
    A=@n72
    B=Saturate(None=@n72)
    Alpha=ScalarParameter() {group=Effect_Decal_Reactive_Emissive_A, parameterName="EDRec_Emissive_A_Gradent_Clamp  1=on"}
   )
   Tex=@n74
  )
  n76 = ComponentMask(None=@n70) {r=true}
  n77 = OneMinus(None=Divide(
    A=Frac(None=@n76)
    B=ScalarParameter() {defaultValue=1, group=Effect_Decal_Reactive_Emissive_A, parameterName=EDRec_Emissive_A_Gradent_Thickness}
   ) {constB=0.2})
  n78 = ScalarParameter() {defaultValue=1, group=Effect_Decal_Reactive_Emissive_A_Reactive, parameterName=EDRec_Emissive_A_Gradent_Falloff, sortPriority=211}
  n79 = Multiply(
   A=StaticSwitchParameter(
    True=ChannelMaskParameter(None=@n75.RGBA) {group=Effect_Decal_Reactive_Emissive_A, parameterName=EDRec_Emissive_A_Gradent_MSk_Channel}.R
    False=Power(Base=@n77, Exp=@n78)
   ) {group=Effect_Decal_Reactive_Emissive_A, parameterName="EDRec_Emissive_A_Gradent ---> use Texture"}
   B=NamedRerouteUsage()
  )
  n80 = Multiply(A=NamedRerouteUsage(), B=@n79)
  n81 = Multiply(
   A=ChannelMaskParameter(None=NamedRerouteUsage()) {group=Effect_Decal_Reactive_Emissive_A, parameterName=EDRec_Emissive_A_Mask_Channel, sortPriority=5}.R
   B=NamedRerouteUsage()
  )
  n82 = StaticSwitch(
   True=Add(A=@n81, B=NamedRerouteUsage())
   False=@n81
   Value=StaticSwitch(True=StaticBool(), False=NamedRerouteUsage(), Value=NamedRerouteUsage())
  )
  n83 = Multiply(A=@n82, B=@n80)
  n84 = Multiply(A=@n83, B=Add(A=@n83, B=@n80))
  n85 = Multiply(A=@n82, B=@n80)
  n86 = ScalarParameter() {group=Effect_Decal_Reactive_Emissive_A, parameterName="EDRec-Emissive-A_Height___X+X---->X", sortPriority=4}
  n87 = ScalarParameter() {defaultValue=1, group=Effect_Decal_Reactive_Emissive_A, parameterName=EDRec-Emissive-A_Height_Contrast, sortPriority=4}
  n88 = MaterialFunctionCall(In=Saturate(None=LinearInterpolate(A=@n84, B=@n85, Alpha=@n86)), Contrast=@n87) {materialFunction=&CheapContrast}
  n89 = ScalarParameter() {group=Effect_Decal_Reactive_Emissive_A, parameterName=EDRec-Emissive-A_Height__invert_Msk, sortPriority=4}
  n90 = LinearInterpolate(A=@n80, B=@n82, Alpha=@n89)
  n91 = LinearInterpolate(A=@n82, B=@n80, Alpha=@n89)
  n92 = ChannelMaskParameter(None=NamedRerouteUsage()) {group=Effect_Decal_Reactive_Emissive_A, parameterName=EDRec_Emissive_A_Mask_Channel, sortPriority=5}
  n93 = NamedRerouteUsage()
  n94 = StaticSwitch(
   True=StaticSwitch(
    True=StaticBool()
    False=StaticBoolParameter() {group=Effect_Decal_Reactive_Emissive_A, parameterName=EDRec_Emissive_A_Mask_Add_Counter, sortPriority=5}
    Value=NamedRerouteUsage()
   )
   False=StaticBool()
   Value=NamedRerouteUsage()
  )
  n95 = StaticSwitchParameter(
   True=StaticSwitch(True=Add(A=@n92.R, B=@n93), False=@n92.R, Value=@n94)
   False=StaticSwitchParameter(
    True=MaterialFunctionCall(Transition Phase=@n90, Height Texture=@n91, Contrast=@n87) {materialFunction=&HeightLerp}.Alpha
    False=StaticSwitchParameter(True=@n88.Result, False=@n79) {group=Effect_Decal_Reactive_Emissive_A, parameterName=EDRec-Emissive-A_Height___X+X, sortPriority=4}
   ) {group=Effect_Decal_Reactive_Emissive_A, parameterName=EDRec-Emissive-A_Height___Lerp, sortPriority=4}
  ) {group=Effect_Decal_Reactive_Emissive_A, parameterName=EDRec-Emissive-A_Only_Mask, sortPriority=4}
  Add(A=@n95, B=@n48)
  n96 = NamedRerouteUsage()
  n97 = ScalarParameter() {defaultValue=1, group=Effect_Decal_Reactive_Emissive_B_Reactive |Gradent, parameterName=EDRec_Emissive_B_Gradent_Time_End, sortPriority=210}
  n98 = LinearInterpolate(
   A=ScalarParameter() {group=Effect_Decal_Reactive_Emissive_B_Reactive |Gradent, parameterName=EDRec_Emissive_B_Gradent_Time_Start, sortPriority=210}
   B=@n97
   Alpha=NamedRerouteUsage()
  )
  n99 = VectorParameter() {
   group=Effect_Decal_Reactive_Emissive_B_Reactive |Gradent
   parameterName=EDRec_Emissive_B_Gradent_Mask Direction
   sortPriority=210
   channelNames {a="", b="", g=X, r=Y}
   defaultValue=<rgba 1 1 0 1>
  }
  n100 = ComponentMask(None=@n99.RGB) {g=true, r=true}
  n101 = Time()
  n102 = ScalarParameter() {group=Effect_Decal_Reactive_Emissive_B_Reactive |Gradent, parameterName=EDRec_Emissive_B_Gradent_Time_X_Speed, sortPriority=210}
  n103 = AppendVector(A=Constant(), B=@n102)
  n104 = VectorParameter() {group=Effect_Decal_Reactive_Emissive_B, parameterName=EDRec_Emissive_B_Gradent_Offset, defaultValue=<rgba 0 0 0 1>}
  n105 = VectorParameter() {group=Effect_Decal_Reactive_Emissive_B, parameterName=EDRec_Emissive_B_Gradent_Speed, defaultValue=<rgba 0 0 0 1>}
  n106 = Multiply(A=Time(), B=@n105.RGB)
  n107 = StaticBoolParameter() {group=Effect_Decal_Reactive_Emissive_B_Reactive, parameterName=EDRec_Emissive_B_Gradent_Reactive---------------------, sortPriority=200}
  n108 = StaticSwitch(
   True=Add(A=Multiply(A=@n98, B=@n100), B=Multiply(A=@n101, B=@n103))
   False=ComponentMask(None=Add(A=@n104.RGB, B=@n106)) {g=true, r=true}
   Value=@n107
  )
  n109 = Add(A=NamedRerouteUsage(), B=@n108)
  n110 = StaticSwitchParameter(True=NamedRerouteUsage(), False=NamedRerouteUsage()) {group=Effect_Decal_Reactive_Emissive_B, parameterName="EDRec_Emissive_B_Warp_ Msk-->Bounds_Warp_Msk"}
  n111 = Multiply(
   A=ScalarParameter() {group=Effect_Decal_Reactive_Emissive_B, parameterName=EDRec_Emissive_B_Warp_Streagth}
   B=Add(
    A=ChannelMaskParameter(None=@n110) {group=Effect_Decal_Reactive_Emissive_B, parameterName=EDRec_Emissive_B_Warp_Mask_Channel}.R
    B=ScalarParameter() {group=Effect_Decal_Reactive_Emissive_B, parameterName=EDRec_Emissive_B_Warp_Offset}
   ) {constB=-0.5}
  )
  n112 = VectorParameter() {
   group=Effect_Decal_Reactive_Emissive_B
   parameterName=EDRec_Emissive_B_Gradent_Rotation
   channelNames {a="Mech\\Character\\Back", b=Vehicle_Side, g=Vehicle, r="Weapon\\Pickaxe"}
  }
  n113 = MaterialFunctionCall(Float4=NamedRerouteUsage()) {materialFunction=&BreakOutFloat4Components}
  n114 = LinearInterpolate(A=@n112."Weapon\\Pickaxe", B=LinearInterpolate(A=@n112.Vehicle, B=@n112.Vehicle, Alpha=@n113.G), Alpha=@n113.R)
  n115 = MaterialFunctionCall(
   UVs=NamedRerouteUsage()
   "Rotation Angle (0-1)"=Add(
    A=ScalarParameter() {group=Effect_Decal_Reactive_Emissive_B, parameterName=EDRec_Emissive_B_Gradent_Rotation}
    B=LinearInterpolate(A=@n114, B=@n112."Mech\\Character\\Back", Alpha=@n113.B)
   )
  ) {materialFunction=&CustomRotator}
  n116 = VectorParameter() {group=Effect_Decal_Reactive_Emissive_B, parameterName=EDRec_Emissive_B_Gradent_Scale, defaultValue=<rgba 1 1 1 5>}
  n117 = Multiply(A=@n116.RGB, B=@n116.RGB)
  n118 = Add(A=Multiply(A=@n115.Rotated Values, B=ComponentMask(None=@n117) {g=true, r=true}), B=@n108)
  n119 = ComponentMask(None=@n118) {g=true, r=true}
  n120 = StaticSwitchParameter(
   True=StaticSwitchParameter(True=Add(A=@n109, B=@n111), False=@n109) {group=Effect_Decal_Reactive_Emissive_B, parameterName=EDRec_Emissive_B_Warp_ Msk}
   False=StaticSwitchParameter(True=Add(A=@n111, B=@n119), False=@n119) {group=Effect_Decal_Reactive_Emissive_B, parameterName=EDRec_Emissive_B_Warp_ Msk}
  ) {group=Effect_Decal_Reactive_Emissive_B, parameterName="EDRec_Emissive_B_Grad UV---->00_UV_Bounds", sortPriority=0}
  n121 = StaticSwitchParameter(True=NamedRerouteUsage(), False=NamedRerouteUsage()) {
   desc=Look To TopSide Shared
   group=Effect_Decal_Reactive_Emissive_B
   parameterName="EDRec_Emissive_B_Gradent__Tex -->TopS_00B_Sharerd_Tex_Msk"
  }
  n122 = StaticSwitchParameter(True=NamedRerouteUsage(), False=@n121) {
   desc=Look To TopSide Shared
   group=Effect_Decal_Reactive_Emissive_B
   parameterName="EDRec_Emissive_B_Gradent__Tex -->TopS_00A_Sharerd_Tex_Msk"
  }
  n123 = TextureSample(
   UVs=LinearInterpolate(
    A=@n120
    B=Saturate(None=@n120)
    Alpha=ScalarParameter() {group=Effect_Decal_Reactive_Emissive_B, parameterName="EDRec_Emissive_B_Gradent_Clamp  1=on"}
   )
   Tex=@n122
  )
  n124 = ComponentMask(None=@n118) {r=true}
  n125 = OneMinus(None=Divide(
    A=Frac(None=@n124)
    B=ScalarParameter() {defaultValue=1, group=Effect_Decal_Reactive_Emissive_B, parameterName=EDRec_Emissive_B_Gradent_Thickness}
   ) {constB=0.2})
  n126 = ScalarParameter() {defaultValue=1, group=Effect_Decal_Reactive_Emissive_B_Reactive, parameterName=EDRec_Emissive_B_Gradent_Falloff, sortPriority=211}
  n127 = Multiply(
   A=StaticSwitchParameter(
    True=ChannelMaskParameter(None=@n123.RGBA) {group=Effect_Decal_Reactive_Emissive_B, parameterName=EDRec_Emissive_B_Gradent_MSk_Channel}.R
    False=Power(Base=@n125, Exp=@n126)
   ) {group=Effect_Decal_Reactive_Emissive_B, parameterName="EDRec_Emissive_B_Gradent ---> use Texture"}
   B=NamedRerouteUsage()
  )
  n128 = Multiply(A=NamedRerouteUsage(), B=@n127)
  n129 = Multiply(
   A=ChannelMaskParameter(None=NamedRerouteUsage()) {group=Effect_Decal_Reactive_Emissive_B, parameterName=EDRec_Emissive_B_Mask_Channel, sortPriority=5}.R
   B=NamedRerouteUsage()
  )
  n130 = Multiply(A=@n129, B=@n128)
  n131 = Multiply(A=@n130, B=Add(A=@n130, B=@n128))
  n132 = Multiply(A=@n129, B=@n127)
  n133 = ScalarParameter() {group=Effect_Decal_Reactive_Emissive_B, parameterName="EDRec-Emissive-B_Height___X+X---->X", sortPriority=4}
  n134 = ScalarParameter() {defaultValue=1, group=Effect_Decal_Reactive_Emissive_B, parameterName=EDRec-Emissive-B_Height_Contrast, sortPriority=4}
  n135 = MaterialFunctionCall(In=Saturate(None=LinearInterpolate(A=@n131, B=@n132, Alpha=@n133)), Contrast=@n134) {materialFunction=&CheapContrast}
  n136 = ScalarParameter() {group=Effect_Decal_Reactive_Emissive_B, parameterName=EDRec-Emissive-B_Height__invert_Msk, sortPriority=4}
  n137 = LinearInterpolate(A=@n128, B=@n129, Alpha=@n136)
  n138 = LinearInterpolate(A=@n129, B=@n128, Alpha=@n136)
  n139 = StaticSwitchParameter(
   True=ChannelMaskParameter(None=NamedRerouteUsage()) {group=Effect_Decal_Reactive_Emissive_B, parameterName=EDRec_Emissive_B_Mask_Channel, sortPriority=5}.R
   False=StaticSwitchParameter(
    True=MaterialFunctionCall(Transition Phase=@n137, Height Texture=@n138, Contrast=@n134) {materialFunction=&HeightLerp}.Alpha
    False=StaticSwitchParameter(True=@n135.Result, False=@n127) {group=Effect_Decal_Reactive_Emissive_B, parameterName=EDRec-Emissive-B_Height___X+X, sortPriority=4}
   ) {group=Effect_Decal_Reactive_Emissive_B, parameterName=EDRec-Emissive-B_Height___Lerp, sortPriority=4}
  ) {group=Effect_Decal_Reactive_Emissive_B, parameterName=EDRec-Emissive-B_Only_Mask, sortPriority=4}
  Add(A=@n139, B=@n96)
  n140 = NamedRerouteUsage()
  n141 = StaticSwitchParameter(True=NamedRerouteUsage(), False=NamedRerouteUsage()) {group=Effect_Decal_Reactive_Emissive_C, parameterName="EDRec_Emissive_C_Warp_ Msk-->Bounds_Warp_Msk"}
  n142 = Multiply(
   A=ScalarParameter() {group=Effect_Decal_Reactive_Emissive_C, parameterName=EDRec_Emissive_C_Warp_Streagth}
   B=Add(
    A=ChannelMaskParameter(None=@n141) {group=Effect_Decal_Reactive_Emissive_C, parameterName=EDRec_Emissive_C_Warp_Mask_Channel}.R
    B=ScalarParameter() {group=Effect_Decal_Reactive_Emissive_C, parameterName=EDRec_Emissive_C_Warp_Offset}
   )
  )
  n143 = ScalarParameter() {defaultValue=1, group=Effect_Decal_Reactive_Emissive_C_Reactive |Gradent, parameterName=EDRec_Emissive_C_Gradent_Time_End, sortPriority=210}
  n144 = LinearInterpolate(
   A=ScalarParameter() {group=Effect_Decal_Reactive_Emissive_C_Reactive |Gradent, parameterName=EDRec_Emissive_C_Gradent_Time_Start, sortPriority=210}
   B=@n143
   Alpha=NamedRerouteUsage()
  )
  n145 = VectorParameter() {
   group=Effect_Decal_Reactive_Emissive_C_Reactive |Gradent
   parameterName=EDRec_Emissive_C_Gradent_Mask Direction
   sortPriority=210
   channelNames {a="", b="", g=X, r=Y}
   defaultValue=<rgba 1 1 0 1>
  }
  n146 = ComponentMask(None=@n145.RGB) {g=true, r=true}
  n147 = Time()
  n148 = ScalarParameter() {group=Effect_Decal_Reactive_Emissive_C_Reactive |Gradent, parameterName=EDRec_Emissive_C_Gradent_Time_X_Speed, sortPriority=210}
  n149 = AppendVector(A=Constant(), B=@n148)
  n150 = VectorParameter() {group=Effect_Decal_Reactive_Emissive_C, parameterName=EDRec_Emissive_C_Gradent_Offset, defaultValue=<rgba 0 0 0 1>}
  n151 = VectorParameter() {group=Effect_Decal_Reactive_Emissive_C, parameterName=EDRec_Emissive_C_Gradent_Speed, defaultValue=<rgba 0 0.5 0 1>}
  n152 = Multiply(A=Time(), B=@n151.RGB)
  n153 = StaticBoolParameter() {group=Effect_Decal_Reactive_Emissive_C_Reactive, parameterName=EDRec_Emissive_C_Gradent_Reactive---------------------, sortPriority=200}
  n154 = StaticSwitch(
   True=Add(A=Multiply(A=@n144, B=@n146), B=Multiply(A=@n147, B=@n149))
   False=ComponentMask(None=Add(A=@n150.RGB, B=@n152)) {g=true, r=true}
   Value=@n153
  )
  n155 = Add(A=NamedRerouteUsage(), B=@n154)
  n156 = VectorParameter() {
   group=Effect_Decal_Reactive_Emissive_C
   parameterName=EDRec_Emissive_C_Gradent_Rotation
   channelNames {a="Mech\\Character\\Back", b=Vehicle_Side, g=Vehicle, r="Weapon\\Pickaxe"}
  }
  n157 = MaterialFunctionCall(Float4=NamedRerouteUsage()) {materialFunction=&BreakOutFloat4Components}
  n158 = LinearInterpolate(A=@n156."Weapon\\Pickaxe", B=LinearInterpolate(A=@n156.Vehicle, B=@n156.Vehicle, Alpha=@n157.G), Alpha=@n157.R)
  n159 = MaterialFunctionCall(
   UVs=NamedRerouteUsage()
   "Rotation Angle (0-1)"=Add(
    A=ScalarParameter() {group=Effect_Decal_Reactive_Emissive_C, parameterName=EDRec_Emissive_C_Gradent_Rotation}
    B=LinearInterpolate(A=@n158, B=@n156."Mech\\Character\\Back", Alpha=@n157.B)
   )
  ) {materialFunction=&CustomRotator}
  n160 = VectorParameter() {group=Effect_Decal_Reactive_Emissive_C, parameterName=EDRec_Emissive_C_Gradent_Scale, defaultValue=<rgba 1 1 1 1>}
  n161 = Multiply(A=@n160.RGB, B=@n160.RGB)
  n162 = Add(A=Multiply(A=@n159.Rotated Values, B=ComponentMask(None=@n161) {g=true, r=true}), B=@n154)
  n163 = ComponentMask(None=@n162) {g=true, r=true}
  n164 = StaticSwitchParameter(
   True=StaticSwitchParameter(True=Add(A=@n142, B=@n155), False=@n155) {group=Effect_Decal_Reactive_Emissive_C, parameterName=EDRec_Emissive_C_Warp_ Msk}
   False=StaticSwitchParameter(True=Add(A=@n142, B=@n163), False=@n163) {group=Effect_Decal_Reactive_Emissive_C, parameterName=EDRec_Emissive_C_Warp_ Msk}
  ) {group=Effect_Decal_Reactive_Emissive_C, parameterName="EDRec_Emissive_C_Grad UV---->00_UV_Bounds", sortPriority=0}
  n165 = StaticSwitchParameter(True=NamedRerouteUsage(), False=NamedRerouteUsage()) {
   desc=Look To TopSide Shared
   group=Effect_Decal_Reactive_Emissive_C
   parameterName="EDRec_Emissive_C_Gradent__Tex -->TopS_00B_Sharerd_Tex_Msk"
  }
  n166 = StaticSwitchParameter(True=NamedRerouteUsage(), False=@n165) {
   desc=Look To TopSide Shared
   group=Effect_Decal_Reactive_Emissive_C
   parameterName="EDRec_Emissive_C_Gradent__Tex -->TopS_00A_Sharerd_Tex_Msk"
  }
  n167 = TextureSample(
   UVs=LinearInterpolate(
    A=@n164
    B=Saturate(None=@n164)
    Alpha=ScalarParameter() {group=Effect_Decal_Reactive_Emissive_C, parameterName="EDRec_Emissive_C_Gradent_Clamp  1=on"}
   )
   Tex=@n166
  )
  n168 = ComponentMask(None=@n162) {r=true}
  n169 = OneMinus(None=Divide(
    A=Frac(None=@n168)
    B=ScalarParameter() {defaultValue=1, group=Effect_Decal_Reactive_Emissive_C, parameterName=EDRec_Emissive_C_Gradent_Thickness}
   ) {constB=0.2})
  n170 = ScalarParameter() {defaultValue=1, group=Effect_Decal_Reactive_Emissive_C_Reactive, parameterName=EDRec_Emissive_C_Gradent_Falloff, sortPriority=211}
  n171 = Multiply(
   A=StaticSwitchParameter(
    True=ChannelMaskParameter(None=@n167.RGBA) {group=Effect_Decal_Reactive_Emissive_C, parameterName=EDRec_Emissive_C_Gradent_MSk_Channel}.R
    False=Power(Base=@n169, Exp=@n170)
   ) {group=Effect_Decal_Reactive_Emissive_C, parameterName="EDRec_Emissive_C_Gradent ---> use Texture"}
   B=NamedRerouteUsage()
  )
  n172 = Multiply(A=NamedRerouteUsage(), B=@n171)
  n173 = Multiply(
   A=ChannelMaskParameter(None=NamedRerouteUsage()) {group=Effect_Decal_Reactive_Emissive_C, parameterName=EDRec_Emissive_C_Mask_Channel, sortPriority=5}.R
   B=NamedRerouteUsage()
  )
  n174 = Multiply(A=@n173, B=@n172)
  n175 = Multiply(A=@n174, B=Add(A=@n174, B=@n172))
  n176 = Multiply(A=@n173, B=@n171)
  n177 = ScalarParameter() {group=Effect_Decal_Reactive_Emissive_C, parameterName="EDRec-Emissive-C_Height___X+X---->X", sortPriority=4}
  n178 = ScalarParameter() {defaultValue=1, group=Effect_Decal_Reactive_Emissive_C, parameterName=EDRec-Emissive-C_Height_Contrast, sortPriority=4}
  n179 = MaterialFunctionCall(In=Saturate(None=LinearInterpolate(A=@n175, B=@n176, Alpha=@n177)), Contrast=@n178) {materialFunction=&CheapContrast}
  n180 = ScalarParameter() {group=Effect_Decal_Reactive_Emissive_C, parameterName=EDRec-Emissive-C_Height__invert_Msk, sortPriority=4}
  n181 = LinearInterpolate(A=@n172, B=@n173, Alpha=@n180)
  n182 = LinearInterpolate(A=@n173, B=@n172, Alpha=@n180)
  n183 = StaticSwitchParameter(
   True=ChannelMaskParameter(None=NamedRerouteUsage()) {group=Effect_Decal_Reactive_Emissive_C, parameterName=EDRec_Emissive_C_Mask_Channel, sortPriority=5}.R
   False=StaticSwitchParameter(
    True=MaterialFunctionCall(Transition Phase=@n181, Height Texture=@n182, Contrast=@n178) {materialFunction=&HeightLerp}.Alpha
    False=StaticSwitchParameter(True=@n179.Result, False=@n171) {group=Effect_Decal_Reactive_Emissive_C, parameterName=EDRec-Emissive-C_Height___X+X, sortPriority=4}
   ) {group=Effect_Decal_Reactive_Emissive_C, parameterName=EDRec-Emissive-C_Height___Lerp, sortPriority=4}
  ) {group=Effect_Decal_Reactive_Emissive_C, parameterName=EDRec-Emissive-C_Only_Mask, sortPriority=4}
  Add(A=@n183, B=@n140)
  CollectionParameter() {collection=&FortniteMaterialParameters, parameterName=EventAudio_Amplitude}
  CollectionParameter() {collection=&FortniteMaterialParameters, parameterName=EventAudio_Amplitude}
  CollectionParameter() {collection=&FortniteMaterialParameters, parameterName=EventAudio_Amplitude}
  CollectionParameter() {collection=&FortniteMaterialParameters, parameterName=EventAudio_Amplitude}
  n184 = NamedRerouteUsage()
  ComponentMask(None=@n184) {a=true}
  n185 = ScalarParameter() {defaultValue=5, group=Effect_Decal_Reactive_Uv_Offset_LR, parameterName=EDRec_DecalB_Random_seed_}
  n186 = NamedRerouteUsage()
  n187 = Time()
  n188 = VectorParameter() {group=Effect_Decal_Reactive_Uv_Offset_LR, parameterName=EDRec_DecalB_Speed, defaultValue=<rgba 0 0 0 1>}
  n189 = ComponentMask(None=@n188.RGB) {g=true, r=true}
  n190 = Add(
   A=MaterialFunctionCall(UVs=@n186, "Rotation Angle (0-1)"=NamedRerouteUsage()) {materialFunction=&CustomRotator}.Rotated Values
   B=Add(A=Multiply(A=@n187, B=@n189), B=ComponentMask(None=@n184) {g=true, r=true})
  ) {constB=0}
  n191 = NamedRerouteUsage()
  n192 = Multiply(A=ComponentMask(None=@n190) {r=true}, B=@n191) {constB=5}
  n193 = Add(A=Divide(A=Floor(None=@n192), B=@n191) {constB=5}) {constB=-0.5}
  n194 = Multiply(A=@n191, B=ScalarParameter() {group=Effect_Decal_Reactive_Uv_Offset_LR, parameterName=EDRec_DecalB_Offset_Y})
  n195 = Add(A=Multiply(A=@n190, B=@n191), B=AppendVector(A=Constant(), B=Multiply(A=@n193, B=@n194))) {constB=0}
  n196 = Add(A=Divide(A=Floor(None=@n195), B=NamedRerouteUsage())) {constB=0.01}
  ComponentMask(None=MaterialFunctionCall(Seed=@n185, Vec2=@n196) {materialFunction=&MF_Hash}.Vec2) {r=true}
  Constant2Vector() {r=1}
  Constant2Vector() {r=1}
  Constant2Vector() {r=1}
  Constant2Vector() {r=1}
  Constant3Vector()
  Constant()
  Constant() {r=2}
  Constant()
  Constant()
  Constant() {r=0.2}
  Constant() {r=10}
  n197 = ScalarParameter() {group=Effect_Decal_Reactive_Uv_Animation_React, parameterName=EDRec_DecalA_BounceRot_Delay-----------------}
  n198 = ScalarParameter() {defaultValue=-9999999, group=Effect_Decal_Reactive_Uv_Animation_React, parameterName=EDRec_DecalA_BounceRot_TimeFrom_BP}
  Divide(
   A=Subtract(A=Time(), B=Add(A=@n197, B=@n198))
   B=ScalarParameter() {defaultValue=1, group=Effect_Decal_Reactive_Uv_Animation_React, parameterName=EDRec_DecalA_BounceRot_Duration}
  )
  Frac()
  Frac()
  Frac(None=@n120)
  Frac(None=@n164)
  n199 = NamedRerouteUsage()
  n200 = Constant()
  n201 = MaterialFunctionCall(
   R=NamedRerouteUsage()
   G=NamedRerouteUsage()
   B=StaticSwitch(True=@n200, False=NamedRerouteUsage(), Value=NamedRerouteUsage())
   A=StaticSwitch(True=@n200, False=NamedRerouteUsage(), Value=NamedRerouteUsage())
  ) {materialFunction=&AppendMany}
  FunctionOutput(None=Multiply(A=NamedRerouteUsage(), B=Multiply(A=@n199, B=@n201.RGBA))) {outputName=Decalmulti_Msk}
  n202 = NamedRerouteUsage()
  n203 = NamedRerouteUsage()
  n204 = StaticSwitch(True=@n203, False=StaticSwitch(True=@n203, False=NamedRerouteUsage(), Value=NamedRerouteUsage()), Value=NamedRerouteUsage())
  n205 = MaterialFunctionCall(Float4=FunctionInput(Preview=TextureSample() {samplerType=SAMPLERTYPE_Masks, texture=/Game/Weapons/FORT_Rifles/Textures/T_Scar_CM}.RGBA) {inputName=CM_Msk, inputType=FunctionInput_Vector4, sortPriority=4}) {materialFunction=&BreakOutFloat4Components}
  n206 = NamedRerouteDeclaration(None=@n205.B) {name=Msk_Accent, nodeColor=<rgba 0 1 0.7656 1>}
  n207 = NamedRerouteUsage()
  n208 = NamedRerouteUsage()
  n209 = NamedRerouteUsage()
  n210 = NamedRerouteUsage()
  n211 = LinearInterpolate(B=@n206, Alpha=Max(A=Max(A=@n207, B=@n208), B=Max(A=@n209, B=@n210)))
  n212 = Max(
   A=NamedRerouteDeclaration(None=@n205.R) {name=Msk_Prime, nodeColor=<rgba 0 1 0.7422 1>}
   B=Min(A=@n206, B=ScalarParameter() {group=Effect_Decal_Reactive, parameterName="EDRec_Accents_Mask  0=Off ", sortPriority=-9})
  )
  n213 = OneMinus(None=@n206)
  n214 = ScalarParameter() {defaultValue=1, group=Effect_Decal_Reactive, parameterName="EDRec_Accents_Mask PickAxe\\Backpack  1=Off", sortPriority=-9}
  n215 = MaterialFunctionCall(Float2=NamedRerouteUsage()) {materialFunction=&BreakOutFloat2Components}
  n216 = LinearInterpolate(A=@n212, B=Min(A=@n212, B=Max(A=@n213, B=@n214)), Alpha=Max(A=@n215.R, B=@n215.R))
  n217 = Saturate(None=Max(A=Max(A=@n211, B=@n216), B=ScalarParameter() {group=Effect_Decal_Reactive, parameterName=EDRec_Remove_Customization Mask, sortPriority=-9}))
  n218 = Multiply(A=Multiply(A=@n204, B=NamedRerouteUsage()), B=@n217)
  n219 = FunctionInput() {bUsePreviewValueAsDefault=true, inputName=Material, inputType=FunctionInput_MaterialAttributes, sortPriority=-20}
  n220 = GetMaterialAttributes(None=@n219) {attributeGetTypes [#1, #2, #3]}
  n221 = NamedRerouteUsage()
  n222 = StaticSwitch(True=@n221, False=StaticSwitch(True=@n221, False=NamedRerouteUsage(), Value=NamedRerouteUsage()), Value=NamedRerouteUsage())
  n223 = ScalarParameter() {defaultValue=1, group=Effect_Decal_Reactive, parameterName="EDRec_Darken_Root_BaseColor   0=on", sortPriority=0}
  n224 = ScalarParameter() {group=Effect_Decal_Reactive, parameterName="EDRec_BaseColor------->Emissive", sortPriority=0}
  n225 = ScalarParameter() {group=Effect_Decal_Reactive, parameterName=EDRec_Roghness_High, sortPriority=55}
  n226 = ScalarParameter() {defaultValue=1, group=Effect_Decal_Reactive, parameterName=EDRec_Roghness_Low, sortPriority=55}
  n227 = NamedRerouteUsage()
  n228 = ChannelMaskParameter(None=@n227) {group=Effect_Decal_Reactive, parameterName=EDRec_Decal_Roughness_Msk_Channel, sortPriority=55}
  n229 = NamedRerouteUsage()
  n230 = StaticSwitch(
   True=@n228.R
   False=StaticSwitch(
    True=@n228.R
    False=ChannelMaskParameter(None=@n229) {group=Effect_Decal_Reactive, parameterName=EDRec_Decal_Roughness_Msk_Channel, sortPriority=55}.R
    Value=NamedRerouteUsage()
   )
   Value=NamedRerouteUsage()
  )
  n231 = ScalarParameter() {group=Effect_Decal_Reactive, parameterName=EDRec_Specular_Low, sortPriority=61}
  n232 = ScalarParameter() {defaultValue=1, group=Effect_Decal_Reactive, parameterName=EDRec_Specular_High, sortPriority=61}
  n233 = ChannelMaskParameter(None=@n227) {group=Effect_Decal_Reactive, parameterName=EDRec_Specular_Msk_Channel, sortPriority=60}
  n234 = StaticSwitch(
   True=@n233.R
   False=StaticSwitch(
    True=@n233.R
    False=ChannelMaskParameter(None=@n229) {group=Effect_Decal_Reactive, parameterName=EDRec_Specular_Msk_Channel, sortPriority=60}.R
    Value=NamedRerouteUsage()
   )
   Value=NamedRerouteUsage()
  )
  n235 = ScalarParameter() {group=Effect_Decal_Reactive, parameterName=EDRec_Metallic_Low, sortPriority=50}
  n236 = ScalarParameter() {defaultValue=1, group=Effect_Decal_Reactive, parameterName=EDRec_Metallic_High, sortPriority=50}
  n237 = ChannelMaskParameter(None=@n227) {group=Effect_Decal_Reactive, parameterName=EDRec_Decal_Metal_Msk_Channel, sortPriority=50}
  n238 = StaticSwitch(
   True=@n237.R
   False=StaticSwitch(
    True=@n237.R
    False=ChannelMaskParameter(None=@n229) {group=Effect_Decal_Reactive, parameterName=EDRec_Decal_Metal_Msk_Channel, sortPriority=50}.R
    Value=NamedRerouteUsage()
   )
   Value=NamedRerouteUsage()
  )
  n239 = LinearInterpolate(
   A=@n220.EmissiveColor
   Alpha=ScalarParameter() {defaultValue=1, group=Effect_Decal_Reactive, parameterName="EDRec_Root_Emissive_Darken   1=on", sortPriority=0}
  ) {constB=0}
  n240 = Multiply(
   A=LinearInterpolate(B=@n222, Alpha=@n224)
   B=ScalarParameter() {defaultValue=1, group=Effect_Decal_Reactive, parameterName=EDRec_Emissive Multplyer, sortPriority=0}
  )
  n241 = VectorParameter() {
   group=Effect_Decal_Reactive
   parameterName="EDRec_Root_Emissive_+--->X"
   sortPriority=0
   channelNames {a="", b="", g="Root Emissive Brightness ", r="Add---->Multply"}
   defaultValue=<rgba 0 1 0 1>
  }
  n242 = Multiply(A=@n239, B=@n241."Root Emissive Brightness ")
  n243 = LinearInterpolate(A=NamedRerouteUsage(), B=NamedRerouteUsage(), Alpha=ScalarParameter() {group=Effect_Decal_Reactive, parameterName="EDRec_Root_Normals 1=On"})
  n244 = NamedRerouteUsage()
  n245 = NamedRerouteUsage()
  n246 = NamedRerouteUsage()
  n247 = Add(A=@n243, B=StaticSwitch(True=@n244, False=StaticSwitch(True=@n244, False=@n245, Value=@n246), Value=NamedRerouteUsage()))
  n248 = StaticSwitchParameter(
   True=StaticSwitchParameter(True=ShadingPathSwitch(Default=@n247, Deferred=@n247, Mobile=@n243) {desc=Migrated from FeatureLevelSwitch}, False=@n247) {defaultValue=true, group=Effect_Decal_Reactive, parameterName=EDRec_Decal_Normals_Android Disable}
   False=@n243
  ) {group=Effect_Decal_Reactive, parameterName=EDRec_Decal_Normals}
  n249 = SetMaterialAttributes!(
   MaterialAttributes=@n220.MaterialAttributes
   BaseColor=LinearInterpolate(A=@n222, B=Multiply(A=@n220.BaseColor, B=@n223), Alpha=@n224) {constB=0}
   Roughness=StaticSwitchParameter(
    True=LinearInterpolate(A=@n225, B=@n226, Alpha=@n230)
    False=ScalarParameter() {defaultValue=0.5, group=Effect_Decal_Reactive, parameterName=EDRec_Roghness, sortPriority=55}
   ) {group=Effect_Decal_Reactive, parameterName=EDRec_Decal_Roughness_Msk, sortPriority=55}
   Specular=StaticSwitchParameter(
    True=LinearInterpolate(A=@n231, B=@n232, Alpha=@n234)
    False=ScalarParameter() {defaultValue=0.5, group=Effect_Decal_Reactive, parameterName=EDRec_Specular, sortPriority=61}
   ) {group=Effect_Decal_Reactive, parameterName=EDRec_Decal_Specular_Msk, sortPriority=60}
   Metallic=StaticSwitchParameter(
    True=LinearInterpolate(A=@n235, B=@n236, Alpha=@n238)
    False=ScalarParameter() {group=Effect_Decal_Reactive, parameterName=EDRec_Metallic, sortPriority=50}
   ) {group=Effect_Decal_Reactive, parameterName=EDRec_Decal_Metal_Msk, sortPriority=50}
   EmissiveColor=LinearInterpolate(A=Add(A=@n239, B=@n240), B=Multiply(A=@n242, B=@n240), Alpha=@n241."Add---->Multply")
   Normal=@n248
  ) {attributeSetTypes [
    #3
    #4
    #5
    #6
    #2
    #1
   ]}
  n250 = Multiply(A=@n217, B=@n204)
  n251 = GetMaterialAttributes(None=StaticSwitchParameter(
    True=BlendMaterialAttributes(A=@n220.MaterialAttributes, B=@n202, Alpha=@n218) {vertexAttributeBlendType=UseA}
    False=BlendMaterialAttributes(A=@n220.MaterialAttributes, B=@n249, Alpha=@n250) {vertexAttributeBlendType=UseA}
   ) {group=Effect_Decal_Reactive, parameterName=EDRec_Emissive_only, sortPriority=-10}) {attributeGetTypes [#2, #3]}
  n252 = NamedRerouteUsage()
  n253 = Constant() {r=1}
  n254 = StaticBoolParameter() {group=Effect_Decal_Reactive_Emissive_A, parameterName=EDRec-Emissive-A-Decal, sortPriority=-1}
  n255 = NamedRerouteUsage()
  n256 = ScalarParameter() {defaultValue=1, group=Effect_Decal_Reactive_Emissive_B, parameterName="EDRec-Emissives B MskByDrawOrder   0=On", sortPriority=0}
  n257 = OneMinus(None=@n256)
  n258 = StaticBoolParameter() {group=Effect_Decal_Reactive_Emissive_B, parameterName=EDRec-Emissive-B-Decal, sortPriority=-1}
  n259 = StaticSwitch(True=Max(A=Min(A=@n255, B=@n257), B=NamedRerouteUsage()), False=@n253, Value=@n258)
  n260 = Max(A=@n255, B=NamedRerouteUsage())
  n261 = ScalarParameter() {defaultValue=1, group=Effect_Decal_Reactive_Emissive_C, parameterName="EDRec-Emissives C MskByDrawOrder   0=On", sortPriority=0}
  n262 = Min(A=@n260, B=OneMinus(None=@n261))
  n263 = NamedRerouteUsage()
  n264 = Constant() {r=1}
  n265 = StaticBoolParameter() {group=Effect_Decal_Reactive_Emissive_C, parameterName=EDRec-Emissive-C-Decal, sortPriority=-1}
  n266 = Max(A=@n260, B=NamedRerouteUsage())
  n267 = ScalarParameter() {defaultValue=1, group=Effect_Decal_Reactive_Emissive_D, parameterName="EDRec-Emissives D MskByDrawOrder   0=On", sortPriority=0}
  n268 = Min(A=@n266, B=OneMinus(None=@n267))
  n269 = NamedRerouteUsage()
  n270 = StaticBoolParameter() {group=Effect_Decal_Reactive_Emissive_D, parameterName=EDRec-Emissive-D-Decal, sortPriority=-1}
  n271 = Multiply(A=StaticSwitch(True=Max(A=@n262, B=@n263), False=@n264, Value=@n265), B=StaticSwitch(True=Max(A=@n268, B=@n269), False=@n264, Value=@n270))
  n272 = LinearInterpolate(B=Multiply(A=StaticSwitch(True=@n252, False=@n253, Value=@n254), B=Multiply(A=@n259, B=@n271)), Alpha=NamedRerouteUsage()) {constA=1}
  n273 = Constant()
  n274 = StaticSwitch(True=NamedRerouteUsage(), False=@n273, Value=@n254)
  n275 = OneMinus(None=@n255)
  n276 = StaticSwitch(True=Multiply(A=NamedRerouteUsage(), B=Max(A=@n275, B=@n256)), False=@n273, Value=@n258)
  n277 = OneMinus(None=@n260)
  n278 = StaticSwitch(True=Multiply(A=Max(A=@n277, B=@n261), B=NamedRerouteUsage()), False=@n273, Value=@n265)
  n279 = OneMinus(None=@n266)
  n280 = StaticSwitch(True=Multiply(A=Max(A=@n279, B=@n267), B=NamedRerouteUsage()), False=@n273, Value=@n270)
  n281 = Add(A=@n276, B=Add(A=@n278, B=@n280))
  n282 = Max(A=@n276, B=Max(A=@n278, B=@n280))
  n283 = Multiply(
   A=NamedRerouteUsage()
   B=LinearInterpolate(
    A=Add(A=@n274, B=@n281)
    B=Max(A=@n274, B=@n282)
    Alpha=ScalarParameter() {group=Effect_Decal_Reactive, parameterName="EDRec-Emissives ------->Max", sortPriority=0}
   )
  )
  n284 = ScalarParameter() {group=Effect_Decal_Reactive, parameterName="EDRec-Emissives ------->BaseColor", sortPriority=0}
  n285 = SetMaterialAttributes!(
   MaterialAttributes=@n251.MaterialAttributes
   EmissiveColor=Add(A=Multiply(A=@n251.EmissiveColor, B=@n272), B=LinearInterpolate(A=@n283, Alpha=@n284) {constB=0})
   BaseColor=Add(A=Multiply(A=@n251.BaseColor, B=@n272), B=LinearInterpolate(B=@n283, Alpha=@n284))
  ) {attributeSetTypes [#2, #3]}
  n286 = GetMaterialAttributes(None=@n285) {attributeGetTypes [#2, #3]}
  n287 = NamedRerouteUsage()
  n288 = Max(A=NamedRerouteUsage(), B=NamedRerouteUsage())
  n289 = NamedRerouteUsage()
  n290 = StaticSwitch(True=LinearInterpolate(A=@n287, Alpha=Max(A=@n288, B=@n289)) {constB=0}, False=@n287, Value=NamedRerouteUsage())
  n291 = VectorParameter() {group=Effect_Decal_Reactive_Counter, parameterName=EDRec-Counter_Emissive_Color, defaultValue=<rgba 1 1 1 1>}
  n292 = Multiply(A=@n291.RGB, B=@n291.RGB)
  n293 = Multiply(
   A=@n292
   B=LinearInterpolate(
    A=ScalarParameter() {defaultValue=1, group=Effect_Decal_Reactive_Counter_Reacive, parameterName=EDRec_Counter_Brightness Start}
    B=ScalarParameter() {group=Effect_Decal_Reactive_Counter_Reacive, parameterName=EDRec_Counter_Brightness End}
    Alpha=NamedRerouteUsage()
   )
  )
  n294 = Multiply(
   A=NamedRerouteUsage()
   B=Multiply(
    A=@n290
    B=StaticSwitchParameter(True=@n293, False=@n292) {group=Effect_Decal_Reactive_Counter, parameterName=EDRec-Counter_Reactive, sortPriority=0}
   )
  )
  n295 = Add(A=@n286.EmissiveColor, B=@n294)
  n296 = ScalarParameter() {defaultValue=1, group=Effect_Decal_Reactive_Counter, parameterName="EDRec-Counter_Emissive--->0=BaseColor"}
  n297 = StaticSwitchParameter(
   True=LinearInterpolate(
    A=ScalarParameter() {defaultValue=1, group=Effect_Decal_Reactive_Counter_Reacive, parameterName=EDRec_Counter_DarkenBase Start}
    B=ScalarParameter() {defaultValue=1, group=Effect_Decal_Reactive_Counter_Reacive, parameterName=EDRec_Counter_DarkenBase End}
    Alpha=NamedRerouteUsage()
   )
   False=ScalarParameter() {defaultValue=1, group=Effect_Decal_Reactive_Counter, parameterName="EDRec_Counter_DarkenBase   0=On"}
  ) {group=Effect_Decal_Reactive_Counter, parameterName=EDRec-Counter_Reactive, sortPriority=0}
  n298 = Multiply(A=@n286.BaseColor, B=Max(A=Min(A=@n297, B=@n296), B=OneMinus(None=@n290)))
  n299 = Add(A=@n298, B=@n294)
  n300 = StaticSwitch(
   True=@n285
   False=SetMaterialAttributes(
    MaterialAttributes=@n286.MaterialAttributes
    EmissiveColor=LinearInterpolate(A=@n286.EmissiveColor, B=@n295, Alpha=@n296)
    BaseColor=LinearInterpolate(A=@n299, B=@n298, Alpha=@n296) {constB=0}
   ) {attributeSetTypes [#2, #3]}
   Value=NamedRerouteUsage()
  )
  n301 = StaticBoolParameter() {desc=use Decal A Location, group=Effect_Decal_Reactive_Counter, parameterName=EDRec-Counter, sortPriority=0}
  FunctionOutput(None=StaticSwitchParameter(True=StaticSwitch(True=@n300, False=@n285, Value=@n301), False=NamedRerouteUsage()) {group=Customization, parameterName=Effect_Multi_Decal_Reactive})
  n302 = Reroute()
  LinearInterpolate(Alpha=@n302) {constA=-3, constB=3}
  n303 = VectorParameter() {
   group=Effect_Decal_Reactive_Uv_Offset_LR
   parameterName=EDRec_DecalD_Rotation--
   channelNames {a=Rotation_B_Speed, b=Rotation_A_Speed, g=Rotation_B, r=Rotation_A}
  }
  n304 = NamedRerouteUsage()
  n305 = Time()
  n306 = VectorParameter() {group=Effect_Decal_Reactive_Uv_Offset_LR, parameterName=EDRec_DecalD_Speed, defaultValue=<rgba 0 0 0 1>}
  n307 = ComponentMask(None=@n306.RGB) {g=true, r=true}
  n308 = NamedRerouteUsage()
  n309 = Add(
   A=MaterialFunctionCall(UVs=@n304, "Rotation Angle (0-1)"=NamedRerouteUsage()) {materialFunction=&CustomRotator}.Rotated Values
   B=Add(A=Multiply(A=@n305, B=@n307), B=ComponentMask(None=@n308) {g=true, r=true})
  ) {constB=0}
  n310 = NamedRerouteUsage()
  n311 = Multiply(A=ComponentMask(None=@n309) {r=true}, B=@n310) {constB=5}
  n312 = Add(A=Divide(A=Floor(None=@n311), B=@n310) {constB=5}) {constB=-0.5}
  n313 = Multiply(A=@n310, B=ScalarParameter() {group=Effect_Decal_Reactive_Uv_Offset_LR, parameterName=EDRec_DecalD_Offset_Y})
  n314 = Add(A=Multiply(A=@n309, B=@n310), B=AppendVector(A=Constant(), B=Multiply(A=@n312, B=@n313))) {constB=0}
  n315 = VectorParameter() {group=Effect_Decal_Reactive_Uv_Offset_LR, parameterName=EDRec_DecalD_Random_seed, defaultValue=<rgba 4 3 0 0>}
  n316 = MaterialFunctionCall!(2 Vector=Add(A=Floor(None=@n314), B=ComponentMask(None=@n315.RGB) {g=true, r=true}), RandomSeed) {materialFunction=&/MF_Random_2}
  n317 = LinearInterpolate(A=@n303.Rotation_A, B=@n303.Rotation_A, Alpha=@n316.utput)
  LinearInterpolate(Alpha=@n317) {constA=-3, constB=3}
  n318 = Multiply(A=ScalarParameter() {group=Effect_Decal_Reactive_Uv_Offset_LR, parameterName=EDRec_DecalC_RND_Msking Speed}, B=Time())
  LinearInterpolate(
   A=@n318
   B=ConstantBiasScale(None=Sine(None=@n318))
   Alpha=ScalarParameter() {group=Effect_Decal_Reactive_Uv_Offset_LR, parameterName=EDRec_DecalC_RND_Msking Speed Sine}
  )
  LinearInterpolate(A=NamedRerouteUsage(), B=NamedRerouteUsage(), Alpha=ScalarParameter() {group=Effect_Decal_Reactive, parameterName="EDRec_Root_Normals 1=On"})
  MaterialFunctionCall(Float2=NamedRerouteUsage()) {materialFunction=&BreakOutFloat2Components}
  n319 = ScalarParameter() {defaultValue=5, group=Effect_Decal_Reactive_Uv_Offset_LR, parameterName=EDRec_DecalA_Random_seed_}
  n320 = NamedRerouteUsage()
  n321 = Time()
  n322 = VectorParameter() {group=Effect_Decal_Reactive_Uv_Offset_LR, parameterName=EDRec_DecalA_Speed, defaultValue=<rgba 0 0 0 1>}
  n323 = ComponentMask(None=@n322.RGB) {g=true, r=true}
  n324 = NamedRerouteUsage()
  n325 = Add(
   A=MaterialFunctionCall(UVs=@n320, "Rotation Angle (0-1)"=NamedRerouteUsage()) {materialFunction=&CustomRotator}.Rotated Values
   B=Add(A=Multiply(A=@n321, B=@n323), B=ComponentMask(None=@n324) {g=true, r=true})
  ) {constB=0}
  n326 = Constant() {r=5}
  n327 = Multiply(A=ComponentMask(None=@n325) {r=true}, B=@n326) {constB=5}
  n328 = Add(A=Divide(A=Floor(None=@n327), B=@n326) {constB=5}) {constB=-0.5}
  n329 = Multiply(A=@n326, B=ScalarParameter() {group=Effect_Decal_Reactive_Uv_Offset_LR, parameterName=EDRec_DecalA_Offset_Y})
  n330 = Add(A=Multiply(A=@n325, B=@n326), B=AppendVector(A=Constant(), B=Multiply(A=@n328, B=@n329))) {constB=0}
  n331 = Add(A=Divide(A=Floor(None=@n330), B=@n326)) {constB=0.01}
  MaterialFunctionCall(Seed=@n319, Vec2=@n331) {materialFunction=&MF_Hash}
  n332 = VectorParameter() {group=Effect_Decal_Reactive_Emissive_UV_00, parameterName=EDRec_Emissive_00_Speed, defaultValue=<rgba 0 0 0 1>}
  Multiply(A=@n332.RGB, B=Time())
  Multiply(A=NamedRerouteUsage())
  Multiply()
  n333 = ComponentMask(None=@n324) {a=true}
  Multiply(A=ScalarParameter() {defaultValue=1, group=Effect_Decal_Reactive_Uv_Offset_LR, parameterName=EDRec_DecalA-Scale Icon}, B=@n333)
  Multiply()
  n334 = VectorParameter() {
   group=Effect_Decal_Reactive_FlipBook
   parameterName=EDRec_FlipBook_FrameB_RND
   channelNames {a="", b=Offset, g=Min, r=Max}
   defaultValue=<rgba 0 0 0 1>
  }
  n335 = Constant2Vector() {g=280.3817, r=574.9868}
  n336 = Add(A=Constant() {r=43758.5469}, B=@n185)
  n337 = Multiply(A=Sine(None=DotProduct(A=@n196, B=@n335)), B=@n336)
  n338 = LinearInterpolate(A=@n334.Min, B=@n334.Min, Alpha=Frac(None=Add(A=@n337, B=@n334.Offset)))
  Multiply(A=@n338, B=@n334.A)
  n339 = VectorParameter() {
   group=Effect_Decal_Reactive_Uv_Offset_LR
   parameterName=EDRec_DecalA_Scale-RND--
   channelNames {a="1= Sine Speed", b=Speed, g=Max, r=Min}
   defaultValue=<rgba 1 1 0 1>
  }
  n340 = Constant2Vector() {g=724.3587, r=235.9474}
  n341 = Add(A=Constant() {r=43758.5469}, B=@n319)
  n342 = Multiply(A=Sine(None=DotProduct(A=@n331, B=@n340)), B=@n341)
  n343 = Add(A=Multiply(A=@n339.Speed, B=Time()), B=@n342)
  n344 = Add(A=Sine(None=@n343))
  n345 = Frac(None=LinearInterpolate(A=@n343, B=Multiply(A=@n344) {constB=0.49}, Alpha=@n339."1= Sine Speed"))
  Multiply(A=MaterialFunctionCall!(X range Y range, Gradient Height, ShowGrid?, Thickness, Color, "f(x)"=@n345) {materialFunction=/Engine/Functions/Engine_MaterialFunctions01/Debug/PlotFunctionOnGraph}."f(x) = y")
  n346 = FunctionInput(Preview=Constant4Vector().RGBA) {bUsePreviewValueAsDefault=true, inputName=IsState, inputType=FunctionInput_Vector4, sortPriority=4}
  NamedRerouteDeclaration(None=@n346) {name=IsStateA, nodeColor=<rgba 0.4297 1 0 1>}
  n347 = PreSkinnedNormal()
  NamedRerouteDeclaration(None=FunctionInput(Preview=VertexInterpolator(VS=@n347).PS) {bUsePreviewValueAsDefault=true, inputName=Pre-Skinned Local Normal, sortPriority=10}) {name=PreSkinNormal, nodeColor=<rgba 0 0.1797 1 1>}
  n348 = VectorParameter() {group=Effect_Decal_Reactive_UV_Offset, parameterName=EDRec_DecalD_Location_Weapon, defaultValue=<rgba 0 0 0 0.5>}
  n349 = AppendVector(A=@n348.RGB, B=@n348.RGB)
  n350 = VectorParameter() {group=Effect_Decal_Reactive_UV_Offset, parameterName=EDRec_DecalD_Location_Vehicle, defaultValue=<rgba 0 0 0 5>}
  n351 = VectorParameter() {group=Effect_Decal_Reactive_UV_Offset, parameterName=EDRec_DecalD_Location_VehicleSide, defaultValue=<rgba 0 0 0 5>}
  n352 = MaterialFunctionCall(Float4=NamedRerouteUsage()) {materialFunction=&BreakOutFloat4Components}
  n353 = LinearInterpolate(A=AppendVector(A=@n350.RGB, B=@n350.RGB), B=AppendVector(A=@n351.RGB, B=@n351.RGB), Alpha=@n352.G)
  n354 = VectorParameter() {group=Effect_Decal_Reactive_UV_Offset, parameterName="EDRec_DecalD_Location_Mech   A=Scale", defaultValue=<rgba 0 0 0 0.5>}
  n355 = AppendVector(A=@n354.RGB, B=@n354.RGB)
  n356 = VectorParameter() {group=Effect_Decal_Reactive_UV_Offset, parameterName="EDRec_DecalD_Location_Character   A=Scale", defaultValue=<rgba 0 0 0 0.5>}
  n357 = AppendVector(A=@n356.RGB, B=@n356.RGB)
  NamedRerouteDeclaration(None=LinearInterpolate(A=LinearInterpolate(A=@n349, B=@n353, Alpha=@n352.R), B=LinearInterpolate(A=@n355, B=@n357, Alpha=@n352.A), Alpha=@n352.B)) {name=Offset_D, nodeColor=<rgba 1 0.7266 0 1>}
  n358 = Step(
   Y=ScalarParameter() {defaultValue=0.6, group=Effect_Decal_Reactive_Emissive_A, parameterName=EDRec_Emissive_A_BC_Alpha_Clamp, sortPriority=0}
   X=NamedRerouteUsage()
  )
  NamedRerouteDeclaration(None=@n358) {name=Alpha_MskE A, nodeColor=<rgba 1 0 0.9375 1>}
  n359 = ScalarParameter() {defaultValue=-9999999, group=Effect_Decal_Reactive_Emissive_A_Reactive, parameterName=EDRec_Emissive_A_Brightness_Time_BP, sortPriority=50}
  n360 = ScalarParameter() {group=Effect_Decal_Reactive_Emissive_A_Reactive, parameterName=EDRec_Emissive_A_Brightness_Delay}
  n361 = Divide(
   A=Subtract(A=Time(), B=Add(A=@n359, B=@n360))
   B=ScalarParameter() {defaultValue=0.25, group=Effect_Decal_Reactive_Emissive_A_Reactive, parameterName=EDRec_Emissive_A_Brightness_Duration}
  )
  NamedRerouteDeclaration(None=Power(
    Base=Saturate(None=@n361)
    Exp=ScalarParameter() {defaultValue=1, group=Effect_Decal_Reactive_Emissive_A_Reactive, parameterName=EDRec_Emissive_A_Brightness_Falloff}
   )) {name=Reactive_Brightness A, nodeColor=<rgba 0.8516 0 1 1>}
  n362 = NamedRerouteUsage()
  n363 = ScalarParameter() {defaultValue=1, group=Effect_Decal_Reactive_Emissive_A, parameterName="EDRec_Emissive_A_DrawBehind BC 0=On", sliderMax=1, sortPriority=4}
  n364 = Saturate(None=Add(A=OneMinus(None=@n362), B=@n363))
  NamedRerouteDeclaration(None=@n364) {name=Alpha_MSk_A, nodeColor=<rgba 1 0 0.8203 1>}
  NamedRerouteDeclaration(None=@n94) {name=Add_Counter_Stat, nodeColor=<rgba 0 0.2969 1 1>}
  n365 = NamedRerouteUsage()
  n366 = ComponentMask(None=@n365) {b=true, g=true, r=true}
  n367 = ComponentMask(None=@n365) {a=true}
  n368 = LinearInterpolate(A=@n366, B=@n366, Alpha=@n367)
  n369 = VectorParameter() {group=Effect_Decal_Reactive_Uv_Animation, parameterName=EDRec_DecalA_Rotation_Pivot, sortPriority=50, defaultValue=<rgba 0 0 0 1>}
  n370 = NamedRerouteUsage()
  n371 = NamedRerouteUsage()
  n372 = NamedRerouteUsage()
  n373 = ScalarParameter() {group=Effect_Decal_Reactive_Uv_Offset_LR, parameterName="EDRec_DecalA_Rotation_Mirroring  1=On", sliderMax=1, sortPriority=0}
  n374 = NamedRerouteDeclaration(None=NamedRerouteUsage()) {name=LR_RotationA, nodeColor=<rgba 0.2422 0 1 1>}
  n375 = ComponentMask(None=NamedRerouteUsage()) {a=true}
  n376 = NamedRerouteUsage()
  n377 = Multiply(A=@n375, B=@n376)
  n378 = StaticSwitchParameter(True=@n377, False=@n376) {desc="Use Scale value Alpha is disabled. ", group=Effect_Decal_Reactive_RE_Offset, parameterName=EDRec_Reactive_Offset, sortPriority=0}
  n379 = ComponentMask(None=NamedRerouteUsage()) {a=true}
  n380 = Multiply(A=@n378, B=ComponentMask(None=NamedRerouteUsage()) {a=true})
  n381 = NamedRerouteUsage()
  n382 = StaticSwitchParameter(True=@n377, False=@n375) {desc="Use Scale value Alpha is disabled. ", group=Effect_Decal_Reactive_RE_Offset, parameterName=EDRec_Reactive_Offset, sortPriority=0}
  n383 = StaticBoolParameter() {group=Effect_Decal_Reactive_UV, parameterName=EDRec_Decal_A---SafeZone-Back}
  n384 = StaticSwitch(True=Multiply(A=@n378, B=ComponentMask(None=@n381) {a=true}), False=@n382, Value=@n383)
  n385 = StaticBoolParameter() {group=Effect_Decal_Reactive_UV, parameterName=EDRec_Decal_A----SafeZone-Middle}
  n386 = StaticBoolParameter() {group=Effect_Decal_Reactive_UV, parameterName=EDRec_Decal_A-----SafeZone-Front}
  n387 = VectorParameter() {group=Effect_Decal_Reactive_Uv_Offset_LR, parameterName=EDRec_DecalA_Scale, defaultValue=<rgba 1 1 1 1>}
  n388 = Multiply(A=StaticSwitch(True=Multiply(A=@n378, B=@n379), False=StaticSwitch(True=@n380, False=@n384, Value=@n385), Value=@n386), B=@n387.RGB)
  n389 = LinearInterpolate(A=@n366, Alpha=@n367) {constB=0}
  n390 = NamedRerouteUsage()
  n391 = StaticSwitch(True=NamedRerouteUsage(), False=NamedRerouteUsage(), Value=@n383)
  n392 = ComponentMask(None=StaticSwitch(True=NamedRerouteUsage(), False=StaticSwitch(True=@n390, False=@n391, Value=@n385), Value=@n386)) {b=true, g=true, r=true}
  n393 = NamedRerouteUsage()
  n394 = StaticSwitch(True=@n393, False=Constant() {r=1}, Value=@n383)
  n395 = StaticSwitch(
   True=StaticBool() {value=true}
   False=NamedRerouteUsage()
   Value=StaticBoolParameter() {group=Effect_Decal_Reactive_Uv_Offset_LR, parameterName="EDRec_DecalA--Side---->LR"}
  )
  n396 = StaticSwitch(True=StaticSwitch(True=@n393, False=StaticSwitch(True=@n393, False=@n394, Value=@n385), Value=@n386), False=@n393, Value=@n395)
  n397 = StaticBool() {value=true}
  n398 = MaterialFunctionCall!(
   Normal_Map
   Offset RT=@n368
   Pivot=@n369.RGB
   "Pivot_L\\R"
   Caculated_Bounds=@n370
   Pre_Skinned_Normals=@n371
   Is State=@n372
   Uvs_Rotation_Mirroring=@n373
   Uvs_Rotation=@n374
   Uvs_TileRate=@n388
   Uvs_Offset=Add(A=@n389, B=Multiply(A=@n392, B=@n396))
   Clamp Y=@n397
   Clamp X=@n397
   Clamp Padding_Comp
  ) {materialFunction=&/MF_Projection_Side_Left_Right_Decal_NormalV2}
  n399 = MaterialFunctionCall!(
   Normal_Map
   Offset Rt=@n368
   Pivot=@n369.RGB
   Caculated_Bounds=@n370
   Pre_Skinned_Normals=@n371
   is_State=@n372
   UV_Rotation=@n374
   Uvs_Offset=Add(A=@n389, B=@n392)
   Uvs_TileRate=@n388
   Uvs_Offset_Pivot
   For_Safe_Zones=StaticBool() {value=true}
   Rotation_Comp
   Clamp Y
   Clamp X
   Clamp Padding_Comp
   No_Rotation
  ) {materialFunction=&/MF_Projection_Side_Normal_V2}
  n400 = Saturate(None=@n399.UVs)
  n401 = ScalarParameter() {defaultValue=1, group=Effect_Decal_Reactive_Uv_Offset_LR, parameterName="EDRec_DecalA_Clamp  1=On", sliderMax=1}
  n402 = StaticBoolParameter() {group=Effect_Decal_Reactive_Uv_Offset_LR, parameterName=EDRec_DecalA--Tileable}
  n403 = StaticSwitch(
   True=NamedRerouteUsage()
   False=StaticSwitch(True=@n398.UVs, False=LinearInterpolate(A=@n399.UVs, B=@n400, Alpha=@n401), Value=@n395)
   Value=@n402
  )
  NamedRerouteDeclaration(None=@n403) {name=Uvs_A, nodeColor=<rgba 1 0.1875 0 1>}
  n404 = TextureObjectParameter() {group=Effect_Decal_Reactive, parameterName=EDRec_Decal_Grad__Tex, sortPriority=10, texture=&T_BlacksRGB}
  NamedRerouteDeclaration(None=@n404) {name=Decal_Grad, nodeColor=<rgba 0 1 0.9062 1>}
  NamedRerouteDeclaration(None=@n254) {name=Emissive A State, nodeColor=<rgba 0 0.2734 1 1>}
  NamedRerouteDeclaration(None=@n205.G) {name=Msk_Curvature, nodeColor=<rgba 0 1 0.7422 1>}
  n405 = VectorParameter() {group=Effect_Decal_Reactive_Uv_Offset_LR, parameterName=EDRec_DecalA_Offset_Weapon, defaultValue=<rgba 0 0 0 0.25>}
  n406 = VectorParameter() {group=Effect_Decal_Reactive_Uv_Offset_LR, parameterName=EDRec_DecalA_Offset_Vehicle, defaultValue=<rgba 0 0 0 0.25>}
  n407 = VectorParameter() {group=Effect_Decal_Reactive_Uv_Offset_LR, parameterName=EDRec_DecalA_Offset_VehicleSide, defaultValue=<rgba 0 0 0 0.25>}
  n408 = MaterialFunctionCall(Float4=NamedRerouteUsage()) {materialFunction=&BreakOutFloat4Components}
  n409 = LinearInterpolate(
   A=AppendVector(A=@n405.RGB, B=@n405.RGB)
   B=LinearInterpolate(A=AppendVector(A=@n406.RGB, B=@n406.RGB), B=AppendVector(A=@n407.RGB, B=@n407.RGB), Alpha=@n408.G)
   Alpha=@n408.R
  )
  n410 = VectorParameter() {group=Effect_Decal_Reactive_Uv_Offset_LR, parameterName=EDRec_DecalA_Offset_Mech, defaultValue=<rgba 0 0 0 0.25>}
  n411 = VectorParameter() {group=Effect_Decal_Reactive_Uv_Offset_LR, parameterName=EDRec_DecalA_Offset_Character, defaultValue=<rgba 0 0 0 0.25>}
  n412 = LinearInterpolate(A=AppendVector(A=@n410.RGB, B=@n410.RGB), B=AppendVector(A=@n411.RGB, B=@n411.RGB), Alpha=@n408.A)
  n413 = VectorParameter() {group=Effect_Decal_Reactive_Uv_Offset_LR, parameterName="EDRec_DecalA_Offset_Backpack----A=Scale", defaultValue=<rgba 0 0 0 0.25>}
  n414 = MaterialFunctionCall(Float2=NamedRerouteUsage()) {materialFunction=&BreakOutFloat2Components}
  n415 = VectorParameter() {group=Effect_Decal_Reactive_Uv_Offset_LR, parameterName=EDRec_DecalA_Offset_Pickaxe, defaultValue=<rgba 0 0 0 0.25>}
  n416 = LinearInterpolate(
   A=LinearInterpolate(A=LinearInterpolate(A=@n409, B=@n412, Alpha=@n408.B), B=AppendVector(A=@n413.RGB, B=@n413.RGB), Alpha=@n414.R)
   B=AppendVector(A=@n415.RGB, B=@n415.RGB)
   Alpha=@n414.G
  )
  n417 = VectorParameter() {group=Effect_Decal_Reactive_Uv_Offset_LR, parameterName=EDRec_DecalA_Offset_Weapon End, defaultValue=<rgba 0 0 0 0.25>}
  n418 = VectorParameter() {group=Effect_Decal_Reactive_Uv_Offset_LR, parameterName=EDRec_DecalA_Offset_Vehicle End, defaultValue=<rgba 0 0 0 0.25>}
  n419 = VectorParameter() {group=Effect_Decal_Reactive_Uv_Offset_LR, parameterName=EDRec_DecalA_Offset_VehicleSide End, defaultValue=<rgba 0 0 0 0.25>}
  n420 = MaterialFunctionCall(Float4=NamedRerouteUsage()) {materialFunction=&BreakOutFloat4Components}
  n421 = LinearInterpolate(
   A=AppendVector(A=@n417.RGB, B=@n417.RGB)
   B=LinearInterpolate(A=AppendVector(A=@n418.RGB, B=@n418.RGB), B=AppendVector(A=@n419.RGB, B=@n419.RGB), Alpha=@n420.G)
   Alpha=@n420.R
  )
  n422 = VectorParameter() {group=Effect_Decal_Reactive_Uv_Offset_LR, parameterName=EDRec_DecalA_Offset_Mech End, defaultValue=<rgba 0 0 0 0.25>}
  n423 = VectorParameter() {group=Effect_Decal_Reactive_Uv_Offset_LR, parameterName=EDRec_DecalA_Offset_Character End, defaultValue=<rgba 0 0 0 0.25>}
  n424 = LinearInterpolate(A=AppendVector(A=@n422.RGB, B=@n422.RGB), B=AppendVector(A=@n423.RGB, B=@n423.RGB), Alpha=@n420.A)
  n425 = VectorParameter() {group=Effect_Decal_Reactive_Uv_Offset_LR, parameterName="EDRec_DecalA_Offset_Backpack----A=Scale End", defaultValue=<rgba 0 0 0 0.25>}
  n426 = MaterialFunctionCall(Float2=NamedRerouteUsage()) {materialFunction=&BreakOutFloat2Components}
  n427 = VectorParameter() {group=Effect_Decal_Reactive_Uv_Offset_LR, parameterName=EDRec_DecalA_Offset_Pickaxe End, defaultValue=<rgba 0 0 0 0.25>}
  n428 = LinearInterpolate(
   A=LinearInterpolate(A=LinearInterpolate(A=@n421, B=@n424, Alpha=@n420.B), B=AppendVector(A=@n425.RGB, B=@n425.RGB), Alpha=@n426.R)
   B=AppendVector(A=@n427.RGB, B=@n427.RGB)
   Alpha=@n426.G
  )
  n429 = StaticSwitchParameter(True=LinearInterpolate(A=@n416, B=@n428, Alpha=NamedRerouteUsage()), False=@n416) {desc="Use Scale value Alpha is disabled. ", group=Effect_Decal_Reactive_RE_Offset, parameterName=EDRec_Reactive_Offset, sortPriority=0}
  NamedRerouteDeclaration(None=@n429) {name=UVs_OffsetA, nodeColor=<rgba 0 0.625 1 1>}
  n430 = StaticSwitch(True=@n398.Clamp_Mask, False=@n399.Safe_Zone_Mask, Value=@n395)
  n431 = StaticSwitch(True=@n430, False=Constant(), Value=@n383)
  NamedRerouteDeclaration(None=StaticSwitch(True=@n430, False=StaticSwitch(True=@n430, False=@n431, Value=@n385), Value=@n386)) {name=Decal_A_Msk_SZ, nodeColor=<rgba 0 1 0.1562 1>}
  n432 = NamedRerouteUsage()
  n433 = ComponentMask(None=@n432) {b=true, g=true, r=true}
  n434 = ComponentMask(None=@n432) {a=true}
  n435 = LinearInterpolate(B=@n433, Alpha=@n434)
  n436 = VectorParameter() {group=Effect_Decal_Reactive_Uv_Animation, parameterName=EDRec_DecalB_Rotation_Pivot, sortPriority=50, defaultValue=<rgba 0 0 0 1>}
  n437 = NamedRerouteUsage()
  n438 = NamedRerouteUsage()
  n439 = NamedRerouteUsage()
  n440 = NamedRerouteDeclaration(None=NamedRerouteUsage()) {name=LR_RotationB, nodeColor=<rgba 0.2422 0 1 1>}
  n441 = NamedRerouteUsage()
  n442 = ComponentMask(None=@n441) {a=true}
  n443 = NamedRerouteUsage()
  n444 = Multiply(A=@n442, B=@n443)
  n445 = StaticSwitchParameter(True=@n444, False=@n443) {desc="Use Scale value Alpha is disabled. ", group=Effect_Decal_Reactive_RE_Offset, parameterName=EDRec_Reactive_Offset, sortPriority=0}
  n446 = NamedRerouteUsage()
  n447 = ComponentMask(None=@n446) {a=true}
  n448 = NamedRerouteUsage()
  n449 = Multiply(A=@n445, B=ComponentMask(None=@n448) {a=true})
  n450 = NamedRerouteUsage()
  n451 = StaticSwitchParameter(True=@n444, False=@n442) {desc="Use Scale value Alpha is disabled. ", group=Effect_Decal_Reactive_RE_Offset, parameterName=EDRec_Reactive_Offset, sortPriority=0}
  n452 = StaticBoolParameter() {group=Effect_Decal_Reactive_UV, parameterName=EDRec_Decal_B---SafeZone-Back}
  n453 = StaticSwitch(True=Multiply(A=@n445, B=ComponentMask(None=@n450) {a=true}), False=@n451, Value=@n452)
  n454 = StaticBoolParameter() {group=Effect_Decal_Reactive_UV, parameterName=EDRec_Decal_B----SafeZone-Middle}
  n455 = StaticBoolParameter() {group=Effect_Decal_Reactive_UV, parameterName=EDRec_Decal_B-----SafeZone-Front}
  n456 = VectorParameter() {group=Effect_Decal_Reactive_Uv_Offset_LR, parameterName=EDRec_DecalB_Scale, defaultValue=<rgba 1 1 1 1>}
  n457 = Multiply(A=StaticSwitch(True=Multiply(A=@n445, B=@n447), False=StaticSwitch(True=@n449, False=@n453, Value=@n454), Value=@n455), B=@n456.RGB)
  n458 = LinearInterpolate(A=@n433, Alpha=@n434) {constB=0}
  n459 = StaticSwitch(True=@n450, False=@n441, Value=@n452)
  n460 = ComponentMask(None=StaticSwitch(True=@n446, False=StaticSwitch(True=@n448, False=@n459, Value=@n454), Value=@n455)) {b=true, g=true, r=true}
  n461 = NamedRerouteUsage()
  n462 = Constant() {r=1}
  n463 = StaticSwitch(True=@n461, False=StaticSwitch(True=@n461, False=StaticSwitch(True=@n461, False=@n462, Value=@n452), Value=@n454), Value=@n455)
  n464 = StaticBool() {value=true}
  n465 = MaterialFunctionCall(
   Offset RT=@n435
   Pivot=@n436.RGB
   Caculated_Bounds=@n437
   Pre_Skinned_Normals=@n438
   Is State=@n439
   Uvs_Rotation=@n440
   Uvs_TileRate=@n457
   Uvs_Offset=Add(A=@n458, B=Multiply(A=@n460, B=@n463))
   Clamp Y=@n464
   Clamp X=@n464
  ) {materialFunction=&MF_Projection_Side_Left_Right_Decal_NormalV2}
  n466 = MaterialFunctionCall(
   Offset Rt=@n435
   Pivot=@n436.RGB
   Caculated_Bounds=@n437
   Pre_Skinned_Normals=@n438
   is_State=@n439
   UV_Rotation=@n440
   Uvs_Offset=Add(A=@n458, B=@n460)
   Uvs_TileRate=@n457
   For_Safe_Zones=StaticBool() {value=true}
  ) {materialFunction=&MF_Projection_Side_Normal_V2}
  n467 = StaticSwitch(
   True=StaticBool() {value=true}
   False=NamedRerouteUsage()
   Value=StaticBoolParameter() {group=Effect_Decal_Reactive_Uv_Offset_LR, parameterName="EDRec_DecalB--Side---->LR"}
  )
  n468 = StaticSwitch(True=@n465.Clamp_Mask, False=@n466.Safe_Zone_Mask, Value=@n467)
  n469 = StaticSwitch(True=@n468, False=Constant(), Value=@n452)
  NamedRerouteDeclaration(None=StaticSwitch(True=@n468, False=StaticSwitch(True=@n468, False=@n469, Value=@n454), Value=@n455)) {name=Decal_B_Msk_SZ, nodeColor=<rgba 0 1 0.1562 1>}
  n470 = Constant()
  n471 = NamedRerouteUsage()
  n472 = ComponentMask(None=@n471) {b=true, g=true, r=true}
  n473 = ComponentMask(None=@n471) {a=true}
  n474 = LinearInterpolate(B=@n472, Alpha=@n473)
  n475 = VectorParameter() {group=Effect_Decal_Reactive_Uv_Animation, parameterName=EDRec_DecalC_Rotation_Pivot, sortPriority=50, defaultValue=<rgba 0 0 0 1>}
  n476 = NamedRerouteUsage()
  n477 = NamedRerouteUsage()
  n478 = NamedRerouteUsage()
  n479 = NamedRerouteDeclaration(None=NamedRerouteUsage()) {name=LR_RotationC, nodeColor=<rgba 0.2422 0 1 1>}
  n480 = NamedRerouteUsage()
  n481 = ComponentMask(None=@n480) {a=true}
  n482 = NamedRerouteUsage()
  n483 = Multiply(A=@n481, B=@n482)
  n484 = StaticSwitchParameter(True=@n483, False=@n482) {desc="Use Scale value Alpha is disabled. ", group=Effect_Decal_Reactive_RE_Offset, parameterName=EDRec_Reactive_Offset, sortPriority=0}
  n485 = NamedRerouteUsage()
  n486 = ComponentMask(None=@n485) {a=true}
  n487 = NamedRerouteUsage()
  n488 = Multiply(A=@n484, B=ComponentMask(None=@n487) {a=true})
  n489 = NamedRerouteUsage()
  n490 = StaticSwitchParameter(True=@n483, False=@n481) {desc="Use Scale value Alpha is disabled. ", group=Effect_Decal_Reactive_RE_Offset, parameterName=EDRec_Reactive_Offset, sortPriority=0}
  n491 = StaticBoolParameter() {group=Effect_Decal_Reactive_UV, parameterName=EDRec_Decal_C---SafeZone-Back}
  n492 = StaticSwitch(True=Multiply(A=@n484, B=ComponentMask(None=@n489) {a=true}), False=@n490, Value=@n491)
  n493 = StaticBoolParameter() {group=Effect_Decal_Reactive_UV, parameterName=EDRec_Decal_C----SafeZone-Middle}
  n494 = StaticBoolParameter() {group=Effect_Decal_Reactive_UV, parameterName=EDRec_Decal_C-----SafeZone-Front}
  n495 = VectorParameter() {group=Effect_Decal_Reactive_Uv_Offset_LR, parameterName=EDRec_DecalC_Scale, defaultValue=<rgba 1 1 1 1>}
  n496 = Multiply(A=StaticSwitch(True=Multiply(A=@n484, B=@n486), False=StaticSwitch(True=@n488, False=@n492, Value=@n493), Value=@n494), B=@n495.RGB)
  n497 = LinearInterpolate(A=@n472, Alpha=@n473) {constB=0}
  n498 = StaticSwitch(True=@n489, False=@n480, Value=@n491)
  n499 = ComponentMask(None=StaticSwitch(True=@n485, False=StaticSwitch(True=@n487, False=@n498, Value=@n493), Value=@n494)) {b=true, g=true, r=true}
  n500 = NamedRerouteUsage()
  n501 = Constant() {r=1}
  n502 = StaticSwitch(True=@n500, False=StaticSwitch(True=@n500, False=StaticSwitch(True=@n500, False=@n501, Value=@n491), Value=@n493), Value=@n494)
  n503 = StaticBool() {value=true}
  n504 = MaterialFunctionCall(
   Offset RT=@n474
   Pivot=@n475.RGB
   Caculated_Bounds=@n476
   Pre_Skinned_Normals=@n477
   Is State=@n478
   Uvs_Rotation=@n479
   Uvs_TileRate=@n496
   Uvs_Offset=Add(A=@n497, B=Multiply(A=@n499, B=@n502))
   Clamp Y=@n503
   Clamp X=@n503
  ) {materialFunction=&MF_Projection_Side_Left_Right_Decal_NormalV2}
  n505 = MaterialFunctionCall(
   Offset Rt=@n474
   Pivot=@n475.RGB
   Caculated_Bounds=@n476
   Pre_Skinned_Normals=@n477
   is_State=@n478
   UV_Rotation=@n479
   Uvs_Offset=Add(A=@n497, B=@n499)
   Uvs_TileRate=@n496
   For_Safe_Zones=StaticBool() {value=true}
  ) {materialFunction=&MF_Projection_Side_Normal_V2}
  n506 = StaticSwitch(
   True=StaticBool() {value=true}
   False=NamedRerouteUsage()
   Value=StaticBoolParameter() {group=Effect_Decal_Reactive_Uv_Offset_LR, parameterName="EDRec_DecalC--Side---->LR"}
  )
  n507 = StaticSwitch(True=@n504.Clamp_Mask, False=@n505.Safe_Zone_Mask, Value=@n506)
  n508 = StaticSwitch(True=@n507, False=StaticSwitch(True=@n507, False=@n470, Value=@n491), Value=@n493)
  n509 = StaticBoolParameter() {group=Effect_Decal_Reactive, parameterName=EDRec_Decal_C-----Disable, sortPriority=-10}
  NamedRerouteDeclaration(None=StaticSwitch(True=@n470, False=StaticSwitch(True=@n507, False=@n508, Value=@n494), Value=@n509)) {name=Decal_C_Msk_SZ, nodeColor=<rgba 0 1 0.1562 1>}
  n510 = Constant()
  n511 = NamedRerouteUsage()
  n512 = ComponentMask(None=@n511) {b=true, g=true, r=true}
  n513 = ComponentMask(None=@n511) {a=true}
  n514 = LinearInterpolate(B=@n512, Alpha=@n513)
  n515 = VectorParameter() {group=Effect_Decal_Reactive_Uv_Animation, parameterName=EDRec_DecalD_Rotation_Pivot, sortPriority=50, defaultValue=<rgba 0 0 0 1>}
  n516 = NamedRerouteUsage()
  n517 = NamedRerouteUsage()
  n518 = NamedRerouteUsage()
  n519 = NamedRerouteDeclaration(None=NamedRerouteUsage()) {name=LR_RotationD, nodeColor=<rgba 0.2422 0 1 1>}
  n520 = NamedRerouteUsage()
  n521 = ComponentMask(None=@n520) {a=true}
  n522 = NamedRerouteUsage()
  n523 = Multiply(A=@n521, B=@n522)
  n524 = StaticSwitchParameter(True=@n523, False=@n522) {desc="Use Scale value Alpha is disabled. ", group=Effect_Decal_Reactive_RE_Offset, parameterName=EDRec_Reactive_Offset, sortPriority=0}
  n525 = NamedRerouteUsage()
  n526 = ComponentMask(None=@n525) {a=true}
  n527 = NamedRerouteUsage()
  n528 = Multiply(A=@n524, B=ComponentMask(None=@n527) {a=true})
  n529 = NamedRerouteUsage()
  n530 = StaticSwitchParameter(True=@n523, False=@n521) {desc="Use Scale value Alpha is disabled. ", group=Effect_Decal_Reactive_RE_Offset, parameterName=EDRec_Reactive_Offset, sortPriority=0}
  n531 = StaticBoolParameter() {group=Effect_Decal_Reactive_UV, parameterName=EDRec_Decal_D---SafeZone-Back}
  n532 = StaticSwitch(True=Multiply(A=@n524, B=ComponentMask(None=@n529) {a=true}), False=@n530, Value=@n531)
  n533 = StaticBoolParameter() {group=Effect_Decal_Reactive_UV, parameterName=EDRec_Decal_D----SafeZone-Middle}
  n534 = StaticBoolParameter() {group=Effect_Decal_Reactive_UV, parameterName=EDRec_Decal_D-----SafeZone-Front}
  n535 = VectorParameter() {group=Effect_Decal_Reactive_Uv_Offset_LR, parameterName=EDRec_DecalD_Scale, defaultValue=<rgba 1 1 1 1>}
  n536 = Multiply(A=StaticSwitch(True=Multiply(A=@n524, B=@n526), False=StaticSwitch(True=@n528, False=@n532, Value=@n533), Value=@n534), B=@n535.RGB)
  n537 = LinearInterpolate(A=@n512, Alpha=@n513) {constB=0}
  n538 = StaticSwitch(True=@n529, False=@n520, Value=@n531)
  n539 = ComponentMask(None=StaticSwitch(True=@n525, False=StaticSwitch(True=@n527, False=@n538, Value=@n533), Value=@n534)) {b=true, g=true, r=true}
  n540 = NamedRerouteUsage()
  n541 = Constant() {r=1}
  n542 = StaticSwitch(True=@n540, False=StaticSwitch(True=@n540, False=StaticSwitch(True=@n540, False=@n541, Value=@n531), Value=@n533), Value=@n534)
  n543 = StaticBool() {value=true}
  n544 = MaterialFunctionCall(
   Offset RT=@n514
   Pivot=@n515.RGB
   Caculated_Bounds=@n516
   Pre_Skinned_Normals=@n517
   Is State=@n518
   Uvs_Rotation=@n519
   Uvs_TileRate=@n536
   Uvs_Offset=Add(A=@n537, B=Multiply(A=@n539, B=@n542))
   Clamp Y=@n543
   Clamp X=@n543
  ) {materialFunction=&MF_Projection_Side_Left_Right_Decal_NormalV2}
  n545 = MaterialFunctionCall(
   Offset Rt=@n514
   Pivot=@n515.RGB
   Caculated_Bounds=@n516
   Pre_Skinned_Normals=@n517
   is_State=@n518
   UV_Rotation=@n519
   Uvs_Offset=Add(A=@n537, B=@n539)
   Uvs_TileRate=@n536
   For_Safe_Zones=StaticBool() {value=true}
  ) {materialFunction=&MF_Projection_Side_Normal_V2}
  n546 = StaticSwitch(
   True=StaticBool() {value=true}
   False=NamedRerouteUsage()
   Value=StaticBoolParameter() {group=Effect_Decal_Reactive_Uv_Offset_LR, parameterName="EDRec_DecalD--Side---->LR"}
  )
  n547 = StaticSwitch(True=@n544.Clamp_Mask, False=@n545.Safe_Zone_Mask, Value=@n546)
  n548 = StaticSwitch(True=@n547, False=StaticSwitch(True=@n547, False=@n510, Value=@n531), Value=@n533)
  n549 = StaticBoolParameter() {group=Effect_Decal_Reactive, parameterName=EDRec_Decal_D-----Disable, sortPriority=-10}
  NamedRerouteDeclaration(None=StaticSwitch(True=@n510, False=StaticSwitch(True=@n547, False=@n548, Value=@n534), Value=@n549)) {name=Decal_D_Msk_SZ, nodeColor=<rgba 0 1 0.1562 1>}
  NamedRerouteDeclaration(None=@n399.UVs No Offsets/Tiling) {name=Uvs Raw, nodeColor=<rgba 0 1 0.9297 1>}
  n550 = LinearInterpolate(
   A=ScalarParameter() {defaultValue=1, group=Effect_Decal_Reactive_Emissive_B_Reactive, parameterName=EDRec_Emissive_B_Brightness_Time__Start}
   B=ScalarParameter() {group=Effect_Decal_Reactive_Emissive_B_Reactive, parameterName=EDRec_Emissive_B_Brightness_Time__End}
   Alpha=NamedRerouteUsage()
  )
  n551 = ScalarParameter() {group=Effect_Decal_Reactive_Emissive_B_Reactive, parameterName=EDRec-Emissive-B_Reactive_Audio__Min}
  n552 = ScalarParameter() {group=Effect_Decal_Reactive_Emissive_B_Reactive, parameterName=EDRec-Emissive-B_Reactive_Audio__Max}
  n553 = NamedRerouteUsage()
  n554 = StaticBoolParameter() {group=Effect_Decal_Reactive_Emissive_B_Reactive, parameterName=EDRec-Emissive-B_Reactive_Audio---------------------}
  n555 = StaticSwitch(True=Max(A=@n550, B=LinearInterpolate(A=@n551, B=@n552, Alpha=@n553)), False=@n550, Value=@n554)
  n556 = Saturate(None=Add(
    A=@n139
    B=ScalarParameter() {defaultValue=1, group=Effect_Decal_Reactive_Emissive_B, parameterName="EDRec_Emissive_B_CC_Mask   0=ON", sliderMax=1}
   ))
  n557 = VectorParameter() {group=Effect_Decal_Reactive_Emissive_B, parameterName=EDRec_Emissive_B_CC Tint, defaultValue=<rgba 1 1 1 1>}
  n558 = ScalarParameter() {group=Effect_Decal_Reactive_Emissive_B, parameterName=EDRec_Emissive_B_CC-Animated}
  n559 = Time()
  n560 = StaticSwitchParameter(True=Add(A=@n111, B=ComponentMask(None=@n108) {r=true}), False=@n127) {group=Effect_Decal_Reactive_Emissive_B, parameterName=EDRec_Emissive_B_Warp_ Msk}
  n561 = Multiply(
   A=ScalarParameter() {defaultValue=1, group=Effect_Decal_Reactive_Emissive_B, parameterName=EDRec_Emissive_B_CC_TileRate}
   B=LinearInterpolate(
    A=@n139
    B=StaticSwitchParameter(True=@n127, False=@n560) {group=Effect_Decal_Reactive_Emissive_B, parameterName="EDRec_Emissive_B_Gradent ---> use Texture"}
    Alpha=ScalarParameter() {group=Effect_Decal_Reactive_Emissive_B, parameterName=EDRec_Emissive_B_CC_Remap--Grad}
   )
  )
  n562 = ScalarParameter() {group=Effect_Decal_Reactive_Emissive_B, parameterName=EDRec_Emissive_B_CC_Offset}
  n563 = StaticSwitch(True=Add(A=NamedRerouteUsage(), B=@n562), False=@n562, Value=NamedRerouteUsage())
  n564 = Frac(None=Add(A=Multiply(A=@n558, B=@n559), B=Add(A=@n561, B=@n563)))
  n565 = CurveAtlasRowParameter(CurveTime=@n564) {
   atlas=/Game/Weapons/WeaponSkins/Wraps/Textures/CruveAtlas_Wraps_01
   curve=/Game/Weapons/WeaponSkins/Wraps/Textures/C_GradSimple
   group=Effect_Decal_Reactive_Emissive_B
   parameterName=EDRec_Emissive_B_CC
  }
  n566 = MaterialFunctionCall!(
   Index=ScalarParameter() {group=Effect_Decal_Reactive_Emissive_B, parameterName="EDRec_Emissive_B_CC  Index"}
   CCinput=@n565
   CCinput_Alpha=@n565
   UV=@n564
  ) {materialFunction=&/MF_Global_CC_Tex}
  n567 = Multiply(A=Multiply(A=@n557.RGB, B=@n566.3V), B=@n566.Alpha)
  n568 = VectorParameter() {group=Effect_Decal_Reactive_Emissive_B, parameterName=EDRec_Emissive_B_Color-Start, defaultValue=<rgba 1 1 1 1>}
  n569 = VectorParameter() {group=Effect_Decal_Reactive_Emissive_B, parameterName=EDRec_Emissive_B_Color-, defaultValue=<rgba 1 1 1 1>}
  n570 = StaticSwitchParameter(True=LinearInterpolate(A=@n568.RGB, B=@n569.RGB, Alpha=NamedRerouteUsage()), False=@n569.RGB) {group=Effect_Decal_Reactive_Emissive_B_Reactive, parameterName=EDRec-Emissive-B_Reactive_Color---------------------}
  n571 = Multiply(
   A=StaticSwitchParameter(True=Multiply(A=@n556, B=@n567), False=Multiply(A=@n139, B=@n570)) {group=Effect_Decal_Reactive_Emissive_B, parameterName="EDRec-Emissive-B-Color--->CCurve"}
   B=NamedRerouteUsage()
  )
  n572 = Multiply(A=@n571, B=@n555)
  n573 = MaterialFunctionCall(
   In=Max(
    A=NamedRerouteUsage()
    B=ScalarParameter() {defaultValue=1, group=Effect_Decal_Reactive_Emissive_B, parameterName="EDRec-Emissive-B_TriMak   0=Active", sortPriority=0}
   )
   Contrast=ScalarParameter() {group=Effect_Decal_Reactive_Emissive_B, parameterName="EDRec-Emissive-B_TriMak  Contrast", sortPriority=0}
  ) {materialFunction=&CheapContrast}
  n574 = Multiply(
   A=Multiply(
    A=StaticSwitchParameter(True=@n572, False=@n571) {group=Effect_Decal_Reactive_Emissive_B_Reactive, parameterName=EDRec-Emissive-B_Reactive_Brightness---------------------}
    B=NamedRerouteUsage()
   )
   B=@n573.Result
  )
  n575 = NamedRerouteUsage()
  n576 = NamedRerouteUsage()
  n577 = StaticBoolParameter() {group=Effect_Decal_Reactive_Emissive_B_Reactive, parameterName=EDRec-Emissive-B-TOD, sortPriority=0}
  n578 = StaticSwitch(
   True=LinearInterpolate(
    B=@n574
    Alpha=StaticSwitchParameter(True=@n575, False=@n576) {group=Effect_Decal_Reactive_RE, parameterName="EDRec_DecalB_Opacity-TOD----->TODB", sortPriority=-1}
   ) {constB=0}
   False=@n574
   Value=@n577
  )
  n579 = NamedRerouteUsage()
  n580 = LinearInterpolate(
   A=@n579
   B=OneMinus(None=@n579)
   Alpha=ScalarParameter() {group=Effect_Decal_Reactive_Emissive_B_Reactive, parameterName="EDRec-Emissive-B_TopS Blend 01<--->02 MSk   1=invert"}
  )
  n581 = LinearInterpolate(B=@n578, Alpha=@n580)
  n582 = StaticBoolParameter() {group=Effect_Decal_Reactive_Emissive_B_Reactive, parameterName="EDRec-Emissive-B_TopS Blend 01<--->02 MSk"}
  n583 = StaticBoolParameter() {group="TopSide_V02_Layer 01 Blend  02", parameterName="TopS Blend 01<--->02"}
  NamedRerouteDeclaration(None=StaticSwitch(True=StaticSwitch(True=@n581, False=@n578, Value=@n582), False=@n578, Value=@n583)) {name=Emissive_B, nodeColor=<rgba 0.2656 1 0 1>}
  n584 = Constant() {r=1}
  n585 = Step(
   Y=ScalarParameter() {defaultValue=0.6, group=Effect_Decal_Reactive_Emissive_B, parameterName=EDRec_Emissive_B_BC_Alpha_Clamp, sortPriority=0}
   X=NamedRerouteUsage()
  )
  n586 = StaticSwitchParameter(
   True=@n584
   False=StaticSwitchParameter(True=@n584, False=@n585) {group=Effect_Decal_Reactive_Emissive_B, parameterName=EDRec-Emissive-B_Height___X+X, sortPriority=4}
  ) {group=Effect_Decal_Reactive_Emissive_B, parameterName=EDRec-Emissive-B_Height___Lerp, sortPriority=4}
  NamedRerouteDeclaration(None=Multiply(
    A=NamedRerouteUsage()
    B=StaticSwitchParameter(True=@n584, False=@n586) {group=Effect_Decal_Reactive_Emissive_B, parameterName=EDRec-Emissive-B_Only_Mask, sortPriority=4}
   )) {name=Emissive_Msk_Simple_B, nodeColor=<rgba 0 1 0.1094 1>}
  n587 = Constant() {r=0.5}
  n588 = VectorParameter() {group=Effect_Decal_Reactive_Emissive_B, parameterName=EDRec_Emissive_B_Msk_Offset, sortPriority=5}
  n589 = MaterialFunctionCall(Float2=Add(A=ComponentMask(None=@n588.RGB) {g=true, r=true}, B=@n115.Rotated Values) {constB=0}) {materialFunction=&BreakOutFloat2Components}
  n590 = VectorParameter() {
   group=Effect_Decal_Reactive_Emissive_B
   parameterName=EDRec_Emissive_B_Decal_Msking_Scale
   sortPriority=5
   channelNames {a=Y Falloff, b=X Falloff, g=Y Size, r=X Size}
   defaultValue=<rgba 0.5 0.5 1 1>
  }
  n591 = SphereMask(A=@n587, B=@n589.R, Radius=@n590.Y Size, Hardness=@n590.Y Size) {attenuationRadius=0.5}
  n592 = SphereMask(A=@n589.G, B=@n587, Radius=@n590.X Size, Hardness=@n590.X Size) {attenuationRadius=0.5}
  n593 = StaticSwitchParameter(True=Power(Base=Multiply(A=@n591, B=@n592)), False=Constant() {r=1}) {
   desc=second layer of bounds that is used to msk
   group=Effect_Decal_Reactive_Emissive_B
   parameterName=EDRec_Emissive_B_Decal_Msking
   sortPriority=5
  }
  NamedRerouteDeclaration(None=@n593) {name=Decal_Msking_B, nodeColor=<rgba 0 1 0.5078 1>}
  NamedRerouteDeclaration(None=@n346) {name=IsState, nodeColor=<rgba 0.1497 1 0.1633 1>}
  n594 = LinearInterpolate(
   A=ScalarParameter() {group=Effect_Decal_Reactive_Emissive_B_Reactive, parameterName=EDRec-Emissive-B_Reactive_Audio_Darken__Min}
   B=ScalarParameter() {group=Effect_Decal_Reactive_Emissive_B_Reactive, parameterName=EDRec-Emissive-B_Reactive_Audio_Darken__Max}
   Alpha=NamedRerouteUsage()
  )
  n595 = LinearInterpolate(
   A=ScalarParameter() {group=Effect_Decal_Reactive_Emissive_B_Reactive, parameterName=EDRec_Emissive_B_Brightness_DarkenBase_Time_Start}
   B=ScalarParameter() {group=Effect_Decal_Reactive_Emissive_B_Reactive, parameterName=EDRec_Emissive_B_Brightness_DarkenBase_Time_End}
   Alpha=NamedRerouteUsage()
  )
  n596 = StaticSwitchParameter(
   True=StaticSwitch(True=Max(A=@n594, B=@n595), False=@n595, Value=NamedRerouteUsage())
   False=ScalarParameter() {group=Effect_Decal_Reactive_Emissive_B, parameterName=EDRec_Emissive_B_DarkenBase-------, sortPriority=4}
  ) {group=Effect_Decal_Reactive_Emissive_B_Reactive, parameterName=EDRec-Emissive-B_Reactive_Brightness---------------------}
  n597 = NamedRerouteUsage()
  n598 = Multiply(A=Multiply(A=@n596, B=@n139), B=@n597)
  n599 = StaticSwitchParameter(True=@n129, False=Constant() {r=1}) {group=Effect_Decal_Reactive_Emissive_B, parameterName=EDRec-Emissive-B_Height___X+X, sortPriority=4}
  n600 = Multiply(A=@n596, B=Saturate(None=@n96))
  n601 = Multiply(
   A=StaticSwitchParameter(
    True=@n129
    False=StaticSwitchParameter(True=@n129, False=@n599) {group=Effect_Decal_Reactive_Emissive_B, parameterName=EDRec-Emissive-B_Height___Lerp, sortPriority=4}
   ) {group=Effect_Decal_Reactive_Emissive_B, parameterName=EDRec-Emissive-B_Only_Mask, sortPriority=4}
   B=@n600
  )
  n602 = ScalarParameter() {group=Effect_Decal_Reactive_Emissive_B, parameterName=EDRec_Emissive_B_DarkenBase-CCAlpha MskSource, sliderMax=1, sortPriority=4}
  n603 = ScalarParameter() {group=Effect_Decal_Reactive_Emissive_B, parameterName="EDRec_Emissive_B_DarkenBase-CCAlpha 1=Use", sliderMax=1, sortPriority=4}
  n604 = StaticSwitchParameter(True=LinearInterpolate(A=@n598, B=LinearInterpolate(A=@n601, B=@n600, Alpha=@n602), Alpha=Ceil(None=@n603)), False=@n598) {group=Effect_Decal_Reactive_Emissive_B, parameterName="EDRec-Emissive-B-Color--->CCurve"}
  n605 = Saturate(None=Multiply(A=@n604, B=NamedRerouteUsage()))
  n606 = NamedRerouteUsage()
  n607 = ScalarParameter() {defaultValue=1, group=Effect_Decal_Reactive_Emissive_B, parameterName="EDRec_Emissive_B_DrawBehind BC 0=On", sliderMax=1, sortPriority=4}
  n608 = Saturate(None=Add(A=OneMinus(None=@n606), B=@n607))
  n609 = LinearInterpolate(B=Max(A=OneMinus(None=@n573.Result), B=OneMinus(None=@n605)), Alpha=@n608) {constA=1}
  n610 = NamedRerouteUsage()
  n611 = NamedRerouteUsage()
  n612 = StaticSwitch(
   True=LinearInterpolate(
    B=@n609
    Alpha=StaticSwitchParameter(True=@n610, False=@n611) {group=Effect_Decal_Reactive_RE, parameterName="EDRec_DecalB_Opacity-TOD----->TODB", sortPriority=-1}
   ) {constA=1, constB=0}
   False=@n609
   Value=@n577
  )
  n613 = LinearInterpolate(B=@n612, Alpha=@n580) {constA=1}
  NamedRerouteDeclaration(None=StaticSwitch(True=StaticSwitch(True=@n613, False=@n612, Value=@n582), False=@n612, Value=@n583)) {name=Emissive_Darken_Base B, nodeColor=<rgba 0 0.1328 1 1>}
  n614 = ScalarParameter() {
   defaultValue=-9999999
   group=Effect_Decal_Reactive_Emissive_B_Reactive |Gradent
   parameterName=EDRec_Emissive_B_Gradent_Time_BP
   sortPriority=212
  }
  n615 = ScalarParameter() {group=Effect_Decal_Reactive_Emissive_B_Reactive |Gradent, parameterName=EDRec_Emissive_B_Gradent_Delay, sortPriority=211}
  n616 = ScalarParameter() {
   defaultValue=0.25
   group=Effect_Decal_Reactive_Emissive_B_Reactive |Gradent
   parameterName=EDRec_Emissive_B_Gradent_Duration
   sortPriority=211
  }
  n617 = Divide(A=Subtract(A=Time(), B=Add(A=@n614, B=@n615)), B=@n616)
  n618 = ScalarParameter() {defaultValue=1, group=Effect_Decal_Reactive_Emissive_B_Reactive |Gradent, parameterName=EDRec_Emissive_B_Gradent_Falloff_, sortPriority=211}
  NamedRerouteDeclaration(None=Power(Base=Saturate(None=@n617), Exp=@n618)) {name=reactive_Gradent_B, nodeColor=<rgba 0 0.5078 1 1>}
  NamedRerouteDeclaration(None=@n585) {name=Alpha_MskEB, nodeColor=<rgba 1 0 0.9375 1>}
  n619 = ScalarParameter() {defaultValue=-9999999, group=Effect_Decal_Reactive_Emissive_B_Reactive, parameterName=EDRec_Emissive_B_Brightness_Time_BP, sortPriority=50}
  n620 = ScalarParameter() {group=Effect_Decal_Reactive_Emissive_B_Reactive, parameterName=EDRec_Emissive_B_Brightness_Delay}
  n621 = Divide(
   A=Subtract(A=Time(), B=Add(A=@n619, B=@n620))
   B=ScalarParameter() {defaultValue=0.25, group=Effect_Decal_Reactive_Emissive_B_Reactive, parameterName=EDRec_Emissive_B_Brightness_Duration}
  )
  NamedRerouteDeclaration(None=Power(
    Base=Saturate(None=@n621)
    Exp=ScalarParameter() {defaultValue=1, group=Effect_Decal_Reactive_Emissive_B_Reactive, parameterName=EDRec_Emissive_B_Brightness_Falloff}
   )) {name=Reactive_Brightness_B, nodeColor=<rgba 0.8516 0 1 1>}
  NamedRerouteDeclaration(None=@n608) {name=Alpha_MSk_B_1, nodeColor=<rgba 1 0 0.8203 1>}
  n622 = NamedRerouteUsage()
  NamedRerouteDeclaration(None=LinearInterpolate(
    A=Constant() {r=1}
    B=Constant3Vector() {constant=<rgba -1 1 1 1>}
    Alpha=MaterialFunctionCall(Float4=@n622) {materialFunction=&BreakOutFloat4Components}.R
   )) {name=SZ_Offset_Comp, nodeColor=<rgba 0.8047 1 0 1>}
  NamedRerouteDeclaration(None=@n217) {name="Msk_Prime\\Accent", nodeColor=<rgba 0 0.6719 1 1>}
  n623 = Saturate(None=@n466.UVs)
  n624 = ScalarParameter() {defaultValue=1, group=Effect_Decal_Reactive_Uv_Offset_LR, parameterName="EDRec_DecalB_Clamp  1=On", sliderMax=1}
  n625 = StaticBoolParameter() {group=Effect_Decal_Reactive_Uv_Offset_LR, parameterName=EDRec_DecalB--Tileable}
  n626 = StaticSwitch(
   True=NamedRerouteUsage()
   False=StaticSwitch(True=@n465.UVs, False=LinearInterpolate(A=@n466.UVs, B=@n623, Alpha=@n624), Value=@n467)
   Value=@n625
  )
  NamedRerouteDeclaration(None=@n626) {name=Uvs_B, nodeColor=<rgba 1 0.1875 0 1>}
  n627 = StaticBoolParameter() {group=Effect_Decal_Reactive_Uv_Offset_LR, parameterName=EDRec_DecalC--Tileable}
  n628 = StaticSwitch(True=NamedRerouteUsage(), False=StaticSwitch(True=@n504.UVs, False=Saturate(None=@n505.UVs), Value=@n506), Value=@n627)
  NamedRerouteDeclaration(None=@n628) {name=Uvs_C, nodeColor=<rgba 1 0.1875 0 1>}
  NamedRerouteDeclaration(None=FunctionInput(Preview=Constant2Vector()) {bUsePreviewValueAsDefault=true, inputName=IsState_B, inputType=FunctionInput_Vector2, sortPriority=4}) {name=IsState_B, nodeColor=<rgba 0.1497 1 0.1633 1>}
  n629 = StaticBoolParameter() {group=Effect_Decal_Reactive_Uv_Offset_LR, parameterName=EDRec_DecalD--Tileable}
  n630 = StaticSwitch(True=NamedRerouteUsage(), False=StaticSwitch(True=@n544.UVs, False=Saturate(None=@n545.UVs), Value=@n546), Value=@n629)
  NamedRerouteDeclaration(None=@n630) {name=Uvs_D, nodeColor=<rgba 1 0.1875 0 1>}
  n631 = VectorParameter() {
   group=Effect_Decal_Reactive_UV_Warp
   parameterName=EDRec_WarpB__tile_speed
   channelNames {a=Speed-y, b=speed-X, g=Tile-Y, r=Tile-X}
   defaultValue=<rgba 1 1 0 0>
  }
  n632 = AppendVector(A=@n631.Tile-X, B=@n631.Tile-X)
  n633 = AppendVector(A=@n631.speed-X, B=@n631.speed-X)
  n634 = Time()
  n635 = TextureSample(UVs=Add(A=Multiply(A=@n626, B=@n632), B=Multiply(A=@n633, B=@n634)), Tex=NamedRerouteUsage())
  NamedRerouteDeclaration(None=@n635.RGBA) {name=Msk_B_Warp, nodeColor=<rgba 0.1484 0 1 1>}
  n636 = LinearInterpolate(
   A=NamedRerouteUsage()
   B=NamedRerouteUsage()
   Alpha=ScalarParameter() {group=Effect_Decal_Reactive_UV_Warp, parameterName="EDRec-WarpA-UV-->Object"}
  )
  n637 = VectorParameter() {
   group=Effect_Decal_Reactive_UV_Warp
   parameterName=EDRec_WarpA__tile_speed
   channelNames {a=Speed-y, b=speed-X, g=Tile-Y, r=Tile-X}
   defaultValue=<rgba 1 1 0 0>
  }
  n638 = AppendVector(A=@n637.Tile-X, B=@n637.Tile-X)
  n639 = AppendVector(A=@n637.speed-X, B=@n637.speed-X)
  n640 = Time()
  n641 = TextureSample(UVs=Add(A=Multiply(A=@n636, B=@n638), B=Multiply(A=@n639, B=@n640)), Tex=NamedRerouteUsage())
  NamedRerouteDeclaration(None=@n641.RGBA) {name=Msk_A_Warp, nodeColor=<rgba 0.1484 0 1 1>}
  n642 = VectorParameter() {
   group=Effect_Decal_Reactive_UV_Warp
   parameterName=EDRec_WarpC__Tile_speed
   channelNames {a=Speed-y, b=speed-X, g=Tile-Y, r=Tile-X}
   defaultValue=<rgba 1 1 0 0>
  }
  n643 = AppendVector(A=@n642.Tile-X, B=@n642.Tile-X)
  n644 = AppendVector(A=@n642.speed-X, B=@n642.speed-X)
  n645 = Time()
  n646 = TextureSample(UVs=Add(A=Multiply(A=@n628, B=@n643), B=Multiply(A=@n644, B=@n645)), Tex=NamedRerouteUsage())
  NamedRerouteDeclaration(None=@n646.RGBA) {name=Msk_C_Warp, nodeColor=<rgba 0.1484 0 1 1>}
  n647 = VectorParameter() {
   group=Effect_Decal_Reactive_UV_Warp
   parameterName=EDRec_WarpD__tile_speed
   channelNames {a=Speed-y, b=speed-X, g=Tile-Y, r=Tile-X}
   defaultValue=<rgba 1 1 0 0>
  }
  n648 = AppendVector(A=@n647.Tile-X, B=@n647.Tile-X)
  n649 = AppendVector(A=@n647.speed-X, B=@n647.speed-X)
  n650 = Time()
  n651 = TextureSample(UVs=Add(A=Multiply(A=@n630, B=@n648), B=Multiply(A=@n649, B=@n650)), Tex=NamedRerouteUsage())
  NamedRerouteDeclaration(None=@n651.RGBA) {name=Msk_D_Warp, nodeColor=<rgba 0.1484 0 1 1>}
  n652 = NamedRerouteUsage()
  n653 = VectorParameter() {
   group=Effect_Decal_Reactive_FlipBook
   parameterName=EDRec_FlipBook_Frame__Offset
   channelNames {a=FrameD, b=FrameC, g=FrameB, r=FrameA}
   defaultValue=<rgba 0 0 0 1>
  }
  n654 = VectorParameter() {
   group=Effect_Decal_Reactive_FlipBook
   parameterName=EDRec_FlipBook_FrameB_Speed--
   channelNames {a=Clamp Max, b=Clamp Min, g="Random  1=Active", r=Speed}
   defaultValue=<rgba 0 0 0 1>
  }
  n655 = Multiply(A=Add(A=Time(), B=@n653.FrameB), B=@n654.Speed)
  n656 = StaticSwitch(True=Add(A=NamedRerouteUsage(), B=@n655), False=@n655, Value=@n625)
  n657 = Frac(None=@n656)
  n658 = MaterialFunctionCall(2 Vector=Add(A=Ceil(None=@n656), B=Constant2Vector() {g=0.2257, r=0.0135})) {materialFunction=&MF_Random_2}
  n659 = NamedRerouteUsage()
  n660 = NamedRerouteUsage()
  n661 = Add(
   A=LinearInterpolate(
    A=Clamp(None=@n657, Min=@n654.Clamp Min, Max=@n654.Clamp Min)
    B=Clamp(None=@n658.utput, Min=@n654.Clamp Min, Max=@n654.Clamp Min)
    Alpha=@n654."Random  1=Active"
   )
   B=Divide(
    A=ScalarParameter() {defaultValue=1, group=Effect_Decal_Reactive_FlipBook, parameterName=EDRec_FlipBook_FrameB}
    B=Multiply(A=@n659, B=@n660)
   ) {constA=9, constB=9}
  )
  n662 = NamedRerouteUsage()
  n663 = NamedRerouteUsage()
  n664 = NamedRerouteUsage()
  n665 = ChannelMaskParameter(None=@n635.RGBA) {group=Effect_Decal_Reactive_UV_Warp, parameterName=EDRec_WarpB_Msk_Channel}
  n666 = ScalarParameter() {group=Effect_Decal_Reactive_UV_Warp, parameterName=EDRec_WarpB_Offset}
  n667 = NamedRerouteUsage()
  n668 = ScalarParameter() {group=Effect_Decal_Reactive_UV_Warp, parameterName=EDRec_WarpB_Strength}
  n669 = ScalarParameter() {group=Effect_Decal_Reactive_UV_Warp, parameterName=EDRec_WarpB_Sine_Speed}
  n670 = Sine(None=Multiply(A=Time(), B=@n669))
  n671 = LinearInterpolate(A=@n668, B=Multiply(A=Add(A=@n670), B=@n668) {constB=0.25}, Alpha=Saturate(None=Ceil(None=@n669)))
  n672 = NamedRerouteUsage()
  n673 = StaticSwitch(
   True=Multiply(A=NamedRerouteUsage(), B=ComponentMask(None=@n664) {g=true})
   False=Multiply(A=Add(A=@n665.R, B=@n666), B=StaticSwitch(True=@n667, False=@n671, Value=@n672))
   Value=NamedRerouteUsage()
  )
  n674 = NamedRerouteUsage()
  n675 = VectorParameter() {group=Effect_Decal_Reactive_UV_Warp | Axis, parameterName=EDRec_WarpB_Axis, channelNames {a="", b="", g=Y, r=X}, defaultValue=<rgba 1 1 0 1>}
  n676 = Add(A=Multiply(A=Multiply(A=@n673, B=@n674), B=AppendVector(A=@n675.X, B=@n675.X)), B=@n626)
  n677 = StaticSwitch(
   True=@n676
   False=StaticSwitchParameter(True=@n676, False=@n626) {group=Effect_Decal_Reactive_UV_Warp, parameterName=EDRec_Warp_Uvs, sortPriority=0}
   Value=NamedRerouteUsage()
  )
  n678 = MaterialFunctionCall(
   "Animation  Phase (0-1)"=StaticSwitch(True=NamedRerouteUsage(), False=StaticSwitch(True=@n652, False=@n661, Value=@n662), Value=NamedRerouteUsage())
   Number of Rows=@n659
   Number of Columns=@n660
   Texture=@n663
   UVs=@n677
   MipBias/Level=NamedRerouteUsage()
   "Clamp Anim (see tooltip)"=StaticBool() {value=true}
  ) {materialFunction=&FlipBook}
  n679 = LinearInterpolate(
   A=@n678.UVs
   B=Frac(None=@n678.UVs)
   Alpha=ScalarParameter() {desc=Use with clamped Textures, group=Effect_Decal_Reactive_FlipBook, parameterName="EDRec_FlipBook_FrameB_1=Frac"}
  )
  n680 = NamedRerouteUsage()
  n681 = DDX(Value=@n680)
  n682 = DDY(Value=@n680)
  n683 = Multiply(A=NamedRerouteUsage(), B=@n680)
  n684 = StaticSwitch(
   True=TextureSample!(UVs=@n679, Tex=@n663, "DDX(UVs)"=@n681, "DDY(UVs)"=@n682, Apply View MipBias) {mipValueMode=TMVM_Derivative}.RGBA
   False=TextureSample!(UVs=@n679, Tex=RecordTextureStreamingInfo!(Tex=@n663, UVs=@n683).Tex, Apply View MipBias).RGBA
   Value=@n625
  )
  n685 = ComponentMask(None=@n684) {a=true}
  n686 = NamedRerouteUsage()
  n687 = StaticSwitchParameter(
   True=StaticSwitch(True=NamedRerouteUsage(), False=Constant(), Value=NamedRerouteUsage())
   False=StaticSwitch(True=Multiply(A=@n685, B=@n686), False=@n685, Value=@n625)
  ) {group=Effect_Decal_Reactive, parameterName=EDRec_Emissive_only, sortPriority=-10}
  NamedRerouteDeclaration(None=@n687) {name=Decal_B_Alpha, nodeColor=<rgba 1 0 0.7734 1>}
  NamedRerouteDeclaration(None=@n219) {name=Material, nodeColor=<rgba 1 0.4219 0 1>}
  n688 = LinearInterpolate(
   A=ScalarParameter() {defaultValue=1, group=Effect_Decal_Reactive_Emissive_C_Reactive, parameterName=EDRec_Emissive_C_Brightness_Time__Start}
   B=ScalarParameter() {group=Effect_Decal_Reactive_Emissive_C_Reactive, parameterName=EDRec_Emissive_C_Brightness_Time__End}
   Alpha=NamedRerouteUsage()
  )
  n689 = ScalarParameter() {group=Effect_Decal_Reactive_Emissive_C_Reactive, parameterName=EDRec-Emissive-C_Reactive_Audio__Min}
  n690 = ScalarParameter() {group=Effect_Decal_Reactive_Emissive_C_Reactive, parameterName=EDRec-Emissive-C_Reactive_Audio__Max}
  n691 = NamedRerouteUsage()
  n692 = StaticBoolParameter() {group=Effect_Decal_Reactive_Emissive_C_Reactive, parameterName=EDRec-Emissive-C_Reactive_Audio---------------------}
  n693 = StaticSwitch(True=Max(A=@n688, B=LinearInterpolate(A=@n689, B=@n690, Alpha=@n691)), False=@n688, Value=@n692)
  n694 = Saturate(None=Add(
    A=@n183
    B=ScalarParameter() {defaultValue=1, group=Effect_Decal_Reactive_Emissive_C, parameterName="EDRec_Emissive_C_CC_Mask   0=ON", sliderMax=1}
   ))
  n695 = VectorParameter() {group=Effect_Decal_Reactive_Emissive_C, parameterName=EDRec_Emissive_C_CC Tint, defaultValue=<rgba 1 1 1 1>}
  n696 = ScalarParameter() {group=Effect_Decal_Reactive_Emissive_C, parameterName=EDRec_Emissive_C_CC-Animated}
  n697 = Time()
  n698 = StaticSwitchParameter(True=Add(A=@n142, B=ComponentMask(None=@n154) {r=true}), False=@n171) {group=Effect_Decal_Reactive_Emissive_C, parameterName=EDRec_Emissive_C_Warp_ Msk}
  n699 = Multiply(
   A=ScalarParameter() {defaultValue=1, group=Effect_Decal_Reactive_Emissive_C, parameterName=EDRec_Emissive_C_CC_TileRate}
   B=LinearInterpolate(
    A=@n183
    B=StaticSwitchParameter(True=@n171, False=@n698) {group=Effect_Decal_Reactive_Emissive_C, parameterName="EDRec_Emissive_C_Gradent ---> use Texture"}
    Alpha=ScalarParameter() {group=Effect_Decal_Reactive_Emissive_C, parameterName=EDRec_Emissive_C_CC_Remap--Grad}
   )
  )
  n700 = ScalarParameter() {group=Effect_Decal_Reactive_Emissive_C, parameterName=EDRec_Emissive_C_CC_Offset}
  n701 = Frac(None=Add(A=Multiply(A=@n696, B=@n697), B=Add(A=@n699, B=@n700)))
  n702 = CurveAtlasRowParameter(CurveTime=@n701) {
   atlas=&CruveAtlas_Wraps_01
   curve=&C_GradSimple
   group=Effect_Decal_Reactive_Emissive_C
   parameterName=EDRec_Emissive_C_CC
  }
  n703 = MaterialFunctionCall(
   Index=ScalarParameter() {group=Effect_Decal_Reactive_Emissive_C, parameterName="EDRec_Emissive_C_CC  Index"}
   CCinput=@n702
   CCinput_Alpha=@n702
   UV=@n701
  ) {materialFunction=&MF_Global_CC_Tex}
  n704 = Multiply(
   A=Multiply(A=@n695.RGB, B=@n703.3V)
   B=ScalarParameter() {defaultValue=1, group=Effect_Decal_Reactive_Emissive_C, parameterName=EDRec_Emissive_C_CC_Brightness}
  )
  n705 = VectorParameter() {group=Effect_Decal_Reactive_Emissive_C, parameterName=EDRec_Emissive_C_Color-Start, defaultValue=<rgba 1 1 1 1>}
  n706 = VectorParameter() {group=Effect_Decal_Reactive_Emissive_C, parameterName=EDRec_Emissive_C_Color-, defaultValue=<rgba 1 1 1 1>}
  n707 = StaticSwitchParameter(True=LinearInterpolate(A=@n705.RGB, B=@n706.RGB, Alpha=NamedRerouteUsage()), False=@n706.RGB) {group=Effect_Decal_Reactive_Emissive_C_Reactive, parameterName=EDRec-Emissive-C_Reactive_Color---------------------}
  n708 = Multiply(
   A=StaticSwitchParameter(True=Multiply(A=@n694, B=@n704), False=Multiply(A=@n183, B=@n707)) {group=Effect_Decal_Reactive_Emissive_C, parameterName="EDRec-Emissive-C-Color--->CCurve"}
   B=NamedRerouteUsage()
  )
  n709 = Multiply(A=@n708, B=@n693)
  n710 = MaterialFunctionCall(
   In=Max(
    A=NamedRerouteUsage()
    B=ScalarParameter() {defaultValue=1, group=Effect_Decal_Reactive_Emissive_C, parameterName="EDRec-Emissive-C_TriMak   0=Active", sortPriority=0}
   )
   Contrast=ScalarParameter() {group=Effect_Decal_Reactive_Emissive_C, parameterName="EDRec-Emissive-C_TriMak  Contrast", sortPriority=0}
  ) {materialFunction=&CheapContrast}
  n711 = Multiply(
   A=Multiply(
    A=StaticSwitchParameter(True=@n709, False=@n708) {group=Effect_Decal_Reactive_Emissive_C_Reactive, parameterName=EDRec-Emissive-C_Reactive_Brightness---------------------}
    B=NamedRerouteUsage()
   )
   B=@n710.Result
  )
  n712 = NamedRerouteUsage()
  n713 = NamedRerouteUsage()
  n714 = StaticBoolParameter() {group=Effect_Decal_Reactive_Emissive_C_Reactive, parameterName=EDRec-Emissive-C-TOD, sortPriority=0}
  n715 = StaticSwitch(
   True=LinearInterpolate(
    B=@n711
    Alpha=StaticSwitchParameter(True=@n712, False=@n713) {group=Effect_Decal_Reactive_RE, parameterName="EDRec_DecalC_Opacity-TOD----->TODB", sortPriority=-1}
   ) {constB=0}
   False=@n711
   Value=@n714
  )
  n716 = NamedRerouteUsage()
  n717 = LinearInterpolate(
   A=@n716
   B=OneMinus(None=@n716)
   Alpha=ScalarParameter() {group=Effect_Decal_Reactive_Emissive_C_Reactive, parameterName="EDRec-Emissive-C_TopS Blend 01<--->02 MSk   1=invert"}
  )
  n718 = LinearInterpolate(B=@n715, Alpha=@n717)
  n719 = StaticBoolParameter() {group=Effect_Decal_Reactive_Emissive_C_Reactive, parameterName="EDRec-Emissive-C_TopS Blend 01<--->02 MSk"}
  n720 = StaticBoolParameter() {group="TopSide_V02_Layer 01 Blend  02", parameterName="TopS Blend 01<--->02"}
  NamedRerouteDeclaration(None=StaticSwitch(True=StaticSwitch(True=@n718, False=@n715, Value=@n719), False=@n715, Value=@n720)) {name=Emissive_C, nodeColor=<rgba 0.2656 1 0 1>}
  n721 = ScalarParameter() {defaultValue=1, group=Effect_Decal_Reactive_Uv_Offset_LR, parameterName="EDRec_DecalA_Scale\\Tile_Vehicle"}
  n722 = ScalarParameter() {defaultValue=1, group=Effect_Decal_Reactive_Uv_Offset_LR, parameterName="EDRec_DecalA_Scale\\Tile_Vehicle_Side"}
  n723 = MaterialFunctionCall(Float4=NamedRerouteUsage()) {materialFunction=&BreakOutFloat4Components}
  n724 = LinearInterpolate(
   A=LinearInterpolate(
    A=ScalarParameter() {defaultValue=1, group=Effect_Decal_Reactive_Uv_Offset_LR, parameterName="EDRec_DecalA_Scale\\Tile_Weapon"}
    B=LinearInterpolate(A=@n721, B=@n722, Alpha=@n723.G)
    Alpha=@n723.R
   )
   B=LinearInterpolate(
    A=ScalarParameter() {defaultValue=1, group=Effect_Decal_Reactive_Uv_Offset_LR, parameterName="EDRec_DecalA_Scale\\Tile_Mech"}
    B=ScalarParameter() {defaultValue=1, group=Effect_Decal_Reactive_Uv_Offset_LR, parameterName="EDRec_DecalA_Scale\\Tile_Character"}
    Alpha=@n723.A
   )
   Alpha=@n723.B
  )
  n725 = ScalarParameter() {defaultValue=1, group=Effect_Decal_Reactive_Uv_Offset_LR, parameterName="EDRec_DecalA_Scale\\Tile_Backpack"}
  n726 = MaterialFunctionCall(Float2=NamedRerouteUsage()) {materialFunction=&BreakOutFloat2Components}
  NamedRerouteDeclaration(None=LinearInterpolate(
    A=LinearInterpolate(A=@n724, B=@n725, Alpha=@n726.R)
    B=ScalarParameter() {defaultValue=1, group=Effect_Decal_Reactive_Uv_Offset_LR, parameterName="EDRec_DecalA_Scale\\Tile_Pickaxe"}
    Alpha=@n726.G
   )) {name=UVs_ScaleA, nodeColor=<rgba 0 0.7188 1 1>}
  n727 = Constant() {r=1}
  n728 = Step(
   Y=ScalarParameter() {defaultValue=0.6, group=Effect_Decal_Reactive_Emissive_C, parameterName=EDRec_Emissive_C_BC_Alpha_Clamp, sortPriority=0}
   X=NamedRerouteUsage()
  )
  n729 = StaticSwitchParameter(
   True=@n727
   False=StaticSwitchParameter(True=@n727, False=@n728) {group=Effect_Decal_Reactive_Emissive_C, parameterName=EDRec-Emissive-C_Height___X+X, sortPriority=4}
  ) {group=Effect_Decal_Reactive_Emissive_C, parameterName=EDRec-Emissive-C_Height___Lerp, sortPriority=4}
  NamedRerouteDeclaration(None=Multiply(
    A=NamedRerouteUsage()
    B=StaticSwitchParameter(True=@n727, False=@n729) {group=Effect_Decal_Reactive_Emissive_C, parameterName=EDRec-Emissive-C_Only_Mask, sortPriority=4}
   )) {name=Emissive_Msk_Simple_C, nodeColor=<rgba 0 1 0.1094 1>}
  n730 = Constant() {r=0.5}
  n731 = VectorParameter() {group=Effect_Decal_Reactive_Emissive_C, parameterName=EDRec_Emissive_C_Msk_Offset, sortPriority=5}
  n732 = MaterialFunctionCall(Float2=Add(A=ComponentMask(None=@n731.RGB) {g=true, r=true}, B=@n159.Rotated Values) {constB=0}) {materialFunction=&BreakOutFloat2Components}
  n733 = VectorParameter() {
   group=Effect_Decal_Reactive_Emissive_C
   parameterName=EDRec_Emissive_C_Decal_Msking_Scale
   sortPriority=5
   channelNames {a=Y Falloff, b=X Falloff, g=Y Size, r=X Size}
   defaultValue=<rgba 0.5 0.5 1 1>
  }
  n734 = SphereMask(A=@n730, B=@n732.R, Radius=@n733.Y Size, Hardness=@n733.Y Size) {attenuationRadius=0.5}
  n735 = SphereMask(A=@n732.G, B=@n730, Radius=@n733.X Size, Hardness=@n733.X Size) {attenuationRadius=0.5}
  n736 = StaticSwitchParameter(True=Power(Base=Multiply(A=@n734, B=@n735)), False=Constant() {r=1}) {
   desc=second layer of bounds that is used to msk
   group=Effect_Decal_Reactive_Emissive_C
   parameterName=EDRec_Emissive_C_Decal_Msking
   sortPriority=5
  }
  NamedRerouteDeclaration(None=@n736) {name=Decal_Msking_C, nodeColor=<rgba 0 1 0.5078 1>}
  n737 = LinearInterpolate(
   A=ScalarParameter() {group=Effect_Decal_Reactive_Emissive_C_Reactive, parameterName=EDRec-Emissive-C_Reactive_Audio_Darken__Min}
   B=ScalarParameter() {group=Effect_Decal_Reactive_Emissive_C_Reactive, parameterName=EDRec-Emissive-C_Reactive_Audio_Darken__Max}
   Alpha=NamedRerouteUsage()
  )
  n738 = LinearInterpolate(
   A=ScalarParameter() {group=Effect_Decal_Reactive_Emissive_C_Reactive, parameterName=EDRec_Emissive_C_Brightness_DarkenBase_Time_Start}
   B=ScalarParameter() {group=Effect_Decal_Reactive_Emissive_C_Reactive, parameterName=EDRec_Emissive_C_Brightness_DarkenBase_Time_End}
   Alpha=NamedRerouteUsage()
  )
  n739 = StaticSwitchParameter(
   True=StaticSwitch(True=Max(A=@n737, B=@n738), False=@n738, Value=NamedRerouteUsage())
   False=ScalarParameter() {group=Effect_Decal_Reactive_Emissive_C, parameterName=EDRec_Emissive_C_DarkenBase-------, sortPriority=4}
  ) {group=Effect_Decal_Reactive_Emissive_C_Reactive, parameterName=EDRec-Emissive-C_Reactive_Brightness---------------------}
  n740 = NamedRerouteUsage()
  n741 = Multiply(A=Multiply(A=@n739, B=@n183), B=@n740)
  n742 = Multiply(A=@n739, B=Saturate(None=@n140))
  n743 = StaticSwitchParameter(True=@n173, False=Constant()) {group=Effect_Decal_Reactive_Emissive_C, parameterName=EDRec-Emissive-C_Height___X+X, sortPriority=4}
  n744 = Multiply(
   A=@n742
   B=StaticSwitchParameter(
    True=@n173
    False=StaticSwitchParameter(True=@n173, False=@n743) {group=Effect_Decal_Reactive_Emissive_C, parameterName=EDRec-Emissive-C_Height___Lerp, sortPriority=4}
   ) {group=Effect_Decal_Reactive_Emissive_C, parameterName=EDRec-Emissive-C_Only_Mask, sortPriority=4}
  )
  n745 = ScalarParameter() {group=Effect_Decal_Reactive_Emissive_C, parameterName=EDRec_Emissive_C_DarkenBase-CCAlpha MskSource, sliderMax=1, sortPriority=4}
  n746 = ScalarParameter() {group=Effect_Decal_Reactive_Emissive_C, parameterName="EDRec_Emissive_C_DarkenBase-CCAlpha 1=Use", sliderMax=1, sortPriority=4}
  n747 = StaticSwitchParameter(True=LinearInterpolate(A=@n741, B=LinearInterpolate(A=@n742, B=@n744, Alpha=@n745), Alpha=Ceil(None=@n746)), False=@n741) {group=Effect_Decal_Reactive_Emissive_C, parameterName="EDRec-Emissive-C-Color--->CCurve"}
  n748 = Saturate(None=Multiply(A=@n747, B=NamedRerouteUsage()))
  n749 = NamedRerouteUsage()
  n750 = ScalarParameter() {defaultValue=1, group=Effect_Decal_Reactive_Emissive_C, parameterName="EDRec_Emissive_C_DrawBehind BC 0=On", sliderMax=1, sortPriority=4}
  n751 = Saturate(None=Add(A=OneMinus(None=@n749), B=@n750))
  n752 = LinearInterpolate(B=Max(A=OneMinus(None=@n748), B=OneMinus(None=@n710.Result)), Alpha=@n751) {constA=1}
  n753 = NamedRerouteUsage()
  n754 = NamedRerouteUsage()
  n755 = StaticSwitch(
   True=LinearInterpolate(
    B=@n752
    Alpha=StaticSwitchParameter(True=@n753, False=@n754) {group=Effect_Decal_Reactive_RE, parameterName="EDRec_DecalC_Opacity-TOD----->TODB", sortPriority=-1}
   ) {constA=1, constB=0}
   False=@n752
   Value=@n714
  )
  n756 = LinearInterpolate(B=@n755, Alpha=@n717) {constA=1}
  NamedRerouteDeclaration(None=StaticSwitch(True=StaticSwitch(True=@n756, False=@n755, Value=@n719), False=@n755, Value=@n720)) {name=Emissive_Darken_Base C, nodeColor=<rgba 0 0.1328 1 1>}
  n757 = ScalarParameter() {
   defaultValue=-9999999
   group=Effect_Decal_Reactive_Emissive_C_Reactive |Gradent
   parameterName=EDRec_Emissive_C_Gradent_Time_BP
   sortPriority=212
  }
  n758 = ScalarParameter() {group=Effect_Decal_Reactive_Emissive_C_Reactive |Gradent, parameterName=EDRec_Emissive_C_Gradent_Delay, sortPriority=211}
  n759 = ScalarParameter() {
   defaultValue=0.25
   group=Effect_Decal_Reactive_Emissive_C_Reactive |Gradent
   parameterName=EDRec_Emissive_C_Gradent_Duration
   sortPriority=211
  }
  n760 = Divide(A=Subtract(A=Time(), B=Add(A=@n757, B=@n758)), B=@n759)
  n761 = ScalarParameter() {defaultValue=1, group=Effect_Decal_Reactive_Emissive_C_Reactive |Gradent, parameterName=EDRec_Emissive_C_Gradent_Falloff_, sortPriority=211}
  NamedRerouteDeclaration(None=Power(Base=Saturate(None=@n760), Exp=@n761)) {name=reactive_Gradent_C, nodeColor=<rgba 0 0.5078 1 1>}
  NamedRerouteDeclaration(None=@n728) {name=Alpha_MskEC, nodeColor=<rgba 1 0 0.9375 1>}
  n762 = ScalarParameter() {defaultValue=-9999999, group=Effect_Decal_Reactive_Emissive_C_Reactive, parameterName=EDRec_Emissive_C_Brightness_Time_BP, sortPriority=50}
  n763 = ScalarParameter() {group=Effect_Decal_Reactive_Emissive_C_Reactive, parameterName=EDRec_Emissive_C_Brightness_Delay}
  n764 = Divide(
   A=Subtract(A=Time(), B=Add(A=@n762, B=@n763))
   B=ScalarParameter() {defaultValue=0.25, group=Effect_Decal_Reactive_Emissive_C_Reactive, parameterName=EDRec_Emissive_C_Brightness_Duration}
  )
  NamedRerouteDeclaration(None=Power(
    Base=Saturate(None=@n764)
    Exp=ScalarParameter() {defaultValue=1, group=Effect_Decal_Reactive_Emissive_C_Reactive, parameterName=EDRec_Emissive_C_Brightness_Falloff}
   )) {name=Reactive_Brightness_C, nodeColor=<rgba 0.8516 0 1 1>}
  NamedRerouteDeclaration(None=@n751) {name=Alpha_MSk_C_1, nodeColor=<rgba 1 0 0.8203 1>}
  NamedRerouteDeclaration(None=MaterialFunctionCall(
    Caculated_Bounds=NamedRerouteUsage()
    is_State=NamedRerouteUsage()
    UV_Rotation=NamedRerouteUsage()
    Uvs_TileRate=NamedRerouteUsage()
    No_Rotation=StaticBoolParameter() {defaultValue=true, group=Effect_Decal_Reactive_Emissive_UV_00, parameterName=EDRec_Emissive_00_Rotation, sortPriority=0}
   ) {materialFunction=&MF_Projection_Side_Normal_V2}.UVs) {name=Uvs_00_Gradent, nodeColor=<rgba 0 0.4609 1 1>}
  n765 = MaterialFunctionCall(Float4=NamedRerouteUsage()) {materialFunction=&BreakOutFloat4Components}
  n766 = LinearInterpolate(
   A=ScalarParameter() {group=Effect_Decal_Reactive_Emissive_UV_00, parameterName="EDRec_Emissive_00_Rotation_Weapon\\pickaxe"}
   B=LinearInterpolate(
    A=ScalarParameter() {group=Effect_Decal_Reactive_Emissive_UV_00, parameterName=EDRec_Emissive_00_Rotation_Vehicle}
    B=ScalarParameter() {group=Effect_Decal_Reactive_Emissive_UV_00, parameterName=EDRec_Emissive_00_Rotation_Vehicle_Side}
    Alpha=@n765.G
   )
   Alpha=@n765.R
  )
  n767 = ScalarParameter() {group=Effect_Decal_Reactive_Emissive_UV_00, parameterName="EDRec_Emissive_00_Mech\\Char\\Back"}
  NamedRerouteDeclaration(None=Add(A=LinearInterpolate(A=@n766, B=@n767, Alpha=@n765.B), B=NamedRerouteUsage())) {name=Emissive_00_Rotation, nodeColor=<rgba 0 0.7188 1 1>}
  n768 = ScalarParameter() {group=Effect_Decal_Reactive_Uv_Offset_LR, parameterName=EDRec_DecalA_Rotation_Weapon}
  n769 = MaterialFunctionCall(Float4=NamedRerouteUsage()) {materialFunction=&BreakOutFloat4Components}
  n770 = LinearInterpolate(
   A=ScalarParameter() {group=Effect_Decal_Reactive_Uv_Offset_LR, parameterName=EDRec_DecalA_Rotation_Vehicle}
   B=ScalarParameter() {group=Effect_Decal_Reactive_Uv_Offset_LR, parameterName=EDRec_DecalA_Rotation_Vehicle_Side}
   Alpha=@n769.G
  )
  n771 = ScalarParameter() {group=Effect_Decal_Reactive_Uv_Offset_LR, parameterName=EDRec_DecalA_Rotation_Mech}
  n772 = ScalarParameter() {group=Effect_Decal_Reactive_Uv_Offset_LR, parameterName=EDRec_DecalA_Rotation_Character}
  n773 = MaterialFunctionCall(Float2=NamedRerouteUsage()) {materialFunction=&BreakOutFloat2Components}
  n774 = LinearInterpolate(
   A=LinearInterpolate(A=LinearInterpolate(A=@n768, B=@n770, Alpha=@n769.R), B=LinearInterpolate(A=@n771, B=@n772, Alpha=@n769.A), Alpha=@n769.B)
   B=ScalarParameter() {group=Effect_Decal_Reactive_Uv_Offset_LR, parameterName=EDRec_DecalA_Rotation_Backpack}
   Alpha=@n773.R
  )
  n775 = ScalarParameter() {group=Effect_Decal_Reactive_Uv_Offset_LR, parameterName=EDRec_DecalA_Rotation_Pickaxe}
  NamedRerouteDeclaration(None=Add(A=LinearInterpolate(A=@n774, B=@n775, Alpha=@n773.G), B=NamedRerouteUsage())) {name=Uvs_RotationA, nodeColor=<rgba 0 0.7188 1 1>}
  n776 = LinearInterpolate(
   A=ScalarParameter() {defaultValue=1, group=Effect_Decal_Reactive_Emissive_D_Reactive, parameterName=EDRec_Emissive_D_Brightness_Time__Start}
   B=ScalarParameter() {group=Effect_Decal_Reactive_Emissive_D_Reactive, parameterName=EDRec_Emissive_D_Brightness_Time__End}
   Alpha=NamedRerouteUsage()
  )
  n777 = ScalarParameter() {group=Effect_Decal_Reactive_Emissive_D_Reactive, parameterName=EDRec-Emissive-D_Reactive_Audio__Min}
  n778 = ScalarParameter() {group=Effect_Decal_Reactive_Emissive_D_Reactive, parameterName=EDRec-Emissive-D_Reactive_Audio__Max}
  n779 = NamedRerouteUsage()
  n780 = StaticBoolParameter() {group=Effect_Decal_Reactive_Emissive_D_Reactive, parameterName=EDRec-Emissive-D_Reactive_Audio---------------------}
  n781 = StaticSwitch(True=Max(A=@n776, B=LinearInterpolate(A=@n777, B=@n778, Alpha=@n779)), False=@n776, Value=@n780)
  n782 = Saturate(None=Add(
    A=@n47
    B=ScalarParameter() {defaultValue=1, group=Effect_Decal_Reactive_Emissive_D, parameterName="EDRec_Emissive_D_CC_Mask   0=ON", sliderMax=1}
   ))
  n783 = VectorParameter() {group=Effect_Decal_Reactive_Emissive_D, parameterName=EDRec_Emissive_D_CC Tint, defaultValue=<rgba 1 1 1 1>}
  n784 = ScalarParameter() {group=Effect_Decal_Reactive_Emissive_D, parameterName=EDRec_Emissive_D_CC-Animated}
  n785 = Time()
  n786 = StaticSwitchParameter(True=Add(A=@n5, B=ComponentMask(None=@n16) {r=true}), False=@n33) {group=Effect_Decal_Reactive_Emissive_D, parameterName=EDRec_Emissive_D_Warp_ Msk}
  n787 = Multiply(
   A=ScalarParameter() {defaultValue=1, group=Effect_Decal_Reactive_Emissive_D, parameterName=EDRec_Emissive_D_CC_TileRate}
   B=LinearInterpolate(
    A=@n47
    B=StaticSwitchParameter(True=@n33, False=@n786) {group=Effect_Decal_Reactive_Emissive_D, parameterName="EDRec_Emissive_D_Gradent ---> use Texture"}
    Alpha=ScalarParameter() {group=Effect_Decal_Reactive_Emissive_D, parameterName=EDRec_Emissive_D_CC_Remap--Grad}
   )
  )
  n788 = ScalarParameter() {group=Effect_Decal_Reactive_Emissive_D, parameterName=EDRec_Emissive_D_CC_Offset}
  n789 = Frac(None=Add(A=Multiply(A=@n784, B=@n785), B=Add(A=@n787, B=@n788)))
  n790 = CurveAtlasRowParameter(CurveTime=@n789) {
   atlas=&CruveAtlas_Wraps_01
   curve=&C_GradSimple
   group=Effect_Decal_Reactive_Emissive_D
   parameterName=EDRec_Emissive_D_CC
  }
  n791 = MaterialFunctionCall(
   Index=ScalarParameter() {group=Effect_Decal_Reactive_Emissive_D, parameterName="EDRec_Emissive_D_CC  Index"}
   CCinput=@n790
   CCinput_Alpha=@n790
   UV=@n789
  ) {materialFunction=&MF_Global_CC_Tex}
  n792 = Multiply(
   A=Multiply(A=@n783.RGB, B=@n791.3V)
   B=ScalarParameter() {defaultValue=1, group=Effect_Decal_Reactive_Emissive_D, parameterName=EDRec_Emissive_D_CC_Brightness}
  )
  n793 = VectorParameter() {group=Effect_Decal_Reactive_Emissive_D, parameterName=EDRec_Emissive_D_Color-Start, defaultValue=<rgba 1 1 1 1>}
  n794 = VectorParameter() {group=Effect_Decal_Reactive_Emissive_D, parameterName=EDRec_Emissive_D_Color-, defaultValue=<rgba 1 1 1 1>}
  n795 = StaticSwitchParameter(True=LinearInterpolate(A=@n793.RGB, B=@n794.RGB, Alpha=NamedRerouteUsage()), False=@n794.RGB) {group=Effect_Decal_Reactive_Emissive_D_Reactive, parameterName=EDRec-Emissive-D_Reactive_Color---------------------}
  n796 = Multiply(
   A=StaticSwitchParameter(True=Multiply(A=@n782, B=@n792), False=Multiply(A=@n47, B=@n795)) {group=Effect_Decal_Reactive_Emissive_D, parameterName="EDRec-Emissive-D-Color--->CCurve"}
   B=NamedRerouteUsage()
  )
  n797 = Multiply(A=@n796, B=@n781)
  n798 = MaterialFunctionCall(
   In=Max(
    A=NamedRerouteUsage()
    B=ScalarParameter() {defaultValue=1, group=Effect_Decal_Reactive_Emissive_D, parameterName="EDRec-Emissive-D_TriMak   0=Active", sortPriority=0}
   )
   Contrast=ScalarParameter() {group=Effect_Decal_Reactive_Emissive_D, parameterName="EDRec-Emissive-D_TriMak  Contrast", sortPriority=0}
  ) {materialFunction=&CheapContrast}
  n799 = Multiply(
   A=Multiply(
    A=StaticSwitchParameter(True=@n797, False=@n796) {group=Effect_Decal_Reactive_Emissive_D_Reactive, parameterName=EDRec-Emissive-D_Reactive_Brightness---------------------}
    B=NamedRerouteUsage()
   )
   B=@n798.Result
  )
  n800 = NamedRerouteUsage()
  n801 = NamedRerouteUsage()
  n802 = StaticBoolParameter() {group=Effect_Decal_Reactive_Emissive_D_Reactive, parameterName=EDRec-Emissive-D-TOD, sortPriority=0}
  n803 = StaticSwitch(
   True=LinearInterpolate(
    B=@n799
    Alpha=StaticSwitchParameter(True=@n800, False=@n801) {group=Effect_Decal_Reactive_RE, parameterName="EDRec_DecalD_Opacity-TOD----->TODB", sortPriority=-1}
   ) {constB=0}
   False=@n799
   Value=@n802
  )
  n804 = NamedRerouteUsage()
  n805 = LinearInterpolate(
   A=@n804
   B=OneMinus(None=@n804)
   Alpha=ScalarParameter() {group=Effect_Decal_Reactive_Emissive_D_Reactive, parameterName="EDRec-Emissive-D_TopS Blend 01<--->02 MSk   1=invert"}
  )
  n806 = LinearInterpolate(B=@n803, Alpha=@n805)
  n807 = StaticBoolParameter() {group=Effect_Decal_Reactive_Emissive_D_Reactive, parameterName="EDRec-Emissive-D_TopS Blend 01<--->02 MSk"}
  n808 = StaticBoolParameter() {group="TopSide_V02_Layer 01 Blend  02", parameterName="TopS Blend 01<--->02"}
  NamedRerouteDeclaration(None=StaticSwitch(True=StaticSwitch(True=@n806, False=@n803, Value=@n807), False=@n803, Value=@n808)) {name=Emissive_D, nodeColor=<rgba 0.2656 1 0 1>}
  n809 = Constant() {r=1}
  n810 = Step(
   Y=ScalarParameter() {defaultValue=0.6, group=Effect_Decal_Reactive_Emissive_D, parameterName=EDRec_Emissive_D_BC_Alpha_Clamp, sortPriority=0}
   X=NamedRerouteUsage()
  )
  n811 = StaticSwitchParameter(
   True=@n809
   False=StaticSwitchParameter(True=@n809, False=@n810) {group=Effect_Decal_Reactive_Emissive_D, parameterName=EDRec-Emissive-D_Height___X+X, sortPriority=4}
  ) {group=Effect_Decal_Reactive_Emissive_D, parameterName=EDRec-Emissive-D_Height___Lerp, sortPriority=4}
  NamedRerouteDeclaration(None=Multiply(
    A=NamedRerouteUsage()
    B=StaticSwitchParameter(True=@n809, False=@n811) {group=Effect_Decal_Reactive_Emissive_D, parameterName=EDRec-Emissive-D_Only_Mask, sortPriority=4}
   )) {name=Emissive_Msk_Simple_D, nodeColor=<rgba 0 1 0.1094 1>}
  n812 = Constant() {r=0.5}
  n813 = VectorParameter() {group=Effect_Decal_Reactive_Emissive_D, parameterName=EDRec_Emissive_D_Msk_Offset, sortPriority=5}
  n814 = MaterialFunctionCall(Float2=Add(A=ComponentMask(None=@n813.RGB) {g=true, r=true}, B=@n21.Rotated Values) {constB=0}) {materialFunction=&BreakOutFloat2Components}
  n815 = VectorParameter() {
   group=Effect_Decal_Reactive_Emissive_D
   parameterName=EDRec_Emissive_D_Decal_Msking_Scale
   sortPriority=5
   channelNames {a=Y Falloff, b=X Falloff, g=Y Size, r=X Size}
   defaultValue=<rgba 0.5 0.5 1 1>
  }
  n816 = SphereMask(A=@n812, B=@n814.R, Radius=@n815.Y Size, Hardness=@n815.Y Size) {attenuationRadius=0.5}
  n817 = SphereMask(A=@n814.G, B=@n812, Radius=@n815.X Size, Hardness=@n815.X Size) {attenuationRadius=0.5}
  n818 = StaticSwitchParameter(True=Power(Base=Multiply(A=@n816, B=@n817)), False=Constant() {r=1}) {
   desc=second layer of bounds that is used to msk
   group=Effect_Decal_Reactive_Emissive_D
   parameterName=EDRec_Emissive_D_Decal_Msking
   sortPriority=5
  }
  NamedRerouteDeclaration(None=@n818) {name=Decal_Msking_D, nodeColor=<rgba 0 1 0.5078 1>}
  n819 = LinearInterpolate(
   A=ScalarParameter() {group=Effect_Decal_Reactive_Emissive_D_Reactive, parameterName=EDRec-Emissive-D_Reactive_Audio_Darken__Min}
   B=ScalarParameter() {group=Effect_Decal_Reactive_Emissive_D_Reactive, parameterName=EDRec-Emissive-D_Reactive_Audio_Darken__Max}
   Alpha=NamedRerouteUsage()
  )
  n820 = LinearInterpolate(
   A=ScalarParameter() {group=Effect_Decal_Reactive_Emissive_D_Reactive, parameterName=EDRec_Emissive_D_Brightness_DarkenBase_Time_Start}
   B=ScalarParameter() {group=Effect_Decal_Reactive_Emissive_D_Reactive, parameterName=EDRec_Emissive_D_Brightness_DarkenBase_Time_End}
   Alpha=NamedRerouteUsage()
  )
  n821 = StaticSwitchParameter(
   True=StaticSwitch(True=Max(A=@n819, B=@n820), False=@n820, Value=NamedRerouteUsage())
   False=ScalarParameter() {group=Effect_Decal_Reactive_Emissive_D, parameterName=EDRec_Emissive_D_DarkenBase-------, sortPriority=4}
  ) {group=Effect_Decal_Reactive_Emissive_D_Reactive, parameterName=EDRec-Emissive-D_Reactive_Brightness---------------------}
  n822 = NamedRerouteUsage()
  n823 = Multiply(A=Multiply(A=@n821, B=@n47), B=@n822)
  n824 = Multiply(A=@n821, B=Saturate(None=@n1))
  n825 = StaticSwitchParameter(True=@n2, False=Constant() {r=1}) {group=Effect_Decal_Reactive_Emissive_D, parameterName=EDRec-Emissive-D_Height___X+X, sortPriority=4}
  n826 = Multiply(
   A=@n824
   B=StaticSwitchParameter(
    True=@n2
    False=StaticSwitchParameter(True=@n2, False=@n825) {group=Effect_Decal_Reactive_Emissive_D, parameterName=EDRec-Emissive-D_Height___Lerp, sortPriority=4}
   ) {group=Effect_Decal_Reactive_Emissive_D, parameterName=EDRec-Emissive-D_Only_Mask, sortPriority=4}
  )
  n827 = ScalarParameter() {group=Effect_Decal_Reactive_Emissive_D, parameterName=EDRec_Emissive_D_DarkenBase-CCAlpha MskSource, sliderMax=1, sortPriority=4}
  n828 = ScalarParameter() {group=Effect_Decal_Reactive_Emissive_D, parameterName="EDRec_Emissive_D_DarkenBase-CCAlpha 1=Use", sliderMax=1, sortPriority=4}
  n829 = StaticSwitchParameter(True=LinearInterpolate(A=@n823, B=LinearInterpolate(A=@n824, B=@n826, Alpha=@n827), Alpha=Ceil(None=@n828)), False=@n823) {group=Effect_Decal_Reactive_Emissive_D, parameterName="EDRec-Emissive-D-Color--->CCurve"}
  n830 = Saturate(None=Multiply(A=@n829, B=NamedRerouteUsage()))
  n831 = NamedRerouteUsage()
  n832 = ScalarParameter() {defaultValue=1, group=Effect_Decal_Reactive_Emissive_D, parameterName="EDRec_Emissive_D_DrawBehind BC 0=On", sliderMax=1, sortPriority=4}
  n833 = Saturate(None=Add(A=OneMinus(None=@n831), B=@n832))
  n834 = LinearInterpolate(B=Max(A=OneMinus(None=@n798.Result), B=OneMinus(None=@n830)), Alpha=@n833) {constA=1}
  n835 = NamedRerouteUsage()
  n836 = NamedRerouteUsage()
  n837 = StaticSwitch(
   True=LinearInterpolate(
    B=@n834
    Alpha=StaticSwitchParameter(True=@n835, False=@n836) {group=Effect_Decal_Reactive_RE, parameterName="EDRec_DecalD_Opacity-TOD----->TODB", sortPriority=-1}
   ) {constA=1, constB=0}
   False=@n834
   Value=@n802
  )
  n838 = LinearInterpolate(B=@n837, Alpha=@n805) {constA=1}
  NamedRerouteDeclaration(None=StaticSwitch(True=StaticSwitch(True=@n838, False=@n837, Value=@n807), False=@n837, Value=@n808)) {name=Emissive_Darken_Base D, nodeColor=<rgba 0 0.1328 1 1>}
  n839 = ScalarParameter() {
   defaultValue=-9999999
   group=Effect_Decal_Reactive_Emissive_D_Reactive|Gradent
   parameterName=EDRec_Emissive_D_Gradent_Time_BP
   sortPriority=212
  }
  n840 = ScalarParameter() {group=Effect_Decal_Reactive_Emissive_D_Reactive|Gradent, parameterName=EDRec_Emissive_D_Gradent_Delay, sortPriority=211}
  n841 = ScalarParameter() {
   defaultValue=0.25
   group=Effect_Decal_Reactive_Emissive_D_Reactive|Gradent
   parameterName=EDRec_Emissive_D_Gradent_Duration
   sortPriority=211
  }
  n842 = Divide(A=Subtract(A=Time(), B=Add(A=@n839, B=@n840)), B=@n841)
  n843 = ScalarParameter() {defaultValue=1, group=Effect_Decal_Reactive_Emissive_D_Reactive|Gradent, parameterName=EDRec_Emissive_D_Gradent_Falloff_, sortPriority=211}
  NamedRerouteDeclaration(None=Power(Base=Saturate(None=@n842), Exp=@n843)) {name=reactive_Gradent_D, nodeColor=<rgba 0 0.5078 1 1>}
  NamedRerouteDeclaration(None=@n810) {name=Alpha_MskED, nodeColor=<rgba 1 0 0.9375 1>}
  n844 = CollectionParameter() {collection=&FortniteMaterialParameters, parameterName=EventAudio_Spectrum}
  n845 = ChannelMaskParameter(None=@n844) {group=Effect_Decal_Reactive_RE_Audio, maskChannel=Alpha, parameterName=EDRec_Audio A Channel, defaultValue=<rgba 0 0 0 1>}
  n846 = CollectionParameter() {collection=&FortniteMaterialParameters, parameterName=Emote_AudioLow}
  n847 = Max(A=@n845.R, B=@n846)
  n848 = ScalarParameter() {defaultValue=0.25, group=Effect_Decal_Reactive_RE_Audio, parameterName=EDRec_Audio A_Low_Remap}
  n849 = ScalarParameter() {defaultValue=1, group=Effect_Decal_Reactive_RE_Audio, parameterName=EDRec_Audio A_High_Remap}
  n850 = Constant()
  n851 = Constant() {r=1}
  n852 = CollectionParameter() {collection=&FortniteMaterialParameters, parameterName=Emote_AudioHigh}
  n853 = Max(
   A=Saturate(None=MaterialFunctionCall(Input=@n847, Input Low=@n848, Input High=@n849, Target Low=@n850, Target High=@n851) {materialFunction=&RemapValueRange}.Result)
   B=@n852
  )
  NamedRerouteDeclaration(None=@n853) {name=AudioSpectrum A, nodeColor=<rgba 1 0.8203 0 1>}
  n854 = ScalarParameter() {defaultValue=-9999999, group=Effect_Decal_Reactive_Emissive_D_Reactive, parameterName=EDRec_Emissive_D_Brightness_Time_BP, sortPriority=50}
  n855 = ScalarParameter() {group=Effect_Decal_Reactive_Emissive_D_Reactive, parameterName=EDRec_Emissive_D_Brightness_Delay}
  n856 = Divide(
   A=Subtract(A=Time(), B=Add(A=@n854, B=@n855))
   B=ScalarParameter() {defaultValue=0.25, group=Effect_Decal_Reactive_Emissive_D_Reactive, parameterName=EDRec_Emissive_D_Brightness_Duration}
  )
  NamedRerouteDeclaration(None=Power(
    Base=Saturate(None=@n856)
    Exp=ScalarParameter() {defaultValue=1, group=Effect_Decal_Reactive_Emissive_D_Reactive, parameterName=EDRec_Emissive_D_Brightness_Falloff}
   )) {name=Reactive_Brightness_D, nodeColor=<rgba 0.8516 0 1 1>}
  NamedRerouteDeclaration(None=@n833) {name=Alpha_MSk_D, nodeColor=<rgba 1 0 0.8203 1>}
  NamedRerouteDeclaration(None=StaticBoolParameter() {group=Effect_Decal_Reactive_RE_Audio, parameterName="EDRec  Audio Reactive Warp"}) {name=Audio_Warp, nodeColor=<rgba 0 0.4141 1 1>}
  n857 = VectorParameter() {group=Effect_Decal_Reactive_Uv_Offset_LR, parameterName=EDRec_DecalB_Offset_Weapon, defaultValue=<rgba 0 0 0 0.25>}
  n858 = VectorParameter() {group=Effect_Decal_Reactive_Uv_Offset_LR, parameterName=EDRec_DecalB_Offset_Vehicle, defaultValue=<rgba 0 0 0 0.25>}
  n859 = VectorParameter() {group=Effect_Decal_Reactive_Uv_Offset_LR, parameterName=EDRec_DecalB_Offset_VehicleSide, defaultValue=<rgba 0 0 0 0.25>}
  n860 = MaterialFunctionCall(Float4=NamedRerouteUsage()) {materialFunction=&BreakOutFloat4Components}
  n861 = LinearInterpolate(
   A=AppendVector(A=@n857.RGB, B=@n857.RGB)
   B=LinearInterpolate(A=AppendVector(A=@n858.RGB, B=@n858.RGB), B=AppendVector(A=@n859.RGB, B=@n859.RGB), Alpha=@n860.G)
   Alpha=@n860.R
  )
  n862 = VectorParameter() {group=Effect_Decal_Reactive_Uv_Offset_LR, parameterName=EDRec_DecalB_Offset_Mech, defaultValue=<rgba 0 0 0 0.25>}
  n863 = VectorParameter() {group=Effect_Decal_Reactive_Uv_Offset_LR, parameterName=EDRec_DecalB_Offset_Character, defaultValue=<rgba 0 0 0 0.25>}
  n864 = LinearInterpolate(A=AppendVector(A=@n862.RGB, B=@n862.RGB), B=AppendVector(A=@n863.RGB, B=@n863.RGB), Alpha=@n860.A)
  n865 = VectorParameter() {group=Effect_Decal_Reactive_Uv_Offset_LR, parameterName="EDRec_DecalB_Offset_Backpack----A=Scale", defaultValue=<rgba 0 0 0 0.25>}
  n866 = MaterialFunctionCall(Float2=NamedRerouteUsage()) {materialFunction=&BreakOutFloat2Components}
  n867 = VectorParameter() {group=Effect_Decal_Reactive_Uv_Offset_LR, parameterName=EDRec_DecalB_Offset_Pickaxe, defaultValue=<rgba 0 0 0 0.25>}
  n868 = LinearInterpolate(
   A=LinearInterpolate(A=LinearInterpolate(A=@n861, B=@n864, Alpha=@n860.B), B=AppendVector(A=@n865.RGB, B=@n865.RGB), Alpha=@n866.R)
   B=AppendVector(A=@n867.RGB, B=@n867.RGB)
   Alpha=@n866.G
  )
  n869 = VectorParameter() {group=Effect_Decal_Reactive_Uv_Offset_LR, parameterName=EDRec_DecalB_Offset_Weapon End, defaultValue=<rgba 0 0 0 0.25>}
  n870 = VectorParameter() {group=Effect_Decal_Reactive_Uv_Offset_LR, parameterName=EDRec_DecalB_Offset_Vehicle End, defaultValue=<rgba 0 0 0 0.25>}
  n871 = VectorParameter() {group=Effect_Decal_Reactive_Uv_Offset_LR, parameterName=EDRec_DecalB_Offset_VehicleSide End, defaultValue=<rgba 0 0 0 0.25>}
  n872 = MaterialFunctionCall(Float4=NamedRerouteUsage()) {materialFunction=&BreakOutFloat4Components}
  n873 = LinearInterpolate(
   A=AppendVector(A=@n869.RGB, B=@n869.RGB)
   B=LinearInterpolate(A=AppendVector(A=@n870.RGB, B=@n870.RGB), B=AppendVector(A=@n871.RGB, B=@n871.RGB), Alpha=@n872.G)
   Alpha=@n872.R
  )
  n874 = VectorParameter() {group=Effect_Decal_Reactive_Uv_Offset_LR, parameterName=EDRec_DecalB_Offset_Mech End, defaultValue=<rgba 0 0 0 0.25>}
  n875 = VectorParameter() {group=Effect_Decal_Reactive_Uv_Offset_LR, parameterName=EDRec_DecalB_Offset_Character End, defaultValue=<rgba 0 0 0 0.25>}
  n876 = LinearInterpolate(A=AppendVector(A=@n874.RGB, B=@n874.RGB), B=AppendVector(A=@n875.RGB, B=@n875.RGB), Alpha=@n872.A)
  n877 = VectorParameter() {group=Effect_Decal_Reactive_Uv_Offset_LR, parameterName="EDRec_DecalB_Offset_Backpack----A=Scale End", defaultValue=<rgba 0 0 0 0.25>}
  n878 = MaterialFunctionCall(Float2=NamedRerouteUsage()) {materialFunction=&BreakOutFloat2Components}
  n879 = VectorParameter() {group=Effect_Decal_Reactive_Uv_Offset_LR, parameterName=EDRec_DecalB_Offset_Pickaxe End, defaultValue=<rgba 0 0 0 0.25>}
  n880 = LinearInterpolate(
   A=LinearInterpolate(A=LinearInterpolate(A=@n873, B=@n876, Alpha=@n872.B), B=AppendVector(A=@n877.RGB, B=@n877.RGB), Alpha=@n878.R)
   B=AppendVector(A=@n879.RGB, B=@n879.RGB)
   Alpha=@n878.G
  )
  n881 = StaticSwitchParameter(True=LinearInterpolate(A=@n868, B=@n880, Alpha=NamedRerouteUsage()), False=@n868) {desc="Use Scale value Alpha is disabled. ", group=Effect_Decal_Reactive_RE_Offset, parameterName=EDRec_Reactive_Offset, sortPriority=0}
  NamedRerouteDeclaration(None=@n881) {name=UVs_OffsetB, nodeColor=<rgba 0 0.625 1 1>}
  n882 = ScalarParameter() {defaultValue=1, group=Effect_Decal_Reactive_RE_Audio, parameterName=EDRec_Audio A Strength}
  n883 = Time()
  n884 = ScalarParameter() {defaultValue=2, group=Effect_Decal_Reactive_RE_Audio, parameterName=EDRec_Audio A Static_Speed}
  n885 = ConstantBiasScale(None=Sine(None=Multiply(A=@n883, B=@n884)))
  n886 = ScalarParameter() {group=Effect_Decal_Reactive_RE_Audio, parameterName=EDRec_Audio A Static_Strength}
  NamedRerouteDeclaration(None=Max(A=Multiply(A=@n853, B=@n882), B=Multiply(A=@n885, B=@n886))) {name=AudioSpectrum A_Static_Speed, nodeColor=<rgba 1 0.8203 0 1>}
  n887 = VectorParameter() {group=Effect_Decal_Reactive_Emissive_UV_00, parameterName=EDRec_Emissive_00_Scale_Weapon, defaultValue=<rgba 1 1 1 1>}
  n888 = VectorParameter() {group=Effect_Decal_Reactive_Emissive_UV_00, parameterName=EDRec_Emissive_00_Scale_Vehicle, defaultValue=<rgba 1 1 1 1>}
  n889 = Multiply(A=@n888.RGB, B=@n888.RGB)
  n890 = VectorParameter() {group=Effect_Decal_Reactive_Emissive_UV_00, parameterName=EDRec_Emissive_00_Scale_VehicleSide, defaultValue=<rgba 1 1 1 1>}
  n891 = Multiply(A=@n890.RGB, B=@n890.RGB)
  n892 = MaterialFunctionCall(Float4=NamedRerouteUsage()) {materialFunction=&BreakOutFloat4Components}
  n893 = VectorParameter() {group=Effect_Decal_Reactive_Emissive_UV_00, parameterName=EDRec_Emissive_00_Scale_Mech, defaultValue=<rgba 1 1 1 1>}
  n894 = VectorParameter() {group=Effect_Decal_Reactive_Emissive_UV_00, parameterName=EDRec_Emissive_00_Scale_Character, defaultValue=<rgba 1 1 1 1>}
  n895 = LinearInterpolate(
   A=LinearInterpolate(A=Multiply(A=@n887.RGB, B=@n887.RGB), B=LinearInterpolate(A=@n889, B=@n891, Alpha=@n892.G), Alpha=@n892.R)
   B=LinearInterpolate(A=Multiply(A=@n893.RGB, B=@n893.RGB), B=Multiply(A=@n894.RGB, B=@n894.RGB), Alpha=@n892.A)
   Alpha=@n892.B
  )
  n896 = VectorParameter() {group=Effect_Decal_Reactive_Emissive_UV_00, parameterName=EDRec_Emissive_00_Scale_Backpack, defaultValue=<rgba 1 1 1 1>}
  n897 = Multiply(A=@n896.RGB, B=@n896.RGB)
  n898 = MaterialFunctionCall(Float2=NamedRerouteUsage()) {materialFunction=&BreakOutFloat2Components}
  n899 = VectorParameter() {group=Effect_Decal_Reactive_Emissive_UV_00, parameterName=EDRec_Emissive_00_Scale_Pickaxe, defaultValue=<rgba 1 1 1 1>}
  NamedRerouteDeclaration(None=LinearInterpolate(A=LinearInterpolate(A=@n895, B=@n897, Alpha=@n898.R), B=Multiply(A=@n899.RGB, B=@n899.RGB), Alpha=@n898.G)) {name=Emissve_00_UVs_Scale, nodeColor=<rgba 0 0.625 1 1>}
  n900 = VectorParameter() {group=Effect_Decal_Reactive_UV_Warp, parameterName=EDRec_Warp_Speed, defaultValue=<rgba 0 0 0 1>}
  n901 = Time()
  n902 = StaticBoolParameter() {defaultValue=true, group=Effect_Decal_Reactive_UV_Warp | Rotation, parameterName=EDRec_Warp_Rotation Disabled, sortPriority=1}
  n903 = TextureSample(
   UVs=MaterialFunctionCall(
    Caculated_Bounds=NamedRerouteUsage()
    is_State=NamedRerouteUsage()
    UV_Rotation=NamedRerouteUsage()
    Uvs_Offset=Multiply(A=@n900.RGB, B=@n901)
    Uvs_TileRate=NamedRerouteUsage()
    No_Rotation=@n902
   ) {materialFunction=&MF_Projection_Side_Normal_V2}.UVs
   Tex=NamedRerouteUsage()
  )
  NamedRerouteDeclaration(None=@n903.RGBA) {name=Warp_Uvs_RGBA, nodeColor=<rgba 0.6406 1 0 1>}
  NamedRerouteDeclaration(None=@n549) {name=Decal_D_State, nodeColor=<rgba 0 0.1797 1 1>}
  NamedRerouteDeclaration(None=@n509) {name=Decal_C_State, nodeColor=<rgba 0 0.1797 1 1>}
  NamedRerouteDeclaration(None=Abs(None=ScalarParameter() {group=Effect_Decal_Reactive_Counter, parameterName=EDRec_Counter__Number})) {name=Count_Number, nodeColor=<rgba 1 0 0.0469 1>}
  n904 = StaticBoolParameter() {group=Effect_Decal_Reactive_Counter, parameterName="EDRec_Counter DecalA---->DecalD", sortPriority=0}
  n905 = StaticSwitch(True=NamedRerouteUsage(), False=NamedRerouteUsage(), Value=@n904)
  n906 = ComponentMask(None=VectorParameter() {group=Effect_Decal_Reactive_Counter, parameterName=EDRec_Counter_UV_Offset}.RGB) {g=true, r=true}
  n907 = VectorParameter() {
   group=Effect_Decal_Reactive_Counter
   parameterName=EDRec_Counter_UV_Scale
   channelNames {a=Uniform, b="", g="", r=""}
   defaultValue=<rgba 1 1 0 0.5>
  }
  n908 = MaterialFunctionCall(
   UVs=MaterialFunctionCall(
    UVs=Add(A=@n905, B=@n906)
    "Rotation Angle (0-1)"=ScalarParameter() {group=Function_Left_Right_Decal V2-Counter, parameterName=FDecalV2_Counter_UV_Rotation}
   ) {materialFunction=&CustomRotator}.Rotated Values
   Texture Scale=ComponentMask(None=Multiply(A=@n907, B=@n907)) {g=true, r=true}
  ) {materialFunction=&ScaleUVsByCenter}
  n909 = Multiply(A=@n908.UVs, B=Constant2Vector() {g=0.1, r=1}) {constA=2, constB=0.1}
  n910 = ComponentMask(None=@n909) {r=true}
  NamedRerouteDeclaration(None=Multiply(
    A=OneMinus(None=@n910)
    B=ScalarParameter() {defaultValue=3, desc="How many Coulums ", group=Effect_Decal_Reactive_Counter, parameterName=EDRec_Counter__Number_Max}
   ) {constB=10}) {name=Count_Max_Number, nodeColor=<rgba 0 0.3672 1 1>}
  NamedRerouteDeclaration(None=@n908.0-1 mask) {name=Count_Uv_Msk, nodeColor=<rgba 0.5703 0 1 1>}
  n911 = ScalarParameter() {defaultValue=1, group=Effect_Decal_Reactive_Uv_Offset_LR, parameterName="EDRec_DecalB_Scale\\Tile_Vehicle"}
  n912 = ScalarParameter() {defaultValue=1, group=Effect_Decal_Reactive_Uv_Offset_LR, parameterName="EDRec_DecalB_Scale\\Tile_Vehicle_Side"}
  n913 = MaterialFunctionCall(Float4=NamedRerouteUsage()) {materialFunction=&BreakOutFloat4Components}
  n914 = LinearInterpolate(
   A=LinearInterpolate(
    A=ScalarParameter() {defaultValue=1, group=Effect_Decal_Reactive_Uv_Offset_LR, parameterName="EDRec_DecalB_Scale\\Tile_Weapon"}
    B=LinearInterpolate(A=@n911, B=@n912, Alpha=@n913.G)
    Alpha=@n913.R
   )
   B=LinearInterpolate(
    A=ScalarParameter() {defaultValue=1, group=Effect_Decal_Reactive_Uv_Offset_LR, parameterName="EDRec_DecalB_Scale\\Tile_Mech"}
    B=ScalarParameter() {defaultValue=1, group=Effect_Decal_Reactive_Uv_Offset_LR, parameterName="EDRec_DecalB_Scale\\Tile_Character"}
    Alpha=@n913.A
   )
   Alpha=@n913.B
  )
  n915 = ScalarParameter() {defaultValue=1, group=Effect_Decal_Reactive_Uv_Offset_LR, parameterName="EDRec_DecalB_Scale\\Tile_Backpack"}
  n916 = MaterialFunctionCall(Float2=NamedRerouteUsage()) {materialFunction=&BreakOutFloat2Components}
  NamedRerouteDeclaration(None=LinearInterpolate(
    A=LinearInterpolate(A=@n914, B=@n915, Alpha=@n916.R)
    B=ScalarParameter() {defaultValue=1, group=Effect_Decal_Reactive_Uv_Offset_LR, parameterName="EDRec_DecalB_Scale\\Tile_Pickaxe"}
    Alpha=@n916.G
   )) {name=UVs_ScaleB, nodeColor=<rgba 0 0.7188 1 1>}
  n917 = NamedRerouteUsage()
  n918 = ComponentMask(None=@n909) {g=true}
  n919 = Constant() {r=10}
  n920 = NamedRerouteUsage()
  n921 = Add(A=Ceil(None=Clamp(None=@n920) {clampMode=CMODE_ClampMin, maxDefault=0}))
  n922 = Frac(None=Divide(A=NamedRerouteUsage(), B=Power(Base=@n919, Exp=@n921) {desc="10-100,000,000"}))
  n923 = Divide(A=Floor(None=Multiply(A=@n922) {constB=100})) {constB=10}
  n924 = NamedRerouteUsage()
  n925 = Divide(A=@n924) {constB=1000000}
  n926 = Clamp(None=NamedRerouteUsage()) {clampMode=CMODE_ClampMax, maxDefault=0}
  n927 = Clamp(None=Subtract(A=Abs(None=@n926)) {desc=shift sampling to make room for decimal place}) {clampMode=CMODE_ClampMin, maxDefault=0}
  n928 = Floor(None=@n927)
  n929 = Frac(None=Multiply(
    A=Add(A=@n924, B=@n925) {constB=0.0001, desc="offset the floating point error precision error... "}
    B=Power(Base=@n919, Exp=@n928) {desc="10-100,000,000"}
   ))
  n930 = Divide(A=Floor(None=Multiply(A=@n929) {constB=10})) {constB=10}
  n931 = Ceil(None=@n927)
  n932 = AppendVector(A=Frac(None=NamedRerouteUsage()), B=LinearInterpolate(A=Add(A=@n918, B=@n923), B=Add(A=@n918, B=@n930), Alpha=Clamp(None=@n931)))
  n933 = NamedRerouteUsage()
  n934 = StaticSwitchParameter(
   True=NamedRerouteUsage()
   False=StaticSwitchParameter(True=NamedRerouteUsage(), False=NamedRerouteUsage()) {desc=Look To TopSide Shared, group=Effect_Decal_Reactive_Counter, parameterName="EDRec_Counter__Tex -->TopS_00B_Sharerd_Tex_Msk"}
  ) {desc=Look To TopSide Shared, group=Effect_Decal_Reactive_Counter, parameterName="EDRec_Counter__Tex -->TopS_00A_Sharerd_Tex_Msk"}
  n935 = TextureSample!(UVs=@n932, Tex=@n934, "DDX(UVs)"=DDX(Value=@n933), "DDY(UVs)"=DDY(Value=@n933), Apply View MipBias) {mipValueMode=TMVM_Derivative}
  n936 = Saturate(None=Multiply(
    A=ChannelMaskParameter(None=@n935.RGBA) {group=Effect_Decal_Reactive_Counter, parameterName=EDRec_Counter_Channel_Msk}.R
    B=NamedRerouteUsage()
   ))
  NamedRerouteDeclaration(None=Saturate(None=Multiply(A=@n917, B=@n936))) {name=Counter, nodeColor=<rgba 1 0.8672 0 1>}
  n937 = ScalarParameter() {defaultValue=-9999999, group=Effect_Decal_Reactive_Counter_Reacive, parameterName=EDRec_Counter_Brightness_Time_BP, sortPriority=212}
  n938 = ScalarParameter() {group=Effect_Decal_Reactive_Counter_Reacive, parameterName=EDRec_Counter_Brightness_Delay, sortPriority=211}
  n939 = Divide(
   A=Subtract(A=Time(), B=Add(A=@n937, B=@n938))
   B=ScalarParameter() {defaultValue=1, group=Effect_Decal_Reactive_Counter_Reacive, parameterName=EDRec_Brightness_Duration, sortPriority=211}
  )
  NamedRerouteDeclaration(None=Power(
    Base=Saturate(None=@n939)
    Exp=ScalarParameter() {defaultValue=1, group=Effect_Decal_Reactive_Counter_Reacive, parameterName=EDRec_Brightness_Falloff, sortPriority=211}
   )) {name=reactive_Count_Brightness, nodeColor=<rgba 0 0.5078 1 1>}
  n940 = ScalarParameter() {defaultValue=1, group=Effect_Decal_Reactive_Counter, parameterName=EDRec_Emissive_Mask_Add_Counter_Brightness, sortPriority=5}
  NamedRerouteDeclaration(None=Multiply(A=@n936, B=@n940)) {name=Counter_output, nodeColor=<rgba 0 1 0.3438 1>}
  NamedRerouteDeclaration(None=@n398.UVs) {name=Uvs_A_LR, nodeColor=<rgba 1 0.1875 0 1>}
  n941 = TextureObjectParameter() {
   group=Effect_Decal_Reactive_Counter
   parameterName=EDRec_Counter__Tex
   texture=/Engine/Functions/Engine_MaterialFunctions02/ExampleContent/DebugNumberStrip
  }
  NamedRerouteDeclaration(None=@n941) {name=Tex_Counter, nodeColor=<rgba 0 0.8125 1 1>}
  NamedRerouteDeclaration(None=@n301) {name=Counter_State, nodeColor=<rgba 1 0.0234 0 1>}
  NamedRerouteDeclaration(None=FunctionInput(Preview=Constant() {r=1}) {bUsePreviewValueAsDefault=true, inputName="Msk_Ts01<--->Ts02", inputType=FunctionInput_Scalar}) {name="Blend TS01--->02", nodeColor=<rgba 0 1 0.6953 1>}
  n942 = ScalarParameter() {group=00-FrontEnd-Overrides, parameterName=00-FE-TODCC-Override}
  n943 = Add(
   A=CollectionParameter() {collection=&FortniteMaterialParameters, parameterName=TimeOfDay_0-1}
   B=@n942
  )
  n944 = CurveAtlasRowParameter(CurveTime=@n943) {
   atlas=&CruveAtlas_Wraps_01
   curve=/Game/Weapons/WeaponSkins/Wraps/Textures/0_ColorCurves/Curve_TimeOfDay
   defaultValue=182
   group=Effect_Decal_Reactive TOD
   parameterName=EDRec_Decal_TODB_CC
  }
  n945 = MaterialFunctionCall(
   Index=ScalarParameter() {defaultValue=182, group=Effect_Decal_Reactive TOD, parameterName="EDRec_Decal_TODB_CC  Index"}
   CCinput=@n944
   CCinput_Alpha=@n944
   UV=@n943
  ) {materialFunction=&MF_Global_CC_Tex}
  n946 = ScalarParameter() {defaultValue=0.5, group=Effect_Decal_Reactive TOD, parameterName=EDRec_Decal_TODB_Morning}
  n947 = ScalarParameter() {group=Effect_Decal_Reactive TOD, parameterName=EDRec_Decal_TODB_Day}
  n948 = Add(A=LinearInterpolate(A=@n942, B=Multiply(A=@n942) {constB=-1}, Alpha=@n947), B=@n947)
  n949 = ScalarParameter() {defaultValue=0.5, group=Effect_Decal_Reactive TOD, parameterName=EDRec_Decal_TODB_Evening}
  n950 = ScalarParameter() {defaultValue=1, group=Effect_Decal_Reactive TOD, parameterName=EDRec_Decal_TODB_Night}
  NamedRerouteDeclaration(None=StaticSwitchParameter(
    True=ChannelMaskParameter(None=@n945.4v) {group=Effect_Decal_Reactive TOD, parameterName=EDRec_Decal_TODB_CC_Channel}.R
    False=MaterialFunctionCall(Value_Morning=@n946, Value_Day=@n948, Value_Evening=@n949, Value_Night=@n950) {materialFunction=&Scalar_Time_of_Day}.Result
   ) {group=Effect_Decal_Reactive TOD, parameterName="EDRec_Decal-TODB --->CCTODB"}) {name=TOD Msk_B, nodeColor=<rgba 0 0.8594 1 1>}
  NamedRerouteDeclaration(None=@n544.UVs) {name=Uvs_D_LR, nodeColor=<rgba 1 0.1406 0 1>}
  NamedRerouteDeclaration(None=@n904) {name=Counter-Uv-Statre, nodeColor=<rgba 0 0.1797 1 1>}
  n951 = ScalarParameter() {group=Effect_Decal_Reactive_Uv_Offset_LR, parameterName=EDRec_DecalB_Rotation_Weapon}
  n952 = MaterialFunctionCall(Float4=NamedRerouteUsage()) {materialFunction=&BreakOutFloat4Components}
  n953 = LinearInterpolate(
   A=ScalarParameter() {group=Effect_Decal_Reactive_Uv_Offset_LR, parameterName=EDRec_DecalB_Rotation_Vehicle}
   B=ScalarParameter() {group=Effect_Decal_Reactive_Uv_Offset_LR, parameterName=EDRec_DecalB_Rotation_Vehicle_Side}
   Alpha=@n952.G
  )
  n954 = ScalarParameter() {group=Effect_Decal_Reactive_Uv_Offset_LR, parameterName=EDRec_DecalB_Rotation_Mech}
  n955 = ScalarParameter() {group=Effect_Decal_Reactive_Uv_Offset_LR, parameterName=EDRec_DecalB_Rotation_Character}
  n956 = MaterialFunctionCall(Float2=NamedRerouteUsage()) {materialFunction=&BreakOutFloat2Components}
  n957 = LinearInterpolate(
   A=LinearInterpolate(A=LinearInterpolate(A=@n951, B=@n953, Alpha=@n952.R), B=LinearInterpolate(A=@n954, B=@n955, Alpha=@n952.A), Alpha=@n952.B)
   B=ScalarParameter() {group=Effect_Decal_Reactive_Uv_Offset_LR, parameterName=EDRec_DecalB_Rotation_Backpack}
   Alpha=@n956.R
  )
  n958 = ScalarParameter() {group=Effect_Decal_Reactive_Uv_Offset_LR, parameterName=EDRec_DecalB_Rotation_Pickaxe}
  NamedRerouteDeclaration(None=Add(A=LinearInterpolate(A=@n957, B=@n958, Alpha=@n956.G), B=NamedRerouteUsage())) {name=Uvs_RotationB, nodeColor=<rgba 0 0.7188 1 1>}
  n959 = Constant()
  n960 = NamedRerouteUsage()
  n961 = LinearInterpolate(
   A=@n960
   B=OneMinus(None=@n960)
   Alpha=ScalarParameter() {group=Effect_Decal_Reactive_RE, parameterName="EDRec_DecalC_Opacity-TopS Blend 01<--->02 MSk   1=invert"}
  )
  n962 = NamedRerouteUsage()
  n963 = Constant()
  n964 = NamedRerouteUsage()
  n965 = Constant() {r=0.5}
  n966 = NamedRerouteUsage()
  n967 = ChannelMaskParameter(None=@n646.RGBA) {group=Effect_Decal_Reactive_UV_Warp, parameterName=EDRec_WarpC_Msk_Channel}
  n968 = ScalarParameter() {group=Effect_Decal_Reactive_UV_Warp, parameterName=EDRec_WarpC_Offset}
  n969 = NamedRerouteUsage()
  n970 = ScalarParameter() {group=Effect_Decal_Reactive_UV_Warp, parameterName=EDRec_WarpC_Strength}
  n971 = ScalarParameter() {group=Effect_Decal_Reactive_UV_Warp, parameterName=EDRec_WarpC_Sine_Speed}
  n972 = Sine(None=Multiply(A=Time(), B=@n971))
  n973 = LinearInterpolate(A=@n970, B=Multiply(A=Add(A=@n972), B=@n970) {constB=0.25}, Alpha=Saturate(None=Ceil(None=@n971)))
  n974 = NamedRerouteUsage()
  n975 = StaticSwitch(
   True=Multiply(A=NamedRerouteUsage(), B=ComponentMask(None=@n966) {b=true})
   False=Multiply(A=Add(A=@n967.R, B=@n968), B=StaticSwitch(True=@n969, False=@n973, Value=@n974))
   Value=NamedRerouteUsage()
  )
  n976 = NamedRerouteUsage()
  n977 = VectorParameter() {group=Effect_Decal_Reactive_UV_Warp | Axis, parameterName=EDRec_WarpC_Axis, channelNames {a="", b="", g=Y, r=X}, defaultValue=<rgba 1 1 0 1>}
  n978 = Add(A=Multiply(A=Multiply(A=@n975, B=@n976), B=AppendVector(A=@n977.X, B=@n977.X)), B=@n628)
  n979 = StaticSwitch(
   True=@n978
   False=StaticSwitchParameter(True=@n978, False=@n628) {group=Effect_Decal_Reactive_UV_Warp, parameterName=EDRec_Warp_Uvs, sortPriority=0}
   Value=NamedRerouteUsage()
  )
  n980 = MaterialFunctionCall(Float2=@n979) {materialFunction=&BreakOutFloat2Components}
  n981 = Multiply(A=SphereMask(A=@n965, B=@n980.R) {attenuationRadius=0.5}, B=SphereMask(A=@n980.G, B=@n965) {attenuationRadius=0.5})
  n982 = NamedRerouteUsage()
  n983 = VectorParameter() {
   group=Effect_Decal_Reactive_FlipBook
   parameterName=EDRec_FlipBook_Frame__Offset
   channelNames {a=FrameD, b=FrameC, g=FrameB, r=FrameA}
   defaultValue=<rgba 0 0 0 1>
  }
  n984 = VectorParameter() {
   group=Effect_Decal_Reactive_FlipBook
   parameterName=EDRec_FlipBook_FrameC_Speed--
   channelNames {a=Clamp Max, b=Clamp Min, g="Random  1=Active", r=Speed}
   defaultValue=<rgba 0 0 0 1>
  }
  n985 = Multiply(A=Add(A=Time(), B=@n983.FrameC), B=@n984.Speed)
  n986 = Frac(None=@n985)
  n987 = MaterialFunctionCall(2 Vector=Add(A=Ceil(None=@n985), B=Constant2Vector() {g=0.1258, r=0.315})) {materialFunction=&MF_Random_2}
  n988 = NamedRerouteUsage()
  n989 = NamedRerouteUsage()
  n990 = Add(
   A=LinearInterpolate(
    A=Clamp(None=@n986, Min=@n984.Clamp Min, Max=@n984.Clamp Min)
    B=Clamp(None=@n987.utput, Min=@n984.Clamp Min, Max=@n984.Clamp Min)
    Alpha=@n984."Random  1=Active"
   )
   B=Divide(
    A=ScalarParameter() {defaultValue=2, group=Effect_Decal_Reactive_FlipBook, parameterName=EDRec_FlipBook_FrameC}
    B=Multiply(A=@n989, B=@n988)
   ) {constA=9, constB=9}
  )
  n991 = NamedRerouteUsage()
  n992 = NamedRerouteUsage()
  n993 = MaterialFunctionCall(
   "Animation  Phase (0-1)"=StaticSwitch(True=NamedRerouteUsage(), False=StaticSwitch(True=@n982, False=@n990, Value=@n991), Value=NamedRerouteUsage())
   Number of Rows=@n989
   Number of Columns=@n988
   Texture=@n992
   UVs=@n979
   MipBias/Level=NamedRerouteUsage()
   "Clamp Anim (see tooltip)"=StaticBool() {value=true}
  ) {materialFunction=&FlipBook}
  n994 = LinearInterpolate(
   A=@n993.UVs
   B=Frac(None=@n993.UVs)
   Alpha=ScalarParameter() {desc=Use with clamped Textures, group=Effect_Decal_Reactive_FlipBook, parameterName="EDRec_FlipBook_FrameC_1=Frac"}
  )
  n995 = NamedRerouteUsage()
  n996 = DDX(Value=@n995)
  n997 = DDY(Value=@n995)
  n998 = TextureSample(UVs=@n994, Tex=@n992, "DDX(UVs)"=@n996, "DDY(UVs)"=@n997) {mipValueMode=TMVM_Derivative}
  n999 = NamedRerouteUsage()
  n1000 = TextureSample!(UVs=@n994, Tex=RecordTextureStreamingInfo(Tex=@n992, UVs=Multiply(A=@n999, B=@n995)).Tex, Apply View MipBias)
  n1001 = StaticSwitch(True=@n998.A, False=@n1000.A, Value=@n627)
  n1002 = Multiply(A=@n1001, B=@n981)
  n1003 = NamedRerouteDeclaration(None=StaticSwitchParameter(True=StaticSwitch(True=@n962, False=@n963, Value=@n964), False=StaticSwitch(True=@n1002, False=@n1001, Value=@n627)) {group=Effect_Decal_Reactive, parameterName=EDRec_Emissive_only, sortPriority=-10}) {name=Decal_C_Alpha, nodeColor=<rgba 1 0 0.7734 1>}
  n1004 = ScalarParameter() {
   defaultValue=0.1
   desc=the point wich you see active to non active.
   group=Effect_Decal_Reactive_RE_Audio
   parameterName=EDRec Audio Opacity_FadePoint
  }
  n1005 = NamedRerouteUsage()
  n1006 = Max(A=SmoothStep(Max=@n1004, Value=NamedRerouteUsage()) {constMax=0.1}, B=@n1005)
  n1007 = NamedRerouteUsage()
  n1008 = StaticSwitch(True=Multiply(A=@n1003, B=StaticSwitch(True=@n1006, False=@n1005, Value=@n1007)), False=@n1003, Value=NamedRerouteUsage())
  n1009 = NamedRerouteUsage()
  n1010 = NamedRerouteUsage()
  n1011 = StaticSwitch(
   True=LinearInterpolate(
    B=@n1008
    Alpha=StaticSwitchParameter(True=@n1009, False=@n1010) {group=Effect_Decal_Reactive_RE, parameterName="EDRec_DecaC_Opacity-TOD----->TODB", sortPriority=-1}
   ) {constB=0}
   False=@n1008
   Value=StaticBoolParameter() {group=Effect_Decal_Reactive_RE, parameterName=EDRec_DecalC_Opacity-TOD, sortPriority=-1}
  )
  n1012 = StaticSwitch(
   True=StaticSwitch(
    True=Multiply(A=@n961, B=@n1011)
    False=@n1011
    Value=StaticBoolParameter() {group=Effect_Decal_Reactive_RE, parameterName="EDRec_DecalC_Opacity-TopS Blend 01<--->02 MSk"}
   )
   False=@n1011
   Value=StaticBoolParameter() {group="TopSide_V02_Layer 01 Blend  02", parameterName="TopS Blend 01<--->02"}
  )
  n1013 = StaticSwitch(True=@n959, False=@n1012, Value=@n509)
  n1014 = NamedRerouteUsage()
  n1015 = LinearInterpolate(
   A=@n1014
   B=OneMinus(None=@n1014)
   Alpha=ScalarParameter() {group=Effect_Decal_Reactive_RE, parameterName="EDRec_DecalB_Opacity-TopS Blend 01<--->02 MSk   1=invert"}
  )
  n1016 = ScalarParameter() {
   defaultValue=0.1
   desc=the point wich you see active to non active.
   group=Effect_Decal_Reactive_RE_Audio
   parameterName=EDRec Audio Opacity_FadePoint
  }
  n1017 = NamedRerouteUsage()
  n1018 = Max(A=SmoothStep(Max=@n1016, Value=NamedRerouteUsage()) {constMax=0.1}, B=@n1017)
  n1019 = NamedRerouteUsage()
  n1020 = StaticSwitch(True=Multiply(A=@n687, B=StaticSwitch(True=@n1018, False=@n1017, Value=@n1019)), False=@n687, Value=NamedRerouteUsage())
  n1021 = NamedRerouteUsage()
  n1022 = NamedRerouteUsage()
  n1023 = StaticSwitch(
   True=LinearInterpolate(
    B=@n1020
    Alpha=StaticSwitchParameter(True=@n1021, False=@n1022) {group=Effect_Decal_Reactive_RE, parameterName="EDRec_DecalB_Opacity-TOD----->TODB", sortPriority=-1}
   ) {constB=0}
   False=@n1020
   Value=StaticBoolParameter() {group=Effect_Decal_Reactive_RE, parameterName=EDRec_DecalB_Opacity-TOD, sortPriority=-1}
  )
  n1024 = StaticSwitch(
   True=StaticSwitch(
    True=Multiply(A=@n1015, B=@n1023)
    False=@n1023
    Value=StaticBoolParameter() {group=Effect_Decal_Reactive_RE, parameterName="EDRec_DecalB_Opacity-TopS Blend 01<--->02 MSk"}
   )
   False=@n1023
   Value=StaticBoolParameter() {group="TopSide_V02_Layer 01 Blend  02", parameterName="TopS Blend 01<--->02"}
  )
  n1025 = NamedRerouteUsage()
  n1026 = LinearInterpolate(
   A=@n1025
   B=OneMinus(None=@n1025)
   Alpha=ScalarParameter() {group=Effect_Decal_Reactive_RE, parameterName="EDRec_DecalA_Opacity-TopS Blend 01<--->02 MSk   1=invert"}
  )
  n1027 = ScalarParameter() {
   defaultValue=0.1
   desc=the point wich you see active to non active.
   group=Effect_Decal_Reactive_RE_Audio
   parameterName=EDRec Audio Opacity_FadePoint
  }
  n1028 = NamedRerouteUsage()
  n1029 = Max(A=SmoothStep(Max=@n1027, Value=NamedRerouteUsage()) {constMax=0.1}, B=@n1028)
  n1030 = NamedRerouteUsage()
  n1031 = NamedRerouteUsage()
  n1032 = Constant()
  n1033 = NamedRerouteUsage()
  n1034 = Constant() {r=0.5}
  n1035 = NamedRerouteUsage()
  n1036 = ChannelMaskParameter(None=@n641.RGBA) {group=Effect_Decal_Reactive_UV_Warp, parameterName=EDRec_WarpA_Msk_Channel}
  n1037 = ScalarParameter() {group=Effect_Decal_Reactive_UV_Warp, parameterName=EDRec_WarpA_Offset}
  n1038 = NamedRerouteUsage()
  n1039 = ScalarParameter() {group=Effect_Decal_Reactive_UV_Warp, parameterName=EDRec_WarpA_Strength}
  n1040 = ScalarParameter() {group=Effect_Decal_Reactive_UV_Warp, parameterName=EDRec_WarpA_Sine_Speed}
  n1041 = Sine(None=Multiply(A=Time(), B=@n1040))
  n1042 = LinearInterpolate(A=@n1039, B=Multiply(A=Add(A=@n1041), B=@n1039) {constB=0.25}, Alpha=Saturate(None=Ceil(None=@n1040)))
  n1043 = NamedRerouteUsage()
  n1044 = StaticSwitch(
   True=Multiply(A=NamedRerouteUsage(), B=ComponentMask(None=@n1035) {r=true})
   False=Multiply(A=Add(A=@n1036.R, B=@n1037), B=StaticSwitch(True=@n1038, False=@n1042, Value=@n1043))
   Value=NamedRerouteUsage()
  )
  n1045 = NamedRerouteUsage()
  n1046 = VectorParameter() {group=Effect_Decal_Reactive_UV_Warp | Axis, parameterName=EDRec_WarpA_Axis, channelNames {a="", b="", g=Y, r=X}, defaultValue=<rgba 1 1 0 1>}
  n1047 = Add(A=Multiply(A=Multiply(A=@n1044, B=@n1045), B=AppendVector(A=@n1046.X, B=@n1046.X)), B=@n403)
  n1048 = StaticSwitch(
   True=@n1047
   False=StaticSwitchParameter(True=@n1047, False=@n403) {group=Effect_Decal_Reactive_UV_Warp, parameterName=EDRec_Warp_Uvs, sortPriority=0}
   Value=NamedRerouteUsage()
  )
  n1049 = MaterialFunctionCall(Float2=@n1048) {materialFunction=&BreakOutFloat2Components}
  n1050 = Multiply(
   A=Multiply(A=SphereMask(A=@n1034, B=@n1049.R) {attenuationRadius=0.5}, B=SphereMask(A=@n1049.G, B=@n1034) {attenuationRadius=0.5})
   B=NamedRerouteUsage()
  )
  n1051 = NamedRerouteUsage()
  n1052 = VectorParameter() {
   group=Effect_Decal_Reactive_FlipBook
   parameterName=EDRec_FlipBook_Frame__Offset
   channelNames {a=FrameD, b=FrameC, g=FrameB, r=FrameA}
   defaultValue=<rgba 0 0 0 1>
  }
  n1053 = VectorParameter() {
   group=Effect_Decal_Reactive_FlipBook
   parameterName=EDRec_FlipBook_FrameA_Speed--
   channelNames {a=Clamp Max, b=Clamp Min, g="Random  1=Active", r=Speed}
   defaultValue=<rgba 0 0 0 1>
  }
  n1054 = Multiply(A=Add(A=Time(), B=@n1052.FrameA), B=@n1053.Speed)
  n1055 = StaticSwitch(True=Add(A=@n1054, B=NamedRerouteUsage()), False=@n1054, Value=@n402)
  n1056 = Frac(None=@n1055)
  n1057 = MaterialFunctionCall(2 Vector=Add(A=Ceil(None=@n1055), B=Constant2Vector() {g=0.2062, r=0.0536})) {materialFunction=&MF_Random_2}
  n1058 = NamedRerouteUsage()
  n1059 = NamedRerouteUsage()
  n1060 = Add(
   A=LinearInterpolate(
    A=Clamp(None=@n1056, Min=@n1053.Clamp Min, Max=@n1053.Clamp Min)
    B=Clamp(None=@n1057.utput, Min=@n1053.Clamp Min, Max=@n1053.Clamp Min)
    Alpha=@n1053."Random  1=Active"
   )
   B=Divide(A=ScalarParameter() {group=Effect_Decal_Reactive_FlipBook, parameterName=EDRec_FlipBook_FrameA}, B=Multiply(A=@n1058, B=@n1059)) {constA=9, constB=9}
  )
  n1061 = NamedRerouteUsage()
  n1062 = NamedRerouteUsage()
  n1063 = MaterialFunctionCall(
   "Animation  Phase (0-1)"=StaticSwitch(True=NamedRerouteUsage(), False=StaticSwitch(True=@n1051, False=@n1060, Value=@n1061), Value=NamedRerouteUsage())
   Number of Rows=@n1058
   Number of Columns=@n1059
   Texture=@n1062
   UVs=@n1048
   MipBias/Level=NamedRerouteUsage()
   "Clamp Anim (see tooltip)"=StaticBool() {value=true}
  ) {materialFunction=&FlipBook}
  n1064 = LinearInterpolate(
   A=@n1063.UVs
   B=Frac(None=@n1063.UVs)
   Alpha=ScalarParameter() {desc=Use with clamped Textures, group=Effect_Decal_Reactive_FlipBook, parameterName="EDRec_FlipBook_FrameA_1=Frac"}
  )
  n1065 = NamedRerouteUsage()
  n1066 = DDX(Value=@n1065)
  n1067 = DDY(Value=@n1065)
  n1068 = TextureSample!(UVs=@n1064, Tex=@n1062, "DDX(UVs)"=@n1066, "DDY(UVs)"=@n1067, Apply View MipBias) {mipValueMode=TMVM_Derivative}
  n1069 = NamedRerouteDeclaration(None=Power(Base=Constant() {r=2}, Exp=ScalarParameter() {group=Effect_Decal_Reactive_FlipBook, parameterName=EDRec_FlipBook_MipBias})) {name=MipBias, nodeColor=<rgba 0 1 0.3438 1>}
  n1070 = TextureSample!(UVs=@n1064, Tex=RecordTextureStreamingInfo(Tex=@n1062, UVs=Multiply(A=@n1069, B=@n1065)).Tex, Apply View MipBias)
  n1071 = StaticSwitch(True=@n1068.A, False=@n1070.A, Value=@n402)
  n1072 = Multiply(A=@n1071, B=@n1050)
  n1073 = NamedRerouteDeclaration(None=StaticSwitchParameter(True=StaticSwitch(True=@n1031, False=@n1032, Value=@n1033), False=StaticSwitch(True=@n1072, False=@n1071, Value=@n402)) {group=Effect_Decal_Reactive, parameterName=EDRec_Emissive_only, sortPriority=-10}) {name=Decal_A_Alpha, nodeColor=<rgba 1 0 0.7734 1>}
  n1074 = StaticSwitch(True=Multiply(A=StaticSwitch(True=@n1029, False=@n1028, Value=@n1030), B=@n1073), False=@n1073, Value=NamedRerouteUsage())
  n1075 = NamedRerouteUsage()
  n1076 = NamedRerouteUsage()
  n1077 = StaticSwitch(
   True=LinearInterpolate(
    B=@n1074
    Alpha=StaticSwitchParameter(True=@n1075, False=@n1076) {group=Effect_Decal_Reactive_RE, parameterName="EDRec_DecalA_Opacity-TOD----->TODB", sortPriority=-1}
   ) {constB=0}
   False=@n1074
   Value=StaticBoolParameter() {group=Effect_Decal_Reactive_RE, parameterName=EDRec_DecalA_Opacity-TOD, sortPriority=-1}
  )
  n1078 = StaticSwitch(
   True=StaticSwitch(
    True=Multiply(A=@n1026, B=@n1077)
    False=@n1077
    Value=StaticBoolParameter() {group=Effect_Decal_Reactive_RE, parameterName="EDRec_DecalA_Opacity-TopS Blend 01<--->02 MSk"}
   )
   False=@n1077
   Value=StaticBoolParameter() {group="TopSide_V02_Layer 01 Blend  02", parameterName="TopS Blend 01<--->02"}
  )
  n1079 = Add(A=@n1078, B=@n1024)
  NamedRerouteDeclaration(None=Saturate(None=Add(A=@n1013, B=@n1079))) {name=Alpha_Set_B without D, nodeColor=<rgba 0.8047 0 1 1>}
  NamedRerouteDeclaration(None=@n46) {name=Add_Counter_Stat_D, nodeColor=<rgba 0 0.2969 1 1>}
  NamedRerouteDeclaration(None=@n270) {name=Emissive D State, nodeColor=<rgba 0 0.9297 1 1>}
  NamedRerouteDeclaration(None=@n1078) {name=Alpha_A_Msk, nodeColor=<rgba 0 0.0625 1 1>}
  NamedRerouteDeclaration(None=@n1024) {name=Alpha_B_Msk, nodeColor=<rgba 0 0.0625 1 1>}
  NamedRerouteDeclaration(None=@n1012) {name=Alpha_C_Msk, nodeColor=<rgba 0 0.0625 1 1>}
  n1080 = NamedRerouteUsage()
  n1081 = LinearInterpolate(
   A=@n1080
   B=OneMinus(None=@n1080)
   Alpha=ScalarParameter() {group=Effect_Decal_Reactive_RE, parameterName="EDRec_DecalD_Opacity-TopS Blend 01<--->02 MSk   1=invert"}
  )
  n1082 = NamedRerouteUsage()
  n1083 = Constant()
  n1084 = NamedRerouteUsage()
  n1085 = NamedRerouteUsage()
  n1086 = VectorParameter() {
   group=Effect_Decal_Reactive_FlipBook
   parameterName=EDRec_FlipBook_Frame__Offset
   channelNames {a=FrameD, b=FrameC, g=FrameB, r=FrameA}
   defaultValue=<rgba 0 0 0 1>
  }
  n1087 = VectorParameter() {
   group=Effect_Decal_Reactive_FlipBook
   parameterName=EDRec_FlipBook_FrameD_Speed--
   channelNames {a=Clamp Max, b=Clamp Min, g="Random  1=Active", r=Speed}
   defaultValue=<rgba 0 0 0 1>
  }
  n1088 = Multiply(A=Add(A=@n1086.FrameD, B=Time()), B=@n1087.Speed)
  n1089 = Frac(None=@n1088)
  n1090 = MaterialFunctionCall(2 Vector=Add(A=Ceil(None=@n1088), B=Constant2Vector() {g=0.0216, r=0.0347})) {materialFunction=&MF_Random_2}
  n1091 = NamedRerouteUsage()
  n1092 = NamedRerouteUsage()
  n1093 = Add(
   A=LinearInterpolate(
    A=Clamp(None=@n1089, Min=@n1087.Clamp Min, Max=@n1087.Clamp Min)
    B=Clamp(None=@n1090.utput, Min=@n1087.Clamp Min, Max=@n1087.Clamp Min)
    Alpha=@n1087."Random  1=Active"
   )
   B=Divide(
    A=ScalarParameter() {defaultValue=3, group=Effect_Decal_Reactive_FlipBook, parameterName=EDRec_FlipBook_FrameD}
    B=Multiply(A=@n1091, B=@n1092)
   ) {constA=9, constB=9}
  )
  n1094 = NamedRerouteUsage()
  n1095 = NamedRerouteUsage()
  n1096 = NamedRerouteUsage()
  n1097 = NamedRerouteUsage()
  n1098 = ChannelMaskParameter(None=@n651.RGBA) {group=Effect_Decal_Reactive_UV_Warp, parameterName=EDRec_WarpD_Msk_Channel}
  n1099 = ScalarParameter() {group=Effect_Decal_Reactive_UV_Warp, parameterName=EDRec_WarpD_Offset}
  n1100 = NamedRerouteUsage()
  n1101 = ScalarParameter() {group=Effect_Decal_Reactive_UV_Warp, parameterName=EDRec_WarpD_Strength}
  n1102 = ScalarParameter() {group=Effect_Decal_Reactive_UV_Warp, parameterName=EDRec_WarpD_Sine_Speed}
  n1103 = Sine(None=Multiply(A=Time(), B=@n1102))
  n1104 = LinearInterpolate(A=@n1101, B=Multiply(A=Add(A=@n1103), B=@n1101) {constB=0.25}, Alpha=Saturate(None=Ceil(None=@n1102)))
  n1105 = NamedRerouteUsage()
  n1106 = StaticSwitch(
   True=Multiply(A=NamedRerouteUsage(), B=ComponentMask(None=@n1097) {a=true})
   False=Multiply(A=Add(A=@n1098.R, B=@n1099), B=StaticSwitch(True=@n1100, False=@n1104, Value=@n1105))
   Value=NamedRerouteUsage()
  )
  n1107 = VectorParameter() {group=Effect_Decal_Reactive_UV_Warp | Axis, parameterName=EDRec_WarpD_Axis, channelNames {a="", b="", g=Y, r=X}, defaultValue=<rgba 1 1 0 1>}
  n1108 = Add(A=Multiply(A=Multiply(A=@n1096, B=@n1106), B=AppendVector(A=@n1107.X, B=@n1107.X)), B=@n630)
  n1109 = StaticSwitch(
   True=@n1108
   False=StaticSwitchParameter(True=@n1108, False=@n630) {group=Effect_Decal_Reactive_UV_Warp, parameterName=EDRec_Warp_Uvs, sortPriority=0}
   Value=NamedRerouteUsage()
  )
  n1110 = MaterialFunctionCall(
   "Animation  Phase (0-1)"=StaticSwitch(True=NamedRerouteUsage(), False=StaticSwitch(True=@n1085, False=@n1093, Value=@n1094), Value=NamedRerouteUsage())
   Number of Rows=@n1091
   Number of Columns=@n1092
   Texture=@n1095
   UVs=@n1109
   MipBias/Level=NamedRerouteUsage()
   "Clamp Anim (see tooltip)"=StaticBool() {value=true}
  ) {materialFunction=&FlipBook}
  n1111 = LinearInterpolate(
   A=@n1110.UVs
   B=Frac(None=@n1110.UVs)
   Alpha=ScalarParameter() {desc=Use with clamped Textures, group=Effect_Decal_Reactive_FlipBook, parameterName="EDRec_FlipBook_FrameD_1=Frac"}
  )
  n1112 = NamedRerouteUsage()
  n1113 = NamedRerouteUsage()
  n1114 = TextureSample(UVs=@n1111, Tex=RecordTextureStreamingInfo(Tex=@n1095, UVs=Multiply(A=@n1112, B=@n1113)).Tex)
  n1115 = Constant() {r=0.5}
  n1116 = MaterialFunctionCall(Float2=@n1109) {materialFunction=&BreakOutFloat2Components}
  n1117 = Multiply(A=SphereMask(A=@n1115, B=@n1116.R) {attenuationRadius=0.5}, B=SphereMask(A=@n1116.G, B=@n1115) {attenuationRadius=0.5})
  n1118 = Multiply(A=@n1114.A, B=@n1117)
  n1119 = NamedRerouteDeclaration(None=StaticSwitchParameter(True=StaticSwitch(True=@n1082, False=@n1083, Value=@n1084), False=StaticSwitch(True=@n1118, False=@n1114.A, Value=@n629)) {group=Effect_Decal_Reactive, parameterName=EDRec_Emissive_only, sortPriority=-10}) {name=Decal_D_Alpha, nodeColor=<rgba 1 0 0.7734 1>}
  n1120 = ScalarParameter() {
   defaultValue=0.1
   desc=the point wich you see active to non active.
   group=Effect_Decal_Reactive_RE_Audio
   parameterName=EDRec Audio Opacity_FadePoint
  }
  n1121 = NamedRerouteUsage()
  n1122 = Max(A=SmoothStep(Max=@n1120, Value=NamedRerouteUsage()) {constMax=0.1}, B=@n1121)
  n1123 = NamedRerouteUsage()
  n1124 = StaticSwitch(True=Multiply(A=@n1119, B=StaticSwitch(True=@n1122, False=@n1121, Value=@n1123)), False=@n1119, Value=NamedRerouteUsage())
  n1125 = NamedRerouteUsage()
  n1126 = NamedRerouteUsage()
  n1127 = StaticSwitch(
   True=LinearInterpolate(
    B=@n1124
    Alpha=StaticSwitchParameter(True=@n1125, False=@n1126) {group=Effect_Decal_Reactive_RE, parameterName="EDRec_DecalD_Opacity-TOD----->TODB", sortPriority=-1}
   ) {constB=0}
   False=@n1124
   Value=StaticBoolParameter() {group=Effect_Decal_Reactive_RE, parameterName=EDRec_DecalD_Opacity-TOD, sortPriority=-1}
  )
  n1128 = StaticSwitch(
   True=StaticSwitch(
    True=Multiply(A=@n1081, B=@n1127)
    False=@n1127
    Value=StaticBoolParameter() {group=Effect_Decal_Reactive_RE, parameterName="EDRec_DecalD_Opacity-TopS Blend 01<--->02 MSk"}
   )
   False=@n1127
   Value=StaticBoolParameter() {group="TopSide_V02_Layer 01 Blend  02", parameterName="TopS Blend 01<--->02"}
  )
  NamedRerouteDeclaration(None=@n1128) {name=Alpha_D_Msk, nodeColor=<rgba 0 0.0625 1 1>}
  n1129 = VectorParameter() {
   group=Effect_Decal_Reactive_FlipBook----
   parameterName=EDRec_Decal_A_Animation--
   channelNames {a=Clamp Max, b=Clamp Min Frames, g="", r=Loop}
   defaultValue=<rgba 1 0 0 1>
  }
  n1130 = Time()
  n1131 = Add(
   A=ScalarParameter() {defaultValue=-9999999, group=Effect_Decal_Reactive_FlipBook----, parameterName=EDRec_Decal_A_Animation_Time_BP}
   B=ScalarParameter() {group=Effect_Decal_Reactive_FlipBook----, parameterName=EDRec_Decal_A_Animation_Delay}
  )
  n1132 = Saturate(None=Divide(
    A=Subtract(A=@n1130, B=@n1131)
    B=ScalarParameter() {defaultValue=1, group=Effect_Decal_Reactive_FlipBook----, parameterName=EDRec_Decal_A_Animation_Duration}
   ))
  n1133 = ScalarParameter() {defaultValue=1, group=Effect_Decal_Reactive_FlipBook----, parameterName=EDRec_Decal_A_Animation_Falloff}
  n1134 = Multiply(A=Multiply(A=Power(Base=@n1132, Exp=@n1133)) {constB=1}, B=@n1129.Loop)
  NamedRerouteDeclaration(None=LinearInterpolate(A=@n1129.Clamp Min Frames, B=@n1129.Clamp Min Frames, Alpha=Frac(None=@n1134))) {name=reactive_A_ainim, nodeColor=<rgba 0 0.5078 1 1>}
  n1135 = VectorParameter() {
   group=Effect_Decal_Reactive_FlipBook----
   parameterName=EDRec_Decal_B_Animation--
   channelNames {a=Clamp Max, b=Clamp Min Frames, g="", r=Loop}
   defaultValue=<rgba 1 0 0 1>
  }
  n1136 = Time()
  n1137 = Add(
   A=ScalarParameter() {defaultValue=-9999999, group=Effect_Decal_Reactive_FlipBook----, parameterName=EDRec_Decal_B_Animation_Time_BP}
   B=ScalarParameter() {group=Effect_Decal_Reactive_FlipBook----, parameterName=EDRec_Decal_B_Animation_Delay}
  )
  n1138 = Saturate(None=Divide(
    A=Subtract(A=@n1136, B=@n1137)
    B=ScalarParameter() {defaultValue=1, group=Effect_Decal_Reactive_FlipBook----, parameterName=EDRec_Decal_B_Animation_Duration}
   ))
  n1139 = ScalarParameter() {defaultValue=1, group=Effect_Decal_Reactive_FlipBook----, parameterName=EDRec_Decal_B_Animation_Falloff}
  n1140 = Multiply(A=Multiply(A=Power(Base=@n1138, Exp=@n1139)) {constB=1}, B=@n1135.Loop)
  NamedRerouteDeclaration(None=LinearInterpolate(A=@n1135.Clamp Min Frames, B=@n1135.Clamp Min Frames, Alpha=Frac(None=@n1140))) {name=reactive_B_ainim, nodeColor=<rgba 0 0.5078 1 1>}
  n1141 = VectorParameter() {
   group=Effect_Decal_Reactive_FlipBook----
   parameterName=EDRec_Decal_C_Animation--
   channelNames {a=Clamp Max, b=Clamp Min Frames, g="", r=Loop}
   defaultValue=<rgba 1 0 0 1>
  }
  n1142 = Time()
  n1143 = Add(
   A=ScalarParameter() {defaultValue=-9999999, group=Effect_Decal_Reactive_FlipBook----, parameterName=EDRec_Decal_C_Animation_Time_BP}
   B=ScalarParameter() {group=Effect_Decal_Reactive_FlipBook----, parameterName=EDRec_Decal_C_Animation_Delay}
  )
  n1144 = Saturate(None=Divide(
    A=Subtract(A=@n1142, B=@n1143)
    B=ScalarParameter() {defaultValue=1, group=Effect_Decal_Reactive_FlipBook----, parameterName=EDRec_Decal_C_Animation_Duration}
   ))
  n1145 = ScalarParameter() {defaultValue=1, group=Effect_Decal_Reactive_FlipBook----, parameterName=EDRec_Decal_C_Animation_Falloff}
  n1146 = Multiply(A=Multiply(A=Power(Base=@n1144, Exp=@n1145)) {constB=1}, B=@n1141.Loop)
  NamedRerouteDeclaration(None=LinearInterpolate(A=@n1141.Clamp Min Frames, B=@n1141.Clamp Min Frames, Alpha=Frac(None=@n1146))) {name=reactive_C_ainim, nodeColor=<rgba 0 0.5078 1 1>}
  n1147 = VectorParameter() {group=Effect_Decal_Reactive_Uv_Offset_LR, parameterName=EDRec_DecalC_Offset_Weapon, defaultValue=<rgba 0 0 0 0.25>}
  n1148 = VectorParameter() {group=Effect_Decal_Reactive_Uv_Offset_LR, parameterName=EDRec_DecalC_Offset_Vehicle, defaultValue=<rgba 0 0 0 0.25>}
  n1149 = VectorParameter() {group=Effect_Decal_Reactive_Uv_Offset_LR, parameterName=EDRec_DecalC_Offset_VehicleSide, defaultValue=<rgba 0 0 0 0.25>}
  n1150 = MaterialFunctionCall(Float4=NamedRerouteUsage()) {materialFunction=&BreakOutFloat4Components}
  n1151 = LinearInterpolate(
   A=AppendVector(A=@n1147.RGB, B=@n1147.RGB)
   B=LinearInterpolate(A=AppendVector(A=@n1148.RGB, B=@n1148.RGB), B=AppendVector(A=@n1149.RGB, B=@n1149.RGB), Alpha=@n1150.G)
   Alpha=@n1150.R
  )
  n1152 = VectorParameter() {group=Effect_Decal_Reactive_Uv_Offset_LR, parameterName=EDRec_DecalC_Offset_Mech, defaultValue=<rgba 0 0 0 0.25>}
  n1153 = VectorParameter() {group=Effect_Decal_Reactive_Uv_Offset_LR, parameterName=EDRec_DecalC_Offset_Character, defaultValue=<rgba 0 0 0 0.25>}
  n1154 = LinearInterpolate(A=AppendVector(A=@n1152.RGB, B=@n1152.RGB), B=AppendVector(A=@n1153.RGB, B=@n1153.RGB), Alpha=@n1150.A)
  n1155 = VectorParameter() {group=Effect_Decal_Reactive_Uv_Offset_LR, parameterName="EDRec_DecalC_Offset_Backpack----A=Scale", defaultValue=<rgba 0 0 0 0.25>}
  n1156 = MaterialFunctionCall(Float2=NamedRerouteUsage()) {materialFunction=&BreakOutFloat2Components}
  n1157 = VectorParameter() {group=Effect_Decal_Reactive_Uv_Offset_LR, parameterName=EDRec_DecalC_Offset_Pickaxe, defaultValue=<rgba 0 0 0 0.25>}
  n1158 = LinearInterpolate(
   A=LinearInterpolate(A=LinearInterpolate(A=@n1151, B=@n1154, Alpha=@n1150.B), B=AppendVector(A=@n1155.RGB, B=@n1155.RGB), Alpha=@n1156.R)
   B=AppendVector(A=@n1157.RGB, B=@n1157.RGB)
   Alpha=@n1156.G
  )
  n1159 = VectorParameter() {group=Effect_Decal_Reactive_Uv_Offset_LR, parameterName=EDRec_DecalC_Offset_Weapon End, defaultValue=<rgba 0 0 0 0.25>}
  n1160 = VectorParameter() {group=Effect_Decal_Reactive_Uv_Offset_LR, parameterName=EDRec_DecalC_Offset_Vehicle End, defaultValue=<rgba 0 0 0 0.25>}
  n1161 = VectorParameter() {group=Effect_Decal_Reactive_Uv_Offset_LR, parameterName=EDRec_DecalC_Offset_VehicleSide End, defaultValue=<rgba 0 0 0 0.25>}
  n1162 = MaterialFunctionCall(Float4=NamedRerouteUsage()) {materialFunction=&BreakOutFloat4Components}
  n1163 = LinearInterpolate(
   A=AppendVector(A=@n1159.RGB, B=@n1159.RGB)
   B=LinearInterpolate(A=AppendVector(A=@n1160.RGB, B=@n1160.RGB), B=AppendVector(A=@n1161.RGB, B=@n1161.RGB), Alpha=@n1162.G)
   Alpha=@n1162.R
  )
  n1164 = VectorParameter() {group=Effect_Decal_Reactive_Uv_Offset_LR, parameterName=EDRec_DecalC_Offset_Mech End, defaultValue=<rgba 0 0 0 0.25>}
  n1165 = VectorParameter() {group=Effect_Decal_Reactive_Uv_Offset_LR, parameterName=EDRec_DecalC_Offset_Character End, defaultValue=<rgba 0 0 0 0.25>}
  n1166 = LinearInterpolate(A=AppendVector(A=@n1164.RGB, B=@n1164.RGB), B=AppendVector(A=@n1165.RGB, B=@n1165.RGB), Alpha=@n1162.A)
  n1167 = VectorParameter() {group=Effect_Decal_Reactive_Uv_Offset_LR, parameterName="EDRec_DecalC_Offset_Backpack----A=Scale End", defaultValue=<rgba 0 0 0 0.25>}
  n1168 = MaterialFunctionCall(Float2=NamedRerouteUsage()) {materialFunction=&BreakOutFloat2Components}
  n1169 = VectorParameter() {group=Effect_Decal_Reactive_Uv_Offset_LR, parameterName=EDRec_DecalC_Offset_Pickaxe End, defaultValue=<rgba 0 0 0 0.25>}
  n1170 = LinearInterpolate(
   A=LinearInterpolate(A=LinearInterpolate(A=@n1163, B=@n1166, Alpha=@n1162.B), B=AppendVector(A=@n1167.RGB, B=@n1167.RGB), Alpha=@n1168.R)
   B=AppendVector(A=@n1169.RGB, B=@n1169.RGB)
   Alpha=@n1168.G
  )
  n1171 = StaticSwitchParameter(True=LinearInterpolate(A=@n1158, B=@n1170, Alpha=NamedRerouteUsage()), False=@n1158) {desc="Use Scale value Alpha is disabled. ", group=Effect_Decal_Reactive_RE_Offset, parameterName=EDRec_Reactive_Offset, sortPriority=0}
  NamedRerouteDeclaration(None=@n1171) {name=UVs_OffsetC, nodeColor=<rgba 0 0.625 1 1>}
  n1172 = VectorParameter() {
   group=Effect_Decal_Reactive_FlipBook----
   parameterName=EDRec_Decal_D_Animation--
   channelNames {a=Clamp Max, b=Clamp Min Frames, g="", r=Loop}
   defaultValue=<rgba 1 0 0 1>
  }
  n1173 = Time()
  n1174 = Add(
   A=ScalarParameter() {defaultValue=-9999999, group=Effect_Decal_Reactive_FlipBook----, parameterName=EDRec_Decal_D_Animation_Time_BP}
   B=ScalarParameter() {group=Effect_Decal_Reactive_FlipBook----, parameterName=EDRec_Decal_D_Animation_Delay}
  )
  n1175 = Saturate(None=Divide(
    A=Subtract(A=@n1173, B=@n1174)
    B=ScalarParameter() {defaultValue=1, group=Effect_Decal_Reactive_FlipBook----, parameterName=EDRec_Decal_D_Animation_Duration}
   ))
  n1176 = ScalarParameter() {defaultValue=1, group=Effect_Decal_Reactive_FlipBook----, parameterName=EDRec_Decal_D_Animation_Falloff}
  n1177 = Multiply(A=Multiply(A=Power(Base=@n1175, Exp=@n1176)) {constB=1}, B=@n1172.Loop)
  NamedRerouteDeclaration(None=LinearInterpolate(A=@n1172.Clamp Min Frames, B=@n1172.Clamp Min Frames, Alpha=Frac(None=@n1177))) {name=reactive_D_ainim, nodeColor=<rgba 0 0.5078 1 1>}
  NamedRerouteDeclaration(None=StaticBoolParameter() {group=Effect_Decal_Reactive_FlipBook----, parameterName=EDRec_Decal_Animation_Reactive, sortPriority=0}) {name=Anim_React_State, nodeColor=<rgba 0.7344 1 0 1>}
  n1178 = StaticBoolParameter() {group=Effect_Decal_Reactive_Emissive_A_Reactive, parameterName=EDRec-Emissive-A_Reactive_Audio---------------------}
  NamedRerouteDeclaration(None=@n1178) {name=Emiss_Audio_A_State, nodeColor=<rgba 0 0.8594 1 1>}
  NamedRerouteDeclaration(None=@n554) {name=Emiss_Audio_B_State, nodeColor=<rgba 0 0.8594 1 1>}
  n1179 = VectorParameter() {
   group=Effect_Decal_Reactive_Uv_Animation
   parameterName=EDRec_DecalA_Sine_Range_Weapon
   sortPriority=500
   channelNames {a="", b="", g=Y Range, r=X Range}
   defaultValue=<rgba 0.1 0.1 0 0>
  }
  n1180 = VectorParameter() {
   group=Effect_Decal_Reactive_Uv_Animation
   parameterName=EDRec_DecalA_Sine_Range_Vehicle
   sortPriority=500
   channelNames {a="", b="", g=Y Range, r=X Range}
   defaultValue=<rgba 0.1 0.1 0 0>
  }
  n1181 = VectorParameter() {
   group=Effect_Decal_Reactive_Uv_Animation
   parameterName=EDRec_DecalA_Sine_Range_VehicleSide
   sortPriority=500
   channelNames {a="", b="", g=Y Range, r=X Range}
   defaultValue=<rgba 0.1 0.1 0 0>
  }
  n1182 = MaterialFunctionCall(Float4=NamedRerouteUsage()) {materialFunction=&BreakOutFloat4Components}
  n1183 = LinearInterpolate(A=@n1179, B=LinearInterpolate(A=@n1180, B=@n1181, Alpha=@n1182.G), Alpha=@n1182.R)
  n1184 = VectorParameter() {
   group=Effect_Decal_Reactive_Uv_Animation
   parameterName=EDRec_DecalA_Sine_Range_Mech
   sortPriority=500
   channelNames {a="", b="", g=Y Range, r=X Range}
   defaultValue=<rgba 0.1 0.1 0 0>
  }
  n1185 = VectorParameter() {
   group=Effect_Decal_Reactive_Uv_Animation
   parameterName=EDRec_DecalA_Sine_Range_Character
   sortPriority=500
   channelNames {a="", b="", g=Y Range, r=X Range}
   defaultValue=<rgba 0.1 0.1 0 0>
  }
  n1186 = LinearInterpolate(A=@n1184, B=@n1185, Alpha=@n1182.A)
  n1187 = VectorParameter() {
   group=Effect_Decal_Reactive_Uv_Animation
   parameterName=EDRec_DecalA_Sine_Range_BackPack
   sortPriority=500
   channelNames {a="", b="", g=Y Range, r=X Range}
   defaultValue=<rgba 0.1 0.1 0 0>
  }
  n1188 = MaterialFunctionCall(Float2=NamedRerouteUsage()) {materialFunction=&BreakOutFloat2Components}
  n1189 = VectorParameter() {
   group=Effect_Decal_Reactive_Uv_Animation
   parameterName=EDRec_DecalA_Sine_Range_Pickaxe
   sortPriority=500
   channelNames {a="", b="", g=Y Range, r=X Range}
   defaultValue=<rgba 0.1 0.1 0 0>
  }
  MaterialExpressionLinearInterpolate_286 = LinearInterpolate(A=LinearInterpolate(A=LinearInterpolate(A=@n1183, B=@n1186, Alpha=@n1182.B), B=@n1187, Alpha=@n1188.R), B=@n1189, Alpha=@n1188.G)
  n1190 = Convert(None=@MaterialExpressionLinearInterpolate_286) {
   nodeName=Break Float3
   convertInputs [{
     type=Vector3
     expressionInput {
      expression=&self:MaterialExpressionLinearInterpolate_286
      inputName=None
      mask=0
      maskA=0
      maskB=0
      maskG=0
      maskR=0
      outputIndex=0
     }
    }]
   convertOutputs [{type=Scalar}, {type=Scalar}, {type=Scalar}]
  }
  n1191 = Time()
  n1192 = VectorParameter() {
   group=Effect_Decal_Reactive_Uv_Animation
   parameterName=EDRec_DecalA_Sine-Speed
   sortPriority=500
   channelNames {a=Y Offset, b=X offset, g=Y Speed, r=X Speed}
  }
  n1193 = AppendVector(A=@n1192.X Speed, B=@n1192.X Speed)
  MaterialExpressionSine_13 = Sine(None=Add(A=Multiply(A=@n1191, B=@n1193), B=AppendVector(A=@n1192.X offset, B=@n1192.X offset)))
  n1194 = Convert(None=@MaterialExpressionSine_13) {
   nodeName=Break Float2
   convertInputs [{
     type=Vector2
     expressionInput {
      expression=&self:MaterialExpressionSine_13
      inputName=None
      mask=0
      maskA=0
      maskB=0
      maskG=0
      maskR=0
      outputIndex=0
     }
    }]
   convertOutputs [{type=Scalar}, {type=Scalar}]
  }
  n1195 = Multiply(A=@n1190, B=@n1194) {constA=0.1}
  n1196 = Multiply(A=@n1190, B=@n1194) {constA=0.1}
  n1197 = Constant()
  n1198 = ScalarParameter() {group=Effect_Decal_Reactive_Uv_Animation, parameterName="EDRec_DecalA_Sine-Rotation   =1", sortPriority=500}
  n1199 = MaterialFunctionCall(R=@n1195, G=@n1196, B=@n1197, A=@n1198) {materialFunction=&AppendMany}
  n1200 = MaterialFunctionCall(R=@n1196, G=@n1195, B=@n1197, A=@n1198) {materialFunction=&AppendMany}
  n1201 = ScalarParameter() {defaultValue=1, group=Effect_Decal_Reactive_Uv_Animation_React, parameterName=EDRec_DecalA__Sine_Strength_Start, sortPriority=100}
  n1202 = ScalarParameter() {defaultValue=1, group=Effect_Decal_Reactive_Uv_Animation_React, parameterName=EDRec_DecalA__Sine_Strength_End, sortPriority=100}
  n1203 = ScalarParameter() {defaultValue=-9999999, group=Effect_Decal_Reactive_Uv_Animation_React, parameterName=EDRec_DecalA__Sine_TimeFrom_BP, sortPriority=100}
  n1204 = Subtract(
   A=Time()
   B=Add(
    A=ScalarParameter() {group=Effect_Decal_Reactive_Uv_Animation_React, parameterName=EDRec_DecalA__Sine_Delay-----------------, sortPriority=100}
    B=@n1203
   )
  )
  n1205 = ScalarParameter() {defaultValue=1, group=Effect_Decal_Reactive_Uv_Animation_React, parameterName=EDRec_DecalA__Sine_Duration, sortPriority=100}
  n1206 = ScalarParameter() {defaultValue=0.25, group=Effect_Decal_Reactive_Uv_Animation_React, parameterName=EDRec_DecalA__Sine_Falloff, sortPriority=100}
  n1207 = Power(Base=Saturate(None=Divide(A=@n1204, B=@n1205)), Exp=@n1206)
  NamedRerouteDeclaration(None=Multiply(A=LinearInterpolate(A=@n1199.RGBA, B=@n1200.RGBA, Alpha=@n1198), B=LinearInterpolate(A=@n1201, B=@n1202, Alpha=@n1207))) {name=Decal_A_Sine, nodeColor=<rgba 0 0.625 1 1>}
  NamedRerouteDeclaration(None=@n692) {name=Emiss_Audio_C_State, nodeColor=<rgba 0 0.8594 1 1>}
  NamedRerouteDeclaration(None=@n780) {name=Emiss_Audio_D_State_1, nodeColor=<rgba 0 0.8594 1 1>}
  n1208 = VectorParameter() {
   group=Effect_Decal_Reactive_Uv_Animation
   parameterName=EDRec_DecalB_Sine_Range_Weapon
   sortPriority=500
   channelNames {a="", b="", g=Y Range, r=X Range}
   defaultValue=<rgba 0.1 0.1 0 0>
  }
  n1209 = VectorParameter() {
   group=Effect_Decal_Reactive_Uv_Animation
   parameterName=EDRec_DecalB_Sine_Range_Vehicle
   sortPriority=500
   channelNames {a="", b="", g=Y Range, r=X Range}
   defaultValue=<rgba 0.1 0.1 0 0>
  }
  n1210 = VectorParameter() {
   group=Effect_Decal_Reactive_Uv_Animation
   parameterName=EDRec_DecalB_Sine_Range_VehicleSide
   sortPriority=500
   channelNames {a="", b="", g=Y Range, r=X Range}
   defaultValue=<rgba 0.1 0.1 0 0>
  }
  n1211 = MaterialFunctionCall(Float4=NamedRerouteUsage()) {materialFunction=&BreakOutFloat4Components}
  n1212 = LinearInterpolate(A=@n1208, B=LinearInterpolate(A=@n1209, B=@n1210, Alpha=@n1211.G), Alpha=@n1211.R)
  n1213 = VectorParameter() {
   group=Effect_Decal_Reactive_Uv_Animation
   parameterName=EDRec_DecalB_Sine_Range_Mech
   sortPriority=500
   channelNames {a="", b="", g=Y Range, r=X Range}
   defaultValue=<rgba 0.1 0.1 0 0>
  }
  n1214 = VectorParameter() {
   group=Effect_Decal_Reactive_Uv_Animation
   parameterName=EDRec_DecalB_Sine_Range_Character
   sortPriority=500
   channelNames {a="", b="", g=Y Range, r=X Range}
   defaultValue=<rgba 0.1 0.1 0 0>
  }
  n1215 = LinearInterpolate(A=@n1213, B=@n1214, Alpha=@n1211.A)
  n1216 = VectorParameter() {
   group=Effect_Decal_Reactive_Uv_Animation
   parameterName=EDRec_DecalB_Sine_Range_BackPack
   sortPriority=500
   channelNames {a="", b="", g=Y Range, r=X Range}
   defaultValue=<rgba 0.1 0.1 0 0>
  }
  n1217 = MaterialFunctionCall(Float2=NamedRerouteUsage()) {materialFunction=&BreakOutFloat2Components}
  n1218 = VectorParameter() {
   group=Effect_Decal_Reactive_Uv_Animation
   parameterName=EDRec_DecalB_Sine_Range_Pickaxe
   sortPriority=500
   channelNames {a="", b="", g=Y Range, r=X Range}
   defaultValue=<rgba 0.1 0.1 0 0>
  }
  MaterialExpressionLinearInterpolate_292 = LinearInterpolate(A=LinearInterpolate(A=LinearInterpolate(A=@n1212, B=@n1215, Alpha=@n1211.B), B=@n1216, Alpha=@n1217.R), B=@n1218, Alpha=@n1217.G)
  n1219 = Convert(None=@MaterialExpressionLinearInterpolate_292) {
   nodeName=Break Float3
   convertInputs [{
     type=Vector3
     expressionInput {
      expression=&self:MaterialExpressionLinearInterpolate_292
      inputName=None
      mask=0
      maskA=0
      maskB=0
      maskG=0
      maskR=0
      outputIndex=0
     }
    }]
   convertOutputs [{type=Scalar}, {type=Scalar}, {type=Scalar}]
  }
  n1220 = Time()
  n1221 = VectorParameter() {
   group=Effect_Decal_Reactive_Uv_Animation
   parameterName=EDRec_DecalB_Sine-Speed
   sortPriority=500
   channelNames {a=Y Offset, b=X offset, g=Y Speed, r=X Speed}
  }
  n1222 = Multiply(A=@n1220, B=@n1221.X Speed)
  n1223 = Multiply(A=@n1219, B=Sine(None=Add(A=@n1222, B=@n1221.X offset))) {constA=0.1}
  n1224 = Multiply(A=@n1220, B=@n1221.Y Speed)
  n1225 = Multiply(A=@n1219, B=Sine(None=Add(A=@n1224, B=@n1221.Y Offset))) {constA=0.1}
  n1226 = Constant()
  n1227 = ScalarParameter() {group=Effect_Decal_Reactive_Uv_Animation, parameterName="EDRec_DecalB_Sine-Rotation   =1", sortPriority=500}
  n1228 = MaterialFunctionCall(R=@n1223, G=@n1225, B=@n1226, A=@n1227) {materialFunction=&AppendMany}
  n1229 = MaterialFunctionCall(R=@n1225, G=@n1223, B=@n1226, A=@n1227) {materialFunction=&AppendMany}
  n1230 = ScalarParameter() {defaultValue=1, group=Effect_Decal_Reactive_Uv_Animation_React, parameterName=EDRec_DecalB__Sine_Strength_Start, sortPriority=100}
  n1231 = ScalarParameter() {defaultValue=1, group=Effect_Decal_Reactive_Uv_Animation_React, parameterName=EDRec_DecalB__Sine_Strength_End, sortPriority=100}
  n1232 = ScalarParameter() {defaultValue=-9999999, group=Effect_Decal_Reactive_Uv_Animation_React, parameterName=EDRec_DecalB__Sine_TimeFrom_BP, sortPriority=100}
  n1233 = Subtract(
   A=Time()
   B=Add(
    A=ScalarParameter() {group=Effect_Decal_Reactive_Uv_Animation_React, parameterName=EDRec_DecalB__Sine_Delay-----------------, sortPriority=100}
    B=@n1232
   )
  )
  n1234 = ScalarParameter() {defaultValue=1, group=Effect_Decal_Reactive_Uv_Animation_React, parameterName=EDRec_DecalB__Sine_Duration, sortPriority=100}
  n1235 = ScalarParameter() {defaultValue=0.25, group=Effect_Decal_Reactive_Uv_Animation_React, parameterName=EDRec_DecalB__Sine_Falloff, sortPriority=100}
  n1236 = Power(Base=Saturate(None=Divide(A=@n1233, B=@n1234)), Exp=@n1235)
  NamedRerouteDeclaration(None=Multiply(A=LinearInterpolate(A=@n1228.RGBA, B=@n1229.RGBA, Alpha=@n1227), B=LinearInterpolate(A=@n1230, B=@n1231, Alpha=@n1236))) {name=Decal_B_Sine, nodeColor=<rgba 0 0.625 1 1>}
  n1237 = VectorParameter() {
   group=Effect_Decal_Reactive_Uv_Animation
   parameterName=EDRec_DecalC_Sine_Range_Weapon
   sortPriority=500
   channelNames {a="", b="", g=Y Range, r=X Range}
   defaultValue=<rgba 0.1 0.1 0 0>
  }
  n1238 = VectorParameter() {
   group=Effect_Decal_Reactive_Uv_Animation
   parameterName=EDRec_DecalC_Sine_Range_Vehicle
   sortPriority=500
   channelNames {a="", b="", g=Y Range, r=X Range}
   defaultValue=<rgba 0.1 0.1 0 0>
  }
  n1239 = VectorParameter() {
   group=Effect_Decal_Reactive_Uv_Animation
   parameterName=EDRec_DecalC_Sine_Range_VehicleSide
   sortPriority=500
   channelNames {a="", b="", g=Y Range, r=X Range}
   defaultValue=<rgba 0.1 0.1 0 0>
  }
  n1240 = MaterialFunctionCall(Float4=NamedRerouteUsage()) {materialFunction=&BreakOutFloat4Components}
  n1241 = LinearInterpolate(A=@n1237, B=LinearInterpolate(A=@n1238, B=@n1239, Alpha=@n1240.G), Alpha=@n1240.R)
  n1242 = VectorParameter() {
   group=Effect_Decal_Reactive_Uv_Animation
   parameterName=EDRec_DecalC_Sine_Range_Mech
   sortPriority=500
   channelNames {a="", b="", g=Y Range, r=X Range}
   defaultValue=<rgba 0.1 0.1 0 0>
  }
  n1243 = VectorParameter() {
   group=Effect_Decal_Reactive_Uv_Animation
   parameterName=EDRec_DecalC_Sine_Range_Character
   sortPriority=500
   channelNames {a="", b="", g=Y Range, r=X Range}
   defaultValue=<rgba 0.1 0.1 0 0>
  }
  n1244 = LinearInterpolate(A=@n1242, B=@n1243, Alpha=@n1240.A)
  n1245 = VectorParameter() {
   group=Effect_Decal_Reactive_Uv_Animation
   parameterName=EDRec_DecalC_Sine_Range_BackPack
   sortPriority=500
   channelNames {a="", b="", g=Y Range, r=X Range}
   defaultValue=<rgba 0.1 0.1 0 0>
  }
  n1246 = MaterialFunctionCall(Float2=NamedRerouteUsage()) {materialFunction=&BreakOutFloat2Components}
  n1247 = VectorParameter() {
   group=Effect_Decal_Reactive_Uv_Animation
   parameterName=EDRec_DecalC_Sine_Range_Pickaxe
   sortPriority=500
   channelNames {a="", b="", g=Y Range, r=X Range}
   defaultValue=<rgba 0.1 0.1 0 0>
  }
  MaterialExpressionLinearInterpolate_298 = LinearInterpolate(A=LinearInterpolate(A=LinearInterpolate(A=@n1241, B=@n1244, Alpha=@n1240.B), B=@n1245, Alpha=@n1246.R), B=@n1247, Alpha=@n1246.G)
  n1248 = Convert(None=@MaterialExpressionLinearInterpolate_298) {
   nodeName=Break Float3
   convertInputs [{
     type=Vector3
     expressionInput {
      expression=&self:MaterialExpressionLinearInterpolate_298
      inputName=None
      mask=0
      maskA=0
      maskB=0
      maskG=0
      maskR=0
      outputIndex=0
     }
    }]
   convertOutputs [{type=Scalar}, {type=Scalar}, {type=Scalar}]
  }
  n1249 = Time()
  n1250 = VectorParameter() {
   group=Effect_Decal_Reactive_Uv_Animation
   parameterName=EDRec_DecalC_Sine-Speed
   sortPriority=500
   channelNames {a=Y Offset, b=X offset, g=Y Speed, r=X Speed}
  }
  n1251 = Multiply(A=@n1249, B=@n1250.X Speed)
  n1252 = Multiply(A=@n1248, B=Sine(None=Add(A=@n1251, B=@n1250.X offset))) {constA=0.1}
  n1253 = Multiply(A=@n1249, B=@n1250.Y Speed)
  n1254 = Multiply(A=@n1248, B=Sine(None=Add(A=@n1253, B=@n1250.Y Offset))) {constA=0.1}
  n1255 = Constant()
  n1256 = ScalarParameter() {group=Effect_Decal_Reactive_Uv_Animation, parameterName="EDRec_DecalC_Sine-Rotation   =1", sortPriority=500}
  n1257 = MaterialFunctionCall(R=@n1252, G=@n1254, B=@n1255, A=@n1256) {materialFunction=&AppendMany}
  n1258 = MaterialFunctionCall(R=@n1254, G=@n1252, B=@n1255, A=@n1256) {materialFunction=&AppendMany}
  n1259 = ScalarParameter() {defaultValue=1, group=Effect_Decal_Reactive_Uv_Animation_React, parameterName=EDRec_DecalC__Sine_Strength_Start, sortPriority=100}
  n1260 = ScalarParameter() {defaultValue=1, group=Effect_Decal_Reactive_Uv_Animation_React, parameterName=EDRec_DecalC__Sine_Strength_End, sortPriority=100}
  n1261 = ScalarParameter() {defaultValue=-9999999, group=Effect_Decal_Reactive_Uv_Animation_React, parameterName=EDRec_DecalC__Sine_TimeFrom_BP, sortPriority=100}
  n1262 = Subtract(
   A=Time()
   B=Add(
    A=ScalarParameter() {group=Effect_Decal_Reactive_Uv_Animation_React, parameterName=EDRec_DecalC__Sine_Delay-----------------, sortPriority=100}
    B=@n1261
   )
  )
  n1263 = ScalarParameter() {defaultValue=1, group=Effect_Decal_Reactive_Uv_Animation_React, parameterName=EDRec_DecalC__Sine_Duration, sortPriority=100}
  n1264 = ScalarParameter() {defaultValue=0.25, group=Effect_Decal_Reactive_Uv_Animation_React, parameterName=EDRec_DecalC__Sine_Falloff, sortPriority=100}
  n1265 = Power(Base=Saturate(None=Divide(A=@n1262, B=@n1263)), Exp=@n1264)
  NamedRerouteDeclaration(None=Multiply(A=LinearInterpolate(A=@n1257.RGBA, B=@n1258.RGBA, Alpha=@n1256), B=LinearInterpolate(A=@n1259, B=@n1260, Alpha=@n1265))) {name=Decal_C_Sine, nodeColor=<rgba 0 0.625 1 1>}
  n1266 = VectorParameter() {
   group=Effect_Decal_Reactive_Uv_Animation
   parameterName=EDRec_DecalD_Sine_Range_Weapon
   sortPriority=500
   channelNames {a="", b="", g=Y Range, r=X Range}
   defaultValue=<rgba 0.1 0.1 0 0>
  }
  n1267 = VectorParameter() {
   group=Effect_Decal_Reactive_Uv_Animation
   parameterName=EDRec_DecalD_Sine_Range_Vehicle
   sortPriority=500
   channelNames {a="", b="", g=Y Range, r=X Range}
   defaultValue=<rgba 0.1 0.1 0 0>
  }
  n1268 = VectorParameter() {
   group=Effect_Decal_Reactive_Uv_Animation
   parameterName=EDRec_DecalD_Sine_Range_VehicleSide
   sortPriority=500
   channelNames {a="", b="", g=Y Range, r=X Range}
   defaultValue=<rgba 0.1 0.1 0 0>
  }
  n1269 = MaterialFunctionCall(Float4=NamedRerouteUsage()) {materialFunction=&BreakOutFloat4Components}
  n1270 = LinearInterpolate(A=@n1266, B=LinearInterpolate(A=@n1267, B=@n1268, Alpha=@n1269.G), Alpha=@n1269.R)
  n1271 = VectorParameter() {
   group=Effect_Decal_Reactive_Uv_Animation
   parameterName=EDRec_DecalD_Sine_Range_Mech
   sortPriority=500
   channelNames {a="", b="", g=Y Range, r=X Range}
   defaultValue=<rgba 0.1 0.1 0 0>
  }
  n1272 = VectorParameter() {
   group=Effect_Decal_Reactive_Uv_Animation
   parameterName=EDRec_DecalD_Sine_Range_Character
   sortPriority=500
   channelNames {a="", b="", g=Y Range, r=X Range}
   defaultValue=<rgba 0.1 0.1 0 0>
  }
  n1273 = LinearInterpolate(A=@n1271, B=@n1272, Alpha=@n1269.A)
  n1274 = VectorParameter() {
   group=Effect_Decal_Reactive_Uv_Animation
   parameterName=EDRec_DecalD_Sine_Range_BackPack
   sortPriority=500
   channelNames {a="", b="", g=Y Range, r=X Range}
   defaultValue=<rgba 0.1 0.1 0 0>
  }
  n1275 = MaterialFunctionCall(Float2=NamedRerouteUsage()) {materialFunction=&BreakOutFloat2Components}
  n1276 = VectorParameter() {
   group=Effect_Decal_Reactive_Uv_Animation
   parameterName=EDRec_DecalD_Sine_Range_Pickaxe
   sortPriority=500
   channelNames {a="", b="", g=Y Range, r=X Range}
   defaultValue=<rgba 0.1 0.1 0 0>
  }
  MaterialExpressionLinearInterpolate_304 = LinearInterpolate(A=LinearInterpolate(A=LinearInterpolate(A=@n1270, B=@n1273, Alpha=@n1269.B), B=@n1274, Alpha=@n1275.R), B=@n1276, Alpha=@n1275.G)
  n1277 = Convert(None=@MaterialExpressionLinearInterpolate_304) {
   nodeName=Break Float3
   convertInputs [{
     type=Vector3
     expressionInput {
      expression=&self:MaterialExpressionLinearInterpolate_304
      inputName=None
      mask=0
      maskA=0
      maskB=0
      maskG=0
      maskR=0
      outputIndex=0
     }
    }]
   convertOutputs [{type=Scalar}, {type=Scalar}, {type=Scalar}]
  }
  n1278 = Time()
  n1279 = VectorParameter() {
   group=Effect_Decal_Reactive_Uv_Animation
   parameterName=EDRec_DecalD_Sine-Speed
   sortPriority=500
   channelNames {a=Y Offset, b=X offset, g=Y Speed, r=X Speed}
  }
  n1280 = Multiply(A=@n1278, B=@n1279.X Speed)
  n1281 = Multiply(A=@n1277, B=Sine(None=Add(A=@n1280, B=@n1279.X offset))) {constA=0.1}
  n1282 = Multiply(A=@n1278, B=@n1279.Y Speed)
  n1283 = Multiply(A=@n1277, B=Sine(None=Add(A=@n1282, B=@n1279.Y Offset))) {constA=0.1}
  n1284 = Constant()
  n1285 = ScalarParameter() {group=Effect_Decal_Reactive_Uv_Animation, parameterName="EDRec_DecalD_Sine-Rotation   =1", sortPriority=500}
  n1286 = MaterialFunctionCall(R=@n1281, G=@n1283, B=@n1284, A=@n1285) {materialFunction=&AppendMany}
  n1287 = MaterialFunctionCall(R=@n1283, G=@n1281, B=@n1284, A=@n1285) {materialFunction=&AppendMany}
  n1288 = ScalarParameter() {defaultValue=1, group=Effect_Decal_Reactive_Uv_Animation_React, parameterName=EDRec_DecalD__Sine_Strength_Start, sortPriority=100}
  n1289 = ScalarParameter() {defaultValue=1, group=Effect_Decal_Reactive_Uv_Animation_React, parameterName=EDRec_DecalD__Sine_Strength_End, sortPriority=100}
  n1290 = ScalarParameter() {defaultValue=-9999999, group=Effect_Decal_Reactive_Uv_Animation_React, parameterName=EDRec_DecalD__Sine_TimeFrom_BP, sortPriority=100}
  n1291 = Subtract(
   A=Time()
   B=Add(
    A=ScalarParameter() {group=Effect_Decal_Reactive_Uv_Animation_React, parameterName=EDRec_DecalD__Sine_Delay-----------------, sortPriority=100}
    B=@n1290
   )
  )
  n1292 = ScalarParameter() {defaultValue=1, group=Effect_Decal_Reactive_Uv_Animation_React, parameterName=EDRec_DecalD__Sine_Duration, sortPriority=100}
  n1293 = ScalarParameter() {defaultValue=0.25, group=Effect_Decal_Reactive_Uv_Animation_React, parameterName=EDRec_DecalD__Sine_Falloff, sortPriority=100}
  n1294 = Power(Base=Saturate(None=Divide(A=@n1291, B=@n1292)), Exp=@n1293)
  NamedRerouteDeclaration(None=Multiply(A=LinearInterpolate(A=@n1286.RGBA, B=@n1287.RGBA, Alpha=@n1285), B=LinearInterpolate(A=@n1288, B=@n1289, Alpha=@n1294))) {name=Decal_D_Sine, nodeColor=<rgba 0 0.625 1 1>}
  n1295 = PreSkinnedLocalBounds()
  n1296 = Subtract(A=VertexInterpolator(VS=PreSkinnedPosition()).PS, B=@n1295.Min)
  n1297 = Subtract(A=@n1295.Max, B=@n1295.Max)
  NamedRerouteDeclaration(None=FunctionInput(Preview=Divide(A=@n1296, B=@n1297)) {bUsePreviewValueAsDefault=true, inputName=Calcated_Bounds, sortPriority=11}) {name=CalBounds, nodeColor=<rgba 1 0.7031 0 1>}
  n1298 = ScalarParameter() {defaultValue=1, group=Effect_Decal_Reactive_Uv_Offset_LR, parameterName="EDRec_DecalC_Scale\\Tile_Vehicle"}
  n1299 = ScalarParameter() {defaultValue=1, group=Effect_Decal_Reactive_Uv_Offset_LR, parameterName="EDRec_DecalC_Scale\\Tile_Vehicle_Side"}
  n1300 = MaterialFunctionCall(Float4=NamedRerouteUsage()) {materialFunction=&BreakOutFloat4Components}
  n1301 = LinearInterpolate(
   A=LinearInterpolate(
    A=ScalarParameter() {defaultValue=1, group=Effect_Decal_Reactive_Uv_Offset_LR, parameterName="EDRec_DecalC_Scale\\Tile_Weapon"}
    B=LinearInterpolate(A=@n1298, B=@n1299, Alpha=@n1300.G)
    Alpha=@n1300.R
   )
   B=LinearInterpolate(
    A=ScalarParameter() {defaultValue=1, group=Effect_Decal_Reactive_Uv_Offset_LR, parameterName="EDRec_DecalC_Scale\\Tile_Mech"}
    B=ScalarParameter() {defaultValue=1, group=Effect_Decal_Reactive_Uv_Offset_LR, parameterName="EDRec_DecalC_Scale\\Tile_Character"}
    Alpha=@n1300.A
   )
   Alpha=@n1300.B
  )
  n1302 = ScalarParameter() {defaultValue=1, group=Effect_Decal_Reactive_Uv_Offset_LR, parameterName="EDRec_DecalC_Scale\\Tile_Backpack"}
  n1303 = MaterialFunctionCall(Float2=NamedRerouteUsage()) {materialFunction=&BreakOutFloat2Components}
  NamedRerouteDeclaration(None=LinearInterpolate(
    A=LinearInterpolate(A=@n1301, B=@n1302, Alpha=@n1303.R)
    B=ScalarParameter() {defaultValue=1, group=Effect_Decal_Reactive_Uv_Offset_LR, parameterName="EDRec_DecalC_Scale\\Tile_Pickaxe"}
    Alpha=@n1303.G
   )) {name=UVs_ScaleC, nodeColor=<rgba 0 0.7188 1 1>}
  n1304 = Multiply(
   A=Add(A=ScalarParameter() {group=Effect_Decal_Reactive_Uv_Animation, parameterName=EDRec_DecalA_Rotation_Speed Offset, sortPriority=0}, B=Time())
   B=ScalarParameter() {group=Effect_Decal_Reactive_Uv_Animation, parameterName="EDRec_DecalA_Rotation_Speed  --------------", sortPriority=0}
  )
  n1305 = Power(
   Base=Saturate()
   Exp=ScalarParameter() {defaultValue=0.25, group=Effect_Decal_Reactive_Uv_Animation_React, parameterName=EDRec_DecalA_BounceRot_Falloff}
  )
  n1306 = LinearInterpolate(A=ScalarParameter() {group=Effect_Decal_Reactive_Uv_Animation_React, parameterName=EDRec_DecalA_BounceRot_Speed_Start}, Alpha=@n1305) {constB=0}
  n1307 = ScalarParameter() {defaultValue=0.1, group=Effect_Decal_Reactive_Uv_Animation, parameterName=EDRec_DecalA_Rotation_Speed Bounce Strength, sortPriority=0}
  n1308 = Multiply(
   A=Sine(None=Add(A=@n1304, B=@n1306))
   B=LinearInterpolate(
    A=ScalarParameter() {defaultValue=0.1, group=Effect_Decal_Reactive_Uv_Animation_React, parameterName=EDRec_DecalA_BounceRot_Speed_Strength_Start}
    B=@n1307
    Alpha=@n1305
   )
  )
  n1309 = ScalarParameter() {group=Effect_Decal_Reactive_Uv_Animation, parameterName="EDRec_DecalA_Rotation_Speed  1=Bounce", sliderMax=1, sortPriority=0}
  n1310 = Ceil(None=@n1309)
  NamedRerouteDeclaration(None=Add(A=NamedRerouteUsage(), B=LinearInterpolate(A=@n1304, B=@n1308, Alpha=@n1310))) {name=Bounce Rotation A, nodeColor=<rgba 1 0.2578 0 1>}
  n1311 = TextureObjectParameter() {group=Effect_Decal_Reactive, parameterName=EDRec_Decal_Mask__Tex, sortPriority=10, texture=&T_BlacksRGB}
  NamedRerouteDeclaration(None=@n1311) {name=Decal_Msk_, nodeColor=<rgba 0 1 0.9062 1>}
  n1312 = Multiply(
   A=Add(A=ScalarParameter() {group=Effect_Decal_Reactive_Uv_Animation, parameterName=EDRec_DecalB_Rotation_Speed Offset, sortPriority=0}, B=Time())
   B=ScalarParameter() {group=Effect_Decal_Reactive_Uv_Animation, parameterName="EDRec_DecalB_Rotation_Speed  --------------", sortPriority=0}
  )
  n1313 = Subtract(
   A=Time()
   B=Add(
    A=ScalarParameter() {group=Effect_Decal_Reactive_Uv_Animation_React, parameterName=EDRec_DecalB_BounceRot_Delay-----------------}
    B=ScalarParameter() {defaultValue=-9999999, group=Effect_Decal_Reactive_Uv_Animation_React, parameterName=EDRec_DecalB_BounceRot_TimeFrom_BP}
   )
  )
  n1314 = ScalarParameter() {defaultValue=1, group=Effect_Decal_Reactive_Uv_Animation_React, parameterName=EDRec_DecalB_BounceRot_Duration}
  n1315 = Power(
   Base=Saturate(None=Divide(A=@n1313, B=@n1314))
   Exp=ScalarParameter() {defaultValue=0.25, group=Effect_Decal_Reactive_Uv_Animation_React, parameterName=EDRec_DecalB_BounceRot_Falloff}
  )
  n1316 = LinearInterpolate(A=ScalarParameter() {group=Effect_Decal_Reactive_Uv_Animation_React, parameterName=EDRec_DecalB_BounceRot_Speed_Start}, Alpha=@n1315) {constB=0}
  n1317 = ScalarParameter() {defaultValue=0.1, group=Effect_Decal_Reactive_Uv_Animation, parameterName=EDRec_DecalB_Rotation_Speed Bounce Strength, sortPriority=0}
  n1318 = Multiply(
   A=Sine(None=Add(A=@n1312, B=@n1316))
   B=LinearInterpolate(
    A=ScalarParameter() {defaultValue=0.1, group=Effect_Decal_Reactive_Uv_Animation_React, parameterName=EDRec_DecalB_BounceRot_Speed_Strength_Start}
    B=@n1317
    Alpha=@n1315
   )
  )
  n1319 = ScalarParameter() {group=Effect_Decal_Reactive_Uv_Animation, parameterName="EDRec_DecalB_Rotation_Speed  1=Bounce", sliderMax=1, sortPriority=0}
  n1320 = Ceil(None=@n1319)
  NamedRerouteDeclaration(None=Add(A=NamedRerouteUsage(), B=LinearInterpolate(A=@n1312, B=@n1318, Alpha=@n1320))) {name=Bounce Rotation B, nodeColor=<rgba 1 0.2578 0 1>}
  n1321 = Multiply(
   A=Add(A=ScalarParameter() {group=Effect_Decal_Reactive_Uv_Animation, parameterName=EDRec_DecalC_Rotation_Speed Offset, sortPriority=0}, B=Time())
   B=ScalarParameter() {group=Effect_Decal_Reactive_Uv_Animation, parameterName="EDRec_DecalC_Rotation_Speed  --------------", sortPriority=0}
  )
  n1322 = Subtract(
   A=Time()
   B=Add(
    A=ScalarParameter() {group=Effect_Decal_Reactive_Uv_Animation_React, parameterName=EDRec_DecalC_BounceRot_Delay-----------------}
    B=ScalarParameter() {defaultValue=-9999999, group=Effect_Decal_Reactive_Uv_Animation_React, parameterName=EDRec_DecalC_BounceRot_TimeFrom_BP}
   )
  )
  n1323 = ScalarParameter() {defaultValue=1, group=Effect_Decal_Reactive_Uv_Animation_React, parameterName=EDRec_DecalC_BounceRot_Duration}
  n1324 = Power(
   Base=Saturate(None=Divide(A=@n1322, B=@n1323))
   Exp=ScalarParameter() {defaultValue=0.25, group=Effect_Decal_Reactive_Uv_Animation_React, parameterName=EDRec_DecalC_BounceRot_Falloff}
  )
  n1325 = LinearInterpolate(A=ScalarParameter() {group=Effect_Decal_Reactive_Uv_Animation_React, parameterName=EDRec_DecalC_BounceRot_Speed_Start}, Alpha=@n1324) {constB=0}
  n1326 = ScalarParameter() {defaultValue=0.1, group=Effect_Decal_Reactive_Uv_Animation, parameterName=EDRec_DecalC_Rotation_Speed Bounce Strength, sortPriority=0}
  n1327 = Multiply(
   A=Sine(None=Add(A=@n1321, B=@n1325))
   B=LinearInterpolate(
    A=ScalarParameter() {defaultValue=0.1, group=Effect_Decal_Reactive_Uv_Animation_React, parameterName=EDRec_DecalC_BounceRot_Speed_Strength_Start}
    B=@n1326
    Alpha=@n1324
   )
  )
  n1328 = ScalarParameter() {group=Effect_Decal_Reactive_Uv_Animation, parameterName="EDRec_DecalC_Rotation_Speed  1=Bounce", sliderMax=1, sortPriority=0}
  n1329 = Ceil(None=@n1328)
  NamedRerouteDeclaration(None=Add(A=NamedRerouteUsage(), B=LinearInterpolate(A=@n1321, B=@n1327, Alpha=@n1329))) {name=Bounce Rotation C, nodeColor=<rgba 1 0.2578 0 1>}
  n1330 = Multiply(
   A=Add(A=ScalarParameter() {group=Effect_Decal_Reactive_Uv_Animation, parameterName=EDRec_DecalD_Rotation_Speed Offset, sortPriority=0}, B=Time())
   B=ScalarParameter() {group=Effect_Decal_Reactive_Uv_Animation, parameterName="EDRec_DecalD_Rotation_Speed  --------------", sortPriority=0}
  )
  n1331 = Subtract(
   A=Time()
   B=Add(
    A=ScalarParameter() {group=Effect_Decal_Reactive_Uv_Animation_React, parameterName=EDRec_DecalD_BounceRot_Delay-----------------}
    B=ScalarParameter() {defaultValue=-9999999, group=Effect_Decal_Reactive_Uv_Animation_React, parameterName=EDRec_DecalD_BounceRot_TimeFrom_BP}
   )
  )
  n1332 = ScalarParameter() {defaultValue=1, group=Effect_Decal_Reactive_Uv_Animation_React, parameterName=EDRec_DecalD_BounceRot_Duration}
  n1333 = Power(
   Base=Saturate(None=Divide(A=@n1331, B=@n1332))
   Exp=ScalarParameter() {defaultValue=0.25, group=Effect_Decal_Reactive_Uv_Animation_React, parameterName=EDRec_DecalD_BounceRot_Falloff}
  )
  n1334 = LinearInterpolate(A=ScalarParameter() {group=Effect_Decal_Reactive_Uv_Animation_React, parameterName=EDRec_DecalD_BounceRot_Speed_Start}, Alpha=@n1333) {constB=0}
  n1335 = ScalarParameter() {defaultValue=0.1, group=Effect_Decal_Reactive_Uv_Animation, parameterName=EDRec_DecalD_Rotation_Speed Bounce Strength, sortPriority=0}
  n1336 = Multiply(
   A=Sine(None=Add(A=@n1330, B=@n1334))
   B=LinearInterpolate(
    A=ScalarParameter() {defaultValue=0.1, group=Effect_Decal_Reactive_Uv_Animation_React, parameterName=EDRec_DecalD_BounceRot_Speed_Strength_Start}
    B=@n1335
    Alpha=@n1333
   )
  )
  n1337 = ScalarParameter() {group=Effect_Decal_Reactive_Uv_Animation, parameterName="EDRec_DecalD_Rotation_Speed  1=Bounce", sliderMax=1, sortPriority=0}
  n1338 = Ceil(None=@n1337)
  NamedRerouteDeclaration(None=Add(A=NamedRerouteUsage(), B=LinearInterpolate(A=@n1330, B=@n1336, Alpha=@n1338))) {name=Bounce Rotation D, nodeColor=<rgba 1 0.2578 0 1>}
  n1339 = MaterialFunctionCall(Color=Abs(None=NamedRerouteUsage())) {materialFunction=&SplitComponents}
  n1340 = MaterialFunctionCall(Color=NamedRerouteUsage()) {materialFunction=&SplitComponents}
  n1341 = LinearInterpolate(A=LinearInterpolate(A=@n1339.R, B=LinearInterpolate(A=@n1339.B, B=@n1339.B, Alpha=@n1340.G), Alpha=@n1340.R), B=@n1339.G, Alpha=@n1340.B)
  NamedRerouteDeclaration(None=@n1341) {name=Tri_Mask_raw, nodeColor=<rgba 1 0 0.4922 1>}
  NamedRerouteDeclaration(None=StaticSwitch(True=@n398.UVs No Offsets/Tiling, False=@n399.UVs No Offsets/Tiling, Value=@n395)) {name=Uvs_A_No, nodeColor=<rgba 1 0.1875 0 1>}
  NamedRerouteDeclaration(None=StaticSwitch(True=@n465.UVs No Offsets/Tiling, False=@n466.UVs No Offsets/Tiling, Value=@n467)) {name=Uvs_B_No, nodeColor=<rgba 1 0.1875 0 1>}
  NamedRerouteDeclaration(None=StaticSwitch(True=@n544.UVs No Offsets/Tiling, False=@n545.UVs No Offsets/Tiling, Value=@n546)) {name=Uvs_D_No, nodeColor=<rgba 1 0.1875 0 1>}
  NamedRerouteDeclaration(None=StaticSwitch(True=@n504.UVs No Offsets/Tiling, False=@n505.UVs No Offsets/Tiling, Value=@n506)) {name=Uvs_C_No, nodeColor=<rgba 1 0.1875 0 1>}
  n1342 = ScalarParameter() {group=Effect_Decal_Reactive_Uv_Offset_LR, parameterName=EDRec_DecalC_Rotation_Weapon}
  n1343 = MaterialFunctionCall(Float4=NamedRerouteUsage()) {materialFunction=&BreakOutFloat4Components}
  n1344 = LinearInterpolate(
   A=ScalarParameter() {group=Effect_Decal_Reactive_Uv_Offset_LR, parameterName=EDRec_DecalC_Rotation_Vehicle}
   B=ScalarParameter() {group=Effect_Decal_Reactive_Uv_Offset_LR, parameterName=EDRec_DecalC_Rotation_Vehicle_Side}
   Alpha=@n1343.G
  )
  n1345 = ScalarParameter() {group=Effect_Decal_Reactive_Uv_Offset_LR, parameterName=EDRec_DecalC_Rotation_Mech}
  n1346 = ScalarParameter() {group=Effect_Decal_Reactive_Uv_Offset_LR, parameterName=EDRec_DecalC_Rotation_Character}
  n1347 = MaterialFunctionCall(Float2=NamedRerouteUsage()) {materialFunction=&BreakOutFloat2Components}
  n1348 = LinearInterpolate(
   A=LinearInterpolate(A=LinearInterpolate(A=@n1342, B=@n1344, Alpha=@n1343.R), B=LinearInterpolate(A=@n1345, B=@n1346, Alpha=@n1343.A), Alpha=@n1343.B)
   B=ScalarParameter() {group=Effect_Decal_Reactive_Uv_Offset_LR, parameterName=EDRec_DecalC_Rotation_Backpack}
   Alpha=@n1347.R
  )
  n1349 = ScalarParameter() {group=Effect_Decal_Reactive_Uv_Offset_LR, parameterName=EDRec_DecalC_Rotation_Pickaxe}
  NamedRerouteDeclaration(None=Add(A=LinearInterpolate(A=@n1348, B=@n1349, Alpha=@n1347.G), B=NamedRerouteUsage())) {name=Uvs_RotationC, nodeColor=<rgba 0 0.7188 1 1>}
  n1350 = VectorParameter() {
   group=Effect_Decal_Reactive_Uv_Offset_LR
   parameterName=EDRec_DecalB_Scale-RND--
   channelNames {a="1= Sine Speed", b=Speed, g=Max, r=Min}
   defaultValue=<rgba 1 1 0 1>
  }
  n1351 = Constant2Vector() {g=724.3587, r=235.9474}
  n1352 = Multiply(A=Sine(None=DotProduct(A=@n196, B=@n1351)), B=@n336)
  n1353 = Add(A=Multiply(A=@n1350.Speed, B=Time()), B=@n1352)
  n1354 = Add(A=Sine(None=@n1353))
  n1355 = Frac(None=LinearInterpolate(A=@n1353, B=Multiply(A=@n1354) {constB=0.49}, Alpha=@n1350."1= Sine Speed"))
  n1356 = MaterialFunctionCall(
   UVs=Frac(None=@n195)
   Texture Scale=Multiply(A=ComponentMask(None=@n184) {a=true}, B=LinearInterpolate(A=@n1350.Min, B=@n1350.Min, Alpha=@n1355))
  ) {materialFunction=&ScaleUVsByCenter}
  n1357 = VectorParameter() {
   group=Effect_Decal_Reactive_Uv_Offset_LR
   parameterName=EDRec_DecalB_Rotation-RND--
   channelNames {a=Rotation_B_Speed, b=Rotation_A_Speed, g=Rotation_B, r=Rotation_A}
  }
  n1358 = Frac(None=Add(A=ScalarParameter() {group=Effect_Decal_Reactive_Uv_Offset_LR, parameterName=EDRec_DecalB_Rotation-RND-Offset}, B=@n337))
  n1359 = Add(
   A=Multiply(A=Time(), B=LinearInterpolate(A=@n1357.Rotation_A_Speed, B=@n1357.Rotation_A_Speed, Alpha=@n1358))
   B=LinearInterpolate(A=@n1357.Rotation_A, B=@n1357.Rotation_A, Alpha=@n1358)
  )
  NamedRerouteDeclaration(None=MaterialFunctionCall(UVs=@n1356.UVs, "Rotation Angle (0-1)"=@n1359) {materialFunction=&CustomRotator}.Rotated Values) {name=UV_B_Tile, nodeColor=<rgba 1 0.7734 0 1>}
  NamedRerouteDeclaration(None=StaticSwitch(
    True=StaticBoolParameter() {group=Effect_Decal_Reactive_FlipBook, parameterName="EDRec_DecalB  Uniqe Texture"}
    False=StaticBool()
    Value=@n625
   )) {name="DecalB  Uniqe", nodeColor=<rgba 1 0 0.3516 1>}
  n1360 = Constant() {r=0.5}
  n1361 = MaterialFunctionCall(Float2=@n677) {materialFunction=&BreakOutFloat2Components}
  n1362 = SphereMask(A=@n1360, B=@n1361.R) {attenuationRadius=0.5}
  n1363 = SphereMask(A=@n1361.G, B=@n1360) {attenuationRadius=0.5}
  NamedRerouteDeclaration(None=Multiply(A=NamedRerouteUsage(), B=Multiply(A=@n1362, B=@n1363))) {name=DecalB_Tile_Msk, nodeColor=<rgba 0.0078 0 1 1>}
  NamedRerouteDeclaration(None=StaticSwitch(True=@n398.UVs, False=@n399.UVs, Value=@n395)) {name=Uvs_A_NotClamp, nodeColor=<rgba 1 0.1875 0 1>}
  n1364 = NamedRerouteUsage()
  n1365 = Time()
  n1366 = VectorParameter() {group=Effect_Decal_Reactive_Uv_Offset_LR, parameterName=EDRec_DecalC_Speed, defaultValue=<rgba 0 0 0 1>}
  n1367 = ComponentMask(None=@n1366.RGB) {g=true, r=true}
  n1368 = NamedRerouteUsage()
  n1369 = Add(
   A=MaterialFunctionCall(UVs=@n1364, "Rotation Angle (0-1)"=NamedRerouteUsage()) {materialFunction=&CustomRotator}.Rotated Values
   B=Add(A=Multiply(A=@n1365, B=@n1367), B=ComponentMask(None=@n1368) {g=true, r=true})
  ) {constB=0}
  n1370 = NamedRerouteUsage()
  n1371 = Multiply(A=ComponentMask(None=@n1369) {r=true}, B=@n1370) {constB=5}
  n1372 = Add(A=Divide(A=Floor(None=@n1371), B=@n1370) {constB=5}) {constB=-0.5}
  n1373 = Multiply(A=@n1370, B=ScalarParameter() {group=Effect_Decal_Reactive_Uv_Offset_LR, parameterName=EDRec_DecalC_Offset_Y})
  n1374 = Add(A=Multiply(A=@n1369, B=@n1370), B=AppendVector(A=Constant(), B=Multiply(A=@n1372, B=@n1373))) {constB=0}
  n1375 = VectorParameter() {
   group=Effect_Decal_Reactive_Uv_Offset_LR
   parameterName=EDRec_DecalC_Scale-RND--
   channelNames {a="1= Sine Speed", b=Speed, g=Max, r=Min}
   defaultValue=<rgba 1 1 0 1>
  }
  n1376 = Add(A=Multiply(A=@n1375.Speed, B=Time()), B=@n302)
  n1377 = Add(A=Sine(None=@n1376))
  n1378 = Frac(None=LinearInterpolate(A=@n1376, B=Multiply(A=@n1377) {constB=0.49}, Alpha=@n1375."1= Sine Speed"))
  n1379 = MaterialFunctionCall(
   UVs=Frac(None=@n1374)
   Texture Scale=Multiply(A=ComponentMask(None=@n1368) {a=true}, B=LinearInterpolate(A=@n1375.Min, B=@n1375.Min, Alpha=@n1378))
  ) {materialFunction=&ScaleUVsByCenter}
  n1380 = VectorParameter() {
   group=Effect_Decal_Reactive_Uv_Offset_LR
   parameterName=EDRec_DecalC_Rotation--
   channelNames {a=Rotation_B_Speed, b=Rotation_A_Speed, g=Rotation_B, r=Rotation_A}
  }
  n1381 = VectorParameter() {group=Effect_Decal_Reactive_Uv_Offset_LR, parameterName=EDRec_DecalC_Random_seed, defaultValue=<rgba 4 3 0 0>}
  n1382 = MaterialFunctionCall(2 Vector=Add(A=Floor(None=@n1374), B=ComponentMask(None=@n1381.RGB) {g=true, r=true})) {materialFunction=&MF_Random_2}
  n1383 = Add(
   A=Multiply(A=Time(), B=LinearInterpolate(A=@n1380.Rotation_A_Speed, B=@n1380.Rotation_A_Speed, Alpha=@n1382.utput))
   B=LinearInterpolate(A=@n1380.Rotation_A, B=@n1380.Rotation_A, Alpha=@n1382.utput)
  )
  NamedRerouteDeclaration(None=MaterialFunctionCall(UVs=@n1379.UVs, "Rotation Angle (0-1)"=@n1383) {materialFunction=&CustomRotator}.Rotated Values) {name=UV_C_Tile, nodeColor=<rgba 1 0.7734 0 1>}
  n1384 = VectorParameter() {
   group=Effect_Decal_Reactive_Uv_Offset_LR
   parameterName=EDRec_DecalD_Scale-RND--
   channelNames {a="1= Sine Speed", b=Speed, g=Max, r=Min}
   defaultValue=<rgba 1 1 0 1>
  }
  n1385 = Add(A=Multiply(A=@n1384.Speed, B=Time()), B=@n317)
  n1386 = Add(A=Sine(None=@n1385))
  n1387 = Frac(None=LinearInterpolate(A=@n1385, B=Multiply(A=@n1386) {constB=0.49}, Alpha=@n1384."1= Sine Speed"))
  n1388 = MaterialFunctionCall(
   UVs=Frac(None=@n314)
   Texture Scale=Multiply(A=ComponentMask(None=@n308) {a=true}, B=LinearInterpolate(A=@n1384.Min, B=@n1384.Min, Alpha=@n1387))
  ) {materialFunction=&ScaleUVsByCenter}
  n1389 = Add(A=Multiply(A=Time(), B=LinearInterpolate(A=@n303.Rotation_A_Speed, B=@n303.Rotation_A_Speed, Alpha=@n316.utput)), B=@n317)
  NamedRerouteDeclaration(None=MaterialFunctionCall(UVs=@n1388.UVs, "Rotation Angle (0-1)"=@n1389) {materialFunction=&CustomRotator}.Rotated Values) {name=UV_D_Tile, nodeColor=<rgba 1 0.7734 0 1>}
  n1390 = Frac(None=@n330)
  n1391 = Multiply(A=@n333, B=LinearInterpolate(A=@n339.Min, B=@n339.Min, Alpha=@n345))
  n1392 = VectorParameter() {
   group=Effect_Decal_Reactive_Uv_Offset_LR
   parameterName=EDRec_DecalA_Rotation-RND--
   channelNames {a=Rotation_B_Speed, b=Rotation_A_Speed, g=Rotation_B, r=Rotation_A}
  }
  n1393 = Constant2Vector() {g=280.3817, r=574.9868}
  n1394 = Multiply(A=Sine(None=DotProduct(A=@n331, B=@n1393)), B=@n341)
  n1395 = Frac(None=Add(A=ScalarParameter() {group=Effect_Decal_Reactive_Uv_Offset_LR, parameterName=EDRec_DecalB_Rotation-RND-Seed-Offset}, B=@n1394))
  n1396 = Add(
   A=Multiply(A=Time(), B=LinearInterpolate(A=@n1392.Rotation_A_Speed, B=@n1392.Rotation_A_Speed, Alpha=@n1395))
   B=LinearInterpolate(A=@n1392.Rotation_A, B=@n1392.Rotation_A, Alpha=@n1395)
  )
  NamedRerouteDeclaration(None=MaterialFunctionCall(
    UVs=MaterialFunctionCall(UVs=@n1390, Texture Scale=@n1391) {materialFunction=&ScaleUVsByCenter}.UVs
    "Rotation Angle (0-1)"=@n1396
   ) {materialFunction=&CustomRotator}.Rotated Values) {name=UV_A_Tile, nodeColor=<rgba 1 0.7734 0 1>}
  n1397 = ScalarParameter() {
   defaultValue=-9999999
   group=Effect_Decal_Reactive_Emissive_A_Reactive | Color
   parameterName=EDRec_Emissive_A_Color_Time_BP
   sortPriority=50
  }
  n1398 = ScalarParameter() {group=Effect_Decal_Reactive_Emissive_A_Reactive | Color, parameterName=EDRec_Emissive_A_Color_Delay}
  n1399 = Divide(
   A=Subtract(A=Time(), B=Add(A=@n1397, B=@n1398))
   B=ScalarParameter() {defaultValue=0.25, group=Effect_Decal_Reactive_Emissive_A_Reactive | Color, parameterName=EDRec_Emissive_A_Color_Duration}
  )
  NamedRerouteDeclaration(None=Power(
    Base=Saturate(None=@n1399)
    Exp=ScalarParameter() {defaultValue=1, group=Effect_Decal_Reactive_Emissive_A_Reactive | Color, parameterName=EDRec_Emissive_A_Color_Falloff}
   )) {name=Reactive_Color A, nodeColor=<rgba 0.8516 0 1 1>}
  NamedRerouteDeclaration(None=@n1383) {name=UV_C_Tile_Rotation, nodeColor=<rgba 1 0.7734 0 1>}
  NamedRerouteDeclaration(None=@n1389) {name=UV_D_Tile_Rotation, nodeColor=<rgba 1 0.7734 0 1>}
  n1400 = VectorParameter() {group=Effect_Decal_Reactive_Uv_Offset_LR, parameterName=EDRec_DecalD_Offset_Weapon, defaultValue=<rgba 0 0 0 0.25>}
  n1401 = VectorParameter() {group=Effect_Decal_Reactive_Uv_Offset_LR, parameterName=EDRec_DecalD_Offset_Vehicle, defaultValue=<rgba 0 0 0 0.25>}
  n1402 = VectorParameter() {group=Effect_Decal_Reactive_Uv_Offset_LR, parameterName=EDRec_DecalD_Offset_VehicleSide, defaultValue=<rgba 0 0 0 0.25>}
  n1403 = MaterialFunctionCall(Float4=NamedRerouteUsage()) {materialFunction=&BreakOutFloat4Components}
  n1404 = LinearInterpolate(
   A=AppendVector(A=@n1400.RGB, B=@n1400.RGB)
   B=LinearInterpolate(A=AppendVector(A=@n1401.RGB, B=@n1401.RGB), B=AppendVector(A=@n1402.RGB, B=@n1402.RGB), Alpha=@n1403.G)
   Alpha=@n1403.R
  )
  n1405 = VectorParameter() {group=Effect_Decal_Reactive_Uv_Offset_LR, parameterName=EDRec_DecalD_Offset_Mech, defaultValue=<rgba 0 0 0 0.25>}
  n1406 = VectorParameter() {group=Effect_Decal_Reactive_Uv_Offset_LR, parameterName=EDRec_DecalD_Offset_Character, defaultValue=<rgba 0 0 0 0.25>}
  n1407 = LinearInterpolate(A=AppendVector(A=@n1405.RGB, B=@n1405.RGB), B=AppendVector(A=@n1406.RGB, B=@n1406.RGB), Alpha=@n1403.A)
  n1408 = VectorParameter() {group=Effect_Decal_Reactive_Uv_Offset_LR, parameterName="EDRec_DecalD_Offset_Backpack----A=Scale", defaultValue=<rgba 0 0 0 0.25>}
  n1409 = MaterialFunctionCall(Float2=NamedRerouteUsage()) {materialFunction=&BreakOutFloat2Components}
  n1410 = VectorParameter() {group=Effect_Decal_Reactive_Uv_Offset_LR, parameterName=EDRec_DecalD_Offset_Pickaxe, defaultValue=<rgba 0 0 0 0.25>}
  n1411 = LinearInterpolate(
   A=LinearInterpolate(A=LinearInterpolate(A=@n1404, B=@n1407, Alpha=@n1403.B), B=AppendVector(A=@n1408.RGB, B=@n1408.RGB), Alpha=@n1409.R)
   B=AppendVector(A=@n1410.RGB, B=@n1410.RGB)
   Alpha=@n1409.G
  )
  n1412 = VectorParameter() {group=Effect_Decal_Reactive_Uv_Offset_LR, parameterName=EDRec_DecalD_Offset_Weapon End, defaultValue=<rgba 0 0 0 0.25>}
  n1413 = VectorParameter() {group=Effect_Decal_Reactive_Uv_Offset_LR, parameterName=EDRec_DecalD_Offset_Vehicle End, defaultValue=<rgba 0 0 0 0.25>}
  n1414 = VectorParameter() {group=Effect_Decal_Reactive_Uv_Offset_LR, parameterName=EDRec_DecalD_Offset_VehicleSide End, defaultValue=<rgba 0 0 0 0.25>}
  n1415 = MaterialFunctionCall(Float4=NamedRerouteUsage()) {materialFunction=&BreakOutFloat4Components}
  n1416 = LinearInterpolate(
   A=AppendVector(A=@n1412.RGB, B=@n1412.RGB)
   B=LinearInterpolate(A=AppendVector(A=@n1413.RGB, B=@n1413.RGB), B=AppendVector(A=@n1414.RGB, B=@n1414.RGB), Alpha=@n1415.G)
   Alpha=@n1415.R
  )
  n1417 = VectorParameter() {group=Effect_Decal_Reactive_Uv_Offset_LR, parameterName=EDRec_DecalD_Offset_Mech End, defaultValue=<rgba 0 0 0 0.25>}
  n1418 = VectorParameter() {group=Effect_Decal_Reactive_Uv_Offset_LR, parameterName=EDRec_DecalD_Offset_Character End, defaultValue=<rgba 0 0 0 0.25>}
  n1419 = LinearInterpolate(A=AppendVector(A=@n1417.RGB, B=@n1417.RGB), B=AppendVector(A=@n1418.RGB, B=@n1418.RGB), Alpha=@n1415.A)
  n1420 = VectorParameter() {group=Effect_Decal_Reactive_Uv_Offset_LR, parameterName="EDRec_DecalD_Offset_Backpack----A=Scale End", defaultValue=<rgba 0 0 0 0.25>}
  n1421 = MaterialFunctionCall(Float2=NamedRerouteUsage()) {materialFunction=&BreakOutFloat2Components}
  n1422 = VectorParameter() {group=Effect_Decal_Reactive_Uv_Offset_LR, parameterName=EDRec_DecalD_Offset_Pickaxe End, defaultValue=<rgba 0 0 0 0.25>}
  n1423 = LinearInterpolate(
   A=LinearInterpolate(A=LinearInterpolate(A=@n1416, B=@n1419, Alpha=@n1415.B), B=AppendVector(A=@n1420.RGB, B=@n1420.RGB), Alpha=@n1421.R)
   B=AppendVector(A=@n1422.RGB, B=@n1422.RGB)
   Alpha=@n1421.G
  )
  n1424 = StaticSwitchParameter(True=LinearInterpolate(A=@n1411, B=@n1423, Alpha=NamedRerouteUsage()), False=@n1411) {desc="Use Scale value Alpha is disabled. ", group=Effect_Decal_Reactive_RE_Offset, parameterName=EDRec_Reactive_Offset, sortPriority=0}
  NamedRerouteDeclaration(None=@n1424) {name=UVs_OffsetD, nodeColor=<rgba 0 0.625 1 1>}
  NamedRerouteDeclaration(None=@n1359) {name=UV_B_Tile_Rotation, nodeColor=<rgba 1 0.7734 0 1>}
  n1425 = ScalarParameter() {group=00-FrontEnd-Overrides, parameterName=00-FE-TODCC-Override}
  n1426 = Add(
   A=CollectionParameter() {collection=&FortniteMaterialParameters, parameterName=TimeOfDay_0-1}
   B=@n1425
  )
  n1427 = CurveAtlasRowParameter(CurveTime=@n1426) {
   atlas=&CruveAtlas_Wraps_01
   curve=&Curve_TimeOfDay
   defaultValue=182
   group=Effect_Decal_Reactive TOD
   parameterName=EDRec_Decal_TOD_CC
  }
  n1428 = MaterialFunctionCall(
   Index=ScalarParameter() {defaultValue=182, group=Effect_Decal_Reactive TOD, parameterName="EDRec_Decal_TOD_CC  Index"}
   CCinput=@n1427
   CCinput_Alpha=@n1427
   UV=@n1426
  ) {materialFunction=&MF_Global_CC_Tex}
  n1429 = ScalarParameter() {defaultValue=0.5, group=Effect_Decal_Reactive TOD, parameterName=EDRec_Decal_TOD_Morning}
  n1430 = ScalarParameter() {group=Effect_Decal_Reactive TOD, parameterName=EDRec_Decal_TOD_Day}
  n1431 = Add(A=LinearInterpolate(A=@n1425, B=Multiply(A=@n1425) {constB=-1}, Alpha=@n1430), B=@n1430)
  n1432 = ScalarParameter() {defaultValue=0.5, group=Effect_Decal_Reactive TOD, parameterName=EDRec_Decal_TOD_Evening}
  n1433 = ScalarParameter() {defaultValue=1, group=Effect_Decal_Reactive TOD, parameterName=EDRec_Decal_TOD_Night}
  NamedRerouteDeclaration(None=StaticSwitchParameter(
    True=ChannelMaskParameter(None=@n1428.4v) {group=Effect_Decal_Reactive TOD, parameterName=EDRec_Decal_TOD_CC_Channel}.R
    False=MaterialFunctionCall(Value_Morning=@n1429, Value_Day=@n1431, Value_Evening=@n1432, Value_Night=@n1433) {materialFunction=&Scalar_Time_of_Day}.Result
   ) {group=Effect_Decal_Reactive TOD, parameterName="EDRec_Decal-TOD --->CCTOD"}) {name=TOD Msk, nodeColor=<rgba 0 0.8594 1 1>}
  NamedRerouteDeclaration(None=@n1396) {name=UV_A_Tile_Rotation, nodeColor=<rgba 1 0.7734 0 1>}
  n1434 = ScalarParameter() {
   defaultValue=-9999999
   group=Effect_Decal_Reactive_Emissive_B_Reactive | Color
   parameterName=EDRec_Emissive_B_Color_Time_BP
   sortPriority=50
  }
  n1435 = ScalarParameter() {group=Effect_Decal_Reactive_Emissive_B_Reactive | Color, parameterName=EDRec_Emissive_B_Color_Delay}
  n1436 = Divide(
   A=Subtract(A=Time(), B=Add(A=@n1434, B=@n1435))
   B=ScalarParameter() {defaultValue=0.25, group=Effect_Decal_Reactive_Emissive_B_Reactive | Color, parameterName=EDRec_Emissive_B_Color_Duration}
  )
  NamedRerouteDeclaration(None=Power(
    Base=Saturate(None=@n1436)
    Exp=ScalarParameter() {defaultValue=1, group=Effect_Decal_Reactive_Emissive_B_Reactive | Color, parameterName=EDRec_Emissive_B_Color_Falloff}
   )) {name=Reactive_Color B, nodeColor=<rgba 0.8516 0 1 1>}
  n1437 = ScalarParameter() {
   defaultValue=-9999999
   group=Effect_Decal_Reactive_Emissive_C_Reactive | Color
   parameterName=EDRec_Emissive_C_Color_Time_BP
   sortPriority=50
  }
  n1438 = ScalarParameter() {group=Effect_Decal_Reactive_Emissive_C_Reactive | Color, parameterName=EDRec_Emissive_C_Color_Delay}
  n1439 = Divide(
   A=Subtract(A=Time(), B=Add(A=@n1437, B=@n1438))
   B=ScalarParameter() {defaultValue=0.25, group=Effect_Decal_Reactive_Emissive_C_Reactive | Color, parameterName=EDRec_Emissive_C_Color_Duration}
  )
  NamedRerouteDeclaration(None=Power(
    Base=Saturate(None=@n1439)
    Exp=ScalarParameter() {defaultValue=1, group=Effect_Decal_Reactive_Emissive_C_Reactive | Color, parameterName=EDRec_Emissive_C_Color_Falloff}
   )) {name=Reactive_Color C, nodeColor=<rgba 0.8516 0 1 1>}
  n1440 = ScalarParameter() {
   defaultValue=-9999999
   group=Effect_Decal_Reactive_Emissive_D_Reactive | Color
   parameterName=EDRec_Emissive_D_Color_Time_BP
   sortPriority=50
  }
  n1441 = ScalarParameter() {group=Effect_Decal_Reactive_Emissive_D_Reactive | Color, parameterName=EDRec_Emissive_D_Color_Delay}
  n1442 = Divide(
   A=Subtract(A=Time(), B=Add(A=@n1440, B=@n1441))
   B=ScalarParameter() {defaultValue=0.25, group=Effect_Decal_Reactive_Emissive_D_Reactive | Color, parameterName=EDRec_Emissive_D_Color_Duration}
  )
  NamedRerouteDeclaration(None=Power(
    Base=Saturate(None=@n1442)
    Exp=ScalarParameter() {defaultValue=1, group=Effect_Decal_Reactive_Emissive_D_Reactive | Color, parameterName=EDRec_Emissive_D_Color_Falloff}
   )) {name=Reactive_Color D, nodeColor=<rgba 0.8516 0 1 1>}
  n1443 = VectorParameter() {
   group=Effect_Decal_Reactive_Uv_Offset_LR
   parameterName=EDRec_DecalA_RND_Msking
   channelNames {a=opacity, b=Offset, g=Min, r=Max}
   defaultValue=<rgba 0 0 0 1>
  }
  n1444 = Add(A=@n342, B=Multiply(A=ScalarParameter() {group=Effect_Decal_Reactive_Uv_Offset_LR, parameterName=EDRec_DecalA_RND_Msking Speed}, B=Time()))
  n1445 = ConstantBiasScale(None=Sine(None=@n1444)) {scale=0.49}
  n1446 = ScalarParameter() {group=Effect_Decal_Reactive_Uv_Offset_LR, parameterName=EDRec_DecalA_RND_Msking Speed Sine}
  n1447 = Frac(None=Add(A=LinearInterpolate(A=@n1444, B=@n1445, Alpha=@n1446), B=@n1443.Offset))
  NamedRerouteDeclaration(None=Multiply(A=SmoothStep(Min=@n1443.Min, Max=@n1443.Min, Value=@n1447), B=@n1443.opacity)) {name=UV_A_Tile_Msk, nodeColor=<rgba 1 0.7734 0 1>}
  n1448 = VectorParameter() {
   group=Effect_Decal_Reactive_FlipBook
   parameterName=EDRec_FlipBook_FrameA_RND
   channelNames {a="", b="Offset ", g=Min, r=Max}
   defaultValue=<rgba 0 0 0 1>
  }
  n1449 = Add(A=@n1394, B=@n1448."Offset ")
  NamedRerouteDeclaration(None=LinearInterpolate(A=@n1448.Min, B=@n1448.Min, Alpha=Frac(None=@n1449))) {name=UV_A_Tile_Frame_Rnd, nodeColor=<rgba 1 0.7734 0 1>}
  n1450 = VectorParameter() {
   group=Effect_Decal_Reactive_Uv_Offset_LR
   parameterName=EDRec_DecalB_RND_Msking
   channelNames {a=opacity, b=Offset, g=Min, r=Max}
   defaultValue=<rgba 0 0 0 1>
  }
  n1451 = Add(A=@n1352, B=Multiply(A=ScalarParameter() {group=Effect_Decal_Reactive_Uv_Offset_LR, parameterName=EDRec_DecalB_RND_Msking Speed}, B=Time()))
  n1452 = ConstantBiasScale(None=Sine(None=@n1451)) {scale=0.49}
  n1453 = ScalarParameter() {group=Effect_Decal_Reactive_Uv_Offset_LR, parameterName=EDRec_DecalB_RND_Msking Speed Sine}
  n1454 = Frac(None=Add(A=LinearInterpolate(A=@n1451, B=@n1452, Alpha=@n1453), B=@n1450.Offset))
  NamedRerouteDeclaration(None=Multiply(A=SmoothStep(Min=@n1450.Min, Max=@n1450.Min, Value=@n1454), B=@n1450.opacity)) {name=UV_B_Tile_Msk, nodeColor=<rgba 1 0.7734 0 1>}
  NamedRerouteDeclaration(None=@n338) {name=UV_B_Tile_Frame_Rnd, nodeColor=<rgba 1 0.7734 0 1>}
  n1455 = ScalarParameter() {defaultValue=1, group=Effect_Decal_Reactive_Uv_Offset_LR, parameterName="EDRec_DecalD_Scale\\Tile_Vehicle"}
  n1456 = ScalarParameter() {defaultValue=1, group=Effect_Decal_Reactive_Uv_Offset_LR, parameterName="EDRec_DecalD_Scale\\Tile_Vehicle_Side"}
  n1457 = MaterialFunctionCall(Float4=NamedRerouteUsage()) {materialFunction=&BreakOutFloat4Components}
  n1458 = ScalarParameter() {defaultValue=1, group=Effect_Decal_Reactive_Uv_Offset_LR, parameterName="EDRec_DecalD_Scale\\Tile_Character", sortPriority=150}
  n1459 = LinearInterpolate(
   A=LinearInterpolate(
    A=ScalarParameter() {defaultValue=1, group=Effect_Decal_Reactive_Uv_Offset_LR, parameterName="EDRec_DecalD_Scale\\Tile_Weapon"}
    B=LinearInterpolate(A=@n1455, B=@n1456, Alpha=@n1457.G)
    Alpha=@n1457.R
   )
   B=LinearInterpolate(
    A=ScalarParameter() {defaultValue=1, group=Effect_Decal_Reactive_Uv_Offset_LR, parameterName="EDRec_DecalD_Scale\\Tile_Mech"}
    B=@n1458
    Alpha=@n1457.A
   )
   Alpha=@n1457.B
  )
  n1460 = ScalarParameter() {defaultValue=1, group=Effect_Decal_Reactive_Uv_Offset_LR, parameterName="EDRec_DecalD_Scale\\Tile_Backpack"}
  n1461 = MaterialFunctionCall(Float2=NamedRerouteUsage()) {materialFunction=&BreakOutFloat2Components}
  n1462 = ScalarParameter() {defaultValue=1, group=Effect_Decal_Reactive_Uv_Offset_LR, parameterName="EDRec_DecalD_Scale\\Tile_Pickaxe", sortPriority=150}
  NamedRerouteDeclaration(None=LinearInterpolate(A=LinearInterpolate(A=@n1459, B=@n1460, Alpha=@n1461.R), B=@n1462, Alpha=@n1461.G)) {name=UVs_ScaleD, nodeColor=<rgba 0 0.7188 1 1>}
  n1463 = VectorParameter() {
   group=Effect_Decal_Reactive_Emissive_A
   parameterName=EDRec_Emissive_A_CC-RND-Offset
   channelNames {a="", b="Seed Offset ", g=Min, r=Max}
   defaultValue=<rgba 0 0 0.5 1>
  }
  n1464 = Add(A=@n342, B=@n1463."Seed Offset ")
  NamedRerouteDeclaration(None=LinearInterpolate(A=@n1463.Min, B=@n1463.Min, Alpha=Frac(None=@n1464))) {name=EDRec_Emissive_A_CC_Rnd, nodeColor=<rgba 1 0.7734 0 1>}
  NamedRerouteDeclaration(None=@n402) {name=Tileable A State, nodeColor=<rgba 0.5937 1 0 1>}
  n1465 = VectorParameter() {
   group=Effect_Decal_Reactive_Emissive_B
   parameterName=EDRec_Emissive_B_CC-RND-Offset
   channelNames {a="", b="Seed Offset ", g=Min, r=Max}
   defaultValue=<rgba 0 0 0.5 1>
  }
  n1466 = Add(A=@n1352, B=@n1465."Seed Offset ")
  NamedRerouteDeclaration(None=LinearInterpolate(A=@n1465.Min, B=@n1465.Min, Alpha=Frac(None=@n1466))) {name=EDRec_Emissive_B_CC_Rnd, nodeColor=<rgba 1 0.7734 0 1>}
  NamedRerouteDeclaration(None=@n625) {name=Tileable B State, nodeColor=<rgba 0.0547 0 1 1>}
  n1467 = VectorParameter() {
   group=Effect_Decal_Reactive_UV_Warp
   parameterName=EDRec_WarpB_Msk_Decal_Strength
   channelNames {a=Decal D, b=Decal C, g=Decal B, r=Decal A}
   defaultValue=<rgba 1 1 1 1>
  }
  NamedRerouteDeclaration(None=@n1467.RGBA) {name=Warp_Uvs_ST, nodeColor=<rgba 0.0547 0 1 1>}
  n1468 = ScalarParameter() {
   desc=The total time of the flip animation before it restarts.
   group=Effect_Decal_Reactive_FlipBook|CCTime
   parameterName=EDRec_Decal_A_CC_Time
  }
  n1469 = Multiply(
   A=Subtract(A=Time(), B=ScalarParameter() {group=Effect_Decal_Reactive_FlipBook|CCTime, parameterName=EDRec_Decal_A_CC_Time_Bp})
   B=Divide(B=@n1468) {constA=3}
  ) {constB=0.04}
  n1470 = Frac(None=Add(A=Frac(None=@n1469), B=ScalarParameter() {group=Effect_Decal_Reactive_FlipBook|CCTime, parameterName=EDRec_Decal_A_CC_offset_Time}))
  n1471 = CurveAtlasRowParameter(CurveTime=@n1470) {
   atlas=&CruveAtlas_Wraps_01
   curve=/Game/Weapons/WeaponSkins/Wraps/Textures/0_ColorCurves/CC_DecalGl_Timing
   defaultValue=254
   group=Effect_Decal_Reactive_FlipBook|CCTime
   parameterName=EDRec_Decal_A_CC
  }
  n1472 = MaterialFunctionCall(
   Index=ScalarParameter() {defaultValue=254, group=Effect_Decal_Reactive_FlipBook|CCTime, parameterName="EDRec_Decal_A_CC  Index"}
   CCinput=@n1471
   CCinput_Alpha=@n1471
   UV=@n1470
  ) {materialFunction=&MF_Global_CC_Tex}
  NamedRerouteDeclaration(None=ChannelMaskParameter(None=@n1472.4v) {group=Effect_Decal_Reactive_FlipBook|CCTime, parameterName=EDRec_Decal_A_CC_Channel_Flip}.R) {name=CCA_FlipBook, nodeColor=<rgba 0.0547 1 0 1>}
  n1473 = ChannelMaskParameter(None=@n1472.4v) {group=Effect_Decal_Reactive_FlipBook|CCTime, parameterName=EDRec_Decal_A_CC_Channel_Warp}
  n1474 = VectorParameter() {
   group=Effect_Decal_Reactive_FlipBook|CCTime
   parameterName=EDRec_Decal_A_CC_Effect
   channelNames {a="", b="", g="", r=Warp}
   defaultValue=<rgba 0 0 0 1>
  }
  n1475 = StaticBoolParameter() {group=Effect_Decal_Reactive_FlipBook|CCTime, parameterName=EDRec-Decal-A-CC}
  NamedRerouteDeclaration(None=StaticSwitch(
    True=LinearInterpolate(B=@n1473.R, Alpha=@n1474.Warp) {constA=1}
    False=Constant() {r=1}
    Value=NamedRerouteDeclaration(None=@n1475) {name=CCA_FlipBook_State, nodeColor=<rgba 0.0547 1 0 1>}
   )) {name=CCA_FlipBook_Warp, nodeColor=<rgba 0.0547 1 0 1>}
  n1476 = ScalarParameter() {
   desc=The total time of the flip animation before it restarts.
   group=Effect_Decal_Reactive_FlipBook|CCTime
   parameterName=EDRec_Decal_B_CC_Time
  }
  n1477 = Multiply(
   A=Subtract(A=Time(), B=ScalarParameter() {group=Effect_Decal_Reactive_FlipBook|CCTime, parameterName=EDRec_Decal_B_CC_Time_Bp})
   B=Divide(B=@n1476) {constA=3}
  ) {constB=0.04}
  n1478 = Frac(None=Add(A=Frac(None=@n1477), B=ScalarParameter() {group=Effect_Decal_Reactive_FlipBook|CCTime, parameterName=EDRec_Decal_B_CC_offset_Time}))
  n1479 = CurveAtlasRowParameter(CurveTime=@n1478) {
   atlas=&CruveAtlas_Wraps_01
   curve=&CC_DecalGl_Timing
   defaultValue=254
   group=Effect_Decal_Reactive_FlipBook|CCTime
   parameterName=EDRec_Decal_B_CC
  }
  n1480 = MaterialFunctionCall(
   Index=ScalarParameter() {defaultValue=254, group=Effect_Decal_Reactive_FlipBook|CCTime, parameterName="EDRec_Decal_B_CC  Index"}
   CCinput=@n1479
   CCinput_Alpha=@n1479
   UV=@n1478
  ) {materialFunction=&MF_Global_CC_Tex}
  NamedRerouteDeclaration(None=ChannelMaskParameter(None=@n1480.4v) {group=Effect_Decal_Reactive_FlipBook|CCTime, parameterName=EDRec_Decal_B_CC_Channel_Flip}.R) {name=CCB_FlipBook, nodeColor=<rgba 0.0547 1 0 1>}
  n1481 = ScalarParameter() {group=Effect_Decal_Reactive_Uv_Offset_LR, parameterName=EDRec_DecalD_Rotation_Weapon}
  n1482 = MaterialFunctionCall(Float4=NamedRerouteUsage()) {materialFunction=&BreakOutFloat4Components}
  n1483 = LinearInterpolate(
   A=ScalarParameter() {group=Effect_Decal_Reactive_Uv_Offset_LR, parameterName=EDRec_DecalD_Rotation_Vehicle}
   B=ScalarParameter() {group=Effect_Decal_Reactive_Uv_Offset_LR, parameterName=EDRec_DecalD_Rotation_Vehicle_Side}
   Alpha=@n1482.G
  )
  n1484 = ScalarParameter() {group=Effect_Decal_Reactive_Uv_Offset_LR, parameterName=EDRec_DecalD_Rotation_Mech}
  n1485 = ScalarParameter() {group=Effect_Decal_Reactive_Uv_Offset_LR, parameterName=EDRec_DecalD_Rotation_Character}
  n1486 = MaterialFunctionCall(Float2=NamedRerouteUsage()) {materialFunction=&BreakOutFloat2Components}
  n1487 = LinearInterpolate(
   A=LinearInterpolate(A=LinearInterpolate(A=@n1481, B=@n1483, Alpha=@n1482.R), B=LinearInterpolate(A=@n1484, B=@n1485, Alpha=@n1482.A), Alpha=@n1482.B)
   B=ScalarParameter() {group=Effect_Decal_Reactive_Uv_Offset_LR, parameterName=EDRec_DecalD_Rotation_Backpack}
   Alpha=@n1486.R
  )
  n1488 = ScalarParameter() {group=Effect_Decal_Reactive_Uv_Offset_LR, parameterName=EDRec_DecalD_Rotation_Pickaxe}
  NamedRerouteDeclaration(None=Add(A=LinearInterpolate(A=@n1487, B=@n1488, Alpha=@n1486.G), B=NamedRerouteUsage())) {name=Uvs_RotationD, nodeColor=<rgba 0 0.7188 1 1>}
  n1489 = ChannelMaskParameter(None=@n1480.4v) {group=Effect_Decal_Reactive_FlipBook|CCTime, parameterName=EDRec_Decal_B_CC_Channel_Warp}
  n1490 = VectorParameter() {
   group=Effect_Decal_Reactive_FlipBook|CCTime
   parameterName=EDRec_Decal_B_CC_Effect
   channelNames {a="", b="", g="", r=Warp}
   defaultValue=<rgba 0 0 0 1>
  }
  n1491 = StaticBoolParameter() {group=Effect_Decal_Reactive_FlipBook|CCTime, parameterName=EDRec-Decal-B-CC}
  NamedRerouteDeclaration(None=StaticSwitch(
    True=LinearInterpolate(B=@n1489.R, Alpha=@n1490.Warp) {constA=1}
    False=Constant() {r=1}
    Value=NamedRerouteDeclaration(None=@n1491) {name=CCB_FlipBook_State, nodeColor=<rgba 0.0547 1 0 1>}
   )) {name=CCB_FlipBook_Warp, nodeColor=<rgba 0.0547 1 0 1>}
  n1492 = ScalarParameter() {
   desc=The total time of the flip animation before it restarts.
   group=Effect_Decal_Reactive_FlipBook|CCTime
   parameterName=EDRec_Decal_C_CC_Time
  }
  n1493 = Multiply(
   A=Subtract(A=Time(), B=ScalarParameter() {group=Effect_Decal_Reactive_FlipBook|CCTime, parameterName=EDRec_Decal_C_CC_Time_Bp})
   B=Divide(B=@n1492) {constA=3}
  ) {constB=0.04}
  n1494 = Frac(None=Add(A=Frac(None=@n1493), B=ScalarParameter() {group=Effect_Decal_Reactive_FlipBook|CCTime, parameterName=EDRec_Decal_C_CC_offset_Time}))
  n1495 = CurveAtlasRowParameter(CurveTime=@n1494) {
   atlas=&CruveAtlas_Wraps_01
   curve=&CC_DecalGl_Timing
   defaultValue=254
   group=Effect_Decal_Reactive_FlipBook|CCTime
   parameterName=EDRec_Decal_C_CC
  }
  n1496 = MaterialFunctionCall(
   Index=ScalarParameter() {defaultValue=254, group=Effect_Decal_Reactive_FlipBook|CCTime, parameterName="EDRec_Decal_C_CC  Index"}
   CCinput=@n1495
   CCinput_Alpha=@n1495
   UV=@n1494
  ) {materialFunction=&MF_Global_CC_Tex}
  NamedRerouteDeclaration(None=ChannelMaskParameter(None=@n1496.4v) {group=Effect_Decal_Reactive_FlipBook|CCTime, parameterName=EDRec_Decal_C_CC_Channel_Flip}.R) {name=CCC_FlipBook, nodeColor=<rgba 0.0547 1 0 1>}
  n1497 = ChannelMaskParameter(None=@n1496.4v) {group=Effect_Decal_Reactive_FlipBook|CCTime, parameterName=EDRec_Decal_C_CC_Channel_Warp}
  n1498 = VectorParameter() {
   group=Effect_Decal_Reactive_FlipBook|CCTime
   parameterName=EDRec_Decal_C_CC_Effect
   channelNames {a="", b="", g="", r=Warp}
   defaultValue=<rgba 0 0 0 1>
  }
  n1499 = StaticBoolParameter() {group=Effect_Decal_Reactive_FlipBook|CCTime, parameterName=EDRec-Decal-C-CC}
  NamedRerouteDeclaration(None=StaticSwitch(
    True=LinearInterpolate(B=@n1497.R, Alpha=@n1498.Warp) {constA=1}
    False=Constant() {r=1}
    Value=NamedRerouteDeclaration(None=@n1499) {name=CCC_FlipBook_State, nodeColor=<rgba 0.0547 1 0 1>}
   )) {name=CCC_FlipBook_Warp, nodeColor=<rgba 0.0547 1 0 1>}
  n1500 = ScalarParameter() {
   desc=The total time of the flip animation before it restarts.
   group=Effect_Decal_Reactive_FlipBook|CCTime
   parameterName=EDRec_Decal_D_CC_Time
  }
  n1501 = Multiply(
   A=Subtract(A=Time(), B=ScalarParameter() {group=Effect_Decal_Reactive_FlipBook|CCTime, parameterName=EDRec_Decal_D_CC_Time_Bp})
   B=Divide(B=@n1500) {constA=3}
  ) {constB=0.04}
  n1502 = Frac(None=Add(A=Frac(None=@n1501), B=ScalarParameter() {group=Effect_Decal_Reactive_FlipBook|CCTime, parameterName=EDRec_Decal_D_CC_offset_Time}))
  n1503 = CurveAtlasRowParameter(CurveTime=@n1502) {
   atlas=&CruveAtlas_Wraps_01
   curve=&CC_DecalGl_Timing
   defaultValue=254
   group=Effect_Decal_Reactive_FlipBook|CCTime
   parameterName=EDRec_Decal_D_CC
  }
  n1504 = MaterialFunctionCall(
   Index=ScalarParameter() {defaultValue=254, group=Effect_Decal_Reactive_FlipBook|CCTime, parameterName="EDRec_Decal_D_CC  Index"}
   CCinput=@n1503
   CCinput_Alpha=@n1503
   UV=@n1502
  ) {materialFunction=&MF_Global_CC_Tex}
  NamedRerouteDeclaration(None=ChannelMaskParameter(None=@n1504.4v) {group=Effect_Decal_Reactive_FlipBook|CCTime, parameterName=EDRec_Decal_D_CC_Channel_Flip}.R) {name=CCD_FlipBook, nodeColor=<rgba 0.0547 1 0 1>}
  n1505 = ChannelMaskParameter(None=@n1504.4v) {group=Effect_Decal_Reactive_FlipBook|CCTime, parameterName=EDRec_Decal_D_CC_Channel_Warp}
  n1506 = VectorParameter() {
   group=Effect_Decal_Reactive_FlipBook|CCTime
   parameterName=EDRec_Decal_D_CC_Effect
   channelNames {a="", b="", g="", r=Warp}
   defaultValue=<rgba 0 0 0 1>
  }
  n1507 = StaticBoolParameter() {group=Effect_Decal_Reactive_FlipBook|CCTime, parameterName=EDRec-Decal-D-CC}
  NamedRerouteDeclaration(None=StaticSwitch(
    True=LinearInterpolate(B=@n1505.R, Alpha=@n1506.Warp) {constA=1}
    False=Constant() {r=1}
    Value=NamedRerouteDeclaration(None=@n1507) {name=CCD_FlipBook_State, nodeColor=<rgba 0.0547 1 0 1>}
   )) {name=CCD_FlipBook_Warp, nodeColor=<rgba 0.0547 1 0 1>}
  n1508 = ScalarParameter() {group=Effect_Decal_Reactive_UV_Warp | Rotation, parameterName=EDRec_Warp_Rotation_Weapon}
  n1509 = MaterialFunctionCall(Float4=NamedRerouteUsage()) {materialFunction=&BreakOutFloat4Components}
  n1510 = LinearInterpolate(
   A=ScalarParameter() {group=Effect_Decal_Reactive_UV_Warp | Rotation, parameterName=EDRec_Warp_Rotation_Vehicle}
   B=ScalarParameter() {group=Effect_Decal_Reactive_UV_Warp | Rotation, parameterName=EDRec_Warp_Rotation_Vehicle_Side}
   Alpha=@n1509.G
  )
  n1511 = ScalarParameter() {group=Effect_Decal_Reactive_UV_Warp | Rotation, parameterName=EDRec_Warp_Rotation_Mech}
  n1512 = ScalarParameter() {group=Effect_Decal_Reactive_UV_Warp | Rotation, parameterName=EDRec_Warp_Rotation_Character}
  n1513 = MaterialFunctionCall(Float2=NamedRerouteUsage()) {materialFunction=&BreakOutFloat2Components}
  n1514 = LinearInterpolate(
   A=LinearInterpolate(A=LinearInterpolate(A=@n1508, B=@n1510, Alpha=@n1509.R), B=LinearInterpolate(A=@n1511, B=@n1512, Alpha=@n1509.A), Alpha=@n1509.B)
   B=ScalarParameter() {group=Effect_Decal_Reactive_UV_Warp | Rotation, parameterName=EDRec_Warp_Rotation_Backpack}
   Alpha=@n1513.R
  )
  n1515 = ScalarParameter() {group=Effect_Decal_Reactive_UV_Warp | Rotation, parameterName=EDRec_Warp_Rotation_Pickaxe}
  NamedRerouteDeclaration(None=Add(A=LinearInterpolate(A=@n1514, B=@n1515, Alpha=@n1513.G), B=NamedRerouteUsage())) {name=Uvs_Rotation_Warp, nodeColor=<rgba 0 0.7188 1 1>}
  n1516 = ScalarParameter() {group=Effect_Decal_Reactive_Emissive_A_Reactive, parameterName=EDRec-Emissive-A_Reactive_Audio__Min}
  n1517 = ScalarParameter() {group=Effect_Decal_Reactive_Emissive_A_Reactive, parameterName=EDRec-Emissive-A_Reactive_Audio__Max}
  n1518 = NamedRerouteUsage()
  n1519 = LinearInterpolate(
   A=ScalarParameter() {defaultValue=1, group=Effect_Decal_Reactive_Emissive_A_Reactive, parameterName=EDRec_Emissive_A_Brightness_Time__Start}
   B=ScalarParameter() {group=Effect_Decal_Reactive_Emissive_A_Reactive, parameterName=EDRec_Emissive_A_Brightness_Time__End}
   Alpha=NamedRerouteUsage()
  )
  n1520 = StaticSwitch(True=Max(A=LinearInterpolate(A=@n1516, B=@n1517, Alpha=@n1518), B=@n1519), False=@n1519, Value=@n1178)
  n1521 = ScalarParameter() {group=Effect_Decal_Reactive_Emissive_A, parameterName=EDRec_Emissive_A_CC-Animated}
  n1522 = Time()
  n1523 = StaticSwitchParameter(True=Add(A=@n50, B=ComponentMask(None=@n62) {r=true}), False=@n79) {group=Effect_Decal_Reactive_Emissive_A, parameterName=EDRec_Emissive_A_Warp_ Msk}
  n1524 = Multiply(
   A=ScalarParameter() {defaultValue=1, group=Effect_Decal_Reactive_Emissive_A, parameterName=EDRec_Emissive_A_CC_TileRate}
   B=LinearInterpolate(
    A=@n95
    B=StaticSwitchParameter(True=@n79, False=@n1523) {group=Effect_Decal_Reactive_Emissive_A, parameterName="EDRec_Emissive_A_Gradent ---> use Texture"}
    Alpha=ScalarParameter() {group=Effect_Decal_Reactive_Emissive_A, parameterName=EDRec_Emissive_A_CC_Remap--Grad}
   )
  )
  n1525 = ScalarParameter() {group=Effect_Decal_Reactive_Emissive_A, parameterName=EDRec_Emissive_A_CC_Offset}
  n1526 = StaticSwitch(True=Add(A=NamedRerouteUsage(), B=@n1525), False=@n1525, Value=NamedRerouteUsage())
  n1527 = Frac(None=Add(A=Multiply(A=@n1521, B=@n1522), B=Add(A=@n1524, B=@n1526)))
  n1528 = CurveAtlasRowParameter(CurveTime=@n1527) {
   atlas=&CruveAtlas_Wraps_01
   curve=&C_GradSimple
   group=Effect_Decal_Reactive_Emissive_A
   parameterName=EDRec_Emissive_A_CC
  }
  n1529 = MaterialFunctionCall(
   Index=ScalarParameter() {group=Effect_Decal_Reactive_Emissive_A, parameterName="EDRec_Emissive_A_CC  Index"}
   CCinput=@n1528
   CCinput_Alpha=@n1528
   UV=@n1527
  ) {materialFunction=&MF_Global_CC_Tex}
  n1530 = ScalarParameter() {group=Effect_Decal_Reactive_Emissive_A, parameterName=EDRec_Emissive_A_DarkenBase-CCAlpha MskSource, sliderMax=1, sortPriority=4}
  n1531 = StaticSwitchParameter(
   True=LinearInterpolate(A=@n95, B=NamedRerouteDeclaration(None=@n1529.Alpha) {name=CC_Alpha_A, nodeColor=<rgba 0 1 0.25 1>}, Alpha=@n1530)
   False=@n95
  ) {group=Effect_Decal_Reactive_Emissive_A, parameterName="EDRec-Emissive-A-Color--->CCurve"}
  NamedRerouteDeclaration(None=StaticSwitchParameter(True=Multiply(A=@n1531, B=@n1520), False=@n1531) {group=Effect_Decal_Reactive_Emissive_A_Reactive, parameterName=EDRec-Emissive-A_Reactive_Brightness---------------------}) {name="Emissive  A  Msk", nodeColor=<rgba 1 0.4453 0 1>}
  NamedRerouteDeclaration(None=ScalarParameter() {defaultValue=0.75, group=Effect_Decal_Reactive_RE, parameterName=EDRec_TimeFrom_BP_Curve, sortPriority=-1}) {name=BP_Curve, nodeColor=<rgba 0.1016 0 1 1>}
  NamedRerouteDeclaration(None=StaticSwitchParameter(True=Multiply(A=@n555, B=@n139), False=@n139) {group=Effect_Decal_Reactive_Emissive_B_Reactive, parameterName=EDRec-Emissive-B_Reactive_Brightness---------------------}) {name="Emissive B  Msk", nodeColor=<rgba 1 0.4453 0 1>}
  n1532 = StaticSwitchParameter(
   True=LinearInterpolate(A=@n183, B=NamedRerouteDeclaration(None=@n703.Alpha) {name=CC_Alpha_C, nodeColor=<rgba 0 1 0.25 1>}, Alpha=@n745)
   False=@n183
  ) {group=Effect_Decal_Reactive_Emissive_C, parameterName="EDRec-Emissive-C-Color--->CCurve"}
  NamedRerouteDeclaration(None=StaticSwitchParameter(True=Multiply(A=@n1532, B=@n693), False=@n1532) {group=Effect_Decal_Reactive_Emissive_C_Reactive, parameterName=EDRec-Emissive-C_Reactive_Brightness---------------------}) {name="Emissive  C  Msk", nodeColor=<rgba 1 0.4453 0 1>}
  n1533 = StaticSwitchParameter(
   True=LinearInterpolate(A=@n47, B=NamedRerouteDeclaration(None=@n791.Alpha) {name=CC_Alpha_D, nodeColor=<rgba 0 1 0.25 1>}, Alpha=@n827)
   False=@n47
  ) {group=Effect_Decal_Reactive_Emissive_D, parameterName="EDRec-Emissive-D-Color--->CCurve"}
  NamedRerouteDeclaration(None=StaticSwitchParameter(True=Multiply(A=@n1533, B=@n781), False=@n1533) {group=Effect_Decal_Reactive_Emissive_D_Reactive, parameterName=EDRec-Emissive-D_Reactive_Brightness---------------------}) {name="Emissive  D  Msk", nodeColor=<rgba 1 0.4453 0 1>}
  NamedRerouteDeclaration(None=@n258) {name=Emissive B State, nodeColor=<rgba 0 0.2734 1 1>}
  NamedRerouteDeclaration(None=@n265) {name=Emissive C State, nodeColor=<rgba 0 0.2734 1 1>}
  n1534 = VectorParameter() {
   group=Effect_Decal_Reactive
   parameterName="EDRec_Simple D  RMS"
   channelNames {a=Spec, b=Metalic, g=Roughness, r=Use}
   defaultValue=<rgba 0 0.5 0 0.5>
  }
  n1535 = Constant()
  n1536 = StaticSwitch(True=@n1534.RGBA, False=@n1535, Value=NamedRerouteUsage())
  n1537 = VectorParameter() {
   group=Effect_Decal_Reactive
   parameterName="EDRec_Simple C  RMS"
   channelNames {a=Spec, b=Metalic, g=Roughness, r=Use}
   defaultValue=<rgba 0 0.5 0 0.5>
  }
  n1538 = StaticSwitch(True=@n1537.RGBA, False=@n1535, Value=NamedRerouteUsage())
  n1539 = NamedRerouteUsage()
  n1540 = VectorParameter() {
   group=Effect_Decal_Reactive
   parameterName="EDRec_Simple B  RMS"
   channelNames {a=Spec, b=Metalic, g=Roughness, r=Use}
   defaultValue=<rgba 0 0.5 0 0.5>
  }
  n1541 = NamedRerouteUsage()
  n1542 = VectorParameter() {
   group=Effect_Decal_Reactive
   parameterName="EDRec_Simple A  RMS"
   channelNames {a=Spec, b=Metalic, g=Roughness, r=Use}
   defaultValue=<rgba 0 0.5 0 0.5>
  }
  MaterialExpressionLinearInterpolate_399 = LinearInterpolate(
   A=LinearInterpolate(
    A=LinearInterpolate(A=@n1536, B=@n1538, Alpha=@n1539)
    B=StaticSwitch(True=@n1540.RGBA, False=@n1535, Value=@n1541)
    Alpha=NamedRerouteUsage()
   )
   B=StaticSwitch(True=@n1542.RGBA, False=@n1535, Value=NamedRerouteUsage())
   Alpha=NamedRerouteUsage()
  )
  n1543 = Convert(None=@MaterialExpressionLinearInterpolate_399) {
   nodeName=Break Float4
   convertInputs [{
     type=Vector4
     defaultValue=<rgba 0 0 0 1>
     expressionInput {
      expression=&self:MaterialExpressionLinearInterpolate_399
      inputName=None
      mask=0
      maskA=0
      maskB=0
      maskG=0
      maskR=0
      outputIndex=0
     }
    }]
   convertOutputs [{type=Scalar}, {type=Scalar}, {type=Scalar}, {type=Scalar}]
  }
  n1544 = SetMaterialAttributes!(
   MaterialAttributes=NamedRerouteUsage()
   BaseColor=Constant3Vector()
   Roughness=@n1543
   Specular=@n1543
   Metallic=@n1543
   EmissiveColor
   Normal=NamedRerouteUsage()
  ) {attributeSetTypes [
    #3
    #4
    #5
    #6
    #2
    #1
   ]}
  NamedRerouteDeclaration(None=@n1544) {name=Simple Material, nodeColor=<rgba 1 0 0.5156 1>}
  NamedRerouteDeclaration(None=@n1543) {name=Simple Use, nodeColor=<rgba 0 0.8359 1 1>}
  NamedRerouteDeclaration(None=@n248) {name=Normal, nodeColor=<rgba 0.9219 0 1 1>}
  n1545 = ScalarParameter() {defaultValue=1, group="Effect_Decal_Reactive_Emissive_A_Reactive | Grad\\Mask", parameterName=EDRec_Emissive_A_Grad Start}
  n1546 = ScalarParameter() {defaultValue=1, group="Effect_Decal_Reactive_Emissive_A_Reactive | Grad\\Mask", parameterName=EDRec_Emissive_A_Grad End}
  n1547 = ScalarParameter() {defaultValue=-9999999, group=Effect_Decal_Reactive_Emissive_A_Reactive, parameterName=EDRec_Emissive_A_Brightness_Time_BP, sortPriority=50}
  n1548 = Subtract(
   A=Time()
   B=Add(
    A=@n1547
    B=ScalarParameter() {group="Effect_Decal_Reactive_Emissive_A_Reactive | Grad\\Mask", parameterName="EDRec_Emissive_A_MAsk\\Grad_Delay"}
   )
  )
  n1549 = ScalarParameter() {defaultValue=0.25, group="Effect_Decal_Reactive_Emissive_A_Reactive | Grad\\Mask", parameterName="EDRec_Emissive_A_MAsk\\Grad_Duration"}
  n1550 = ScalarParameter() {defaultValue=1, group="Effect_Decal_Reactive_Emissive_A_Reactive | Grad\\Mask", parameterName="EDRec_Emissive_A_MAsk\\Grad_Falloff"}
  n1551 = Power(Base=Saturate(None=Divide(A=@n1548, B=@n1549)), Exp=@n1550)
  n1552 = Constant() {r=1}
  n1553 = StaticBoolParameter() {group=Effect_Decal_Reactive_Emissive_A_Reactive, parameterName=EDRec-Emissive-A_Reactive_Brightness---------------------}
  NamedRerouteDeclaration(None=StaticSwitch(True=LinearInterpolate(A=@n1545, B=@n1546, Alpha=@n1551), False=@n1552, Value=@n1553)) {name=Reactive_Grad A, nodeColor=<rgba 0.8516 0 1 1>}
  n1554 = ScalarParameter() {defaultValue=1, group="Effect_Decal_Reactive_Emissive_A_Reactive | Grad\\Mask", parameterName=EDRec_Emissive_A_Mask Start}
  n1555 = ScalarParameter() {defaultValue=1, group="Effect_Decal_Reactive_Emissive_A_Reactive | Grad\\Mask", parameterName=EDRec_Emissive_A_Mask End}
  NamedRerouteDeclaration(None=StaticSwitch(True=LinearInterpolate(A=@n1554, B=@n1555, Alpha=@n1551), False=@n1552, Value=@n1553)) {name=Reactive_Mask A, nodeColor=<rgba 0.8516 0 1 1>}
  n1556 = TextureObjectParameter() {
   group=TopSide_V02
   parameterName=TopS_00A_Sharerd_Tex_Msk
   samplerType=SAMPLERTYPE_Masks
   sortPriority=249
   texture=&T_BlackMask
  }
  n1557 = TextureObjectParameter() {
   group=TopSide_V02
   parameterName=TopS_00B_Sharerd_Tex_Msk
   samplerType=SAMPLERTYPE_Masks
   sortPriority=249
   texture=&T_BlackMask
  }
  n1558 = StaticSwitchParameter(
   True=NamedRerouteDeclaration(None=@n1556) {name=Tex_00A_Shared_Tex, nodeColor=<rgba 1 0.8906 0 1>}
   False=StaticSwitchParameter(True=NamedRerouteDeclaration(None=@n1557) {name=Tex_00B_Shared_Tex, nodeColor=<rgba 1 0.8906 0 1>}, False=@n1311) {group=Effect_Decal_Reactive, parameterName="EDRec_Decal_Mask__Tex -------->TopS_00B_Sharerd_Tex_Msk"}
  ) {group=Effect_Decal_Reactive, parameterName="EDRec_Decal_Mask__Tex -------->TopS_00A_Sharerd_Tex_Msk"}
  n1559 = TextureObjectParameter() {group=Effect_Decal_Reactive, parameterName=EDRec_Decal_BC__Tex, sortPriority=10, texture=&T_BlacksRGB}
  NamedRerouteDeclaration(None=StaticSwitchParameter(True=NamedRerouteDeclaration(None=@n1558) {name=Decal_Msk, nodeColor=<rgba 0 1 0.9062 1>}, False=@n1559) {group=Effect_Decal_Reactive, parameterName=EDRec_Emissive_only, sortPriority=-10}) {name=Decal_BC, nodeColor=<rgba 0 1 0.9063 1>}
  n1560 = ScalarParameter() {defaultValue=1, group="Effect_Decal_Reactive_Emissive_B_Reactive | Grad\\Mask", parameterName=EDRec_Emissive_B_Grad Start}
  n1561 = ScalarParameter() {defaultValue=1, group="Effect_Decal_Reactive_Emissive_B_Reactive | Grad\\Mask", parameterName=EDRec_Emissive_B_Grad End}
  n1562 = ScalarParameter() {defaultValue=-9999999, group=Effect_Decal_Reactive_Emissive_B_Reactive, parameterName=EDRec_Emissive_B_Brightness_Time_BP, sortPriority=50}
  n1563 = Subtract(
   A=Time()
   B=Add(
    A=@n1562
    B=ScalarParameter() {group="Effect_Decal_Reactive_Emissive_B_Reactive | Grad\\Mask", parameterName="EDRec_Emissive_B_MAsk\\Grad_Delay"}
   )
  )
  n1564 = ScalarParameter() {defaultValue=0.25, group="Effect_Decal_Reactive_Emissive_B_Reactive | Grad\\Mask", parameterName="EDRec_Emissive_B_MAsk\\Grad_Duration"}
  n1565 = ScalarParameter() {defaultValue=1, group="Effect_Decal_Reactive_Emissive_B_Reactive | Grad\\Mask", parameterName="EDRec_Emissive_B_MAsk\\Grad_Falloff"}
  n1566 = Power(Base=Saturate(None=Divide(A=@n1563, B=@n1564)), Exp=@n1565)
  n1567 = Constant() {r=1}
  n1568 = StaticBoolParameter() {group=Effect_Decal_Reactive_Emissive_B_Reactive, parameterName=EDRec-Emissive-B_Reactive_Brightness---------------------}
  NamedRerouteDeclaration(None=StaticSwitch(True=LinearInterpolate(A=@n1560, B=@n1561, Alpha=@n1566), False=@n1567, Value=@n1568)) {name=Reactive_Grad B, nodeColor=<rgba 0.8516 0 1 1>}
  n1569 = ScalarParameter() {defaultValue=1, group="Effect_Decal_Reactive_Emissive_B_Reactive | Grad\\Mask", parameterName=EDRec_Emissive_B_Mask Start}
  n1570 = ScalarParameter() {defaultValue=1, group="Effect_Decal_Reactive_Emissive_B_Reactive | Grad\\Mask", parameterName=EDRec_Emissive_B_Mask End}
  NamedRerouteDeclaration(None=StaticSwitch(True=LinearInterpolate(A=@n1569, B=@n1570, Alpha=@n1566), False=@n1567, Value=@n1568)) {name=Reactive_Mask B, nodeColor=<rgba 0.8516 0 1 1>}
  n1571 = ScalarParameter() {defaultValue=1, group="Effect_Decal_Reactive_Emissive_C_Reactive | Grad\\Mask", parameterName=EDRec_Emissive_C_Mask Start}
  n1572 = ScalarParameter() {defaultValue=1, group="Effect_Decal_Reactive_Emissive_C_Reactive | Grad\\Mask", parameterName=EDRec_Emissive_C_Mask End}
  n1573 = ScalarParameter() {defaultValue=-9999999, group=Effect_Decal_Reactive_Emissive_C_Reactive, parameterName=EDRec_Emissive_C_Brightness_Time_BP, sortPriority=50}
  n1574 = Subtract(
   A=Time()
   B=Add(
    A=@n1573
    B=ScalarParameter() {group="Effect_Decal_Reactive_Emissive_C_Reactive | Grad\\Mask", parameterName="EDRec_Emissive_C_MAsk\\Grad_Delay"}
   )
  )
  n1575 = ScalarParameter() {defaultValue=0.25, group="Effect_Decal_Reactive_Emissive_C_Reactive | Grad\\Mask", parameterName="EDRec_Emissive_C_MAsk\\Grad_Duration"}
  n1576 = ScalarParameter() {defaultValue=1, group="Effect_Decal_Reactive_Emissive_C_Reactive | Grad\\Mask", parameterName="EDRec_Emissive_C_MAsk\\Grad_Falloff"}
  n1577 = Power(Base=Saturate(None=Divide(A=@n1574, B=@n1575)), Exp=@n1576)
  n1578 = Constant() {r=1}
  n1579 = StaticBoolParameter() {group=Effect_Decal_Reactive_Emissive_C_Reactive, parameterName=EDRec-Emissive-C_Reactive_Brightness---------------------}
  NamedRerouteDeclaration(None=StaticSwitch(True=LinearInterpolate(A=@n1571, B=@n1572, Alpha=@n1577), False=@n1578, Value=@n1579)) {name=Reactive_Mask C, nodeColor=<rgba 0.8516 0 1 1>}
  n1580 = ScalarParameter() {defaultValue=1, group="Effect_Decal_Reactive_Emissive_C_Reactive | Grad\\Mask", parameterName=EDRec_Emissive_C_Grad Start}
  n1581 = ScalarParameter() {defaultValue=1, group="Effect_Decal_Reactive_Emissive_C_Reactive | Grad\\Mask", parameterName=EDRec_Emissive_C_Grad End}
  NamedRerouteDeclaration(None=StaticSwitch(True=LinearInterpolate(A=@n1580, B=@n1581, Alpha=@n1577), False=@n1578, Value=@n1579)) {name=Reactive_Grad C, nodeColor=<rgba 0.8516 0 1 1>}
  n1582 = ScalarParameter() {defaultValue=1, group="Effect_Decal_Reactive_Emissive_D_Reactive | Grad\\Mask", parameterName=EDRec_Emissive_D_Mask Start}
  n1583 = ScalarParameter() {defaultValue=1, group="Effect_Decal_Reactive_Emissive_D_Reactive | Grad\\Mask", parameterName=EDRec_Emissive_D_Mask End}
  n1584 = ScalarParameter() {defaultValue=-9999999, group=Effect_Decal_Reactive_Emissive_D_Reactive, parameterName=EDRec_Emissive_D_Brightness_Time_BP, sortPriority=50}
  n1585 = Subtract(
   A=Time()
   B=Add(
    A=@n1584
    B=ScalarParameter() {group="Effect_Decal_Reactive_Emissive_D_Reactive | Grad\\Mask", parameterName="EDRec_Emissive_D_MAsk\\Grad_Delay"}
   )
  )
  n1586 = ScalarParameter() {defaultValue=0.25, group="Effect_Decal_Reactive_Emissive_D_Reactive | Grad\\Mask", parameterName="EDRec_Emissive_D_MAsk\\Grad_Duration"}
  n1587 = ScalarParameter() {defaultValue=1, group="Effect_Decal_Reactive_Emissive_D_Reactive | Grad\\Mask", parameterName="EDRec_Emissive_D_MAsk\\Grad_Falloff"}
  n1588 = Power(Base=Saturate(None=Divide(A=@n1585, B=@n1586)), Exp=@n1587)
  n1589 = Constant() {r=1}
  n1590 = StaticBoolParameter() {group=Effect_Decal_Reactive_Emissive_D_Reactive, parameterName=EDRec-Emissive-D_Reactive_Brightness---------------------}
  NamedRerouteDeclaration(None=StaticSwitch(True=LinearInterpolate(A=@n1582, B=@n1583, Alpha=@n1588), False=@n1589, Value=@n1590)) {name=Reactive_Mask D, nodeColor=<rgba 0.8516 0 1 1>}
  n1591 = ScalarParameter() {defaultValue=1, group="Effect_Decal_Reactive_Emissive_D_Reactive | Grad\\Mask", parameterName=EDRec_Emissive_D_Grad Start}
  n1592 = ScalarParameter() {defaultValue=1, group="Effect_Decal_Reactive_Emissive_D_Reactive | Grad\\Mask", parameterName=EDRec_Emissive_D_Grad End}
  NamedRerouteDeclaration(None=StaticSwitch(True=LinearInterpolate(A=@n1591, B=@n1592, Alpha=@n1588), False=@n1589, Value=@n1590)) {name=Reactive_Grad D, nodeColor=<rgba 0.8516 0 1 1>}
  n1593 = ScalarParameter() {defaultValue=20, group=Effect_Decal_Reactive, parameterName=EDRec_Tri_Contrast, sortPriority=-1}
  NamedRerouteDeclaration(None=Max(
    A=MaterialFunctionCall(In=@n1341, Contrast=@n1593) {materialFunction=&CheapContrast}.Result
    B=ScalarParameter() {group=Effect_Decal_Reactive, parameterName="EDRec_Tri_Disable =1", sortPriority=-1}
   )) {name=Tri_Mask, nodeColor=<rgba 1 0 0.4922 1>}
  n1594 = VectorParameter() {group=Effect_Decal_Reactive_FlipBook, parameterName=EDRec_DecalA__Sheet_Phase, defaultValue=<rgba 0 0 0 1>}
  n1595 = MaterialFunctionCall(Color=NamedRerouteUsage()) {materialFunction=&SplitComponents}
  n1596 = NamedRerouteUsage()
  n1597 = Multiply(A=Time(), B=ScalarParameter() {group=Effect_Decal_Reactive_Uv_Animation, parameterName=EDRec_DecalA_Rotation_Speed})
  n1598 = MaterialFunctionCall!(
   For_Safe_Zones=StaticBool() {value=true}
   Normal_Map
   Caculated_Bounds=NamedRerouteUsage()
   Pre_Skinned_Normals=NamedRerouteUsage()
   Is Vehicle=@n1595.R
   Is VehicleSide=@n1595.R
   Is Mech=@n1595.R
   Vehicle_Rotation
   UV_Rotation Bypass=NamedRerouteDeclaration(None=Add(A=@n1596, B=@n1597)) {name=Rotation_A, nodeColor=<rgba 0 1 0.2969 1>}
   Mech_Rotation
   Uvs_Offset=ComponentMask(None=NamedRerouteUsage()) {b=true, g=true, r=true}
   Offset-after rotation
   Uvs_TileRate=ComponentMask(None=Constant() {r=0.1}) {a=true}
  ) {materialFunction=&/MF_Projection_Side_Normal}
  n1599 = ScalarParameter() {defaultValue=2, group=Effect_Decal_Reactive_FlipBook, parameterName=EDRec_Decal_Sheet_Number Rows/Colums, sortPriority=0}
  n1600 = Multiply(A=Add(A=@n1598.UVs, B=ComponentMask(None=@n1594.RGB) {g=true, r=true}), B=Divide(B=@n1599) {constB=3})
  n1601 = NamedRerouteUsage()
  n1602 = TextureSample!(UVs=@n1600, Tex=NamedRerouteUsage(), Level=@n1601, Apply View MipBias) {mipValueMode=TMVM_MipLevel}
  n1603 = Abs(None=Multiply(A=Add(B=@n1598.UVs) {constA=-0.5}) {constB=2})
  n1604 = Saturate(None=Ceil(None=OneMinus(None=@n1603)))
  n1605 = Multiply(A=ComponentMask(None=@n1604) {r=true}, B=ComponentMask(None=@n1604) {g=true})
  n1606 = Multiply(A=@n1602.A, B=@n1605)
  n1607 = StaticSwitch(True=Multiply(A=@n1606, B=NamedRerouteUsage()), False=@n1606, Value=NamedRerouteUsage())
  n1608 = VectorParameter() {group=Effect_Decal_Reactive_FlipBook, parameterName=EDRec_DecalB__Sheet_Phase, defaultValue=<rgba 0 0 0 1>}
  n1609 = MaterialFunctionCall(Color=NamedRerouteUsage()) {materialFunction=&SplitComponents}
  n1610 = NamedRerouteUsage()
  n1611 = Multiply(A=Time(), B=ScalarParameter() {group=Effect_Decal_Reactive_Uv_Animation, parameterName=EDRec_DecalB_Rotation_Speed})
  n1612 = NamedRerouteUsage()
  n1613 = MaterialFunctionCall(
   For_Safe_Zones=StaticBool() {value=true}
   Caculated_Bounds=NamedRerouteUsage()
   Pre_Skinned_Normals=NamedRerouteUsage()
   Is Vehicle=@n1609.R
   Is VehicleSide=@n1609.R
   Is Mech=@n1609.R
   UV_Rotation Bypass=NamedRerouteDeclaration(None=Add(A=@n1610, B=@n1611)) {name=Rotation_B, nodeColor=<rgba 0 1 0.2969 1>}
   Uvs_Offset=ComponentMask(None=@n1612) {b=true, g=true, r=true}
   Uvs_TileRate=ComponentMask(None=@n1612) {a=true}
  ) {materialFunction=&MF_Projection_Side_Normal}
  n1614 = Multiply(A=Add(A=@n1613.UVs, B=ComponentMask(None=@n1608.RGB) {g=true, r=true}), B=Divide(B=@n1599) {constB=3})
  n1615 = NamedRerouteUsage()
  n1616 = TextureSample(UVs=@n1614, Tex=NamedRerouteUsage(), Level=@n1615) {mipValueMode=TMVM_MipLevel}
  n1617 = Abs(None=Multiply(A=Add(B=@n1613.UVs) {constA=-0.5}) {constB=2})
  n1618 = Saturate(None=Ceil(None=OneMinus(None=@n1617)))
  n1619 = Multiply(A=ComponentMask(None=@n1618) {r=true}, B=ComponentMask(None=@n1618) {g=true})
  n1620 = Multiply(A=@n1616.A, B=@n1619)
  n1621 = StaticSwitch(True=Multiply(A=@n1620, B=NamedRerouteUsage()), False=@n1620, Value=NamedRerouteUsage())
  n1622 = VectorParameter() {group=Effect_Decal_Reactive_FlipBook, parameterName=EDRec_DecalC__Sheet_Phase, defaultValue=<rgba 0 0 0 1>}
  n1623 = MaterialFunctionCall(Color=NamedRerouteUsage()) {materialFunction=&SplitComponents}
  n1624 = NamedRerouteUsage()
  n1625 = Multiply(A=Time(), B=ScalarParameter() {group=Effect_Decal_Reactive_Uv_Animation, parameterName=EDRec_DecalC_Rotation_Speed})
  n1626 = NamedRerouteUsage()
  n1627 = MaterialFunctionCall(
   For_Safe_Zones=StaticBool() {value=true}
   Caculated_Bounds=NamedRerouteUsage()
   Pre_Skinned_Normals=NamedRerouteUsage()
   Is Vehicle=@n1623.R
   Is VehicleSide=@n1623.R
   Is Mech=@n1623.R
   UV_Rotation Bypass=NamedRerouteDeclaration(None=Add(A=@n1624, B=@n1625)) {name=Rotation_C, nodeColor=<rgba 0 1 0.2969 1>}
   Uvs_Offset=ComponentMask(None=@n1626) {b=true, g=true, r=true}
   Uvs_TileRate=ComponentMask(None=@n1626) {a=true}
  ) {materialFunction=&MF_Projection_Side_Normal}
  n1628 = Multiply(A=Add(A=@n1627.UVs, B=ComponentMask(None=@n1622.RGB) {g=true, r=true}), B=Divide(B=@n1599) {constB=3})
  n1629 = NamedRerouteUsage()
  n1630 = TextureSample(UVs=@n1628, Tex=NamedRerouteUsage(), Level=@n1629) {mipValueMode=TMVM_MipLevel}
  n1631 = Abs(None=Multiply(A=Add(B=@n1627.UVs) {constA=-0.5}) {constB=2})
  n1632 = Saturate(None=Ceil(None=OneMinus(None=@n1631)))
  n1633 = Multiply(A=ComponentMask(None=@n1632) {r=true}, B=ComponentMask(None=@n1632) {g=true})
  n1634 = Multiply(A=@n1630.A, B=@n1633)
  n1635 = StaticSwitch(True=Multiply(A=@n1634, B=NamedRerouteUsage()), False=@n1634, Value=NamedRerouteUsage())
  n1636 = VectorParameter() {group=Effect_Decal_Reactive_FlipBook, parameterName=EDRec_DecalD__Sheet_Phase, defaultValue=<rgba 0 0 0 1>}
  n1637 = MaterialFunctionCall(Color=NamedRerouteUsage()) {materialFunction=&SplitComponents}
  n1638 = NamedRerouteUsage()
  n1639 = Multiply(A=Time(), B=ScalarParameter() {group=Effect_Decal_Reactive_Uv_Animation, parameterName=EDRec_DecalD_Rotation_Speed})
  n1640 = NamedRerouteUsage()
  n1641 = MaterialFunctionCall(
   For_Safe_Zones=StaticBool() {value=true}
   Caculated_Bounds=NamedRerouteUsage()
   Pre_Skinned_Normals=NamedRerouteUsage()
   Is Vehicle=@n1637.R
   Is VehicleSide=@n1637.R
   Is Mech=@n1637.R
   UV_Rotation Bypass=NamedRerouteDeclaration(None=Add(A=@n1638, B=@n1639)) {name=Rotation_D, nodeColor=<rgba 0 1 0.2969 1>}
   Uvs_Offset=ComponentMask(None=@n1640) {b=true, g=true, r=true}
   Uvs_TileRate=ComponentMask(None=@n1640) {a=true}
  ) {materialFunction=&MF_Projection_Side_Normal}
  n1642 = Multiply(A=Add(A=@n1641.UVs, B=ComponentMask(None=@n1636.RGB) {g=true, r=true}), B=Divide(B=@n1599) {constB=3})
  n1643 = NamedRerouteUsage()
  n1644 = TextureSample(UVs=@n1642, Tex=NamedRerouteUsage(), Level=@n1643) {mipValueMode=TMVM_MipLevel}
  n1645 = Abs(None=Multiply(A=Add(B=@n1641.UVs) {constA=-0.5}) {constB=2})
  n1646 = Saturate(None=Ceil(None=OneMinus(None=@n1645)))
  n1647 = Multiply(A=ComponentMask(None=@n1646) {r=true}, B=ComponentMask(None=@n1646) {g=true})
  n1648 = Multiply(A=@n1644.A, B=@n1647)
  n1649 = StaticSwitch(True=Multiply(A=@n1648, B=NamedRerouteUsage()), False=@n1648, Value=NamedRerouteUsage())
  n1650 = Add(A=Add(A=@n1607, B=@n1621), B=Add(A=@n1635, B=@n1649))
  NamedRerouteDeclaration(None=Multiply(A=NamedRerouteUsage(), B=Saturate(None=@n1650))) {name=Alpha_Set_A, nodeColor=<rgba 0.8281 1 0 1>}
  n1651 = StaticSwitch(True=Constant(), False=@n1128, Value=@n549)
  n1652 = Add(A=@n1079, B=StaticSwitch(True=@n959, False=Add(A=@n1012, B=@n1651), Value=@n509))
  NamedRerouteDeclaration(None=Multiply(A=NamedRerouteUsage(), B=Saturate(None=@n1652))) {name=Alpha_Set_B, nodeColor=<rgba 0.8047 0 1 1>}
  n1653 = ScalarParameter() {defaultValue=-9999999, group=Effect_Decal_Reactive_RE, parameterName=EDRec_TimeFrom_BP_DecalA, sortPriority=0}
  n1654 = Subtract(A=Time(), B=Add(A=ScalarParameter() {group=Effect_Decal_Reactive_RE, parameterName=EDRec_DecalA_Delay--------, sortPriority=0}, B=@n1653))
  n1655 = Saturate(None=Divide(A=@n1654, B=ScalarParameter() {defaultValue=0.25, group=Effect_Decal_Reactive_RE, parameterName=EDRec_DecalA_Duration, sortPriority=0}))
  n1656 = NamedRerouteUsage()
  NamedRerouteDeclaration(None=LinearInterpolate(
    A=ScalarParameter() {defaultValue=1, group=Effect_Decal_Reactive_RE, parameterName=EDRec_DecalA_Opacity_Start, sortPriority=0}
    B=ScalarParameter() {defaultValue=1, group=Effect_Decal_Reactive_RE, parameterName=EDRec_DecalA_Opacity_End, sortPriority=0}
    Alpha=Power(Base=@n1655, Exp=@n1656)
   )) {name=Reactive_Opacity_A, nodeColor=<rgba 0.217 0.5686 1 1>}
  n1657 = VectorParameter() {group=Effect_Decal_Reactive, parameterName=EDRec_Decal_BackGroundColor, defaultValue=<rgba 0 0 0 1>}
  n1658 = Multiply(A=@n1114.RGB, B=@n1128)
  n1659 = StaticSwitch(
   True=VectorParameter() {group=Effect_Decal_Reactive, parameterName=EDRec_Decal_BackGroundColor, defaultValue=<rgba 0 0 0 1>}.RGB
   False=LinearInterpolate(A=StaticSwitch(True=@n1657.RGB, False=@n1658, Value=@n549), B=StaticSwitch(True=@n998.RGB, False=@n1000.RGB, Value=@n627), Alpha=@n1012)
   Value=@n509
  )
  n1660 = ComponentMask(None=@n684) {b=true, g=true, r=true}
  NamedRerouteDeclaration(None=LinearInterpolate(A=LinearInterpolate(A=@n1659, B=@n1660, Alpha=@n1024), B=StaticSwitch(True=@n1068.RGB, False=@n1070.RGB, Value=@n402), Alpha=@n1078)) {name=BC_Set_B, nodeColor=<rgba 0.8047 0 1 1>}
  n1661 = ScalarParameter() {group=Effect_Decal_Reactive_RE_Rotation, parameterName=EDRec_DecalD_Rotation_Amount, sortPriority=21}
  n1662 = ScalarParameter() {defaultValue=0.25, group=Effect_Decal_Reactive_RE_Rotation, parameterName=EDRec_DecalD_Rotation-Duration, sortPriority=21}
  n1663 = ScalarParameter() {defaultValue=-9999999, group=Effect_Decal_Reactive_RE, parameterName=EDRec_TimeFrom_BP_DecalD, sortPriority=20}
  n1664 = Subtract(
   A=Time()
   B=Add(A=ScalarParameter() {group=Effect_Decal_Reactive_RE, parameterName=EDRec_DecalD_Delay-----------------, sortPriority=20}, B=@n1663)
  )
  n1665 = Power(
   Base=Saturate(None=Divide(A=@n1664, B=@n1662))
   Exp=ScalarParameter() {defaultValue=1, group=Effect_Decal_Reactive_RE_Rotation, parameterName=EDRec_DecalD_Rotation-Curve, sortPriority=21}
  )
  n1666 = StaticBoolParameter() {defaultValue=true, group=Effect_Decal_Reactive_RE_Rotation, parameterName=EDRec_DecalD_Rotation_Reactive, sortPriority=-10}
  NamedRerouteDeclaration(None=StaticSwitch(True=LinearInterpolate(B=@n1661, Alpha=@n1665), False=Constant(), Value=@n1666)) {name=Reactive_Rotation_D, nodeColor=<rgba 0.217 0.5686 1 1>}
  n1667 = ScalarParameter() {group=Effect_Decal_Reactive_RE_Rotation, parameterName=EDRec_DecalC_Rotation_Amount, sortPriority=11}
  n1668 = ScalarParameter() {defaultValue=-9999999, group=Effect_Decal_Reactive_RE, parameterName=EDRec_TimeFrom_BP_DecalC, sortPriority=10}
  n1669 = Subtract(A=Time(), B=Add(A=ScalarParameter() {group=Effect_Decal_Reactive_RE, parameterName=EDRec_DecalC_Delay---------, sortPriority=10}, B=@n1668))
  n1670 = ScalarParameter() {defaultValue=0.25, group=Effect_Decal_Reactive_RE_Rotation, parameterName=EDRec_DecalC_Rotation-Duration, sortPriority=11}
  n1671 = Power(
   Base=Saturate(None=Divide(A=@n1669, B=@n1670))
   Exp=ScalarParameter() {defaultValue=1, group=Effect_Decal_Reactive_RE_Rotation, parameterName=EDRec_DecalC_Rotation-Curve, sortPriority=11}
  )
  NamedRerouteDeclaration(None=StaticSwitch(True=LinearInterpolate(B=@n1667, Alpha=@n1671), False=Constant(), Value=@n1666)) {name=Reactive_Rotation_C, nodeColor=<rgba 0.217 0.5686 1 1>}
  n1672 = ScalarParameter() {group=Effect_Decal_Reactive_RE_Rotation, parameterName=EDRec_DecalB_Rotation_Amount, sortPriority=7}
  n1673 = ScalarParameter() {defaultValue=-9999999, group=Effect_Decal_Reactive_RE, parameterName=EDRec_TimeFrom_BP_DecalB, sortPriority=6}
  n1674 = Subtract(A=Time(), B=Add(A=ScalarParameter() {group=Effect_Decal_Reactive_RE, parameterName=EDRec_DecalB_Delay--------, sortPriority=6}, B=@n1673))
  n1675 = ScalarParameter() {defaultValue=0.25, group=Effect_Decal_Reactive_RE_Rotation, parameterName=EDRec_DecaB_Rotation-Duration, sortPriority=7}
  n1676 = Power(
   Base=Saturate(None=Divide(A=@n1674, B=@n1675))
   Exp=ScalarParameter() {defaultValue=1, group=Effect_Decal_Reactive_RE_Rotation, parameterName=EDRec_DecalB_Rotation-Curve, sortPriority=7}
  )
  NamedRerouteDeclaration(None=StaticSwitch(True=LinearInterpolate(B=@n1672, Alpha=@n1676), False=Constant(), Value=@n1666)) {name=Reactive_Rotation_B, nodeColor=<rgba 0.217 0.5686 1 1>}
  n1677 = ScalarParameter() {group=Effect_Decal_Reactive_RE_Rotation, parameterName=EDRec_DecalA_Rotation_Amount, sortPriority=5}
  n1678 = ScalarParameter() {defaultValue=0.25, group=Effect_Decal_Reactive_RE_Rotation, parameterName=EDRec_DecaA_Rotation-Duration, sortPriority=5}
  n1679 = Power(
   Base=Saturate(None=Divide(A=@n1654, B=@n1678))
   Exp=ScalarParameter() {defaultValue=1, group=Effect_Decal_Reactive_RE_Rotation, parameterName=EDRec_DecalA_Rotation-Curve, sortPriority=5}
  )
  NamedRerouteDeclaration(None=StaticSwitch(True=LinearInterpolate(B=@n1677, Alpha=@n1679), False=Constant(), Value=@n1666)) {name=Reactive_Rotation_A, nodeColor=<rgba 0.217 0.5686 1 1>}
  n1680 = Divide(A=TextureProperty(Tex=NamedRerouteUsage()), B=Multiply(A=NamedRerouteUsage(), B=NamedRerouteUsage()))
  n1681 = ScalarParameter() {defaultValue=0.5, group=Effect_Decal_Reactive, parameterName=EDRec_Miplevel_Flip_Adjustment, sortPriority=0}
  NamedRerouteDeclaration(None=MaterialFunctionCall!(UVs=NamedRerouteUsage(), Texture Size=Multiply(A=@n1680, B=@n1681)) {materialFunction=/Engine/Functions/Engine_MaterialFunctions01/Texturing/ComputeMipLevel}.MipLevel) {name=mips, nodeColor=<rgba 1 0 0.1641 1>}
  n1682 = Divide(A=TextureProperty(Tex=NamedRerouteUsage()), B=Multiply(A=@n1599, B=@n1599))
  n1683 = ScalarParameter() {defaultValue=0.5, group=Effect_Decal_Reactive, parameterName=EDRec_Miplevel_Flip_Adjustment, sortPriority=0}
  NamedRerouteDeclaration(None=MaterialFunctionCall(Texture Size=Multiply(A=@n1682, B=@n1683)) {materialFunction=&ComputeMipLevel}.MipLevel) {name=mips_A, nodeColor=<rgba 1 0 0.1641 1>}
  n1684 = TextureObjectParameter() {
   group=Effect_Decal_Reactive
   parameterName=EDRec_Decal_Normal__Tex
   samplerType=SAMPLERTYPE_Normal
   sortPriority=10
   texture=&FlatNormalBlue
  }
  NamedRerouteDeclaration(None=@n1684) {name=Decal_Nrml, nodeColor=<rgba 0 1 0.9062 1>}
  n1685 = Saturate(None=Divide(A=@n1674, B=ScalarParameter() {defaultValue=0.25, group=Effect_Decal_Reactive_RE, parameterName=EDRec_DecalB_Duration, sortPriority=6}))
  n1686 = NamedRerouteUsage()
  NamedRerouteDeclaration(None=LinearInterpolate(
    A=ScalarParameter() {defaultValue=1, group=Effect_Decal_Reactive_RE, parameterName=EDRec_DecalB_Opacity_Start, sortPriority=6}
    B=ScalarParameter() {defaultValue=1, group=Effect_Decal_Reactive_RE, parameterName=EDRec_DecalB_Opacity_End, sortPriority=6}
    Alpha=Power(Base=@n1685, Exp=@n1686)
   )) {name=Reactive_Opacity_B, nodeColor=<rgba 0.217 0.5686 1 1>}
  NamedRerouteDeclaration(None=@n1598.UVs) {name=UVsA, nodeColor=<rgba 1 0 0.3984 1>}
  NamedRerouteDeclaration(None=@n1613.UVs) {name=UVsB, nodeColor=<rgba 1 0 0.3984 1>}
  NamedRerouteDeclaration(None=GetMaterialAttributes(None=FunctionInput() {bUsePreviewValueAsDefault=true, inputName=Root_Matieral, inputType=FunctionInput_MaterialAttributes, sortPriority=-20}) {attributeGetTypes [#1]}.Normal) {name=Root_Normals, nodeColor=<rgba 0 1 0.2031 1>}
  n1687 = TextureSample(UVs=@n1642, Tex=NamedRerouteUsage(), Level=@n1643) {mipValueMode=TMVM_MipLevel}
  n1688 = StaticBool() {value=true}
  n1689 = NamedRerouteUsage()
  n1690 = NamedRerouteUsage()
  n1691 = NamedRerouteUsage()
  n1692 = NamedRerouteUsage()
  n1693 = LinearInterpolate(
   A=LinearInterpolate(
    B=MaterialFunctionCall!(
     Normal_Map
     Normals_Source=@n1687.RGB
     Normals_Source_Use=@n1688
     UVs
     Rotation_Comp=@n1689
     UV_Rotation
     Pre_Skinned_Normals=@n1690
     is_State=@n1691
     "Left\\Right"
     No_Rotation
    ) {materialFunction=&/MF_UVs_Side_Fix_Normals}.Raw_Normal
    Alpha=@n1649
   )
   B=MaterialFunctionCall(
    Normals_Source=TextureSample(UVs=@n1628, Tex=@n1692, Level=@n1629) {mipValueMode=TMVM_MipLevel}.RGB
    Normals_Source_Use=StaticBool() {value=true}
    UV_Rotation=NamedRerouteUsage()
    Pre_Skinned_Normals=NamedRerouteUsage()
    is_State=NamedRerouteUsage()
   ) {materialFunction=&MF_UVs_Side_Fix_Normals}.Raw_Normal
   Alpha=@n1635
  )
  n1694 = NamedRerouteUsage()
  n1695 = Multiply(
   A=@n1619
   B=MaterialFunctionCall(
    Normals_Source=TextureSample(UVs=@n1614, Tex=@n1694, Level=@n1615) {mipValueMode=TMVM_MipLevel}.RGB
    Normals_Source_Use=StaticBool() {value=true}
    Rotation_Comp=NamedRerouteUsage()
    Pre_Skinned_Normals=NamedRerouteUsage()
    is_State=NamedRerouteUsage()
   ) {materialFunction=&MF_UVs_Side_Fix_Normals}.Raw_Normal
  )
  n1696 = MaterialFunctionCall(
   Normals_Source=TextureSample(UVs=@n1600, Tex=NamedRerouteUsage(), Level=@n1601) {mipValueMode=TMVM_MipLevel}.RGB
   Normals_Source_Use=StaticBool() {value=true}
   UV_Rotation=NamedRerouteUsage()
   Pre_Skinned_Normals=NamedRerouteUsage()
   is_State=NamedRerouteUsage()
  ) {materialFunction=&MF_UVs_Side_Fix_Normals}
  NamedRerouteDeclaration(None=LinearInterpolate(A=LinearInterpolate(A=@n1693, B=@n1695, Alpha=@n1621), B=Multiply(A=@n1605, B=@n1696.Raw_Normal), Alpha=@n1607)) {name=Side_Normals, nodeColor=<rgba 0 0.6953 1 1>}
  n1697 = ScalarParameter() {defaultValue=1, group=Effect_Decal_Reactive_UV_Warp, parameterName=EDRec_Warp_Scale_Weapon}
  n1698 = MaterialFunctionCall(Float4=NamedRerouteUsage()) {materialFunction=&BreakOutFloat4Components}
  n1699 = LinearInterpolate(
   A=ScalarParameter() {defaultValue=1, group=Effect_Decal_Reactive_UV_Warp, parameterName=EDRec_Warp_Scale_Vehicle}
   B=ScalarParameter() {defaultValue=1, group=Effect_Decal_Reactive_UV_Warp, parameterName=EDRec_Warp_Scale_Vehicle_Side}
   Alpha=@n1698.G
  )
  n1700 = ScalarParameter() {defaultValue=1, group=Effect_Decal_Reactive_UV_Warp, parameterName=EDRec_Warp_Scale_Mech}
  n1701 = ScalarParameter() {defaultValue=1, group=Effect_Decal_Reactive_UV_Warp, parameterName=EDRec_Warp_Scale_Character}
  n1702 = MaterialFunctionCall(Float2=NamedRerouteUsage()) {materialFunction=&BreakOutFloat2Components}
  n1703 = LinearInterpolate(
   A=LinearInterpolate(A=LinearInterpolate(A=@n1697, B=@n1699, Alpha=@n1698.R), B=LinearInterpolate(A=@n1700, B=@n1701, Alpha=@n1698.A), Alpha=@n1698.B)
   B=ScalarParameter() {defaultValue=1, group=Effect_Decal_Reactive_UV_Warp, parameterName=EDRec_Warp_Scale_Backpack}
   Alpha=@n1702.R
  )
  n1704 = ScalarParameter() {defaultValue=1, group=Effect_Decal_Reactive_UV_Warp, parameterName=EDRec_Warp_Scale_Pickaxe}
  NamedRerouteDeclaration(None=Multiply(
    A=LinearInterpolate(A=@n1703, B=@n1704, Alpha=@n1702.G)
    B=VectorParameter() {group=Effect_Decal_Reactive_UV_Warp, parameterName=EDRec_Warp_Scale_Mast, defaultValue=<rgba 1 1 1 1>}.RGB
   )) {name=UVs_Warp_ScaleC, nodeColor=<rgba 0 0.7188 1 1>}
  NamedRerouteDeclaration(None=@n220.Normal) {name=Base_Normal, nodeColor=<rgba 0 0.4844 1 1>}
  n1705 = NamedRerouteUsage()
  n1706 = NamedRerouteUsage()
  n1707 = LinearInterpolate(
   A=LinearInterpolate(B=TextureSample(UVs=@n1642, Tex=@n1705, Level=@n1706) {mipValueMode=TMVM_MipLevel}.RGBA, Alpha=@n1649)
   B=TextureSample(UVs=@n1628, Tex=NamedRerouteUsage(), Level=NamedRerouteUsage()) {mipValueMode=TMVM_MipLevel}.RGBA
   Alpha=@n1635
  )
  n1708 = TextureSample(UVs=@n1614, Tex=NamedRerouteUsage(), Level=NamedRerouteUsage()) {mipValueMode=TMVM_MipLevel}
  n1709 = NamedRerouteUsage()
  n1710 = NamedRerouteUsage()
  NamedRerouteDeclaration(None=LinearInterpolate(
    A=LinearInterpolate(A=@n1707, B=@n1708.RGBA, Alpha=@n1621)
    B=TextureSample(UVs=@n1600, Tex=@n1709, Level=@n1710) {mipValueMode=TMVM_MipLevel}.RGBA
    Alpha=@n1607
   )) {name=Side_Msk, nodeColor=<rgba 1 0.4453 0 1>}
  n1711 = Saturate(None=Divide(A=@n1669, B=ScalarParameter() {defaultValue=0.25, group=Effect_Decal_Reactive_RE, parameterName=EDRec_DecalC_Duration, sortPriority=10}))
  n1712 = NamedRerouteUsage()
  NamedRerouteDeclaration(None=LinearInterpolate(
    A=ScalarParameter() {defaultValue=1, group=Effect_Decal_Reactive_RE, parameterName=EDRec_DecalC_Opacity_Start, sortPriority=10}
    B=ScalarParameter() {defaultValue=1, group=Effect_Decal_Reactive_RE, parameterName=EDRec_DecalC_Opacity_End, sortPriority=10}
    Alpha=Power(Base=@n1711, Exp=@n1712)
   )) {name=Reactive_Opacity_C, nodeColor=<rgba 0.217 0.5686 1 1>}
  NamedRerouteDeclaration(None=StaticBoolParameter() {group=Effect_Decal_Reactive, parameterName="EDRec---->Left\\Right", sortPriority=-10}) {name="Left\\right\\State", nodeColor=<rgba 0 0.0156 1 1>}
  n1713 = LinearInterpolate(A=Multiply(A=Multiply(A=@n1644.RGB, B=@n1647), B=@n1644.A), B=Multiply(A=@n1630.RGB, B=@n1633), Alpha=@n1635)
  n1714 = Multiply(A=@n1616.RGB, B=@n1619)
  NamedRerouteDeclaration(None=LinearInterpolate(A=LinearInterpolate(A=@n1713, B=@n1714, Alpha=@n1621), B=Multiply(A=@n1602.RGB, B=@n1605), Alpha=@n1607)) {name=Side_Bc, nodeColor=<rgba 0.8984 0 1 1>}
  n1715 = VectorParameter() {group=Effect_Decal_Reactive, parameterName=EDRec_Decal_BackGroundMsk}
  n1716 = DDX(Value=@n1113)
  n1717 = DDY(Value=@n1113)
  n1718 = TextureSample!(UVs=@n1111, Tex=NamedRerouteUsage(), "DDX(UVs)"=@n1716, "DDY(UVs)"=@n1717, Apply View MipBias) {mipValueMode=TMVM_Derivative}
  n1719 = Multiply(A=@n1117, B=@n1718.RGBA)
  n1720 = LinearInterpolate(
   A=@n1715.RGBA
   B=NamedRerouteDeclaration(None=StaticSwitch(True=@n1719, False=@n1718.RGBA, Value=@n629)) {name=Msk_D, nodeColor=<rgba 1 0.75 0 1>}
   Alpha=@n1128
  )
  n1721 = TextureSample(UVs=@n994, Tex=NamedRerouteUsage(), "DDX(UVs)"=@n996, "DDY(UVs)"=@n997) {mipValueMode=TMVM_Derivative}
  n1722 = StaticSwitch(True=Multiply(A=@n981, B=@n1721.RGBA), False=@n1721.RGBA, Value=@n627)
  n1723 = StaticSwitch(
   True=VectorParameter() {group=Effect_Decal_Reactive, parameterName=EDRec_Decal_BackGroundMsk}.RGBA
   False=LinearInterpolate(
    A=StaticSwitch(True=@n1715.RGBA, False=@n1720, Value=@n549)
    B=NamedRerouteDeclaration(None=@n1722) {name=Msk_C, nodeColor=<rgba 1 0.75 0 1>}
    Alpha=@n1012
   )
   Value=@n509
  )
  n1724 = TextureSample(UVs=@n679, Tex=NamedRerouteUsage(), "DDX(UVs)"=@n681, "DDY(UVs)"=@n682) {mipValueMode=TMVM_Derivative}
  n1725 = NamedRerouteDeclaration(None=StaticSwitch(True=Multiply(A=@n686, B=@n1724.RGBA), False=@n1724.RGBA, Value=@n625)) {name=Msk_B, nodeColor=<rgba 1 0.75 0 1>}
  n1726 = TextureSample(UVs=@n1064, Tex=NamedRerouteUsage(), "DDX(UVs)"=@n1066, "DDY(UVs)"=@n1067) {mipValueMode=TMVM_Derivative}
  n1727 = StaticSwitch(True=Multiply(A=@n1050, B=@n1726.RGBA), False=@n1726.RGBA, Value=@n402)
  NamedRerouteDeclaration(None=LinearInterpolate(
    A=LinearInterpolate(A=@n1723, B=@n1725, Alpha=@n1024)
    B=NamedRerouteDeclaration(None=@n1727) {name=Msk_A, nodeColor=<rgba 1 0.75 0 1>}
    Alpha=@n1078
   )) {name=Msk_Set_B, nodeColor=<rgba 0.9922 0 1 1>}
  n1728 = Constant()
  n1729 = NamedRerouteUsage()
  n1730 = NamedRerouteUsage()
  n1731 = NamedRerouteUsage()
  n1732 = LinearInterpolate(
   B=MaterialFunctionCall(
    Normals_Source=TextureSample(UVs=@n1111, Tex=@n1729, "DDX(UVs)"=@n1716, "DDY(UVs)"=@n1717) {mipValueMode=TMVM_Derivative}.RGB
    Normals_Source_Use=StaticBool() {value=true}
    UVs=@n1109
    UV_Rotation=StaticSwitch(True=@n1730, False=@n1731, Value=@n629)
    Pre_Skinned_Normals=NamedRerouteUsage()
    is_State=NamedRerouteUsage()
    "Left\\Right"=NamedRerouteUsage()
   ) {materialFunction=&MF_UVs_Side_Fix_Normals}.Raw_Normal
   Alpha=@n1128
  )
  n1733 = TextureSample(UVs=@n994, Tex=NamedRerouteUsage(), "DDX(UVs)"=@n996, "DDY(UVs)"=@n997) {mipValueMode=TMVM_Derivative}
  n1734 = StaticBool() {value=true}
  n1735 = StaticSwitch(True=NamedRerouteUsage(), False=NamedRerouteUsage(), Value=@n627)
  n1736 = NamedRerouteUsage()
  n1737 = NamedRerouteUsage()
  n1738 = NamedRerouteUsage()
  n1739 = StaticSwitch(
   True=@n959
   False=LinearInterpolate(
    A=StaticSwitch(True=@n1728, False=@n1732, Value=@n549)
    B=MaterialFunctionCall(
     Normals_Source=@n1733.RGB
     Normals_Source_Use=@n1734
     UVs=@n979
     UV_Rotation=@n1735
     Pre_Skinned_Normals=@n1736
     is_State=@n1737
     "Left\\Right"=@n1738
    ) {materialFunction=&MF_UVs_Side_Fix_Normals}.Raw_Normal
    Alpha=@n1012
   )
   Value=@n509
  )
  n1740 = MaterialFunctionCall(
   Normals_Source=TextureSample(UVs=@n679, Tex=NamedRerouteUsage(), "DDX(UVs)"=@n681, "DDY(UVs)"=@n682) {mipValueMode=TMVM_Derivative}.RGB
   Normals_Source_Use=StaticBool() {value=true}
   UVs=@n677
   UV_Rotation=StaticSwitch(True=NamedRerouteUsage(), False=NamedRerouteUsage(), Value=@n625)
   Pre_Skinned_Normals=NamedRerouteUsage()
   is_State=NamedRerouteUsage()
   "Left\\Right"=@n467
  ) {materialFunction=&MF_UVs_Side_Fix_Normals}
  n1741 = TextureSample(UVs=@n1064, Tex=NamedRerouteUsage(), "DDX(UVs)"=@n1066, "DDY(UVs)"=@n1067) {mipValueMode=TMVM_Derivative}
  n1742 = StaticBool() {value=true}
  n1743 = StaticSwitch(True=NamedRerouteUsage(), False=NamedRerouteUsage(), Value=@n402)
  n1744 = NamedRerouteUsage()
  n1745 = NamedRerouteUsage()
  NamedRerouteDeclaration(None=LinearInterpolate(
    A=LinearInterpolate(A=@n1739, B=@n1740.Raw_Normal, Alpha=@n1024)
    B=MaterialFunctionCall(
     Normals_Source=@n1741.RGB
     Normals_Source_Use=@n1742
     UVs=@n1048
     UV_Rotation=@n1743
     Pre_Skinned_Normals=@n1744
     is_State=@n1745
     "Left\\Right"=@n395
    ) {materialFunction=&MF_UVs_Side_Fix_Normals}.Raw_Normal
    Alpha=@n1078
   )) {name=Nrml_Set_B, nodeColor=<rgba 0.9922 0 1 1>}
  NamedRerouteDeclaration(None=@n1627.UVs) {name=UVsC, nodeColor=<rgba 1 0 0.3984 1>}
  NamedRerouteDeclaration(None=@n1641.UVs) {name=UVsD, nodeColor=<rgba 1 0 0.3984 1>}
  n1746 = Saturate(None=Divide(A=@n1664, B=ScalarParameter() {defaultValue=0.25, group=Effect_Decal_Reactive_RE, parameterName=EDRec_DecalD_Duration, sortPriority=20}))
  n1747 = NamedRerouteUsage()
  NamedRerouteDeclaration(None=LinearInterpolate(
    A=ScalarParameter() {defaultValue=1, group=Effect_Decal_Reactive_RE, parameterName=EDRec_DecalD_Opacity_Start, sortPriority=20}
    B=ScalarParameter() {defaultValue=1, group=Effect_Decal_Reactive_RE, parameterName=EDRec_DecalD_Opacity_End, sortPriority=20}
    Alpha=Power(Base=@n1746, Exp=@n1747)
   )) {name=Reactive_Opacity_D, nodeColor=<rgba 0.217 0.5686 1 1>}
  NamedRerouteDeclaration(None=StaticBoolParameter() {group=Customization, parameterName="Custmization  V2", sortPriority=-500}) {name=Cust_State, nodeColor=<rgba 0.125 0 1 1>}
  n1748 = ScalarParameter() {group=Effect_Decal_Reactive_RE_Offset, parameterName=EDRec_DecalA_Offset_Delay--------, sortPriority=5}
  n1749 = ScalarParameter() {defaultValue=-9999999, group=Effect_Decal_Reactive_RE_Offset, parameterName=EDRec_TimeFrom_BP_Offset_DecalA, sortPriority=5}
  n1750 = StaticBoolParameter() {group=Effect_Decal_Reactive_RE_Offset, parameterName=EDRec_USe Separate _TimeFrom_BP_offset, sortPriority=0}
  n1751 = StaticSwitch(True=@n1749, False=NamedRerouteUsage(), Value=@n1750)
  n1752 = Divide(
   A=Subtract(A=Time(), B=Add(A=@n1748, B=@n1751))
   B=ScalarParameter() {defaultValue=0.25, group=Effect_Decal_Reactive_RE_Offset, parameterName=EDRec_DecalA_Offset_Duration, sortPriority=5}
  )
  NamedRerouteDeclaration(None=Power(
    Base=Saturate(None=@n1752)
    Exp=ScalarParameter() {defaultValue=1, group=Effect_Decal_Reactive_RE_Offset, parameterName=EDRec_DecalA_Offset_Fallof, sortPriority=5}
   )) {name=Reactive_OffsetA, nodeColor=<rgba 0 1 0.2266 1>}
  n1753 = ScalarParameter() {group=Effect_Decal_Reactive_RE_Offset, parameterName=EDRec_DecalB_Offset_Delay--------, sortPriority=10}
  n1754 = ScalarParameter() {defaultValue=-9999999, group=Effect_Decal_Reactive_RE_Offset, parameterName=EDRec_TimeFrom_BP_Offset_DecalB, sortPriority=10}
  n1755 = StaticSwitch(True=@n1754, False=NamedRerouteUsage(), Value=@n1750)
  n1756 = Divide(
   A=Subtract(A=Time(), B=Add(A=@n1753, B=@n1755))
   B=ScalarParameter() {defaultValue=0.25, group=Effect_Decal_Reactive_RE_Offset, parameterName=EDRec_DecalB_Offset_Duration, sortPriority=10}
  )
  NamedRerouteDeclaration(None=Power(
    Base=Saturate(None=@n1756)
    Exp=ScalarParameter() {defaultValue=1, group=Effect_Decal_Reactive_RE_Offset, parameterName=EDRec_DecalB_Offset_Fallof, sortPriority=10}
   )) {name=Reactive_OffsetB, nodeColor=<rgba 0 1 0.2266 1>}
  n1757 = ScalarParameter() {group=Effect_Decal_Reactive_RE_Offset, parameterName=EDRec_DecaC_Offset_Delay--------, sortPriority=20}
  n1758 = ScalarParameter() {defaultValue=-9999999, group=Effect_Decal_Reactive_RE_Offset, parameterName=EDRec_TimeFrom_BP_Offset_DecalC, sortPriority=20}
  n1759 = StaticSwitch(True=@n1758, False=NamedRerouteUsage(), Value=@n1750)
  n1760 = Divide(
   A=Subtract(A=Time(), B=Add(A=@n1757, B=@n1759))
   B=ScalarParameter() {defaultValue=0.25, group=Effect_Decal_Reactive_RE_Offset, parameterName=EDRec_DecalC_Offset_Duration, sortPriority=20}
  )
  NamedRerouteDeclaration(None=Power(
    Base=Saturate(None=@n1760)
    Exp=ScalarParameter() {defaultValue=1, group=Effect_Decal_Reactive_RE_Offset, parameterName=EDRec_DecalC_Offset_Fallof, sortPriority=20}
   )) {name=Reactive_OffsetC, nodeColor=<rgba 0 1 0.2266 1>}
  n1761 = ScalarParameter() {group=Effect_Decal_Reactive_RE_Offset, parameterName=EDRec_DecalD_Offset_Delay--------, sortPriority=25}
  n1762 = ScalarParameter() {defaultValue=-9999999, group=Effect_Decal_Reactive_RE_Offset, parameterName=EDRec_TimeFrom_BP_Offset_DecalD, sortPriority=25}
  n1763 = StaticSwitch(True=@n1762, False=NamedRerouteUsage(), Value=@n1750)
  n1764 = Divide(
   A=Subtract(A=Time(), B=Add(A=@n1761, B=@n1763))
   B=ScalarParameter() {defaultValue=0.25, group=Effect_Decal_Reactive_RE_Offset, parameterName=EDRec_DecalD_Offset_Duration, sortPriority=25}
  )
  NamedRerouteDeclaration(None=Power(
    Base=Saturate(None=@n1764)
    Exp=ScalarParameter() {defaultValue=1, group=Effect_Decal_Reactive_RE_Offset, parameterName=EDRec_DecalD_Offset_Fallof, sortPriority=25}
   )) {name=Reactive_OffsetD, nodeColor=<rgba 0 1 0.2266 1>}
  NamedRerouteDeclaration(None=@n428) {name=UVs_OffsetA_1, nodeColor=<rgba 0 0.625 1 1>}
  NamedRerouteDeclaration(None=@n880) {name=UVs_OffsetB_1, nodeColor=<rgba 0 0.625 1 1>}
  NamedRerouteDeclaration(None=@n1170) {name=UVs_OffsetC_1, nodeColor=<rgba 0 0.625 1 1>}
  NamedRerouteDeclaration(None=@n1423) {name=UVs_OffsetD_1, nodeColor=<rgba 0 0.625 1 1>}
  NamedRerouteDeclaration(None=@n1653) {name=TimeFormBP A, nodeColor=<rgba 0 1 0.625 1>}
  n1765 = VectorParameter() {group=Effect_Decal_Reactive_UV_Offset, parameterName=EDRec_DecalA_Location_Weapon, defaultValue=<rgba 0 0 0 0.5>}
  n1766 = AppendVector(A=@n1765.RGB, B=@n1765.RGB)
  n1767 = VectorParameter() {group=Effect_Decal_Reactive_UV_Offset, parameterName=EDRec_DecalA_Location_Vehicle, defaultValue=<rgba 0 0 0 5>}
  n1768 = VectorParameter() {group=Effect_Decal_Reactive_UV_Offset, parameterName=EDRec_DecalA_Location_VehicleSide, defaultValue=<rgba 0 0 0 5>}
  n1769 = MaterialFunctionCall(Float4=NamedRerouteUsage()) {materialFunction=&BreakOutFloat4Components}
  n1770 = LinearInterpolate(A=AppendVector(A=@n1767.RGB, B=@n1767.RGB), B=AppendVector(A=@n1768.RGB, B=@n1768.RGB), Alpha=@n1769.G)
  n1771 = VectorParameter() {group=Effect_Decal_Reactive_UV_Offset, parameterName="EDRec_DecalA_Location_Mech   A=Scale", defaultValue=<rgba 0 0 0 0.5>}
  n1772 = AppendVector(A=@n1771.RGB, B=@n1771.RGB)
  n1773 = VectorParameter() {group=Effect_Decal_Reactive_UV_Offset, parameterName="EDRec_DecalA_Location_Character   A=Scale", defaultValue=<rgba 0 0 0 0.5>}
  n1774 = AppendVector(A=@n1773.RGB, B=@n1773.RGB)
  NamedRerouteDeclaration(None=LinearInterpolate(A=LinearInterpolate(A=@n1766, B=@n1770, Alpha=@n1769.R), B=LinearInterpolate(A=@n1772, B=@n1774, Alpha=@n1769.A), Alpha=@n1769.B)) {name=Offset_A, nodeColor=<rgba 1 0.7266 0 1>}
  NamedRerouteDeclaration(None=@n1673) {name=TimeFormBP B, nodeColor=<rgba 0 1 0.625 1>}
  NamedRerouteDeclaration(None=@n1663) {name=TimeFormBP D, nodeColor=<rgba 0 1 0.625 1>}
  NamedRerouteDeclaration(None=@n1668) {name=TimeFormBP C, nodeColor=<rgba 0 1 0.625 1>}
  NamedRerouteDeclaration(None=StaticBoolParameter() {defaultValue=true, group=Effect_Decal_Reactive_RE, parameterName=EDRec_DecalD_Opacity_Reactive, sortPriority=-10}) {name=Reactive opacity State, nodeColor=<rgba 0.5703 1 0 1>}
  n1775 = ScalarParameter() {group=Effect_Decal_Reactive_UV_Warp, parameterName=EDRec_Warp_Strength_Vehicle}
  n1776 = ScalarParameter() {group=Effect_Decal_Reactive_UV_Warp, parameterName=EDRec_Warp_Strength_Vehicle_Side}
  n1777 = MaterialFunctionCall(Float4=NamedRerouteUsage()) {materialFunction=&BreakOutFloat4Components}
  n1778 = LinearInterpolate(
   A=LinearInterpolate(
    A=ScalarParameter() {group=Effect_Decal_Reactive_UV_Warp, parameterName=EDRec_Warp_Strength_Weapon}
    B=LinearInterpolate(A=@n1775, B=@n1776, Alpha=@n1777.G)
    Alpha=@n1777.R
   )
   B=LinearInterpolate(
    A=ScalarParameter() {group=Effect_Decal_Reactive_UV_Warp, parameterName=EDRec_Warp_Strength_Mech}
    B=ScalarParameter() {group=Effect_Decal_Reactive_UV_Warp, parameterName=EDRec_Warp_Strength_Character}
    Alpha=@n1777.A
   )
   Alpha=@n1777.B
  )
  n1779 = ScalarParameter() {group=Effect_Decal_Reactive_UV_Warp, parameterName=EDRec_Warp_Strength_Backpack}
  n1780 = MaterialFunctionCall(Float2=NamedRerouteUsage()) {materialFunction=&BreakOutFloat2Components}
  NamedRerouteDeclaration(None=LinearInterpolate(
    A=LinearInterpolate(A=@n1778, B=@n1779, Alpha=@n1780.R)
    B=ScalarParameter() {group=Effect_Decal_Reactive_UV_Warp, parameterName=EDRec_Warp_Strength_Pickaxe}
    Alpha=@n1780.G
   )) {name=UVs_Warp_Strength, nodeColor=<rgba 0 0.7188 1 1>}
  n1781 = NamedRerouteUsage()
  n1782 = NamedRerouteUsage()
  NamedRerouteDeclaration(None=StaticSwitchParameter(
    True=StaticSwitchParameter(True=@n1781, False=@n1782) {group=Effect_Decal_Reactive_UV_Warp, parameterName="EDRec_Warp Decal_Msk -------->TopS_00B_Sharerd_Tex_Msk"}
    False=@n1311
   ) {group=Effect_Decal_Reactive_UV_Warp, parameterName="EDRec_Warp Decal_Msk -------->TopS_00A_Sharerd_Tex_Msk"}) {name=Tex_Warp, nodeColor=<rgba 0 1 0.1328 1>}
  n1783 = CollectionParameter() {collection=&FortniteMaterialParameters, parameterName=EventAudio_Spectrum}
  n1784 = ChannelMaskParameter(None=@n1783) {group=Effect_Decal_Reactive_RE_Audio, maskChannel=Alpha, parameterName=EDRec_Audio B Channel, defaultValue=<rgba 0 0 0 1>}
  n1785 = CollectionParameter() {collection=&FortniteMaterialParameters, parameterName=Emote_AudioLow}
  n1786 = Max(A=@n1784.R, B=@n1785)
  n1787 = ScalarParameter() {defaultValue=0.25, group=Effect_Decal_Reactive_RE_Audio, parameterName=EDRec_Audio B_Low_Remap}
  n1788 = ScalarParameter() {defaultValue=1, group=Effect_Decal_Reactive_RE_Audio, parameterName=EDRec_Audio B_High_Remap}
  n1789 = Constant()
  n1790 = Constant() {r=1}
  n1791 = CollectionParameter() {collection=&FortniteMaterialParameters, parameterName=Emote_AudioHigh}
  n1792 = Max(
   A=Saturate(None=MaterialFunctionCall(Input=@n1786, Input Low=@n1787, Input High=@n1788, Target Low=@n1789, Target High=@n1790) {materialFunction=&RemapValueRange}.Result)
   B=@n1791
  )
  NamedRerouteDeclaration(None=@n1792) {name=AudioSpectrum B, nodeColor=<rgba 1 0.8203 0 1>}
  n1793 = ScalarParameter() {defaultValue=1, group=Effect_Decal_Reactive_RE_Audio, parameterName=EDRec_Audio B Strength}
  n1794 = Time()
  n1795 = ScalarParameter() {defaultValue=2, group=Effect_Decal_Reactive_RE_Audio, parameterName=EDRec_Audio B Static_Speed}
  n1796 = ConstantBiasScale(None=Sine(None=Multiply(A=@n1794, B=@n1795)))
  n1797 = ScalarParameter() {group=Effect_Decal_Reactive_RE_Audio, parameterName=EDRec_Audio B Static_Strength}
  NamedRerouteDeclaration(None=Max(A=Multiply(A=@n1792, B=@n1793), B=Multiply(A=@n1796, B=@n1797))) {name=AudioSpectrum B_Static_Speed, nodeColor=<rgba 1 0.8203 0 1>}
  n1798 = CollectionParameter() {collection=&FortniteMaterialParameters, parameterName=EventAudio_Spectrum}
  n1799 = ChannelMaskParameter(None=@n1798) {group=Effect_Decal_Reactive_RE_Audio, maskChannel=Alpha, parameterName=EDRec_Audio C Channel, defaultValue=<rgba 0 0 0 1>}
  n1800 = CollectionParameter() {collection=&FortniteMaterialParameters, parameterName=Emote_AudioLow}
  n1801 = Max(A=@n1799.R, B=@n1800)
  n1802 = ScalarParameter() {defaultValue=0.25, group=Effect_Decal_Reactive_RE_Audio, parameterName=EDRec_Audio C_Low_Remap}
  n1803 = ScalarParameter() {defaultValue=1, group=Effect_Decal_Reactive_RE_Audio, parameterName=EDRec_Audio C_High_Remap}
  n1804 = Constant()
  n1805 = Constant() {r=1}
  n1806 = CollectionParameter() {collection=&FortniteMaterialParameters, parameterName=Emote_AudioHigh}
  n1807 = Max(
   A=Saturate(None=MaterialFunctionCall(Input=@n1801, Input Low=@n1802, Input High=@n1803, Target Low=@n1804, Target High=@n1805) {materialFunction=&RemapValueRange}.Result)
   B=@n1806
  )
  NamedRerouteDeclaration(None=@n1807) {name=AudioSpectrum C, nodeColor=<rgba 1 0.8203 0 1>}
  n1808 = VectorParameter() {group=Effect_Decal_Reactive_UV_Offset, parameterName=EDRec_DecalB_Location_Weapon, defaultValue=<rgba 0 0 0 0.5>}
  n1809 = AppendVector(A=@n1808.RGB, B=@n1808.RGB)
  n1810 = VectorParameter() {group=Effect_Decal_Reactive_UV_Offset, parameterName=EDRec_DecalB_Location_Vehicle, defaultValue=<rgba 0 0 0 5>}
  n1811 = VectorParameter() {group=Effect_Decal_Reactive_UV_Offset, parameterName=EDRec_DecalB_Location_VehicleSide, defaultValue=<rgba 0 0 0 5>}
  n1812 = MaterialFunctionCall(Float4=NamedRerouteUsage()) {materialFunction=&BreakOutFloat4Components}
  n1813 = LinearInterpolate(A=AppendVector(A=@n1810.RGB, B=@n1810.RGB), B=AppendVector(A=@n1811.RGB, B=@n1811.RGB), Alpha=@n1812.G)
  n1814 = VectorParameter() {group=Effect_Decal_Reactive_UV_Offset, parameterName="EDRec_DecalB_Location_Mech   A=Scale", defaultValue=<rgba 0 0 0 0.5>}
  n1815 = AppendVector(A=@n1814.RGB, B=@n1814.RGB)
  n1816 = VectorParameter() {group=Effect_Decal_Reactive_UV_Offset, parameterName="EDRec_DecalB_Location_Character A=Scale", defaultValue=<rgba 0 0 0 0.5>}
  n1817 = AppendVector(A=@n1816.RGB, B=@n1816.RGB)
  NamedRerouteDeclaration(None=LinearInterpolate(A=LinearInterpolate(A=@n1809, B=@n1813, Alpha=@n1812.R), B=LinearInterpolate(A=@n1815, B=@n1817, Alpha=@n1812.A), Alpha=@n1812.B)) {name=Offset_B, nodeColor=<rgba 1 0.7266 0 1>}
  n1818 = ScalarParameter() {defaultValue=1, group=Effect_Decal_Reactive_RE_Audio, parameterName=EDRec_Audio C Strength}
  n1819 = Time()
  n1820 = ScalarParameter() {defaultValue=2, group=Effect_Decal_Reactive_RE_Audio, parameterName=EDRec_Audio C Static_Speed}
  n1821 = ConstantBiasScale(None=Sine(None=Multiply(A=@n1819, B=@n1820)))
  n1822 = ScalarParameter() {group=Effect_Decal_Reactive_RE_Audio, parameterName=EDRec_Audio C Static_Strength}
  NamedRerouteDeclaration(None=Max(A=Multiply(A=@n1807, B=@n1818), B=Multiply(A=@n1821, B=@n1822))) {name=AudioSpectrum C_Static_Speed, nodeColor=<rgba 1 0.8203 0 1>}
  n1823 = CollectionParameter() {collection=&FortniteMaterialParameters, parameterName=EventAudio_Spectrum}
  n1824 = ChannelMaskParameter(None=@n1823) {group=Effect_Decal_Reactive_RE_Audio, maskChannel=Alpha, parameterName=EDRec_Audio D Channel, defaultValue=<rgba 0 0 0 1>}
  n1825 = CollectionParameter() {collection=&FortniteMaterialParameters, parameterName=Emote_AudioLow}
  n1826 = Max(A=@n1824.R, B=@n1825)
  n1827 = ScalarParameter() {defaultValue=0.25, group=Effect_Decal_Reactive_RE_Audio, parameterName=EDRec_Audio D_Low_Remap}
  n1828 = ScalarParameter() {defaultValue=1, group=Effect_Decal_Reactive_RE_Audio, parameterName=EDRec_Audio D_High_Remap}
  n1829 = Constant()
  n1830 = Constant() {r=1}
  n1831 = CollectionParameter() {collection=&FortniteMaterialParameters, parameterName=Emote_AudioHigh}
  n1832 = Max(
   A=Saturate(None=MaterialFunctionCall(Input=@n1826, Input Low=@n1827, Input High=@n1828, Target Low=@n1829, Target High=@n1830) {materialFunction=&RemapValueRange}.Result)
   B=@n1831
  )
  NamedRerouteDeclaration(None=@n1832) {name=AudioSpectrum D, nodeColor=<rgba 1 0.8203 0 1>}
  n1833 = ScalarParameter() {defaultValue=1, group=Effect_Decal_Reactive_RE_Audio, parameterName=EDRec_Audio D Strength}
  n1834 = Time()
  n1835 = ScalarParameter() {defaultValue=2, group=Effect_Decal_Reactive_RE_Audio, parameterName=EDRec_Audio D Static_Speed}
  n1836 = ConstantBiasScale(None=Sine(None=Multiply(A=@n1834, B=@n1835)))
  n1837 = ScalarParameter() {group=Effect_Decal_Reactive_RE_Audio, parameterName=EDRec_Audio D Static_Strength}
  NamedRerouteDeclaration(None=Max(A=Multiply(A=@n1832, B=@n1833), B=Multiply(A=@n1836, B=@n1837))) {name=AudioSpectrum D_Static_Speed, nodeColor=<rgba 1 0.8203 0 1>}
  NamedRerouteDeclaration(None=StaticBoolParameter() {group=Effect_Decal_Reactive_RE_Audio, parameterName="EDRec  Audio Reactive Opacity"}) {name=Audio_Opacity, nodeColor=<rgba 0 0.4141 1 1>}
  n1838 = ChannelMaskParameter(None=@n903.RGBA) {group=Effect_Decal_Reactive_UV_Warp, parameterName=EDRec_WarpB_Msk_Channel}
  n1839 = ScalarParameter() {group=Effect_Decal_Reactive_UV_Warp, parameterName=EDRec_WarpB_Offset}
  n1840 = ScalarParameter() {group=Effect_Decal_Reactive_UV_Warp, parameterName=EDRec_WarpB_Strength}
  NamedRerouteDeclaration(None=Multiply(A=Add(A=@n1838.R, B=@n1839), B=@n1840)) {name=Warp_Uvs, nodeColor=<rgba 0.0547 0 1 1>}
  NamedRerouteDeclaration(None=StaticBoolParameter() {group=Effect_Decal_Reactive_UV_Warp, parameterName=EDRec_Warp_Simple_Uvs, sortPriority=1}) {name=Warp_Uv_State, nodeColor=<rgba 0.2656 1 0 1>}
  n1841 = VectorParameter() {group=Effect_Decal_Reactive_UV_Offset, parameterName=EDRec_DecalC_Location_Weapon, defaultValue=<rgba 0 0 0 0.5>}
  n1842 = AppendVector(A=@n1841.RGB, B=@n1841.RGB)
  n1843 = VectorParameter() {group=Effect_Decal_Reactive_UV_Offset, parameterName=EDRec_DecalC_Location_Vehicle, defaultValue=<rgba 0 0 0 5>}
  n1844 = VectorParameter() {group=Effect_Decal_Reactive_UV_Offset, parameterName=EDRec_DecalC_Location_VehicleSide, defaultValue=<rgba 0 0 0 5>}
  n1845 = MaterialFunctionCall(Float4=NamedRerouteUsage()) {materialFunction=&BreakOutFloat4Components}
  n1846 = LinearInterpolate(A=AppendVector(A=@n1843.RGB, B=@n1843.RGB), B=AppendVector(A=@n1844.RGB, B=@n1844.RGB), Alpha=@n1845.G)
  n1847 = VectorParameter() {group=Effect_Decal_Reactive_UV_Offset, parameterName="EDRec_DecalC_Location_Mech   A=Scale", defaultValue=<rgba 0 0 0 0.5>}
  n1848 = AppendVector(A=@n1847.RGB, B=@n1847.RGB)
  n1849 = VectorParameter() {group=Effect_Decal_Reactive_UV_Offset, parameterName="EDRec_DecalC_Location_Character A=Scale", defaultValue=<rgba 0 0 0 0.5>}
  n1850 = AppendVector(A=@n1849.RGB, B=@n1849.RGB)
  NamedRerouteDeclaration(None=LinearInterpolate(A=LinearInterpolate(A=@n1842, B=@n1846, Alpha=@n1845.R), B=LinearInterpolate(A=@n1848, B=@n1850, Alpha=@n1845.A), Alpha=@n1845.B)) {name=Offset_C, nodeColor=<rgba 1 0.7266 0 1>}
  NamedRerouteDeclaration(None=ScalarParameter() {defaultValue=2, group=Effect_Decal_Reactive_FlipBook, parameterName=EDRec_FlipBook_Number_Rows, sortPriority=0}) {name="Rows_####", nodeColor=<rgba 0 0.5312 1 1>}
  NamedRerouteDeclaration(None=ScalarParameter() {defaultValue=2, group=Effect_Decal_Reactive_FlipBook, parameterName=EDRec_FlipBook_Number_Columns, sortPriority=0}) {name="Colums_####", nodeColor=<rgba 0 0.5312 1 1>}
  n1851 = Constant4Vector() {constant=<rgba 0 0 0 0.5>}
  NamedRerouteDeclaration(None=FunctionInput(Preview=@n1851.RGBA) {bUsePreviewValueAsDefault=true, inputName=Object-SafeZone-Back, inputType=FunctionInput_Vector4, sortPriority=100}) {name=SafeZone_Back, nodeColor=<rgba 0 0.9297 1 1>}
  NamedRerouteDeclaration(None=FunctionInput(Preview=@n1851.RGBA) {bUsePreviewValueAsDefault=true, inputName=Object-SafeZone-Middle, inputType=FunctionInput_Vector4, sortPriority=101}) {name=SafeZone_Middle, nodeColor=<rgba 0 0.9297 1 1>}
  NamedRerouteDeclaration(None=FunctionInput(Preview=@n1851.RGBA) {bUsePreviewValueAsDefault=true, inputName=Object-SafeZone-Front, inputType=FunctionInput_Vector4, sortPriority=102}) {name=SafeZone_Front, nodeColor=<rgba 0 0.9297 1 1>}
  n1852 = Saturate(None=Add(
    A=@n95
    B=ScalarParameter() {defaultValue=1, group=Effect_Decal_Reactive_Emissive_A, parameterName="EDRec_Emissive_A_CC_Mask   0=ON", sliderMax=1}
   ))
  n1853 = VectorParameter() {group=Effect_Decal_Reactive_Emissive_A, parameterName=EDRec_Emissive_A_CC Tint, defaultValue=<rgba 1 1 1 1>}
  n1854 = Multiply(
   A=Multiply(A=@n1853.RGB, B=@n1529.3V)
   B=ScalarParameter() {defaultValue=1, group=Effect_Decal_Reactive_Emissive_A, parameterName=EDRec_Emissive_A_CC_Brightness}
  )
  n1855 = VectorParameter() {group=Effect_Decal_Reactive_Emissive_A, parameterName=EDRec_Emissive_A_Color-Start, defaultValue=<rgba 1 1 1 1>}
  n1856 = VectorParameter() {group=Effect_Decal_Reactive_Emissive_A, parameterName=EDRec_Emissive_A_Color-, defaultValue=<rgba 1 1 1 1>}
  n1857 = StaticSwitchParameter(True=LinearInterpolate(A=@n1855.RGB, B=@n1856.RGB, Alpha=NamedRerouteUsage()), False=@n1856.RGB) {group=Effect_Decal_Reactive_Emissive_A_Reactive, parameterName=EDRec-Emissive-A_Reactive_Color---------------------}
  n1858 = Multiply(
   A=StaticSwitchParameter(True=Multiply(A=@n1852, B=@n1854), False=Multiply(A=@n95, B=@n1857)) {group=Effect_Decal_Reactive_Emissive_A, parameterName="EDRec-Emissive-A-Color--->CCurve"}
   B=NamedRerouteUsage()
  )
  n1859 = Multiply(A=@n1858, B=@n1520)
  n1860 = MaterialFunctionCall(
   In=Max(
    A=NamedRerouteUsage()
    B=ScalarParameter() {defaultValue=1, group=Effect_Decal_Reactive_Emissive_A, parameterName="EDRec-Emissive-A_TriMak   0=Active", sortPriority=0}
   )
   Contrast=ScalarParameter() {group=Effect_Decal_Reactive_Emissive_A, parameterName="EDRec-Emissive-A_TriMak  Contrast", sortPriority=0}
  ) {materialFunction=&CheapContrast}
  n1861 = Multiply(
   A=Multiply(
    A=StaticSwitchParameter(True=@n1859, False=@n1858) {group=Effect_Decal_Reactive_Emissive_A_Reactive, parameterName=EDRec-Emissive-A_Reactive_Brightness---------------------}
    B=NamedRerouteUsage()
   )
   B=@n1860.Result
  )
  n1862 = NamedRerouteUsage()
  n1863 = NamedRerouteUsage()
  n1864 = StaticBoolParameter() {group=Effect_Decal_Reactive_Emissive_A_Reactive, parameterName=EDRec-Emissive-A-TOD, sortPriority=0}
  n1865 = StaticSwitch(
   True=LinearInterpolate(
    B=@n1861
    Alpha=StaticSwitchParameter(True=@n1862, False=@n1863) {group=Effect_Decal_Reactive_RE, parameterName="EDRec_DecalA_Opacity-TOD----->TODB", sortPriority=-1}
   ) {constB=0}
   False=@n1861
   Value=@n1864
  )
  n1866 = NamedRerouteUsage()
  n1867 = LinearInterpolate(
   A=@n1866
   B=OneMinus(None=@n1866)
   Alpha=ScalarParameter() {group=Effect_Decal_Reactive_Emissive_A_Reactive, parameterName="EDRec-Emissive-A_TopS Blend 01<--->02 MSk   1=invert"}
  )
  n1868 = LinearInterpolate(B=@n1865, Alpha=@n1867)
  n1869 = StaticBoolParameter() {group=Effect_Decal_Reactive_Emissive_A_Reactive, parameterName="EDRec-Emissive-A_TopS Blend 01<--->02 MSk"}
  n1870 = StaticBoolParameter() {group="TopSide_V02_Layer 01 Blend  02", parameterName="TopS Blend 01<--->02"}
  NamedRerouteDeclaration(None=StaticSwitch(True=StaticSwitch(True=@n1868, False=@n1865, Value=@n1869), False=@n1865, Value=@n1870)) {name=Emissive_A, nodeColor=<rgba 0.2656 1 0 1>}
  n1871 = Constant() {r=1}
  n1872 = StaticSwitchParameter(
   True=@n1871
   False=StaticSwitchParameter(True=@n1871, False=@n358) {group=Effect_Decal_Reactive_Emissive_A, parameterName=EDRec-Emissive-A_Height___X+X, sortPriority=4}
  ) {group=Effect_Decal_Reactive_Emissive_A, parameterName=EDRec-Emissive-A_Height___Lerp, sortPriority=4}
  NamedRerouteDeclaration(None=Multiply(
    A=NamedRerouteUsage()
    B=StaticSwitchParameter(True=@n1871, False=@n1872) {group=Effect_Decal_Reactive_Emissive_A, parameterName=EDRec-Emissive-A_Only_Mask, sortPriority=4}
   )) {name=Emissive_Msk_Simple A, nodeColor=<rgba 0 1 0.1094 1>}
  n1873 = Constant() {r=0.5}
  n1874 = VectorParameter() {group=Effect_Decal_Reactive_Emissive_A, parameterName=EDRec_Emissive_A_Msk_Offset, sortPriority=5}
  n1875 = MaterialFunctionCall(Float2=Add(A=ComponentMask(None=@n1874.RGB) {g=true, r=true}, B=@n67.Rotated Values) {constB=0}) {materialFunction=&BreakOutFloat2Components}
  n1876 = VectorParameter() {
   group=Effect_Decal_Reactive_Emissive_A
   parameterName=EDRec_Emissive_A_Decal_Msking_Scale
   sortPriority=5
   channelNames {a=Y Falloff, b=X Falloff, g=Y Size, r=X Size}
   defaultValue=<rgba 0.5 0.5 1 1>
  }
  n1877 = SphereMask(A=@n1873, B=@n1875.R, Radius=@n1876.Y Size, Hardness=@n1876.Y Size) {attenuationRadius=0.5}
  n1878 = SphereMask(A=@n1875.G, B=@n1873, Radius=@n1876.X Size, Hardness=@n1876.X Size) {attenuationRadius=0.5}
  n1879 = StaticSwitchParameter(True=Power(Base=Multiply(A=@n1877, B=@n1878)), False=Constant() {r=1}) {
   desc=second layer of bounds that is used to msk
   group=Effect_Decal_Reactive_Emissive_A
   parameterName=EDRec_Emissive_A_Decal_Msking
   sortPriority=5
  }
  NamedRerouteDeclaration(None=@n1879) {name=Decal_Msking A, nodeColor=<rgba 0 1 0.5078 1>}
  n1880 = LinearInterpolate(
   A=ScalarParameter() {group=Effect_Decal_Reactive_Emissive_A_Reactive, parameterName=EDRec-Emissive-A_Reactive_Audio_Darken__Min}
   B=ScalarParameter() {group=Effect_Decal_Reactive_Emissive_A_Reactive, parameterName=EDRec-Emissive-A_Reactive_Audio_Darken__Max}
   Alpha=NamedRerouteUsage()
  )
  n1881 = LinearInterpolate(
   A=ScalarParameter() {group=Effect_Decal_Reactive_Emissive_A_Reactive, parameterName=EDRec_Emissive_A_Brightness_DarkenBase_Time_Start}
   B=ScalarParameter() {group=Effect_Decal_Reactive_Emissive_A_Reactive, parameterName=EDRec_Emissive_A_Brightness_DarkenBase_Time_End}
   Alpha=NamedRerouteUsage()
  )
  n1882 = StaticSwitchParameter(
   True=StaticSwitch(True=Max(A=@n1880, B=@n1881), False=@n1881, Value=NamedRerouteUsage())
   False=ScalarParameter() {group=Effect_Decal_Reactive_Emissive_A, parameterName=EDRec_Emissive_A_DarkenBase-------, sortPriority=4}
  ) {group=Effect_Decal_Reactive_Emissive_A_Reactive, parameterName=EDRec-Emissive-A_Reactive_Brightness---------------------}
  n1883 = NamedRerouteUsage()
  n1884 = Multiply(A=Multiply(A=@n1882, B=@n95), B=@n1883)
  n1885 = Multiply(A=@n1882, B=Saturate(None=@n48))
  n1886 = StaticSwitchParameter(True=@n82, False=Constant() {r=1}) {group=Effect_Decal_Reactive_Emissive_A, parameterName=EDRec-Emissive-A_Height___X+X, sortPriority=4}
  n1887 = Multiply(
   A=StaticSwitchParameter(
    True=@n82
    False=StaticSwitchParameter(True=@n82, False=@n1886) {group=Effect_Decal_Reactive_Emissive_A, parameterName=EDRec-Emissive-A_Height___Lerp, sortPriority=4}
   ) {group=Effect_Decal_Reactive_Emissive_A, parameterName=EDRec-Emissive-A_Only_Mask, sortPriority=4}
   B=@n1885
  )
  n1888 = ScalarParameter() {group=Effect_Decal_Reactive_Emissive_A, parameterName="EDRec_Emissive_A_DarkenBase-CCAlpha 1=Use", sliderMax=1, sortPriority=4}
  n1889 = StaticSwitchParameter(True=LinearInterpolate(A=@n1884, B=LinearInterpolate(A=@n1885, B=@n1887, Alpha=@n1530), Alpha=Ceil(None=@n1888)), False=@n1884) {group=Effect_Decal_Reactive_Emissive_A, parameterName="EDRec-Emissive-A-Color--->CCurve"}
  n1890 = Saturate(None=Multiply(A=@n1889, B=NamedRerouteUsage()))
  n1891 = LinearInterpolate(B=Max(A=OneMinus(None=@n1860.Result), B=OneMinus(None=@n1890)), Alpha=@n364) {constA=1}
  n1892 = NamedRerouteUsage()
  n1893 = NamedRerouteUsage()
  n1894 = StaticSwitch(
   True=LinearInterpolate(
    B=@n1891
    Alpha=StaticSwitchParameter(True=@n1892, False=@n1893) {group=Effect_Decal_Reactive_RE, parameterName="EDRec_DecalA_Opacity-TOD----->TODB", sortPriority=-1}
   ) {constA=1, constB=0}
   False=@n1891
   Value=@n1864
  )
  n1895 = LinearInterpolate(B=@n1894, Alpha=@n1867) {constA=1}
  NamedRerouteDeclaration(None=StaticSwitch(True=StaticSwitch(True=@n1895, False=@n1894, Value=@n1869), False=@n1894, Value=@n1870)) {name=Emissive_Darken_Base A, nodeColor=<rgba 0 0.1328 1 1>}
  n1896 = ScalarParameter() {
   defaultValue=-9999999
   group=Effect_Decal_Reactive_Emissive_A_Reactive |Gradent
   parameterName=EDRec_Emissive_A_Gradent_Time_BP
   sortPriority=212
  }
  n1897 = ScalarParameter() {group=Effect_Decal_Reactive_Emissive_A_Reactive |Gradent, parameterName=EDRec_Emissive_A_Gradent_Delay, sortPriority=211}
  n1898 = ScalarParameter() {
   defaultValue=0.25
   group=Effect_Decal_Reactive_Emissive_A_Reactive |Gradent
   parameterName=EDRec_Emissive_A_Gradent_Duration
   sortPriority=211
  }
  n1899 = Divide(A=Subtract(A=Time(), B=Add(A=@n1896, B=@n1897)), B=@n1898)
  n1900 = ScalarParameter() {defaultValue=1, group=Effect_Decal_Reactive_Emissive_A_Reactive |Gradent, parameterName=EDRec_Emissive_A_Gradent_Falloff_, sortPriority=211}
  NamedRerouteDeclaration(None=Power(Base=Saturate(None=@n1899), Exp=@n1900)) {name=reactive_Gradent A, nodeColor=<rgba 0 0.5078 1 1>}
  NamedRerouteUsage()
  NamedRerouteUsage()
  NamedRerouteUsage()
  NamedRerouteUsage()
  NamedRerouteUsage()
  NamedRerouteUsage()
  NamedRerouteUsage()
  NamedRerouteUsage()
  NamedRerouteUsage()
  NamedRerouteUsage()
  NamedRerouteUsage()
  NamedRerouteUsage()
  Power(
   Base=Saturate(None=@n657)
   Exp=ScalarParameter() {defaultValue=1, desc="adjust frame speed ", group=Effect_Decal_Reactive_FlipBook, parameterName=EDRec_FlipBook_FrameB_Pause}
  )
  Reroute(None=@n904)
  Reroute(None=@n545.UVs No Offsets/Tiling)
  Reroute(None=@n33)
  Reroute(None=@n186)
  Reroute(None=@n1364)
  Reroute(None=@n304)
  Reroute(None=@n320)
  Saturate(None=@n1356.UVs)
  n1901 = Add(A=@n1382.utput, B=Multiply(A=Time()) {constB=0})
  Saturate(None=LinearInterpolate(Alpha=Sine(None=@n1901)) {constA=0.4, constB=1.25})
  Saturate(None=@n1379.UVs)
  n1902 = Add(A=@n316.utput, B=Multiply(A=Time()) {constB=0})
  Saturate(None=LinearInterpolate(Alpha=Sine(None=@n1902)) {constA=0.4, constB=1.25})
  Saturate(None=@n1388.UVs)
  Saturate()
  ScalarParameter() {defaultValue=0.5, group=Effect_Decal_Reactive_Uv_Offset_LR, parameterName=EDRec_DecalB-Scale Icon}
  ScalarParameter() {group=Effect_Decal_Reactive_Uv_Offset_LR, parameterName=EDRec_DecalB_Rotation_B_Speed}
  ScalarParameter() {group=Effect_Decal_Reactive_Uv_Offset_LR, parameterName=EDRec_DecalB_Rotation_A_Speed}
  ScalarParameter() {defaultValue=0.5, group=Effect_Decal_Reactive_Uv_Offset_LR, parameterName=EDRec_DecalC-Scale Icon}
  ScalarParameter() {group=Effect_Decal_Reactive_Uv_Offset_LR, parameterName=EDRec_DecalB_Rotation_B_Speed}
  ScalarParameter() {group=Effect_Decal_Reactive_Uv_Offset_LR, parameterName=EDRec_DecalB_Rotation_A_Speed}
  ScalarParameter() {defaultValue=0.5, group=Effect_Decal_Reactive_Uv_Offset_LR, parameterName=EDRec_DecalD-Scale Icon}
  ScalarParameter() {group=Effect_Decal_Reactive_Uv_Offset_LR, parameterName=EDRec_DecalB_Rotate_Tile}
  ScalarParameter() {group=Effect_Decal_Reactive_Uv_Offset_LR, parameterName=EDRec_DecalA_Rotate_Tile}
  ScalarParameter() {group=Effect_Decal_Reactive_Uv_Offset_LR, parameterName=EDRec_DecalC_Rotate_Tile}
  ScalarParameter() {group=Effect_Decal_Reactive_Uv_Offset_LR, parameterName=EDRec_DecalD_Rotate_Tile}
  StaticBoolParameter() {group=Effect_Decal_Reactive_Counter, parameterName="EDRec_Decal_Counter_Pos_L\\R", sortPriority=-10}
  StaticBool() {value=true}
  StaticSwitchParameter(
   True=LinearInterpolate(A=@n139, B=NamedRerouteDeclaration(None=@n566.Alpha) {name=CC_Alpha_B, nodeColor=<rgba 0 1 0.25 1>}, Alpha=@n602)
   False=@n139
  ) {group=Effect_Decal_Reactive_Emissive_B, parameterName="EDRec-Emissive-B-Color--->CCurve"}
  StaticSwitchParameter(True=Multiply(A=NamedRerouteUsage(), B=NamedRerouteUsage()), False=@n1858) {
   desc=uses the Mask texture as a diffues emissive
   group=Effect_Decal_Reactive_Emissive_A
   parameterName="EDRec-Emissive-A--BC-Msk--->Simple-->Msk_Tex"
   sortPriority=500
  }
  StaticSwitchParameter(True=Multiply(A=NamedRerouteUsage(), B=NamedRerouteUsage()), False=@n571) {
   desc=uses the Mask texture as a diffues emissive
   group=Effect_Decal_Reactive_Emissive_B
   parameterName="EDRec-Emissive-B--BC-Msk--->Simple-->Msk_Tex"
   sortPriority=500
  }
  StaticSwitchParameter(True=Multiply(A=@n1882, B=@n1883), False=@n1889) {
   desc=uses the Mask texture as a diffues emissive
   group=Effect_Decal_Reactive_Emissive_A
   parameterName="EDRec-Emissive-A--BC-Msk--->Simple-->Msk_Tex"
   sortPriority=500
  }
  StaticSwitchParameter(True=Multiply(A=@n596, B=@n597), False=@n604) {
   desc=uses the Mask texture as a diffues emissive
   group=Effect_Decal_Reactive_Emissive_B
   parameterName="EDRec-Emissive-B--BC-Msk--->Simple-->Msk_Tex"
   sortPriority=500
  }
  StaticSwitchParameter(True=Multiply(A=NamedRerouteUsage(), B=NamedRerouteUsage()), False=@n708) {
   desc=uses the Mask texture as a diffues emissive
   group=Effect_Decal_Reactive_Emissive_C
   parameterName="EDRec-Emissive-C--BC-Msk--->Simple-->Msk_Tex"
   sortPriority=500
  }
  StaticSwitchParameter(True=Multiply(A=@n739, B=@n740), False=@n747) {
   desc=uses the Mask texture as a diffues emissive
   group=Effect_Decal_Reactive_Emissive_C
   parameterName="EDRec-Emissive-C--BC-Msk--->Simple-->Msk_Tex"
   sortPriority=500
  }
  StaticSwitchParameter(True=Multiply(A=NamedRerouteUsage(), B=NamedRerouteUsage()), False=@n796) {
   desc=uses the Mask texture as a diffues emissive
   group=Effect_Decal_Reactive_Emissive_D
   parameterName="EDRec-Emissive-D--BC-Msk--->Simple-->Msk_Tex"
   sortPriority=500
  }
  StaticSwitchParameter(True=Multiply(A=@n821, B=@n822), False=@n829) {
   desc=uses the Mask texture as a diffues emissive
   group=Effect_Decal_Reactive_Emissive_D
   parameterName="EDRec-Emissive-D--BC-Msk--->Simple-->Msk_Tex"
   sortPriority=500
  }
  n1903 = Time()
  n1904 = ScalarParameter() {group=Effect_Decal_Reactive_UV_Warp, parameterName=EDRec_WarpB_Sine_Speed}
  n1905 = Add(A=Sine(None=Multiply(A=@n1903, B=@n1904)))
  n1906 = Ceil(None=@n1904)
  StaticSwitch(
   True=NamedRerouteUsage()
   False=LinearInterpolate(A=@n1840, B=Multiply(A=@n1905, B=@n1840) {constB=0.25}, Alpha=Saturate(None=@n1906))
   Value=NamedRerouteUsage()
  )
  TextureSample!(UVs=@n932, Tex=@n934, Level=NamedRerouteUsage(), Apply View MipBias) {mipValueMode=TMVM_MipLevel}
  VectorParameter() {
   group=Effect_Decal_Reactive_FlipBook
   parameterName=EDRec_FlipBook_Frame__Offset
   channelNames {a=FrameD, b=FrameC, g=FrameB, r=FrameA}
   defaultValue=<rgba 0 0 0 1>
  }
 ]}
