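# Material function for weapon wrap skins that produces one decal layer's UVs, base color, alpha, mask and normal outputs.
# Projection functions build side or left/right decal UVs from pre-skinned bounds and normals, with switchable Front/Mid/Back safe-zone placement, rotation and sine motion.
# Switches add tiling, UV warping and a flipbook driven by time, randomness, reactive or color-curve inputs.
# Textures sample with explicit derivatives, and normals get rotation and side correction.
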MaterialFunction /Game/Weapons/WeaponSkins/Wraps/MaterialFunctions/MF_MultiDecal_Layer
graph {
 properties {description=Tileable Frame Random}
 nodes [
  MaterialExpressionFunctionInput_24 = FunctionInput() {bUsePreviewValueAsDefault=true, inputName=FlipBook Frame Speed, inputType=FunctionInput_Vector4, sortPriority=30}
  n1 = Convert(None=@MaterialExpressionFunctionInput_24) {
   nodeName=Break Float4
   convertInputs [{
     type=Vector4
     defaultValue=<rgba 0 0 0 1>
     expressionInput {
      expression=&self:MaterialExpressionFunctionInput_24
      inputName=None
      mask=0
      maskA=0
      maskB=0
      maskG=0
      maskR=0
      outputIndex=0
     }
    }]
   convertOutputs [{type=Scalar}, {type=Scalar}, {type=Scalar}, {type=Scalar}]
  }
  n2 = Add(
   A=FunctionInput() {bUsePreviewValueAsDefault=true, inputName=Flipbook Offset, inputType=FunctionInput_Scalar, sortPriority=30}
   B=Multiply(A=Time(), B=@n1)
  )
  n3 = Constant()
  n4 = StaticSwitch(
   True=Add(
    A=@n2
    B=FunctionInput(Preview=@n3) {bUsePreviewValueAsDefault=true, inputName=Tileable Frame, inputType=FunctionInput_Scalar, sortPriority=100}
   )
   False=@n2
   Value=NamedRerouteUsage()
  )
  n5 = MaterialFunctionCall!(2 Vector=Add(A=Ceil(None=@n4), B=Constant2Vector() {g=0.2062, r=0.0536}), RandomSeed) {materialFunction=&/MF_Random_2}
  Clamp(None=@n5.utput, Min=@n1, Max=@n1)
  n6 = Frac(None=@n4)
  Clamp(None=@n6, Min=@n1, Max=@n1)
  Constant() {r=2}
  n7 = NamedRerouteUsage()
  Frac(None=@n7)
  n8 = NamedRerouteDeclaration(None=FunctionInput() {bUsePreviewValueAsDefault=true, inputName=Rotation Bounce}) {name=LR_RotationA, nodeColor=<rgba 0.2422 0 1 1>}
  n9 = FunctionInput() {bUsePreviewValueAsDefault=true, inputName=SafeZone--Front, inputType=FunctionInput_Vector4, sortPriority=-2}
  n10 = ComponentMask(None=@n9) {a=true}
  n11 = FunctionInput(Preview=Constant() {r=1}) {bUsePreviewValueAsDefault=true, inputName=SafeZone--OverScale, inputType=FunctionInput_Scalar, sortPriority=-3}
  n12 = FunctionInput() {bUsePreviewValueAsDefault=true, inputName=SafeZone--Mid, inputType=FunctionInput_Vector4, sortPriority=-2}
  n13 = Multiply(A=@n11, B=ComponentMask(None=@n12) {a=true})
  n14 = FunctionInput() {bUsePreviewValueAsDefault=true, inputName=SafeZone--Back, inputType=FunctionInput_Vector4, sortPriority=-2}
  n15 = FunctionInput(Preview=Constant4Vector() {constant=<rgba 0 0 0 0.25>}.RGBA) {bUsePreviewValueAsDefault=true, inputName=Uv Offset/Scale, inputType=FunctionInput_Vector4}
  n16 = ComponentMask(None=@n15) {a=true}
  n17 = StaticSwitch(True=Multiply(A=@n11, B=ComponentMask(None=@n14) {a=true}), False=@n16, Value=NamedRerouteUsage())
  n18 = NamedRerouteUsage()
  n19 = Multiply(
   A=StaticSwitch(True=Multiply(A=@n11, B=@n10), False=StaticSwitch(True=@n13, False=@n17, Value=@n18), Value=NamedRerouteUsage())
   B=FunctionInput(Preview=Constant() {r=1}) {bUsePreviewValueAsDefault=true, inputName=NonUniform Scale, previewValue=<1 1 1 1>}
  )
  n20 = StaticSwitch(True=@n14, False=@n15, Value=NamedRerouteUsage())
  n21 = NamedRerouteUsage()
  n22 = ComponentMask(None=StaticSwitch(True=@n9, False=StaticSwitch(True=@n12, False=@n20, Value=@n21), Value=NamedRerouteUsage())) {b=true, g=true, r=true}
  n23 = NamedRerouteUsage()
  n24 = StaticSwitch(True=@n23, False=Constant() {r=1}, Value=NamedRerouteUsage())
  n25 = NamedRerouteUsage()
  n26 = StaticSwitch(
   True=StaticSwitch(True=@n23, False=StaticSwitch(True=@n23, False=@n24, Value=@n25), Value=NamedRerouteUsage())
   False=@n23
   Value=NamedRerouteUsage()
  )
  n27 = FunctionInput(Preview=Constant()) {bUsePreviewValueAsDefault=true, inputName=Sine Animation, inputType=FunctionInput_Vector4}
  n28 = ComponentMask(None=@n27) {b=true, g=true, r=true}
  n29 = ComponentMask(None=@n27) {a=true}
  n30 = LinearInterpolate(A=@n28, Alpha=@n29) {constB=0}
  n31 = StaticBool()
  n32 = FunctionInput(Preview=Constant()) {bUsePreviewValueAsDefault=true, inputName=Rotation_Pivot offset}
  n33 = LinearInterpolate(A=@n28, B=@n28, Alpha=@n29)
  n34 = MaterialFunctionCall!(
   Normal_Map
   Offset RT=@n33
   Pivot=@n32
   "Pivot_L\\R"
   Caculated_Bounds=NamedRerouteUsage()
   Pre_Skinned_Normals=NamedRerouteUsage()
   Is State=NamedRerouteUsage()
   Uvs_Rotation_Mirroring=FunctionInput() {bUsePreviewValueAsDefault=true, inputName=Rotation_Mirroring, inputType=FunctionInput_Scalar}
   Uvs_Rotation=@n8
   Uvs_TileRate=@n19
   Uvs_Offset=Add(A=@n30, B=Multiply(A=@n22, B=@n26))
   Clamp Y=@n31
   Clamp X=@n31
   Clamp Padding_Comp
  ) {materialFunction=&/MF_Projection_Side_Left_Right_Decal_NormalV2}
  n35 = MaterialFunctionCall!(
   Normal_Map
   Offset Rt=@n33
   Pivot=@n32
   Caculated_Bounds=NamedRerouteUsage()
   Pre_Skinned_Normals=NamedRerouteUsage()
   is_State=NamedRerouteUsage()
   UV_Rotation=@n8
   Uvs_Offset=Add(A=@n30, B=@n22)
   Uvs_TileRate=@n19
   Uvs_Offset_Pivot
   For_Safe_Zones=StaticBool() {value=true}
   Rotation_Comp
   Clamp Y
   Clamp X
   Clamp Padding_Comp
   No_Rotation
  ) {materialFunction=&/MF_Projection_Side_Normal_V2}
  n36 = StaticSwitch(
   True=NamedRerouteUsage()
   False=StaticSwitch(True=NamedRerouteDeclaration(None=@n34.UVs) {name=Uvs_A_LR, nodeColor=<rgba 1 0.1875 0 1>}, False=@n35.UVs, Value=NamedRerouteUsage())
   Value=NamedRerouteUsage()
  )
  FunctionOutput(None=@n36) {outputName=Uvs, sortPriority=200}
  n37 = StaticSwitch(True=@n34.UVs No Offsets/Tiling, False=@n35.UVs No Offsets/Tiling, Value=NamedRerouteUsage())
  FunctionOutput(None=@n37) {outputName=UvsNoOffsets, sortPriority=200}
  n38 = NamedRerouteUsage()
  n39 = FunctionInput() {bUsePreviewValueAsDefault=true, inputName=Flipbook ColorCurve Animation, inputType=FunctionInput_Scalar, sortPriority=30}
  n40 = MaterialFunctionCall(Input=@n6, Input Low=Constant(), Input High=Constant() {r=1}, Target Low=@n1, Target High=@n1) {materialFunction=&RemapValueRange}
  n41 = MaterialFunctionCall(Input=@n5.utput, Input Low=Constant(), Input High=Constant() {r=1}, Target Low=@n1, Target High=@n1) {materialFunction=&RemapValueRange}
  n42 = FunctionInput(Preview=Constant()) {bUsePreviewValueAsDefault=true, inputName=Flipbook Frame, inputType=FunctionInput_Scalar, sortPriority=30}
  n43 = MaterialFunctionCall(Color=NamedRerouteUsage()) {materialFunction=&SplitComponents}
  n44 = Multiply(A=@n43.R, B=@n43.R)
  n45 = StaticBool()
  n46 = StaticSwitch(
   True=FunctionInput() {bUsePreviewValueAsDefault=true, inputName=Flipbook Reactive Anim, inputType=FunctionInput_Scalar, sortPriority=30}
   False=Add(A=LinearInterpolate(A=@n40.Result, B=@n41.Result, Alpha=@n1), B=Divide(A=@n42, B=@n44) {constA=9, constB=9})
   Value=FunctionInput(Preview=@n45) {bUsePreviewValueAsDefault=true, inputName=Flipbook Reactive State, inputType=FunctionInput_StaticBool, sortPriority=30}
  )
  n47 = FunctionInput(Preview=@n45) {bUsePreviewValueAsDefault=true, inputName=Flipbook ColorCurve State, inputType=FunctionInput_StaticBool, sortPriority=30}
  n48 = MaterialFunctionCall(Color=NamedRerouteUsage()) {materialFunction=&SplitComponents}
  n49 = Saturate(None=@n7)
  n50 = FunctionInput(Preview=Constant() {r=1}) {inputName=Uv_Calmp pxiel Comp, inputType=FunctionInput_Scalar}
  n51 = NamedRerouteUsage()
  MaterialExpressionSaturate_3 = Saturate(None=@n51)
  n52 = Convert(None=@MaterialExpressionSaturate_3) {
   nodeName=Break Float2
   convertInputs [{
     type=Vector2
     expressionInput {
      expression=&self:MaterialExpressionSaturate_3
      inputName=None
      mask=0
      maskA=0
      maskB=0
      maskG=0
      maskR=0
      outputIndex=0
     }
    }]
   convertOutputs [{type=Scalar}, {type=Scalar}]
  }
  n53 = FunctionInput(Preview=Constant()) {inputName="Uv_Calmp----->Tile\\Speed", inputType=FunctionInput_Vector4}
  n54 = Multiply(A=Time(), B=ComponentMask(None=@n53) {a=true, b=true})
  MaterialExpressionFrac_2 = Frac(None=Add(A=@n51, B=Frac(None=@n54)) {constB=0})
  n55 = Convert(None=@MaterialExpressionFrac_2) {
   nodeName=Break Float2
   convertInputs [{
     type=Vector2
     expressionInput {
      expression=&self:MaterialExpressionFrac_2
      inputName=None
      mask=0
      maskA=0
      maskB=0
      maskG=0
      maskR=0
      outputIndex=0
     }
    }]
   convertOutputs [{type=Scalar}, {type=Scalar}]
  }
  MaterialExpressionSaturate_1 = Saturate(None=@n53)
  n56 = Convert(None=@MaterialExpressionSaturate_1) {
   nodeName=Break Float4
   convertInputs [{
     type=Vector4
     defaultValue=<rgba 0 0 0 1>
     expressionInput {
      expression=&self:MaterialExpressionSaturate_1
      inputName=None
      mask=0
      maskA=0
      maskB=0
      maskG=0
      maskR=0
      outputIndex=0
     }
    }]
   convertOutputs [{type=Scalar}, {type=Scalar}, {type=Scalar}, {type=Scalar}]
  }
  n57 = AppendVector(A=LinearInterpolate(A=@n52, B=@n55, Alpha=@n56), B=LinearInterpolate(A=@n52, B=@n55, Alpha=@n56))
  n58 = MaterialFunctionCall(
   "Animation  Phase (0-1)"=Frac(None=StaticSwitch(True=@n39, False=@n46, Value=@n47))
   Number of Rows=@n48.R
   Number of Columns=@n48.R
   UVs=StaticSwitch(
    True=MaterialFunctionCall(UVs=@n49, Texture Scale=@n50) {materialFunction=&ScaleUVsByCenter}.UVs
    False=MaterialFunctionCall(UVs=@n57, Texture Scale=@n50) {materialFunction=&ScaleUVsByCenter}.UVs
    Value=NamedRerouteUsage()
   )
   "Clamp Anim (see tooltip)"=StaticBool()
  ) {materialFunction=&FlipBook}
  n59 = Constant()
  n60 = MaterialFunctionCall!(
   Normal_Map
   Normals_Source=TextureSample!(
    UVs=@n58.UVs
    Tex=FunctionInput() {bUsePreviewValueAsDefault=true, inputName=Z-Input-N-Tex, inputType=FunctionInput_Texture2D, sortPriority=550}
    "DDX(UVs)"=DDX(Value=@n38)
    "DDY(UVs)"=DDY(Value=@n38)
    Apply View MipBias
   ) {mipValueMode=TMVM_Derivative}.RGB
   Normals_Source_Use=StaticBool() {value=true}
   UVs=NamedRerouteUsage()
   Rotation_Comp
   UV_Rotation=StaticSwitch(
    True=FunctionInput(Preview=@n59) {bUsePreviewValueAsDefault=true, inputName=Tileable Rotation, inputType=FunctionInput_Scalar, sortPriority=100}
    False=NamedRerouteUsage()
    Value=NamedRerouteUsage()
   )
   Pre_Skinned_Normals=NamedRerouteUsage()
   is_State=NamedRerouteUsage()
   "Left\\Right"=NamedRerouteUsage()
   No_Rotation
  ) {materialFunction=&/MF_UVs_Side_Fix_Normals}
  FunctionOutput(None=@n60.Raw_Normal) {outputName=Tex--Normals}
  n61 = NamedRerouteUsage()
  n62 = TextureSample(
   UVs=@n58.UVs
   Tex=FunctionInput() {bUsePreviewValueAsDefault=true, inputName=Z-Input-Msk-Tex, inputType=FunctionInput_Texture2D, sortPriority=550}
   "DDX(UVs)"=DDX(Value=@n61)
   "DDY(UVs)"=DDY(Value=@n61)
  ) {mipValueMode=TMVM_Derivative}
  n63 = StaticSwitch(True=Multiply(A=NamedRerouteUsage(), B=@n62.RGBA), False=@n62.RGBA, Value=NamedRerouteUsage())
  FunctionOutput(None=@n63) {outputName=Tex---Msk}
  n64 = NamedRerouteUsage()
  n65 = FunctionInput(Preview=TextureObject() {texture=/Game/Weapons/WeaponSkins/Wraps/Textures/UVCheaker}) {bUsePreviewValueAsDefault=true, inputName=Z-Input-BC-Tex, inputType=FunctionInput_Texture2D, sortPriority=550}
  n66 = NamedRerouteUsage()
  n67 = TextureSample(UVs=@n58.UVs, Tex=@n65, "DDX(UVs)"=DDX(Value=@n66), "DDY(UVs)"=DDY(Value=@n66)) {mipValueMode=TMVM_Derivative}
  n68 = StaticSwitch(
   True=FunctionInput(Preview=Constant()) {inputName=Emissive, inputType=FunctionInput_Scalar, sortPriority=200}
   False=StaticSwitch(True=Multiply(A=@n64, B=@n67.A), False=@n67.A, Value=NamedRerouteUsage())
   Value=FunctionInput(Preview=StaticBool()) {bUsePreviewValueAsDefault=true, inputName=Emissive only State, inputType=FunctionInput_StaticBool, sortPriority=200}
  )
  FunctionOutput(None=@n68) {outputName=Tex---BC Alpha}
  FunctionOutput(None=@n67.RGB) {outputName=Tex---BC}
  FunctionOutput(None=@n34.UVs) {outputName="Uvs_L\\R", sortPriority=200}
  Multiply(A=@n58.UV Center)
  n69 = Constant() {r=0.5}
  n70 = MaterialFunctionCall(Float2=NamedRerouteUsage()) {materialFunction=&BreakOutFloat2Components}
  n71 = FunctionInput(Preview=Constant() {r=1}) {bUsePreviewValueAsDefault=true, inputName=Tileable Mask, inputType=FunctionInput_Scalar, sortPriority=100}
  Multiply(A=Multiply(A=SphereMask(A=@n69, B=@n70.R) {attenuationRadius=0.5}, B=SphereMask(A=@n70.G, B=@n69) {attenuationRadius=0.5}), B=@n71)
  n72 = FunctionInput(Preview=Constant()) {bUsePreviewValueAsDefault=true, inputName=Z-input-isState, inputType=FunctionInput_Vector4, sortPriority=500}
  NamedRerouteDeclaration(None=@n72) {name=is State, nodeColor=<rgba 0 1 0.4141 1>}
  n73 = StaticBool()
  NamedRerouteDeclaration(None=FunctionInput(Preview=@n73) {bUsePreviewValueAsDefault=true, inputName=L/R State, inputType=FunctionInput_StaticBool}) {name=LR State, nodeColor=<rgba 0 0.9766 1 1>}
  NamedRerouteDeclaration(None=@n71) {name=Tile Msking, nodeColor=<rgba 0 1 0.25 1>}
  NamedRerouteDeclaration(None=@n36) {name=Uvs, nodeColor=<rgba 1 0.1875 0 1>}
  n74 = PreSkinnedLocalBounds()
  n75 = Subtract(A=VertexInterpolator(VS=PreSkinnedPosition()).PS, B=@n74.Min)
  n76 = MaterialFunctionCall(Color=Subtract(A=@n74.Max, B=@n74.Max)) {materialFunction=&SplitComponents}
  n77 = Max(A=Max(A=@n76.R, B=@n76.R), B=@n76.B)
  NamedRerouteDeclaration(None=FunctionInput(Preview=Divide(A=@n75, B=@n77)) {bUsePreviewValueAsDefault=true, inputName=Z-input--CalculateBounds, sortPriority=500}) {name=Caclcated_Bounds, nodeColor=<rgba 0 0.9297 1 1>}
  n78 = StaticSwitch(True=@n34.Clamp_Mask, False=@n35.Safe_Zone_Mask, Value=NamedRerouteUsage())
  n79 = StaticSwitch(True=@n78, False=Constant(), Value=NamedRerouteUsage())
  n80 = NamedRerouteUsage()
  NamedRerouteDeclaration(None=StaticSwitch(True=@n78, False=StaticSwitch(True=@n78, False=@n79, Value=@n80), Value=NamedRerouteUsage())) {name=Decal_A_Msk_SZ, nodeColor=<rgba 0 1 0.1562 1>}
  NamedRerouteDeclaration() {name=PreSkin_Normal, nodeColor=<rgba 0.8281 1 0 1>}
  NamedRerouteDeclaration(None=LinearInterpolate(
    A=Constant() {r=1}
    B=Constant3Vector() {constant=<rgba -1 1 1 1>}
    Alpha=MaterialFunctionCall(Float4=@n72) {materialFunction=&BreakOutFloat4Components}.R
   )) {name=SZ_Offset_Comp, nodeColor=<rgba 0.8047 1 0 1>}
  n81 = NamedRerouteUsage()
  n82 = LinearInterpolate(
   A=@n81
   B=OneMinus(None=@n81)
   Alpha=ScalarParameter() {group=Effect_Decal_Reactive_RE, parameterName="EDRec_DecalA_Opacity-TopS Blend 01<--->02 MSk   1=invert"}
  )
  n83 = ScalarParameter() {
   defaultValue=0.1
   desc=the point wich you see active to non active.
   group=Effect_Decal_Reactive_RE_Audio
   parameterName=EDRec Audio Opacity_FadePoint
  }
  n84 = NamedRerouteUsage()
  n85 = Max(A=SmoothStep(Max=@n83, Value=NamedRerouteUsage()) {constMax=0.1}, B=@n84)
  n86 = NamedRerouteUsage()
  n87 = NamedRerouteDeclaration(None=@n68) {name=Decal_A_Alpha, nodeColor=<rgba 1 0 0.7734 1>}
  n88 = StaticSwitch(True=Multiply(A=@n87, B=StaticSwitch(True=@n85, False=@n84, Value=@n86)), False=@n87, Value=NamedRerouteUsage())
  n89 = NamedRerouteUsage()
  n90 = NamedRerouteUsage()
  n91 = StaticSwitch(
   True=LinearInterpolate(
    B=@n88
    Alpha=StaticSwitchParameter(True=@n89, False=@n90) {group=Effect_Decal_Reactive_RE, parameterName="EDRec_DecalA_Opacity-TOD----->TODB", sortPriority=-1}
   ) {constB=0}
   False=@n88
   Value=StaticBoolParameter() {group=Effect_Decal_Reactive_RE, parameterName=EDRec_DecalA_Opacity-TOD, sortPriority=-1}
  )
  n92 = StaticSwitch(
   True=StaticSwitch(
    True=Multiply(A=@n82, B=@n91)
    False=@n91
    Value=StaticBoolParameter() {group=Effect_Decal_Reactive_RE, parameterName="EDRec_DecalA_Opacity-TopS Blend 01<--->02 MSk"}
   )
   False=@n91
   Value=StaticBoolParameter() {group="TopSide_V02_Layer 01 Blend  02", parameterName="TopS Blend 01<--->02"}
  )
  n93 = Add(A=@n92)
  NamedRerouteDeclaration(None=Saturate(None=Add(B=@n93))) {name=Alpha_Set_B without D, nodeColor=<rgba 0.8047 0 1 1>}
  NamedRerouteDeclaration(None=@n92) {name=Alpha_A_Msk, nodeColor=<rgba 0 0.0625 1 1>}
  NamedRerouteDeclaration(None=FunctionInput(Preview=@n73) {bUsePreviewValueAsDefault=true, inputName=SafeZone---Front, inputType=FunctionInput_StaticBool, sortPriority=-1}) {name=SZ Front State, nodeColor=<rgba 1 0.7031 0 1>}
  NamedRerouteDeclaration(None=@n37) {name=Uvs_A_No, nodeColor=<rgba 1 0.1875 0 1>}
  NamedRerouteDeclaration(None=StaticSwitch(True=@n34.UVs, False=@n35.UVs, Value=NamedRerouteUsage())) {name=Uvs_A_NotClamp, nodeColor=<rgba 1 0.1875 0 1>}
  NamedRerouteDeclaration(None=FunctionInput() {bUsePreviewValueAsDefault=true, inputName=Tilable SysTem, inputType=FunctionInput_Vector2, sortPriority=100}) {name=UV_Tile System, nodeColor=<rgba 1 0.7734 0 1>}
  NamedRerouteDeclaration(None=FunctionInput(Preview=@n73) {bUsePreviewValueAsDefault=true, inputName=Tileable State, inputType=FunctionInput_StaticBool, sortPriority=100}) {name=Tileable State, nodeColor=<rgba 0.5938 1 0 1>}
  n94 = Add(A=@n93)
  NamedRerouteDeclaration(None=Multiply(A=NamedRerouteUsage(), B=Saturate(None=@n94))) {name=Alpha_Set_B, nodeColor=<rgba 0.8047 0 1 1>}
  NamedRerouteDeclaration(None=FunctionInput(Preview=@n73) {bUsePreviewValueAsDefault=true, inputName=SafeZone---Back, inputType=FunctionInput_StaticBool, sortPriority=-1}) {name=SZ Back State, nodeColor=<rgba 1 0.7031 0 1>}
  NamedRerouteDeclaration(None=LinearInterpolate(B=@n67.RGB, Alpha=@n92)) {name=BC_Set_B, nodeColor=<rgba 0.8047 0 1 1>}
  NamedRerouteDeclaration(None=FunctionInput(Preview=@n73) {bUsePreviewValueAsDefault=true, inputName=SafeZone---Mid, inputType=FunctionInput_StaticBool, sortPriority=-1}) {name=SZ Mid State, nodeColor=<rgba 1 0.7031 0 1>}
  NamedRerouteDeclaration(None=LinearInterpolate(B=NamedRerouteDeclaration(None=@n63) {name=Msk_A, nodeColor=<rgba 1 0.75 0 1>}, Alpha=@n92)) {name=Msk_Set_B, nodeColor=<rgba 0.9922 0 1 1>}
  NamedRerouteDeclaration(None=LinearInterpolate(B=@n60.Raw_Normal, Alpha=@n92)) {name=Nrml_Set_B, nodeColor=<rgba 0.9922 0 1 1>}
  NamedRerouteDeclaration(None=FunctionInput(Preview=Constant3Vector() {constant=<rgba 2 2 0 0>}) {bUsePreviewValueAsDefault=true, inputName="FlipBook Colums\\Rows", sortPriority=30}) {name=Frame C/R, nodeColor=<rgba 1 0.7734 0 1>}
  n95 = PreSkinnedNormal()
  NamedRerouteDeclaration(None=FunctionInput(Preview=VertexInterpolator(VS=@n95).PS) {bUsePreviewValueAsDefault=true, inputName=Z-input--PreSkin, sortPriority=500}) {name=Preskin, nodeColor=<rgba 0 0.7891 1 1>}
  n96 = NamedRerouteUsage()
  n97 = Add(A=FunctionInput() {bUsePreviewValueAsDefault=true, inputName=Warp Uvs, inputType=FunctionInput_Vector2, sortPriority=200}, B=@n96)
  n98 = FunctionInput(Preview=StaticBool()) {bUsePreviewValueAsDefault=true, inputName=warp UVs State, inputType=FunctionInput_StaticBool, sortPriority=200}
  n99 = StaticBool()
  NamedRerouteDeclaration(None=StaticSwitch(
    True=@n97
    False=StaticSwitch(True=@n97, False=@n96, Value=@n98)
    Value=FunctionInput(Preview=@n99) {bUsePreviewValueAsDefault=true, inputName=warp Audio State, inputType=FunctionInput_StaticBool, sortPriority=200}
   )) {name=Uvs_Flip, nodeColor=<rgba 0.5703 0 1 1>}
  n100 = MaterialFunctionCall() {materialFunction=&SplitComponents}
  NamedRerouteDeclaration(None=@n100.R) {name="Rows_####", nodeColor=<rgba 0 0.5312 1 1>}
  NamedRerouteDeclaration(None=@n100.G) {name="Colums_####", nodeColor=<rgba 0 0.5312 1 1>}
  NamedRerouteUsage()
  n101 = Constant() {desc="UV Scale for tiling ", r=1}
  n102 = Multiply(A=@n16, B=@n101)
  n103 = StaticBool() {desc=reactive offset}
  Reroute(None=StaticSwitch(True=@n102, False=@n101, Value=@n103))
  Saturate(None=@n58.UVs)
  StaticSwitch(True=@n102, False=@n16, Value=@n103)
  n104 = Power(Base=Constant() {r=2}, Exp=ScalarParameter() {group=Effect_MultDecal_V02, parameterName=EMDv2_MipBias})
  n105 = NamedRerouteUsage()
  n106 = TextureSample!(UVs=@n58.UVs, Tex=RecordTextureStreamingInfo!(Tex=@n65, UVs=Multiply(A=@n104, B=@n105)).Tex, Apply View MipBias)
  n107 = NamedRerouteUsage()
  StaticSwitch(True=@n67.RGB, False=@n106.RGB, Value=@n107)
  StaticSwitch(True=@n67.A, False=@n106.A, Value=@n107)
  TextureObject() {texture=/Engine/Functions/Engine_MaterialFunctions02/ExampleContent/Textures/flipbook}
 ]
}
