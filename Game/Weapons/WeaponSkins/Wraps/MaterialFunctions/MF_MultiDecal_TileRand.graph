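# Material function for weapon wrap skins that tiles decal UVs and randomizes each tile.
# UVs, defaulting to pre-skinned position normalized by local bounds, are rotated, offset, scrolled and split into a grid with per-column Y staggering.
# Sine-dot hashes of seeded cell IDs drive each tile's time-animated scale and rotation.
# The same hashes feed outputs for tile masking, flipbook value and emissive CC offset, alongside the tile UV and rotation.
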MaterialFunction /Game/Weapons/WeaponSkins/Wraps/MaterialFunctions/MF_MultiDecal_TileRand
graph {nodes [
  MaterialExpressionFunctionInput_9 = FunctionInput(Preview=Constant() {r=1.0123}) {inputName=Rnd Seed, inputType=FunctionInput_Vector4}
  n1 = Convert(None=@MaterialExpressionFunctionInput_9) {
   nodeName=Break Float4
   convertInputs [{
     type=Vector4
     defaultValue=<rgba 0 0 0 1>
     expressionInput {
      expression=&self:MaterialExpressionFunctionInput_9
      inputName=None
      mask=0
      maskA=0
      maskB=0
      maskG=0
      maskR=0
      outputIndex=0
     }
    }]
   convertOutputs [{type=Scalar}, {type=Scalar}, {type=Scalar}, {type=Scalar}]
  }
  n2 = Constant() {r=43758.5469}
  Add(A=@n1, B=@n2)
  n3 = PreSkinnedPosition()
  n4 = PreSkinnedLocalBounds()
  n5 = MaterialFunctionCall(Color=Subtract(A=@n4.Max, B=@n4.Max)) {materialFunction=&SplitComponents}
  n6 = Divide(A=Subtract(A=VertexInterpolator(VS=@n3).PS, B=@n4.Min), B=Max(A=Max(A=@n5.R, B=@n5.R), B=@n5.B))
  n7 = FunctionInput(Preview=ComponentMask(None=NamedRerouteDeclaration(None=@n6) {name=Caclcated_Bounds, nodeColor=<rgba 0 0.9297 1 1>}) {g=true, r=true}) {inputName=UVs-NO offsets, inputType=FunctionInput_Vector2}
  n8 = FunctionInput(Preview=Constant()) {inputName=UVs-Rotation, inputType=FunctionInput_Scalar}
  n9 = Constant()
  n10 = FunctionInput(Preview=@n9) {bUsePreviewValueAsDefault=true, inputName=UVs-offsets}
  n11 = FunctionInput(Preview=@n9) {inputName=UVs_tile Speed, inputType=FunctionInput_Vector4}
  MaterialExpressionAdd_209 = Add(
   A=Add(
    A=MaterialFunctionCall(UVs=@n7, "Rotation Angle (0-1)"=@n8) {materialFunction=&CustomRotator}.Rotated Values
    B=ComponentMask(None=@n10) {g=true, r=true}
   ) {constB=0}
   B=Multiply(A=Time(), B=ComponentMask(None=@n11) {g=true, r=true})
  )
  n12 = Convert(None=@MaterialExpressionAdd_209) {
   nodeName=Break Float2
   convertInputs [{
     type=Vector2
     expressionInput {
      expression=&self:MaterialExpressionAdd_209
      inputName=None
      mask=0
      maskA=0
      maskB=0
      maskG=0
      maskR=0
      outputIndex=0
     }
    }]
   convertOutputs [{type=Scalar}, {type=Scalar}]
  }
  n13 = AppendVector(A=Frac(None=@n12), B=@n12)
  MaterialExpressionFunctionInput_7 = FunctionInput(Preview=Constant3Vector() {constant=<rgba 10 1 0 0>}) {inputName="UVs-TileRate\\Scale"}
  n14 = Convert(None=@MaterialExpressionFunctionInput_7) {
   nodeName=Break Float3
   convertInputs [{
     type=Vector3
     expressionInput {
      expression=&self:MaterialExpressionFunctionInput_7
      inputName=None
      mask=0
      maskA=0
      maskB=0
      maskG=0
      maskR=0
      outputIndex=0
     }
    }]
   convertOutputs [{type=Scalar}, {type=Scalar}, {type=Scalar}]
  }
  n15 = Multiply(A=ComponentMask(None=@n13) {r=true}, B=@n14) {constB=5}
  n16 = Add(A=Divide(A=Floor(None=@n15), B=@n14) {constB=5}, B=ComponentMask(None=@n11) {a=true}) {constB=-0.5}
  n17 = Constant()
  MaterialExpressionMultiply_0 = Multiply(A=Time(), B=@n11)
  n18 = Convert(None=@MaterialExpressionMultiply_0) {
   nodeName=Break Float4
   convertInputs [{
     type=Vector4
     defaultValue=<rgba 0 0 0 1>
     expressionInput {
      expression=&self:MaterialExpressionMultiply_0
      inputName=None
      mask=0
      maskA=0
      maskB=0
      maskG=0
      maskR=0
      outputIndex=0
     }
    }]
   convertOutputs [{type=Scalar}, {type=Scalar}, {type=Scalar}, {type=Scalar}]
  }
  n19 = Add(A=Multiply(A=@n14, B=FunctionInput(Preview=@n17) {inputName=UVs-TileRate_Y_Offset, inputType=FunctionInput_Scalar}), B=@n18)
  n20 = Add(A=Multiply(A=@n13, B=@n14), B=AppendVector(A=Constant(), B=Multiply(A=@n16, B=@n19))) {constB=0}
  n21 = Divide(A=Floor(None=@n20), B=@n14)
  Add(A=@n21) {constB=0.01}
  Add(A=@n2, B=@n1)
  ComponentMask(None=@n10) {a=true}
  Constant() {r=0.2}
  MaterialExpressionFunctionInput_6 = FunctionInput(Preview=Constant4Vector() {constant=<rgba 1 1 0 0>}.RGBA) {inputName=Rnd-Scale, inputType=FunctionInput_Vector4}
  n22 = Convert(None=@MaterialExpressionFunctionInput_6) {
   nodeName=Break Float4
   convertInputs [{
     type=Vector4
     defaultValue=<rgba 0 0 0 1>
     expressionInput {
      expression=&self:MaterialExpressionFunctionInput_6
      inputName=None
      mask=0
      maskA=0
      maskB=0
      maskG=0
      maskR=0
      outputIndex=0
     }
    }]
   convertOutputs [{type=Scalar}, {type=Scalar}, {type=Scalar}, {type=Scalar}]
  }
  n23 = Add(A=@n1, B=@n21)
  n24 = Constant2Vector() {g=724.3587, r=235.9474}
  n25 = Multiply(A=Sine(None=DotProduct(A=@n23, B=@n24)), B=@n2)
  n26 = Add(A=Multiply(A=@n22, B=Time()), B=@n25)
  n27 = Add(A=Sine(None=@n26))
  n28 = Frac(None=LinearInterpolate(A=@n26, B=Multiply(A=@n27) {constB=0.49}, Alpha=@n22))
  n29 = LinearInterpolate(A=@n22, B=@n22, Alpha=@n28)
  MaterialExpressionFunctionInput_8 = FunctionInput(Preview=Constant()) {inputName=Rnd-Rotation, inputType=FunctionInput_Vector4}
  n30 = Convert(None=@MaterialExpressionFunctionInput_8) {
   nodeName=Break Float4
   convertInputs [{
     type=Vector4
     defaultValue=<rgba 0 0 0 1>
     expressionInput {
      expression=&self:MaterialExpressionFunctionInput_8
      inputName=None
      mask=0
      maskA=0
      maskB=0
      maskG=0
      maskR=0
      outputIndex=0
     }
    }]
   convertOutputs [{type=Scalar}, {type=Scalar}, {type=Scalar}, {type=Scalar}]
  }
  n31 = Add(A=@n1, B=@n21)
  n32 = Constant2Vector() {g=280.3817, r=574.9868}
  n33 = Multiply(A=Sine(None=DotProduct(A=@n31, B=@n32)), B=@n2)
  n34 = Frac(None=Add(A=FunctionInput() {inputName=Rnd-Rotation_offset, inputType=FunctionInput_Scalar}, B=@n33))
  n35 = Add(A=Multiply(A=Time(), B=LinearInterpolate(A=@n30, B=@n30, Alpha=@n34)), B=LinearInterpolate(A=@n30, B=@n30, Alpha=@n34))
  n36 = MaterialFunctionCall(
   UVs=MaterialFunctionCall(UVs=Frac(None=@n20), Texture Scale=Multiply(A=@n14, B=@n29)) {materialFunction=&ScaleUVsByCenter}.UVs
   "Rotation Angle (0-1)"=@n35
  ) {materialFunction=&CustomRotator}
  FunctionOutput(None=@n36.Rotated Values) {outputName=Uv Tile}
  FunctionOutput(None=@n35) {outputName=Uv_Rotation}
  MaterialExpressionFunctionInput_14 = FunctionInput(Preview=Constant4Vector().RGBA) {inputName=Rnd Msking, inputType=FunctionInput_Vector4}
  n37 = Convert(None=@MaterialExpressionFunctionInput_14) {
   nodeName=Break Float4
   convertInputs [{
     type=Vector4
     defaultValue=<rgba 0 0 0 1>
     expressionInput {
      expression=&self:MaterialExpressionFunctionInput_14
      inputName=None
      mask=0
      maskA=0
      maskB=0
      maskG=0
      maskR=0
      outputIndex=0
     }
    }]
   convertOutputs [{type=Scalar}, {type=Scalar}, {type=Scalar}, {type=Scalar}]
  }
  MaterialExpressionFunctionInput_11 = FunctionInput() {inputName="Rnd Msk Speed\\Sine"}
  n38 = Convert(None=@MaterialExpressionFunctionInput_11) {
   nodeName=Break Float3
   convertInputs [{
     type=Vector3
     expressionInput {
      expression=&self:MaterialExpressionFunctionInput_11
      inputName=None
      mask=0
      maskA=0
      maskB=0
      maskG=0
      maskR=0
      outputIndex=0
     }
    }]
   convertOutputs [{type=Scalar}, {type=Scalar}, {type=Scalar}]
  }
  n39 = Add(A=@n25, B=Multiply(A=@n38, B=Time()))
  n40 = Add(A=@n39, B=@n37)
  n41 = LinearInterpolate(A=@n39, B=ConstantBiasScale(None=Sine(None=@n40)) {scale=0.49}, Alpha=@n38)
  n42 = LinearInterpolate(A=Add(A=@n41, B=@n37), B=@n41, Alpha=@n38)
  n43 = Multiply(A=SmoothStep(Min=@n37, Max=@n37, Value=Frac(None=@n42)), B=@n37)
  FunctionOutput(None=@n43) {outputName=Tile Msking}
  MaterialExpressionFunctionInput_13 = FunctionInput() {inputName=Rnd Flipbook, inputType=FunctionInput_Vector4}
  n44 = Convert(None=@MaterialExpressionFunctionInput_13) {
   nodeName=Break Float4
   convertInputs [{
     type=Vector4
     defaultValue=<rgba 0 0 0 1>
     expressionInput {
      expression=&self:MaterialExpressionFunctionInput_13
      inputName=None
      mask=0
      maskA=0
      maskB=0
      maskG=0
      maskR=0
      outputIndex=0
     }
    }]
   convertOutputs [{type=Scalar}, {type=Scalar}, {type=Scalar}, {type=Scalar}]
  }
  n45 = LinearInterpolate(A=@n44, B=@n44, Alpha=Frac(None=Add(A=@n44, B=@n33)))
  FunctionOutput(None=@n45) {outputName=Tile Flipbook}
  MaterialExpressionFunctionInput_12 = FunctionInput() {inputName=Rnd CC offset, inputType=FunctionInput_Vector4}
  n46 = Convert(None=@MaterialExpressionFunctionInput_12) {
   nodeName=Break Float4
   convertInputs [{
     type=Vector4
     defaultValue=<rgba 0 0 0 1>
     expressionInput {
      expression=&self:MaterialExpressionFunctionInput_12
      inputName=None
      mask=0
      maskA=0
      maskB=0
      maskG=0
      maskR=0
      outputIndex=0
     }
    }]
   convertOutputs [{type=Scalar}, {type=Scalar}, {type=Scalar}, {type=Scalar}]
  }
  n47 = LinearInterpolate(A=@n46, B=@n46, Alpha=Frac(None=Add(A=@n25, B=@n46)))
  FunctionOutput(None=@n47) {outputName=Emissive CC}
  Multiply(A=MaterialFunctionCall!(X range Y range, Gradient Height, ShowGrid?, Thickness, Color, "f(x)"=@n28) {materialFunction=/Engine/Functions/Engine_MaterialFunctions01/Debug/PlotFunctionOnGraph}."f(x) = y")
  NamedRerouteDeclaration(None=@n36.Rotated Values) {name=UV_A_Tile, nodeColor=<rgba 1 0.7734 0 1>}
  NamedRerouteDeclaration(None=@n35) {name=UV_A_Tile_Rotation, nodeColor=<rgba 1 0.7734 0 1>}
  NamedRerouteDeclaration(None=@n43) {name=UV_A_Tile_Msk, nodeColor=<rgba 1 0.7734 0 1>}
  NamedRerouteDeclaration(None=@n45) {name=UV_A_Tile_Frame_Rnd, nodeColor=<rgba 1 0.7734 0 1>}
  NamedRerouteDeclaration(None=@n47) {name=EDRec_Emissive_A_CC_Rnd, nodeColor=<rgba 1 0.7734 0 1>}
  ScalarParameter() {group=Effect_Decal_Reactive_Uv_Offset_LR, parameterName=EDRec_DecalB_Rotation-RND-Seed-Offset}
  ScalarParameter() {group=Effect_Decal_Reactive_Uv_Offset_LR, parameterName=EDRec_DecalA_RND_Msking Speed}
  ScalarParameter() {group=Effect_Decal_Reactive_Uv_Offset_LR, parameterName=EDRec_DecalA_RND_Msking Speed Sine}
  VectorParameter() {
   group=Effect_Decal_Reactive_Uv_Offset_LR
   parameterName=EDRec_DecalA_Scale-RND--
   channelNames {a="1= Sine Speed", b=Speed, g=Max, r=Min}
   defaultValue=<rgba 1 1 0 1>
  }
 ]}
