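# Material function for wraps that blends animated, flipbook-capable decals onto an incoming material, with safe-zone inputs for placement.
# Object-type masks select offset, rotation and scale for weapons, vehicles, mechs, characters, backpacks and pickaxes.
# MF_MultiDecal_Layer combines these with sine motion, texture UV warp and random tiling, then feeds base color, PBR, normal and emissive.
# Time-stamp parameters and color-curve atlases drive reactive animation, while static switches gate optional features.
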MaterialFunction /Game/Weapons/WeaponSkins/Wraps/MaterialFunctions/MF_Effect_MultiDecal_V2
graph {
 properties {description=SafeZ}
 nodes [
  n1 = TextureSample() {texture=&T_BlacksRGB}
  n2 = MaterialFunctionCall(Float4=FunctionInput(Preview=@n1.RGBA) {bUsePreviewValueAsDefault=true, inputName=CM_Msk, inputType=FunctionInput_Vector4, sortPriority=4}) {materialFunction=&BreakOutFloat4Components}
  Add(A=NamedRerouteDeclaration(None=@n2.R) {name=Msk_Prime, nodeColor=<rgba 0 1 0.7422 1>})
  Constant3Vector()
  n3 = BlendMaterialAttributes(A=NamedRerouteUsage(), B=NamedRerouteUsage(), Alpha=Saturate(None=@n2.A)) {vertexAttributeBlendType=UseA}
  FunctionOutput(None=@n3) {outputName=Result_1}
  n4 = ScalarParameter() {group=Effect_MultDecal_V02 UV Warp | DecalA, parameterName=EMDv2_WarpA_Strength}
  n5 = ScalarParameter() {group=Effect_MultDecal_V02 UV Warp | DecalA, parameterName=EMDv2_WarpA_Sine_Speed}
  n6 = Sine(None=Multiply(A=Time(), B=@n5))
  LinearInterpolate(A=@n4, B=Multiply(A=Add(A=@n6), B=@n4) {constB=0.25}, Alpha=Saturate(None=Ceil(None=@n5)))
  n7 = VectorParameter() {
   group=Effect_MultDecal_V02 Uv Offset| Decal_A
   parameterName=EMDv2_DecalA_Offset_Weapon
   channelNames {a=Scale, b="", g=Y Offset, r=X Offset}
   defaultValue=<rgba 0 0 0 0.25>
  }
  n8 = VectorParameter() {
   group=Effect_MultDecal_V02 Uv Offset| Decal_A
   parameterName=EMDv2_DecalA_Offset_Vehicle
   channelNames {a=Scale, b="", g=Y Offset, r=X Offset}
   defaultValue=<rgba 0 0 0 0.25>
  }
  n9 = VectorParameter() {
   group=Effect_MultDecal_V02 Uv Offset| Decal_A
   parameterName=EMDv2_DecalA_Offset_VehicleSide
   channelNames {a=Scale, b="", g=Y Offset, r=X Offset}
   defaultValue=<rgba 0 0 0 0.25>
  }
  n10 = MaterialFunctionCall(Float4=NamedRerouteUsage()) {materialFunction=&BreakOutFloat4Components}
  n11 = LinearInterpolate(
   A=AppendVector(A=@n7.RGB, B=@n7.RGB)
   B=LinearInterpolate(A=AppendVector(A=@n8.RGB, B=@n8.RGB), B=AppendVector(A=@n9.RGB, B=@n9.RGB), Alpha=@n10.G)
   Alpha=@n10.R
  )
  n12 = VectorParameter() {
   group=Effect_MultDecal_V02 Uv Offset| Decal_A
   parameterName=EMDv2_DecalA_Offset_Mech
   channelNames {a=Scale, b="", g=Y Offset, r=X Offset}
   defaultValue=<rgba 0 0 0 0.25>
  }
  n13 = VectorParameter() {
   group=Effect_MultDecal_V02 Uv Offset| Decal_A
   parameterName=EMDv2_DecalA_Offset_Character
   channelNames {a=Scale, b="", g=Y Offset, r=X Offset}
   defaultValue=<rgba 0 0 0 0.25>
  }
  n14 = LinearInterpolate(A=AppendVector(A=@n12.RGB, B=@n12.RGB), B=AppendVector(A=@n13.RGB, B=@n13.RGB), Alpha=@n10.A)
  n15 = VectorParameter() {
   group=Effect_MultDecal_V02 Uv Offset| Decal_A
   parameterName=EMDv2_DecalA_Offset_Backpack
   channelNames {a=Scale, b="", g=Y Offset, r=X Offset}
   defaultValue=<rgba 0 0 0 0.25>
  }
  n16 = MaterialFunctionCall(Float2=NamedRerouteUsage()) {materialFunction=&BreakOutFloat2Components}
  n17 = VectorParameter() {
   group=Effect_MultDecal_V02 Uv Offset| Decal_A
   parameterName=EMDv2_DecalA_Offset_Pickaxe
   channelNames {a=Scale, b="", g=Y Offset, r=X Offset}
   defaultValue=<rgba 0 0 0 0.25>
  }
  n18 = LinearInterpolate(
   A=LinearInterpolate(A=LinearInterpolate(A=@n11, B=@n14, Alpha=@n10.B), B=AppendVector(A=@n15.RGB, B=@n15.RGB), Alpha=@n16.R)
   B=AppendVector(A=@n17.RGB, B=@n17.RGB)
   Alpha=@n16.G
  )
  NamedRerouteDeclaration(None=@n18) {name=UVs_OffsetA, nodeColor=<rgba 0 0.625 1 1>}
  NamedRerouteDeclaration(None=@n2.G) {name=Msk_Curvature, nodeColor=<rgba 0 1 0.7422 1>}
  n19 = TextureSample() {texture=/Game/Weapons/FORT_Rifles/Textures/SCAR_D}
  n20 = Constant3Vector() {constant=<rgba 0 0.5 0.5 0>}
  n21 = SetMaterialAttributes!(
   MaterialAttributes
   Normal=TextureSample() {samplerType=SAMPLERTYPE_Normal, texture=/Game/Weapons/FORT_Rifles/T_SCAR_N}.RGB
   BaseColor=@n19.RGB
   Metallic=@n20.R
   Specular=@n20.R
   Roughness=@n20.R
  ) {attributeSetTypes [
    #1
    #2
    #3
    #4
    #5
   ]}
  n22 = GetMaterialAttributes(None=FunctionInput(Preview=@n21) {bUsePreviewValueAsDefault=true, inputName=Material, inputType=FunctionInput_MaterialAttributes}) {attributeGetTypes [
    #2
    #5
    #1
    #3
    #6
   ]}
  NamedRerouteDeclaration(None=@n22.MaterialAttributes) {name=Material, nodeColor=<rgba 0.7812 1 0 1>}
  n23 = VectorParameter() {
   group=Effect_MultDecal_V02 Uv Offset| Decal_A
   parameterName=EMDv2_DecalA Rotation_Pivot offset
   channelNames {a="", b="", g=Y Offset, r=X Offset}
   defaultValue=<rgba 0 0 0 1>
  }
  n24 = VectorParameter() {
   group=Effect_MultDecal_V02 Uv Offset| Decal_A
   parameterName="EMDv2_DecalA Clamp---->Tile\\Speed"
   channelNames {a=y-Speed, b=X-Speed, g=Y Axies, r=X Axies}
  }
  n25 = VectorParameter() {
   group=Effect_MultDecal_V02 Uv Offset| Decal_A
   parameterName=EMDv2_DecalA_NonUnformScale
   channelNames {a="", b="", g=Y Scale, r=X Scale}
   defaultValue=<rgba 1 1 1 1>
  }
  n26 = VectorParameter() {
   group="Effect_MultDecal_V02  FlipBook| DecalA"
   parameterName=EMDv2_DecalA_Frame_Speed--
   channelNames {a=Clamp MAx, b=Clamp Min, g="Random 1=Active", r=Speed}
   defaultValue=<rgba 0 0 0 1>
  }
  n27 = MaterialFunctionCall!(
   SafeZone--OverScale=NamedRerouteUsage()
   SafeZone--Mid=NamedRerouteUsage()
   SafeZone--Front=NamedRerouteUsage()
   SafeZone--Back=NamedRerouteUsage()
   SafeZone---Back=StaticBoolParameter() {group=Effect_MultDecal_V02 Uv Offset | Decal_A | SafeZone, parameterName=EMDv2_Decal_A----SafeZone-Back}
   SafeZone---Mid=StaticBoolParameter() {group=Effect_MultDecal_V02 Uv Offset | Decal_A | SafeZone, parameterName=EMDv2_Decal_A----SafeZone-Middle}
   SafeZone---Front=StaticBoolParameter() {group=Effect_MultDecal_V02 Uv Offset | Decal_A | SafeZone, parameterName=EMDv2_Decal_A-----SafeZone-Front}
   Uv Offset/Scale=NamedRerouteUsage()
   Sine Animation=NamedRerouteUsage()
   Rotation_Pivot offset=@n23.RGB
   "Uv_Calmp----->Tile\\Speed"=@n24.RGBA
   Uv_Calmp pxiel Comp=Add(B=ScalarParameter() {group=Effect_MultDecal_V02 Uv Offset| Decal_A, parameterName=EMDv2_DecalA Clamp Pixel Comp}) {constA=1}
   Rotation_Mirroring=ScalarParameter() {group=Effect_MultDecal_V02 Uv Offset| Decal_A, parameterName=EMDv2_DecalA_Rotation-Mirror-------}
   NonUniform Scale=@n25.RGB
   Rotation Bounce=NamedRerouteUsage()
   L/R State=StaticBoolParameter() {group=Effect_MultDecal_V02 Uv Offset| Decal_A, parameterName="EMDv2-DecalA--Side---->LR"}
   Flipbook Frame=ScalarParameter() {group="Effect_MultDecal_V02  FlipBook| DecalA", parameterName="EMDv2_DecalA_Frame ####"}
   Flipbook ColorCurve Animation=NamedRerouteUsage()
   "FlipBook Colums\\Rows"=NamedRerouteUsage()
   Flipbook ColorCurve State=NamedRerouteUsage()
   Flipbook Reactive Anim=NamedRerouteUsage()
   Flipbook Offset=ScalarParameter() {group="Effect_MultDecal_V02  FlipBook| DecalA", parameterName=EMDv2_DecalA_Offset}
   FlipBook Frame Speed=@n26.RGBA
   Flipbook Reactive State=NamedRerouteUsage()
   Tileable State=StaticBoolParameter() {group=Effect_MultDecal_V02 Uv Rnd Tile, parameterName=EMDv2-DecalA--Rnd-Tile}
   Tilable SysTem=NamedRerouteUsage()
   Tileable Rotation=NamedRerouteUsage()
   Tileable Mask=NamedRerouteUsage()
   Tileable Frame=NamedRerouteUsage()
   Warp Uvs=NamedRerouteUsage()
   warp UVs State=StaticBoolParameter() {group=Effect_MultDecal_V02 UV Warp | DecalA, parameterName=EMDv2-DecalA--Warp}
   warp Audio State
   Emissive only State
   Emissive
   Z-input-isState=NamedRerouteUsage()
   Z-input--PreSkin=NamedRerouteUsage()
   Z-input--CalculateBounds=NamedRerouteUsage()
   Z-Input-BC-Tex=NamedRerouteUsage()
   Z-Input-N-Tex=NamedRerouteUsage()
   Z-Input-Msk-Tex=NamedRerouteUsage()
  ) {materialFunction=&/MF_MultiDecal_Layer}
  NamedRerouteDeclaration(None=@n27.Uvs) {name=Uvs, nodeColor=<rgba 1 0.1875 0 1>}
  n28 = TextureObjectParameter() {
   group=Effect_MultDecal_V02 Emissive
   parameterName=EMDv2_Grad__Tex
   sortPriority=10
   texture=&T_BlacksRGB
  }
  NamedRerouteDeclaration(None=@n28) {name=Tex_Grad, nodeColor=<rgba 0 1 0.9062 1>}
  n29 = VectorParameter() {
   group="Effect_MultDecal_V02  FlipBook"
   parameterName=EMDv2--FlipBook Frames---Count
   channelNames {a="", b="", g=Row Amount, r=Columes Amount}
   defaultValue=<rgba 2 2 0 1>
  }
  NamedRerouteDeclaration(None=@n29.RGB) {name=FlipBrook Frames, nodeColor=<rgba 0.9688 1 0 1>}
  n30 = SetMaterialAttributes(Normal=Constant3Vector() {constant=<rgba 0 0 1 0>}, BaseColor=@n19.RGB, Metallic=@n20.R, Specular=@n20.R, Roughness=@n20.R) {attributeSetTypes [
    #1
    #2
    #3
    #4
    #5
   ]}
  n31 = GetMaterialAttributes(None=FunctionInput(Preview=@n30) {bUsePreviewValueAsDefault=true, inputName=Material_Root, inputType=FunctionInput_MaterialAttributes}) {attributeGetTypes [
    #2
    #5
    #1
    #3
    #6
   ]}
  NamedRerouteDeclaration(None=@n31.Normal) {name=Root_Normal, nodeColor=<rgba 0.4062 0 1 1>}
  NamedRerouteDeclaration(None=@n31.MaterialAttributes) {name=Root_Material, nodeColor=<rgba 0.7812 1 0 1>}
  n32 = LinearInterpolate(
   A=NamedRerouteUsage()
   B=NamedRerouteUsage()
   Alpha=ScalarParameter() {group=Effect_MultDecal_V02 UV Warp | DecalA, parameterName="EMDv2-WarpA-DecalSpace-->ObjectBounds"}
  )
  n33 = VectorParameter() {
   group=Effect_MultDecal_V02 UV Warp | DecalA
   parameterName=EMDv2_WarpA__Tile/Speed
   channelNames {a=Speed-y, b=speed-X, g=Tile-Y, r=Tile-X}
   defaultValue=<rgba 1 1 0 0>
  }
  n34 = AppendVector(A=@n33.Tile-X, B=@n33.Tile-X)
  n35 = AppendVector(A=@n33.speed-X, B=@n33.speed-X)
  n36 = Time()
  n37 = TextureSample(UVs=Add(A=Multiply(A=@n32, B=@n34), B=Multiply(A=@n35, B=@n36)), Tex=NamedRerouteUsage())
  NamedRerouteDeclaration(None=@n37.RGBA) {name=Msk_A_Warp, nodeColor=<rgba 0.1484 0 1 1>}
  n38 = VectorParameter() {
   group=Effect_MultDecal_V02 Uv Rnd Tile | DecalA | Objects
   parameterName="EMDv2_DecalA_TileRate\\Scale_Weapon"
   sortPriority=500
   channelNames {a="", b="", g=Scale, r=TileRate}
   defaultValue=<rgba 5 0.75 0 0>
  }
  n39 = VectorParameter() {
   group=Effect_MultDecal_V02 Uv Rnd Tile | DecalA | Objects
   parameterName="EMDv2_DecalA_TileRate\\Scale_Vehicle"
   sortPriority=500
   channelNames {a="", b="", g=Scale, r=TileRate}
   defaultValue=<rgba 5 0.75 0 0>
  }
  n40 = VectorParameter() {
   group=Effect_MultDecal_V02 Uv Rnd Tile | DecalA | Objects
   parameterName="EMDv2_DecalA_TileRate\\Scale_VehicleSide"
   sortPriority=500
   channelNames {a="", b="", g=Scale, r=TileRate}
   defaultValue=<rgba 5 0.75 0 0>
  }
  n41 = MaterialFunctionCall(Float4=NamedRerouteUsage()) {materialFunction=&BreakOutFloat4Components}
  n42 = VectorParameter() {
   group=Effect_MultDecal_V02 Uv Rnd Tile | DecalA | Objects
   parameterName="EMDv2_DecalA_TileRate\\Scale_Mech"
   sortPriority=500
   channelNames {a="", b="", g=Scale, r=TileRate}
   defaultValue=<rgba 5 0.75 0 0>
  }
  n43 = VectorParameter() {
   group=Effect_MultDecal_V02 Uv Rnd Tile | DecalA | Objects
   parameterName="EMDv2_DecalA_TileRate\\Scale_Character"
   sortPriority=500
   channelNames {a="", b="", g=Scale, r=TileRate}
   defaultValue=<rgba 5 0.75 0 0>
  }
  n44 = LinearInterpolate(
   A=LinearInterpolate(A=@n38, B=LinearInterpolate(A=@n39, B=@n40, Alpha=@n41.G), Alpha=@n41.R)
   B=LinearInterpolate(A=@n42, B=@n43, Alpha=@n41.A)
   Alpha=@n41.B
  )
  n45 = VectorParameter() {
   group=Effect_MultDecal_V02 Uv Rnd Tile | DecalA | Objects
   parameterName="EMDv2_DecalA_TileRate\\Scale_BackPack"
   sortPriority=500
   channelNames {a="", b="", g=Scale, r=TileRate}
   defaultValue=<rgba 5 0.75 0 0>
  }
  n46 = MaterialFunctionCall(Float2=NamedRerouteUsage()) {materialFunction=&BreakOutFloat2Components}
  n47 = VectorParameter() {
   group=Effect_MultDecal_V02 Uv Rnd Tile | DecalA | Objects
   parameterName="EMDv2_DecalA_TileRate\\Scale_Pickaxe"
   sortPriority=500
   channelNames {a="", b="", g=Scale, r=TileRate}
   defaultValue=<rgba 5 0.75 0 0>
  }
  NamedRerouteDeclaration(None=LinearInterpolate(A=LinearInterpolate(A=@n44, B=@n45, Alpha=@n46.R), B=@n47, Alpha=@n46.G)) {name=DecalA Rnd Tile, nodeColor=<rgba 0.1484 1 0 1>}
  NamedRerouteDeclaration(None=FunctionInput(Preview=Constant4Vector() {constant=<rgba 0.2 0.2 0 0.5>}.RGBA) {bUsePreviewValueAsDefault=true, inputName=Object-SafeZone-Back, inputType=FunctionInput_Vector4, sortPriority=100}) {name=SafeZone_Back, nodeColor=<rgba 0 0.9297 1 1>}
  n48 = NamedRerouteUsage()
  n49 = ChannelMaskParameter(None=@n37.RGBA) {group=Effect_MultDecal_V02 UV Warp | DecalA, parameterName=EMDv2-WarpA-Msk_Channel}
  n50 = ScalarParameter() {group=Effect_MultDecal_V02 UV Warp | DecalA, parameterName=EMDv2_WarpA_Offset}
  n51 = NamedRerouteUsage()
  n52 = NamedRerouteUsage()
  n53 = StaticSwitch(
   True=Multiply(A=NamedRerouteUsage(), B=ComponentMask(None=@n48) {r=true})
   False=Multiply(A=Add(A=@n49.R, B=@n50), B=StaticSwitch(True=@n51, False=@n4, Value=@n52))
   Value=NamedRerouteUsage()
  )
  n54 = NamedRerouteUsage()
  n55 = VectorParameter() {
   group=Effect_MultDecal_V02 UV Warp | DecalA
   parameterName=EMDv2-WarpA_Axis Mask
   channelNames {a="", b="", g=Y mask, r=X mask}
   defaultValue=<rgba 1 1 0 1>
  }
  NamedRerouteDeclaration(None=Multiply(A=Multiply(A=@n53, B=@n54), B=AppendVector(A=@n55.X mask, B=@n55.X mask))) {name=Warp A, nodeColor=<rgba 1 0 0.1875 1>}
  n56 = VectorParameter() {
   group=Effect_MultDecal_V02 UV Warp
   parameterName=EMDv2_Warp_Speed
   channelNames {a="", b="", g=Y Speed, r=X Speed}
   defaultValue=<rgba 0 0 0 1>
  }
  n57 = Time()
  n58 = StaticBoolParameter() {defaultValue=true, group=Effect_MultDecal_V02 UV Warp | Rotation, parameterName=EMDv2_Warp_Rotation Disabled, sortPriority=1}
  n59 = TextureSample(
   UVs=MaterialFunctionCall!(
    Normal_Map
    Offset Rt
    Pivot
    Caculated_Bounds=NamedRerouteUsage()
    Pre_Skinned_Normals
    is_State=NamedRerouteUsage()
    UV_Rotation=NamedRerouteUsage()
    Uvs_Offset=Multiply(A=@n56.RGB, B=@n57)
    Uvs_TileRate=NamedRerouteUsage()
    Uvs_Offset_Pivot
    For_Safe_Zones
    Rotation_Comp
    Clamp Y
    Clamp X
    Clamp Padding_Comp
    No_Rotation=@n58
   ) {materialFunction=&/MF_Projection_Side_Normal_V2}.UVs
   Tex=NamedRerouteUsage()
  )
  NamedRerouteDeclaration(None=@n59.RGBA) {name=Warp_Uvs_RGBA, nodeColor=<rgba 0.6406 1 0 1>}
  NamedRerouteDeclaration(None=StaticBoolParameter() {group=Effect_MultDecal_V02, parameterName=EMDv2_Normals}) {name=State-Normal, nodeColor=<rgba 0.3359 1 0 1>}
  NamedRerouteDeclaration(None=@n27."Uvs_L\\R") {name=Uvs_A_LR, nodeColor=<rgba 1 0.1875 0 1>}
  n60 = TextureObjectParameter() {
   group=Effect_MultDecal_V02
   parameterName=EMDv2__Counter__Tex
   texture=/Engine/Functions/Engine_MaterialFunctions02/ExampleContent/DebugNumberStrip
  }
  NamedRerouteDeclaration(None=@n60) {name=Tex_Counter, nodeColor=<rgba 0 0.8125 1 1>}
  n61 = VectorParameter() {
   group=Effect_MultDecal_V02 | PBR | DecalA
   parameterName=EMDv2_DecalA_BaseColor--Emissive
   channelNames {a="", b=Darken Root Emissive, g="BaseColor----->Emissive", r="BaseColor---->Black"}
   defaultValue=<rgba 0 0 1 0>
  }
  n62 = VectorParameter() {
   group=Effect_MultDecal_V02 | PBR | DecalA
   parameterName=EMDv2_DecalA_PBR_Roughness
   channelNames {a="", b=0, g=Max, r=Min}
   defaultValue=<rgba 0 1 0 1>
  }
  n63 = ChannelMaskParameter(None=@n27.Tex---Msk) {group=Effect_MultDecal_V02, parameterName=EMDv2_Use Msk--Tex--Roughness Channle}
  n64 = VectorParameter() {
   group=Effect_MultDecal_V02 | PBR | DecalA
   parameterName=EMDv2_DecalA_PBR
   channelNames {a="", b=Metalic, g=Specular, r=Roughness}
   defaultValue=<rgba 0.5 0.5 0 1>
  }
  n65 = VectorParameter() {
   group=Effect_MultDecal_V02 | PBR | DecalA
   parameterName=EMDv2_DecalA_PBR_Spec
   channelNames {a="", b=0, g=Max, r=Min}
   defaultValue=<rgba 0.5 1 0 1>
  }
  n66 = ChannelMaskParameter(None=@n27.Tex---Msk) {group=Effect_MultDecal_V02, parameterName=EMDv2_Use Msk--Tex--Spec Channle}
  n67 = VectorParameter() {
   group=Effect_MultDecal_V02 | PBR | DecalA
   parameterName=EMDv2_DecalA_PBR_Metalic
   channelNames {a="", b=0, g=Max, r=Min}
   defaultValue=<rgba 0 1 0 1>
  }
  n68 = ChannelMaskParameter(None=@n27.Tex---Msk) {group=Effect_MultDecal_V02, parameterName=EMDv2_Use Msk--Tex--Metalic Channle}
  n69 = LinearInterpolate(
   A=NamedRerouteUsage()
   B=NamedRerouteUsage()
   Alpha=ScalarParameter() {group=Effect_MultDecal_V02 | PBR | DecalA, parameterName="EMDv2_DecalA_Normals ------>Root"}
  )
  n70 = LinearInterpolate(
   A=@n69
   B=Add(A=@n27.Tex--Normals, B=@n69)
   Alpha=ScalarParameter() {defaultValue=1, group=Effect_MultDecal_V02 | PBR | DecalA, parameterName=EMDv2_DecalA_Normal Strength}
  )
  n71 = ShadingPathSwitch(Default=@n70, Deferred=@n70, Mobile=@n69) {desc=Migrated from FeatureLevelSwitch}
  n72 = NamedRerouteUsage()
  n73 = LinearInterpolate(A=NamedRerouteUsage(), Alpha=@n61.Darken Root Emissive) {constB=0}
  n74 = SetMaterialAttributes!(
   MaterialAttributes=NamedRerouteUsage()
   BaseColor=LinearInterpolate(A=@n27.Tex---BC, Alpha=@n61."BaseColor---->Black") {constB=0}
   Roughness=StaticSwitch(True=LinearInterpolate(A=@n62.Min, B=@n62.Min, Alpha=@n63.R), False=@n64.Roughness, Value=NamedRerouteUsage())
   Specular=StaticSwitch(True=LinearInterpolate(A=@n65.Min, B=@n65.Min, Alpha=@n66.R), False=@n64.Specular, Value=NamedRerouteUsage())
   Metallic=StaticSwitch(True=LinearInterpolate(A=@n67.Min, B=@n67.Min, Alpha=@n68.R), False=@n64.Metalic, Value=NamedRerouteUsage())
   Normal=StaticSwitch(True=StaticSwitch(True=@n71, False=@n70, Value=@n72), False=@n69, Value=NamedRerouteUsage())
   EmissiveColor=LinearInterpolate(A=@n73, B=Add(A=@n27.Tex---BC, B=@n73), Alpha=@n61."BaseColor----->Emissive") {constB=0}
  ) {attributeSetTypes [
    #2
    #5
    #4
    #3
    #1
    #6
   ]}
  NamedRerouteDeclaration(None=BlendMaterialAttributes(A=NamedRerouteUsage(), B=@n74, Alpha=NamedRerouteUsage()) {vertexAttributeBlendType=UseA}) {name=Decal A, nodeColor=<rgba 0 0.3906 1 1>}
  n75 = VectorParameter() {
   group="Effect_MultDecal_V02  FlipBook| DecalA | Reactive"
   parameterName=EMDv2_DecalA_Animation--
   channelNames {a=Clamp Max, b=Clamp Min Frames, g="", r=Loop}
   defaultValue=<rgba 1 0 0 1>
  }
  n76 = Time()
  n77 = Add(
   A=ScalarParameter() {defaultValue=-9999999, group="Effect_MultDecal_V02  FlipBook| DecalA | Reactive", parameterName=EMDv2_DecalA_Animation_Time_BP}
   B=ScalarParameter() {group="Effect_MultDecal_V02  FlipBook| DecalA | Reactive", parameterName=EMDv2_DecalA_Animation_Delay}
  )
  n78 = Saturate(None=Divide(
    A=Subtract(A=@n76, B=@n77)
    B=ScalarParameter() {defaultValue=1, group="Effect_MultDecal_V02  FlipBook| DecalA | Reactive", parameterName=EMDv2_DecalA_Animation_Duration}
   ))
  n79 = ScalarParameter() {defaultValue=1, group="Effect_MultDecal_V02  FlipBook| DecalA | Reactive", parameterName=EMDv2_DecalA_Animation_Falloff}
  n80 = Multiply(A=Multiply(A=Power(Base=@n78, Exp=@n79)) {constB=1}, B=@n75.Loop)
  NamedRerouteDeclaration(None=LinearInterpolate(A=@n75.Clamp Min Frames, B=@n75.Clamp Min Frames, Alpha=Frac(None=@n80))) {name=FlipBook_React_AnimA, nodeColor=<rgba 0 0.5078 1 1>}
  NamedRerouteDeclaration(None=StaticBoolParameter() {defaultValue=true, group=Effect_MultDecal_V02, parameterName=EMDv2_Normals Android Disable}) {name=State-Normal_Android, nodeColor=<rgba 0.3359 1 0 1>}
  NamedRerouteDeclaration(None=StaticBoolParameter() {group="Effect_MultDecal_V02  FlipBook| DecalA | Reactive", parameterName=EMDv2_DecalA_Animation_Reactive, sortPriority=0}) {name=FlipBook_React_StateA, nodeColor=<rgba 0.7344 1 0 1>}
  n81 = VectorParameter() {
   group=Effect_MultDecal_V02 Uv Animation| Sine | DecalA | Objects
   parameterName=EMDv2_DecalA_Sine_Range_Weapon
   sortPriority=500
   channelNames {a="", b="", g=Y Range, r=X Range}
   defaultValue=<rgba 0.1 0.1 0 0>
  }
  n82 = VectorParameter() {
   group=Effect_MultDecal_V02 Uv Animation| Sine | DecalA | Objects
   parameterName=EMDv2_DecalA_Sine_Range_Vehicle
   sortPriority=500
   channelNames {a="", b="", g=Y Range, r=X Range}
   defaultValue=<rgba 0.1 0.1 0 0>
  }
  n83 = VectorParameter() {
   group=Effect_MultDecal_V02 Uv Animation| Sine | DecalA | Objects
   parameterName=EMDv2_DecalA_Sine_Range_VehicleSide
   sortPriority=500
   channelNames {a="", b="", g=Y Range, r=X Range}
   defaultValue=<rgba 0.1 0.1 0 0>
  }
  n84 = MaterialFunctionCall(Float4=NamedRerouteUsage()) {materialFunction=&BreakOutFloat4Components}
  n85 = LinearInterpolate(A=@n81, B=LinearInterpolate(A=@n82, B=@n83, Alpha=@n84.G), Alpha=@n84.R)
  n86 = VectorParameter() {
   group=Effect_MultDecal_V02 Uv Animation| Sine | DecalA | Objects
   parameterName=EMDv2_DecalA_Sine_Range_Mech
   sortPriority=500
   channelNames {a="", b="", g=Y Range, r=X Range}
   defaultValue=<rgba 0.1 0.1 0 0>
  }
  n87 = VectorParameter() {
   group=Effect_MultDecal_V02 Uv Animation| Sine | DecalA | Objects
   parameterName=EMDv2_DecalA_Sine_Range_Character
   sortPriority=500
   channelNames {a="", b="", g=Y Range, r=X Range}
   defaultValue=<rgba 0.1 0.1 0 0>
  }
  n88 = LinearInterpolate(A=@n86, B=@n87, Alpha=@n84.A)
  n89 = VectorParameter() {
   group=Effect_MultDecal_V02 Uv Animation| Sine | DecalA | Objects
   parameterName=EMDv2_DecalA_Sine_Range_BackPack
   sortPriority=500
   channelNames {a="", b="", g=Y Range, r=X Range}
   defaultValue=<rgba 0.1 0.1 0 0>
  }
  n90 = MaterialFunctionCall(Float2=NamedRerouteUsage()) {materialFunction=&BreakOutFloat2Components}
  n91 = VectorParameter() {
   group=Effect_MultDecal_V02 Uv Animation| Sine | DecalA | Objects
   parameterName=EMDv2_DecalA_Sine_Range_Pickaxe
   sortPriority=500
   channelNames {a="", b="", g=Y Range, r=X Range}
   defaultValue=<rgba 0.1 0.1 0 0>
  }
  MaterialExpressionLinearInterpolate_286 = LinearInterpolate(A=LinearInterpolate(A=LinearInterpolate(A=@n85, B=@n88, Alpha=@n84.B), B=@n89, Alpha=@n90.R), B=@n91, Alpha=@n90.G)
  n92 = Convert(None=@MaterialExpressionLinearInterpolate_286) {
   nodeName=Break Float3
   convertInputs [{
     type=Vector3
     expressionInput {
      expression=&self:MaterialExpressionLinearInterpolate_286
      inputName=None
      mask=0
      maskA=0
      maskB=0
      maskG=0
      maskR=0
      outputIndex=0
     }
    }]
   convertOutputs [{type=Scalar}, {type=Scalar}, {type=Scalar}]
  }
  n93 = Time()
  n94 = VectorParameter() {
   group=Effect_MultDecal_V02 Uv Animation| Sine | DecalA
   parameterName=EMDv2_DecalA_Sine-Speed
   sortPriority=500
   channelNames {a=Y Offset, b=X offset, g=Y Speed, r=X Speed}
  }
  n95 = AppendVector(A=@n94.X Speed, B=@n94.X Speed)
  MaterialExpressionSine_13 = Sine(None=Add(A=Multiply(A=@n93, B=@n95), B=AppendVector(A=@n94.X offset, B=@n94.X offset)))
  n96 = Convert(None=@MaterialExpressionSine_13) {
   nodeName=Break Float2
   convertInputs [{
     type=Vector2
     expressionInput {
      expression=&self:MaterialExpressionSine_13
      inputName=None
      mask=0
      maskA=0
      maskB=0
      maskG=0
      maskR=0
      outputIndex=0
     }
    }]
   convertOutputs [{type=Scalar}, {type=Scalar}]
  }
  n97 = Multiply(A=@n92, B=@n96) {constA=0.1}
  n98 = Multiply(A=@n92, B=@n96) {constA=0.1}
  n99 = Constant()
  n100 = ScalarParameter() {group=Effect_MultDecal_V02 Uv Animation| Sine | DecalA, parameterName="EMDv2_DecalA_Sine-Rotation---->Match Uv Direction", sortPriority=500}
  n101 = MaterialFunctionCall(R=@n97, G=@n98, B=@n99, A=@n100) {materialFunction=&AppendMany}
  n102 = MaterialFunctionCall(R=@n98, G=@n97, B=@n99, A=@n100) {materialFunction=&AppendMany}
  n103 = ScalarParameter() {
   defaultValue=1
   group=Effect_MultDecal_V02 Uv Animation| Sine | DecalA | Reactive
   parameterName=EMDv2_DecalA__Sine_Strength_Start
   sortPriority=100
  }
  n104 = ScalarParameter() {
   defaultValue=1
   group=Effect_MultDecal_V02 Uv Animation| Sine | DecalA | Reactive
   parameterName=EMDv2_DecalA__Sine_Strength_End
   sortPriority=100
  }
  n105 = ScalarParameter() {group=Effect_MultDecal_V02 Uv Animation| Sine | DecalA | Reactive, parameterName=EMDv2_DecalA__Sine_Delay-----------------, sortPriority=100}
  n106 = ScalarParameter() {
   defaultValue=-9999999
   group=Effect_MultDecal_V02 Uv Animation| Sine | DecalA | Reactive
   parameterName=EMDv2_DecalA__Sine_TimeFrom_BP
   sortPriority=100
  }
  n107 = Subtract(A=Time(), B=Add(A=@n105, B=@n106))
  n108 = ScalarParameter() {
   defaultValue=1
   group=Effect_MultDecal_V02 Uv Animation| Sine | DecalA | Reactive
   parameterName=EMDv2_DecalA__Sine_Duration
   sortPriority=100
  }
  n109 = ScalarParameter() {
   defaultValue=0.25
   group=Effect_MultDecal_V02 Uv Animation| Sine | DecalA | Reactive
   parameterName=EMDv2_DecalA__Sine_Falloff
   sortPriority=100
  }
  n110 = Power(Base=Saturate(None=Divide(A=@n107, B=@n108)), Exp=@n109)
  NamedRerouteDeclaration(None=Multiply(A=LinearInterpolate(A=@n101.RGBA, B=@n102.RGBA, Alpha=@n100), B=LinearInterpolate(A=@n103, B=@n104, Alpha=@n110))) {name=Decal_A_Sine, nodeColor=<rgba 0 0.625 1 1>}
  n111 = ScalarParameter() {group=Effect_MultDecal_V02 Uv Offset| Decal_A, parameterName=EMDv2_DecalA_Rotation_Weapon}
  n112 = MaterialFunctionCall(Float4=NamedRerouteUsage()) {materialFunction=&BreakOutFloat4Components}
  n113 = LinearInterpolate(
   A=ScalarParameter() {group=Effect_MultDecal_V02 Uv Offset| Decal_A, parameterName=EMDv2_DecalA_Rotation_Vehicle}
   B=ScalarParameter() {group=Effect_MultDecal_V02 Uv Offset| Decal_A, parameterName=EMDv2_DecalA_Rotation_Vehicle_Side}
   Alpha=@n112.G
  )
  n114 = ScalarParameter() {group=Effect_MultDecal_V02 Uv Offset| Decal_A, parameterName=EMDv2_DecalA_Rotation_Mech}
  n115 = ScalarParameter() {group=Effect_MultDecal_V02 Uv Offset| Decal_A, parameterName=EMDv2_DecalA_Rotation_Character}
  n116 = MaterialFunctionCall(Float2=NamedRerouteUsage()) {materialFunction=&BreakOutFloat2Components}
  n117 = LinearInterpolate(
   A=LinearInterpolate(A=LinearInterpolate(A=@n111, B=@n113, Alpha=@n112.R), B=LinearInterpolate(A=@n114, B=@n115, Alpha=@n112.A), Alpha=@n112.B)
   B=ScalarParameter() {group=Effect_MultDecal_V02 Uv Offset| Decal_A, parameterName=EMDv2_DecalA_Rotation_Backpack}
   Alpha=@n116.R
  )
  n118 = ScalarParameter() {group=Effect_MultDecal_V02 Uv Offset| Decal_A, parameterName=EMDv2_DecalA_Rotation_Pickaxe}
  n119 = ScalarParameter() {group=Effect_MultDecal_V02 Uv Offset| Decal_A | Reactive | Rotation, parameterName=EMDv2_DecalA_Rotation_Amount_Start, sortPriority=5}
  n120 = ScalarParameter() {group=Effect_MultDecal_V02 Uv Offset| Decal_A | Reactive | Rotation, parameterName=EMDv2_DecaA_Rotationy Delay, sortPriority=0}
  n121 = ScalarParameter() {
   defaultValue=-9999999
   group=Effect_MultDecal_V02 Uv Offset| Decal_A | Reactive | Rotation
   parameterName=EMDv2_DecaA_Rotation_BP_DecalA
   sortPriority=0
  }
  n122 = ScalarParameter() {
   defaultValue=0.25
   group=Effect_MultDecal_V02 Uv Offset| Decal_A | Reactive | Rotation
   parameterName=EMDv2_DecaA_Rotation-Duration
   sortPriority=5
  }
  n123 = Divide(A=Subtract(A=Time(), B=Add(A=@n120, B=@n121)), B=@n122)
  n124 = ScalarParameter() {
   defaultValue=1
   group=Effect_MultDecal_V02 Uv Offset| Decal_A | Reactive | Rotation
   parameterName=EMDv2_DecalA_Rotation-Falloff
   sortPriority=5
  }
  n125 = LinearInterpolate(A=@n119, Alpha=Power(Base=Saturate(None=@n123), Exp=@n124)) {constB=0}
  NamedRerouteDeclaration(None=Add(A=LinearInterpolate(A=@n117, B=@n118, Alpha=@n116.G), B=@n125) {constB=0}) {name=Uvs_RotationA, nodeColor=<rgba 0 0.7188 1 1>}
  NamedRerouteDeclaration(None=FunctionInput(Preview=Constant4Vector() {constant=<rgba 0 0.2 0 0.5>}.RGBA) {bUsePreviewValueAsDefault=true, inputName=Object-SafeZone-Middle, inputType=FunctionInput_Vector4, sortPriority=101}) {name=SafeZone_Middle, nodeColor=<rgba 0 0.9297 1 1>}
  n126 = ScalarParameter() {group=Effect_MultDecal_V02 Uv Animation| Rotation | DecalA, parameterName="EMDv2_DecalA_Rotation_Speed  --------------", sortPriority=0}
  n127 = Multiply(
   A=Add(
    A=ScalarParameter() {group=Effect_MultDecal_V02 Uv Animation| Rotation | DecalA, parameterName=EMDv2_DecalA_Rotation_Speed Offset, sortPriority=0}
    B=Time()
   )
   B=@n126
  )
  n128 = ScalarParameter() {
   defaultValue=-9999999
   group=Effect_MultDecal_V02 Uv Animation| Rotation | DecalA | Reactive
   parameterName=EMDv2_DecalA_BounceRot_TimeFrom_BP
  }
  n129 = Subtract(
   A=Time()
   B=Add(
    A=ScalarParameter() {group=Effect_MultDecal_V02 Uv Animation| Rotation | DecalA | Reactive, parameterName=EMDv2_BounceRot_Delay-----------------}
    B=@n128
   )
  )
  n130 = ScalarParameter() {defaultValue=1, group=Effect_MultDecal_V02 Uv Animation| Rotation | DecalA | Reactive, parameterName=EMDv2_DecalA_BounceRot_Duration}
  n131 = ScalarParameter() {defaultValue=0.25, group=Effect_MultDecal_V02 Uv Animation| Rotation | DecalA | Reactive, parameterName=EMDv2_DecalA_BounceRot_Falloff}
  n132 = Power(Base=Saturate(None=Divide(A=@n129, B=@n130)), Exp=@n131)
  n133 = LinearInterpolate(
   A=ScalarParameter() {group=Effect_MultDecal_V02 Uv Animation| Rotation | DecalA | Reactive, parameterName=EMDv2_DecalA_BounceRot_Speed_Start}
   Alpha=@n132
  ) {constB=0}
  n134 = ScalarParameter() {defaultValue=0.1, group=Effect_MultDecal_V02 Uv Animation| Rotation | DecalA | Reactive, parameterName=EMDv2_DecalA_BounceRot_Strength_Start}
  n135 = ScalarParameter() {defaultValue=0.1, group=Effect_MultDecal_V02 Uv Animation| Rotation | DecalA, parameterName=EMDv2_DecalA_Rotation_Bounce Strength}
  n136 = Multiply(A=Sine(None=Add(A=@n127, B=@n133)), B=LinearInterpolate(A=@n134, B=@n135, Alpha=@n132))
  n137 = ScalarParameter() {
   group=Effect_MultDecal_V02 Uv Animation| Rotation | DecalA
   parameterName="EMDv2_DecalA_Rotation_Speed----->Bounce"
   sliderMax=1
   sortPriority=0
  }
  n138 = Saturate(None=@n137)
  NamedRerouteDeclaration(None=Add(A=NamedRerouteUsage(), B=LinearInterpolate(A=@n127, B=@n136, Alpha=@n138))) {name=Bounce Rotation A, nodeColor=<rgba 1 0.2578 0 1>}
  n139 = TextureObjectParameter() {group=Effect_MultDecal_V02, parameterName=EMDv2_Mask__Tex, texture=&T_BlacksRGB}
  NamedRerouteDeclaration(None=@n139) {name=Tex_Msk, nodeColor=<rgba 0 1 0.9062 1>}
  n140 = MaterialFunctionCall(Color=Abs(None=NamedRerouteUsage())) {materialFunction=&SplitComponents}
  n141 = MaterialFunctionCall(Color=NamedRerouteUsage()) {materialFunction=&SplitComponents}
  n142 = LinearInterpolate(A=LinearInterpolate(A=@n140.R, B=LinearInterpolate(A=@n140.B, B=@n140.B, Alpha=@n141.G), Alpha=@n141.R), B=@n140.G, Alpha=@n141.B)
  NamedRerouteDeclaration(None=@n142) {name=Tri_Mask_raw, nodeColor=<rgba 1 0 0.4922 1>}
  NamedRerouteDeclaration(None=@n27.UvsNoOffsets) {name=Uvs_A_No, nodeColor=<rgba 1 0.1875 0 1>}
  NamedRerouteDeclaration(None=FunctionInput(Preview=Constant4Vector() {constant=<rgba -0.25 0.2 0 0.5>}.RGBA) {bUsePreviewValueAsDefault=true, inputName=Object-SafeZone-Front, inputType=FunctionInput_Vector4, sortPriority=102}) {name=SafeZone_Front, nodeColor=<rgba 0 0.9297 1 1>}
  NamedRerouteDeclaration(None=StaticBoolParameter() {group=Effect_MultDecal_V02, parameterName=EMDv2_Use Msk--Tex--Roughness}) {name=State-Roughness, nodeColor=<rgba 0.3359 1 0 1>}
  NamedRerouteDeclaration(None=StaticBoolParameter() {group=Effect_MultDecal_V02, parameterName=EMDv2_Use Msk--Tex--Spec}) {name=State-Spec, nodeColor=<rgba 0.3359 1 0 1>}
  n143 = VectorParameter() {
   group=Effect_MultDecal_V02 UV Warp
   parameterName=EMDv2_WarpB_Msk_Decal_Strength
   channelNames {a=Decal D, b=Decal C, g=Decal B, r=Decal A}
   defaultValue=<rgba 1 1 1 1>
  }
  NamedRerouteDeclaration(None=@n143.RGBA) {name=Warp_Uvs_ST, nodeColor=<rgba 0.0547 0 1 1>}
  n144 = ScalarParameter() {
   desc=The total time of the flip animation before it restarts.
   group="Effect_MultDecal_V02  FlipBook| DecalA | CCTime"
   parameterName=EMDv2_Decal_A_CC_Time
  }
  n145 = Multiply(
   A=Subtract(A=Time(), B=ScalarParameter() {group="Effect_MultDecal_V02  FlipBook| DecalA | CCTime", parameterName=EMDv2_Decal_A_CC_Time_Bp})
   B=Divide(B=@n144) {constA=3}
  ) {constB=0.04}
  n146 = ScalarParameter() {
   desc=Reactive Trigger point in CC
   group="Effect_MultDecal_V02  FlipBook| DecalA | CCTime"
   parameterName=EMDv2_Decal_A_CC_Time_Bp_Offset_Trigger
  }
  n147 = Frac(None=Add(A=Frac(None=@n145), B=@n146))
  n148 = CurveAtlasRowParameter(CurveTime=@n147) {
   atlas=/Game/Weapons/WeaponSkins/Wraps/Textures/CruveAtlas_Wraps_01
   curve=/Game/Weapons/WeaponSkins/Wraps/Textures/0_ColorCurves/CC_DecalGl_Timing
   defaultValue=254
   group="Effect_MultDecal_V02  FlipBook| DecalA | CCTime"
   parameterName=EMDv2_Decal_A_CC
  }
  n149 = MaterialFunctionCall!(
   Index=ScalarParameter() {defaultValue=254, group="Effect_MultDecal_V02  FlipBook| DecalA | CCTime", parameterName="EMDv2_Decal_A_CC  Index"}
   CCinput=@n148
   CCinput_Alpha=@n148
   UV=@n147
  ) {materialFunction=&/MF_Global_CC_Tex}
  NamedRerouteDeclaration(None=ChannelMaskParameter(None=@n149.4v) {group="Effect_MultDecal_V02  FlipBook| DecalA | CCTime", parameterName=EMDv2_Decal_A_CC_Channel_Flip}.R) {name=CCA_FlipBook, nodeColor=<rgba 0.0547 1 0 1>}
  n150 = ChannelMaskParameter(None=@n149.4v) {group="Effect_MultDecal_V02  FlipBook| DecalA | CCTime", parameterName=EMDv2_Decal_A_CC_Channel_Warp}
  n151 = VectorParameter() {
   group="Effect_MultDecal_V02  FlipBook| DecalA | CCTime"
   parameterName=EMDv2_Decal_A_CC_Effect
   channelNames {a="", b="", g="", r=Warp}
   defaultValue=<rgba 0 0 0 1>
  }
  n152 = StaticBoolParameter() {group="Effect_MultDecal_V02  FlipBook| DecalA | CCTime", parameterName=EMDv2-Decal-A-CC}
  NamedRerouteDeclaration(None=StaticSwitch(
    True=LinearInterpolate(B=@n150.R, Alpha=@n151.Warp) {constA=1}
    False=Constant() {r=1}
    Value=NamedRerouteDeclaration(None=@n152) {name=CCA_FlipBook_State, nodeColor=<rgba 0.0547 1 0 1>}
   )) {name=CCA_FlipBook_Warp, nodeColor=<rgba 0.0547 1 0 1>}
  NamedRerouteDeclaration(None=StaticBoolParameter() {group=Effect_MultDecal_V02, parameterName=EMDv2_Use Msk--Tex--Metalic}) {name=State-Metalic, nodeColor=<rgba 0.3359 1 0 1>}
  n153 = ScalarParameter() {group=Effect_MultDecal_V02 UV Warp | Rotation, parameterName=EMDv2_Warp_Rotation_Weapon}
  n154 = MaterialFunctionCall(Float4=NamedRerouteUsage()) {materialFunction=&BreakOutFloat4Components}
  n155 = LinearInterpolate(
   A=ScalarParameter() {group=Effect_MultDecal_V02 UV Warp | Rotation, parameterName=EMDv2_Warp_Rotation_Vehicle}
   B=ScalarParameter() {group=Effect_MultDecal_V02 UV Warp | Rotation, parameterName=EMDv2_Warp_Rotation_Vehicle_Side}
   Alpha=@n154.G
  )
  n156 = ScalarParameter() {group=Effect_MultDecal_V02 UV Warp | Rotation, parameterName=EMDv2_Warp_Rotation_Mech}
  n157 = ScalarParameter() {group=Effect_MultDecal_V02 UV Warp | Rotation, parameterName=EMDv2_Warp_Rotation_Character}
  n158 = MaterialFunctionCall(Float2=NamedRerouteUsage()) {materialFunction=&BreakOutFloat2Components}
  n159 = LinearInterpolate(
   A=LinearInterpolate(A=LinearInterpolate(A=@n153, B=@n155, Alpha=@n154.R), B=LinearInterpolate(A=@n156, B=@n157, Alpha=@n154.A), Alpha=@n154.B)
   B=ScalarParameter() {group=Effect_MultDecal_V02 UV Warp | Rotation, parameterName=EMDv2_Warp_Rotation_Backpack}
   Alpha=@n158.R
  )
  n160 = ScalarParameter() {group=Effect_MultDecal_V02 UV Warp | Rotation, parameterName=EMDv2_Warp_Rotation_Pickaxe}
  NamedRerouteDeclaration(None=Add(A=LinearInterpolate(A=@n159, B=@n160, Alpha=@n158.G), B=NamedRerouteUsage())) {name=Uvs_Rotation_Warp, nodeColor=<rgba 0 0.7188 1 1>}
  NamedRerouteDeclaration(None=StaticBoolParameter() {group=Effect_MultDecal_V02, parameterName=EMDv2_Emissive Only, sortPriority=0}) {name=State-Emissive only, nodeColor=<rgba 0.3359 1 0 1>}
  n161 = TextureObjectParameter() {
   group=TopSide_V02
   parameterName=TopS_00A_Sharerd_Tex_Msk
   samplerType=SAMPLERTYPE_Masks
   texture=&T_BlackMask
  }
  n162 = TextureObjectParameter() {
   group=TopSide_V02
   parameterName=TopS_00B_Sharerd_Tex_Msk
   samplerType=SAMPLERTYPE_Masks
   texture=&T_BlackMask
  }
  n163 = StaticSwitchParameter(
   True=NamedRerouteDeclaration(None=@n161) {name=Tex_00A_Shared_Tex, nodeColor=<rgba 1 0.8906 0 1>}
   False=StaticSwitchParameter(True=NamedRerouteDeclaration(None=@n162) {name=Tex_00B_Shared_Tex, nodeColor=<rgba 1 0.8906 0 1>}, False=@n139) {group=Effect_MultDecal_V02, parameterName="EMDv2_Mask__Tex -------->TopS_00B_Sharerd_Tex_Msk"}
  ) {group=Effect_MultDecal_V02, parameterName="EMDv2_Mask__Tex -------->TopS_00A_Sharerd_Tex_Msk"}
  NamedRerouteDeclaration(None=StaticSwitch(
    True=NamedRerouteDeclaration(None=@n163) {name=Tex_Msk_, nodeColor=<rgba 0 1 0.9062 1>}
    False=TextureObjectParameter() {group=Effect_MultDecal_V02, parameterName=EMDv2_BC__Tex, texture=&T_BlacksRGB}
    Value=NamedRerouteUsage()
   )) {name=Tex_BC, nodeColor=<rgba 0 1 0.9062 1>}
  NamedRerouteDeclaration(None=@n1.RGBA) {name=CM_Mask, nodeColor=<rgba 1 0 0.3047 1>}
  n164 = ScalarParameter() {defaultValue=20, group=Effect_MultDecal_V02, parameterName=EMDv2_Tri_Contrast, sortPriority=-1}
  NamedRerouteDeclaration(None=Max(
    A=MaterialFunctionCall(In=@n142, Contrast=@n164) {materialFunction=&CheapContrast}.Result
    B=ScalarParameter() {group=Effect_MultDecal_V02, parameterName="EMDv2_Tri_Disable =1", sortPriority=-1}
   )) {name=Tri_Mask, nodeColor=<rgba 1 0 0.4922 1>}
  n165 = VectorParameter() {
   group=Effect_MultDecal_V02 Uv Rnd Tile | DecalA
   parameterName="EMDv2_DecalA_Rnd  Rotation"
   channelNames {a=Rotation_B_Speed, b=Rotation_A_Speed, g=Rotation_B, r=Rotation_A}
  }
  n166 = VectorParameter() {
   desc="disable Y-speed if using Y offset speed "
   group=Effect_MultDecal_V02 Uv Rnd Tile | DecalA
   parameterName=EMDv2_DecalA_Rnd_Speed
   channelNames {a="Y Position ", b=Y Speed Offset, g="Y Speed ", r=X Speed}
   defaultValue=<rgba 0 0 0 -0.5>
  }
  n167 = VectorParameter() {
   group=Effect_MultDecal_V02 Uv Rnd Tile | DecalA
   parameterName=EMDv2_DecalA_Rnd Scale
   channelNames {a="1= Sine Speed", b=Speed, g=Max, r=Min}
   defaultValue=<rgba 1 1 0 1>
  }
  n168 = VectorParameter() {
   group=Effect_MultDecal_V02 Uv Rnd Tile | DecalA
   parameterName=EMDv2_DecalA_Rnd Seeds
   channelNames {a="", b="", g=Seed B, r=Seed A}
   defaultValue=<rgba 5 3 0 1>
  }
  n169 = VectorParameter() {
   group=Effect_MultDecal_V02 Uv Rnd Tile | DecalA
   parameterName=EMDv2_DecalA_Rnd Msking Speed
   channelNames {a="", b="", g=Speed Sin, r=Speed}
  }
  n170 = VectorParameter() {
   group=Effect_MultDecal_V02 Uv Rnd Tile | DecalA
   parameterName=EMDv2_DecalA_Rnd CC Offset
   channelNames {a="", b="Seed Offset ", g=Max, r=Min}
   defaultValue=<rgba 3 5 2 1>
  }
  n171 = VectorParameter() {group=Effect_MultDecal_V02 Uv Rnd Tile | DecalA, parameterName=EMDv2_DecalA_Rnd FlipBooik, channelNames {a="", b="Offset ", g=Max, r=Min}}
  n172 = VectorParameter() {
   group=Effect_MultDecal_V02 Uv Rnd Tile | DecalA
   parameterName=EMDv2_DecalA_Rnd Msking
   channelNames {a=opacity, b=Offset, g=Max, r=Min}
   defaultValue=<rgba -0.01 -0.01 0 1>
  }
  n173 = MaterialFunctionCall!(
   Rnd-Rotation=@n165.RGBA
   UVs-NO offsets=NamedRerouteUsage()
   UVs_tile Speed=@n166.RGBA
   UVs-offsets=NamedRerouteUsage()
   UVs-TileRate_Y_Offset=ScalarParameter() {group=Effect_MultDecal_V02 Uv Rnd Tile | DecalA, parameterName=EMDv2_DecalA_Rnd_TileRate Y Offest}
   Rnd-Scale=@n167.RGBA
   "UVs-TileRate\\Scale"=NamedRerouteUsage()
   UVs-Rotation=NamedRerouteUsage()
   Rnd Seed=@n168.RGBA
   "Rnd Msk Speed\\Sine"=@n169.RGB
   Rnd CC offset=@n170.RGBA
   Rnd Flipbook=@n171.RGBA
   Rnd Msking=@n172.RGBA
   Rnd-Rotation_offset=ScalarParameter() {group=Effect_MultDecal_V02 Uv Rnd Tile | DecalA, parameterName="EMDv2_DecalA_Rnd  Rotation Offest"}
  ) {materialFunction=&/MF_MultiDecal_TileRand}
  NamedRerouteDeclaration(None=@n173.Uv_Rotation) {name=Rnd Tile A Rotation, nodeColor=<rgba 0 0.3906 1 1>}
  NamedRerouteDeclaration(None=@n173.Tile Msking) {name=Rnd Tile A Msking, nodeColor=<rgba 0 0.3906 1 1>}
  NamedRerouteDeclaration(None=@n173.Tile Flipbook) {name=Rnd Tile A FlipBook, nodeColor=<rgba 0 0.3906 1 1>}
  NamedRerouteDeclaration(None=@n173.Emissive CC) {name=Rnd Tile A CC Emissive, nodeColor=<rgba 0 0.3906 1 1>}
  NamedRerouteDeclaration(None=@n125) {name=Reactive_Rotation_A, nodeColor=<rgba 0.217 0.5686 1 1>}
  NamedRerouteDeclaration(None=@n173.Uv Tile) {name=Rnd Tile A Uv Tile, nodeColor=<rgba 0 0.3906 1 1>}
  NamedRerouteDeclaration(None=@n27.Tex---BC Alpha) {name=Decal A Alpha, nodeColor=<rgba 1 0.1875 0 1>}
  n174 = TextureObjectParameter() {
   group=Effect_MultDecal_V02
   parameterName=EMDv2_Normal__Tex
   samplerType=SAMPLERTYPE_Normal
   texture=&FlatNormalBlue
  }
  NamedRerouteDeclaration(None=@n174) {name=Tex_N, nodeColor=<rgba 0 1 0.9062 1>}
  NamedRerouteDeclaration(None=@n22.EmissiveColor) {name=Mat Emissive, nodeColor=<rgba 0.7812 1 0 1>}
  NamedRerouteDeclaration(None=FunctionInput(Preview=Constant4Vector().RGBA) {bUsePreviewValueAsDefault=true, inputName=IsState, inputType=FunctionInput_Vector4, sortPriority=4}) {name=IsState, nodeColor=<rgba 0.1497 1 0.1633 1>}
  n175 = ScalarParameter() {defaultValue=1, group=Effect_MultDecal_V02 UV Warp | TileRate, parameterName=EMDv2_Warp_TileRate_Weapon}
  n176 = MaterialFunctionCall(Float4=NamedRerouteUsage()) {materialFunction=&BreakOutFloat4Components}
  n177 = LinearInterpolate(
   A=ScalarParameter() {defaultValue=1, group=Effect_MultDecal_V02 UV Warp | TileRate, parameterName=EMDv2_Warp_TileRate_Vehicle}
   B=ScalarParameter() {defaultValue=1, group=Effect_MultDecal_V02 UV Warp | TileRate, parameterName=EMDv2_Warp_TileRate_Vehicle_Side}
   Alpha=@n176.G
  )
  n178 = ScalarParameter() {defaultValue=1, group=Effect_MultDecal_V02 UV Warp | TileRate, parameterName=EMDv2_Warp_TileRate_Scale_Mech}
  n179 = ScalarParameter() {defaultValue=1, group=Effect_MultDecal_V02 UV Warp | TileRate, parameterName=EMDv2_Warp_TileRate_Character}
  n180 = MaterialFunctionCall(Float2=NamedRerouteUsage()) {materialFunction=&BreakOutFloat2Components}
  n181 = LinearInterpolate(
   A=LinearInterpolate(A=LinearInterpolate(A=@n175, B=@n177, Alpha=@n176.R), B=LinearInterpolate(A=@n178, B=@n179, Alpha=@n176.A), Alpha=@n176.B)
   B=ScalarParameter() {defaultValue=1, group=Effect_MultDecal_V02 UV Warp | TileRate, parameterName=EMDv2_Warp_TileRate_Backpack}
   Alpha=@n180.R
  )
  n182 = ScalarParameter() {defaultValue=1, group=Effect_MultDecal_V02 UV Warp | TileRate, parameterName=EMDv2_Warp_TileRate_Pickaxe}
  n183 = VectorParameter() {
   group=Effect_MultDecal_V02 UV Warp | TileRate
   parameterName=EMDv2_Warp_NonUnformScale
   channelNames {a="", b="", g=Y Scale, r=X Scale}
   defaultValue=<rgba 1 1 1 1>
  }
  NamedRerouteDeclaration(None=Multiply(A=LinearInterpolate(A=@n181, B=@n182, Alpha=@n180.G), B=@n183.RGB)) {name=UVs_Warp_Scale, nodeColor=<rgba 0 0.7188 1 1>}
  NamedRerouteDeclaration(None=FunctionInput(Preview=Constant2Vector()) {bUsePreviewValueAsDefault=true, inputName=IsState_B, inputType=FunctionInput_Vector2, sortPriority=4}) {name=IsState_B, nodeColor=<rgba 0.1497 1 0.1633 1>}
  n184 = PreSkinnedLocalBounds()
  n185 = Subtract(A=VertexInterpolator(VS=PreSkinnedPosition()).PS, B=@n184.Min)
  n186 = MaterialFunctionCall(Color=Subtract(A=@n184.Max, B=@n184.Max)) {materialFunction=&SplitComponents}
  n187 = Max(A=Max(A=@n186.R, B=@n186.R), B=@n186.B)
  NamedRerouteDeclaration(None=FunctionInput(Preview=Divide(A=@n185, B=@n187)) {bUsePreviewValueAsDefault=true, inputName=Calcated_Bounds, sortPriority=11}) {name=Caclcated_Bounds, nodeColor=<rgba 0 0.9297 1 1>}
  n188 = ScalarParameter() {defaultValue=1, group=Effect_MultDecal_V02 Uv Offset | Decal_A | SafeZone, parameterName=EMDv2_DecalA_SZ_Scale_Vehicle}
  n189 = ScalarParameter() {defaultValue=1, group=Effect_MultDecal_V02 Uv Offset | Decal_A | SafeZone, parameterName=EMDv2_DecalA_SZ_Scale_Vehicle_Side}
  n190 = MaterialFunctionCall(Float4=NamedRerouteUsage()) {materialFunction=&BreakOutFloat4Components}
  n191 = LinearInterpolate(
   A=LinearInterpolate(
    A=ScalarParameter() {defaultValue=1, group=Effect_MultDecal_V02 Uv Offset | Decal_A | SafeZone, parameterName=EMDv2_DecalA_SZ_Scale_Weapon}
    B=LinearInterpolate(A=@n188, B=@n189, Alpha=@n190.G)
    Alpha=@n190.R
   )
   B=LinearInterpolate(
    A=ScalarParameter() {defaultValue=1, group=Effect_MultDecal_V02 Uv Offset | Decal_A | SafeZone, parameterName=EMDv2_DecalA_SZ_Scale_Mech}
    B=ScalarParameter() {defaultValue=1, group=Effect_MultDecal_V02 Uv Offset | Decal_A | SafeZone, parameterName=EMDv2_DecalA_SZ_Scale_Character}
    Alpha=@n190.A
   )
   Alpha=@n190.B
  )
  n192 = ScalarParameter() {defaultValue=1, group=Effect_MultDecal_V02 Uv Offset | Decal_A | SafeZone, parameterName=EMDv2_DecalA_SZ_Scale_Backpack}
  n193 = MaterialFunctionCall(Float2=NamedRerouteUsage()) {materialFunction=&BreakOutFloat2Components}
  NamedRerouteDeclaration(None=LinearInterpolate(
    A=LinearInterpolate(A=@n191, B=@n192, Alpha=@n193.R)
    B=ScalarParameter() {defaultValue=1, group=Effect_MultDecal_V02 Uv Offset | Decal_A | SafeZone, parameterName=EMDv2_DecalA_SZ_Scale_Pickaxe}
    Alpha=@n193.G
   )) {name=Uvs_SZ_Scale, nodeColor=<rgba 0 0.7188 1 1>}
  n194 = NamedRerouteUsage()
  NamedRerouteDeclaration(None=StaticSwitchParameter(
    True=NamedRerouteUsage()
    False=StaticSwitchParameter(True=@n194, False=@n139) {group=Effect_MultDecal_V02 UV Warp, parameterName="EMDv2_Mask__Tex Warp Msk -------->TopS_00B_Sharerd_Tex_Msk"}
   ) {group=Effect_MultDecal_V02 UV Warp, parameterName="EMDv2_Mask__Tex Warp Msk  -------->TopS_00A_Sharerd_Tex_Msk"}) {name=Tex_Warp, nodeColor=<rgba 0 1 0.1328 1>}
  n195 = PreSkinnedNormal()
  NamedRerouteDeclaration(None=FunctionInput(Preview=VertexInterpolator(VS=@n195).PS) {bUsePreviewValueAsDefault=true, inputName=Pre-Skinned Local Normal, sortPriority=10}) {name=PreSkin_Normal, nodeColor=<rgba 0.8281 1 0 1>}
  n196 = ChannelMaskParameter(None=@n59.RGBA) {group=Effect_MultDecal_V02 UV Warp, parameterName=EMDv2_Warp_Msk_Channel}
  n197 = ScalarParameter() {group=Effect_MultDecal_V02 UV Warp, parameterName=EMDv2_Warp_Offset}
  NamedRerouteDeclaration(None=Multiply(A=Add(A=@n196.R, B=@n197), B=ScalarParameter() {group=Effect_MultDecal_V02 UV Warp, parameterName=EMDv2_Warp_Strength})) {name=Warp_Uvs, nodeColor=<rgba 0.0547 0 1 1>}
  NamedRerouteDeclaration(None=StaticBoolParameter() {group=Effect_MultDecal_V02 UV Warp, parameterName=EMDv2_Warp_Simple_Uvs, sortPriority=1}) {name=Warp_Uv_State, nodeColor=<rgba 0.2656 1 0 1>}
  NamedRerouteDeclaration(None=@n22.Normal) {name=Mat Normal, nodeColor=<rgba 0.4062 0 1 1>}
  ScalarParameter() {group=Effect_MultDecal_V02, parameterName="EMDv2_Normals ------>Root"}
  StaticBool() {value=true}
  StaticSwitchParameter(
   True=BlendMaterialAttributes(A=@n3, B=NamedRerouteUsage(), Alpha=NamedRerouteDeclaration(None=@n2.B) {name=Msk_Accent, nodeColor=<rgba 0 1 0.7656 1>}) {vertexAttributeBlendType=UseA}
   False=NamedRerouteUsage()
  ) {group=Customization, parameterName=Effect_Multi_Decal_V2}
  StaticSwitchParameter(True=LinearInterpolate(A=@n18, Alpha=NamedRerouteUsage()), False=@n18) {desc="Use Scale value Alpha is disabled. ", group=Effect_MultDecal_V02 Uv Offset, parameterName=EMDv2_Reactive_Offset, sortPriority=0}
  TextureSample() {samplerType=SAMPLERTYPE_Masks, texture=/Game/Weapons/FORT_Rifles/Textures/T_Scar_CM}
 ]
}
