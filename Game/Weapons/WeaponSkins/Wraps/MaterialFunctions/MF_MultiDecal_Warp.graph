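# MF_MultiDecal_Warp is a material function in the weapon-skin wraps folder, named for warping across multiple decals.
# Its expression graph is empty, with no inputs, outputs or nodes shown, so it defines no warp logic here.
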MaterialFunction /Game/Weapons/WeaponSkins/Wraps/MaterialFunctions/MF_MultiDecal_Warp
graph {expressions {}}
