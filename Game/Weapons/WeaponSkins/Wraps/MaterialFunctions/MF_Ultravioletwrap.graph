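# Material function for the animated Ultraviolet wrap skin, blended over input material attributes by a customization mask behind a static switch.
# UVs built from PreSkin_Bounds and pan speeds adjust for weapon, vehicle, side-project vehicle or mech.
# Panning samples of one packed texture layer blue over two purple masks on near-black.
# A scrolling gradient pulses blue-area emissive, while a Blueprint-fed timestamp drives a decaying reactive glow on both colors.
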MaterialFunction /Game/Weapons/WeaponSkins/Wraps/MaterialFunctions/MF_Ultravioletwrap.MF_UltravioletWrap
graph {nodes [
  n1 = FunctionInput() {inputType=FunctionInput_MaterialAttributes}
  n2 = GetMaterialAttributes(None=@n1) {attributeGetTypes [#1]}
  n3 = VectorParameter() {parameterName="BluePanSpeed ", defaultValue=<rgba -0.08 -0.02 0 0>}
  n4 = Time()
  n5 = ScalarParameter() {defaultValue=1, parameterName=Vehicle Pan Mult}
  n6 = ScalarParameter() {defaultValue=1, parameterName=Side Project Vehicle Pan Mult}
  n7 = FunctionInput() {bUsePreviewValueAsDefault=true, inputName=Side Project Vehicle, inputType=FunctionInput_Scalar}
  n8 = FunctionInput() {bUsePreviewValueAsDefault=true, inputName=isVehicle, inputType=FunctionInput_Scalar, sortPriority=4}
  n9 = FunctionInput() {bUsePreviewValueAsDefault=true, inputName=isMech, inputType=FunctionInput_Scalar}
  n10 = LinearInterpolate(
   A=LinearInterpolate(A=ScalarParameter() {defaultValue=1, parameterName=Weapon Pan Mult}, B=LinearInterpolate(A=@n5, B=@n6, Alpha=@n7), Alpha=@n8)
   B=ScalarParameter() {defaultValue=1, parameterName=Mech Pan Mult}
   Alpha=@n9
  )
  n11 = Multiply(A=Multiply(A=ComponentMask(None=@n3.RGB) {g=true, r=true}, B=@n4), B=@n10)
  n12 = MaterialFunctionCall(Color=FunctionInput() {inputName=PreSkin_Bounds}) {materialFunction=&SplitComponents}
  n13 = AppendVector(A=@n12.G, B=@n12.G)
  n14 = LinearInterpolate(A=AppendVector(A=@n12.R, B=@n12.R), B=AppendVector(A=@n12.R, B=@n12.R), Alpha=@n7)
  n15 = OneMinus(None=@n12.B)
  n16 = Multiply(
   A=LinearInterpolate(A=LinearInterpolate(A=@n13, B=@n14, Alpha=@n8), B=AppendVector(A=@n15, B=@n12.R), Alpha=@n9)
   B=ScalarParameter() {defaultValue=1.25, parameterName=Grid Scale}
  )
  n17 = TextureObject() {samplerType=SAMPLERTYPE_LinearColor, texture=/Game/Weapons/WeaponSkins/Wraps/Textures/Ultraviolet/UV_BluePurple5_ParallaxStraighten}
  n18 = TextureSample(
   UVs=MaterialFunctionCall(UVs=Add(A=@n11, B=@n16), "Rotation Angle (0-1)"=ScalarParameter() {defaultValue=0.25, parameterName=Blue Rot}) {materialFunction=&CustomRotator}.Rotated Values
   Tex=@n17
  )
  n19 = Multiply(A=ComponentMask(None=VectorParameter() {parameterName=EmissivePulsePanSpeed, defaultValue=<rgba -0.5 0 0 0>}.RGB) {g=true, r=true}, B=@n4)
  n20 = MaterialFunctionCall(
   UVs=Add(A=Multiply(A=@n19, B=@n10), B=@n16)
   "Rotation Angle (0-1)"=ScalarParameter() {defaultValue=0.25, parameterName="Emissive Pan Rotation (0-1)"}
  ) {materialFunction=&CustomRotator}
  n21 = Multiply(
   A=Multiply(
    A=@n18.G
    B=TextureSample(UVs=@n20.Rotated Values) {samplerType=SAMPLERTYPE_Masks, texture=/Game/Weapons/WeaponSkins/Wraps/Textures/GilaMonster/T_Wraps_LizardStripes_Grad}.R
   )
   B=VectorParameter() {parameterName=Emissive Color, defaultValue=<rgba 0.1406 0.8603 1 0>}.RGB
  )
  n22 = ScalarParameter() {defaultValue=2, parameterName=Emissive Mult}
  n23 = Divide(
   A=Subtract(A=Time(), B=ScalarParameter() {group=00-Reactive-GTFB, parameterName=00-GameTimeFromBP})
   B=ScalarParameter() {parameterName=ULT_ReactPeriod}
  )
  n24 = ScalarParameter() {defaultValue=1, parameterName=ULT_ReactFalloff}
  n25 = OneMinus(None=Saturate(None=Power(Base=@n23, Exp=@n24)))
  n26 = VectorParameter() {parameterName=Purple, defaultValue=<rgba 0.7128 0.0169 0.9911 1>}
  n27 = ScalarParameter() {defaultValue=10, parameterName=React Bost Color}
  n28 = ComponentMask(None=VectorParameter() {parameterName=Purple1PanSpeed, defaultValue=<rgba 0.06 0.02 0 0>}.RGB) {g=true, r=true}
  n29 = Add(
   A=Multiply(A=Multiply(A=@n28, B=@n4), B=@n10)
   B=Multiply(A=@n16, B=ScalarParameter() {defaultValue=1, parameterName="Foreground Purple Scale Mult "})
  )
  n30 = ScalarParameter() {parameterName=PurpleRot}
  n31 = ComponentMask(None=VectorParameter() {parameterName=Purple2PanSpeed, defaultValue=<rgba 0.03 0.01 0 0>}.RGB) {g=true, r=true}
  n32 = Add(
   A=Multiply(A=Multiply(A=@n31, B=@n4), B=@n10)
   B=Multiply(A=@n16, B=ScalarParameter() {defaultValue=1.25, parameterName=Background Mask Scale Mult})
  )
  n33 = Add(
   A=TextureSample(
    UVs=MaterialFunctionCall(UVs=@n29, "Rotation Angle (0-1)"=@n30) {materialFunction=&CustomRotator}.Rotated Values
    Tex=@n17
   ).R
   B=TextureSample(
    UVs=MaterialFunctionCall(UVs=@n32, "Rotation Angle (0-1)"=@n30) {materialFunction=&CustomRotator}.Rotated Values
    Tex=@n17
   ).B
  )
  n34 = Saturate(None=Subtract(A=Saturate(None=@n33), B=@n18.G))
  n35 = Multiply(A=Multiply(A=@n25, B=Multiply(A=@n26.RGB, B=@n27)), B=@n34)
  n36 = VectorParameter() {parameterName=Blue, defaultValue=<rgba 0 0.6571 0.8854 1>}
  n37 = Multiply(A=Multiply(A=@n25, B=Multiply(A=@n36.RGB, B=@n27)), B=@n18.G)
  n38 = Multiply(A=@n34, B=@n26.RGB)
  n39 = Multiply(A=@n18.G, B=@n36.RGB)
  n40 = VectorParameter() {parameterName=Background Black, defaultValue=<rgba 0.009 0.009 0.009 0>}
  n41 = OneMinus(None=Saturate(None=Add(A=@n33, B=@n18.G)))
  n42 = SetMaterialAttributes!(
   MaterialAttributes=@n2.MaterialAttributes
   EmissiveColor=Add(A=Multiply(A=@n21, B=@n22), B=Add(A=@n35, B=@n37))
   BaseColor=Add(A=Add(A=@n38, B=@n39), B=Multiply(A=@n40.RGB, B=@n41))
   Specular=LinearInterpolate(B=ScalarParameter() {defaultValue=0.5, parameterName=Max Spec}, Alpha=@n2.Specular)
  ) {attributeSetTypes [#2, #3, #1]}
  n43 = FunctionInput() {bUsePreviewValueAsDefault=true, inputName="Customization Mask (_CM B)", inputType=FunctionInput_Scalar, sortPriority=24}
  FunctionOutput(None=StaticSwitchParameter(True=BlendMaterialAttributes(A=@n1, B=@n42, Alpha=@n43), False=@n1) {group=Customization_Wrap, parameterName=Use Ultraviolet Wrap})
 ]}
