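# Material function for the HoloFoil weapon, vehicle and mech wrap, switched by Wrap_HoloFoilTUM.
# Camera direction and pre-skinned bounds shift gradient lookups on side-projected UVs, with per-target tuning.
# Fresnel and a radial mask shape the gradient into base color and emissive, then a masked side decal overlays it.
# The customization mask blends an MF_Silk-processed foil with the decal version; Multicolor and chroma-noise branches do not reach the output.
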MaterialFunction /Game/Weapons/WeaponSkins/Wraps/MaterialFunctions/MF_Wrap_HoloFoilTum.MF_Wrap_HoloFoilTUM
graph {nodes [
  n1 = Constant()
  n2 = MaterialFunctionCall(
   R=FunctionInput(Preview=Constant()) {inputName=isVehicle, inputType=FunctionInput_Scalar, sortPriority=5}
   G=FunctionInput(Preview=@n1) {bUsePreviewValueAsDefault=true, inputName=Side Project Vehicle, inputType=FunctionInput_Scalar}
   B=FunctionInput(Preview=@n1) {bUsePreviewValueAsDefault=true, inputName=isMech, inputType=FunctionInput_Scalar, sortPriority=4}
  ) {materialFunction=&AppendMany}
  n3 = MaterialFunctionCall(Color=@n2.RGB) {materialFunction=&SplitComponents}
  ComponentMask(None=Saturate(None=Add(A=@n3.R, B=@n3.R))) {r=true}
  n4 = VectorParameter() {group=Wrap_TUM_Base, parameterName="TUM_BaseColor-A=Rough -----", defaultValue=<rgba 0.7135 0.7135 0.7135 0.5>}
  n5 = ScalarParameter() {defaultValue=10, group=Wrap_TUM_HoloFoil, parameterName=TUM_FresnelExponent_Weapon, sortPriority=-100}
  n6 = ScalarParameter() {defaultValue=3.5, group=Wrap_TUM_HoloFoil, parameterName=TUM_FresnelExponent_Vehicle, sortPriority=-100}
  n7 = MaterialFunctionCall(Color=@n2.RGB) {materialFunction=&SplitComponents}
  n8 = FunctionInput() {inputName=Material, inputType=FunctionInput_MaterialAttributes, sortPriority=-100}
  n9 = GetMaterialAttributes(None=@n8) {attributeGetTypes [
    #1
    #2
    #3
    #4
    #5
   ]}
  n10 = Transform(None=@n9.Normal)
  n11 = MaterialFunctionCall(
   In=Fresnel(ExponentIn=LinearInterpolate(A=@n5, B=@n6, Alpha=@n7.R), Normal=@n10)
   Contrast=ScalarParameter() {group=Wrap_TUM_HoloFoil, parameterName=TUM_FresnelContrast, sortPriority=-100}
  ) {materialFunction=&CheapContrast}
  n12 = VertexInterpolator(VS=PreSkinnedPosition())
  n13 = PreSkinnedLocalBounds()
  n14 = FunctionInput(Preview=Divide(A=Subtract(A=@n12.PS, B=@n13.Min), B=Subtract(A=@n13.Max, B=@n13.Max))) {bUsePreviewValueAsDefault=true, inputName=Calcated_Bounds, sortPriority=11}
  n15 = ComponentMask(None=@n14) {b=true}
  n16 = MaterialFunctionCall() {materialFunction=&CameraDirectionVector}
  n17 = ScalarParameter() {defaultValue=0.5, group=Wrap_TUM_HoloFoil, parameterName=TUM_ViewAnlgeFidelity}
  n18 = Multiply(A=DotProduct(A=Constant3Vector() {constant=<rgba 0 0 1 0>}, B=@n16.Result), B=@n17)
  n19 = Multiply(A=DotProduct(A=Constant3Vector() {constant=<rgba 1 0 0 0>}, B=@n16.Result), B=@n17)
  n20 = Add(A=@n18, B=@n19)
  n21 = Multiply(A=DotProduct(A=Constant3Vector() {constant=<rgba 0 1 0 0>}, B=@n16.Result), B=@n17)
  n22 = Add(A=ComponentMask(None=@n14) {r=true}, B=@n19)
  n23 = Add(A=ComponentMask(None=@n14) {g=true}, B=@n21)
  n24 = LinearInterpolate(
   A=Add(A=@n15, B=Add(A=@n20, B=@n21))
   B=Add(A=Add(A=@n22, B=@n23), B=Add(A=@n15, B=@n18))
   Alpha=ScalarParameter() {group=Wrap_TUM_HoloFoil, parameterName=TUM_Use3Vector}
  )
  n25 = ScalarParameter() {defaultValue=1, group=Wrap_TUM_HoloFoil, parameterName=TUM_Layer_01_ChromaOffset}
  n26 = ScalarParameter() {defaultValue=1, group=Wrap_TUM_HoloFoil, parameterName=TUM_01_TileRate_Gradient_Weapon}
  n27 = MaterialFunctionCall(Color=@n2.RGB) {materialFunction=&SplitComponents}
  n28 = LinearInterpolate(
   A=ScalarParameter() {defaultValue=1, group=Wrap_TUM_HoloFoil, parameterName=TUM_01_TileRate_Gradient_Vehicle}
   B=ScalarParameter() {defaultValue=1, group=Wrap_TUM_HoloFoil, parameterName=TUM_01_TileRate_Gradient_VehicleSide}
   Alpha=@n27.G
  )
  n29 = MaterialFunctionCall(Color=@n2.RGB) {materialFunction=&SplitComponents}
  n30 = VectorParameter() {group=Wrap_TUM_UVs, parameterName=TUM_01_Offset_Weapon, defaultValue=<rgba 0 0 0 1>}
  n31 = VectorParameter() {desc=must activate TS_Vehicle_Side_01 UVs, group=Wrap_TUM_UVs, parameterName=TUM_01_Offset_Vehicle_Side, defaultValue=<rgba 0 0 0 1>}
  n32 = MaterialFunctionCall(Color=@n2.RGB) {materialFunction=&SplitComponents}
  n33 = LinearInterpolate(A=VectorParameter() {group=Wrap_TUM_UVs, parameterName=TUM_01_Offset_Vehicle, defaultValue=<rgba 0 0 0 1>}.RGB, B=@n31.RGB, Alpha=@n32.G)
  n34 = ScalarParameter() {defaultValue=1, group=Wrap_TUM_UVs, parameterName=TUM_01_TileRate_Weapon}
  n35 = LinearInterpolate(
   A=ScalarParameter() {defaultValue=1, group=Wrap_TUM_UVs, parameterName=TUM_01_TileRate_Vehicle}
   B=ScalarParameter() {defaultValue=1, group=Wrap_TUM_UVs, parameterName=TUM_01_TileRate_VehicleSide}
   Alpha=@n32.G
  )
  n36 = FunctionInput(Preview=VertexInterpolator(VS=PreSkinnedNormal()).PS) {bUsePreviewValueAsDefault=true, inputName=Pre-Skinned Local Normal, sortPriority=10}
  n37 = MaterialFunctionCall!(
   For_Safe_Zones
   Normal_Map
   Caculated_Bounds=@n14
   Pre_Skinned_Normals=@n36
   Is Vehicle=@n29.R
   Is VehicleSide=@n29.R
   Is Mech=@n29.R
   Vehicle_Rotation=ScalarParameter() {group=Wrap_TUM_UVs, parameterName=TUM_Rotation_Vehicle}
   UV_Rotation Bypass
   Mech_Rotation=ScalarParameter() {defaultValue=0.25, group=Wrap_TUM_UVs, parameterName=TUM_Rotation_Mech -----}
   Uvs_Offset=LinearInterpolate(
    A=LinearInterpolate(A=@n30.RGB, B=@n33, Alpha=@n32.R)
    B=VectorParameter() {group=Wrap_TUM_UVs, parameterName=TUM_01_Offset_Mech -----, defaultValue=<rgba 0 0 0 1>}.RGB
    Alpha=@n32.B
   )
   Offset-after rotation
   Uvs_TileRate=LinearInterpolate(
    A=LinearInterpolate(A=@n34, B=@n35, Alpha=@n32.R)
    B=ScalarParameter() {defaultValue=1, group=Wrap_TUM_UVs, parameterName=TUM_01_TileRate_Mech -----}
    Alpha=@n32.B
   )
  ) {materialFunction=&/MF_Projection_Side_Normal}
  n38 = Multiply(
   A=LinearInterpolate(
    A=LinearInterpolate(A=@n26, B=@n28, Alpha=@n27.R)
    B=ScalarParameter() {defaultValue=1, group=Wrap_TUM_HoloFoil, parameterName=TUM_01_TileRate_Gradient_Mech -----}
    Alpha=@n27.B
   )
   B=@n37.UVs No Offsets/Tiling
  )
  n39 = VectorParameter() {group=Wrap_TUM_HoloFoil, parameterName=TUM_01_Offset_Gradient_Weapon, defaultValue=<rgba 0 0 0 1>}
  n40 = VectorParameter() {group=Wrap_TUM_HoloFoil, parameterName=TUM_01_Offset_Gradient_Vehicle_Side, defaultValue=<rgba 0 0 0 1>}
  n41 = LinearInterpolate(
   A=VectorParameter() {group=Wrap_TUM_HoloFoil, parameterName=TUM_01_Offset_Gradient_Vehicle, defaultValue=<rgba 0 0 0 1>}.RGB
   B=@n40.RGB
   Alpha=@n27.G
  )
  n42 = ComponentMask(None=LinearInterpolate(
    A=LinearInterpolate(A=@n39.RGB, B=@n41, Alpha=@n27.R)
    B=VectorParameter() {group=Wrap_TUM_HoloFoil, parameterName=TUM_01_Offset_Gradient_Mech -----, defaultValue=<rgba 0 0 0 1>}.RGB
    Alpha=@n27.B
   )) {g=true, r=true}
  n43 = TextureSample(
   UVs=Add(A=Add(A=@n24, B=@n25), B=Add(A=@n38, B=@n42))
   Tex=TextureObjectParameter() {group=Wrap_TUM_Textures, parameterName=TUM_FresnelGradient, texture=&T_BlacksRGB}
  )
  n44 = VectorParameter() {group=Wrap_TUM_HoloFoil, parameterName=TUM_CurveRadial_OffsetWeapon, sortPriority=-100, defaultValue=<rgba 0.1 1 0 1>}
  n45 = VectorParameter() {group=Wrap_TUM_HoloFoil, parameterName=TUM_CurveRadial_OffsetVehicle, sortPriority=-100, defaultValue=<rgba 0.1 1 0 1>}
  n46 = VectorParameter() {group=Wrap_TUM_HoloFoil, parameterName=TUM_CurveRadial_OffsetVehicleSide, sortPriority=-100, defaultValue=<rgba 0.1 1 0 1>}
  n47 = VectorParameter() {group=Wrap_TUM_HoloFoil, parameterName=TUM_CurveRadial_OffsetMech, sortPriority=-100, defaultValue=<rgba 0.1 1 0 1>}
  n48 = LinearInterpolate(A=LinearInterpolate(A=@n44.RGB, B=LinearInterpolate(A=@n45.RGB, B=@n46.RGB, Alpha=@n7.G), Alpha=@n7.R), B=@n47.RGB, Alpha=@n7.B)
  n49 = ScalarParameter() {defaultValue=1, group=Wrap_TUM_HoloFoil, parameterName=TUM_CurveRadial_Size_Weapon, sortPriority=-100}
  n50 = LinearInterpolate(
   A=ScalarParameter() {defaultValue=1, group=Wrap_TUM_HoloFoil, parameterName=TUM_CurveRadial_Size_Vehicle, sortPriority=-100}
   B=ScalarParameter() {defaultValue=1, group=Wrap_TUM_HoloFoil, parameterName=TUM_CurveRadial_Size_VehicleSide, sortPriority=-100}
   Alpha=@n7.G
  )
  n51 = SphereMask(
   A=Add(A=ComponentMask(None=@n48) {g=true, r=true}, B=@n37.UVs No Offsets/Tiling)
   B=Constant() {r=0.5}
   Radius=LinearInterpolate(
    A=LinearInterpolate(A=@n49, B=@n50, Alpha=@n7.R)
    B=ScalarParameter() {defaultValue=1, group=Wrap_TUM_HoloFoil, parameterName=TUM_CurveRadial_Size_Mech, sortPriority=-100}
    Alpha=@n7.B
   )
   Hardness=ScalarParameter() {group=Wrap_TUM_HoloFoil, parameterName=TUM_CurveRadial_Hardness, sortPriority=-100}
  ) {hardnessPercent=0}
  n52 = LinearInterpolate(A=@n43.RGB, B=LinearInterpolate(A=@n4.RGB, B=@n43.RGB, Alpha=@n11.Result), Alpha=@n51)
  n53 = ScalarParameter() {defaultValue=0.7, group=Wrap_TUM_Base, parameterName=TUM_Spec}
  n54 = Multiply(
   A=LinearInterpolate(A=@n43.RGB, B=LinearInterpolate(B=@n43.RGB, Alpha=@n11.Result), Alpha=@n51)
   B=ScalarParameter() {defaultValue=0.5, group=Wrap_TUM_HoloFoil, parameterName=TUM_Holofoil_EmissiveAmount}
  )
  n55 = ScalarParameter() {defaultValue=1, group=Wrap_TUM_Base, parameterName=TUM_Metal}
  n56 = SetMaterialAttributes!(
   MaterialAttributes=@n9.MaterialAttributes
   Metallic=@n55
   Roughness=@n4.A
   BaseColor=@n52
   Normal=@n9.Normal
   Specular=@n53
   EmissiveColor=@n54
  ) {attributeSetTypes [
    #3
    #4
    #5
    #1
    #6
    #2
   ]}
  n57 = MaterialFunctionCall!(Alpha=Constant() {r=1}, Material=@n56) {materialFunction=/Game/Weapons/WeaponSkins/MaterialFunction/MF_Silk}
  n58 = ScalarParameter() {defaultValue=5, group=Wrap_TUM_HoloFoil, parameterName=TUM_FresnelExponent_BaseColor, sortPriority=-100}
  n59 = Multiply(A=LinearInterpolate(Alpha=Fresnel(ExponentIn=@n58, Normal=@n10)) {constA=1, constB=0}, B=@n52)
  n60 = TextureSample(
   UVs=@n37.UVs
   Tex=TextureObjectParameter() {group=Wrap_TUM_Textures, parameterName=TUM_SideDecalBaseColor, texture=&T_BlacksRGB}
  )
  n61 = TextureObjectParameter() {group=Wrap_TUM_Textures, parameterName=TUM_SideDecalMask, texture=&T_BlacksRGB}
  n62 = TextureSample(UVs=@n37.UVs, Tex=@n61)
  n63 = ScalarParameter() {defaultValue=0.5, group=Wrap_TUM_Base, parameterName=TUM_RoughnessDecal}
  n64 = Multiply(A=@n62.B, B=Power(Base=@n60.RGB) {constExponent=5})
  n65 = ScalarParameter() {defaultValue=2, group=Wrap_TUM_Base, parameterName=TUM_SideDecal_EmissiveAmount}
  n66 = SetMaterialAttributes!(
   MaterialAttributes=@n9.MaterialAttributes
   BaseColor=LinearInterpolate(A=LinearInterpolate(A=@n59, B=@n60.RGB, Alpha=@n62.R), Alpha=@n62.G) {constB=0}
   Metallic=LinearInterpolate(A=@n55, B=ScalarParameter() {group=Wrap_TUM_Base, parameterName=TUM_MetalDecal}, Alpha=@n62.R) {desc=MetalBlend}
   Specular=LinearInterpolate(A=@n53, B=ScalarParameter() {defaultValue=0.7, group=Wrap_TUM_Base, parameterName=TUM_SpecDecal}, Alpha=@n62.R) {desc=Spec}
   Roughness=LinearInterpolate(A=@n4.A, B=Multiply(A=@n63, B=@n60.A), Alpha=@n62.R)
   EmissiveColor=LinearInterpolate(A=@n54, B=Multiply(A=@n64, B=@n65), Alpha=@n62.R)
  ) {attributeSetTypes [
    #5
    #3
    #6
    #4
    #2
   ]}
  n67 = Saturate(None=Add(
    A=FunctionInput() {bUsePreviewValueAsDefault=true, inputName="Customization Mask (_CM B)", inputType=FunctionInput_Scalar, sortPriority=24}
    B=ScalarParameter() {group=Wrap_TUM_Base, parameterName=TUM_Remove_Customization Mask, sortPriority=-10}
   ))
  FunctionOutput(None=StaticSwitchParameter(True=BlendMaterialAttributes(A=@n57.Result, B=@n66, Alpha=@n67) {vertexAttributeBlendType=UseA}, False=@n8) {group=Customization_Wrap, parameterName=Wrap_HoloFoilTUM})
  n68 = GetMaterialAttributes(None=@n8) {attributeGetTypes [
    #5
    #3
    #4
    #2
   ]}
  n69 = ScalarParameter() {defaultValue=1, group=Multicolor, parameterName=Multicolor_FresnelExponent, sortPriority=10}
  n70 = Saturate(None=Multiply(
    A=Fresnel(ExponentIn=@n69) {baseReflectFraction=0, exponent=1}
    B=ScalarParameter() {defaultValue=1, group=Multicolor, parameterName=Multicolor_FresnelMultiplier, sortPriority=11}
   ) {constB=2})
  n71 = Multiply(A=@n70, B=ScalarParameter() {defaultValue=1, group=Multicolor, parameterName=Multicolor_ViewAngleUVTile, sortPriority=22})
  n72 = Multiply(
   A=DotProduct(
    A=Constant3Vector() {constant=<rgba 1 1 1 0>}
    B=MaterialFunctionCall() {materialFunction=&CameraDirectionVector}.Result
   )
   B=ScalarParameter() {defaultValue=0.5, group=Multicolor, parameterName=Multicolor_ViewAngleUVOffsetStrength, sortPriority=23}
  ) {constB=0.5}
  n73 = VertexInterpolator(VS=PreSkinnedPosition())
  n74 = PreSkinnedLocalBounds()
  n75 = Multiply(
   A=ComponentMask(None=Divide(A=@n73.PS, B=@n74.Max)) {b=true}
   B=ScalarParameter() {defaultValue=1, group=Multicolor, parameterName=Multicolor_VerticalUVTiling, sortPriority=20}
  )
  n76 = ScalarParameter() {group=Multicolor, parameterName=Multicolor_VerticalUVOffset, sortPriority=21}
  n77 = TextureObjectParameter() {
   group=Multicolor
   parameterName=Multicolor_Gradient
   sortPriority=0
   texture=/Game/Characters/Player/Female/Medium/Bodies/F_Med_Soldier_01/Skins/F_MED_COM_Taxi_Upgraded/Skins/Multi/T_Multicolor_Gradient_D
  }
  n78 = TextureSample(
   UVs=MaterialFunctionCall(X=Add(A=@n71, B=@n72), Y=Add(A=@n75, B=@n76)) {materialFunction=&MakeFloat2}.Result
   Tex=@n77
  )
  n79 = Multiply(A=@n70, B=ScalarParameter() {defaultValue=1, group=Multicolor, parameterName=Multicolor_BaseColorGradient, sortPriority=30})
  n80 = ScalarParameter() {group=Multicolor, parameterName=Multicolor_Metallic, sortPriority=31}
  n81 = Add(A=@n68.Roughness, B=ScalarParameter() {group=Multicolor, parameterName=Multicolor_RoughnessModify, sortPriority=33})
  n82 = ScalarParameter() {defaultValue=1, group=Multicolor, parameterName=Multicolor_EmissiveGradient, sortPriority=34}
  SetMaterialAttributes!(
   MaterialAttributes=@n68.MaterialAttributes
   BaseColor=LinearInterpolate(
    A=LinearInterpolate(A=@n68.BaseColor, B=@n78.RGB, Alpha=@n79)
    B=@n70
    Alpha=ScalarParameter() {group=Multicolor, parameterName=Multicolor_DebugShowMask, sortPriority=40}
   )
   Metallic=Saturate(None=LinearInterpolate(A=@n68.Metallic, B=@n80, Alpha=@n70))
   Roughness=Saturate(None=LinearInterpolate(A=@n68.Roughness, B=@n81, Alpha=@n70))
   EmissiveColor=LinearInterpolate(A=@n68.EmissiveColor, B=@n78.RGB, Alpha=Multiply(A=@n70, B=@n82))
  ) {attributeSetTypes [
    #5
    #3
    #4
    #2
   ]}
  n83 = TextureObjectParameter() {
   group=Wrap_TUM_HoloFoil
   parameterName=TUM_ChromaDistortionNoise
   samplerType=SAMPLERTYPE_LinearColor
   texture=/Game/Characters/Player/Female/Medium/Bodies/F_MED_HighTower_Squash/Textures/T_Swirly_Noise_highContrast1_Characters
  }
  n84 = Constant() {r=1}
  n85 = ScalarParameter() {defaultValue=1, group=Wrap_TUM_HoloFoil, parameterName=TUM_ChromaDistortionNoise_Tiling_Weapon}
  n86 = ScalarParameter() {defaultValue=1, group=Wrap_TUM_HoloFoil, parameterName=TUM_ChromaDistortionNoise_Tiling_Vehicle}
  n87 = Multiply(
   A=@n14
   B=LinearInterpolate(
    A=LinearInterpolate(A=@n85, B=@n86, Alpha=@n3.R)
    B=ScalarParameter() {defaultValue=1, group=Wrap_TUM_HoloFoil, parameterName=TUM_ChromaDistortionNoise_Tiling_Mech}
    Alpha=@n3.B
   )
  )
  Subtract(
   A=@n14
   B=Multiply(
    A=ScalarParameter() {defaultValue=0.25, group=Wrap_TUM_HoloFoil, parameterName=TUM_ChromaDistortionNoiseStrength}
    B=MaterialFunctionCall(TextureObject=@n83, TextureSize=@n84, WorldPosition=@n87, World Space Normal=@n36) {materialFunction=&/WorldAlignedTexture}.XYZ Texture
   )
  )
  TextureObjectParameter() {group=Wrap_TUM_Textures, parameterName=TUM_FresnelGradient, texture=&T_BlacksRGB}
  TextureSample(UVs=@n37.UVs, Tex=@n61)
 ]}
