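# Material function for customization wraps that picks the blend style for primary triplanar diffuse, roughness, normal and emissive.
# With UseGradient on, an alpha blends the primary layer toward the base material or second triplanar layer, chosen by static bool.
# The panning-mask or DiscoFloor switch instead uses MF_WrapDiffusePan to mask between simple-gradient values and primary layer with MF_Tri_Pann_Fresnel_Apply.
# Spec always comes from the simple-gradient blend; the pan mask is also output.
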MaterialFunction /Game/Packages/Fortress_SharedMaterials/Base_Material_Functions/BaseCustomizationFunctions/GradientBlendStyles
graph {nodes [
  n1 = VectorParameter() {group=Primary Triplaner Gradient, parameterName=SimpleGradientDiffuse, sortPriority=-18, defaultValue=<rgba 0 0 0 1>}
  n2 = FunctionInput() {inputName=Gradient_Diffuse, sortPriority=4}
  n3 = FunctionInput() {inputName=Gradient_Roughness, inputType=FunctionInput_Scalar, sortPriority=5}
  n4 = FunctionInput() {inputName=Gradient_Normal, sortPriority=6}
  n5 = FunctionInput() {inputName=RoughnessAlphaBlend, inputType=FunctionInput_Scalar, sortPriority=7}
  n6 = MaterialFunctionCall!(
   A_01=@n1.RGB
   B_01=@n2
   A_02=ScalarParameter() {group=Primary Triplaner Gradient, parameterName=SimpleGradientRoughness, sortPriority=-17}
   B_02=@n3
   A_03=StaticSwitchParameter(True=FunctionInput() {inputName=BaseNormal, sortPriority=-1}, False=@n4) {group=Primary Triplaner Gradient, parameterName=SimpleGradientNormal, sortPriority=-16}
   B_03=@n4
   A_04=ScalarParameter() {defaultValue=0.5, group=Primary Triplaner Gradient, parameterName=SimpleGradientSpec, sortPriority=-17}
   B_04=FunctionInput() {inputName=Specular, sortPriority=-1}
   Alpha=StaticSwitch(
    True=@n5
    False=Constant() {r=1}
    Value=StaticBoolParameter() {group=Primary Triplaner Gradient, parameterName=-SimpleGradent-Triplaner-02----------, sortPriority=-19}
   )
  ) {materialFunction=&/MF_Blend4}
  n7 = FunctionInput() {inputName=Primary_emissive, sortPriority=4}
  n8 = FunctionInput() {inputName=Primary_Diffuse, sortPriority=1}
  n9 = MaterialFunctionCall!(Emissive_Map=@n7, BaseColor_Map=@n8) {materialFunction=/Game/Weapons/WeaponSkins/Wraps/MaterialFunctions/MF_Tri_Pann_Fresnel_Apply}
  n10 = FunctionInput() {inputName=Primary_Roughness, inputType=FunctionInput_Scalar, sortPriority=2}
  n11 = FunctionInput() {inputName=Primary_Normal, sortPriority=3}
  n12 = MaterialFunctionCall!(
   NormalizedLocalPosition=FunctionInput() {inputName=NormalizedLocalPosition, sortPriority=8}
   ForwardDirection=FunctionInput() {inputName=Forward Vector, sortPriority=9}
   isVehicle=FunctionInput() {inputName=isVehicle, sortPriority=-1}
   isMech=FunctionInput() {inputName=isMech, inputType=FunctionInput_Scalar}
   DiffusMask=FunctionInput() {inputName=Primary_Mask, sortPriority=3}
  ) {materialFunction=&/MF_WrapDiffusePan}
  n13 = MaterialFunctionCall(
   A_01=@n6.Out_01
   B_01=@n9.Basecolor_output
   A_02=@n6.Out_01
   B_02=@n10
   A_03=@n6.Out_01
   B_03=@n11
   A_04=@n7
   B_04=@n9.Basecolor_output
   Alpha=@n12.Result
  ) {materialFunction=&MF_Blend4}
  n14 = GetMaterialAttributes(None=FunctionInput() {inputName=Material, inputType=FunctionInput_MaterialAttributes, sortPriority=-1}) {attributeGetTypes [
    #1
    #2
    #3
    #4
   ]}
  n15 = StaticBoolParameter() {desc=Blend Between Triplaner 01 to Triplaner 02, group=Primary Triplaner Gradient, parameterName=BlendToTriplaner 02, sortPriority=-15}
  n16 = MaterialFunctionCall!(
   A_01=@n14.BaseColor
   B_01=@n2
   A_02=@n14.BaseColor
   B_02=@n3
   A_03=@n14.BaseColor
   B_03=@n4
   A_04=@n14.BaseColor
   B_04=Constant()
   Alpha=@n15
  ) {materialFunction=&/MF_Switch4}
  n17 = MaterialFunctionCall(A_01=@n8, B_01=@n16.Out_01, A_02=@n10, B_02=@n16.Out_01, A_03=@n11, B_03=@n16.Out_01, A_04=@n7, B_04=@n16.Out_01, Alpha=@n5) {materialFunction=&MF_Blend4}
  n18 = StaticBoolParameter() {
   desc="uses Gradent to Blend Between two materials "
   group=Primary Triplaner Gradient
   parameterName=-UseGradient---------------
   sortPriority=-50
  }
  n19 = MaterialFunctionCall(A_01=@n17.Out_01, B_01=@n8, A_02=@n17.Out_01, B_02=@n10, A_03=@n17.Out_01, B_03=@n11, A_04=@n17.Out_01, B_04=@n7, Alpha=@n18) {materialFunction=&MF_Switch4}
  n20 = StaticBoolParameter() {
   desc="Mask out triplaner 01 by Mask's green channel "
   group=Primary Triplaner DiffusePanningMask
   parameterName=-Triplaner-01-Panning-Mask----------------
   sortPriority=-5
  }
  n21 = MaterialFunctionCall(
   A_01=@n13.Out_01
   B_01=@n19.Out_01
   A_02=@n13.Out_01
   B_02=@n19.Out_01
   A_03=@n13.Out_01
   B_03=@n19.Out_01
   A_04=@n13.Out_01
   B_04=@n19.Out_01
   Alpha=StaticSwitch(True=StaticBool() {value=true}, False=@n20, Value=StaticBoolParameter() {group=Customization, parameterName=DiscoFloor Wrap})
  ) {materialFunction=&MF_Switch4}
  FunctionOutput(None=@n21.Out_01) {outputName=Diffuse}
  FunctionOutput(None=@n21.Out_02) {outputName=Roughness, sortPriority=1}
  FunctionOutput(None=@n21.Out_03) {outputName=Normal, sortPriority=2}
  FunctionOutput(None=@n21.Out_04) {outputName=Emissive, sortPriority=2}
  FunctionOutput(None=@n12.Result) {outputName=WrapDiffuseMask}
  FunctionOutput(None=@n6.Out_04) {outputName=Spec, sortPriority=2}
 ]}
