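# Material function that builds emissive for wrap materials on weapons, vehicles and mechs, outputting emissive plus the first height-blend mask.
# Primary path tints a mask channel, optionally modulated by a gradient panning along local forward position, with per-type offsets, sine motion and mask-driven UV warp.
# Static switches can replace that with MF_WrapEmissiveHeightBlend for triplanar layer 01.
# MF_WrapEmissiveHeightBlend2 optionally supplies layer 02, either added or mask-blended over the result.
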MaterialFunction /Game/Packages/Fortress_SharedMaterials/Base_Material_Functions/BaseCustomizationFunctions/MF_WrapEmissive
graph {nodes [
  n1 = FunctionInput() {bUsePreviewValueAsDefault=true, inputName=isMech, inputType=FunctionInput_Scalar}
  n2 = FunctionInput() {inputName=HeightMapMask_2, inputType=FunctionInput_Scalar}
  n3 = FunctionInput() {bUsePreviewValueAsDefault=true, inputName=isVehicle}
  n4 = FunctionInput() {inputName=NormalizedLocalPosition}
  n5 = FunctionInput() {inputName=ForwardDirection}
  n6 = MaterialFunctionCall!(
   NormalizedLocalPosition=@n4
   ForwardDirection=@n5
   isVehicle=@n3
   IsStatsB=NamedRerouteUsage()
   IsMech=@n1
   IsCharacter=NamedRerouteUsage()
   EmissiveMask=@n2
   EmissiveMask_B
  ) {materialFunction=&/MF_WrapEmissiveHeightBlend2}
  n7 = FunctionInput() {inputName=WrapDiffuseMask}
  n8 = FunctionInput() {inputName=EmissiveMask}
  n9 = StaticSwitch(
   True=ComponentMask(None=@n8) {r=true}
   False=ComponentMask(None=@n8) {g=true}
   Value=StaticBoolParameter() {desc=Uses Green By default, group=Primary Triplaner Emissive, parameterName=Use R Channel For Emissive, sortPriority=1}
  )
  n10 = MaterialFunctionCall!(
   NormalizedLocalPosition=@n4
   ForwardDirection=@n5
   isVehicle=@n3
   IsMech=@n1
   EmissiveMask=StaticSwitchParameter(True=@n9, False=FunctionInput() {inputName=HeightMapMask, inputType=FunctionInput_Scalar}) {group=Primary Triplaner Emissive, parameterName=Ues EmissiveMask/Height, sortPriority=16}
   EmissiveMask_B=StaticSwitchParameter(True=@n2, False=ComponentMask(None=@n8) {b=true}) {group=Primary Triplaner Emissive, parameterName=Use MaskFromTriplaner 01/02, sortPriority=13}
  ) {materialFunction=&/MF_WrapEmissiveHeightBlend}
  n11 = ScalarParameter() {defaultValue=1, desc=Adjust the spread of the gradent, group=Primary Triplaner Emissive, parameterName=Emissive PanSpread, sortPriority=7}
  n12 = MaterialFunctionCall(Float3=Multiply(A=@n4, B=@n5)) {materialFunction=&BreakOutFloat3Components}
  n13 = LinearInterpolate(A=@n12.G, B=@n12.G, Alpha=ComponentMask(None=@n3) {r=true})
  n14 = LinearInterpolate(
   A=ScalarParameter() {group=Primary Triplaner Emissive, parameterName=Emissive PanWeaponOffset, sortPriority=8}
   B=ScalarParameter() {group=Primary Triplaner Emissive, parameterName=Emissive PanVehicleOffset, sortPriority=8}
   Alpha=@n3
  )
  n15 = ScalarParameter() {group=Primary Triplaner Emissive, parameterName=Emissive PanMechOffset, sortPriority=8}
  n16 = Multiply(A=@n11, B=Add(A=LinearInterpolate(A=@n13, B=@n12.B, Alpha=@n1), B=LinearInterpolate(A=@n14, B=@n15, Alpha=@n1)))
  n17 = ScalarParameter() {defaultValue=1, group=Primary Triplaner Emissive, parameterName=Emis_PanningMaskSineMax, sortPriority=7}
  n18 = Constant() {r=1}
  n19 = ScalarParameter() {group=Primary Triplaner Emissive, parameterName=Emis_PanningMaskSineMin, sortPriority=7}
  n20 = Time()
  n21 = ScalarParameter() {defaultValue=1, group=Primary Triplaner Emissive, parameterName=Emis_PanningMaskPanSpeed_Y, sortPriority=22}
  n22 = StaticSwitchParameter(
   True=AppendVector(A=@n21, B=ScalarParameter() {group=Primary Triplaner Emissive, parameterName=Emis_PanningMaskPanSpeed_X, sortPriority=21})
   False=@n21
  ) {group=Primary Triplaner Emissive, parameterName=Emis_PanningMaskFullTexture, sortPriority=-3}
  n23 = Subtract(A=Sine(None=Multiply(A=@n20, B=@n22)))
  n24 = Add(B=LinearInterpolate(A=AppendVector(A=@n17, B=@n18), B=AppendVector(A=@n19, B=@n18), Alpha=Divide(A=@n23)))
  n25 = Multiply(A=Time(), B=ScalarParameter() {defaultValue=0.5, group=Primary Triplaner Emissive, parameterName=Emissive PanSpeed, sortPriority=8})
  n26 = ChannelMaskParameter(None=@n8) {group=Primary Triplaner Emissive, maskChannel=Blue, parameterName=UVOffsetChannel, defaultValue=<rgba 0 0 1 0>}
  n27 = ScalarParameter() {group=Primary Triplaner Emissive, parameterName=UVOffsetU}
  n28 = ScalarParameter() {group=Primary Triplaner Emissive, parameterName=UVOffsetV}
  n29 = Add(
   A=StaticSwitchParameter(True=Add(A=@n16, B=@n24), False=Add(A=@n16, B=@n25)) {group=Primary Triplaner Emissive, parameterName=Emissve_Sine_pan, sortPriority=-3}
   B=Multiply(A=@n26.R, B=AppendVector(A=@n27, B=@n28))
  )
  n30 = ScalarParameter() {desc=Warp uvs useing mask source, group=Primary Triplaner Emissive, parameterName="WarpEmissiveUvs (mask Channel)", sortPriority=9}
  n31 = Multiply(A=@n9, B=@n30)
  n32 = ScalarParameter() {group=Primary Triplaner Emissive, parameterName=UseMultiplyGradient}
  n33 = TextureObjectParameter() {
   group=Primary Triplaner Emissive
   parameterName=Emissive Gradient
   samplerType=SAMPLERTYPE_Masks
   sortPriority=6
   texture=/Game/Global/Textures/Default/Masks_and_Gradients/T_VertGradient_MASK
  }
  n34 = TextureSample(UVs=LinearInterpolate(A=Add(A=@n29, B=@n31), B=@n29, Alpha=@n32), Tex=@n33) {samplerSource=SSM_Wrap_WorldGroupSettings, texture=/Game/Weapons/WeaponSkins/Textures/T_Swirly_Noise_highContrast}
  n35 = Multiply(A=@n34.R, B=@n31)
  n36 = VectorParameter() {group=Primary Triplaner Emissive, parameterName=Emissive Color, sortPriority=3, defaultValue=<rgba 0 0 0 1>}
  n37 = Multiply(A=@n9, B=@n36.RGB)
  n38 = StaticSwitch(
   True=Multiply(A=LinearInterpolate(A=@n34.R, B=@n35, Alpha=@n32), B=@n37)
   False=@n37
   Value=StaticBoolParameter() {group=Primary Triplaner Emissive, parameterName=Emissive Panning, sortPriority=5}
  )
  n39 = Multiply(A=@n38, B=OneMinus(None=FunctionInput() {inputName=UseMaskGradent, inputType=FunctionInput_Scalar}))
  n40 = StaticBoolParameter() {group=Primary Triplaner Emissive, parameterName=MaskEmissiveWithGradent, sortPriority=2}
  n41 = StaticBoolParameter() {group=Primary Triplaner Emissive, parameterName="Emissive (Uses Mask for Metal and Emissive)", sortPriority=-1}
  n42 = StaticBoolParameter() {
   desc=Uses Alpha From Triplaner 01 for emissive
   group=Primary Triplaner Emissive
   parameterName=-Triplaner-01-Emissive-Height Blend-------------
   sortPriority=15
  }
  n43 = StaticSwitch(
   True=StaticSwitchParameter(True=LinearInterpolate(A=@n10.Result, Alpha=@n7) {constB=0}, False=@n10.Result) {group=Primary Triplaner Emissive, parameterName=Triplaner 01 Emissive masked By DiffusePanning, sortPriority=15}
   False=StaticSwitch(
    True=StaticSwitch(True=@n39, False=@n38, Value=@n40)
    False=Constant()
    Value=FunctionInput(Preview=@n41) {inputName=UseMaskForEmissive, inputType=FunctionInput_StaticBool}
   )
   Value=@n42
  )
  n44 = LinearInterpolate(A=@n43, B=@n6.Result, Alpha=@n6.Mask)
  n45 = Add(A=@n6.Result, B=@n43)
  n46 = StaticBoolParameter() {
   desc=Add or mask it out from the other emmisve
   group=Primary Triplaner Emissive
   parameterName="Mask Tripaner 01 Emissive form Tri 02 "
   sortPriority=20
  }
  n47 = StaticBoolParameter() {
   desc=Uses Alpha From Triplaner 02for emissive
   group=Primary Triplaner Emissive
   parameterName=-Triplaner-02-Emissive-Height-Blend-------------
   sortPriority=50
  }
  FunctionOutput(None=StaticSwitch(True=StaticSwitch(True=@n44, False=@n45, Value=@n46), False=@n43, Value=@n47))
  FunctionOutput(None=@n10.Mask) {outputName=Mask}
  NamedRerouteDeclaration(None=FunctionInput() {bUsePreviewValueAsDefault=true, inputName=IsCharacter, inputType=FunctionInput_Scalar}) {name=IsCharacter, nodeColor=<rgba 0 0.9253 1 1>}
  n48 = FunctionInput() {bUsePreviewValueAsDefault=true, inputName=IsStatsB, inputType=FunctionInput_Vector2}
  NamedRerouteDeclaration(None=ComponentMask(None=@n48) {r=true}) {name=IsBackpack, nodeColor=<rgba 0 0.9253 1 1>}
  NamedRerouteDeclaration(None=ComponentMask(None=@n48) {g=true}) {name=IsPickaxe, nodeColor=<rgba 0 0.9253 1 1>}
  NamedRerouteDeclaration(None=@n48) {name=IsStatsB, nodeColor=<rgba 0 0.9253 1 1>}
 ]}
