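# ThresholdBlend is a material function in Fortress shared materials that optionally applies an adjustable threshold offset to a scalar mask, outputting a gradient.
# It subtracts the default constant from the mask, then adds the saturated threshold amount multiplied by blend rate and clamped to a maximum.
# The UseThresholdBlend static switch selects either that saturated result or the untouched mask.
# Its parameters sit in the Primary Triplaner Gradient group.
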MaterialFunction /Game/Packages/Fortress_SharedMaterials/Base_Material_Functions/BaseCustomizationFunctions/ThresholdBlend
graph {nodes [
  n1 = FunctionInput() {inputName=Mask, inputType=FunctionInput_Scalar, sortPriority=6}
  n2 = Saturate(None=ScalarParameter() {group=Primary Triplaner Gradient, parameterName=GradientThresholdAmount, sortPriority=-9})
  n3 = ScalarParameter() {defaultValue=2, group=Primary Triplaner Gradient, parameterName=BlendRate, sortPriority=-8}
  n4 = Add(
   A=Subtract(A=@n1)
   B=Clamp(
    None=Multiply(A=@n2, B=@n3)
    Max=ScalarParameter() {defaultValue=1, group=Primary Triplaner Gradient, parameterName=Gradient Threshold Max, sortPriority=-7}
   )
  )
  FunctionOutput(None=StaticSwitch(
    True=Saturate(None=@n4)
    False=@n1
    Value=StaticBoolParameter() {group=Primary Triplaner Gradient, parameterName=UseThresholdBlend, sortPriority=-10}
   )) {outputName=Gradient, sortPriority=1}
 ]}
