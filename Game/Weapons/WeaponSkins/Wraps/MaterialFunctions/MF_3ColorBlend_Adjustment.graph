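# Material function for weapon wrap skins that blends three Vector4 colors A, B and C along a scalar Alpha.
# Saturated SmoothSteps on Alpha and its inverse mix A toward B and C toward B, then Alpha lerps between both results.
# A_B_Min_Max supplies R and B as both Min and Max of their SmoothSteps, giving hard thresholds.
# G and A only feed Power branches that never reach the output.
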MaterialFunction /Game/Weapons/WeaponSkins/Wraps/MaterialFunctions/MF_3ColorBlend_Adjustment
graph {nodes [
  Constant() {r=1}
  n1 = FunctionInput() {bUsePreviewValueAsDefault=true, inputName=A, inputType=FunctionInput_Vector4, previewValue=<1 0 0 1>}
  n2 = FunctionInput() {bUsePreviewValueAsDefault=true, inputName=B, inputType=FunctionInput_Vector4, sortPriority=1, previewValue=<0 1 0 1>}
  n3 = MaterialFunctionCall(Float4=FunctionInput(Preview=Constant4Vector() {constant=<rgba 0 0.5 0 0.5>}.RGBA) {desc="A min\\ max     B min \\Max", inputName=A_B_Min_Max, inputType=FunctionInput_Vector4}) {materialFunction=&BreakOutFloat4Components}
  n4 = FunctionInput(Preview=MaterialFunctionCall() {materialFunction=&LinearGradient}.UGradient) {bUsePreviewValueAsDefault=true, inputName=Alpha, inputType=FunctionInput_Scalar, sortPriority=4}
  n5 = Saturate(None=SmoothStep(Min=@n3.R, Max=@n3.R, Value=@n4) {constMax=0.6, constValue=0.5})
  n6 = FunctionInput() {bUsePreviewValueAsDefault=true, inputName=C, inputType=FunctionInput_Vector4, sortPriority=1, previewValue=<0 0 1 1>}
  n7 = Saturate(None=SmoothStep(Min=@n3.B, Max=@n3.B, Value=OneMinus(None=@n4)) {constMax=0.6})
  FunctionOutput(None=LinearInterpolate(A=LinearInterpolate(A=@n1, B=@n2, Alpha=@n5), B=LinearInterpolate(A=@n6, B=@n2, Alpha=@n7), Alpha=@n4) {desc="Blend it!"})
  n8 = OneMinus(None=@n4)
  Saturate(None=Power(Base=Multiply(A=@n8, B=@n3.B), Exp=@n3.A) {constExponent=1})
  Saturate(None=Power(Base=Multiply(A=@n4, B=@n3.R) {constB=2}, Exp=@n3.G) {constExponent=1})
 ]}
