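# Weapon wrap material function that fakes three lights on the mobile shading path and passes attributes through unchanged elsewhere.
# MPC_3LightsMobile drives each light, with static switches for spot or point and object-relative position; only Light1 flickers.
# Custom HLSL computes diffuse and Phong specular from vertex normals, shaped by Metallic, Roughness and Specular, with linear falloff.
# The clamped sum is multiplied by BaseColor and added to Emissive.
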MaterialFunction /Game/Weapons/WeaponSkins/Wraps/MaterialFunctions/MF_3LightsMobile
graph {nodes [
  n1 = MaterialFunctionCall() {materialFunction=/Engine/Functions/Engine_MaterialFunctions02/WorldPositionOffset/ObjectScale}
  n2 = Max(A=Max(A=@n1.Scale X, B=@n1.Scale X), B=@n1.Scale Z)
  Divide(A=Constant() {r=1}, B=@n2)
  n3 = FunctionInput() {inputType=FunctionInput_MaterialAttributes}
  n4 = GetMaterialAttributes(None=@n3) {attributeGetTypes [
    #1
    #2
    #3
    #4
    #5
   ]}
  MaterialExpressionWorldPosition_0 = WorldPosition()
  n5 = ComponentMask(None=CollectionParameter() {collection=&/MPC_3LightsMobile, parameterName=Light1Pos}) {b=true, g=true, r=true}
  n6 = ObjectPositionWS()
  MaterialExpressionComponentMask_0 = ComponentMask(None=StaticSwitchParameter(True=Add(A=@n5, B=@n6), False=@n5) {parameterName=3lightLocalPosition}) {b=true, g=true, r=true}
  MaterialExpressionCustom_11 = Custom!(WorldPos=@MaterialExpressionWorldPosition_0.XYZ, LightPosition=@MaterialExpressionComponentMask_0) {
   code="return normalize(WorldPos - LightPosition);"
   description=LightToPixel
   inputs [
    {inputName=WorldPos, input {expression=&self:MaterialExpressionWorldPosition_0}}
    {
     inputName=LightPosition
     input {expression=&self:MaterialExpressionComponentMask_0}
    }
   ]
  }
  MaterialExpressionNormalize_1 = Normalize(VectorInput=ComponentMask(None=CollectionParameter() {collection=&MPC_3LightsMobile, parameterName=Light1Dir}) {b=true, g=true, r=true})
  MaterialExpressionCollectionParameter_10 = CollectionParameter() {collection=&MPC_3LightsMobile, parameterName=Light1SpotCutoff}
  MaterialExpressionComponentMask_2 = ComponentMask(None=CollectionParameter() {collection=&MPC_3LightsMobile, parameterName=Light1Color}) {b=true, g=true, r=true}
  n7 = OneMinus(None=Multiply(A=@n4.Metallic, B=ScalarParameter() {defaultValue=0.5, parameterName=MetallicDiffuseAttenuation}))
  n8 = Time()
  n9 = CollectionParameter() {collection=&MPC_3LightsMobile, parameterName=Light1FlickerSpeed}
  n10 = ConstantBiasScale(None=Sine(None=Multiply(A=@n8, B=@n9)))
  n11 = OneMinus(None=CollectionParameter() {collection=&MPC_3LightsMobile, parameterName=Light1FlickerIntensity})
  n12 = Multiply(
   A=Saturate(None=Add(A=@n10, B=@n11))
   B=CollectionParameter() {collection=&MPC_3LightsMobile, parameterName=Light1Intensity}
  )
  MaterialExpressionMultiply_6 = Multiply(A=@n7, B=@n12)
  MaterialExpressionVertexNormalWS_0 = VertexNormalWS()
  MaterialExpressionCameraPositionWS_0 = CameraPositionWS()
  MaterialExpressionMultiply_2 = Multiply(A=OneMinus(None=@n4.Roughness), B=ScalarParameter() {defaultValue=24, parameterName=SpecExpMax})
  MaterialExpressionMultiply_3 = Multiply(A=@n4.Specular, B=@n12)
  MaterialExpressionCustom_1 = Custom!(
   Color=@MaterialExpressionComponentMask_2
   DiffuseIntensity=@MaterialExpressionMultiply_6
   LightToPixel=@MaterialExpressionCustom_11
   Normal=@MaterialExpressionVertexNormalWS_0
   CameraPos=@MaterialExpressionCameraPositionWS_0
   WorldPos=@MaterialExpressionWorldPosition_0.XYZ
   SpecularPower=@MaterialExpressionMultiply_2
   SpecularIntensity=@MaterialExpressionMultiply_3
  ) {
   code="    float DiffuseFactor = dot(Normal, -LightToPixel);\r\n\r\n    float3 DiffuseColor = float3(0, 0, 0);\r\n    float3 SpecularColor = float3(0, 0, 0);\r\n\r\n    if (DiffuseFactor > 0) {\r\n        DiffuseColor = float3(Color * DiffuseIntensity * DiffuseFactor);\r\n        float3 VertexToEye = normalize(CameraPos - WorldPos);\r\n        float3 LightReflect = normalize(reflect(LightToPixel, Normal));\r\n        float SpecularFactor = dot(VertexToEye, LightReflect);\r\n        if (SpecularFactor > 0) {\r\n            SpecularFactor = pow(SpecularFactor, SpecularPower) /** DiffuseFactor*/;\r\n            SpecularColor = float3(Color * SpecularIntensity * SpecularFactor);\r\n        }\r\n    }\r\n\r\n    return (DiffuseColor + SpecularColor);"
   description=CalcLightInternal
   inputs [
    {inputName=Color, input {expression=&self:MaterialExpressionComponentMask_2}}
    {
     inputName=DiffuseIntensity
     input {expression=&self:MaterialExpressionMultiply_6}
    }
    {inputName=LightToPixel, input {expression="<self>1"}}
    {inputName=Normal, input {expression=&self:MaterialExpressionVertexNormalWS_0}}
    {
     inputName=CameraPos
     input {expression=&self:MaterialExpressionCameraPositionWS_0}
    }
    {inputName=WorldPos, input {expression=&self:MaterialExpressionWorldPosition_0}}
    {inputName=SpecularPower, input {expression=&self:MaterialExpressionMultiply_2}}
    {
     inputName=SpecularIntensity
     input {expression=&self:MaterialExpressionMultiply_3}
    }
   ]
  }
  MaterialExpressionCollectionParameter_11 = CollectionParameter() {collection=&MPC_3LightsMobile, parameterName=Light1Radius}
  MaterialExpressionCustom_2 = Custom!(
   CalcLightInternal=@MaterialExpressionCustom_1
   WorldPos=@MaterialExpressionWorldPosition_0.XYZ
   LightPosition=@MaterialExpressionComponentMask_0
   Radius=@MaterialExpressionCollectionParameter_11
  ) {
   code="float3 LightDirection = WorldPos - LightPosition;\r\nfloat Distance = length(LightDirection);\r\nreturn CalcLightInternal * (1 - (Distance/Radius));"
   description=CalcPointLight
   inputs [
    {
     inputName=CalcLightInternal
     input {expression=&self:MaterialExpressionCustom_1}
    }
    {inputName=WorldPos, input {expression=&self:MaterialExpressionWorldPosition_0}}
    {
     inputName=LightPosition
     input {expression=&self:MaterialExpressionComponentMask_0}
    }
    {
     inputName=Radius
     input {expression=&self:MaterialExpressionCollectionParameter_11}
    }
   ]
  }
  n13 = Custom!(
   LightToPixel=@MaterialExpressionCustom_11
   LightDirection=@MaterialExpressionNormalize_1
   LightCutoff=@MaterialExpressionCollectionParameter_10
   CalcPointLight=@MaterialExpressionCustom_2
  ) {
   code="float SpotFactor = dot(LightToPixel, LightDirection);\r\n\r\nif (SpotFactor > LightCutoff) {\r\n\r\n\treturn CalcPointLight * (1.0 - (1.0 - SpotFactor) * 1.0/(1.0 - LightCutoff));\r\n}\r\nelse \r\n{\r\n\treturn float3(0,0,0);\r\n}"
   description=CalcSpotLight
   inputs [
    {inputName=LightToPixel, input {expression=&self:MaterialExpressionCustom_11}}
    {
     inputName=LightDirection
     input {expression=&self:MaterialExpressionNormalize_1}
    }
    {
     inputName=LightCutoff
     input {expression=&self:MaterialExpressionCollectionParameter_10}
    }
    {inputName=CalcPointLight, input {expression=&self:MaterialExpressionCustom_2}}
   ]
  }
  n14 = StaticSwitchParameter(True=@n13, False=@MaterialExpressionCustom_2) {group=Light1, parameterName=Light1IsSpot}
  n15 = ComponentMask(None=CollectionParameter() {collection=&MPC_3LightsMobile, parameterName=Light2Pos}) {b=true, g=true, r=true}
  n16 = ObjectPositionWS()
  MaterialExpressionComponentMask_3 = ComponentMask(None=StaticSwitchParameter(True=Add(A=@n15, B=@n16), False=@n15) {parameterName=3lightLocalPosition}) {b=true, g=true, r=true}
  MaterialExpressionCustom_10 = Custom!(WorldPos=@MaterialExpressionWorldPosition_0.XYZ, LightPosition=@MaterialExpressionComponentMask_3) {
   code="return normalize(WorldPos - LightPosition);"
   description=LightToPixel
   inputs [
    {inputName=WorldPos, input {expression=&self:MaterialExpressionWorldPosition_0}}
    {
     inputName=LightPosition
     input {expression=&self:MaterialExpressionComponentMask_3}
    }
   ]
  }
  MaterialExpressionNormalize_0 = Normalize(VectorInput=ComponentMask(None=CollectionParameter() {collection=&MPC_3LightsMobile, parameterName=Light2Dir}) {b=true, g=true, r=true})
  MaterialExpressionCollectionParameter_13 = CollectionParameter() {collection=&MPC_3LightsMobile, parameterName=Light2SpotCutoff}
  MaterialExpressionComponentMask_5 = ComponentMask(None=CollectionParameter() {collection=&MPC_3LightsMobile, parameterName=Light2Color}) {b=true, g=true, r=true}
  n17 = CollectionParameter() {collection=&MPC_3LightsMobile, parameterName=Light2Intensity}
  MaterialExpressionMultiply_7 = Multiply(A=@n7, B=@n17)
  MaterialExpressionMultiply_4 = Multiply(A=@n4.Specular, B=@n17)
  MaterialExpressionCustom_4 = Custom!(
   Color=@MaterialExpressionComponentMask_5
   DiffuseIntensity=@MaterialExpressionMultiply_7
   LightToPixel=@MaterialExpressionCustom_10
   Normal=@MaterialExpressionVertexNormalWS_0
   CameraPos=@MaterialExpressionCameraPositionWS_0
   WorldPos=@MaterialExpressionWorldPosition_0.XYZ
   SpecularPower=@MaterialExpressionMultiply_2
   SpecularIntensity=@MaterialExpressionMultiply_4
  ) {
   code="    float DiffuseFactor = dot(Normal, -LightToPixel);\r\n\r\n    float3 DiffuseColor = float3(0, 0, 0);\r\n    float3 SpecularColor = float3(0, 0, 0);\r\n\r\n    if (DiffuseFactor > 0) {\r\n        DiffuseColor = float3(Color * DiffuseIntensity * DiffuseFactor);\r\n        float3 VertexToEye = normalize(CameraPos - WorldPos);\r\n        float3 LightReflect = normalize(reflect(LightToPixel, Normal));\r\n        float SpecularFactor = dot(VertexToEye, LightReflect);\r\n        if (SpecularFactor > 0) {\r\n            SpecularFactor = pow(SpecularFactor, SpecularPower) /** DiffuseFactor*/;\r\n            SpecularColor = float3(Color * SpecularIntensity * SpecularFactor);\r\n        }\r\n    }\r\n\r\n    return (DiffuseColor + SpecularColor);"
   description=CalcLightInternal
   inputs [
    {inputName=Color, input {expression=&self:MaterialExpressionComponentMask_5}}
    {
     inputName=DiffuseIntensity
     input {expression=&self:MaterialExpressionMultiply_7}
    }
    {inputName=LightToPixel, input {expression=&self:MaterialExpressionCustom_10}}
    {inputName=Normal, input {expression=&self:MaterialExpressionVertexNormalWS_0}}
    {
     inputName=CameraPos
     input {expression=&self:MaterialExpressionCameraPositionWS_0}
    }
    {inputName=WorldPos, input {expression=&self:MaterialExpressionWorldPosition_0}}
    {inputName=SpecularPower, input {expression=&self:MaterialExpressionMultiply_2}}
    {
     inputName=SpecularIntensity
     input {expression=&self:MaterialExpressionMultiply_4}
    }
   ]
  }
  MaterialExpressionCollectionParameter_14 = CollectionParameter() {collection=&MPC_3LightsMobile, parameterName=Light2Radius}
  MaterialExpressionCustom_3 = Custom!(
   CalcLightInternal=@MaterialExpressionCustom_4
   WorldPos=@MaterialExpressionWorldPosition_0.XYZ
   LightPosition=@MaterialExpressionComponentMask_3
   Radius=@MaterialExpressionCollectionParameter_14
  ) {
   code="float3 LightDirection = WorldPos - LightPosition;\r\nfloat Distance = length(LightDirection);\r\nreturn CalcLightInternal * (1 - (Distance/Radius));"
   description=CalcPointLight
   inputs [
    {
     inputName=CalcLightInternal
     input {expression=&self:MaterialExpressionCustom_4}
    }
    {inputName=WorldPos, input {expression=&self:MaterialExpressionWorldPosition_0}}
    {
     inputName=LightPosition
     input {expression=&self:MaterialExpressionComponentMask_3}
    }
    {
     inputName=Radius
     input {expression=&self:MaterialExpressionCollectionParameter_14}
    }
   ]
  }
  n18 = Custom!(
   LightToPixel=@MaterialExpressionCustom_10
   LightDirection=@MaterialExpressionNormalize_0
   LightCutoff=@MaterialExpressionCollectionParameter_13
   CalcPointLight=@MaterialExpressionCustom_3
  ) {
   code="float SpotFactor = dot(LightToPixel, LightDirection);\r\n\r\nif (SpotFactor > LightCutoff) {\r\n\r\n\treturn CalcPointLight * (1.0 - (1.0 - SpotFactor) * 1.0/(1.0 - LightCutoff));\r\n}\r\nelse \r\n{\r\n\treturn float3(0,0,0);\r\n}"
   description=CalcSpotLight
   inputs [
    {inputName=LightToPixel, input {expression=&self:MaterialExpressionCustom_10}}
    {
     inputName=LightDirection
     input {expression=&self:MaterialExpressionNormalize_0}
    }
    {
     inputName=LightCutoff
     input {expression=&self:MaterialExpressionCollectionParameter_13}
    }
    {inputName=CalcPointLight, input {expression=&self:MaterialExpressionCustom_3}}
   ]
  }
  n19 = StaticSwitchParameter(True=@n18, False=@MaterialExpressionCustom_3) {group=Light2, parameterName=Light2IsSpot}
  n20 = ComponentMask(None=CollectionParameter() {collection=&MPC_3LightsMobile, parameterName=Light3Pos}) {b=true, g=true, r=true}
  n21 = ObjectPositionWS()
  MaterialExpressionComponentMask_6 = ComponentMask(None=StaticSwitchParameter(True=Add(A=@n20, B=@n21), False=@n20) {parameterName=3lightLocalPosition}) {b=true, g=true, r=true}
  MaterialExpressionCustom_9 = Custom!(WorldPos=@MaterialExpressionWorldPosition_0.XYZ, LightPosition=@MaterialExpressionComponentMask_6) {
   code="return normalize(WorldPos - LightPosition);"
   description=LightToPixel
   inputs [
    {inputName=WorldPos, input {expression=&self:MaterialExpressionWorldPosition_0}}
    {
     inputName=LightPosition
     input {expression=&self:MaterialExpressionComponentMask_6}
    }
   ]
  }
  MaterialExpressionNormalize_2 = Normalize(VectorInput=ComponentMask(None=CollectionParameter() {collection=&MPC_3LightsMobile, parameterName=Light3Dir}) {b=true, g=true, r=true})
  MaterialExpressionCollectionParameter_18 = CollectionParameter() {collection=&MPC_3LightsMobile, parameterName=Light3SpotCutoff}
  MaterialExpressionComponentMask_8 = ComponentMask(None=CollectionParameter() {collection=&MPC_3LightsMobile, parameterName=Light3Color}) {b=true, g=true, r=true}
  n22 = CollectionParameter() {collection=&MPC_3LightsMobile, parameterName=Light3Intensity}
  MaterialExpressionMultiply_8 = Multiply(A=@n7, B=@n22)
  MaterialExpressionMultiply_5 = Multiply(A=@n4.Specular, B=@n22)
  MaterialExpressionCustom_7 = Custom!(
   Color=@MaterialExpressionComponentMask_8
   DiffuseIntensity=@MaterialExpressionMultiply_8
   LightToPixel=@MaterialExpressionCustom_9
   Normal=@MaterialExpressionVertexNormalWS_0
   CameraPos=@MaterialExpressionCameraPositionWS_0
   WorldPos=@MaterialExpressionWorldPosition_0.XYZ
   SpecularPower=@MaterialExpressionMultiply_2
   SpecularIntensity=@MaterialExpressionMultiply_5
  ) {
   code="    float DiffuseFactor = dot(Normal, -LightToPixel);\r\n\r\n    float3 DiffuseColor = float3(0, 0, 0);\r\n    float3 SpecularColor = float3(0, 0, 0);\r\n\r\n    if (DiffuseFactor > 0) {\r\n        DiffuseColor = float3(Color * DiffuseIntensity * DiffuseFactor);\r\n        float3 VertexToEye = normalize(CameraPos - WorldPos);\r\n        float3 LightReflect = normalize(reflect(LightToPixel, Normal));\r\n        float SpecularFactor = dot(VertexToEye, LightReflect);\r\n        if (SpecularFactor > 0) {\r\n            SpecularFactor = pow(SpecularFactor, SpecularPower) /** DiffuseFactor*/;\r\n            SpecularColor = float3(Color * SpecularIntensity * SpecularFactor);\r\n        }\r\n    }\r\n\r\n    return (DiffuseColor + SpecularColor);"
   description=CalcLightInternal
   inputs [
    {inputName=Color, input {expression=&self:MaterialExpressionComponentMask_8}}
    {
     inputName=DiffuseIntensity
     input {expression=&self:MaterialExpressionMultiply_8}
    }
    {inputName=LightToPixel, input {expression=&self:MaterialExpressionCustom_9}}
    {inputName=Normal, input {expression=&self:MaterialExpressionVertexNormalWS_0}}
    {
     inputName=CameraPos
     input {expression=&self:MaterialExpressionCameraPositionWS_0}
    }
    {inputName=WorldPos, input {expression=&self:MaterialExpressionWorldPosition_0}}
    {inputName=SpecularPower, input {expression=&self:MaterialExpressionMultiply_2}}
    {
     inputName=SpecularIntensity
     input {expression=&self:MaterialExpressionMultiply_5}
    }
   ]
  }
  MaterialExpressionCollectionParameter_17 = CollectionParameter() {collection=&MPC_3LightsMobile, parameterName=Light3Radius}
  MaterialExpressionCustom_6 = Custom!(
   CalcLightInternal=@MaterialExpressionCustom_7
   WorldPos=@MaterialExpressionWorldPosition_0.XYZ
   LightPosition=@MaterialExpressionComponentMask_6
   Radius=@MaterialExpressionCollectionParameter_17
  ) {
   code="float3 LightDirection = WorldPos - LightPosition;\r\nfloat Distance = length(LightDirection);\r\nreturn CalcLightInternal * (1 - (Distance/Radius));"
   description=CalcPointLight
   inputs [
    {
     inputName=CalcLightInternal
     input {expression=&self:MaterialExpressionCustom_7}
    }
    {inputName=WorldPos, input {expression=&self:MaterialExpressionWorldPosition_0}}
    {
     inputName=LightPosition
     input {expression=&self:MaterialExpressionComponentMask_6}
    }
    {
     inputName=Radius
     input {expression=&self:MaterialExpressionCollectionParameter_17}
    }
   ]
  }
  n23 = Custom!(
   LightToPixel=@MaterialExpressionCustom_9
   LightDirection=@MaterialExpressionNormalize_2
   LightCutoff=@MaterialExpressionCollectionParameter_18
   CalcPointLight=@MaterialExpressionCustom_6
  ) {
   code="float SpotFactor = dot(LightToPixel, LightDirection);\r\n\r\nif (SpotFactor > LightCutoff) {\r\n\r\n\treturn CalcPointLight * (1.0 - (1.0 - SpotFactor) * 1.0/(1.0 - LightCutoff));\r\n}\r\nelse \r\n{\r\n\treturn float3(0,0,0);\r\n}"
   description=CalcSpotLight
   inputs [
    {inputName=LightToPixel, input {expression=&self:MaterialExpressionCustom_9}}
    {
     inputName=LightDirection
     input {expression=&self:MaterialExpressionNormalize_2}
    }
    {
     inputName=LightCutoff
     input {expression=&self:MaterialExpressionCollectionParameter_18}
    }
    {inputName=CalcPointLight, input {expression=&self:MaterialExpressionCustom_6}}
   ]
  }
  n24 = Add(
   A=Add(A=Max(A=@n14) {constB=0}, B=Max(A=@n19) {constB=0}) {constB=0}
   B=Max(A=StaticSwitchParameter(True=@n23, False=@MaterialExpressionCustom_6) {group=Light3, parameterName=Light3IsSpot}) {constB=0}
  ) {constB=0}
  n25 = Add(A=@n4.EmissiveColor, B=Multiply(A=@n4.BaseColor, B=@n24))
  FunctionOutput(None=ShadingPathSwitch(
    Default=@n3
    Mobile=SetMaterialAttributes!(MaterialAttributes=@n4.MaterialAttributes, EmissiveColor=@n25) {attributeSetTypes [#1]}
   ) {desc=Migrated from FeatureLevelSwitch})
  Multiply(A=@n2) {constB=0.01}
  PixelNormalWS()
  ScalarParameter() {parameterName=Param_4}
  VertexInterpolator(VS=@n24)
 ]}
