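# Material function for weapon skin wraps that composites two gear layers, E over F, each with its own UVs, mask UVs and tint.
# Mask red times gear texture alpha drives the blends for color, combined mask, tint-alpha depth, and the Normals texture over the BackGroundMask input.
# Emissive takes a selectable Normals channel, masked and colored, with switches for clamping and a red-indexed gradient texture lookup.
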MaterialFunction /Game/Weapons/WeaponSkins/Wraps/MaterialFunctions/MF_AlDay_Gears
graph {nodes [
  n1 = TextureObject() {texture=&T_BlacksRGB}
  n2 = FunctionInput(Preview=@n1) {bUsePreviewValueAsDefault=true, inputName=Gear_mask_Red, inputType=FunctionInput_Texture2D}
  n3 = TextureSample(UVs=FunctionInput() {bUsePreviewValueAsDefault=true, inputName=Uvs_GearF_Mask, inputType=FunctionInput_Vector2, sortPriority=12}, Tex=@n2) {texture=&T_BlacksRGB}
  n4 = FunctionInput() {bUsePreviewValueAsDefault=true, inputName=Uvs_GearF, inputType=FunctionInput_Vector2, sortPriority=13}
  n5 = FunctionInput(Preview=@n1) {bUsePreviewValueAsDefault=true, inputName=Gears, inputType=FunctionInput_Texture2D, sortPriority=1}
  n6 = TextureSample(UVs=@n4, Tex=@n5) {texture=&T_BlacksRGB}
  n7 = Multiply(A=@n3.R, B=@n6.A)
  n8 = Multiply(A=@n7, B=@n6.RGB)
  n9 = VectorParameter() {group=Wrap_AldGears, parameterName=GearF_Tint, defaultValue=<rgba 0.2 0.2 0.2 0.5>}
  n10 = FunctionInput() {bUsePreviewValueAsDefault=true, inputName=Uvs_GearE, inputType=FunctionInput_Vector2, sortPriority=11}
  n11 = TextureSample(UVs=@n10, Tex=@n5) {texture=&T_BlacksRGB}
  n12 = TextureSample(UVs=FunctionInput() {bUsePreviewValueAsDefault=true, inputName=Uvs_GearE_Mask, inputType=FunctionInput_Vector2, sortPriority=10}, Tex=@n2) {texture=&T_BlacksRGB}
  n13 = Multiply(A=@n12.R, B=@n11.A)
  n14 = Multiply(A=@n11.RGB, B=@n13)
  n15 = VectorParameter() {group=Wrap_AldGears, parameterName=GearE_Tint, defaultValue=<rgba 1.5 1.5 1.5 0.9>}
  FunctionOutput(None=LinearInterpolate(A=Multiply(A=@n8, B=@n9.RGB) {constB=0.25}, B=Multiply(A=@n14, B=@n15.RGB), Alpha=@n13)) {outputName=Combined_Output}
  FunctionOutput(None=Saturate(None=Add(A=@n13, B=@n7))) {outputName=Gear_Mask, sortPriority=1}
  n16 = FunctionInput() {bUsePreviewValueAsDefault=true, inputName=BackGroundMask, previewValue=<1 1 1 1>}
  n17 = FunctionInput() {inputName=Normals, inputType=FunctionInput_Texture2D, sortPriority=1}
  n18 = TextureSample(UVs=@n4, Tex=@n17) {texture=&T_BlacksRGB}
  n19 = TextureSample(UVs=@n10, Tex=@n17) {texture=&T_BlacksRGB}
  FunctionOutput(None=LinearInterpolate(A=LinearInterpolate(A=@n16, B=@n18.RGB, Alpha=@n7), B=@n19.RGB, Alpha=@n13)) {outputName=Texture_Non_alpha, sortPriority=1}
  FunctionOutput(None=LinearInterpolate(A=@n9.A, B=@n15.A, Alpha=@n13)) {outputName=Gear_Mask_Depth, sortPriority=1}
  n20 = ChannelMaskParameter(None=@n19.RGB) {group=Wrap_AldGears, maskChannel=Green, parameterName=Emissive_Channel_Mask, sortPriority=55, defaultValue=<rgba 0 1 0 0>}
  n21 = VectorParameter() {group=Wrap_AldGears, parameterName=Emissive_E_ColorCW, sortPriority=56, defaultValue=<rgba 1.5 1.5 1.5 0.9>}
  n22 = ChannelMaskParameter(None=@n18.RGB) {group=Wrap_AldGears, maskChannel=Green, parameterName=Emissive_Channel_Mask, sortPriority=55, defaultValue=<rgba 0 1 0 0>}
  n23 = VectorParameter() {group=Wrap_AldGears, parameterName=Emissive_E_ColorCW, sortPriority=56, defaultValue=<rgba 1.5 1.5 1.5 0.9>}
  n24 = Add(A=Multiply(A=Multiply(A=@n12.R, B=@n20.R), B=@n21.RGB), B=Multiply(A=Multiply(A=@n3.R, B=@n22.R), B=@n23.RGB))
  n25 = StaticSwitchParameter(True=Clamp(None=@n24) {maxDefault=0.99}, False=@n24) {group=Wrap_ClockWork, parameterName=Emissive_Clamp_Range, sortPriority=52}
  n26 = TextureSampleParameter2D(UVs=ComponentMask(None=@n25) {r=true}) {group=Wrap_ClockWork, parameterName=Emissive_GradCW, sortPriority=51, texture=/Game/Global/Textures/Default/Masks_and_Gradients/3pxRGBImage}
  n27 = Multiply(A=@n26.A, B=ScalarParameter() {defaultValue=1, group=Wrap_ClockWork, parameterName=Emissive_Grad_Intensity, sortPriority=51})
  FunctionOutput(None=StaticSwitchParameter(True=Multiply(A=@n26.RGB, B=@n27), False=@n25) {group=Wrap_ClockWork, parameterName=Emissive_Trexture_Grad, sortPriority=50}) {outputName=Emissive, sortPriority=1}
  Multiply()
 ]}
