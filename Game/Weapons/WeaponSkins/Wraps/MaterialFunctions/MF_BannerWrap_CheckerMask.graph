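# Material function for banner weapon wraps that stamps a banner icon into alternating cells of a checkerboard.
# The checker comes from rotated input UVs, while the icon tiles use rotated or raw UVs via a static switch, scaled about each cell center.
# A second, slightly offset texture sample gives a shadow channel beside the icon, both cropped by per-axis SphereMasks.
# Several other pattern branches, including the OffsetLoc one, never reach the single Banner output.
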MaterialFunction /Game/Weapons/WeaponSkins/Wraps/MaterialFunctions/MF_BannerWrap_CheckerMask
graph {nodes [
  n1 = Constant() {r=0.5}
  n2 = PreSkinnedPosition()
  n3 = Multiply(A=Constant3Vector() {constant=<rgba 1 1 1 0>}, B=Constant() {r=22})
  n4 = FunctionInput(Preview=Divide(A=VertexInterpolator(VS=@n2).PS, B=Multiply(A=@n3) {constB=-1})) {inputName=In-UVs, inputType=FunctionInput_Vector2, sortPriority=-100}
  n5 = Rotator(
   Coordinate=@n4
   Time=FunctionInput(Preview=Constant()) {bUsePreviewValueAsDefault=true, inputName=Rotation, inputType=FunctionInput_Scalar, sortPriority=-9}
  )
  n6 = Add(
   A=FunctionInput(Preview=Constant()) {bUsePreviewValueAsDefault=true, inputName=OffsetLoc, inputType=FunctionInput_Vector2, sortPriority=-8}
   B=@n5
  )
  n7 = ComponentMask(None=@n6) {r=true}
  n8 = Constant() {r=0.5}
  n9 = Multiply(A=Ceil(None=Add(A=@n7, B=@n8) {constB=0.5}), B=@n8) {constB=0.5}
  n10 = ComponentMask(None=@n6) {g=true}
  n11 = Subtract(A=AppendVector(A=@n7, B=Add(A=@n9, B=@n10)), B=OneMinus(None=@n8)) {constB=0.5}
  n12 = Round(None=Frac(None=Multiply(A=@n1, B=@n11)) {desc=mod})
  ComponentMask(None=@n12) {g=true}
  Constant() {r=0.956}
  Constant() {r=0.021}
  Constant() {r=0.017}
  n13 = Multiply(A=Constant() {r=0.5}, B=@n5)
  n14 = MaterialFunctionCall(Float2=Round(None=Frac(None=@n13))) {materialFunction=&BreakOutFloat2Components}
  n15 = Add(A=Multiply(A=@n14.R, B=@n14.R), B=Multiply(A=OneMinus(None=@n14.R), B=OneMinus(None=@n14.G)))
  n16 = StaticBool()
  n17 = Frac(None=StaticSwitch(
    True=@n5
    False=@n4
    Value=FunctionInput(Preview=@n16) {bUsePreviewValueAsDefault=true, inputName=UseRotation, inputType=FunctionInput_StaticBool, sortPriority=-9}
   ))
  n18 = MaterialFunctionCall(
   UVs=@n17
   Texture Scale=FunctionInput(Preview=Constant() {r=0.7}) {bUsePreviewValueAsDefault=true, inputName=Scale2, inputType=FunctionInput_Scalar, sortPriority=-7}
  ) {materialFunction=&ScaleUVsByCenter}
  n19 = FunctionInput(Preview=TextureObject() {texture=/Game/UI/Foundation/Textures/Banner/Standard/T-Banners-Icons-Chevrons-L}) {inputName=Texture, inputType=FunctionInput_Texture2D, sortPriority=50}
  n20 = TextureSample(UVs=@n18.UVs, Tex=@n19) {samplerSource=SSM_Wrap_WorldGroupSettings}
  n21 = Add(A=@n18.UVs, B=Multiply(A=Constant2Vector() {g=1}, B=Constant() {r=-0.02}))
  n22 = MaterialFunctionCall(X=@n20.R, Y=Multiply(A=OneMinus(None=@n20.R), B=TextureSample(UVs=@n21, Tex=@n19) {samplerSource=SSM_Wrap_WorldGroupSettings}.R)) {materialFunction=&MakeFloat2}
  n23 = Constant() {r=0.5}
  n24 = MaterialFunctionCall(Float2=@n21) {materialFunction=&BreakOutFloat2Components}
  n25 = Constant() {r=0.48}
  n26 = Constant() {r=0.95}
  n27 = SphereMask(A=@n23, B=@n24.R, Radius=@n25, Hardness=@n26) {attenuationRadius=0.5, hardnessPercent=0}
  n28 = SphereMask(A=@n24.G, B=@n23, Radius=@n25, Hardness=@n26) {attenuationRadius=0.5, hardnessPercent=0}
  FunctionOutput(None=Multiply(A=Multiply(A=@n15, B=@n22.Result), B=Multiply(A=@n27, B=@n28))) {outputName=Banner}
  n29 = Constant() {r=-1}
  n30 = Constant2Vector() {g=0.25, r=0.25}
  n31 = Multiply(A=Constant() {r=0.5}, B=@n5)
  n32 = Frac(None=Add(A=Multiply(A=@n29, B=@n30), B=@n31))
  n33 = Constant() {r=0.96}
  n34 = MaterialFunctionCall(UVs=@n32, Texture Scale=@n33) {materialFunction=&ScaleUVsByCenter}
  n35 = Constant() {r=0.05}
  n36 = Constant()
  n37 = Frac(None=Add(A=@n30, B=@n31))
  n38 = MaterialFunctionCall(UVs=@n37, Texture Scale=@n33) {materialFunction=&ScaleUVsByCenter}
  Max(
   A=Saturate(None=SphereMask(A=@n32, B=@n34.UVs, Radius=@n35, Hardness=@n36) {attenuationRadius=0.5, hardnessPercent=0})
   B=Saturate(None=SphereMask(A=@n37, B=@n38.UVs, Radius=@n35, Hardness=@n36) {attenuationRadius=0.5, hardnessPercent=0})
  )
  n39 = Multiply(A=Constant() {r=0.25}, B=@n5)
  n40 = MaterialFunctionCall(Float2=@n39) {materialFunction=&BreakOutFloat2Components}
  n41 = Add(A=Constant() {r=0.25}, B=@n40.R)
  n42 = Round(None=Frac(None=MaterialFunctionCall(X=@n41, Y=@n40.G) {materialFunction=&MakeFloat2}.Result))
  n43 = OneMinus(None=@n42)
  n44 = Add(
   A=Multiply(A=ComponentMask(None=@n42) {r=true}, B=ComponentMask(None=@n42) {g=true})
   B=Multiply(A=ComponentMask(None=@n43) {r=true}, B=ComponentMask(None=@n43) {g=true})
  )
  n45 = Round(None=Frac(None=@n39))
  n46 = Multiply(A=ComponentMask(None=@n45) {r=true}, B=ComponentMask(None=@n45) {g=true})
  n47 = OneMinus(None=@n45)
  n48 = Multiply(A=ComponentMask(None=@n47) {r=true}, B=ComponentMask(None=@n47) {g=true})
  n49 = Add(A=OneMinus(None=Add(A=@n46, B=@n48)), B=OneMinus(None=@n15))
  Multiply(A=Saturate(None=Add(A=@n44, B=@n49)), B=@n15)
  Multiply(A=ComponentMask(None=@n12) {r=true}, B=Frac(None=@n11) {desc=mod})
  n50 = Constant() {r=0.96}
  n51 = Multiply(
   A=Add(
    A=MaterialFunctionCall(UVs=@n17, Texture Scale=@n50) {materialFunction=&ScaleUVsByCenter}.UVs
    B=Constant() {r=-0.5}
   )
   B=Constant() {r=2}
  )
  n52 = MaterialFunctionCall(Float2=OneMinus(None=Abs(None=@n51))) {materialFunction=&BreakOutFloat2Components}
  n53 = Multiply(A=@n52.R, B=@n52.R)
  n54 = Constant() {r=6}
  Saturate(None=Add(A=Multiply(A=@n53, B=@n54), B=Constant() {r=-0.02}))
 ]}
