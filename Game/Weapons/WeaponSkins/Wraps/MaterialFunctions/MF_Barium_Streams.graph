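# Material function for the Barium weapon wrap that outputs an animated radial mask of streams around an adjustable elliptical hole.
# UVs are offset by state flags and location input, then converted to radial coordinates for two differently panned and tiled samples of one mask texture.
# A distance-faded scrolling warp texture and three ring ripples at input radii distort both layers; ripples also perturb the hole.
# Both layers are hole-masked, sharply contrasted, summed and saturated.
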MaterialFunction /Game/Weapons/WeaponSkins/Wraps/MaterialFunctions/MF_Barium_Streams
graph {nodes [
  Constant2Vector() {g=0.6, r=1}
  n1 = Multiply(A=Time(), B=ComponentMask(None=FunctionInput() {inputName=Streams_Speed, sortPriority=50}) {g=true, r=true})
  n2 = Constant2Vector() {g=1, r=-0.5}
  n3 = TextureCoordinate()
  n4 = Constant2Vector() {g=0.5, r=-0.5}
  n5 = MaterialFunctionCall(Float4=FunctionInput() {inputName=isstates, inputType=FunctionInput_Vector4}) {materialFunction=&BreakOutFloat4Components}
  n6 = LinearInterpolate(A=@n4, B=LinearInterpolate(A=Constant2Vector() {g=-0.5, r=-0.5}, B=@n4, Alpha=@n5.G), Alpha=@n5.R)
  n7 = Constant2Vector() {g=0.5, r=-0.5}
  n8 = FunctionInput(Preview=Constant4Vector() {constant=<rgba -0.5 0.1 0 0>}.RGBA) {inputName=Location/Scale, inputType=FunctionInput_Vector4, sortPriority=32}
  n9 = Add(
   A=Add(A=FunctionInput(Preview=@n3) {inputName=Uvs, inputType=FunctionInput_Vector2}, B=LinearInterpolate(A=@n6, B=@n7, Alpha=@n5.B))
   B=ComponentMask(None=@n8) {g=true, r=true}
  )
  n10 = MaterialFunctionCall(Vector or UVs=@n9) {materialFunction=&VectorToRadialValue}
  n11 = Add(A=Add(A=Multiply(A=@n1, B=@n2), B=@n10.Radial Coordinates), B=Constant2Vector() {g=0.5}) {constB=0.5}
  n12 = ComponentMask(None=FunctionInput() {inputName=StreamsA_TileRate, sortPriority=50}) {g=true, r=true}
  n13 = ComponentMask(None=FunctionInput() {inputName=StreamsWarp_TileRate, sortPriority=100}) {g=true, r=true}
  n14 = Time()
  n15 = ComponentMask(None=FunctionInput() {inputName=StreamsWarp_Speed, sortPriority=100}) {g=true, r=true}
  n16 = TextureSample(
   UVs=Add(A=Multiply(A=@n10.Radial Coordinates, B=@n13), B=Multiply(A=@n14, B=@n15)) {constB=0}
   Tex=FunctionInput() {inputName=Texture_Warp, inputType=FunctionInput_Texture2D, sortPriority=5}
  ) {samplerType=SAMPLERTYPE_LinearColor}
  n17 = Multiply(
   A=SmoothStep(
    Max=ScalarParameter() {defaultValue=0.4, group=Wrap_Barium_Streams, parameterName=Barium_StreamsWarp_Fallof}
    Value=MaterialFunctionCall(Vector 2=@n9) {materialFunction=/Engine/Functions/Engine_MaterialFunctions02/Utility/VectorLength}.V2 Length
   ) {constMax=0.15, constMin=0.1}
   B=Multiply(A=Add(A=@n16.R) {constB=-0.5}, B=ScalarParameter() {defaultValue=0.1, group=Wrap_Barium_Streams, parameterName=Barium_StreamsWarp_Amount})
  )
  n18 = ScalarParameter() {defaultValue=0.05, group=Wrap_Barium_Streams, parameterName=Barium_Ripple_Thickness}
  n19 = MaterialFunctionCall(Vector 2=@n9) {materialFunction=&VectorLength}
  n20 = Subtract(
   A=@n19.V2 Length
   B=FunctionInput(Preview=Constant()) {bUsePreviewValueAsDefault=true, inputName=Warp_PositionA, inputType=FunctionInput_Scalar, sortPriority=100}
  ) {constB=0}
  n21 = OneMinus(None=SmoothStep(Max=@n18, Value=Abs(None=@n20)) {constMax=0.05})
  n22 = ScalarParameter() {defaultValue=0.02, group=Wrap_Barium_Streams, parameterName=Barium_Ripple_Strength}
  n23 = Subtract(
   A=@n19.V2 Length
   B=FunctionInput() {bUsePreviewValueAsDefault=true, inputName=Warp_PositionC, inputType=FunctionInput_Scalar, sortPriority=100}
  ) {constB=0}
  n24 = OneMinus(None=SmoothStep(Max=@n18, Value=Abs(None=@n23)) {constMax=0.05})
  n25 = FunctionInput() {bUsePreviewValueAsDefault=true, inputName=Warp_PositionB, inputType=FunctionInput_Scalar, sortPriority=100}
  n26 = SmoothStep(Max=@n18, Value=Abs(None=Subtract(A=@n19.V2 Length, B=@n25) {constB=0})) {constMax=0.05}
  n27 = Add(
   A=Add(A=Multiply(A=@n21, B=@n22) {constB=0.02}, B=Multiply(A=@n24, B=@n22) {constB=0.02})
   B=Multiply(A=OneMinus(None=@n26), B=@n22) {constB=0.02}
  )
  n28 = Add(A=Add(A=Multiply(A=@n11, B=@n12), B=@n17), B=@n27)
  n29 = FunctionInput() {inputName=Texture_Msk, inputType=FunctionInput_Texture2D, sortPriority=5}
  n30 = Multiply(
   A=@n9
   B=ComponentMask(None=VectorParameter() {group=Wrap_Barium_Streams, parameterName=Barium_Streams_hole_Size, defaultValue=<rgba 1 0.6 0 1>}.RGB) {g=true, r=true}
  )
  n31 = SmoothStep(
   Max=Multiply(B=ComponentMask(None=@n8) {a=true}) {constA=0.75}
   Value=Add(A=MaterialFunctionCall(Vector 2=@n30) {materialFunction=&VectorLength}.V2 Length, B=@n27)
  ) {constMax=0.15}
  n32 = Multiply(A=Add(A=TextureSample(UVs=@n28, Tex=@n29) {samplerType=SAMPLERTYPE_LinearColor}.R) {constB=0.2}, B=@n31)
  n33 = Constant() {r=10}
  n34 = Add(A=@n10.Radial Coordinates, B=@n1) {constB=0}
  n35 = ComponentMask(None=FunctionInput() {inputName=StreamsB_TileRate, sortPriority=50}) {g=true, r=true}
  n36 = Add(A=Add(A=Multiply(A=@n34, B=@n35), B=@n17), B=@n27)
  n37 = Multiply(A=Add(A=TextureSample(UVs=@n36, Tex=@n29) {samplerType=SAMPLERTYPE_LinearColor}.R) {constB=0.2}, B=@n31)
  n38 = Saturate(None=Add(
    A=MaterialFunctionCall(In=@n32, Contrast=@n33) {materialFunction=&CheapContrast}.Result
    B=MaterialFunctionCall(In=@n37, Contrast=@n33) {materialFunction=&CheapContrast}.Result
   ))
  FunctionOutput(None=@n38)
  Multiply(A=@n31)
  Multiply(A=@n38)
 ]}
