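# Material function for weapon wraps that produces a fake cartoon specular value from sunlight direction and surface normal.
# It reads the sunlight vector from FortniteMaterialParameters but copies only its R component into all three axes, so normalizing leaves a diagonal direction.
# The saturated dot product with the world pixel normal becomes the U coordinate, at constant V, for a grayscale lookup in T_Fake_Spec_M, which is the output.
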MaterialFunction /Game/Weapons/WeaponSkins/Wraps/MaterialFunctions/MF_Cartoon_Spec
graph {nodes [
  n1 = CollectionParameter() {
   collection=&FortniteMaterialParameters
   parameterName=SunlightDirectionalVector
  }
  n2 = MaterialFunctionCall(Float4=@n1) {materialFunction=&BreakOutFloat4Components}
  n3 = MaterialFunctionCall(X=@n2.R, Y=@n2.R, Z=@n2.R) {materialFunction=&MakeFloat3}
  n4 = Saturate(None=DotProduct(A=Normalize(VectorInput=@n3.Result), B=PixelNormalWS()))
  n5 = Constant()
  FunctionOutput(None=TextureSample(UVs=MaterialFunctionCall(X=@n4, Y=@n5) {materialFunction=&MakeFloat2}.Result) {samplerType=SAMPLERTYPE_LinearGrayscale, texture=/Game/Weapons/WeaponSkins/Wraps/Textures/Cartoon/T_Fake_Spec_M}.RGB)
 ]}
