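# Material function for weapon skin wraps that picks one planar projection coordinate pair from local position, as a cheap alternative to triplanar mapping.
# Two nested Ifs test absolute pre-skinned local normal components against a fixed threshold: Y selects XZ, otherwise Z selects XY, with YZ as fallback.
# The hard switch outputs a single 2D coordinate, with no blending or texture sampling inside the function.
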MaterialFunction /Game/Weapons/WeaponSkins/Wraps/MaterialFunctions/MF_Cheap_Tri
graph {nodes [
  n1 = MaterialFunctionCall(Float3=FunctionInput() {bUsePreviewValueAsDefault=true, inputName=Pre-Skinned Local Normal, sortPriority=2}) {materialFunction=&BreakOutFloat3Components}
  n2 = Constant() {r=0.65}
  n3 = FunctionInput() {bUsePreviewValueAsDefault=true, inputName=Local Position, sortPriority=1}
  n4 = Abs(None=@n1.B)
  n5 = ComponentMask(None=@n3) {g=true, r=true}
  n6 = ComponentMask(None=@n3) {b=true, g=true}
  FunctionOutput(None=If(A=Abs(None=@n1.G), B=@n2, "A > B"=ComponentMask(None=@n3) {b=true, r=true}, "A < B"=If(A=@n4, B=@n2, "A > B"=@n5, "A < B"=@n6)))
 ]}
