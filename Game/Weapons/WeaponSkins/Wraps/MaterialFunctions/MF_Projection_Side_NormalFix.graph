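# Weapon wrap material function that corrects normal map orientation for axis-aligned planar projection, using pre-skinned normals.
# It swizzles the input normal for X, Y and Z projections, scales its depth by the matching surface normal component and rotates it by Rotation_Value.
# IsState channels select the axis through nested lerps.
# The surface's axis component is subtracted, and the difference goes from local to tangent space as Raw_Normal, or added to flat Z as Normal.
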MaterialFunction /Game/Weapons/WeaponSkins/Wraps/MaterialFunctions/MF_Projection_Side_NormalFix
graph {nodes [
  n1 = FunctionInput(Preview=TextureSample(Tex=TextureObject() {samplerType=SAMPLERTYPE_Normal, texture=/Engine/Functions/Engine_MaterialFunctions02/ExampleContent/Textures/SphereRenderNormalsMap}) {samplerType=SAMPLERTYPE_Normal}.RGB) {bUsePreviewValueAsDefault=true, inputName=Normals input}
  n2 = MaterialFunctionCall(Color=@n1) {materialFunction=&SplitComponents}
  n3 = FunctionInput(Preview=VertexInterpolator(VS=PreSkinnedNormal()).PS) {bUsePreviewValueAsDefault=true, inputName=Pre_Skinned_Normals, sortPriority=3}
  n4 = ComponentMask(None=@n3) {r=true}
  n5 = AppendVector(A=AppendVector(A=Multiply(A=@n2.B, B=@n4), B=@n2.R), B=Multiply(A=@n2.G) {constB=-1})
  n6 = FunctionInput() {inputName=Rotation_Value, inputType=FunctionInput_Scalar, sortPriority=1}
  n7 = RotateAboutAxis(NormalizedRotationAxis=Constant3Vector() {constant=<rgba 1 0 0 0>}, RotationAngle=@n6, PivotPoint=Constant3Vector(), Position=@n5)
  n8 = AppendVector(A=ComponentMask(None=@n1) {g=true, r=true}, B=Multiply(A=@n2.B, B=ComponentMask(None=@n3) {b=true}))
  n9 = Add(
   A=RotateAboutAxis(
    NormalizedRotationAxis=Constant3Vector() {constant=<rgba 0 0 1 0>}
    RotationAngle=Multiply(A=@n6) {constB=-1}
    PivotPoint=Constant3Vector()
    Position=@n8
   )
   B=@n8
  )
  n10 = MaterialFunctionCall(Color=@n1) {materialFunction=&SplitComponents}
  n11 = ComponentMask(None=@n3) {g=true}
  n12 = AppendVector(A=AppendVector(A=@n10.R, B=Multiply(A=@n10.B, B=@n11)), B=Multiply(A=@n10.G) {constB=-1})
  n13 = Add(
   A=RotateAboutAxis(
    NormalizedRotationAxis=Constant3Vector() {constant=<rgba 0 1 0 0>}
    RotationAngle=Multiply(A=@n6) {constB=-1}
    PivotPoint=Constant3Vector()
    Position=@n12
   )
   B=@n12
  )
  n14 = MaterialFunctionCall(Color=FunctionInput() {inputName=IsState, sortPriority=2}) {materialFunction=&SplitComponents}
  n15 = Constant3Vector() {constant=<rgba 0 1 0 0>}
  n16 = Multiply(A=@n6) {constB=-1}
  n17 = Constant3Vector()
  n18 = MaterialFunctionCall(Color=@n1) {materialFunction=&SplitComponents}
  n19 = ComponentMask(None=@n3) {g=true}
  n20 = AppendVector(A=AppendVector(A=@n18.R, B=Multiply(A=@n18.B, B=@n19)), B=Multiply(A=@n18.G) {constB=-1})
  n21 = LinearInterpolate(
   A=LinearInterpolate(A=Add(A=@n7, B=@n5), B=LinearInterpolate(A=@n9, B=@n13, Alpha=@n14.G), Alpha=@n14.R)
   B=Add(A=RotateAboutAxis(NormalizedRotationAxis=@n15, RotationAngle=@n16, PivotPoint=@n17, Position=@n20), B=@n20)
   Alpha=@n14.B
  )
  n22 = Constant3Vector() {constant=<rgba 1 0 0 0>}
  n23 = Multiply(A=Constant3Vector() {constant=<rgba 0 0 1 0>}, B=@n3)
  n24 = Multiply(A=Constant3Vector() {constant=<rgba 0 1 0 0>}, B=@n3)
  n25 = LinearInterpolate(
   A=LinearInterpolate(A=Multiply(A=@n22, B=@n3), B=LinearInterpolate(A=@n23, B=@n24, Alpha=@n14.G), Alpha=@n14.R)
   B=Multiply(A=Constant3Vector() {constant=<rgba 0 1 0 0>}, B=@n3)
   Alpha=@n14.B
  )
  n26 = Transform(None=Subtract(A=@n21, B=@n25)) {transformSourceType=TRANSFORMSOURCE_Local, transformType=TRANSFORM_Tangent}
  FunctionOutput(None=Add(A=Constant3Vector() {constant=<rgba 0 0 1 0>}, B=@n26)) {outputName=Normal, sortPriority=2}
  FunctionOutput(None=@n26) {description=add to object normals, outputName=Raw_Normal, sortPriority=3}
  Multiply() {constB=-1}
  Normalize(VectorInput=@n25)
  Normalize(VectorInput=@n21)
 ]}
