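# Material function for clockwork weapon wraps that composites four gear layers (A to D) from shared gear, mask and normal textures.
# Mask red times gear alpha sets each layer's coverage, blending B, C, D, then A on top into tinted color, mask, tint-alpha depth and normals.
# Normals pass through a side-fix function with per-gear rotation.
# Emissive uses a selectable normal-texture channel, masked and tinted per gear, with optional clamping and gradient lookup.
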MaterialFunction /Game/Weapons/WeaponSkins/Wraps/MaterialFunctions/MF_ClockWork_Gears
graph {nodes [
  n1 = TextureObject() {texture=/Game/Global/Textures/Default/Blanks/T_Black_srgb_EmptyAlpha}
  n2 = FunctionInput(Preview=@n1) {inputName=Gear_mask_Red, inputType=FunctionInput_Texture2D}
  n3 = TextureSample(UVs=FunctionInput() {inputName=Uvs_GearB_Mask, inputType=FunctionInput_Vector2, sortPriority=12}, Tex=@n2) {texture=&T_BlacksRGB}
  n4 = FunctionInput() {inputName=Uvs_GearB, inputType=FunctionInput_Vector2, sortPriority=13}
  n5 = FunctionInput(Preview=@n1) {inputName=Gears, inputType=FunctionInput_Texture2D, sortPriority=1}
  n6 = TextureSample(UVs=@n4, Tex=@n5) {texture=&T_BlacksRGB}
  n7 = Multiply(A=@n3.R, B=@n6.A)
  n8 = VectorParameter() {group=Wrap_ClockWork, parameterName=GearB_Tint, defaultValue=<rgba 0.2 0.2 0.2 0.5>}
  n9 = Multiply(A=Multiply(A=@n7, B=@n6.RGB), B=@n8.RGB) {constB=0.25}
  n10 = TextureSample(UVs=FunctionInput() {inputName=Uvs_GearC_Mask, inputType=FunctionInput_Vector2, sortPriority=14}, Tex=@n2) {texture=&T_BlacksRGB}
  n11 = FunctionInput() {inputName=Uvs_GearC, inputType=FunctionInput_Vector2, sortPriority=16}
  n12 = TextureSample(UVs=@n11, Tex=@n5) {texture=&T_BlacksRGB}
  n13 = Multiply(A=@n10.R, B=@n12.A)
  n14 = VectorParameter() {group=Wrap_ClockWork, parameterName=GearC_Tint, defaultValue=<rgba 0.6 0.6 0.6 0.5>}
  n15 = TextureSample(UVs=FunctionInput() {inputName=Uvs_GearD_Mask, inputType=FunctionInput_Vector2, sortPriority=17}, Tex=@n2) {texture=&T_BlacksRGB}
  n16 = FunctionInput() {inputName=Uvs_GearD, inputType=FunctionInput_Vector2, sortPriority=18}
  n17 = TextureSample(UVs=@n16, Tex=@n5) {texture=&T_BlacksRGB}
  n18 = Multiply(A=@n15.R, B=@n17.A)
  n19 = VectorParameter() {group=Wrap_ClockWork, parameterName=GearD_Tint, defaultValue=<rgba 1 1 1 0.75>}
  n20 = LinearInterpolate(
   A=Multiply(A=Multiply(A=@n13, B=@n12.RGB), B=@n14.RGB) {constB=0.25}
   B=Multiply(A=Multiply(A=@n18, B=@n17.RGB), B=@n19.RGB) {constB=0.25}
   Alpha=@n18
  )
  n21 = Saturate(None=Add(A=@n13, B=@n18))
  n22 = FunctionInput(Preview=TextureCoordinate()) {inputName=Uvs_GearA, inputType=FunctionInput_Vector2, sortPriority=11}
  n23 = TextureSample(UVs=@n22, Tex=@n5) {texture=&T_BlacksRGB}
  n24 = TextureSample(UVs=FunctionInput() {inputName=Uvs_GearA_Mask, inputType=FunctionInput_Vector2, sortPriority=10}, Tex=@n2) {texture=&T_BlacksRGB}
  n25 = Multiply(A=@n24.R, B=@n23.A)
  n26 = Multiply(A=@n23.RGB, B=@n25)
  n27 = VectorParameter() {group=Wrap_ClockWork, parameterName=GearA_Tint, defaultValue=<rgba 1.5 1.5 1.5 0.9>}
  FunctionOutput(None=LinearInterpolate(A=LinearInterpolate(A=@n9, B=@n20, Alpha=@n21), B=Multiply(A=@n26, B=@n27.RGB), Alpha=@n25)) {outputName=Combined_Output}
  n28 = Add(A=@n25, B=@n7)
  FunctionOutput(None=Saturate(None=Add(A=@n28, B=@n21))) {outputName=Gear_Mask, sortPriority=1}
  n29 = FunctionInput(Preview=TextureObject() {samplerType=SAMPLERTYPE_Normal, texture=&FlatNormalBlue}) {inputName=Normals, inputType=FunctionInput_Texture2D, sortPriority=1}
  n30 = TextureSample(UVs=@n4, Tex=@n29) {texture=&T_BlacksRGB}
  n31 = ComponentMask(None=NamedRerouteUsage()) {g=true}
  n32 = NamedRerouteUsage()
  n33 = NamedRerouteUsage()
  n34 = LinearInterpolate(
   A=FunctionInput() {bUsePreviewValueAsDefault=true, inputName=BackGroundMask, previewValue=<1 1 1 1>}
   B=MaterialFunctionCall!(
    Normal_Map
    Normals_Source=@n30.RGB
    Normals_Source_Use=StaticBool() {value=true}
    UVs
    Rotation_Comp=Constant() {r=0.25}
    UV_Rotation=@n31
    Pre_Skinned_Normals=@n32
    is_State=@n33
    "Left\\Right"
    No_Rotation
   ) {materialFunction=&/MF_UVs_Side_Fix_Normals}.Normal
   Alpha=@n7
  )
  n35 = TextureSample(UVs=@n16, Tex=@n29) {texture=&T_BlacksRGB}
  n36 = ComponentMask(None=NamedRerouteUsage()) {a=true}
  n37 = NamedRerouteUsage()
  n38 = NamedRerouteUsage()
  n39 = TextureSample(UVs=@n11, Tex=@n29) {texture=&T_BlacksRGB}
  n40 = ComponentMask(None=NamedRerouteUsage()) {b=true}
  n41 = NamedRerouteUsage()
  n42 = NamedRerouteUsage()
  n43 = LinearInterpolate(
   A=MaterialFunctionCall(
    Normals_Source=@n39.RGB
    Normals_Source_Use=StaticBool() {value=true}
    Rotation_Comp=Constant() {r=0.25}
    UV_Rotation=@n40
    Pre_Skinned_Normals=@n41
    is_State=@n42
   ) {materialFunction=&MF_UVs_Side_Fix_Normals}.Normal
   B=MaterialFunctionCall(
    Normals_Source=@n35.RGB
    Normals_Source_Use=StaticBool() {value=true}
    Rotation_Comp=Constant() {r=0.25}
    UV_Rotation=@n36
    Pre_Skinned_Normals=@n37
    is_State=@n38
   ) {materialFunction=&MF_UVs_Side_Fix_Normals}.Normal
   Alpha=@n18
  )
  n44 = TextureSample(UVs=@n22, Tex=@n29) {texture=&T_BlacksRGB}
  n45 = StaticBool() {value=true}
  n46 = Constant() {r=0.25}
  n47 = ComponentMask(None=NamedRerouteUsage()) {r=true}
  n48 = NamedRerouteUsage()
  n49 = NamedRerouteUsage()
  FunctionOutput(None=LinearInterpolate(
    A=LinearInterpolate(A=@n34, B=@n43, Alpha=@n21)
    B=MaterialFunctionCall(Normals_Source=@n44.RGB, Normals_Source_Use=@n45, Rotation_Comp=@n46, UV_Rotation=@n47, Pre_Skinned_Normals=@n48, is_State=@n49) {materialFunction=&MF_UVs_Side_Fix_Normals}.Normal
    Alpha=@n25
   )) {outputName=Texture_Normals, sortPriority=1}
  n50 = LinearInterpolate(A=@n14.A, B=@n19.A, Alpha=@n18)
  FunctionOutput(None=LinearInterpolate(A=LinearInterpolate(A=@n8.A, B=@n50, Alpha=@n21), B=@n27.A, Alpha=@n25)) {outputName=Gear_Mask_Depth, sortPriority=1}
  n51 = ChannelMaskParameter(None=@n44.RGBA) {group=Wrap_ClockWork, maskChannel=Green, parameterName=Emissive_Channel_Mask, sortPriority=55, defaultValue=<rgba 0 1 0 0>}
  n52 = Multiply(A=@n24.R, B=@n51.R)
  n53 = VectorParameter() {group=Wrap_ClockWork, parameterName=Emissive_A_ColorCW, sortPriority=56, defaultValue=<rgba 1.5 1.5 1.5 0.9>}
  n54 = ChannelMaskParameter(None=@n30.RGBA) {group=Wrap_ClockWork, maskChannel=Green, parameterName=Emissive_Channel_Mask, sortPriority=55, defaultValue=<rgba 0 1 0 0>}
  n55 = Multiply(A=@n3.R, B=@n54.R)
  n56 = VectorParameter() {group=Wrap_ClockWork, parameterName=Emissive_B_ColorCW, sortPriority=57, defaultValue=<rgba 1.5 1.5 1.5 0.9>}
  n57 = ChannelMaskParameter(None=NamedRerouteDeclaration(None=@n39.RGBA) {name=Normal_C, nodeColor=<rgba 0 0.3672 1 1>}) {group=Wrap_ClockWork, maskChannel=Green, parameterName=Emissive_Channel_Mask, sortPriority=55, defaultValue=<rgba 0 1 0 0>}
  n58 = Multiply(A=@n10.R, B=@n57.R)
  n59 = VectorParameter() {group=Wrap_ClockWork, parameterName=Emissive_C_ColorCW, sortPriority=58, defaultValue=<rgba 1.5 1.5 1.5 0.9>}
  n60 = ChannelMaskParameter(None=NamedRerouteDeclaration(None=@n35.RGBA) {name=Normal_D, nodeColor=<rgba 0 0.3672 1 1>}) {group=Wrap_ClockWork, maskChannel=Green, parameterName=Emissive_Channel_Mask, sortPriority=55, defaultValue=<rgba 0 1 0 0>}
  n61 = Multiply(A=@n15.R, B=@n60.R)
  n62 = VectorParameter() {group=Wrap_ClockWork, parameterName=Emissive_D_ColorCW, sortPriority=59, defaultValue=<rgba 1.5 1.5 1.5 0.9>}
  n63 = Add(A=Add(A=Multiply(A=@n52, B=@n53.RGB), B=Multiply(A=@n55, B=@n56.RGB)), B=Add(A=Multiply(A=@n58, B=@n59.RGB), B=Multiply(A=@n61, B=@n62.RGB)))
  n64 = StaticSwitchParameter(True=Clamp(None=@n63) {maxDefault=0.99}, False=@n63) {group=Wrap_ClockWork, parameterName=Emissive_Clamp_Range, sortPriority=52}
  n65 = TextureSampleParameter2D(UVs=ComponentMask(None=@n64) {r=true}) {group=Wrap_ClockWork, parameterName=Emissive_GradCW, sortPriority=51, texture=/Game/Global/Textures/Default/Masks_and_Gradients/3pxRGBImage}
  n66 = Multiply(A=@n65.A, B=ScalarParameter() {defaultValue=1, group=Wrap_ClockWork, parameterName=Emissive_Grad_Intensity, sortPriority=51})
  FunctionOutput(None=StaticSwitchParameter(True=Multiply(A=@n65.RGB, B=@n66), False=@n64) {group=Wrap_ClockWork, parameterName=Emissive_Trexture_Grad, sortPriority=50}) {outputName=Emissive, sortPriority=1}
  n67 = NamedRerouteUsage()
  n68 = LinearInterpolate(A=NamedRerouteUsage(), B=NamedRerouteUsage(), Alpha=@n18)
  FunctionOutput(None=LinearInterpolate(A=LinearInterpolate(A=@n67, B=@n68, Alpha=@n21), B=NamedRerouteUsage(), Alpha=@n25)) {outputName=Texture_RGBA, sortPriority=1}
  MaterialFunctionCall!(Normals input=@n39.RGB, Rotation_Value=@n40, IsState=@n42, Pre_Skinned_Normals=@n41) {materialFunction=&/MF_Projection_Side_NormalFix}
  MaterialFunctionCall(Normals input=@n30.RGB, Rotation_Value=@n31, IsState=@n33, Pre_Skinned_Normals=@n32) {materialFunction=&MF_Projection_Side_NormalFix}
  MaterialFunctionCall(Normals input=@n44.RGB, Rotation_Value=@n47, IsState=@n49, Pre_Skinned_Normals=@n48) {materialFunction=&MF_Projection_Side_NormalFix}
  MaterialFunctionCall(Normals input=@n35.RGB, Rotation_Value=@n36, IsState=@n38, Pre_Skinned_Normals=@n37) {materialFunction=&MF_Projection_Side_NormalFix}
  NamedRerouteDeclaration(None=FunctionInput(Preview=Constant()) {bUsePreviewValueAsDefault=true, inputName=Rotation, inputType=FunctionInput_Vector4, sortPriority=100}) {name=Rotation, nodeColor=<rgba 0.1497 1 0.1633 1>}
  NamedRerouteDeclaration(None=@n30.RGBA) {name=Normal_B, nodeColor=<rgba 0 0.3672 1 1>}
  NamedRerouteDeclaration(None=FunctionInput(Preview=Constant4Vector().RGBA) {bUsePreviewValueAsDefault=true, inputName=IsState, inputType=FunctionInput_Vector4, sortPriority=90}) {name=IsState, nodeColor=<rgba 0.1497 1 0.1633 1>}
  NamedRerouteDeclaration(None=@n44.RGBA) {name=Normal_A, nodeColor=<rgba 0 0.3672 1 1>}
  n69 = PreSkinnedNormal()
  NamedRerouteDeclaration(None=FunctionInput(Preview=VertexInterpolator(VS=@n69).PS) {bUsePreviewValueAsDefault=true, inputName=Local Normals, sortPriority=80}) {name=PreSkin_Normal, nodeColor=<rgba 0.8281 1 0 1>}
 ]}
