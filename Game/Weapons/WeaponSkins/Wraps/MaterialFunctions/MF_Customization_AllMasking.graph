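# Material function for weapon wraps that selects which customization mask regions of a weapon are covered.
# It takes base and secondary masks as inputs and derives a primary mask as one minus their sum.
# A chain of lerps, driven by Customization_Masking scalar parameters, switches from the default base mask to single regions or summed pairs, with later switches overriding earlier ones.
# MASK_CoverEntireWeapon comes last, serving as both value and alpha to force full coverage at one.
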MaterialFunction /Game/Weapons/WeaponSkins/Wraps/MaterialFunctions/MF_Customization_AllMasking
graph {nodes [
  n1 = FunctionInput() {
   bUsePreviewValueAsDefault=true
   desc=base customization
   inputName="Customization Mask (_CM B Top left)"
   inputType=FunctionInput_Scalar
   sortPriority=24
  }
  n2 = FunctionInput() {
   bUsePreviewValueAsDefault=true
   desc=secondary
   inputName="Customization Mask 2 (_CM B Top Right)"
   inputType=FunctionInput_Scalar
   sortPriority=25
  }
  n3 = LinearInterpolate(A=@n1, B=@n2, Alpha=ScalarParameter() {group=Customization_Masking, parameterName=MASK_OnlySecondary, sortPriority=1})
  n4 = Add(A=@n2, B=@n1) {desc="Base + Secondary = exposed primary"}
  n5 = OneMinus(None=@n4) {desc=Primary}
  n6 = ScalarParameter() {group=Customization_Masking, parameterName=MASK_OnlyPrimary, sortPriority=2}
  n7 = LinearInterpolate(
   A=LinearInterpolate(
    A=LinearInterpolate(A=@n3, B=@n5, Alpha=@n6)
    B=@n4
    Alpha=ScalarParameter() {group=Customization_Masking, parameterName=MASK_BaseAndSecondary, sortPriority=3}
   )
   B=Add(A=@n5, B=@n1)
   Alpha=ScalarParameter() {group=Customization_Masking, parameterName=MASK_BaseAndPrimary, sortPriority=4}
  )
  n8 = Add(A=@n5, B=@n2)
  n9 = ScalarParameter() {group=Customization_Masking, parameterName=MASK_SecondaryAndPrimary, sortPriority=5}
  n10 = ScalarParameter() {group=Customization_Masking, parameterName=MASK_CoverEntireWeapon, sortPriority=0}
  FunctionOutput(None=LinearInterpolate(A=LinearInterpolate(A=@n7, B=@n8, Alpha=@n9), B=@n10, Alpha=@n10))
 ]}
