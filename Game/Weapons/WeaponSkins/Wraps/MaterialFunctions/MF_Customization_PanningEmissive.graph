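# Material function that applies the shared PanningEmissive effect to material attributes for weapon skin wraps.
# It passes the Material, isVehicle and Customization Mask (as Alpha) inputs into PanningEmissive and outputs its result.
# The pre-skinned position and normal inputs are declared but unwired.
# A Crevasse edge-mask chain with FX_Crystalline black-level and contrast parameters does not reach the output.
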MaterialFunction /Game/Weapons/WeaponSkins/Wraps/MaterialFunctions/MF_Customization_PanningEmissive
graph {nodes [
  FunctionInput(Preview=Constant()) {bUsePreviewValueAsDefault=true, inputName=------------------, inputType=FunctionInput_Scalar, sortPriority=-19}
  FunctionInput() {bUsePreviewValueAsDefault=true, inputName=PreSkinLocalPosition}
  FunctionInput() {bUsePreviewValueAsDefault=true, inputName=PreSkinnedLocalNormal}
  FunctionOutput(None=MaterialFunctionCall!(
    Alpha=FunctionInput() {bUsePreviewValueAsDefault=true, inputName="Customization Mask (_CM B)", inputType=FunctionInput_Scalar, sortPriority=24}
    isVehicle=FunctionInput() {inputName=isVehicle, inputType=FunctionInput_Scalar}
    Gmap
    UVOffset
    Local Normal
    Local Position
    EmissiveColorOverride
    Material=FunctionInput() {inputName=Material, inputType=FunctionInput_MaterialAttributes, sortPriority=-200}
   ) {materialFunction=/Game/Weapons/WeaponSkins/MaterialFunction/PanningEmissive}.Result)
  n1 = FunctionInput() {bUsePreviewValueAsDefault=true, inputName="Customization Crevasse (_CM G)", inputType=FunctionInput_Scalar, sortPriority=21}
  n2 = ScalarParameter() {defaultValue=1.5, group=FX_Crystalline, parameterName=Edge Mask Black Levels, sortPriority=1}
  Saturate(None=MaterialFunctionCall(
    In=MaterialFunctionCall!(Image=@n1, Black intensity level=@n2) {materialFunction=/Engine/Functions/Engine_MaterialFunctions02/Texturing/RaiseBlackLevelsByPercentage}.Result
    Contrast=ScalarParameter() {defaultValue=-0.2, group=FX_Crystalline, parameterName=Edge Mask Contrast, sortPriority=2}
   ) {materialFunction=&CheapContrast}.Result)
 ]}
