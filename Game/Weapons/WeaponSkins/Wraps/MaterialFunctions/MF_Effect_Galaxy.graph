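# Weapon wrap material function that, behind a static switch, blends an animated galaxy look over the incoming material using primary and accent masks.
# Panning galaxy and star textures use virtual-plane or screen-space UVs, scaled per object type (weapon, vehicle, mech, character, backpack, pickaxe).
# Hue shift, optional color-curve tinting, fresnel and curvature-mask edges shape the emissive, with time-triggered reactive edge and Gmap color.
# A cheap normal-selected triplanar projection supplies optional roughness masking.
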MaterialFunction /Game/Weapons/WeaponSkins/Wraps/MaterialFunctions/MF_Effect_Galaxy
graph {nodes [
  n1 = NamedRerouteUsage()
  Clamp(
   None=Add(A=@n1) {constB=-0.5}
   Min=ScalarParameter() {defaultValue=-0.1, group=Effect_Galaxy_Edge, parameterName=Eff_Galaxy_Edge_Brightnees_MinClamp}
  ) {minDefault=-0.1}
  n2 = NamedRerouteUsage()
  n3 = NamedRerouteUsage()
  n4 = ScalarParameter() {defaultValue=1, group=Effect_Galaxy, parameterName="Eff_Galaxy_Primary_Msk    1=On", sortPriority=0}
  n5 = TextureSample() {samplerType=SAMPLERTYPE_Masks, texture=/Game/Weapons/FORT_Rifles/Textures/T_Scar_CM}
  n6 = MaterialFunctionCall(Float4=FunctionInput(Preview=@n5.RGBA) {inputName=CM_Msk, inputType=FunctionInput_Vector4, sortPriority=4}) {materialFunction=&BreakOutFloat4Components}
  n7 = NamedRerouteDeclaration(None=@n6.R) {name=Msk_Prime, nodeColor=<rgba 0 1 0.7422 1>}
  n8 = NamedRerouteDeclaration(None=@n6.B) {name=Msk_Accent, nodeColor=<rgba 0 1 0.7656 1>}
  n9 = ScalarParameter() {group=Effect_Galaxy, parameterName="Eff_Galaxy_Accent_Msk    1=On", sortPriority=0}
  n10 = Max(
   A=Max(A=Min(A=@n4, B=@n7), B=Min(A=@n8, B=@n9))
   B=ScalarParameter() {group=Effect_Galaxy, parameterName="Eff_Galaxy_Disable_All_Msk    1=On", sortPriority=0}
  )
  n11 = TextureSample() {samplerType=SAMPLERTYPE_Normal, texture=/Game/Weapons/FORT_Rifles/T_SCAR_N}
  n12 = Constant3Vector() {constant=<rgba 1 0 0 0>}
  n13 = GetMaterialAttributes(None=FunctionInput(Preview=SetMaterialAttributes!(MaterialAttributes, Normal=@n11.RGB, BaseColor=@n12) {attributeSetTypes [#1, #2]}) {inputName=Material, inputType=FunctionInput_MaterialAttributes}) {attributeGetTypes [
    #2
    #3
    #1
    #4
   ]}
  FunctionOutput(None=StaticSwitchParameter(
    True=BlendMaterialAttributes(A=@n2, B=@n3, Alpha=@n10) {vertexAttributeBlendType=UseA}
    False=NamedRerouteDeclaration(None=@n13.MaterialAttributes) {name=Root_Material, nodeColor=<rgba 0.7812 1 0 1>}
   ) {group=Customization, parameterName=Effect_Galaxy})
  InverseLinearInterpolate(
   A=Constant() {r=45}
   B=Constant() {r=90}
   Value=MaterialFunctionCall() {materialFunction=/Engine/Functions/Engine_MaterialFunctions02/Utility/FOV}.FOV angle in degrees
  )
  n14 = Subtract(
   A=Time()
   B=Add(
    A=ScalarParameter() {defaultValue=-9999999, group=Effect_Galaxy_Edge|Reactive, parameterName=Eff_Galaxy_Gmap_TFBP, sortPriority=600}
    B=ScalarParameter() {group=Effect_Galaxy_Edge|Reactive, parameterName=Eff_Galaxy_Gmap_Delay}
   )
  )
  n15 = ScalarParameter() {defaultValue=0.25, group=Effect_Galaxy_Edge|Reactive, parameterName=Eff_Galaxy_Gmap_Duration}
  n16 = Power(
   Base=Saturate(None=Divide(A=@n14, B=@n15))
   Exp=ScalarParameter() {defaultValue=0.25, group=Effect_Galaxy_Edge|Reactive, parameterName=Eff_Galaxy_Gmap_Falloff}
  )
  LinearInterpolate(
   A=ScalarParameter() {defaultValue=1, group=Effect_Galaxy_Edge|Reactive, parameterName=Eff_Galaxy_Gmap_Brightness_Start}
   B=ScalarParameter() {defaultValue=1, group=Effect_Galaxy_Edge|Reactive, parameterName=Eff_Galaxy_Gmap_Brightness_End}
   Alpha=@n16
  )
  n17 = VectorParameter() {group=Effect_Galaxy_Edge, parameterName=Eff_Galaxy_Edge_Color, channelNames {a=desaturate, b="", g="", r=""}, defaultValue=<rgba 1 1 1 0>}
  n18 = VectorParameter() {group=Effect_Galaxy_Galaxy, parameterName=Eff_Galaxy_BaseColor, sortPriority=-10, defaultValue=<rgba 0 0 0 1>}
  n19 = NamedRerouteUsage()
  n20 = Desaturation(
   None=MaterialFunctionCall(Hue Shift Percentage=ScalarParameter() {group=Effect_Galaxy_Galaxy, parameterName=Eff_Galaxy_GalaxyHueShift}, Texture=@n19) {materialFunction=&HueShift}.Result
   Fraction=ScalarParameter() {defaultValue=0.33, group=Effect_Galaxy_Galaxy, parameterName=Eff_Galaxy_GalaxyDesaturationAmount}
  )
  n21 = NamedRerouteUsage()
  n22 = VectorParameter() {
   group=Effect_Galaxy_Galaxy
   parameterName=Eff_Galaxy_CC_WarpS
   channelNames {a="", b=Gmap Strength, g=Curvature Strength, r=Galaxy Strenght}
   defaultValue=<rgba 0 0 0 1>
  }
  n23 = Multiply(A=NamedRerouteUsage(), B=@n22.Curvature Strength)
  n24 = Multiply(A=ChannelMaskParameter(None=@n19) {group=Effect_Galaxy_Galaxy, parameterName=Eff_Galaxy_CC_Warp_Channel}.R, B=@n22.Galaxy Strenght)
  n25 = Add(
   A=Multiply(A=Time(), B=ScalarParameter() {group=Effect_Galaxy_Galaxy, parameterName=Eff_Galaxy_CC_Speed})
   B=Add(A=Multiply(A=@n21, B=@n22.Gmap Strength), B=Add(A=@n23, B=@n24))
  )
  n26 = ScalarParameter() {defaultValue=1, group=Effect_Galaxy_Galaxy, parameterName=Eff_Galaxy_CC_TileRate}
  n27 = Frac(None=Add(A=Multiply(A=@n25, B=@n26), B=ScalarParameter() {group=Effect_Galaxy_Galaxy, parameterName=Eff_Galaxy_CC_Offset}))
  n28 = CurveAtlasRowParameter(CurveTime=@n27) {
   atlas=/Game/Weapons/WeaponSkins/Wraps/Textures/CruveAtlas_Wraps_01
   curve=/Game/Weapons/WeaponSkins/Wraps/Textures/0_ColorCurves/CC_Spectum
   defaultValue=98
   group=Effect_Galaxy_Galaxy
   parameterName="Eff_Galaxy_CC  Tint"
  }
  n29 = Multiply(
   A=MaterialFunctionCall!(Index=ScalarParameter() {group=Effect_Galaxy_Galaxy, parameterName="Eff_Galaxy_CC  Tint  Index"}, CCinput=@n28, CCinput_Alpha=@n28, UV=@n27) {materialFunction=&/MF_Global_CC_Tex}.3V
   B=ScalarParameter() {defaultValue=1, group=Effect_Galaxy_Galaxy, parameterName=Eff_Galaxy_CC_Brightness}
  )
  n30 = Multiply(A=@n20, B=@n29)
  n31 = ScalarParameter() {group=Effect_Galaxy_Galaxy, parameterName=Eff_Galaxy_CC_Affect every layer}
  n32 = VectorParameter() {group=Effect_Galaxy_Galaxy, parameterName=Eff_Galaxy_BGMultTint, defaultValue=<rgba 0.1095 0.2307 0.5271 1>}
  n33 = Add(
   A=@n18.RGB
   B=StaticSwitchParameter(True=LinearInterpolate(A=@n30, B=@n20, Alpha=@n31), False=Multiply(A=@n20, B=@n32.RGB)) {group=Effect_Galaxy_Galaxy, parameterName="Eff_Galaxy_BGMultTint---->CC"}
  )
  n34 = Desaturation(None=@n20, Fraction=ScalarParameter() {defaultValue=0.5, group=Effect_Galaxy_Stars, parameterName=Eff_Galaxy_Star Saturation})
  n35 = VectorParameter() {group=Effect_Galaxy_Stars, parameterName=Eff_Galaxy_Star_Small_Brightness, defaultValue=<rgba 50 50 50 1>}
  n36 = Multiply(A=NamedRerouteUsage(), B=ScalarParameter() {defaultValue=8, group=Effect_Galaxy_Stars, parameterName=Eff_Galaxy_Star_Small_Tiling})
  n37 = VectorParameter() {group=Effect_Galaxy_Stars, parameterName=Eff_Galaxy_Star_Small_PanSpeed, defaultValue=<rgba -0.02 0.02 0 0>}
  n38 = ComponentMask(None=@n37.RGB) {g=true, r=true}
  n39 = ChannelMaskParameter(None=TextureSample(UVs=Panner(Coordinate=@n36, Speed=@n38) {speedY=0.02}, Tex=NamedRerouteUsage()).RGBA) {group=Effect_Galaxy, maskChannel=Blue, parameterName=Eff_Galaxy_Star_Small_Channel, defaultValue=<rgba 0 0 1 0>}
  n40 = LinearInterpolate(A=@n33, B=Multiply(A=@n34, B=@n35.RGB), Alpha=Multiply(A=@n39.R, B=@n39.R))
  n41 = VectorParameter() {group=Effect_Galaxy_Stars, parameterName=Eff_Galaxy_Star_Med_Brightness, defaultValue=<rgba 200 200 200 1>}
  n42 = Multiply(A=@n34, B=@n41.RGB)
  n43 = Multiply(A=NamedRerouteUsage(), B=ScalarParameter() {defaultValue=3, group=Effect_Galaxy_Stars, parameterName=Eff_Galaxy_Star_Med_Tiling})
  n44 = ComponentMask(None=VectorParameter() {group=Effect_Galaxy_Stars, parameterName=Eff_Galaxy_Star_Med_PanSpeed}.RGB) {g=true, r=true}
  n45 = ChannelMaskParameter(None=TextureSample(UVs=Panner(Coordinate=@n43, Speed=@n44) {speedX=0.133, speedY=0.2}, Tex=NamedRerouteUsage()).RGBA) {group=Effect_Galaxy, parameterName=Eff_Galaxy_Star_Med_Channel}
  n46 = NamedRerouteUsage()
  n47 = VectorParameter() {group=Effect_Galaxy_Stars, parameterName=Eff_Galaxy_Star_Med_Msk_PanSpeed, defaultValue=<rgba 0.133 0.2 0 0>}
  n48 = ComponentMask(None=@n47.RGB) {g=true, r=true}
  n49 = ChannelMaskParameter(None=TextureSample(UVs=Panner(Coordinate=@n46, Speed=@n48) {speedX=0.133, speedY=0.2}, Tex=NamedRerouteUsage()) {samplerSource=SSM_Wrap_WorldGroupSettings, samplerType=SAMPLERTYPE_Masks}.RGBA) {group=Effect_Galaxy, maskChannel=Green, parameterName=Eff_Galaxy_Star_Med_Noise_Channel, defaultValue=<rgba 0 1 0 0>}
  n50 = Multiply(A=Multiply(A=@n45.R, B=@n45.R), B=@n49.R)
  n51 = StaticBoolParameter() {group=Effect_Galaxy_Stars, parameterName="Eff_Galaxy_Star  Disable", sortPriority=0}
  n52 = VectorParameter() {group=Effect_Galaxy, parameterName=Eff_Galaxy_Fres_Color, channelNames {a=Opacity, b="", g="", r=""}, defaultValue=<rgba 1 1 1 1>}
  n53 = Fresnel(ExponentIn=ScalarParameter() {defaultValue=5, group=Effect_Galaxy, parameterName=Eff_Galaxy_Fres_EX}) {baseReflectFraction=0}
  n54 = LinearInterpolate(
   A=StaticSwitch(True=@n33, False=LinearInterpolate(A=@n40, B=@n42, Alpha=@n50), Value=@n51)
   B=@n52
   Alpha=Saturate(None=Multiply(A=@n52.Opacity, B=@n53))
  )
  n55 = VectorParameter() {
   group=Effect_Galaxy_Edge
   parameterName=Eff_Galaxy_Edge_Color_Start
   channelNames {a=desaturate, b="", g="", r=""}
   defaultValue=<rgba 1 1 1 0>
  }
  n56 = Subtract(
   A=Time()
   B=Add(
    A=ScalarParameter() {defaultValue=-9999999, group=Effect_Galaxy_Edge|Reactive, parameterName=Eff_Galaxy_Edge_TFBP, sortPriority=600}
    B=ScalarParameter() {group=Effect_Galaxy_Edge|Reactive, parameterName=Eff_Galaxy_Edge_Delay}
   )
  )
  n57 = ScalarParameter() {defaultValue=0.25, group=Effect_Galaxy_Edge|Reactive, parameterName=Eff_Galaxy_Edge_Duration}
  n58 = Power(
   Base=Saturate(None=Divide(A=@n56, B=@n57))
   Exp=ScalarParameter() {defaultValue=0.25, group=Effect_Galaxy_Edge|Reactive, parameterName=Eff_Galaxy_Edge_Falloff}
  )
  n59 = LinearInterpolate(A=@n55, B=@n17, Alpha=@n58)
  n60 = Multiply(
   A=Multiply(
    A=Desaturation(None=@n54, Fraction=@n17.desaturate)
    B=ScalarParameter() {defaultValue=4, group=Effect_Galaxy_Edge, parameterName=Eff_Galaxy_Edge_Brightnees}
   )
   B=@n59
  )
  n61 = LinearInterpolate(
   A=ScalarParameter() {defaultValue=1, group=Effect_Galaxy_Edge|Reactive, parameterName=Eff_Galaxy_Edge_Brightness_Start}
   B=ScalarParameter() {defaultValue=1, group=Effect_Galaxy_Edge|Reactive, parameterName=Eff_Galaxy_Edge_Brightness_End}
   Alpha=@n58
  )
  Multiply(A=@n60, B=@n61)
  n62 = Multiply(A=@n1, B=@n59)
  Multiply(A=@n62, B=@n61)
  NamedRerouteDeclaration(None=@n13.Normal) {name=Root_Normal, nodeColor=<rgba 0.4062 0 1 1>}
  n63 = MaterialFunctionCall(Float3=Abs(None=NamedRerouteUsage())) {materialFunction=&BreakOutFloat3Components}
  n64 = Constant() {r=0.5}
  n65 = NamedRerouteUsage()
  n66 = ComponentMask(None=@n65) {b=true, desc=YZ, g=true}
  n67 = ComponentMask(None=@n65) {desc=YX, g=true, r=true}
  n68 = ComponentMask(None=@n65) {b=true, desc=XZ, r=true}
  n69 = If(A=@n63.R, B=@n64, "A > B"=@n66, "A < B"=If(A=@n63.B, B=@n64, "A > B"=@n67, "A < B"=@n68))
  n70 = Constant() {r=0.5}
  n71 = Constant() {r=0.5}
  n72 = MaterialFunctionCall(Float4=NamedRerouteUsage()) {materialFunction=&BreakOutFloat4Components}
  n73 = LinearInterpolate(
   A=If(A=@n63.B, B=@n70, "A > B"=@n67, "A < B"=If(A=@n63.R, B=@n70, "A > B"=@n66, "A < B"=@n68))
   B=If(A=@n63.G, B=@n71, "A > B"=@n68, "A < B"=If(A=@n63.B, B=@n71, "A > B"=@n67, "A < B"=@n66))
   Alpha=@n72.G
  )
  n74 = Constant() {r=0.5}
  n75 = If(A=@n63.R, B=@n74, "A > B"=@n66, "A < B"=@n67)
  NamedRerouteDeclaration(None=LinearInterpolate(A=LinearInterpolate(A=@n69, B=@n73, Alpha=@n72.R), B=If(A=@n63.G, B=@n74, "A > B"=@n68, "A < B"=@n75), Alpha=@n72.B)) {name=UVS_Cheap_Tri_1, nodeColor=<rgba 1 0 0.8438 1>}
  n76 = VectorParameter() {
   group=Effect_Galaxy_Edge | Gmap Objects
   parameterName=Eff_Galaxy_Gmap_Range_Weapon
   channelNames {a="", b=High, g=Low, r=Use}
   defaultValue=<rgba 0 0.2 0.8 1>
  }
  n77 = VectorParameter() {
   group=Effect_Galaxy_Edge | Gmap Objects
   parameterName=Eff_Galaxy_Gmap_Range_Vehicle
   channelNames {a="", b=high, g=low, r=use}
   defaultValue=<rgba 0 0.2 0.8 1>
  }
  n78 = VectorParameter() {
   group=Effect_Galaxy_Edge | Gmap Objects
   parameterName=Eff_Galaxy_Gmap_Range_VehicleSide
   channelNames {a="", b=High, g=Low, r=Use}
   defaultValue=<rgba 0 0.2 0.8 1>
  }
  n79 = MaterialFunctionCall(Float4=NamedRerouteUsage()) {materialFunction=&BreakOutFloat4Components}
  n80 = VectorParameter() {
   group=Effect_Galaxy_Edge | Gmap Objects
   parameterName=Eff_Galaxy_Gmap_Range_Mech
   channelNames {a="", b=High, g=Low, r=Use}
   defaultValue=<rgba 0 0.2 0.8 1>
  }
  n81 = VectorParameter() {
   group=Effect_Galaxy_Edge | Gmap Objects
   parameterName=Eff_Galaxy_Gmap_Range_Character
   channelNames {a="", b=High, g=Low, r=Use}
   defaultValue=<rgba 0 0.2 0.8 1>
  }
  n82 = LinearInterpolate(
   A=LinearInterpolate(A=@n76.RGB, B=LinearInterpolate(A=@n77.RGB, B=@n78.RGB, Alpha=@n79.G), Alpha=@n79.R)
   B=LinearInterpolate(A=@n80.RGB, B=@n81.RGB, Alpha=@n79.A)
   Alpha=@n79.B
  )
  n83 = VectorParameter() {
   group=Effect_Galaxy_Edge | Gmap Objects
   parameterName=Eff_Galaxy_Gmap_Range_Backpack
   channelNames {a="", b=High, g=Low, r=Use}
   defaultValue=<rgba 0 0.2 0.8 1>
  }
  n84 = MaterialFunctionCall(Float2=NamedRerouteUsage()) {materialFunction=&BreakOutFloat2Components}
  n85 = VectorParameter() {
   group=Effect_Galaxy_Edge | Gmap Objects
   parameterName=Eff_Galaxy_Gmap_Range_Pickaxe
   channelNames {a="", b=High, g=Low, r=Use}
   defaultValue=<rgba 0 0.2 0.8 1>
  }
  NamedRerouteDeclaration(None=LinearInterpolate(A=LinearInterpolate(A=@n82, B=@n83.RGB, Alpha=@n84.R), B=@n85.RGB, Alpha=@n84.G)) {name=Range, nodeColor=<rgba 0 0.625 1 1>}
  n86 = FunctionInput(Preview=Constant4Vector() {constant=<rgba 0 0 5 -1>}.RGBA) {bUsePreviewValueAsDefault=true, inputName=Object-SafeZone-Virtual-Pivot, inputType=FunctionInput_Vector4, sortPriority=105}
  NamedRerouteDeclaration(None=@n86) {name=SafeZone VP, nodeColor=<rgba 0.5 1 0 1>}
  n87 = FunctionInput(Preview=Constant()) {bUsePreviewValueAsDefault=true, inputName=Object-SafeZone-VP-Character-PreView-Switch, inputType=FunctionInput_Scalar, sortPriority=105}
  NamedRerouteDeclaration(None=@n87) {name=SafeZone Character-Switch, nodeColor=<rgba 1 0.9375 0 1>}
  NamedRerouteDeclaration(None=@n16) {name=Gmap reactive, nodeColor=<rgba 0 1 0.25 1>}
  NamedRerouteDeclaration(None=@n5.RGBA) {name=CM_Mask, nodeColor=<rgba 1 0 0.3047 1>}
  n88 = TextureObjectParameter() {group=Effect_Galaxy, parameterName=Eff_Galaxy__Tex_Gal, sortPriority=5, texture=&T_BlacksRGB}
  NamedRerouteDeclaration(None=@n88) {name=Tex_Galx, nodeColor=<rgba 1 0 0.6094 1>}
  n89 = TextureObjectParameter() {group=Effect_Galaxy, parameterName=Eff_Galaxy__Tex_Star_Msk, sortPriority=5, texture=&T_BlacksRGB}
  NamedRerouteDeclaration(None=@n89) {name=Tex_Stars_Msk, nodeColor=<rgba 1 0 0.6094 1>}
  n90 = Multiply(A=ScalarParameter() {defaultValue=1, group=Effect_Galaxy_Edge, parameterName=Eff_Galaxy_TileRate}, B=NamedRerouteUsage())
  n91 = ComponentMask(None=VectorParameter() {group=Effect_Galaxy_Galaxy, parameterName=Eff_Galaxy_PanSpeed, defaultValue=<rgba 0.005 0.015 0 0>}.RGB) {g=true, r=true}
  NamedRerouteDeclaration(None=TextureSample(UVs=Panner(Coordinate=@n90, Speed=@n91) {speedY=0.02}, Tex=NamedRerouteUsage()).RGB) {name=Galaxy_Base, nodeColor=<rgba 1 0 0.5391 1>}
  n92 = LinearInterpolate(
   A=LinearInterpolate(A=@n54, B=@n60, Alpha=@n1)
   B=Add(A=@n54, B=@n62)
   Alpha=Saturate(None=ScalarParameter() {group=Effect_Galaxy_Edge, parameterName="Eff_Galaxy_Edge---->Additive"})
  )
  n93 = Multiply(A=@n29, B=@n92)
  NamedRerouteDeclaration(None=StaticSwitchParameter(True=LinearInterpolate(A=@n92, B=@n93, Alpha=@n31), False=@n92) {group=Effect_Galaxy_Galaxy, parameterName="Eff_Galaxy_BGMultTint---->CC"}) {name=Galaxy, nodeColor=<rgba 0 0.25 1 1>}
  n94 = ScalarParameter() {defaultValue=0.85, group=Effect_Galaxy, parameterName=Eff_Galaxy_Rougness}
  n95 = NamedRerouteUsage()
  n96 = NamedRerouteUsage()
  n97 = ChannelMaskParameter(None=TextureSample(UVs=Multiply(A=@n95, B=@n96), Tex=NamedRerouteUsage()).RGBA) {group=Effect_Galaxy, maskChannel=Alpha, parameterName=Eff_Galaxy_Rougness_Tri_Msk_Channel, defaultValue=<rgba 0 0 0 1>}
  n98 = LinearInterpolate(A=@n94, B=@n97.R, Alpha=ScalarParameter() {group=Effect_Galaxy, parameterName="Eff_Galaxy_Rougness_Tri_Msk   1=On"})
  n99 = Add(A=NamedRerouteUsage()) {constB=-0.5}
  n100 = ScalarParameter() {defaultValue=4, group=Effect_Galaxy_Edge, parameterName=Eff_Galaxy_Edge_Strength}
  n101 = SmoothStep(
   Min=ScalarParameter() {defaultValue=0.2, group=Effect_Galaxy_Edge, parameterName=Eff_Galaxy_Edge_Hardness_Min}
   Max=ScalarParameter() {defaultValue=0.3, group=Effect_Galaxy_Edge, parameterName=Eff_Galaxy_Edge_Hardness_Max}
   Value=Power(
    Base=Multiply(A=@n99, B=@n100)
    Exp=ScalarParameter() {defaultValue=1, group=Effect_Galaxy_Edge, parameterName=Eff_Galaxy_Edge_Curvature_Fallof}
   )
  )
  n102 = Saturate(None=@n101)
  n103 = LinearInterpolate(B=NamedRerouteUsage(), Alpha=ScalarParameter() {group=Effect_Galaxy, parameterName="Eff_Galaxy_Root_Emissive 1=on", sortPriority=0})
  n104 = NamedRerouteUsage()
  n105 = LinearInterpolate(
   A=VectorParameter() {group=Effect_Galaxy_Edge, parameterName=Eff_Galaxy_Gmap_Color_Start, defaultValue=<rgba 0 0 0 1>}.RGB
   B=VectorParameter() {group=Effect_Galaxy_Edge, parameterName=Eff_Galaxy_Gmap_Color, defaultValue=<rgba 0 0 0 1>}.RGB
   Alpha=NamedRerouteUsage()
  )
  n106 = VectorParameter() {group=Effect_Galaxy_Edge, parameterName=Eff_Galaxy_Gmap_Range, channelNames {a="", b="", g=Max, r=Min}, defaultValue=<rgba 1 0 0 1>}
  n107 = NamedRerouteUsage()
  n108 = MaterialFunctionCall(Color=NamedRerouteUsage()) {materialFunction=&SplitComponents}
  n109 = LinearInterpolate(A=SmoothStep(Min=@n106.Min, Max=@n106.Min, Value=@n107), B=SmoothStep(Min=@n108.G, Max=@n108.G, Value=@n107), Alpha=@n108.R)
  n110 = SetMaterialAttributes!(
   MaterialAttributes=NamedRerouteUsage()
   BaseColor=Constant()
   Roughness=LinearInterpolate(
    A=StaticSwitch(True=@n94, False=@n98, Value=@n51)
    B=ScalarParameter() {defaultValue=0.85, group=Effect_Galaxy, parameterName=Eff_Galaxy_Rougness_Edge}
    Alpha=@n102
   )
   Specular=LinearInterpolate(
    A=ScalarParameter() {group=Effect_Galaxy, parameterName=Eff_Galaxy_Spec}
    B=ScalarParameter() {defaultValue=0.5, group=Effect_Galaxy, parameterName=Eff_Galaxy_Spec_Edge}
    Alpha=@n102
   )
   Metallic=ScalarParameter() {group=Effect_Galaxy, parameterName=Eff_Galaxy_Metallic}
   Normal
   EmissiveColor=Add(A=Add(A=@n103, B=@n104), B=Multiply(A=@n105, B=@n109))
  ) {attributeSetTypes [
    #2
    #3
    #5
    #6
    #1
    #4
   ]}
  NamedRerouteDeclaration(None=@n110) {name=Mat_Galaxy, nodeColor=<rgba 0 0.3906 1 1>}
  NamedRerouteDeclaration(None=@n13.EmissiveColor) {name=Root_Emissive, nodeColor=<rgba 0.4062 0 1 1>}
  NamedRerouteDeclaration(None=@n101) {name=Curvature_output, nodeColor=<rgba 0 1 0.8125 1>}
  n111 = ScalarParameter() {defaultValue=1, group=Effect_Galaxy_Uvs, parameterName=Eff_Roughness_Scale_Vehicle}
  n112 = ScalarParameter() {defaultValue=1, group=Effect_Galaxy_Uvs, parameterName=Eff_Roughness_Scale_Vehicle_Side}
  n113 = MaterialFunctionCall(Float4=NamedRerouteUsage()) {materialFunction=&BreakOutFloat4Components}
  n114 = LinearInterpolate(
   A=LinearInterpolate(
    A=ScalarParameter() {defaultValue=1, group=Effect_Galaxy_Uvs, parameterName=Eff_Roughness_Scale_Weapon}
    B=LinearInterpolate(A=@n111, B=@n112, Alpha=@n113.G)
    Alpha=@n113.R
   )
   B=LinearInterpolate(
    A=ScalarParameter() {defaultValue=1, group=Effect_Galaxy_Uvs, parameterName=Eff_Roughness_Scale_Warp_Mech}
    B=ScalarParameter() {defaultValue=1, group=Effect_Galaxy_Uvs, parameterName=Eff_Roughness_Scale_Character}
    Alpha=@n113.A
   )
   Alpha=@n113.B
  )
  n115 = ScalarParameter() {defaultValue=1, group=Effect_Galaxy_Uvs, parameterName=Eff_Roughness_Scale_Backpack}
  n116 = MaterialFunctionCall(Float2=NamedRerouteUsage()) {materialFunction=&BreakOutFloat2Components}
  NamedRerouteDeclaration(None=LinearInterpolate(
    A=LinearInterpolate(A=@n114, B=@n115, Alpha=@n116.R)
    B=ScalarParameter() {defaultValue=1, group=Effect_Galaxy_Uvs, parameterName=Eff_Roughness_Scale_Pickaxe}
    Alpha=@n116.G
   )) {name=Texture_Scale_1, nodeColor=<rgba 0 0.7188 1 1>}
  n117 = ScalarParameter() {group=Effect_Galaxy, parameterName=Eff_Galaxy_WarpStrength}
  n118 = Add(A=NamedRerouteUsage()) {constB=-0.5}
  n119 = Divide(B=MaterialFunctionCall() {materialFunction=&ObjectScale}.Scale Z)
  n120 = NamedRerouteUsage()
  n121 = ComponentMask(None=@n120) {b=true, g=true, r=true}
  n122 = MaterialFunctionCall!(Center Point World Position=LinearInterpolate(
    A=TransformPosition(None=@n121) {transformType=TRANSFORMPOSSOURCE_World}
    B=ComponentMask(None=@n120) {b=true, g=true, r=true}
    Alpha=NamedRerouteUsage()
   )) {materialFunction=/Engine/Functions/Engine_MaterialFunctions02/UVs/WorldSpaceAlignedScreenCoordinates}
  n123 = Constant() {r=-2500}
  n124 = PreSkinnedLocalBounds()
  n125 = MaterialFunctionCall(Float3=Subtract(A=@n124.Max, B=@n124.Max)) {materialFunction=&BreakOutFloat3Components}
  n126 = Multiply(A=ComponentMask(None=@n120) {a=true}, B=Max(A=Max(A=@n125.R, B=@n125.R), B=@n125.B))
  n127 = Add(A=Multiply(A=Multiply(A=@n119, B=@n122.Offset and Scaled UVS), B=Divide(A=@n123, B=@n126))) {constB=0.5}
  n128 = Constant3Vector() {constant=<rgba 1 1 0 0>}
  n129 = MaterialFunctionCall() {materialFunction=&CameraDirectionVector}
  n130 = Multiply(B=Normalize(VectorInput=Multiply(A=@n128, B=@n129.Result))) {constA=0.25}
  n131 = ActorPositionWS()
  n132 = Constant3Vector() {constant=<rgba 0 0 90 1>}
  n133 = NamedRerouteUsage()
  n134 = Multiply(A=@n133) {constB=0.5}
  n135 = NamedRerouteUsage()
  n136 = MaterialFunctionCall(
   Plane Normal Axis=Normalize(VectorInput=Add(A=@n130, B=@n129.Result))
   Plane Center=Add(A=Multiply(A=@n129.Result) {constB=50}, B=Add(A=@n131, B=@n132))
   UV Scale=Multiply(
    A=LinearInterpolate(A=@n134, B=@n133, Alpha=@n135)
    B=MaterialFunctionCall() {materialFunction=/Engine/Functions/Engine_MaterialFunctions02/WorldPositionOffset/ObjectScale}.Scale Z
   )
  ) {materialFunction=&VirtualPlaneCoordinates}
  NamedRerouteDeclaration(None=Add(
    A=Multiply(A=@n117, B=@n118)
    B=StaticSwitchParameter(True=@n127, False=@n136.UVs) {group=Effect_Galaxy_Uvs, parameterName="Eff_Galaxy_FOV---->ScreenSpace"}
   )) {name=ScreenSpaceUVs, nodeColor=<rgba 0.5469 0 1 1>}
  NamedRerouteDeclaration(None=FunctionInput(Preview=Constant4Vector().RGBA) {bUsePreviewValueAsDefault=true, inputName=IsState, inputType=FunctionInput_Vector4, sortPriority=4}) {name=IsState, nodeColor=<rgba 0.1497 1 0.1633 1>}
  n137 = ScalarParameter() {defaultValue=160, group=Effect_Galaxy_Uvs, parameterName=Eff_Galaxy_Scale_Vehicle}
  n138 = ScalarParameter() {defaultValue=160, group=Effect_Galaxy_Uvs, parameterName=Eff_Galaxy_Scale_Vehicle_Side}
  n139 = MaterialFunctionCall(Float4=NamedRerouteUsage()) {materialFunction=&BreakOutFloat4Components}
  n140 = LinearInterpolate(
   A=LinearInterpolate(
    A=ScalarParameter() {defaultValue=160, group=Effect_Galaxy_Uvs, parameterName=Eff_Galaxy_Scale_Weapon}
    B=LinearInterpolate(A=@n137, B=@n138, Alpha=@n139.G)
    Alpha=@n139.R
   )
   B=LinearInterpolate(
    A=ScalarParameter() {defaultValue=160, group=Effect_Galaxy_Uvs, parameterName=Eff_Galaxy_Scale_Mech}
    B=ScalarParameter() {defaultValue=160, group=Effect_Galaxy_Uvs, parameterName=Eff_Galaxy_Scale_Character}
    Alpha=@n139.A
   )
   Alpha=@n139.B
  )
  n141 = ScalarParameter() {defaultValue=160, group=Effect_Galaxy_Uvs, parameterName=Eff_Galaxy_Scale_Backpack}
  n142 = MaterialFunctionCall(Float2=NamedRerouteUsage()) {materialFunction=&BreakOutFloat2Components}
  NamedRerouteDeclaration(None=LinearInterpolate(
    A=LinearInterpolate(A=@n140, B=@n141, Alpha=@n142.R)
    B=ScalarParameter() {defaultValue=160, group=Effect_Galaxy_Uvs, parameterName=Eff_Galaxy_Scale_Pickaxe}
    Alpha=@n142.G
   )) {name=Texture_Scale, nodeColor=<rgba 0 0.7188 1 1>}
  n143 = InverseLinearInterpolate(
   A=Constant() {r=45}
   B=Constant() {r=90}
   Value=MaterialFunctionCall() {materialFunction=&FOV}.FOV angle in degrees
  )
  n144 = MaterialFunctionCall(Float4=NamedRerouteUsage()) {materialFunction=&BreakOutFloat4Components}
  n145 = LinearInterpolate(A=LinearInterpolate(A=@n143, Alpha=@n144.R), B=LinearInterpolate(B=@n143, Alpha=@n144.A) {constA=1}, Alpha=@n144.B)
  n146 = MaterialFunctionCall(Float2=NamedRerouteUsage()) {materialFunction=&BreakOutFloat2Components}
  NamedRerouteDeclaration(None=LinearInterpolate(A=LinearInterpolate(A=@n145, B=@n143, Alpha=@n146.R), B=@n143, Alpha=@n146.G)) {name=Fov_State, nodeColor=<rgba 0 0.7188 1 1>}
  NamedRerouteDeclaration(None=FunctionInput(Preview=Constant2Vector()) {bUsePreviewValueAsDefault=true, inputName=IsState_B, inputType=FunctionInput_Vector2, sortPriority=4}) {name=IsState_B, nodeColor=<rgba 0.1497 1 0.1633 1>}
  n147 = PreSkinnedLocalBounds()
  n148 = Subtract(A=VertexInterpolator(VS=PreSkinnedPosition()).PS, B=@n147.Min)
  n149 = MaterialFunctionCall(Color=Subtract(A=@n147.Max, B=@n147.Max)) {materialFunction=&SplitComponents}
  n150 = Max(A=Max(A=@n149.R, B=@n149.R), B=@n149.B)
  NamedRerouteDeclaration(None=FunctionInput(Preview=Divide(A=@n148, B=@n150)) {bUsePreviewValueAsDefault=true, inputName=Calcated_Bounds, sortPriority=11}) {name=Caclcated_Bounds, nodeColor=<rgba 0 0.9297 1 1>}
  NamedRerouteDeclaration(None=@n6.G) {name=Msk_Curvature, nodeColor=<rgba 0 1 0.7422 1>}
  n151 = PreSkinnedNormal()
  NamedRerouteDeclaration(None=FunctionInput(Preview=VertexInterpolator(VS=@n151).PS) {bUsePreviewValueAsDefault=true, inputName=Pre-Skinned Local Normal, sortPriority=10}) {name=PreSkin_Normal, nodeColor=<rgba 0.8281 1 0 1>}
  NamedRerouteDeclaration(None=@n6.A) {name=Msk_Gmap, nodeColor=<rgba 0 1 0.7656 1>}
  Saturate(None=Add(A=NamedRerouteUsage()) {constB=-0.5})
  TransformPosition(None=WorldPosition() {worldPositionShaderOffset=WPT_ExcludeAllShaderOffsets}.XYZ) {transformSourceType=TRANSFORMPOSSOURCE_World}
  VertexNormalWS()
 ]}
