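# Material function for weapon wrap skins that layers an optional iridescent effect onto incoming material attributes.
# Scaled, wrapped Fresnel indexes a gradient texture or curve-atlas color, feeding base color or emissive.
# The full variant also reshapes roughness and metallic, while a simplified switch skips this and changes masking.
# Customization, topside, base-color and decal masks drive a standard layer blend, routed to the main or "after" output by an after-accent switch.
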MaterialFunction /Game/Weapons/WeaponSkins/Wraps/MaterialFunctions/MF_Effect_Iridescent
graph {nodes [
  n1 = Multiply(
   A=NamedRerouteUsage()
   B=ScalarParameter() {defaultValue=1, group=Effect_Iridescent_Simple, parameterName="Irad_Msk_Primary   1=On  0=Off", sortPriority=16}
  )
  n2 = Multiply(
   A=NamedRerouteUsage()
   B=ScalarParameter() {defaultValue=1, group=Effect_Iridescent_Simple, parameterName="Irad_Msk_Accent   1=On  0=Off", sortPriority=16}
  )
  Add(A=@n1, B=@n2)
  n3 = NamedRerouteUsage()
  n4 = FunctionInput() {inputName=Material, inputType=FunctionInput_MaterialAttributes, sortPriority=-200}
  n5 = GetMaterialAttributes(None=@n4) {attributeGetTypes [
    #1
    #2
    #3
    #4
    #5
   ]}
  n6 = Fresnel(
   ExponentIn=ScalarParameter() {defaultValue=0.5, group=Effect_Iridescent, parameterName=Irad_Fresnel_Exponent, sortPriority=10}
   BaseReflectFractionIn=ScalarParameter() {defaultValue=0.08, group=Effect_Iridescent, parameterName=Irad_Fresnel_BaseReflecFraction, sortPriority=10}
   Normal=LinearInterpolate(A=@n5.Normal, B=PixelNormalWS(), Alpha=ScalarParameter() {group=Effect_Iridescent, parameterName=Irad_Use_Pixel_Normals, sortPriority=0})
  )
  n7 = ScalarParameter() {defaultValue=0.99, group=Effect_Iridescent, parameterName=Irad_Fresnel_CC_TileRate, sortPriority=20}
  n8 = Frac(None=Add(A=Multiply(A=@n6, B=@n7), B=ScalarParameter() {group=Effect_Iridescent, parameterName=Irad_Fresnel_CC_Offset, sortPriority=20}))
  n9 = TextureSampleParameter2D(UVs=@n8) {group=Effect_Iridescent, parameterName=Irad_Fresnel_Tex, sortPriority=0, texture=&T_BlacksRGB}
  n10 = CurveAtlasRowParameter(CurveTime=@n8) {
   atlas=/Game/Weapons/WeaponSkins/Wraps/Textures/CruveAtlas_Wraps_01
   curve=/Game/Weapons/WeaponSkins/Wraps/Textures/C_GradSimple
   group=Effect_Iridescent
   parameterName=Irad_Fresnel_ColorCurve
   sortPriority=25
  }
  n11 = MaterialFunctionCall!(
   Index=ScalarParameter() {group=Effect_Iridescent, parameterName="Irad_Fresnel_ColorCurve  Index", sortPriority=25}
   CCinput=@n10
   CCinput_Alpha=@n10
   UV
  ) {materialFunction=&/MF_Global_CC_Tex}
  n12 = StaticBoolParameter() {group=Effect_Iridescent, parameterName="Irad_Fresnel_ColorCurve ---->Texture", sortPriority=0}
  n13 = Add(
   A=ScalarParameter() {defaultValue=1, group=Effect_Iridescent, parameterName="Irad_Fresnel_CC_Mask using Fresnel  0=Off", sortPriority=20}
   B=@n6
  )
  n14 = ScalarParameter() {defaultValue=2, group=Effect_Iridescent, parameterName=Irad_Fresnel_Intensity, sortPriority=20}
  n15 = Multiply(A=Multiply(A=Saturate(None=@n13), B=StaticSwitch(True=@n9.RGB, False=@n11.3V, Value=@n12)), B=@n14)
  n16 = ScalarParameter() {group=Effect_Iridescent, parameterName=Irad_Make_Emissive, sortPriority=20}
  n17 = Add(A=LinearInterpolate(A=@n15, Alpha=@n16) {constB=0}, B=@n5.BaseColor)
  n18 = Add(A=LinearInterpolate(B=@n15, Alpha=@n16), B=@n5.EmissiveColor)
  n19 = SetMaterialAttributes!(MaterialAttributes=@n5.MaterialAttributes, BaseColor=@n17, Roughness, Normal, Metallic, EmissiveColor=@n18) {attributeSetTypes [
    #1
    #2
    #3
    #4
    #5
   ]}
  n20 = Clamp(None=LinearInterpolate(Alpha=Clamp(None=@n5.Roughness) {maxDefault=0.5}) {constB=2})
  n21 = ScalarParameter() {defaultValue=1, group=Effect_Iridescent, parameterName=Irad_Roughness_pits}
  n22 = ScalarParameter() {defaultValue=1.3262, group=Effect_Iridescent, parameterName=Irad_Depth_Contrast}
  n23 = Clamp(None=Multiply(A=Power(Base=@n20, Exp=@n22), B=ScalarParameter() {defaultValue=1.2614, group=Effect_Iridescent, parameterName=Irad_Depth_intensity}))
  n24 = SetMaterialAttributes(
   MaterialAttributes=@n5.MaterialAttributes
   BaseColor=Multiply(A=@n17, B=LinearInterpolate(Alpha=@n20) {constA=0.5})
   Roughness=Saturate(None=LinearInterpolate(A=@n21, B=@n5.Roughness, Alpha=@n23))
   Normal=@n5.Normal
   Metallic=Multiply(A=@n14, B=StaticSwitch(True=@n9.A, False=@n11.Alpha, Value=@n12))
   EmissiveColor=@n18
  ) {attributeSetTypes [
    #1
    #2
    #3
    #4
    #5
   ]}
  n25 = FunctionInput(Preview=Constant()) {bUsePreviewValueAsDefault=true, inputName=Decal_Mask, inputType=FunctionInput_Scalar}
  n26 = FunctionInput() {inputName=TS_Mask_02, inputType=FunctionInput_Vector4, sortPriority=50}
  n27 = FunctionInput() {inputName=TS_Mask_04, inputType=FunctionInput_Vector4, sortPriority=50}
  n28 = FunctionInput() {inputName=TS_Mask_01, inputType=FunctionInput_Vector4, sortPriority=49}
  n29 = ChannelMaskParameter(None=@n28) {group=Effect_Iridescent_Simple, parameterName=Irad_TS_Mask_01_Channels, sortPriority=10}
  n30 = StaticSwitchParameter(
   True=ChannelMaskParameter(None=FunctionInput() {inputName=TS_Mask_03, inputType=FunctionInput_Vector4, sortPriority=49}) {group=Effect_Iridescent_Simple, parameterName=Irad_TS_Mask_03_Channels, sortPriority=10}.R
   False=StaticSwitchParameter(
    True=ChannelMaskParameter(None=@n27) {group=Effect_Iridescent_Simple, parameterName=Irad_TS_Mask_04_Channels, sortPriority=10}.R
    False=@n29.R
   ) {group=Effect_Iridescent_Simple, parameterName="Irad_TS_Mask_01-------->TS_Mask_04", sortPriority=1}
  ) {group=Effect_Iridescent_Simple, parameterName="Irad_TS_Mask_01-------->TS_Mask_03", sortPriority=1}
  n31 = StaticBoolParameter() {group=Customization, parameterName="Custmization  V2", sortPriority=-500}
  n32 = StaticSwitchParameter(
   True=StaticSwitchParameter(
    True=ChannelMaskParameter(None=@n26) {group=Effect_Iridescent_Simple, parameterName=Irad_TS_Mask_02_Channels, sortPriority=10}.R
    False=StaticSwitch(True=@n30, False=@n29.R, Value=@n31)
   ) {group=Effect_Iridescent_Simple, parameterName="Irad_TS_Mask_01-------->TS_Mask_02", sortPriority=1}
   False=ScalarParameter() {defaultValue=1, group=Effect_Iridescent_Simple, parameterName=Irad_Amount_Simple, sortPriority=0}
  ) {desc="Only works with Topside masks. ", group=Effect_Iridescent_Simple, parameterName="Irad_Simple-------->TS_Mask_01", sortPriority=1}
  n33 = ScalarParameter() {group=Effect_Iridescent_Simple, parameterName="Irad_Msk_Disable_Accent+Primary   1=on  0=off", sortPriority=15}
  n34 = Multiply(A=@n32, B=Saturate(None=Add(A=@n1, B=@n33)))
  n35 = NamedRerouteUsage()
  n36 = NamedRerouteUsage()
  n37 = FunctionInput(Preview=Saturate(None=Add(A=@n35, B=@n36))) {bUsePreviewValueAsDefault=true, inputName="Customization Mask (_CM B)", inputType=FunctionInput_Scalar, sortPriority=24}
  n38 = Add(A=@n37, B=ScalarParameter() {group=Effect_Iridescent, parameterName=Irad_Remove_Customization Mask, sortPriority=0})
  n39 = StaticSwitchParameter(
   True=ChannelMaskParameter(None=@n28) {group=Effect_Iridescent, parameterName=Irad_TS_Mask_01, sortPriority=5}.R
   False=ChannelMaskParameter(None=@n5.BaseColor) {group=Effect_Iridescent, parameterName=Irad_Basecolor_Channel_Mask_, sortPriority=5}.R
  ) {group=Effect_Iridescent, parameterName="Irad_Basecolor_Channel_Mask---->TS_Mask_01", sortPriority=5}
  n40 = StaticSwitchParameter(
   True=Saturate(None=Add(A=@n34, B=@n2))
   False=Multiply(
    A=Saturate(None=@n38)
    B=StaticSwitchParameter(True=@n39, False=@n37) {group=Effect_Iridescent, parameterName="Irad_Basecolor_Channel_Mask ", sortPriority=5}
   )
  ) {group=Effect_Iridescent, parameterName=Irad_Simplfied, sortPriority=-1}
  n41 = Add(A=@n25, B=@n40)
  n42 = LinearInterpolate(
   A=Multiply(A=OneMinus(None=@n25), B=@n40)
   B=@n41
   Alpha=ScalarParameter() {defaultValue=1, group=Effect_Iridescent_Simple, parameterName="Irad_Mask Left/Right_Decal 1=off 0=on", sortPriority=0}
  )
  n43 = MaterialFunctionCall!(
   Base Material=@n3
   Top Material=StaticSwitchParameter(True=@n19, False=@n24) {group=Effect_Iridescent, parameterName=Irad_Simplfied, sortPriority=-1}
   Alpha=StaticSwitchParameter(True=Saturate(None=@n42), False=@n40) {group=Customization, parameterName=Function_Left_Right_Decal}
  ) {materialFunction=/Engine/Functions/MaterialLayerFunctions/MatLayerBlend_Standard}
  FunctionOutput(None=StaticSwitchParameter(
    True=StaticSwitchParameter(True=@n3, False=@n43.Blended Material) {group=Customization, parameterName=Effect_Iridescent_After_Accent}
    False=@n3
   ) {group=Customization, parameterName="Effect_Iridescent "})
  FunctionOutput(None=StaticSwitchParameter(True=@n43.Blended Material, False=@n3) {group=Customization, parameterName=Effect_Iridescent_After_Accent}) {outputName=after, sortPriority=1}
  Multiply(A=NamedRerouteUsage(), B=ScalarParameter() {group=Effect_Iridescent_Simple, parameterName="Irad_Msk_Accent   Fill white", sortPriority=16})
  NamedRerouteDeclaration(None=@n4) {name=Root_material, nodeColor=<rgba 0.9688 0 1 1>}
  n44 = TextureSample() {samplerType=SAMPLERTYPE_Masks, texture=/Game/Weapons/FORT_Rifles/Textures/T_Scar_CM}
  n45 = FunctionInput(Preview=@n44.G) {bUsePreviewValueAsDefault=true, inputName=Mask Crevise, inputType=FunctionInput_Scalar, sortPriority=24, previewValue=<1 0 0 1>}
  NamedRerouteDeclaration(None=@n45) {name=Msk_Curvature, nodeColor=<rgba 0 1 0.7422 1>}
  n46 = FunctionInput(Preview=@n44.R) {bUsePreviewValueAsDefault=true, inputName=Mask Primary, inputType=FunctionInput_Scalar, sortPriority=25, previewValue=<1 0 0 1>}
  NamedRerouteDeclaration(None=@n46) {name=Msk_Primary, nodeColor=<rgba 0 1 0.7656 1>}
  n47 = FunctionInput(Preview=@n44.B) {bUsePreviewValueAsDefault=true, inputName=Mask Secondary, inputType=FunctionInput_Scalar, sortPriority=26, previewValue=<1 0 0 1>}
  NamedRerouteDeclaration(None=@n47) {name=Msk_Accent, nodeColor=<rgba 0 1 0.7656 1>}
  Reroute()
  Reroute(None=@n4)
  Saturate(None=@n41)
  Saturate(None=Add(A=@n39, B=@n2))
  StaticSwitchParameter() {group=Effect_Iridescent, parameterName=Irad_Simplfied, sortPriority=-1}
 ]}
