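# Weapon wrap material function that scales the incoming material's emissive by time of day, enabled by a static switch that otherwise passes attributes through.
# A second switch picks the multiplier: a channel-masked color curve via MF_Global_CC_Tex, or Scalar_Time_of_Day using Morning, Day, Evening and Night scalars.
# The curve samples the TimeOfDay collection value plus a front-end override, which also adjusts the Day scalar.
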MaterialFunction /Game/Weapons/WeaponSkins/Wraps/MaterialFunctions/MF_Effect_TOD
graph {nodes [
  n1 = GetMaterialAttributes(None=FunctionInput() {inputName=Material, inputType=FunctionInput_MaterialAttributes}) {attributeGetTypes [
    #1
    #2
    #3
    #4
   ]}
  n2 = NamedRerouteDeclaration(None=@n1.MaterialAttributes) {name=Root_Material, nodeColor=<rgba 0.7812 1 0 1>}
  n3 = ScalarParameter() {group=00-FrontEnd-Overrides, parameterName=00-FE-TODCC-Override}
  n4 = Add(
   A=CollectionParameter() {collection=&FortniteMaterialParameters, parameterName=TimeOfDay_0-1}
   B=@n3
  )
  n5 = CurveAtlasRowParameter(CurveTime=@n4) {
   atlas=/Game/Weapons/WeaponSkins/Wraps/Textures/CruveAtlas_Wraps_01
   curve=/Game/Weapons/WeaponSkins/Wraps/Textures/0_ColorCurves/Curve_TimeOfDay
   defaultValue=182
   group=Effect_TOD_Emissive
   parameterName=Effect_TOD_TOD_CC
  }
  n6 = MaterialFunctionCall!(
   Index=ScalarParameter() {defaultValue=182, group=Effect_TOD_Emissive, parameterName="Effect_TOD_TOD_CC  Index"}
   CCinput=@n5
   CCinput_Alpha=@n5
   UV=@n4
  ) {materialFunction=&/MF_Global_CC_Tex}
  n7 = ScalarParameter() {defaultValue=0.5, group=Effect_TOD_Emissive, parameterName=Effect_TOD_Morning}
  n8 = ScalarParameter() {defaultValue=1, group=Effect_TOD_Emissive, parameterName=Effect_TOD_Day}
  n9 = Add(A=LinearInterpolate(A=@n3, B=Multiply(A=@n3) {constB=-1}, Alpha=@n8), B=@n8)
  n10 = ScalarParameter() {defaultValue=0.5, group=Effect_TOD_Emissive, parameterName=Effect_TOD_Evening}
  n11 = ScalarParameter() {defaultValue=0.25, group=Effect_TOD_Emissive, parameterName=Effect_TOD_Night}
  n12 = Multiply(
   A=@n1.EmissiveColor
   B=StaticSwitchParameter(
    True=ChannelMaskParameter(None=@n6.4v) {group=Effect_TOD_Emissive, parameterName=Effect_TOD_CC_Channel}.R
    False=MaterialFunctionCall(Value_Morning=@n7, Value_Day=@n9, Value_Evening=@n10, Value_Night=@n11) {materialFunction=&Scalar_Time_of_Day}.Result
   ) {group=Effect_TOD_Emissive, parameterName="Effect_TOD --->CCTOD"}
  )
  FunctionOutput(None=StaticSwitchParameter(
    True=SetMaterialAttributes!(MaterialAttributes=@n2, EmissiveColor=@n12) {attributeSetTypes [#4]}
    False=@n2
   ) {group=Customization, parameterName=Effect_TOD_Emissive})
  NamedRerouteDeclaration(None=@n1.Normal) {name=Root_Normal, nodeColor=<rgba 0.4062 0 1 1>}
 ]}
