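# Material function for weapon skin wraps that adds an optional, masked Fresnel rim glow to a material's emissive color.
# A Fresnel term from the transformed Normal input is scaled by a color parameter and the Primary Mask's red channel, with exponent exposed.
# The result adds to the incoming EmissiveColor within the material attributes.
# The "Use Emissive Fresnel" static switch selects this or passes the attributes through unchanged.
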MaterialFunction /Game/Weapons/WeaponSkins/Wraps/MaterialFunctions/MF_FresnelEmissive
graph {nodes [
  n1 = FunctionInput(Preview=MaterialFunctionCall() {materialFunction=/Game/Characters/Enemies/MaterialAttributes/MaterialAttributeTester}.Result) {inputName=Materials, inputType=FunctionInput_MaterialAttributes}
  n2 = GetMaterialAttributes(None=@n1) {attributeGetTypes [#1, #2]}
  n3 = FunctionInput() {bUsePreviewValueAsDefault=true, inputName=Primary Mask, sortPriority=3}
  n4 = Fresnel(
   ExponentIn=ScalarParameter() {defaultValue=3, group=Emissive Fresnel, parameterName=EmissiveFresnel_Exponent}
   Normal=Transform(None=FunctionInput() {inputName=Normal, sortPriority=1})
  ) {baseReflectFraction=0}
  n5 = VectorParameter() {group=Emissive Fresnel, parameterName=EmissiveFresnel_Color, defaultValue=<rgba 1 1 1 1>}
  n6 = Add(A=@n2.EmissiveColor, B=Multiply(A=ComponentMask(None=@n3) {r=true}, B=Multiply(A=@n4, B=@n5.RGB)))
  FunctionOutput(None=StaticSwitchParameter(
    True=SetMaterialAttributes!(MaterialAttributes=@n2.MaterialAttributes, EmissiveColor=@n6) {attributeSetTypes [#1]}
    False=@n1
   ) {group=Customization, parameterName=Use Emissive Fresnel})
  Multiply(A=@n2.EmissiveColor, B=@n5.RGB)
 ]}
