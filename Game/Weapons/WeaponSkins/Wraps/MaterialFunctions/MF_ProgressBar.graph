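# Material function for weapon wrap skins that draws an animated horizontal progress bar mask with a fading tail.
# Two custom HLSL step nodes on UV.y are subtracted to cut a band set by VerticalOffset and Thickness.
# UVs rotated by ProgressBarTipAngle supply a U gradient that scrolls with Time and Speed and wraps through Frac.
# A Power by ProgressBarTailContrast shapes the tail before it multiplies the band.
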MaterialFunction /Game/Weapons/WeaponSkins/Wraps/MaterialFunctions/MF_ProgressBar
graph {nodes [
  n1 = FunctionInput(Preview=TextureCoordinate()) {bUsePreviewValueAsDefault=true, inputName=TexCoord, inputType=FunctionInput_Vector2}
  MaterialExpressionComponentMask_0 = ComponentMask(None=@n1) {g=true}
  MaterialExpressionFunctionInput_0 = FunctionInput(Preview=Constant() {r=0.5}) {bUsePreviewValueAsDefault=true, inputName=VerticalOffset}
  MaterialExpressionAdd_0 = Add(
   A=@MaterialExpressionFunctionInput_0
   B=FunctionInput(Preview=Constant() {r=0.1}) {bUsePreviewValueAsDefault=true, inputName=Thickness, inputType=FunctionInput_Scalar}
  )
  n2 = Custom!(a=@MaterialExpressionComponentMask_0, b=@MaterialExpressionAdd_0) {
   code="return step(a,b);"
   description=Step
   outputType=CMOT_Float1
   inputs [
    {inputName=a, input {expression=&self:MaterialExpressionComponentMask_0}}
    {inputName=b, input {expression=&self:MaterialExpressionAdd_0}}
   ]
  }
  n3 = Custom(a=@MaterialExpressionComponentMask_0, b=@MaterialExpressionFunctionInput_0) {
   code="return step(a,b);"
   description=Step
   outputType=CMOT_Float1
   inputs [
    {inputName=a, input {expression=&self:MaterialExpressionComponentMask_0}}
    {inputName=b, input {expression=&self:MaterialExpressionFunctionInput_0}}
   ]
  }
  n4 = Rotator(
   Coordinate=@n1
   Time=FunctionInput(Preview=Constant()) {bUsePreviewValueAsDefault=true, inputName=ProgressBarTipAngle, inputType=FunctionInput_Scalar}
  )
  n5 = FunctionInput() {bUsePreviewValueAsDefault=true, inputName=Speed, inputType=FunctionInput_Scalar, previewValue=<1 0 0 1>}
  n6 = Time()
  n7 = Frac(None=Subtract(A=ComponentMask(None=@n4) {r=true}, B=Multiply(A=@n5, B=@n6)))
  n8 = FunctionInput(Preview=Constant() {r=10}) {bUsePreviewValueAsDefault=true, inputName=ProgressBarTailContrast, inputType=FunctionInput_Scalar}
  FunctionOutput(None=Multiply(A=Subtract(A=@n2, B=@n3), B=Power(Base=@n7, Exp=@n8))) {outputName=Output}
 ]}
