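# Material function for weapon-skin wraps that projects an atlased normal map from one side, with vehicle, vehicle-side and mech variants.
# Builds UVs from pre-skinned position and local bounds, then applies offset, rotation, centered scale and atlas cell placement.
# Swizzles the sampled normal per projection axis, rotates it, subtracts the masked mesh normal and converts local to tangent space.
# Outputs the normal, raw delta, UVs and safe-zone mask.
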MaterialFunction /Game/Weapons/WeaponSkins/Wraps/MaterialFunctions/MF_Projection_Side_Nomal_Atlas
graph {nodes [
  Constant2Vector() {g=-0.5, r=0.5}
  Constant2Vector() {g=-0.5, r=0.5}
  n1 = Constant3Vector() {constant=<rgba 0 1 0 0>}
  n2 = MaterialFunctionCall(
   R=FunctionInput(Preview=Constant()) {bUsePreviewValueAsDefault=true, inputName=Is Vehicle, inputType=FunctionInput_Scalar, sortPriority=5}
   G=FunctionInput(Preview=Constant()) {bUsePreviewValueAsDefault=true, inputName=Is VehicleSide, inputType=FunctionInput_Scalar, sortPriority=6}
   B=FunctionInput(Preview=Constant()) {bUsePreviewValueAsDefault=true, inputName=Is Mech, inputType=FunctionInput_Scalar, sortPriority=7}
  ) {materialFunction=&AppendMany}
  n3 = MaterialFunctionCall(Color=@n2.RGB) {materialFunction=&SplitComponents}
  n4 = LinearInterpolate(A=LinearInterpolate(A=@n1, B=LinearInterpolate(B=@n1, Alpha=@n3.G), Alpha=@n3.R), B=@n1, Alpha=@n3.B)
  n5 = PreSkinnedPosition()
  n6 = PreSkinnedLocalBounds()
  n7 = Divide(A=Subtract(A=VertexInterpolator(VS=@n5).PS, B=@n6.Min), B=Subtract(A=@n6.Max, B=@n6.Max))
  n8 = MaterialFunctionCall(Color=Multiply(
    A=FunctionInput(Preview=@n7) {bUsePreviewValueAsDefault=true, inputName=Caculated_Bounds, sortPriority=1}
    B=Constant3Vector() {constant=<rgba 1 1 -1 0>}
   )) {materialFunction=&SplitComponents}
  n9 = AppendVector(A=@n8.R, B=@n8.R)
  n10 = LinearInterpolate(A=AppendVector(A=@n8.G, B=@n8.G), B=LinearInterpolate(A=AppendVector(A=@n8.R, B=@n8.R), B=@n9, Alpha=@n3.G), Alpha=@n3.R)
  n11 = Constant3Vector()
  n12 = Add(
   A=Add(A=ComponentMask(None=@n4) {g=true, r=true}, B=LinearInterpolate(A=@n10, B=@n9, Alpha=@n3.B))
   B=ComponentMask(None=FunctionInput(Preview=@n11) {bUsePreviewValueAsDefault=true, description=defualt 0, inputName=Uvs_Offset, sortPriority=12}) {g=true, r=true}
  )
  n13 = FunctionInput() {
   bUsePreviewValueAsDefault=true
   desc=will bypass vehicle and mech rotation
   description=defualt 0
   inputName=UV_Rotation
   inputType=FunctionInput_Scalar
   sortPriority=10
  }
  n14 = Constant3Vector() {constant=<rgba 1 1 1 0>}
  n15 = MaterialFunctionCall(
   UVs=MaterialFunctionCall(UVs=@n12, Rotation Center=Constant2Vector() {g=0.5, r=0.5}, "Rotation Angle (0-1)"=@n13) {materialFunction=&CustomRotator}.Rotated Values
   Texture Scale=ComponentMask(None=FunctionInput(Preview=@n14) {bUsePreviewValueAsDefault=true, description=defualt 0, inputName=Uvs_Scale, sortPriority=12}) {g=true, r=true}
  ) {materialFunction=&ScaleUVsByCenter}
  n16 = FunctionInput(Preview=Constant2Vector() {g=2, r=2}) {
   bUsePreviewValueAsDefault=true
   desc=2x2
   description=defualt 0
   inputName="Atlas_Row\\Colums Amount"
   inputType=FunctionInput_Vector2
   sortPriority=50
  }
  n17 = FunctionInput(Preview=Constant2Vector()) {
   bUsePreviewValueAsDefault=true
   desc=0x0
   description=defualt 0
   inputName=Atlas_offset_location
   inputType=FunctionInput_Vector2
   sortPriority=50
  }
  n18 = Add(A=Multiply(A=@n15.UVs, B=Divide(B=@n16)), B=@n17)
  n19 = TextureSample(
   UVs=@n18
   Tex=FunctionInput(Preview=TextureObject() {samplerType=SAMPLERTYPE_Normal, texture=/Engine/Functions/Engine_MaterialFunctions02/ExampleContent/Textures/SphereRenderNormalsMap}) {bUsePreviewValueAsDefault=true, inputName=Normal_Map, inputType=FunctionInput_Texture2D}
  ) {samplerSource=SSM_Wrap_WorldGroupSettings, samplerType=SAMPLERTYPE_Normal}
  n20 = MaterialFunctionCall(Color=@n19.RGB) {materialFunction=&SplitComponents}
  n21 = FunctionInput(Preview=VertexInterpolator(VS=PreSkinnedNormal()).PS) {bUsePreviewValueAsDefault=true, inputName=Pre_Skinned_Normals, sortPriority=2}
  n22 = ComponentMask(None=@n21) {r=true}
  n23 = AppendVector(A=AppendVector(A=Multiply(A=@n20.B, B=@n22), B=@n20.R), B=Multiply(A=@n20.G) {constB=-1})
  n24 = RotateAboutAxis(NormalizedRotationAxis=Constant3Vector() {constant=<rgba 1 0 0 0>}, RotationAngle=@n13, PivotPoint=Constant3Vector(), Position=@n23)
  n25 = AppendVector(A=ComponentMask(None=@n19.RGB) {g=true, r=true}, B=Multiply(A=@n20.B, B=ComponentMask(None=@n21) {b=true}))
  n26 = Add(
   A=RotateAboutAxis(
    NormalizedRotationAxis=Constant3Vector() {constant=<rgba 0 0 1 0>}
    RotationAngle=Multiply(A=@n13) {constB=-1}
    PivotPoint=Constant3Vector()
    Position=@n25
   )
   B=@n25
  )
  n27 = MaterialFunctionCall(Color=@n19.RGB) {materialFunction=&SplitComponents}
  n28 = ComponentMask(None=@n21) {g=true}
  n29 = AppendVector(A=AppendVector(A=@n27.R, B=Multiply(A=@n27.B, B=@n28)), B=Multiply(A=@n27.G) {constB=-1})
  n30 = Add(
   A=RotateAboutAxis(
    NormalizedRotationAxis=Constant3Vector() {constant=<rgba 0 1 0 0>}
    RotationAngle=Multiply(A=@n13) {constB=-1}
    PivotPoint=Constant3Vector()
    Position=@n29
   )
   B=@n29
  )
  n31 = MaterialFunctionCall(Color=@n2.RGB) {materialFunction=&SplitComponents}
  n32 = Constant3Vector() {constant=<rgba 0 1 0 0>}
  n33 = Multiply(A=@n13) {constB=-1}
  n34 = Constant3Vector()
  n35 = MaterialFunctionCall(Color=@n19.RGB) {materialFunction=&SplitComponents}
  n36 = ComponentMask(None=@n21) {g=true}
  n37 = AppendVector(A=AppendVector(A=@n35.R, B=Multiply(A=@n35.B, B=@n36)), B=Multiply(A=@n35.G) {constB=-1})
  n38 = LinearInterpolate(
   A=LinearInterpolate(A=Add(A=@n24, B=@n23), B=LinearInterpolate(A=@n26, B=@n30, Alpha=@n31.G), Alpha=@n31.R)
   B=Add(A=RotateAboutAxis(NormalizedRotationAxis=@n32, RotationAngle=@n33, PivotPoint=@n34, Position=@n37), B=@n37)
   Alpha=@n31.B
  )
  n39 = Constant3Vector() {constant=<rgba 1 0 0 0>}
  n40 = Multiply(A=Constant3Vector() {constant=<rgba 0 0 1 0>}, B=@n21)
  n41 = Multiply(A=Constant3Vector() {constant=<rgba 0 1 0 0>}, B=@n21)
  n42 = LinearInterpolate(
   A=LinearInterpolate(A=Multiply(A=@n39, B=@n21), B=LinearInterpolate(A=@n40, B=@n41, Alpha=@n31.G), Alpha=@n31.R)
   B=Multiply(A=Constant3Vector() {constant=<rgba 0 1 0 0>}, B=@n21)
   Alpha=@n31.B
  )
  n43 = Transform(None=Subtract(A=@n38, B=@n42)) {transformSourceType=TRANSFORMSOURCE_Local, transformType=TRANSFORM_Tangent}
  FunctionOutput(None=Add(A=Constant3Vector() {constant=<rgba 0 0 1 0>}, B=@n43)) {outputName=Normal, sortPriority=2}
  FunctionOutput(None=@n18) {outputName=UVs, sortPriority=1}
  FunctionOutput(None=@n43) {description=add to object normals, outputName=Raw_Normal, sortPriority=3}
  FunctionOutput(None=@n12) {outputName=UV_No_offset, sortPriority=-1}
  FunctionOutput(None=@n15.0-1 mask) {outputName=Safe_Zone_Mask}
  Multiply(A=@n13) {constB=-1}
  Normalize(VectorInput=@n42)
  Normalize(VectorInput=@n38)
 ]}
