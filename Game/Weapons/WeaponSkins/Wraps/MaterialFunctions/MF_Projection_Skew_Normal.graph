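# Material function for weapon wrap skins that projects a normal map using skewed planar UVs from pre-skinned local bounds.
# Skew Amount blends two projections, Is Vehicle switches axes, and tiling and offset follow.
# Sampled normals are swizzled by pre-skinned normals, converted to tangent space, and output with UVs and an alternate-axis contrast mask.
# The default bounds calculation divides by Max minus Max, which is zero.
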MaterialFunction /Game/Weapons/WeaponSkins/Wraps/MaterialFunctions/MF_Projection_Skew_Normal
graph {nodes [
  n1 = PreSkinnedLocalBounds()
  n2 = Subtract(A=VertexInterpolator(VS=PreSkinnedPosition()).PS, B=@n1.Min)
  n3 = Subtract(A=@n1.Max, B=@n1.Max)
  n4 = MaterialFunctionCall(Color=FunctionInput(Preview=Divide(A=@n2, B=@n3)) {bUsePreviewValueAsDefault=true, inputName=Caculated_Bounds, sortPriority=1}) {materialFunction=&SplitComponents}
  n5 = OneMinus(None=@n4.B)
  n6 = FunctionInput(Preview=Constant()) {bUsePreviewValueAsDefault=true, inputName=Is Vehicle, inputType=FunctionInput_Scalar, sortPriority=5}
  n7 = OneMinus(None=@n4.R)
  n8 = OneMinus(None=@n4.B)
  n9 = FunctionInput(Preview=Constant() {r=0.5}) {bUsePreviewValueAsDefault=true, description=0.5, inputName=Skew Amount, inputType=FunctionInput_Scalar, sortPriority=5}
  n10 = LinearInterpolate(
   A=LinearInterpolate(A=AppendVector(A=@n4.G, B=@n5), B=AppendVector(A=@n4.R, B=@n4.R), Alpha=@n6)
   B=LinearInterpolate(A=AppendVector(A=@n4.G, B=@n7), B=AppendVector(A=@n4.R, B=@n8), Alpha=@n6)
   Alpha=@n9
  )
  n11 = FunctionInput(Preview=Constant3Vector() {constant=<rgba 1 2 1 0>}) {bUsePreviewValueAsDefault=true, description=defualt 0, inputName=Uvs_TileRate, sortPriority=12}
  n12 = Constant3Vector()
  n13 = Add(
   A=Multiply(A=@n10, B=ComponentMask(None=@n11) {g=true, r=true})
   B=ComponentMask(None=FunctionInput(Preview=@n12) {bUsePreviewValueAsDefault=true, description=defualt 0, inputName=Uvs_Offset, sortPriority=12}) {g=true, r=true}
  ) {constB=0}
  n14 = TextureSample(
   UVs=@n13
   Tex=FunctionInput(Preview=TextureObject() {samplerType=SAMPLERTYPE_Normal, texture=/Engine/Functions/Engine_MaterialFunctions02/ExampleContent/Textures/SphereRenderNormalsMap}) {bUsePreviewValueAsDefault=true, inputName=Normal_Map, inputType=FunctionInput_Texture2D}
  ) {samplerType=SAMPLERTYPE_Normal}
  n15 = MaterialFunctionCall(Color=@n14.RGB) {materialFunction=&SplitComponents}
  n16 = FunctionInput(Preview=VertexInterpolator(VS=PreSkinnedNormal()).PS) {bUsePreviewValueAsDefault=true, inputName=Pre_Skinned_Normals, sortPriority=2}
  n17 = AppendVector(A=Multiply(A=@n15.B, B=ComponentMask(None=@n16) {r=true}), B=@n15.R)
  n18 = Multiply(A=@n15.G) {constB=-1}
  n19 = MaterialFunctionCall(Color=@n14.RGB) {materialFunction=&SplitComponents}
  n20 = AppendVector(A=Multiply(A=@n19.G) {constB=-1}, B=@n19.R)
  n21 = ComponentMask(None=@n16) {b=true}
  n22 = Multiply(A=@n19.B, B=@n21)
  n23 = MaterialFunctionCall(Color=@n14.RGB) {materialFunction=&SplitComponents}
  n24 = AppendVector(A=@n23.R, B=Multiply(A=@n23.B, B=ComponentMask(None=@n16) {g=true}))
  n25 = Multiply(A=@n23.G) {constB=-1}
  n26 = MaterialFunctionCall(Color=@n14.RGB) {materialFunction=&SplitComponents}
  n27 = AppendVector(A=@n26.R, B=@n26.R)
  n28 = ComponentMask(None=@n16) {b=true}
  n29 = Multiply(A=@n26.B, B=@n28)
  n30 = LinearInterpolate(
   A=LinearInterpolate(A=AppendVector(A=@n17, B=@n18), B=AppendVector(A=@n20, B=@n22), Alpha=Abs(None=@n21))
   B=LinearInterpolate(A=AppendVector(A=@n24, B=@n25), B=AppendVector(A=@n27, B=@n29), Alpha=Abs(None=@n28))
   Alpha=@n6
  )
  n31 = LinearInterpolate(
   A=Multiply(A=Constant3Vector() {constant=<rgba 1 0 1 0>}, B=@n16)
   B=Multiply(A=Constant3Vector() {constant=<rgba 0 1 1 0>}, B=@n16)
   Alpha=@n6
  )
  n32 = Transform(None=Subtract(A=@n30, B=@n31)) {transformSourceType=TRANSFORMSOURCE_Local, transformType=TRANSFORM_Tangent}
  FunctionOutput(None=Add(A=Constant3Vector() {constant=<rgba 0 0 1 0>}, B=@n32)) {outputName=Normal, sortPriority=5}
  FunctionOutput(None=@n13) {outputName=UVs}
  n33 = Abs(None=@n16)
  n34 = ComponentMask(None=@n33) {r=true}
  n35 = ComponentMask(None=@n33) {b=true}
  n36 = ComponentMask(None=@n33) {g=true}
  n37 = LinearInterpolate(A=Saturate(None=Add(A=@n34, B=@n35)), B=Saturate(None=Add(A=@n35, B=@n36)), Alpha=@n6)
  n38 = FunctionInput(Preview=Constant() {r=1}) {bUsePreviewValueAsDefault=true, description=1, inputName=Alternate Axies contrast, inputType=FunctionInput_Scalar, sortPriority=100}
  FunctionOutput(None=Saturate(None=MaterialFunctionCall(In=@n37, Contrast=@n38) {materialFunction=&CheapContrast}.Result)) {outputName=Masking Alternat Axies, sortPriority=100}
  FunctionOutput(None=@n32) {description=add to object normals, outputName=Raw_Normal, sortPriority=10}
  FunctionOutput(None=@n10) {outputName=UVs No Offsets/Tiling, sortPriority=1}
  Multiply(A=@n26.G) {constB=-1}
  Normalize(VectorInput=@n31)
  Normalize(VectorInput=@n30)
 ]}
