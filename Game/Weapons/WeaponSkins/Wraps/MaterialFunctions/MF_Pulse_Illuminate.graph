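# Material function for weapon skin wraps that outputs a scalar brightness pulsing over time between low and high values.
# Time times frequency drives a sine, which is scaled by amplitude and offset to form the lerp alpha between both brightness levels.
# A static switch optionally clamps that wave to a min/max range.
# Every input defaults to a named parameter, exposing the controls to material instances unless overridden.
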MaterialFunction /Game/Weapons/WeaponSkins/Wraps/MaterialFunctions/MF_Pulse_Illuminate
graph {nodes [
  n1 = ScalarParameter() {defaultValue=1, group=Primary Triplaner DiffusePanningMask, parameterName=Pulse_Brightness_Low, sortPriority=1}
  n2 = ScalarParameter() {defaultValue=50, group=Primary Triplaner DiffusePanningMask, parameterName=Pulse_Brightness_High, sortPriority=2}
  n3 = Multiply(
   A=Time()
   B=FunctionInput(Preview=ScalarParameter() {defaultValue=1, group=Primary Triplaner DiffusePanningMask, parameterName=Pulse_Frequency, sortPriority=5}) {bUsePreviewValueAsDefault=true, inputName=Frequency, inputType=FunctionInput_Scalar, sortPriority=2}
  )
  n4 = ScalarParameter() {defaultValue=0.5, group=Primary Triplaner DiffusePanningMask, parameterName=Pulse_Amplitude, sortPriority=7}
  n5 = Add(
   A=Multiply(
    A=Sine(None=@n3)
    B=FunctionInput(Preview=@n4) {bUsePreviewValueAsDefault=true, inputName=Amplitude, inputType=FunctionInput_Scalar, sortPriority=4}
   ) {constB=0.5}
   B=FunctionInput(Preview=ScalarParameter() {defaultValue=1, group=Primary Triplaner DiffusePanningMask, parameterName=Pulse_SineOffset, sortPriority=6}) {bUsePreviewValueAsDefault=true, inputName=Sinewave-Offset, inputType=FunctionInput_Scalar, sortPriority=3}
  )
  n6 = Clamp(
   None=@n5
   Min=FunctionInput(Preview=ScalarParameter() {group=Primary Triplaner DiffusePanningMask, parameterName=Pulse_Sinewave_Clamp_Min, sortPriority=3}) {bUsePreviewValueAsDefault=true, inputName=Clamp-Min, inputType=FunctionInput_Scalar, sortPriority=6}
   Max=FunctionInput(Preview=ScalarParameter() {defaultValue=1, group=Primary Triplaner DiffusePanningMask, parameterName=Pulse_Sinewave_Clamp_Max, sortPriority=4}) {bUsePreviewValueAsDefault=true, inputName=Clamp-Max, inputType=FunctionInput_Scalar, sortPriority=7}
  )
  n7 = FunctionInput(Preview=StaticBoolParameter() {group=Flicker, parameterName=Clamp-SineWave}) {bUsePreviewValueAsDefault=true, inputName=Clamp Sine Wave?, inputType=FunctionInput_StaticBool, sortPriority=5}
  FunctionOutput(None=LinearInterpolate(
    A=FunctionInput(Preview=@n1) {bUsePreviewValueAsDefault=true, inputName=Brightness-Low, inputType=FunctionInput_Scalar}
    B=FunctionInput(Preview=@n2) {bUsePreviewValueAsDefault=true, inputName=Brightness-High, inputType=FunctionInput_Scalar, sortPriority=1}
    Alpha=StaticSwitch(True=@n6, False=@n5, Value=@n7)
   ))
 ]}
