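# Weapon wrap material function that projects UVs and world positions onto virtual planes around an object's safe-zone pivot.
# A switch blends between local-to-world and raw pivot as centre, feeding a camera-facing VirtualPlaneCoordinatesAbs call.
# Default scale combines pivot alpha, a bounds term (Max minus itself as wired) and platform-switched object Z scale.
# An inline ray-plane intersection with configurable axes adds single-plane UVs and two masks, using Abs against opposite-direction artifacts.
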MaterialFunction /Game/Weapons/WeaponSkins/Wraps/MaterialFunctions/MF_SafeZone_VirtualPlane
graph {nodes [
  Constant3Vector()
  n1 = FunctionInput(Preview=Constant4Vector() {constant=<rgba 0 0 0 1>}.RGBA) {bUsePreviewValueAsDefault=true, inputName=Object-SafeZone-Virtual-Pivot, inputType=FunctionInput_Vector4, sortPriority=1}
  n2 = ComponentMask(None=@n1) {b=true, g=true, r=true}
  n3 = FunctionInput() {bUsePreviewValueAsDefault=true, inputName=Object-SafeZone-Virtual-CH-Switch, inputType=FunctionInput_Scalar, sortPriority=2}
  n4 = TransformPosition(None=LinearInterpolate(A=@n2, B=Constant3Vector(), Alpha=@n3)) {transformType=TRANSFORMPOSSOURCE_World}
  n5 = CameraPositionWS()
  n6 = Normalize(VectorInput=Subtract(A=@n4, B=FunctionInput(Preview=@n5) {bUsePreviewValueAsDefault=true, desc="0,0,1", description="0,0,1", inputName=Camera _pos}))
  n7 = FunctionInput(Preview=Constant3Vector() {constant=<rgba 0 0 1 1>}) {bUsePreviewValueAsDefault=true, desc="0,0,1", description="0,0,1", inputName=Axis}
  n8 = LinearInterpolate(A=@n4, B=@n2, Alpha=@n3)
  n9 = FunctionInput() {bUsePreviewValueAsDefault=true, desc=0.0.0, description=0.0.0, inputName=PlaneCenter_distortion, sortPriority=60, previewValue=<0 0 0 0>}
  n10 = PreSkinnedLocalBounds()
  n11 = MaterialFunctionCall(Float3=Subtract(A=@n10.Max, B=@n10.Max)) {materialFunction=&BreakOutFloat3Components}
  n12 = Max(A=Max(A=@n11.R, B=@n11.R), B=@n11.B)
  n13 = MaterialFunctionCall() {materialFunction=/Engine/Functions/Engine_MaterialFunctions02/WorldPositionOffset/ObjectScale}
  n14 = Constant() {r=1}
  n15 = MaterialFunctionCall("Xb1,PS4"=@n13.Scale Z, "PC-High&&Epic,XSX,PS5"=@n13.Scale Z, "PC-Perf,Android"=@n14, Switch=@n14, "PC-Low(NotPerf)&&Medium"=@n13.Scale Z) {materialFunction=&PlatformSwitch_1D}
  n16 = Multiply(
   A=Multiply(A=ComponentMask(None=@n1) {a=true}, B=Multiply(A=@n12, B=@n15.Result))
   B=FunctionInput(Preview=Constant() {r=1}) {bUsePreviewValueAsDefault=true, inputName=Scale_Multplyer, inputType=FunctionInput_Vector2}
  )
  n17 = MaterialFunctionCall!(
   Plane Normal Axis=@n6
   Plane X Axis=Normalize(VectorInput=CrossProduct(A=@n6, B=@n7))
   Plane Center=Add(A=NamedRerouteDeclaration(None=@n8) {name=Plane center, nodeColor=<rgba 0 0.4141 1 1>}, B=@n9)
   UV Scale=NamedRerouteDeclaration(None=FunctionInput(Preview=@n16) {bUsePreviewValueAsDefault=true, inputName=Scale, inputType=FunctionInput_Vector2}) {name=Uv Scale, nodeColor=<rgba 0 0.4141 1 1>}
  ) {materialFunction=/Game/Athena/Environments/Nexus/Materials/VirtualPlaneCoordinatesAbs}
  FunctionOutput(None=@n17.UVs) {outputName=Uvs}
  FunctionOutput(None=@n17.World Position) {outputName=World Position, sortPriority=1}
  FunctionOutput(None=@n17.World Position - Centered) {outputName=World Position Centered, sortPriority=2}
  n18 = NamedRerouteUsage()
  n19 = NamedRerouteUsage()
  n20 = DotProduct(A=@n19, B=CameraPositionWS() {desc=Ray Origin})
  n21 = Add(A=@n9, B=NamedRerouteUsage())
  n22 = DotProduct(A=@n19, B=@n21) {desc=Distance to Plane}
  n23 = Subtract(A=WorldPosition().XYZ, B=CameraPositionWS()) {desc=Direction Vector}
  n24 = DotProduct(A=@n19, B=@n23) {desc=VectorDotPlaneNormal}
  n25 = Divide(A=Multiply(A=Subtract(A=@n20, B=@n22) {desc=Plane Distance}) {constB=-1}, B=@n24) {desc=Intersection Time1}
  n26 = Abs(None=@n25) {desc=Added this Abs to fix pure white in oposite directions}
  n27 = Subtract(A=Add(A=CameraPositionWS() {desc=Ray Origin}, B=Multiply(A=@n26, B=@n23)) {desc=Intersection Position}, B=@n21)
  n28 = DotProduct(A=@n18, B=@n27)
  n29 = MaterialFunctionCall!(Vector1=@n19, Vector2=Constant3Vector() {constant=<rgba 0 0.0002 -1 0>}) {materialFunction=/Engine/Functions/Engine_MaterialFunctions02/Math/CreateThirdOrthogonalVector}
  n30 = DotProduct(A=Normalize(VectorInput=CrossProduct(A=@n18, B=@n29.Vector1)) {desc=Plane Y}, B=@n27)
  n31 = Add(A=Divide(A=AppendVector(A=@n28, B=@n30), B=NamedRerouteUsage())) {constB=0.5}
  FunctionOutput(None=@n31) {description=The 2D UVs of the virtual plane coordinates., outputName=Single_plane_UVs, sortPriority=50}
  n32 = Saturate(None=OneMinus(None=@n24))
  FunctionOutput(None=@n32) {description=The 2D UVs of the virtual plane coordinates., outputName=Single_plane_Masking on side, sortPriority=52}
  n33 = Constant()
  If(A=@n26, B=Constant(), "A > B"=AppendVector(A=@n28, B=@n30), "A == B"=@n33, "A < B"=@n33)
  n34 = Saturate(None=Divide(A=@n25) {constB=1})
  Multiply(
   A=TextureSample(UVs=@n31) {texture=/Game/Weapons/WeaponSkins/Wraps/Textures/UVCheaker_Clamped}.RGB
   B=FunctionOutput(None=@n34) {description=The 2D UVs of the virtual plane coordinates., outputName=Single_plane_Masking, sortPriority=51}
  )
  Multiply(A=@n34, B=@n32)
  n35 = Constant3Vector() {constant=<rgba 0 1 0 0>}
  n36 = FunctionInput(Preview=@n35) {bUsePreviewValueAsDefault=true, desc="0,1,0", description="0,1,0", inputName=SinglePlane X Axis, sortPriority=51, previewValue=<0 0 1 1>}
  NamedRerouteDeclaration(None=@n36) {name=Plane X, nodeColor=<rgba 0 0.4141 1 1>}
  n37 = FunctionInput(Preview=Constant3Vector() {constant=<rgba 0 0 1 0>}) {
   bUsePreviewValueAsDefault=true
   desc="0,0,1"
   description="0,0,1"
   inputName=SinglePlane Normal Axis
   sortPriority=50
   previewValue=<0 0 1 1>
  }
  NamedRerouteDeclaration(None=@n37) {name=Plane Normal Axis, nodeColor=<rgba 0 0.4141 1 1>}
  Transform(None=@n35) {transformSourceType=TRANSFORMSOURCE_Local}
 ]}
