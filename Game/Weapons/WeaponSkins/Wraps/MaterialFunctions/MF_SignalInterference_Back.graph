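# Material function that builds the animated back glitch mask for the Signal Interference weapon wrap.
# It pans two packed mask textures over the Bounds coordinates at different speeds, combining vertical lines and angled masks.
# One scrolling hash-mark layer cuts into that pattern and another adds to it.
# A static green-channel mask, offset and scaled by parameters, confines the result, which is saturated and output as Back_UV.
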MaterialFunction /Game/Weapons/WeaponSkins/Wraps/MaterialFunctions/MF_SignalInterference_Back
graph {nodes [
  n1 = FunctionInput(Preview=TextureCoordinate()) {bUsePreviewValueAsDefault=true, inputName=Bounds, inputType=FunctionInput_Vector2}
  n2 = AppendVector(
   A=ScalarParameter() {defaultValue=0.1, group=Signal_Interference, parameterName=SI_GlitchBackMask_OffsetX, sortPriority=45}
   B=ScalarParameter() {defaultValue=0.15, group=Signal_Interference, parameterName=SI_GlitchBackMask_OffsetY, sortPriority=45}
  )
  n3 = FunctionInput(Preview=TextureObject() {samplerType=SAMPLERTYPE_Masks, texture=/Game/Weapons/WeaponSkins/Wraps/Textures/SignalInterference/T_Wrap_SignalInterference_M}) {desc="Green = Mask, Blue = Hash Marks", inputName=Texture, inputType=FunctionInput_Texture2D, sortPriority=1}
  n4 = TextureSample(
   UVs=Multiply(
    A=Add(A=@n1, B=@n2)
    B=ScalarParameter() {defaultValue=0.75, group=Signal_Interference, parameterName=SI_GlitchBackMask_Scale, sortPriority=45}
   )
   Tex=@n3
  ) {samplerType=SAMPLERTYPE_Masks}
  n5 = Add(A=@n1, B=Multiply(A=Time(), B=Constant2Vector() {r=0.05}))
  n6 = Constant() {r=2}
  n7 = FunctionInput(Preview=TextureObject() {samplerType=SAMPLERTYPE_Masks, texture=/Game/Weapons/WeaponSkins/Wraps/Textures/SignalInterference/Wrap_SignalInterference_M2}) {desc="Red = Vertical, Green = Angled Mask, Blue = Diagonal", inputName=Texture_1, inputType=FunctionInput_Texture2D, sortPriority=1}
  n8 = Time()
  n9 = Constant2Vector() {r=0.05}
  n10 = Multiply(A=Add(A=@n1, B=Multiply(A=@n8, B=@n9)), B=Constant() {r=1})
  n11 = Multiply(
   A=TextureSample(UVs=Multiply(A=@n5, B=@n6), Tex=@n7) {samplerType=SAMPLERTYPE_Masks}.R
   B=OneMinus(None=TextureSample(UVs=@n10, Tex=@n7) {samplerType=SAMPLERTYPE_Masks}.G)
  )
  n12 = Multiply(A=Time(), B=Constant2Vector() {r=0.075})
  n13 = TextureSample(UVs=Multiply(A=Add(A=@n1, B=@n12), B=Constant() {r=1}), Tex=@n7) {samplerType=SAMPLERTYPE_Masks}
  n14 = Time()
  n15 = Multiply(A=@n14, B=Constant2Vector() {g=0.0175, r=0.05})
  n16 = TextureSample(UVs=Multiply(A=Add(A=@n1, B=@n15), B=Constant() {r=2}), Tex=@n3)
  n17 = Add(A=Multiply(A=@n14, B=Constant2Vector() {g=-0.025, r=0.075}), B=@n1)
  n18 = Constant() {r=2}
  n19 = Add(A=Multiply(A=Add(A=@n11, B=@n13.G), B=OneMinus(None=@n16.B)), B=TextureSample(UVs=Multiply(A=@n17, B=@n18), Tex=@n3).B)
  FunctionOutput(None=Saturate(None=Multiply(A=@n4.G, B=@n19))) {outputName=Back_UV, sortPriority=2}
 ]}
