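# Material function for weapon skin wraps that outputs sphere-projected normals, pointing radially from an offset pivot through each point of the mesh.
# It adds a PivotOffset input, defaulting to a vector parameter, to vertex-interpolated pre-skinned position, then normalizes the result.
# X and Z are multiplied by bounds-extent ratios, but both divide Y by itself, so no aspect correction occurs.
# Scaled bounds Min and Max are computed but left unconnected.
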MaterialFunction /Game/Weapons/WeaponSkins/Wraps/MaterialFunctions/MF_SphereProjectedNormals
graph {nodes [
  n1 = PreSkinnedPosition()
  n2 = VectorParameter() {group=SphereProjectedNormals, parameterName=SphereProjectedLocaltionNormals_Offset}
  n3 = MaterialFunctionCall(Float3=Add(A=VertexInterpolator(VS=@n1).PS, B=FunctionInput(Preview=@n2.RGB) {bUsePreviewValueAsDefault=true, inputName=PivotOffset})) {materialFunction=&BreakOutFloat3Components}
  n4 = PreSkinnedLocalBounds()
  n5 = MaterialFunctionCall(Float3=@n4.Extents) {materialFunction=&BreakOutFloat3Components}
  n6 = Divide(A=@n5.G, B=@n5.G) {desc=A}
  n7 = Divide(A=@n5.G, B=@n5.G) {desc=B}
  n8 = MaterialFunctionCall(X=Multiply(A=@n3.R, B=@n6), Y=@n3.G, Z=Multiply(A=@n3.B, B=@n7)) {materialFunction=&MakeFloat3}
  FunctionOutput(None=Normalize(VectorInput=@n8.Result))
  n9 = MaterialFunctionCall(Float3=@n4.Min) {materialFunction=&BreakOutFloat3Components}
  MaterialFunctionCall(X=Multiply(A=@n9.R, B=@n6), Y=@n9.G, Z=Multiply(A=@n9.B, B=@n7)) {desc=Min, materialFunction=&MakeFloat3}
  n10 = MaterialFunctionCall(Float3=@n4.Max) {materialFunction=&BreakOutFloat3Components}
  MaterialFunctionCall(X=Multiply(A=@n10.R, B=@n6), Y=@n10.G, Z=Multiply(A=@n10.B, B=@n7)) {desc=Max, materialFunction=&MakeFloat3}
  Reroute(None=@n8.Result) {desc=Position}
 ]}
