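# Material function for weapon wraps that shifts a triplanar material toward a glowing look at night.
# A time-of-day scalar, zero by day, partial at morning and evening and full at night, blends base color toward a darkened version.
# The same scalar blends emissive toward the Diffuse input scaled by the roughness input and a glow strength, intended as the second triplanar layer.
# A static switch passes material attributes through unchanged when disabled.
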MaterialFunction /Game/Weapons/WeaponSkins/Wraps/MaterialFunctions/MF_TimeofDayEmissiveAdjustments
graph {nodes [
  n1 = FunctionInput() {inputName=Material, inputType=FunctionInput_MaterialAttributes}
  n2 = GetMaterialAttributes(None=@n1) {attributeGetTypes [#1, #2]}
  n3 = Multiply(
   A=@n2.BaseColor
   B=ScalarParameter() {defaultValue=1, group=Primary Triplaner TOD Emissive, parameterName=TOD_DarkenTriplaner01, sortPriority=91}
  )
  n4 = MaterialFunctionCall(Value_Morning=Constant() {r=0.15}, Value_Day=Constant(), Value_Evening=Constant() {r=0.15}, Value_Night=Constant() {r=1}) {materialFunction=&Scalar_Time_of_Day}
  n5 = Multiply(
   A=FunctionInput() {inputName=Diffuse, sortPriority=1}
   B=Multiply(
    A=FunctionInput() {inputName=Roghness, inputType=FunctionInput_Scalar, sortPriority=3}
    B=ScalarParameter() {defaultValue=1, group=Primary Triplaner TOD Emissive, parameterName=TOD_Triplaner02GlowStrength, sortPriority=92}
   )
  )
  n6 = StaticSwitchParameter(
   True=SetMaterialAttributes!(
    MaterialAttributes=@n2.MaterialAttributes
    BaseColor=LinearInterpolate(A=@n2.BaseColor, B=@n3, Alpha=@n4.Result)
    EmissiveColor=LinearInterpolate(A=@n2.EmissiveColor, B=@n5, Alpha=@n4.Result)
   ) {attributeSetTypes [#1, #2]}
   False=@n1
  ) {
   desc="Puts Triplaner 02 into Emmisive and darkens Triplanerr 01 at night time. "
   group=Primary Triplaner
   parameterName=TOD Emissive - Triplaner 02 Is Emissive
   sortPriority=90
  }
  FunctionOutput(None=@n6)
 ]}
