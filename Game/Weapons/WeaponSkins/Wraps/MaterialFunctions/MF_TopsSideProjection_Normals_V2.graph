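# Material function for weapon skin wraps that projects a normal map along pre-skinned local axes, outputting an additive tangent-space surface normal.
# Bounds-based coordinates feed three rotatable, scalable, offsettable samples, reoriented per axis and blended by contrast-adjusted pre-skinned normals.
# IsVehicle and IsMech flags select which two axes receive texture, with the third optionally filled flat.
# An identically blended flat axis normal is subtracted before the local-to-tangent transform, keeping only XY.
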MaterialFunction /Game/Weapons/WeaponSkins/Wraps/MaterialFunctions/MF_TopsSideProjection_Normals_V2
graph {nodes [
  n1 = MaterialFunctionCall(Color=FunctionInput() {inputName=Rotation_Axis, sortPriority=1}) {materialFunction=&SplitComponents}
  n2 = PreSkinnedPosition()
  n3 = PreSkinnedLocalBounds()
  n4 = Divide(A=Subtract(A=VertexInterpolator(VS=@n2).PS, B=@n3.Min), B=Subtract(A=@n3.Max, B=@n3.Max))
  n5 = Multiply(
   A=FunctionInput(Preview=Multiply(A=@n4) {constB=5}) {inputName=Caculated_Bounds, sortPriority=3}
   B=Constant3Vector() {constant=<rgba 1 1 -1 0>}
  )
  n6 = Add(
   A=MaterialFunctionCall(UVs=ComponentMask(None=@n5) {b=true, g=true}, "Rotation Angle (0-1)"=@n1.R) {materialFunction=&CustomRotator}.Rotated Values
   B=NamedRerouteUsage()
  )
  n7 = NamedRerouteUsage()
  n8 = FunctionInput(Preview=TextureObject() {samplerType=SAMPLERTYPE_Normal, texture=/Engine/Functions/Engine_MaterialFunctions02/ExampleContent/Textures/SphereRenderNormalsMap}) {inputName=Texture_Normals, inputType=FunctionInput_Texture2D}
  n9 = TextureSample(UVs=Add(A=Multiply(A=@n6, B=@n7), B=NamedRerouteUsage()), Tex=@n8) {samplerType=SAMPLERTYPE_Normal}
  n10 = FunctionInput(Preview=VertexInterpolator(VS=PreSkinnedNormal()).PS) {inputName=Pre_Skinned_Normals, sortPriority=4}
  n11 = ComponentMask(None=@n10) {r=true}
  n12 = Multiply(A=@n9.R) {constB=-1}
  n13 = MaterialFunctionCall(Float4=FunctionInput() {bUsePreviewValueAsDefault=true, inputName=IsState, inputType=FunctionInput_Vector4, sortPriority=5}) {materialFunction=&BreakOutFloat4Components}
  n14 = NamedRerouteDeclaration(None=@n13.B) {name=IsMech, nodeColor=<rgba 0 0.3203 1 1>}
  n15 = AppendVector(A=AppendVector(A=Multiply(A=@n9.B, B=@n11), B=LinearInterpolate(A=@n9.R, B=@n12, Alpha=@n14)), B=Multiply(A=@n9.G) {constB=-1})
  n16 = Add(
   A=RotateAboutAxis(
    NormalizedRotationAxis=Constant3Vector() {constant=<rgba 1 0 0 0>}
    RotationAngle=Multiply(A=@n1.R) {constB=-1}
    PivotPoint=Constant3Vector() {constant=<rgba 0 0 0 1>}
    Position=@n15
   )
   B=@n15
  )
  n17 = Add(
   A=MaterialFunctionCall(UVs=ComponentMask(None=@n5) {g=true, r=true}, "Rotation Angle (0-1)"=@n1.B) {materialFunction=&CustomRotator}.Rotated Values
   B=NamedRerouteUsage()
  )
  n18 = NamedRerouteUsage()
  n19 = TextureSample(UVs=Add(A=Multiply(A=@n17, B=@n18), B=NamedRerouteUsage()), Tex=@n8) {samplerType=SAMPLERTYPE_Normal}
  n20 = ComponentMask(None=@n10) {b=true}
  n21 = AppendVector(A=ComponentMask(None=@n19.RGB) {g=true, r=true}, B=Multiply(A=@n19.B, B=@n20))
  n22 = Add(
   A=RotateAboutAxis(
    NormalizedRotationAxis=Constant3Vector() {constant=<rgba 0 0 1 0>}
    RotationAngle=Multiply(A=@n1.B) {constB=-1}
    PivotPoint=Constant3Vector() {constant=<rgba 0 0 0 1>}
    Position=@n21
   )
   B=@n21
  )
  n23 = FunctionInput(Preview=Constant() {r=1}) {inputName=Blend_Contrast, inputType=FunctionInput_Scalar, sortPriority=5}
  n24 = MaterialFunctionCall(In=Abs(None=@n20), Contrast=@n23) {materialFunction=&CheapContrast}
  n25 = ComponentMask(None=@n10) {g=true}
  n26 = MaterialFunctionCall(In=Abs(None=@n25), Contrast=@n23) {materialFunction=&CheapContrast}
  n27 = Saturate(None=Add(A=@n24.Result, B=@n26.Result))
  n28 = LinearInterpolate(A=@n16, B=@n22, Alpha=@n27)
  n29 = Multiply(A=Constant3Vector() {constant=<rgba 0 1 0 0>}, B=@n25)
  n30 = FunctionInput(Preview=Constant() {r=1}) {bUsePreviewValueAsDefault=true, inputName=Fill_alternate_Axis_Opacity, inputType=FunctionInput_Scalar, sortPriority=6}
  n31 = MaterialFunctionCall(In=Abs(None=@n11), Contrast=@n23) {materialFunction=&CheapContrast}
  n32 = Add(A=@n31.Result, B=Saturate(None=Subtract(A=@n24.Result, B=@n26.Result)))
  n33 = Multiply(A=@n30, B=OneMinus(None=Saturate(None=@n32)))
  n34 = Add(
   A=MaterialFunctionCall(UVs=ComponentMask(None=@n5) {b=true, r=true}, "Rotation Angle (0-1)"=@n1.G) {materialFunction=&CustomRotator}.Rotated Values
   B=NamedRerouteUsage()
  )
  n35 = NamedRerouteUsage()
  n36 = TextureSample(UVs=Add(A=Multiply(A=@n34, B=@n35), B=NamedRerouteUsage()), Tex=@n8) {samplerType=SAMPLERTYPE_Normal}
  n37 = Multiply(A=@n36.R) {constB=-1}
  n38 = AppendVector(A=AppendVector(A=LinearInterpolate(A=@n36.R, B=@n37, Alpha=@n14), B=Multiply(A=@n36.B, B=@n25)), B=Multiply(A=@n36.G) {constB=-1})
  n39 = Add(
   A=RotateAboutAxis(
    NormalizedRotationAxis=Constant3Vector() {constant=<rgba 0 -1 0 0>}
    RotationAngle=Multiply(A=@n1.G) {constB=-1}
    PivotPoint=Constant3Vector() {constant=<rgba 0 0 0 1>}
    Position=@n38
   )
   B=@n38
  )
  n40 = Saturate(None=Add(A=@n24.Result, B=@n31.Result))
  n41 = LinearInterpolate(A=@n39, B=@n22, Alpha=@n40) {desc=Float3}
  n42 = Multiply(A=Constant3Vector() {constant=<rgba 1 0 0 0>}, B=@n11)
  n43 = Multiply(A=OneMinus(None=@n27), B=@n30)
  n44 = NamedRerouteDeclaration(None=@n13.R) {name=IsVehicle, nodeColor=<rgba 0 0.3203 1 1>}
  n45 = Constant3Vector() {constant=<rgba 0 0 1 0>}
  n46 = Add(A=@n31.Result, B=@n26.Result)
  n47 = Multiply(A=OneMinus(None=Saturate(None=@n46)), B=@n30)
  n48 = LinearInterpolate(
   A=LinearInterpolate(A=LinearInterpolate(A=@n28, B=@n29, Alpha=@n33), B=LinearInterpolate(A=@n41, B=@n42, Alpha=@n43), Alpha=@n44)
   B=LinearInterpolate(A=LinearInterpolate(A=@n16, B=@n39, Alpha=@n27), B=Multiply(A=@n45, B=@n20), Alpha=@n47)
   Alpha=@n14
  )
  n49 = Multiply(A=Constant3Vector() {constant=<rgba 1 0 0 0>}, B=@n11)
  n50 = Multiply(A=Constant3Vector() {constant=<rgba 0 0 1 0>}, B=@n20)
  n51 = LinearInterpolate(A=@n49, B=@n50, Alpha=@n27)
  n52 = Multiply(A=Constant3Vector() {constant=<rgba 0 1 0 0>}, B=@n25)
  n53 = LinearInterpolate(A=@n52, B=@n50, Alpha=@n40)
  n54 = LinearInterpolate(
   A=LinearInterpolate(A=LinearInterpolate(A=@n51, B=@n52, Alpha=@n33), B=LinearInterpolate(A=@n53, B=@n49, Alpha=@n43), Alpha=@n44)
   B=LinearInterpolate(A=LinearInterpolate(A=@n49, B=@n52, Alpha=@n27), B=@n50, Alpha=@n47)
   Alpha=@n14
  )
  n55 = Subtract(A=@n48, B=@n54)
  FunctionOutput(None=Multiply(
    A=Transform(None=@n55) {transformSourceType=TRANSFORMSOURCE_Local, transformType=TRANSFORM_Tangent}
    B=Constant3Vector() {constant=<rgba 1 1 0 0>}
   )) {outputName=Add_Surface_Normal}
  Multiply(A=@n48)
  Multiply(A=@n19.RGB, B=Constant3Vector() {constant=<rgba -1 -1 1 0>})
  Multiply(
   A=TextureSample(UVs=TextureCoordinate()) {samplerType=SAMPLERTYPE_Normal, texture=&SphereRenderNormalsMap}.RGB
   B=Constant3Vector() {constant=<rgba 1 1 0 0>}
  )
  NamedRerouteDeclaration(None=@n13.A) {name=IsCharacter, nodeColor=<rgba 0 0.3203 1 1>}
  NamedRerouteDeclaration(None=FunctionInput() {bUsePreviewValueAsDefault=true, inputName=IsStatsB, inputType=FunctionInput_Vector2, sortPriority=5}) {name=IsStatsB, nodeColor=<rgba 0 0.4844 1 1>}
  n56 = Constant()
  NamedRerouteDeclaration(None=ComponentMask(None=FunctionInput(Preview=@n56) {bUsePreviewValueAsDefault=true, inputName=Offset, sortPriority=200}) {g=true, r=true}) {name=Offset, nodeColor=<rgba 0 0.3672 1 1>}
  n57 = Constant() {r=1}
  NamedRerouteDeclaration(None=ComponentMask(None=FunctionInput(Preview=@n57) {bUsePreviewValueAsDefault=true, inputName=Scale, sortPriority=200}) {g=true, r=true}) {name=Scale, nodeColor=<rgba 0 0.3672 1 1>}
  n58 = Constant()
  NamedRerouteDeclaration(None=ComponentMask(None=FunctionInput(Preview=@n58) {bUsePreviewValueAsDefault=true, inputName=Offset_Pivot, sortPriority=200}) {g=true, r=true}) {name=Offset_PV, nodeColor=<rgba 0 0.3672 1 1>}
  Normalize(VectorInput=@n54)
  Normalize(VectorInput=@n48)
 ]}
