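# Material function for weapon-skin wraps that projects layered top/side patterns onto weapons, vehicles, mechs, characters, backpacks and pickaxes, selecting UV offsets per item type.
# Side-projection UVs, optionally polar, panned or warped, feed gradient or texture masks combined with curvature and layer masks by height-lerp or multiply.
# These masks blend material layers over the root material and output emissive masks.
# Static switches gate features, including time-of-day and audio-reactive modulation.
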MaterialFunction /Game/Weapons/WeaponSkins/Wraps/MaterialFunctions/MF_TopSideProjection_Wrap_V02
graph {nodes [
  n1 = MaterialFunctionCall(Float4=NamedRerouteUsage()) {materialFunction=&BreakOutFloat4Components}
  n2 = LinearInterpolate(
   A=VectorParameter() {group=TopSide_V02_Layer-01_Uvs, parameterName=TopS_01_Offset_Weapon, defaultValue=<rgba 0 0 0 1>}.RGB
   B=LinearInterpolate(
    A=VectorParameter() {group=TopSide_V02_Layer-01_Uvs, parameterName=TopS_01_Offset_Vehicle, defaultValue=<rgba 0 0 0 1>}.RGB
    B=VectorParameter() {group=TopSide_V02_Layer-01_Uvs, parameterName=TopS_01_Offset_VehicleSide, defaultValue=<rgba 0 0 0 1>}.RGB
    Alpha=@n1.G
   )
   Alpha=@n1.R
  )
  n3 = LinearInterpolate(
   A=VectorParameter() {group=TopSide_V02_Layer-01_Uvs, parameterName=TopS_01_Offset_Mech, defaultValue=<rgba 0 0 0 1>}.RGB
   B=VectorParameter() {group=TopSide_V02_Layer-01_Uvs, parameterName=TopS_01_Offset_Character, defaultValue=<rgba 0 0 0 1>}.RGB
   Alpha=@n1.A
  )
  n4 = MaterialFunctionCall(Float2=NamedRerouteUsage()) {materialFunction=&BreakOutFloat2Components}
  n5 = LinearInterpolate(
   A=LinearInterpolate(
    A=LinearInterpolate(A=@n2, B=@n3, Alpha=@n1.B)
    B=VectorParameter() {group=TopSide_V02_Layer-01_Uvs, parameterName=TopS_01_Offset_Backpack, defaultValue=<rgba 0 0 0 1>}.RGB
    Alpha=@n4.R
   )
   B=VectorParameter() {group=TopSide_V02_Layer-01_Uvs, parameterName=TopS_01_Offset_Pickaxe, defaultValue=<rgba 0 0 0 1>}.RGB
   Alpha=@n4.G
  )
  n6 = VectorParameter() {group=TopSide_V02_Layer-01_Uvs, parameterName=TopS_01_Panning_Speed------------, defaultValue=<rgba 0 0 0 1>}
  n7 = Multiply(A=Time(), B=@n6.RGB)
  Add(A=@n5, B=@n7)
  n8 = MaterialFunctionCall(Float4=NamedRerouteUsage()) {materialFunction=&BreakOutFloat4Components}
  n9 = LinearInterpolate(
   A=VectorParameter() {group=TopSide_V02_Layer-04_Uvs, parameterName=TopS_04_Offset_Weapon, defaultValue=<rgba 0 0 0 1>}.RGB
   B=LinearInterpolate(
    A=VectorParameter() {group=TopSide_V02_Layer-04_Uvs, parameterName=TopS_04_Offset_Vehicle, defaultValue=<rgba 0 0 0 1>}.RGB
    B=VectorParameter() {group=TopSide_V02_Layer-04_Uvs, parameterName=TopS_04_Offset_VehicleSide, defaultValue=<rgba 0 0 0 1>}.RGB
    Alpha=@n8.G
   )
   Alpha=@n8.R
  )
  n10 = LinearInterpolate(
   A=VectorParameter() {group=TopSide_V02_Layer-04_Uvs, parameterName=TopS_04_Offset_Mech, defaultValue=<rgba 0 0 0 1>}.RGB
   B=VectorParameter() {group=TopSide_V02_Layer-04_Uvs, parameterName=TopS_04_Offset_Character, defaultValue=<rgba 0 0 0 1>}.RGB
   Alpha=@n8.A
  )
  n11 = MaterialFunctionCall(Float2=NamedRerouteUsage()) {materialFunction=&BreakOutFloat2Components}
  n12 = LinearInterpolate(
   A=LinearInterpolate(
    A=LinearInterpolate(A=@n9, B=@n10, Alpha=@n8.B)
    B=VectorParameter() {group=TopSide_V02_Layer-04_Uvs, parameterName=TopS_04_Offset_Backpack, defaultValue=<rgba 0 0 0 1>}.RGB
    Alpha=@n11.R
   )
   B=VectorParameter() {group=TopSide_V02_Layer-04_Uvs, parameterName=TopS_04_Offset_Pickaxe, defaultValue=<rgba 0 0 0 1>}.RGB
   Alpha=@n11.G
  )
  n13 = VectorParameter() {
   group=TopSide_V02_Layer-04_Uvs
   parameterName=TopS_04_Panning_Speed------------
   channelNames {a="0=step animation", b="", g="", r=""}
   defaultValue=<rgba 0 0 0 1>
  }
  n14 = Multiply(A=Time(), B=@n13)
  Add(A=@n12, B=@n14)
  n15 = MaterialFunctionCall(Float4=NamedRerouteUsage()) {materialFunction=&BreakOutFloat4Components}
  n16 = LinearInterpolate(
   A=VectorParameter() {group=TopSide_V02_Layer-03_Uvs, parameterName=TopS_03_Offset_Weapon, defaultValue=<rgba 0 0 0 1>}.RGB
   B=LinearInterpolate(
    A=VectorParameter() {group=TopSide_V02_Layer-03_Uvs, parameterName=TopS_03_Offset_Vehicle, defaultValue=<rgba 0 0 0 1>}.RGB
    B=VectorParameter() {group=TopSide_V02_Layer-03_Uvs, parameterName=TopS_03_Offset_VehicleSide, defaultValue=<rgba 0 0 0 1>}.RGB
    Alpha=@n15.G
   )
   Alpha=@n15.R
  )
  n17 = LinearInterpolate(
   A=VectorParameter() {group=TopSide_V02_Layer-03_Uvs, parameterName=TopS_03_Offset_Mech, defaultValue=<rgba 0 0 0 1>}.RGB
   B=VectorParameter() {group=TopSide_V02_Layer-03_Uvs, parameterName=TopS_03_Offset_Character, defaultValue=<rgba 0 0 0 1>}.RGB
   Alpha=@n15.A
  )
  n18 = MaterialFunctionCall(Float2=NamedRerouteUsage()) {materialFunction=&BreakOutFloat2Components}
  n19 = LinearInterpolate(
   A=LinearInterpolate(
    A=LinearInterpolate(A=@n16, B=@n17, Alpha=@n15.B)
    B=VectorParameter() {group=TopSide_V02_Layer-03_Uvs, parameterName=TopS_03_Offset_Backpack, defaultValue=<rgba 0 0 0 1>}.RGB
    Alpha=@n18.R
   )
   B=VectorParameter() {group=TopSide_V02_Layer-03_Uvs, parameterName=TopS_03_Offset_Pickaxe, defaultValue=<rgba 0 0 0 1>}.RGB
   Alpha=@n18.G
  )
  n20 = VectorParameter() {
   group=TopSide_V02_Layer-03_Uvs
   parameterName=TopS_03_Panning_Speed------------
   channelNames {a="0=step animation", b="", g="", r=""}
   defaultValue=<rgba 0 0 0 1>
  }
  n21 = Multiply(A=Time(), B=@n20)
  Add(A=@n19, B=@n21)
  n22 = NamedRerouteUsage()
  n23 = NamedRerouteUsage()
  Add(A=@n22, B=@n23)
  Add()
  n24 = Constant4Vector()
  AppendVector(A=@n24.RGBA, B=@n24.RGBA)
  n25 = Constant4Vector()
  AppendVector(A=@n25.RGBA, B=@n25.RGBA)
  n26 = Constant4Vector()
  AppendVector(A=@n26.RGBA, B=@n26.RGBA)
  n27 = TextureSample() {samplerType=SAMPLERTYPE_Masks, texture=/Game/Weapons/FORT_Rifles/Textures/T_Scar_CM}
  n28 = FunctionInput(Preview=@n27.R) {bUsePreviewValueAsDefault=true, inputName=Mask Primary, inputType=FunctionInput_Scalar, sortPriority=24, previewValue=<1 0 0 1>}
  n29 = FunctionInput(Preview=@n27.B) {bUsePreviewValueAsDefault=true, inputName=Mask Secondary, inputType=FunctionInput_Scalar, sortPriority=24, previewValue=<1 0 0 1>}
  n30 = Saturate(None=Subtract(A=NamedRerouteUsage(), B=NamedRerouteUsage()))
  n31 = LinearInterpolate(A=ScalarParameter() {group=TopSide_V02, parameterName="TopS_Mask_Accent  0=Off"}, Alpha=@n30) {constB=0}
  n32 = Add(
   A=Add(A=@n28, B=Multiply(A=@n29, B=@n31))
   B=LinearInterpolate(A=ScalarParameter() {group=TopSide_V02, parameterName="TopS_Mask_Primary\\Accent  1=Off"}, Alpha=@n30) {constB=0}
  )
  n33 = Saturate(None=Add(A=@n32, B=NamedRerouteUsage()))
  n34 = Saturate(None=@n32)
  n35 = StaticBoolParameter() {group="TopSide_V02_Layer 01 Blend  02", parameterName="TopS Blend Affect_Primary\\Accent Mask"}
  n36 = BlendMaterialAttributes(
   A=NamedRerouteUsage()
   B=NamedRerouteUsage()
   Alpha=StaticSwitch(True=StaticSwitch(True=@n33, False=@n34, Value=@n35), False=@n34, Value=NamedRerouteUsage())
  ) {vertexAttributeBlendType=UseA}
  BlendMaterialAttributes(A=@n36, B=NamedRerouteUsage()) {vertexAttributeBlendType=UseA}
  n37 = ScalarParameter() {desc=disable all values to let root material show thur, group=TopSide_V02_Layer 01, parameterName="TopS_01_base <--->Root Mat", sliderMin=1}
  n38 = Saturate(None=@n37)
  Ceil(None=@n38)
  CollectionParameter() {collection=&FortniteMaterialParameters, parameterName=EventAudio_Amplitude}
  CollectionParameter() {collection=&FortniteMaterialParameters, parameterName=EventAudio_Amplitude}
  CollectionParameter() {collection=&FortniteMaterialParameters, parameterName=EventAudio_Amplitude}
  ComponentMask(None=NamedRerouteUsage()) {a=true}
  n39 = NamedRerouteUsage()
  n40 = VectorParameter() {
   group="TopSide_V02_Layer 01 Blend  02"
   parameterName=TopS Blend-TS 01__Curvature_Amount
   channelNames {a=Grad_Pits, b=Grad_Peaks, g=Msks_Pits, r=Msks_Peaks}
  }
  n41 = NamedRerouteUsage()
  n42 = ChannelMaskParameter(None=NamedRerouteUsage()) {group="TopSide_V02_Layer 01 Blend  02", parameterName=TopS Blend-TS 02_Mask_Channel}
  n43 = NamedRerouteUsage()
  n44 = NamedRerouteUsage()
  n45 = StaticSwitchParameter(
   True=ChannelMaskParameter(None=NamedRerouteUsage()) {group="TopSide_V02_Layer 01 Blend  02", parameterName=TopS Blend-TS 03_Mask_Channel}.R
   False=StaticSwitchParameter(
    True=ChannelMaskParameter(None=@n43) {group="TopSide_V02_Layer 01 Blend  02", parameterName=TopS Blend-TS 04_Mask_Channel}.R
    False=ChannelMaskParameter(None=@n44) {group="TopSide_V02_Layer 01 Blend  02", parameterName=TopS Blend-TS 01_Mask_Channel}.R
   ) {group="TopSide_V02_Layer 01 Blend  02", parameterName="TopS Blend-TS 01_Mask---->TS 04"}
  ) {group="TopSide_V02_Layer 01 Blend  02", parameterName="TopS Blend-TS 01_Mask---->TS 03"}
  n46 = VectorParameter() {
   group="TopSide_V02_Layer 01 Blend  02"
   parameterName=TopS Blend-TS-Tri-Mask
   channelNames {a=Contrast, b="AltAxiesMSk TopSide ----->Side", g="Grad   0=on", r="Hieght  0=on"}
   defaultValue=<rgba 1 1 0 1>
  }
  n47 = MaterialFunctionCall(In=LinearInterpolate(A=NamedRerouteUsage(), B=NamedRerouteUsage(), Alpha=@n46."AltAxiesMSk TopSide ----->Side"), Contrast=@n46.Contrast) {materialFunction=&CheapContrast}
  n48 = Add(A=@n47.Result, B=@n46."Hieght  0=on")
  n49 = Add(
   A=Add(A=Multiply(A=@n39, B=@n40.Msks_Peaks), B=Multiply(A=@n41, B=@n40.Msks_Pits))
   B=Multiply(
    A=StaticSwitchParameter(True=@n42.R, False=@n45) {group="TopSide_V02_Layer 01 Blend  02", parameterName="TopS Blend-TS 01_Mask---->TS 02"}
    B=Saturate(None=@n48)
   )
  )
  n50 = ScalarParameter() {group="TopSide_V02_Layer 01 Blend  02", parameterName=TopS Blend HeightLerp Swap Msks 1-active}
  n51 = NamedRerouteUsage()
  n52 = NamedRerouteUsage()
  n53 = Add(A=@n47.Result, B=@n46."Grad   0=on")
  n54 = StaticSwitchParameter(True=NamedRerouteUsage(), False=NamedRerouteUsage()) {group="TopSide_V02_Layer 01 Blend  02", parameterName=TopS Blend FullWeaponBounds}
  n55 = Constant()
  n56 = NamedRerouteUsage()
  n57 = NamedRerouteUsage()
  n58 = StaticSwitchParameter(
   True=NamedRerouteUsage()
   False=StaticSwitchParameter(
    True=NamedRerouteUsage()
    False=StaticSwitchParameter(True=@n56, False=@n57) {group="TopSide_V02_Layer 01 Blend  02", parameterName="TopS Blend 01_UVS Offset------>SafeZoneBack"}
   ) {group="TopSide_V02_Layer 01 Blend  02", parameterName="TopS Blend 01_UVS Offset------>SafeZoneMiddle"}
  ) {group="TopSide_V02_Layer 01 Blend  02", parameterName="TopS Blend 01_UVS Offset------>SafeZoneFront"}
  n59 = Constant() {r=2}
  n60 = MaterialFunctionCall!(
   UV Coord=MaterialFunctionCall!(
    Normal_Map=NamedRerouteUsage()
    Offset Rt
    Pivot
    Caculated_Bounds=@n54
    Pre_Skinned_Normals=NamedRerouteUsage()
    is_State=NamedRerouteUsage()
    UV_Rotation=@n55
    Uvs_Offset=ComponentMask(None=@n58) {b=true, g=true, r=true}
    Uvs_TileRate
    Uvs_Offset_Pivot=@n55
    For_Safe_Zones=StaticBool() {value=true}
    Rotation_Comp
    Clamp Y
    Clamp X
    Clamp Padding_Comp
    No_Rotation
   ) {materialFunction=&/MF_Projection_Side_Normal_V2}.UVs
   LenghtScale=@n59
   RadialScale=@n59
  ) {materialFunction=/Game/Global/MaterialFunctions/TechArt/MF_PolarCoordinates}
  n61 = ComponentMask(None=NamedRerouteUsage()) {g=true, r=true}
  n62 = NamedRerouteUsage()
  n63 = StaticBoolParameter() {group="TopSide_V02_Layer 01 Blend  02", parameterName=TopS Blend_Grad----Texture}
  n64 = MaterialFunctionCall(
   Normal_Map=NamedRerouteUsage()
   Caculated_Bounds=@n54
   Pre_Skinned_Normals=NamedRerouteUsage()
   is_State=NamedRerouteUsage()
   UV_Rotation=NamedRerouteUsage()
   Uvs_Offset=StaticSwitch(True=NamedRerouteUsage(), False=Constant(), Value=@n63)
   Uvs_TileRate=StaticSwitch(True=NamedRerouteUsage(), False=Constant() {r=1}, Value=@n63)
   Uvs_Offset_Pivot=NamedRerouteUsage()
  ) {materialFunction=&MF_Projection_Side_Normal_V2}
  n65 = StaticSwitchParameter(
   True=Add(A=Multiply(A=@n60.Result, B=@n61), B=ComponentMask(None=@n62) {g=true, r=true})
   False=StaticSwitchParameter(True=@n64.UVs No Rotation, False=@n64.UVs No Rotation) {group="TopSide_V02_Layer 01 Blend  02", parameterName=TopS Blend 01_No_Rotation}
  ) {group="TopSide_V02_Layer 01 Blend  02", parameterName="TopS Blend 01_UVS  ----->Polar"}
  n66 = ScalarParameter() {group="TopSide_V02_Layer 01 Blend  02  TOD", parameterName=TopS_01_Blend_TOD_Min}
  n67 = ScalarParameter() {group="TopSide_V02_Layer 01 Blend  02  TOD", parameterName=TopS_01_Blend_TOD_Max}
  n68 = NamedRerouteUsage()
  n69 = NamedRerouteUsage()
  n70 = StaticSwitchParameter(
   True=Add(A=@n65, B=LinearInterpolate(A=@n66, B=@n67, Alpha=@n68))
   False=StaticSwitchParameter(True=Add(A=@n65, B=@n69), False=@n65) {group="TopSide_V02_Layer 01 Blend  02 Reactive", parameterName=TopS Blend Reactive}
  ) {group="TopSide_V02_Layer 01 Blend  02 Reactive", parameterName=TopS Blend TOD Offset}
  n71 = VectorParameter() {
   group="TopSide_V02_Layer 01 Blend  02"
   parameterName=TopS Blend_Grad-Tex-Clamp
   channelNames {a="", b="", g="TileRate ", r="Clamp  1=on"}
   defaultValue=<rgba 0 0.995 0 1>
  }
  n72 = LinearInterpolate(A=@n70, B=Multiply(A=Saturate(None=@n70), B=@n71."TileRate "), Alpha=@n71."Clamp  1=on")
  n73 = NamedRerouteUsage()
  n74 = ChannelMaskParameter(None=NamedRerouteUsage()) {group="TopSide_V02_Layer 01 Blend  02", parameterName=TopS _04_Blend_Warp_Mask_Channel, sortPriority=201}
  n75 = ChannelMaskParameter(None=NamedRerouteUsage()) {group="TopSide_V02_Layer 01 Blend  02", parameterName=TopS_01_Blend_Warp_Mask_Channel, sortPriority=200}
  n76 = StaticSwitchParameter(
   True=ChannelMaskParameter(None=NamedRerouteUsage()) {group="TopSide_V02_Layer 01 Blend  02", parameterName=TopS _02_Blend_Warp_Mask_Channel, sortPriority=201}.R
   False=StaticSwitchParameter(
    True=ChannelMaskParameter(None=@n73) {group="TopSide_V02_Layer 01 Blend  02", parameterName=TopS _03_Blend_Warp_Mask_Channel, sortPriority=201}.R
    False=StaticSwitchParameter(True=@n74.R, False=@n75.R) {group="TopSide_V02_Layer 01 Blend  02", parameterName="TopS_01_Blend_Warp-TS 01---->TS 04", sortPriority=200}
   ) {group="TopSide_V02_Layer 01 Blend  02", parameterName="TopS_01_Blend_Warp-TS 01---->TS 03", sortPriority=200}
  ) {group="TopSide_V02_Layer 01 Blend  02", parameterName="TopS_01_Blend_Warp-TS 01---->TS 02", sortPriority=200}
  n77 = ScalarParameter() {group="TopSide_V02_Layer 01 Blend  02", parameterName=TopS_01_Blend_Warp_Offset, sortPriority=200}
  n78 = Add(
   A=@n72
   B=Multiply(
    A=Add(A=@n76, B=@n77)
    B=ScalarParameter() {group="TopSide_V02_Layer 01 Blend  02", parameterName=TopS_01_Blend_Warp_Amount, sortPriority=200}
   )
  )
  n79 = StaticSwitchParameter(True=NamedRerouteUsage(), False=NamedRerouteUsage()) {group="TopSide_V02_Layer 01 Blend  02", parameterName=TopS Blend_Grad-Tex_Mask Share_00B_Msk_Tex}
  n80 = TextureObjectParameter() {
   group="TopSide_V02_Layer 01 Blend  02"
   parameterName=TopS Blend_Grad-Tex_Mask
   samplerType=SAMPLERTYPE_Masks
   texture=&T_BlackMask
  }
  n81 = ChannelMaskParameter(None=TextureSample!(
    UVs=StaticSwitchParameter(True=@n78, False=@n72) {group="TopSide_V02_Layer 01 Blend  02", parameterName=TopS_01_Blend_Warp, sortPriority=200}
    Tex=StaticSwitchParameter(True=@n79, False=@n80) {group="TopSide_V02_Layer 01 Blend  02", parameterName=TopS Blend_Grad-Tex_Mask Share_00A_Msk_Tex}
    "DDX(UVs)"=DDX(Value=@n64.UVs No Offsets/Tiling)
    "DDY(UVs)"=DDY(Value=@n64.UVs No Offsets/Tiling)
    Apply View MipBias
   ) {mipValueMode=TMVM_Derivative}.RGBA) {group="TopSide_V02_Layer 01 Blend  02", parameterName=TopS Blend_Grad-Tex_Mask_Channel}
  n82 = Saturate(None=Divide(A=ComponentMask(None=@n70) {r=true}, B=NamedRerouteUsage()))
  n83 = ScalarParameter() {defaultValue=2, group="TopSide_V02_Layer 01 Blend  02", parameterName=TopS Grad__Pow}
  n84 = StaticSwitchParameter(True=NamedRerouteUsage(), False=OneMinus(None=Power(Base=@n82, Exp=@n83))) {group="TopSide_V02_Layer 01 Blend  02", parameterName=TopS Blend_Grad----TOD}
  n85 = Add(
   A=Add(A=Multiply(A=@n51, B=@n40.Grad_Peaks), B=Multiply(A=@n52, B=@n40.Grad_Pits))
   B=Multiply(A=Saturate(None=@n53), B=StaticSwitch(True=@n81.R, False=@n84, Value=@n63))
  )
  n86 = LinearInterpolate(A=@n85, B=@n49, Alpha=@n50)
  n87 = LinearInterpolate(A=@n49, B=@n85, Alpha=@n50)
  n88 = ScalarParameter() {defaultValue=1, group="TopSide_V02_Layer 01 Blend  02", parameterName=TopS Blend-Contrast}
  n89 = Multiply(A=@n49, B=@n85)
  n90 = Multiply(A=@n89, B=Add(A=@n89, B=@n85))
  n91 = ScalarParameter() {group="TopSide_V02_Layer 01 Blend  02", parameterName="TopS Blend X+X---->X"}
  n92 = MaterialFunctionCall(In=Saturate(None=LinearInterpolate(A=@n90, B=@n89, Alpha=@n91)), Contrast=@n88) {materialFunction=&CheapContrast}
  n93 = StaticSwitchParameter(
   True=@n49
   False=StaticSwitchParameter(
    True=MaterialFunctionCall(Transition Phase=@n86, Height Texture=@n87, Contrast=@n88) {materialFunction=&HeightLerp}.Alpha
    False=StaticSwitchParameter(True=@n92.Result, False=@n85) {group="TopSide_V02_Layer 01 Blend  02", parameterName=TopS Blend X+X}
   ) {group="TopSide_V02_Layer 01 Blend  02", parameterName=TopS Blend HeightLerp}
  ) {group="TopSide_V02_Layer 01 Blend  02", parameterName=TopS Blend Straight Mask}
  n94 = LinearInterpolate(
   A=@n93
   B=OneMinus(None=@n93)
   Alpha=ScalarParameter() {group="TopSide_V02_Layer 01 Blend  02 Reactive", parameterName="TopS Blend-InvertMsk   1=invert"}
  )
  n95 = NamedRerouteUsage()
  n96 = Multiply(
   A=Saturate(None=@n94)
   B=LinearInterpolate(
    A=ChannelMaskParameter(None=@n95) {group="TopSide_V02_Layer 01 Blend  02", parameterName=TopS_Split_Uv_Blend-TS 02_Mask_Channel}.R
    Alpha=NamedRerouteUsage()
   )
  )
  n97 = StaticSwitchParameter(True=NamedRerouteUsage(), False=StaticSwitch(True=Saturate(None=@n96), False=@n94, Value=NamedRerouteUsage())) {group="TopSide_V02_Layer 01 Blend  02", parameterName=TopS Blend TOD}
  n98 = StaticSwitchParameter(True=Multiply(A=@n97, B=NamedRerouteUsage()), False=@n97) {group="TopSide_V02_Layer 01 Blend  02 Reactive", parameterName=TopS Blend Reactive}
  ComponentMask(None=@n98) {r=true}
  ComponentMask() {g=true, r=true}
  Constant2Vector()
  Constant()
  Constant()
  Constant() {r=1}
  n99 = MaterialFunctionCall(Float4=NamedRerouteUsage()) {materialFunction=&BreakOutFloat4Components}
  n100 = MaterialFunctionCall(
   R=LinearInterpolate(A=@n99.R, Alpha=ScalarParameter() {group=TopSide_V02, parameterName="TopS_IS_Roll   1=ON"}) {constB=0}
   G=@n99.G
   B=@n99.G
   A=@n99.G
  ) {materialFunction=&AppendMany}
  n101 = Constant()
  n102 = MaterialFunctionCall(Caculated_Bounds=NamedRerouteUsage(), is_State=@n100.RGBA, Uvs_Offset=@n101, Uvs_TileRate=Constant() {r=1}, Uvs_Offset_Pivot=@n101) {materialFunction=&MF_Projection_Side_Normal_V2}
  n103 = MaterialFunctionCall(Float4=NamedRerouteUsage()) {materialFunction=&BreakOutFloat4Components}
  n104 = LinearInterpolate(
   A=ScalarParameter() {defaultValue=-1, group=TopSide_V02_Layer Emissive_02, parameterName=TopS_Em02_CC_UVMap_Offset_Weapon, sortPriority=500}
   B=ScalarParameter() {defaultValue=-1, group=TopSide_V02_Layer Emissive_02, parameterName=TopS_Em02_CC_UVMap_Offset_Vehicle, sortPriority=500}
   Alpha=@n103.R
  )
  n105 = ScalarParameter() {defaultValue=-1, group=TopSide_V02_Layer Emissive_02, parameterName=TopS_Em02_CC_UVMap_Offset_Mech, sortPriority=500}
  n106 = ScalarParameter() {
   defaultValue=-1
   desc="negative value inverts "
   group=TopSide_V02_Layer Emissive_02
   parameterName=TopS_Em02_CC_UVMap_Falloff_Weapon
   sortPriority=500
  }
  n107 = ScalarParameter() {
   defaultValue=-1
   desc="negative value inverts "
   group=TopSide_V02_Layer Emissive_02
   parameterName=TopS_Em02_CC_UVMap_Falloff_Vehicle
   sortPriority=500
  }
  n108 = ScalarParameter() {
   defaultValue=-1
   desc="negative value inverts "
   group=TopSide_V02_Layer Emissive_02
   parameterName=TopS_Em02_CC_UVMap_Falloff_Mech
   sortPriority=500
  }
  n109 = Divide(
   A=Add(A=ComponentMask(None=@n102.UVs No Rotation) {r=true}, B=LinearInterpolate(A=@n104, B=@n105, Alpha=@n103.B))
   B=LinearInterpolate(A=LinearInterpolate(A=@n106, B=@n107, Alpha=@n103.R), B=@n108, Alpha=@n103.B)
  )
  Frac(None=@n109)
  n110 = MaterialFunctionCall(Float4=NamedRerouteUsage()) {materialFunction=&BreakOutFloat4Components}
  n111 = ScalarParameter() {group=TopSide_V02, parameterName="TopS_IS_Roll   1=ON"}
  n112 = Constant()
  n113 = MaterialFunctionCall(
   Caculated_Bounds=NamedRerouteUsage()
   is_State=MaterialFunctionCall(R=LinearInterpolate(A=@n110.R, Alpha=@n111) {constB=0}, G=@n110.G, B=@n110.G, A=@n110.G) {materialFunction=&AppendMany}.RGBA
   Uvs_Offset=@n112
   Uvs_TileRate=Constant() {r=1}
   Uvs_Offset_Pivot=@n112
  ) {materialFunction=&MF_Projection_Side_Normal_V2}
  n114 = MaterialFunctionCall(Float4=NamedRerouteUsage()) {materialFunction=&BreakOutFloat4Components}
  n115 = LinearInterpolate(
   A=ScalarParameter() {defaultValue=-1, group=TopSide_V02_Layer Emissive_03, parameterName=TopS_Em03_CC_UVMap_Offset_Weapon, sortPriority=500}
   B=ScalarParameter() {defaultValue=-1, group=TopSide_V02_Layer Emissive_03, parameterName=TopS_Em03_CC_UVMap_Offset_Vehicle, sortPriority=500}
   Alpha=@n114.R
  )
  n116 = ScalarParameter() {defaultValue=-1, group=TopSide_V02_Layer Emissive_03, parameterName=TopS_Em03_CC_UVMap_Offset_Mech, sortPriority=500}
  n117 = ScalarParameter() {
   defaultValue=-1
   desc="negative value inverts "
   group=TopSide_V02_Layer Emissive_03
   parameterName=TopS_Em03_CC_UVMap_Falloff_Weapon
   sortPriority=500
  }
  n118 = ScalarParameter() {
   defaultValue=-1
   desc="negative value inverts "
   group=TopSide_V02_Layer Emissive_03
   parameterName=TopS_Em03_CC_UVMap_Falloff_Vehicle
   sortPriority=500
  }
  n119 = ScalarParameter() {
   defaultValue=-1
   desc="negative value inverts "
   group=TopSide_V02_Layer Emissive_03
   parameterName=TopS_Em03_CC_UVMap_Falloff_Mech
   sortPriority=500
  }
  n120 = Divide(
   A=Add(A=ComponentMask(None=@n113.UVs No Rotation) {r=true}, B=LinearInterpolate(A=@n115, B=@n116, Alpha=@n114.B))
   B=LinearInterpolate(A=LinearInterpolate(A=@n117, B=@n118, Alpha=@n114.R), B=@n119, Alpha=@n114.B)
  )
  Frac(None=@n120)
  Frac()
  FunctionInput() {bUsePreviewValueAsDefault=true, inputName=TS Mask_01, inputType=FunctionInput_Vector4, sortPriority=24, previewValue=<1 0 0 1>}
  FunctionInput() {bUsePreviewValueAsDefault=true, inputName=TS Mask_02, inputType=FunctionInput_Vector4, sortPriority=24, previewValue=<1 0 0 1>}
  n121 = NamedRerouteUsage()
  n122 = StaticBoolParameter() {group=TopSide_V02, parameterName=TopSideProjection_V02}
  FunctionOutput(None=StaticSwitch(True=@n36, False=@n121, Value=@n122))
  n123 = NamedRerouteUsage()
  FunctionOutput(None=@n123) {outputName=TS_Layer01_Mask}
  n124 = NamedRerouteUsage()
  FunctionOutput(None=@n124) {outputName=TS_Layer02_Mask}
  n125 = FunctionInput() {inputName=Triplaners-Masks textures}
  n126 = Constant() {r=1}
  FunctionOutput(None=StaticSwitch(True=@n123, False=AppendVector(A=@n125, B=@n126), Value=@n122)) {outputName=Mask_Override}
  n127 = NamedRerouteUsage()
  n128 = ChannelMaskParameter(None=NamedRerouteUsage()) {group=TopSide_V02_Layer Emissive_01, parameterName=TopS Emissive_01-TS 03_Mask_Channel}
  n129 = StaticSwitchParameter(
   True=ChannelMaskParameter(None=NamedRerouteUsage()) {group=TopSide_V02_Layer Emissive_01, parameterName=TopS Emissive_01-TS 04_Mask_Channel}.R
   False=ChannelMaskParameter(None=NamedRerouteUsage()) {group=TopSide_V02_Layer Emissive_01, parameterName=TopS Emissive_01-TS 01_Mask_Channel}.R
  ) {group=TopSide_V02_Layer Emissive_01, parameterName="TopS Emissive_01-TS 01_Mask---->TS 04"}
  n130 = NamedRerouteUsage()
  n131 = VectorParameter() {
   group=TopSide_V02_Layer Emissive_01
   parameterName=TopS Emissive_01__Curvature_Amount
   channelNames {a=Grad_Pits, b=Grad_Peaks, g=Msks_Pits, r=Msks_Peaks}
  }
  n132 = NamedRerouteUsage()
  n133 = Add(
   A=StaticSwitchParameter(
    True=ChannelMaskParameter(None=@n127) {group=TopSide_V02_Layer Emissive_01, parameterName=TopS Emissive_01-TS 02_Mask_Channel}.R
    False=StaticSwitchParameter(True=@n128.R, False=@n129) {group=TopSide_V02_Layer Emissive_01, parameterName="TopS Emissive_01-TS 01_Mask---->TS 03"}
   ) {group=TopSide_V02_Layer Emissive_01, parameterName="TopS Emissive_01-TS 01_Mask---->TS 02"}
   B=Add(A=Multiply(A=@n130, B=@n131.Msks_Peaks), B=Multiply(A=@n132, B=@n131.Msks_Pits))
  )
  n134 = NamedRerouteUsage()
  n135 = NamedRerouteUsage()
  n136 = VectorParameter() {
   group=TopSide_V02_Layer Emissive_01
   parameterName=TopS Emissive-01-Tri-Mask
   channelNames {a=Contrast, b="AltAxiesMSk TopSide ----->Side", g="Grad   0=on", r="Hieght  0=on"}
   defaultValue=<rgba 1 1 0 1>
  }
  n137 = MaterialFunctionCall(In=LinearInterpolate(A=NamedRerouteUsage(), B=NamedRerouteUsage(), Alpha=@n136."AltAxiesMSk TopSide ----->Side"), Contrast=@n136.Contrast) {materialFunction=&CheapContrast}
  n138 = Add(A=@n137.Result, B=@n136."Grad   0=on")
  n139 = VectorParameter() {
   group=TopSide_V02_Layer Emissive_01
   parameterName=TopS Emissive_01_Grad-Tex--Polar_Msk
   channelNames {a=Inner UV Offset, b=Outer, g=Innier, r="0=Use Mask_Values"}
   defaultValue=<rgba 0 1 0 0>
  }
  n140 = NamedRerouteUsage()
  n141 = NamedRerouteUsage()
  n142 = Constant()
  n143 = StaticBoolParameter() {group=TopSide_V02_Layer Emissive_01, parameterName=TopS Emissive_01_Grad----Texture}
  n144 = StaticSwitch(
   True=StaticBoolParameter() {group=TopSide_V02_Layer Emissive_01, parameterName="TopS Emissive_01_Grad-Tex-Uvs---->Polar"}
   False=StaticBool()
   Value=@n143
  )
  n145 = MaterialFunctionCall(
   Normal_Map=NamedRerouteUsage()
   Caculated_Bounds=StaticSwitchParameter(True=NamedRerouteUsage(), False=NamedRerouteUsage()) {group=TopSide_V02_Layer Emissive_01, parameterName="TopS Emissive_01  FullWeaponBounds"}
   Pre_Skinned_Normals=NamedRerouteUsage()
   is_State=NamedRerouteUsage()
   UV_Rotation=NamedRerouteUsage()
   Uvs_Offset=StaticSwitch(True=@n140, False=StaticSwitch(True=@n141, False=@n142, Value=@n143), Value=@n144)
   Uvs_TileRate=StaticSwitch(True=NamedRerouteUsage(), False=Constant() {r=1}, Value=@n143)
   Uvs_Offset_Pivot=StaticSwitch(True=Constant(), False=@n140, Value=@n144)
   For_Safe_Zones=@n144
  ) {materialFunction=&MF_Projection_Side_Normal_V2}
  n146 = Constant() {r=2}
  n147 = MaterialFunctionCall(
   UV Coord=StaticSwitch(True=NamedRerouteUsage(), False=@n145.UVs No Rotation, Value=NamedRerouteUsage())
   LenghtScale=@n146
   RadialScale=@n146
  ) {materialFunction=&MF_PolarCoordinates}
  n148 = ComponentMask(None=@n147.Result) {r=true}
  n149 = SmoothStep(Min=@n139.Innier, Max=@n139.Innier, Value=@n148)
  n150 = VectorParameter() {group=TopSide_V02_Layer Emissive_01, parameterName=TopS Emissive_01_Grad-Tex--PolarTileRate, defaultValue=<rgba 1 1 0 1>}
  n151 = Multiply(A=@n147.Result, B=ComponentMask(None=@n150.RGB) {g=true, r=true})
  n152 = ComponentMask(None=@n141) {g=true, r=true}
  n153 = Constant()
  n154 = LinearInterpolate(B=@n139.Inner UV Offset, Alpha=@n148)
  n155 = StaticSwitch(
   True=Add(A=Add(A=@n151, B=@n152), B=AppendVector(A=@n153, B=@n154))
   False=StaticSwitchParameter(True=@n145.UVs No Rotation, False=@n145.UVs No Rotation) {group=TopSide_V02_Layer Emissive_01, parameterName="TopS Emissive_01  No_Rotation"}
   Value=@n144
  )
  n156 = StaticSwitch(
   True=Add(A=NamedRerouteUsage(), B=@n155)
   False=@n155
   Value=NamedRerouteDeclaration(None=StaticBoolParameter() {group=TopSide_V02_Layer Emissive_01 Reactive, parameterName=TopS Emissive_01_Offset----Reactive--------------}) {name=Emissive_01_Offset_bool, nodeColor=<rgba 0.3594 0 1 1>}
  )
  n157 = VectorParameter() {
   group=TopSide_V02_Layer Emissive_01
   parameterName=TopS Emissive_01_Grad-Tex-Clamp
   channelNames {a="", b="", g="TileRate ", r="Clamp  1=on"}
   defaultValue=<rgba 0 0.995 0 1>
  }
  n158 = LinearInterpolate(A=@n156, B=Multiply(A=Saturate(None=@n156), B=@n157."TileRate "), Alpha=@n157."Clamp  1=on")
  n159 = NamedRerouteUsage()
  n160 = ChannelMaskParameter(None=NamedRerouteUsage()) {group=TopSide_V02_Layer Emissive_01, parameterName=TopS_04_Em01_Warp_Mask_Channel, sortPriority=201}
  n161 = ChannelMaskParameter(None=NamedRerouteUsage()) {group=TopSide_V02_Layer Emissive_01, parameterName=TopS_01_Em01_Warp_Mask_Channel, sortPriority=200}
  n162 = StaticSwitchParameter(
   True=ChannelMaskParameter(None=NamedRerouteUsage()) {group=TopSide_V02_Layer Emissive_01, parameterName=TopS_02_Em01_Warp_Mask_Channel, sortPriority=201}.R
   False=StaticSwitchParameter(
    True=ChannelMaskParameter(None=@n159) {group=TopSide_V02_Layer Emissive_01, parameterName=TopS_03_Em01_Warp_Mask_Channel, sortPriority=201}.R
    False=StaticSwitchParameter(True=@n160.R, False=@n161.R) {group=TopSide_V02_Layer Emissive_01, parameterName="TopS_01_Em01_Warp-TS 01---->TS 04", sortPriority=200}
   ) {group=TopSide_V02_Layer Emissive_01, parameterName="TopS_01_Em01_Warp-TS 01---->TS 03", sortPriority=200}
  ) {group=TopSide_V02_Layer Emissive_01, parameterName="TopS_01_Em01_Warp-TS 01---->TS 02", sortPriority=200}
  n163 = Add(A=@n162, B=ScalarParameter() {group=TopSide_V02_Layer Emissive_01, parameterName=TopS_Em01_Warp_Offset, sortPriority=201})
  n164 = ScalarParameter() {group=TopSide_V02_Layer Emissive_01, parameterName=TopS_Em01_Warp_Amount, sortPriority=201}
  n165 = ScalarParameter() {group=TopSide_V02_Layer Emissive_01 Reactive, parameterName="TopS Audio 01 Emissive_Warp  Low", sortPriority=152}
  n166 = ScalarParameter() {defaultValue=0.1, group=TopSide_V02_Layer Emissive_01 Reactive, parameterName="TopS Audio 01 Emissive_Warp  High", sortPriority=152}
  n167 = NamedRerouteUsage()
  n168 = StaticSwitch(
   True=LinearInterpolate(
    A=@n164
    B=LinearInterpolate(A=@n165, B=@n166, Alpha=@n167)
    Alpha=ScalarParameter() {group=TopSide_V02_Layer Emissive_01 Reactive, parameterName="TopS Audio 01 Emissive_Warp  1=active", sortPriority=152}
   )
   False=@n164
   Value=NamedRerouteUsage()
  )
  n169 = NamedRerouteUsage()
  n170 = VectorParameter() {
   group=TopSide_V02_Layer Emissive_01
   parameterName=TopS_Em01_Warp_Curvature_Amount
   sortPriority=201
   channelNames {a=PitsUVS Grad, b=PeaksUVs Grad, g=PitsUVS, r=PeaksUVs}
  }
  n171 = NamedRerouteUsage()
  n172 = Add(A=Add(A=@n158, B=Multiply(A=@n163, B=@n168)), B=Add(A=Multiply(A=@n169, B=@n170.PeaksUVs), B=Multiply(A=@n171, B=@n170.PitsUVS)))
  n173 = StaticSwitchParameter(True=NamedRerouteUsage(), False=NamedRerouteUsage()) {group=TopSide_V02_Layer Emissive_01, parameterName=TopS Emissive_01_Grad-Tex_Mask Share_00B_Msk_Tex}
  n174 = TextureObjectParameter() {
   group=TopSide_V02_Layer Emissive_01
   parameterName=TopS Emissive_01_Grad-Tex_Mask
   samplerType=SAMPLERTYPE_Masks
   texture=&T_BlackMask
  }
  n175 = MaterialFunctionCall(Float2=TextureCoordinate()) {materialFunction=&BreakOutFloat2Components}
  n176 = ChannelMaskParameter(None=TextureSample(
    UVs=StaticSwitchParameter(True=@n172, False=@n158) {group=TopSide_V02_Layer Emissive_01, parameterName=TopS Emissive_01_Warp, sortPriority=200}
    Tex=StaticSwitchParameter(True=@n173, False=@n174) {group=TopSide_V02_Layer Emissive_01, parameterName=TopS Emissive_01_Grad-Tex_Mask Share_00A_Msk_Tex}
    "DDX(UVs)"=DDX(Value=@n175.R)
    "DDY(UVs)"=DDY(Value=@n175.G)
   ) {mipValueMode=TMVM_Derivative}.RGBA) {group=TopSide_V02_Layer Emissive_01, parameterName=TopS Emissive_01_Grad-Tex_Mask_Channel}
  n177 = StaticSwitch(True=Multiply(A=Max(A=@n139."0=Use Mask_Values", B=@n149), B=@n176.R), False=@n176.R, Value=@n144)
  n178 = ComponentMask(None=@n156) {r=true}
  n179 = Divide(A=@n178, B=NamedRerouteUsage())
  n180 = OneMinus(None=Power(Base=Saturate(None=@n179), Exp=ScalarParameter() {defaultValue=2, group=TopSide_V02_Layer Emissive_01, parameterName=TopS_Em01__Pow}))
  n181 = Add(
   A=Add(A=Multiply(A=@n134, B=@n131.Grad_Peaks), B=Multiply(A=@n135, B=@n131.Grad_Pits))
   B=Multiply(A=Saturate(None=@n138), B=StaticSwitch(True=@n177, False=@n180, Value=@n143))
  )
  n182 = NamedRerouteUsage()
  n183 = Saturate(None=Multiply(A=@n181, B=@n182))
  n184 = ScalarParameter() {group=TopSide_V02_Layer Emissive_01 Reactive, parameterName="TopS Emissive_01-Brightness-Reactive--->Gradent Mult", sliderMax=1}
  n185 = Add(A=@n181, B=@n182)
  n186 = ScalarParameter() {group=TopSide_V02_Layer Emissive_01 Reactive, parameterName="TopS Emissive_01-Brightness-Reactive--->Gradent Add", sliderMax=1}
  n187 = StaticSwitch(
   True=LinearInterpolate(A=LinearInterpolate(A=@n181, B=@n183, Alpha=@n184) {constA=1}, B=Saturate(None=@n185), Alpha=@n186) {constA=1}
   False=@n181
   Value=NamedRerouteUsage()
  )
  n188 = Multiply(A=Saturate(None=Add(A=@n137.Result, B=@n136."Hieght  0=on")), B=@n133)
  n189 = ScalarParameter() {group=TopSide_V02_Layer Emissive_01, parameterName=TopS Emissive_01 HeightLerp_Swap_msk}
  n190 = LinearInterpolate(A=@n187, B=@n188, Alpha=@n189)
  n191 = LinearInterpolate(A=@n188, B=@n187, Alpha=@n189)
  n192 = ScalarParameter() {defaultValue=1, group=TopSide_V02_Layer Emissive_01, parameterName=TopS Emissive_01-Contrast}
  n193 = Multiply(A=@n188, B=@n187)
  n194 = Multiply(A=@n193, B=Add(A=@n193, B=@n187))
  n195 = ScalarParameter() {group=TopSide_V02_Layer Emissive_01, parameterName="TopS Emissive_01 X+X------>X"}
  n196 = MaterialFunctionCall(In=Saturate(None=LinearInterpolate(A=@n194, B=@n193, Alpha=@n195)), Contrast=@n192) {materialFunction=&CheapContrast}
  n197 = StaticSwitchParameter(
   True=@n133
   False=StaticSwitchParameter(
    True=MaterialFunctionCall(Transition Phase=@n190, Height Texture=@n191, Contrast=@n192) {materialFunction=&HeightLerp}.Alpha
    False=StaticSwitchParameter(True=@n196.Result, False=@n181) {group=TopSide_V02_Layer Emissive_01, parameterName=TopS Emissive_01 X+X}
   ) {group=TopSide_V02_Layer Emissive_01, parameterName=TopS Emissive_01 HeightLerp}
  ) {group=TopSide_V02_Layer Emissive_01, parameterName=TopS Emissive_01 Straight Mask}
  n198 = StaticBoolParameter() {group=TopSide_V02_Layer Emissive_01, parameterName=TopS Emissive_01, sortPriority=0}
  FunctionOutput(None=StaticSwitch(True=@n197, False=Constant(), Value=@n198)) {outputName=Emissive 01 Msk}
  n199 = NamedRerouteUsage()
  n200 = ChannelMaskParameter(None=NamedRerouteUsage()) {group=TopSide_V02_Layer Emissive_02, parameterName=TopS Emissive_02-TS 03_Mask_Channel}
  n201 = StaticSwitchParameter(
   True=ChannelMaskParameter(None=NamedRerouteUsage()) {group=TopSide_V02_Layer Emissive_02, parameterName=TopS Emissive_02-TS 04_Mask_Channel}.R
   False=ChannelMaskParameter(None=NamedRerouteUsage()) {group=TopSide_V02_Layer Emissive_02, parameterName=TopS Emissive_02-TS 01_Mask_Channel}.R
  ) {group=TopSide_V02_Layer Emissive_02, parameterName="TopS Emissive_02-TS 01_Mask---->TS 04"}
  n202 = NamedRerouteUsage()
  n203 = VectorParameter() {
   group=TopSide_V02_Layer Emissive_02
   parameterName=TopS Emissive_02__Curvature_Amount
   channelNames {a=Grad_Pits, b=Grad_Peaks, g=Msks_Pits, r=Msks_Peaks}
  }
  n204 = NamedRerouteUsage()
  n205 = Add(
   A=StaticSwitchParameter(
    True=ChannelMaskParameter(None=@n199) {group=TopSide_V02_Layer Emissive_02, parameterName=TopS Emissive_02-TS 02_Mask_Channel}.R
    False=StaticSwitchParameter(True=@n200.R, False=@n201) {group=TopSide_V02_Layer Emissive_02, parameterName="TopS Emissive_02-TS 01_Mask---->TS 03"}
   ) {group=TopSide_V02_Layer Emissive_02, parameterName="TopS Emissive_02-TS 01_Mask---->TS 02"}
   B=Add(A=Multiply(A=@n202, B=@n203.Msks_Peaks), B=Multiply(A=@n204, B=@n203.Msks_Pits))
  )
  n206 = VectorParameter() {
   group=TopSide_V02_Layer Emissive_02
   parameterName=TopS Emissive-02-Tri-Mask
   channelNames {a=Contrast, b="AltAxiesMSk TopSide ----->Side", g="Grad   0=on", r="Hieght  0=on"}
   defaultValue=<rgba 1 1 0 1>
  }
  n207 = MaterialFunctionCall(In=LinearInterpolate(A=NamedRerouteUsage(), B=NamedRerouteUsage(), Alpha=@n206."AltAxiesMSk TopSide ----->Side"), Contrast=@n206.Contrast) {materialFunction=&CheapContrast}
  n208 = Multiply(A=Saturate(None=Add(A=@n207.Result, B=@n206."Hieght  0=on")), B=@n205)
  n209 = ScalarParameter() {group=TopSide_V02_Layer Emissive_02, parameterName=TopS Emissive_02 HeightLerp Swap Msk}
  n210 = NamedRerouteUsage()
  n211 = NamedRerouteUsage()
  n212 = Add(A=@n207.Result, B=@n206."Grad   0=on")
  n213 = VectorParameter() {
   group=TopSide_V02_Layer Emissive_02
   parameterName=TopS Emissive_02_Grad-Tex--Polar_Msk
   channelNames {a=Inner UV Offset, b=Outer, g=Innier, r="0=Use Mask_Values"}
   defaultValue=<rgba 0 1 0 0>
  }
  n214 = NamedRerouteUsage()
  n215 = NamedRerouteUsage()
  n216 = Constant()
  n217 = StaticBoolParameter() {group=TopSide_V02_Layer Emissive_02, parameterName=TopS Emissive_02_Grad----Texture}
  n218 = StaticSwitch(
   True=StaticBoolParameter() {group=TopSide_V02_Layer Emissive_02, parameterName="TopS Emissive_02_Grad-Tex-Uvs---->Polar"}
   False=StaticBool()
   Value=@n217
  )
  n219 = StaticSwitch(True=NamedRerouteUsage(), False=Constant() {r=1}, Value=@n217)
  n220 = MaterialFunctionCall(
   Caculated_Bounds=StaticSwitchParameter(True=NamedRerouteUsage(), False=NamedRerouteUsage()) {group=TopSide_V02_Layer Emissive_02, parameterName="TopS Emissive_02  FullWeaponBounds"}
   Pre_Skinned_Normals=NamedRerouteUsage()
   is_State=NamedRerouteUsage()
   UV_Rotation=NamedRerouteUsage()
   Uvs_Offset=StaticSwitch(True=@n214, False=StaticSwitch(True=@n215, False=@n216, Value=@n217), Value=@n218)
   Uvs_TileRate=@n219
   Uvs_Offset_Pivot=StaticSwitch(True=Constant(), False=@n214, Value=@n218)
   For_Safe_Zones=@n218
  ) {materialFunction=&MF_Projection_Side_Normal_V2}
  n221 = Constant() {r=2}
  n222 = MaterialFunctionCall(
   UV Coord=StaticSwitch(True=NamedRerouteUsage(), False=@n220.UVs No Rotation, Value=NamedRerouteUsage())
   LenghtScale=@n221
   RadialScale=@n221
  ) {materialFunction=&MF_PolarCoordinates}
  n223 = ComponentMask(None=@n222.Result) {r=true}
  n224 = SmoothStep(Min=@n213.Innier, Max=@n213.Innier, Value=@n223)
  n225 = VectorParameter() {group=TopSide_V02_Layer Emissive_02, parameterName=TopS Emissive_02_Grad-Tex--PolarTileRate, defaultValue=<rgba 1 1 0 1>}
  n226 = Multiply(A=@n222.Result, B=ComponentMask(None=@n225.RGB) {g=true, r=true})
  n227 = ComponentMask(None=@n215) {g=true, r=true}
  n228 = Constant()
  n229 = LinearInterpolate(B=@n213.Inner UV Offset, Alpha=@n223)
  n230 = StaticSwitch(
   True=Add(A=Add(A=@n226, B=@n227), B=AppendVector(A=@n228, B=@n229))
   False=StaticSwitchParameter(True=@n220.UVs No Rotation, False=@n220.UVs No Rotation) {group=TopSide_V02_Layer Emissive_02, parameterName="TopS Emissive_02  No_Rotation"}
   Value=@n218
  )
  n231 = StaticSwitch(
   True=Add(A=NamedRerouteUsage(), B=@n230)
   False=@n230
   Value=NamedRerouteDeclaration(None=StaticBoolParameter() {group=TopSide_V02_Layer Emissive_02 Reactive, parameterName=TopS Emissive_02_Offset----Reactive--------------}) {name=Emissive_02_Offset_bool, nodeColor=<rgba 0.3594 0 1 1>}
  )
  n232 = VectorParameter() {
   group=TopSide_V02_Layer Emissive_02
   parameterName=TopS Emissive_02_Grad-Tex-Clamp
   channelNames {a="", b="", g="TileRate ", r="Clamp  1=on"}
   defaultValue=<rgba 0 0.995 0 1>
  }
  n233 = LinearInterpolate(A=@n231, B=Multiply(A=Saturate(None=@n231), B=@n232."TileRate "), Alpha=@n232."Clamp  1=on")
  n234 = NamedRerouteUsage()
  n235 = ChannelMaskParameter(None=NamedRerouteUsage()) {group=TopSide_V02_Layer Emissive_02, parameterName=TopS_04_Em02_Warp_Mask_Channel, sortPriority=201}
  n236 = ChannelMaskParameter(None=NamedRerouteUsage()) {group=TopSide_V02_Layer Emissive_02, parameterName=TopS_01_Em02_Warp_Mask_Channel, sortPriority=200}
  n237 = StaticSwitchParameter(
   True=ChannelMaskParameter(None=NamedRerouteUsage()) {group=TopSide_V02_Layer Emissive_02, parameterName=TopS_02_Em02_Warp_Mask_Channel, sortPriority=201}.R
   False=StaticSwitchParameter(
    True=ChannelMaskParameter(None=@n234) {group=TopSide_V02_Layer Emissive_02, parameterName=TopS_03_Em02_Warp_Mask_Channel, sortPriority=201}.R
    False=StaticSwitchParameter(True=@n235.R, False=@n236.R) {group=TopSide_V02_Layer Emissive_02, parameterName="TopS_01_Em02_Warp-TS 01---->TS 04", sortPriority=200}
   ) {group=TopSide_V02_Layer Emissive_02, parameterName="TopS_01_Em02_Warp-TS 01---->TS 03", sortPriority=200}
  ) {group=TopSide_V02_Layer Emissive_02, parameterName="TopS_01_Em02_Warp-TS 01---->TS 02", sortPriority=200}
  n238 = Add(A=@n237, B=ScalarParameter() {group=TopSide_V02_Layer Emissive_02, parameterName=TopS_Em02_Warp_Offset, sortPriority=201})
  n239 = ScalarParameter() {group=TopSide_V02_Layer Emissive_02, parameterName=TopS_Em02_Warp_Amount, sortPriority=201}
  n240 = ScalarParameter() {group=TopSide_V02_Layer Emissive_02 Reactive, parameterName="TopS Audio 02 Emissive_Warp  Low", sortPriority=213}
  n241 = ScalarParameter() {defaultValue=0.1, group=TopSide_V02_Layer Emissive_02 Reactive, parameterName="TopS Audio 02 Emissive_Warp  High", sortPriority=213}
  n242 = NamedRerouteUsage()
  n243 = StaticSwitch(
   True=LinearInterpolate(
    A=@n239
    B=LinearInterpolate(A=@n240, B=@n241, Alpha=@n242)
    Alpha=ScalarParameter() {group=TopSide_V02_Layer Emissive_02 Reactive, parameterName="TopS Audio 02 Emissive_Warp  1=active", sortPriority=213}
   )
   False=@n239
   Value=NamedRerouteUsage()
  )
  n244 = NamedRerouteUsage()
  n245 = VectorParameter() {
   group=TopSide_V02_Layer Emissive_02
   parameterName=TopS_Em02_Warp_Curvature_Amount
   sortPriority=201
   channelNames {a=PitsUVS Grad, b=PeaksUVs Grad, g=PitsUVS, r=PeaksUVs}
  }
  n246 = NamedRerouteUsage()
  n247 = Add(A=Add(A=@n233, B=Multiply(A=@n238, B=@n243)), B=Add(A=Multiply(A=@n244, B=@n245.PeaksUVs), B=Multiply(A=@n246, B=@n245.PitsUVS)))
  n248 = StaticSwitchParameter(True=NamedRerouteUsage(), False=NamedRerouteUsage()) {group=TopSide_V02_Layer Emissive_02, parameterName=TopS Emissive_02_Grad-Tex_Mask Share_00B_Msk_Tex}
  n249 = TextureObjectParameter() {
   group=TopSide_V02_Layer Emissive_02
   parameterName=TopS Emissive_02_Grad-Tex_Mask
   samplerType=SAMPLERTYPE_Masks
   texture=&T_BlackMask
  }
  n250 = MaterialFunctionCall(Float2=TextureCoordinate()) {materialFunction=&BreakOutFloat2Components}
  n251 = ChannelMaskParameter(None=TextureSample(
    UVs=StaticSwitchParameter(True=@n247, False=@n233) {group=TopSide_V02_Layer Emissive_02, parameterName=TopS Emissive_02_Warp, sortPriority=200}
    Tex=StaticSwitchParameter(True=@n248, False=@n249) {group=TopSide_V02_Layer Emissive_02, parameterName=TopS Emissive_02_Grad-Tex_Mask Share_00A_Msk_Tex}
    "DDX(UVs)"=DDX(Value=@n250.R)
    "DDY(UVs)"=DDY(Value=@n250.G)
   ) {mipValueMode=TMVM_Derivative}.RGBA) {group=TopSide_V02_Layer Emissive_02, maskChannel=Alpha, parameterName=TopS Emissive_02_Grad-Tex_Mask_Channel, defaultValue=<rgba 0 0 0 1>}
  n252 = StaticSwitch(True=Multiply(A=Max(A=@n213."0=Use Mask_Values", B=@n224), B=@n251.R), False=@n251.R, Value=@n218)
  n253 = ComponentMask(None=@n231) {r=true}
  n254 = Divide(A=@n253, B=NamedRerouteUsage())
  n255 = OneMinus(None=Power(Base=Saturate(None=@n254), Exp=ScalarParameter() {defaultValue=2, group=TopSide_V02_Layer Emissive_02, parameterName=TopS_Em02__Pow}))
  n256 = Add(
   A=Add(A=Multiply(A=@n210, B=@n203.Grad_Peaks), B=Multiply(A=@n211, B=@n203.Grad_Pits))
   B=Multiply(A=Saturate(None=@n212), B=StaticSwitch(True=@n252, False=@n255, Value=@n217))
  )
  n257 = NamedRerouteUsage()
  n258 = Multiply(A=@n256, B=@n257)
  n259 = ScalarParameter() {group=TopSide_V02_Layer Emissive_02 Reactive, parameterName="TopS Emissive_02-Brightness-Reactive--->Gradent  Mult", sliderMax=1}
  n260 = Add(A=@n256, B=@n257)
  n261 = ScalarParameter() {group=TopSide_V02_Layer Emissive_02 Reactive, parameterName="TopS Emissive_02-Brightness-Reactive--->Gradent  Add", sliderMax=1}
  n262 = StaticSwitch(
   True=LinearInterpolate(A=LinearInterpolate(A=@n256, B=@n258, Alpha=@n259) {constA=1}, B=Saturate(None=@n260), Alpha=@n261)
   False=@n256
   Value=NamedRerouteUsage()
  )
  n263 = ScalarParameter() {defaultValue=1, group=TopSide_V02_Layer Emissive_02, parameterName=TopS Emissive_02-Contrast}
  n264 = MaterialFunctionCall(
   Transition Phase=LinearInterpolate(A=@n262, B=@n208, Alpha=@n209)
   Height Texture=LinearInterpolate(A=@n208, B=@n262, Alpha=@n209)
   Contrast=@n263
  ) {materialFunction=&HeightLerp}
  n265 = Multiply(A=@n208, B=@n262)
  n266 = Multiply(A=@n265, B=Add(A=@n265, B=@n262))
  n267 = ScalarParameter() {group=TopSide_V02_Layer Emissive_02, parameterName="TopS Emissive_02 X+X------>X"}
  n268 = MaterialFunctionCall(In=Saturate(None=LinearInterpolate(A=@n266, B=@n265, Alpha=@n267)), Contrast=@n263) {materialFunction=&CheapContrast}
  n269 = StaticSwitchParameter(
   True=@n205
   False=StaticSwitchParameter(
    True=@n264.Alpha
    False=StaticSwitchParameter(True=@n268.Result, False=@n256) {group=TopSide_V02_Layer Emissive_02, parameterName=TopS Emissive_02 X+X}
   ) {group=TopSide_V02_Layer Emissive_02, parameterName=TopS Emissive_02 HeightLerp}
  ) {group=TopSide_V02_Layer Emissive_02, parameterName=TopS Emissive_02 Straight Mask}
  n270 = StaticBoolParameter() {group=TopSide_V02_Layer Emissive_02, parameterName=TopS Emissive_02, sortPriority=0}
  FunctionOutput(None=StaticSwitch(True=@n269, False=Constant(), Value=@n270)) {outputName=Emissive 02 Msk}
  n271 = NamedRerouteUsage()
  n272 = ChannelMaskParameter(None=NamedRerouteUsage()) {group=TopSide_V02_Layer Emissive_03, parameterName=TopS Emissive_03-TS 03_Mask_Channel}
  n273 = StaticSwitchParameter(
   True=ChannelMaskParameter(None=NamedRerouteUsage()) {group=TopSide_V02_Layer Emissive_03, parameterName=TopS Emissive_03-TS 04_Mask_Channel}.R
   False=ChannelMaskParameter(None=NamedRerouteUsage()) {group=TopSide_V02_Layer Emissive_03, parameterName=TopS Emissive_03-TS 01_Mask_Channel}.R
  ) {group=TopSide_V02_Layer Emissive_03, parameterName="TopS Emissive_03-TS 01_Mask---->TS 04"}
  n274 = NamedRerouteUsage()
  n275 = VectorParameter() {
   group=TopSide_V02_Layer Emissive_03
   parameterName=TopS Emissive_03__Curvature_Amount
   channelNames {a=Grad_Pits, b=Grad_Peaks, g=Msks_Pits, r=Msks_Peaks}
  }
  n276 = NamedRerouteUsage()
  n277 = Add(
   A=StaticSwitchParameter(
    True=ChannelMaskParameter(None=@n271) {group=TopSide_V02_Layer Emissive_03, parameterName=TopS Emissive_03-TS 02_Mask_Channel}.R
    False=StaticSwitchParameter(True=@n272.R, False=@n273) {group=TopSide_V02_Layer Emissive_03, parameterName="TopS Emissive_03-TS 01_Mask---->TS 03"}
   ) {group=TopSide_V02_Layer Emissive_03, parameterName="TopS Emissive_03-TS 01_Mask---->TS 02"}
   B=Add(A=Multiply(A=@n274, B=@n275.Msks_Peaks), B=Multiply(A=@n276, B=@n275.Msks_Pits))
  )
  n278 = VectorParameter() {
   group=TopSide_V02_Layer Emissive_03
   parameterName=TopS Emissive-03-Tri-Mask
   channelNames {a=Contrast, b="AltAxiesMSk TopSide ----->Side", g="Grad   0=on", r="Hieght  0=on"}
   defaultValue=<rgba 1 1 0 1>
  }
  n279 = MaterialFunctionCall(In=LinearInterpolate(A=NamedRerouteUsage(), B=NamedRerouteUsage(), Alpha=@n278."AltAxiesMSk TopSide ----->Side"), Contrast=@n278.Contrast) {materialFunction=&CheapContrast}
  n280 = Multiply(A=Saturate(None=Add(A=@n279.Result, B=@n278."Hieght  0=on")), B=@n277)
  n281 = ScalarParameter() {group=TopSide_V02_Layer Emissive_03, parameterName=TopS Emissive_03 HeightLerp Swap Msk}
  n282 = NamedRerouteUsage()
  n283 = NamedRerouteUsage()
  n284 = Add(A=@n279.Result, B=@n278."Grad   0=on")
  n285 = VectorParameter() {
   group=TopSide_V02_Layer Emissive_03
   parameterName=TopS Emissive_03_Grad-Tex--Polar_Msk
   channelNames {a=Inner UV Offset, b=Outer, g=Innier, r="0=Use Mask_Values"}
   defaultValue=<rgba 0 1 0 0>
  }
  n286 = NamedRerouteUsage()
  n287 = StaticBoolParameter() {group=TopSide_V02_Layer Emissive_03, parameterName=TopS Emissive_03_Grad----Texture}
  n288 = StaticSwitch(True=NamedRerouteUsage(), False=Constant(), Value=@n287)
  n289 = StaticSwitch(
   True=StaticBoolParameter() {group=TopSide_V02_Layer Emissive_03, parameterName="TopS Emissive_03_Grad-Tex-Uvs---->Polar"}
   False=StaticBool()
   Value=@n287
  )
  n290 = MaterialFunctionCall(
   Caculated_Bounds=StaticSwitchParameter(True=NamedRerouteUsage(), False=NamedRerouteUsage()) {group=TopSide_V02_Layer Emissive_03, parameterName="TopS Emissive_03  FullWeaponBounds"}
   Pre_Skinned_Normals=NamedRerouteUsage()
   is_State=NamedRerouteUsage()
   UV_Rotation=NamedRerouteUsage()
   Uvs_Offset=StaticSwitch(True=@n286, False=@n288, Value=@n289)
   Uvs_TileRate=StaticSwitch(True=NamedRerouteUsage(), False=Constant() {r=1}, Value=@n287)
   Uvs_Offset_Pivot=StaticSwitch(True=Constant(), False=@n286, Value=@n289)
   For_Safe_Zones=@n289
  ) {materialFunction=&MF_Projection_Side_Normal_V2}
  n291 = Constant() {r=2}
  n292 = MaterialFunctionCall(
   UV Coord=StaticSwitch(True=NamedRerouteUsage(), False=@n290.UVs No Rotation, Value=NamedRerouteUsage())
   LenghtScale=@n291
   RadialScale=@n291
  ) {materialFunction=&MF_PolarCoordinates}
  n293 = ComponentMask(None=@n292.Result) {r=true}
  n294 = SmoothStep(Min=@n285.Innier, Max=@n285.Innier, Value=@n293)
  n295 = VectorParameter() {group=TopSide_V02_Layer Emissive_03, parameterName=TopS Emissive_03_Grad-Tex--PolarTileRate, defaultValue=<rgba 1 1 0 1>}
  n296 = Multiply(A=@n292.Result, B=ComponentMask(None=@n295.RGB) {g=true, r=true})
  n297 = ComponentMask(None=@n288) {g=true, r=true}
  n298 = Constant()
  n299 = LinearInterpolate(B=@n285.Inner UV Offset, Alpha=@n293)
  n300 = StaticSwitch(
   True=Add(A=Add(A=@n296, B=@n297), B=AppendVector(A=@n298, B=@n299))
   False=StaticSwitchParameter(True=@n290.UVs No Rotation, False=@n290.UVs No Rotation) {group=TopSide_V02_Layer Emissive_03, parameterName="TopS Emissive_03  No_Rotation"}
   Value=@n289
  )
  n301 = StaticSwitch(
   True=Add(A=NamedRerouteUsage(), B=@n300)
   False=@n300
   Value=NamedRerouteDeclaration(None=StaticBoolParameter() {group=TopSide_V02_Layer Emissive_03 Reactive, parameterName=TopS Emissive_03_Offset----Reactive--------------}) {name=Emissive_03_Offset_bool, nodeColor=<rgba 0.3594 0 1 1>}
  )
  n302 = VectorParameter() {
   group=TopSide_V02_Layer Emissive_03
   parameterName=TopS Emissive_03_Grad-Tex-Clamp
   channelNames {a="", b="", g="TileRate ", r="Clamp  1=on"}
   defaultValue=<rgba 0 0.995 0 1>
  }
  n303 = LinearInterpolate(A=@n301, B=Multiply(A=Saturate(None=@n301), B=@n302."TileRate "), Alpha=@n302."Clamp  1=on")
  n304 = NamedRerouteUsage()
  n305 = ChannelMaskParameter(None=NamedRerouteUsage()) {group=TopSide_V02_Layer Emissive_03, parameterName=TopS_04_Em03_Warp_Mask_Channel, sortPriority=201}
  n306 = ChannelMaskParameter(None=NamedRerouteUsage()) {group=TopSide_V02_Layer Emissive_03, parameterName=TopS_01_Em03_Warp_Mask_Channel, sortPriority=200}
  n307 = StaticSwitchParameter(
   True=ChannelMaskParameter(None=NamedRerouteUsage()) {group=TopSide_V02_Layer Emissive_03, parameterName=TopS_02_Em03_Warp_Mask_Channel, sortPriority=201}.R
   False=StaticSwitchParameter(
    True=ChannelMaskParameter(None=@n304) {group=TopSide_V02_Layer Emissive_03, parameterName=TopS_03_Em03_Warp_Mask_Channel, sortPriority=201}.R
    False=StaticSwitchParameter(True=@n305.R, False=@n306.R) {group=TopSide_V02_Layer Emissive_03, parameterName="TopS_01_Em03_Warp-TS 01---->TS 04", sortPriority=200}
   ) {group=TopSide_V02_Layer Emissive_03, parameterName="TopS_01_Em03_Warp-TS 01---->TS 03", sortPriority=200}
  ) {group=TopSide_V02_Layer Emissive_03, parameterName="TopS_01_Em03_Warp-TS 01---->TS 02", sortPriority=200}
  n308 = Add(A=@n307, B=ScalarParameter() {group=TopSide_V02_Layer Emissive_03, parameterName=TopS_Em03_Warp_Offset, sortPriority=201})
  n309 = ScalarParameter() {group=TopSide_V02_Layer Emissive_03, parameterName=TopS_Em03_Warp_Amount, sortPriority=201}
  n310 = ScalarParameter() {group=TopSide_V02_Layer Emissive_03 Reactive, parameterName="TopS Audio 03 Emissive_Warp  Low", sortPriority=213}
  n311 = ScalarParameter() {defaultValue=0.1, group=TopSide_V02_Layer Emissive_03 Reactive, parameterName="TopS Audio 03 Emissive_Warp  High", sortPriority=213}
  n312 = NamedRerouteUsage()
  n313 = StaticSwitch(
   True=LinearInterpolate(
    A=@n309
    B=LinearInterpolate(A=@n310, B=@n311, Alpha=@n312)
    Alpha=ScalarParameter() {group=TopSide_V02_Layer Emissive_03 Reactive, parameterName="TopS Audio 03 Emissive_Warp  1=active", sortPriority=213}
   )
   False=@n309
   Value=NamedRerouteUsage()
  )
  n314 = NamedRerouteUsage()
  n315 = VectorParameter() {
   group=TopSide_V02_Layer Emissive_03
   parameterName=TopS_Em03_Warp_Curvature_Amount
   sortPriority=201
   channelNames {a=PitsUVS Grad, b=PeaksUVs Grad, g=PitsUVS, r=PeaksUVs}
  }
  n316 = NamedRerouteUsage()
  n317 = Add(A=Add(A=@n303, B=Multiply(A=@n308, B=@n313)), B=Add(A=Multiply(A=@n314, B=@n315.PeaksUVs), B=Multiply(A=@n316, B=@n315.PitsUVS)))
  n318 = StaticSwitchParameter(True=NamedRerouteUsage(), False=NamedRerouteUsage()) {group=TopSide_V02_Layer Emissive_03, parameterName=TopS Emissive_03_Grad-Tex_Mask Share_00B_Msk_Tex}
  n319 = TextureObjectParameter() {
   group=TopSide_V02_Layer Emissive_03
   parameterName=TopS Emissive_03_Grad-Tex_Mask
   samplerType=SAMPLERTYPE_Masks
   texture=&T_BlackMask
  }
  n320 = MaterialFunctionCall(Float2=TextureCoordinate()) {materialFunction=&BreakOutFloat2Components}
  n321 = ChannelMaskParameter(None=TextureSample(
    UVs=StaticSwitchParameter(True=@n317, False=@n303) {group=TopSide_V02_Layer Emissive_03, parameterName=TopS Emissive_03_Warp, sortPriority=200}
    Tex=StaticSwitchParameter(True=@n318, False=@n319) {group=TopSide_V02_Layer Emissive_03, parameterName=TopS Emissive_03_Grad-Tex_Mask Share_00A_Msk_Tex}
    "DDX(UVs)"=DDX(Value=@n320.R)
    "DDY(UVs)"=DDY(Value=@n320.G)
   ) {mipValueMode=TMVM_Derivative}.RGBA) {group=TopSide_V02_Layer Emissive_03, maskChannel=Alpha, parameterName=TopS Emissive_03_Grad-Tex_Mask_Channel, defaultValue=<rgba 0 0 0 1>}
  n322 = StaticSwitch(True=Multiply(A=Max(A=@n285."0=Use Mask_Values", B=@n294), B=@n321.R), False=@n321.R, Value=@n289)
  n323 = Divide(A=ComponentMask(None=@n301) {r=true}, B=NamedRerouteUsage())
  n324 = OneMinus(None=Power(Base=Saturate(None=@n323), Exp=ScalarParameter() {defaultValue=2, group=TopSide_V02_Layer Emissive_03, parameterName=TopS_Em03__Pow}))
  n325 = Add(
   A=Add(A=Multiply(A=@n282, B=@n275.Grad_Peaks), B=Multiply(A=@n283, B=@n275.Grad_Pits))
   B=Multiply(A=Saturate(None=@n284), B=StaticSwitch(True=@n322, False=@n324, Value=@n287))
  )
  n326 = NamedRerouteUsage()
  n327 = Multiply(A=@n325, B=@n326)
  n328 = ScalarParameter() {group=TopSide_V02_Layer Emissive_03 Reactive, parameterName="TopS Emissive_03-Brightness-Reactive--->Gradent  Mult", sliderMax=1}
  n329 = Add(A=@n325, B=@n326)
  n330 = ScalarParameter() {group=TopSide_V02_Layer Emissive_03 Reactive, parameterName="TopS Emissive_03-Brightness-Reactive--->Gradent  Add", sliderMax=1}
  n331 = StaticSwitch(
   True=LinearInterpolate(A=LinearInterpolate(A=@n325, B=@n327, Alpha=@n328) {constA=1}, B=Saturate(None=@n329), Alpha=@n330)
   False=@n325
   Value=NamedRerouteUsage()
  )
  n332 = LinearInterpolate(A=@n331, B=@n280, Alpha=@n281)
  n333 = LinearInterpolate(A=@n280, B=@n331, Alpha=@n281)
  n334 = ScalarParameter() {defaultValue=1, group=TopSide_V02_Layer Emissive_03, parameterName=TopS Emissive_03-Contrast}
  n335 = Multiply(A=@n280, B=@n331)
  n336 = Multiply(A=@n335, B=Add(A=@n335, B=@n331))
  n337 = ScalarParameter() {group=TopSide_V02_Layer Emissive_03, parameterName="TopS Emissive_03 X+X---->X"}
  n338 = MaterialFunctionCall(In=Saturate(None=LinearInterpolate(A=@n336, B=@n335, Alpha=@n337)), Contrast=@n334) {materialFunction=&CheapContrast}
  n339 = StaticSwitchParameter(
   True=@n277
   False=StaticSwitchParameter(
    True=MaterialFunctionCall(Transition Phase=@n332, Height Texture=@n333, Contrast=@n334) {materialFunction=&HeightLerp}.Alpha
    False=StaticSwitchParameter(True=@n338.Result, False=@n325) {group=TopSide_V02_Layer Emissive_03, parameterName=TopS Emissive_03 X+X}
   ) {group=TopSide_V02_Layer Emissive_03, parameterName=TopS Emissive_03 HeightLerp}
  ) {group=TopSide_V02_Layer Emissive_03, parameterName=TopS Emissive_03 Straight Mask}
  n340 = StaticBoolParameter() {group=TopSide_V02_Layer Emissive_03, parameterName=TopS Emissive_03, sortPriority=0}
  FunctionOutput(None=StaticSwitch(True=@n339, False=Constant(), Value=@n340)) {outputName=Emissive 03 Msk}
  FunctionOutput(None=NamedRerouteUsage()) {outputName=TS_Layer03_Mask}
  FunctionOutput(None=NamedRerouteUsage()) {outputName=TS_Layer04_Mask}
  FunctionOutput(None=@n98) {outputName="Blend Tops01<--->Topsside01 Msk"}
  n341 = ScalarParameter() {group=TopSide_V02_Layer-01_Uvs | Warp, parameterName="TopS_01_Warp_Rotation_Mech\\Character\\Backpack", sortPriority=300}
  n342 = MaterialFunctionCall(Float4=NamedRerouteUsage()) {materialFunction=&BreakOutFloat4Components}
  LinearInterpolate(
   A=@n341
   B=ScalarParameter() {group=TopSide_V02_Layer-01_Uvs | Warp, parameterName=TopS_01_Warp_Rotation_Character, sortPriority=300}
   Alpha=@n342.A
  )
  n343 = VectorParameter() {group=TopSide_V02_Layer-02_Uvs, parameterName="TopS_02_Rotation_Weapon\\Pickaxe", defaultValue=<rgba 0 0 -0.25 1>}
  n344 = VectorParameter() {group=TopSide_V02_Layer-02_Uvs, parameterName=TopS_02_Rotation_Vehicle_Side, defaultValue=<rgba 0 0 0 1>}
  n345 = MaterialFunctionCall(Float4=NamedRerouteUsage()) {materialFunction=&BreakOutFloat4Components}
  n346 = LinearInterpolate(
   A=VectorParameter() {group=TopSide_V02_Layer-02_Uvs, parameterName=TopS_02_Rotation_Vehicle, defaultValue=<rgba 0 0 0 1>}.RGB
   B=@n344.RGB
   Alpha=@n345.G
  )
  n347 = VectorParameter() {group=TopSide_V02_Layer-02_Uvs, parameterName="TopS_02_Rotation_Mech\\Character\\Backpack", defaultValue=<rgba 0.25 0.25 0 1>}
  n348 = VectorParameter() {group=TopSide_V02_Layer-02_Uvs, parameterName="TopS_02_Rotation_Mech A 1=on", defaultValue=<rgba 0.25 0.25 0 0>}
  n349 = LinearInterpolate(A=@n347.RGB, B=@n348.RGB, Alpha=@n348.RGB)
  n350 = MaterialFunctionCall(Float2=NamedRerouteUsage()) {materialFunction=&BreakOutFloat2Components}
  n351 = LinearInterpolate(
   A=LinearInterpolate(A=LinearInterpolate(A=@n343.RGB, B=@n346, Alpha=@n345.R), B=LinearInterpolate(A=@n349, B=@n347.RGB, Alpha=@n345.A), Alpha=@n345.B)
   B=@n347.RGB
   Alpha=@n350.R
  )
  LinearInterpolate(
   A=@n351
   B=VectorParameter() {group=TopSide_V02_Layer-02_Uvs, parameterName=TopS_02_Rotation_Pickaxe, defaultValue=<rgba 0 0 0 1>}.RGB
   Alpha=@n350.G
  )
  n352 = ScalarParameter() {group=TopSide_V02_Layer-01_Uvs | Warp, parameterName=TopS_01_Warp_Rotation_Vehicle, sortPriority=300}
  n353 = ScalarParameter() {group=TopSide_V02_Layer-01_Uvs | Warp, parameterName=TopS_01_Warp_Rotation_Vehicle_Side, sortPriority=300}
  n354 = LinearInterpolate(
   A=LinearInterpolate(
    A=ScalarParameter() {group=TopSide_V02_Layer-01_Uvs | Warp, parameterName="TopS_01_Warp_Rotation_Weapon\\Pickaxe", sortPriority=300}
    B=LinearInterpolate(A=@n352, B=@n353, Alpha=@n342.G)
    Alpha=@n342.R
   )
   B=@n341
   Alpha=@n342.B
  )
  n355 = MaterialFunctionCall(Float2=NamedRerouteUsage()) {materialFunction=&BreakOutFloat2Components}
  LinearInterpolate(
   A=LinearInterpolate(
    A=@n354
    B=ScalarParameter() {group=TopSide_V02_Layer-01_Uvs | Warp, parameterName=TopS_01_Warp_Rotation_Backpack, sortPriority=300}
    Alpha=@n355.R
   )
   B=ScalarParameter() {group=TopSide_V02_Layer-01_Uvs | Warp, parameterName=TopS_01_Warp_Rotation_Pickaxe, sortPriority=300}
   Alpha=@n355.G
  )
  n356 = ScalarParameter() {group=TopSide_V02_Layer-02_Uvs | Warp, parameterName="TopS_02_Warp_Rotation_Mech\\Character\\Backpack", sortPriority=300}
  n357 = MaterialFunctionCall(Float4=NamedRerouteUsage()) {materialFunction=&BreakOutFloat4Components}
  LinearInterpolate(
   A=@n356
   B=ScalarParameter() {group=TopSide_V02_Layer-02_Uvs | Warp, parameterName=TopS_02_Warp_Rotation_Character, sortPriority=300}
   Alpha=@n357.A
  )
  n358 = ScalarParameter() {group=TopSide_V02_Layer-02_Uvs | Warp, parameterName=TopS_02_Warp_Rotation_Vehicle, sortPriority=300}
  n359 = ScalarParameter() {group=TopSide_V02_Layer-02_Uvs | Warp, parameterName=TopS_02_Warp_Rotation_Vehicle_Side, sortPriority=300}
  n360 = LinearInterpolate(
   A=LinearInterpolate(
    A=ScalarParameter() {group=TopSide_V02_Layer-02_Uvs | Warp, parameterName="TopS_02_Warp_Rotation_Weapon\\Pickaxe", sortPriority=300}
    B=LinearInterpolate(A=@n358, B=@n359, Alpha=@n357.G)
    Alpha=@n357.R
   )
   B=@n356
   Alpha=@n357.B
  )
  n361 = MaterialFunctionCall(Float2=NamedRerouteUsage()) {materialFunction=&BreakOutFloat2Components}
  LinearInterpolate(
   A=LinearInterpolate(
    A=@n360
    B=ScalarParameter() {group=TopSide_V02_Layer-02_Uvs | Warp, parameterName=TopS_02_Warp_Rotation_Backpack, sortPriority=300}
    Alpha=@n361.R
   )
   B=ScalarParameter() {group=TopSide_V02_Layer-02_Uvs | Warp, parameterName=TopS_02_Warp_Rotation_Pickaxe, sortPriority=300}
   Alpha=@n361.G
  )
  n362 = ScalarParameter() {group=TopSide_V02_Layer-02_Uvs, parameterName="TopS_02_Side_Rotation_Mech\\Character\\Backpack"}
  n363 = MaterialFunctionCall(Float4=NamedRerouteUsage()) {materialFunction=&BreakOutFloat4Components}
  LinearInterpolate(A=@n362, B=ScalarParameter() {group=TopSide_V02_Layer-02_Uvs, parameterName=TopS_02_Side_Rotation_Character}, Alpha=@n363.A)
  n364 = ScalarParameter() {group=TopSide_V02_Layer-02_Uvs, parameterName=TopS_02_Side_Rotation_Vehicle}
  n365 = ScalarParameter() {group=TopSide_V02_Layer-02_Uvs, parameterName=TopS_02_Side_Rotation_Vehicle_Side}
  n366 = LinearInterpolate(
   A=LinearInterpolate(
    A=ScalarParameter() {group=TopSide_V02_Layer-02_Uvs, parameterName="TopS_02_Side_Rotation_Weapon\\Pickaxe"}
    B=LinearInterpolate(A=@n364, B=@n365, Alpha=@n363.G)
    Alpha=@n363.R
   )
   B=@n362
   Alpha=@n363.B
  )
  n367 = MaterialFunctionCall(Float2=NamedRerouteUsage()) {materialFunction=&BreakOutFloat2Components}
  LinearInterpolate(
   A=LinearInterpolate(A=@n366, B=ScalarParameter() {group=TopSide_V02_Layer-02_Uvs, parameterName=TopS_02_Side_Rotation_Backpack}, Alpha=@n367.R)
   B=ScalarParameter() {group=TopSide_V02_Layer-02_Uvs, parameterName=TopS_02_Side_Rotation_Pickaxe}
   Alpha=@n367.G
  )
  n368 = ScalarParameter() {defaultValue=0.5, group=TopSide_V02_Layer Emissive_01, parameterName="TopS_Em01_Rotation_Mech\\Character\\BackPack"}
  n369 = MaterialFunctionCall(Float4=NamedRerouteUsage()) {materialFunction=&BreakOutFloat4Components}
  LinearInterpolate(
   A=@n368
   B=ScalarParameter() {defaultValue=0.5, group=TopSide_V02_Layer Emissive_01, parameterName=TopS_Em01_Rotation_Character}
   Alpha=@n369.A
  )
  n370 = ScalarParameter() {defaultValue=0.5, group=TopSide_V02_Layer Emissive_01, parameterName=TopS_Em01_Rotation_Vehicle}
  n371 = ScalarParameter() {defaultValue=0.5, group=TopSide_V02_Layer Emissive_01, parameterName=TopS_Em01_Rotation_Vehicle_Side}
  n372 = LinearInterpolate(
   A=LinearInterpolate(
    A=ScalarParameter() {defaultValue=0.5, group=TopSide_V02_Layer Emissive_01, parameterName="TopS_Em01_Rotation_Weapon\\Pickaxe"}
    B=LinearInterpolate(A=@n370, B=@n371, Alpha=@n369.G)
    Alpha=@n369.R
   )
   B=@n368
   Alpha=@n369.B
  )
  n373 = MaterialFunctionCall(Float2=NamedRerouteUsage()) {materialFunction=&BreakOutFloat2Components}
  LinearInterpolate(
   A=LinearInterpolate(
    A=@n372
    B=ScalarParameter() {defaultValue=0.5, group=TopSide_V02_Layer Emissive_01, parameterName=TopS_Em01_Rotation_Backpack}
    Alpha=@n373.R
   )
   B=ScalarParameter() {defaultValue=0.5, group=TopSide_V02_Layer Emissive_01, parameterName=TopS_Em01_Rotation_Pickaxe}
   Alpha=@n373.G
  )
  LinearInterpolate(Alpha=ScalarParameter() {group=TopSide_V02_Layer Emissive_02, parameterName="TopS Emissive_02_ColorCuve--->Use-UVS   1=On"})
  n374 = ScalarParameter() {defaultValue=0.5, group=TopSide_V02_Layer Emissive_02, parameterName="TopS_Em02_Rotation_Mech\\Character\\Backpack"}
  n375 = MaterialFunctionCall(Float4=NamedRerouteUsage()) {materialFunction=&BreakOutFloat4Components}
  LinearInterpolate(
   A=@n374
   B=ScalarParameter() {defaultValue=0.5, group=TopSide_V02_Layer Emissive_02, parameterName=TopS_Em02_Rotation_Character}
   Alpha=@n375.A
  )
  n376 = ScalarParameter() {defaultValue=0.5, group=TopSide_V02_Layer Emissive_02, parameterName=TopS_Em02_Rotation_Vehicle}
  n377 = ScalarParameter() {defaultValue=0.5, group=TopSide_V02_Layer Emissive_02, parameterName=TopS_Em02_Rotation_Vehicle_Side}
  n378 = LinearInterpolate(
   A=LinearInterpolate(
    A=ScalarParameter() {defaultValue=0.5, group=TopSide_V02_Layer Emissive_02, parameterName="TopS_Em02_Rotation_Weapon\\Pickaxe"}
    B=LinearInterpolate(A=@n376, B=@n377, Alpha=@n375.G)
    Alpha=@n375.R
   )
   B=@n374
   Alpha=@n375.B
  )
  n379 = MaterialFunctionCall(Float2=NamedRerouteUsage()) {materialFunction=&BreakOutFloat2Components}
  LinearInterpolate(
   A=LinearInterpolate(
    A=@n378
    B=ScalarParameter() {defaultValue=0.5, group=TopSide_V02_Layer Emissive_02, parameterName=TopS_Em02_Rotation_Backpack}
    Alpha=@n379.R
   )
   B=ScalarParameter() {defaultValue=0.5, group=TopSide_V02_Layer Emissive_02, parameterName=TopS_Em02_Rotation_Pickaxe}
   Alpha=@n379.G
  )
  n380 = VectorParameter() {group=TopSide_V02_Layer-03_Uvs, parameterName="TopS_03_Rotation_Mech\\Character\\Backpack", defaultValue=<rgba 0.25 0.25 0 1>}
  n381 = VectorParameter() {group=TopSide_V02_Layer-03_Uvs, parameterName=TopS_03_Rotation_Character, defaultValue=<rgba 0.25 0.25 0 1>}
  n382 = MaterialFunctionCall(Float4=NamedRerouteUsage()) {materialFunction=&BreakOutFloat4Components}
  LinearInterpolate(A=@n380.RGB, B=@n381.RGB, Alpha=@n382.A)
  n383 = VectorParameter() {group=TopSide_V02_Layer-03_Uvs, parameterName="TopS_03_Rotation_Weapon\\Pickaxe", defaultValue=<rgba 0 0 -0.25 1>}
  n384 = VectorParameter() {group=TopSide_V02_Layer-03_Uvs, parameterName=TopS_03_Rotation_Vehicle, defaultValue=<rgba 0 0 0 1>}
  n385 = VectorParameter() {group=TopSide_V02_Layer-03_Uvs, parameterName=TopS_03_Rotation_Vehicle_Side, defaultValue=<rgba 0 0 0 1>}
  n386 = LinearInterpolate(A=LinearInterpolate(A=@n383.RGB, B=LinearInterpolate(A=@n384.RGB, B=@n385.RGB, Alpha=@n382.G), Alpha=@n382.R), B=@n380.RGB, Alpha=@n382.B)
  n387 = MaterialFunctionCall(Float2=NamedRerouteUsage()) {materialFunction=&BreakOutFloat2Components}
  LinearInterpolate(
   A=LinearInterpolate(
    A=@n386
    B=VectorParameter() {group=TopSide_V02_Layer-03_Uvs, parameterName=TopS_03_Rotation_Backpack, defaultValue=<rgba 0 0 0 1>}.RGB
    Alpha=@n387.R
   )
   B=VectorParameter() {group=TopSide_V02_Layer-03_Uvs, parameterName=TopS_03_Rotation_Pickaxe, defaultValue=<rgba 0 0 0 1>}.RGB
   Alpha=@n387.G
  )
  n388 = ScalarParameter() {group=TopSide_V02_Layer-03_Uvs | Warp, parameterName="TopS_03_Warp_Rotation_Mech\\Character\\Backpack", sortPriority=300}
  n389 = MaterialFunctionCall(Float4=NamedRerouteUsage()) {materialFunction=&BreakOutFloat4Components}
  LinearInterpolate(
   A=@n388
   B=ScalarParameter() {group=TopSide_V02_Layer-03_Uvs | Warp, parameterName=TopS_03_Warp_Rotation_Character, sortPriority=300}
   Alpha=@n389.A
  )
  n390 = ScalarParameter() {group=TopSide_V02_Layer-03_Uvs | Warp, parameterName=TopS_03_Warp_Rotation_Vehicle, sortPriority=300}
  n391 = ScalarParameter() {group=TopSide_V02_Layer-03_Uvs | Warp, parameterName=TopS_03_Warp_Rotation_Vehicle_Side, sortPriority=300}
  n392 = LinearInterpolate(
   A=LinearInterpolate(
    A=ScalarParameter() {group=TopSide_V02_Layer-03_Uvs | Warp, parameterName="TopS_03_Warp_Rotation_Weapon\\Pickaxe", sortPriority=300}
    B=LinearInterpolate(A=@n390, B=@n391, Alpha=@n389.G)
    Alpha=@n389.R
   )
   B=@n388
   Alpha=@n389.B
  )
  n393 = MaterialFunctionCall(Float2=NamedRerouteUsage()) {materialFunction=&BreakOutFloat2Components}
  LinearInterpolate(
   A=LinearInterpolate(
    A=@n392
    B=ScalarParameter() {group=TopSide_V02_Layer-03_Uvs | Warp, parameterName=TopS_03_Warp_Rotation_Backpack, sortPriority=300}
    Alpha=@n393.R
   )
   B=ScalarParameter() {group=TopSide_V02_Layer-03_Uvs | Warp, parameterName=TopS_03_Warp_Rotation_Pickaxe, sortPriority=300}
   Alpha=@n393.G
  )
  n394 = ScalarParameter() {group=TopSide_V02_Layer-03_Uvs, parameterName="TopS_03_Side_Rotation_Mech\\Character\\Backpack"}
  n395 = MaterialFunctionCall(Float4=NamedRerouteUsage()) {materialFunction=&BreakOutFloat4Components}
  LinearInterpolate(A=@n394, B=ScalarParameter() {group=TopSide_V02_Layer-03_Uvs, parameterName=TopS_03_Side_Rotation_Character}, Alpha=@n395.A)
  n396 = ScalarParameter() {group=TopSide_V02_Layer-03_Uvs, parameterName=TopS_03_Side_Rotation_Vehicle}
  n397 = ScalarParameter() {group=TopSide_V02_Layer-03_Uvs, parameterName=TopS_03_Side_Rotation_Vehicle_Side}
  n398 = LinearInterpolate(
   A=LinearInterpolate(
    A=ScalarParameter() {group=TopSide_V02_Layer-03_Uvs, parameterName="TopS_03_Side_Rotation_Weapon\\Pickaxe"}
    B=LinearInterpolate(A=@n396, B=@n397, Alpha=@n395.G)
    Alpha=@n395.R
   )
   B=@n394
   Alpha=@n395.B
  )
  n399 = MaterialFunctionCall(Float2=NamedRerouteUsage()) {materialFunction=&BreakOutFloat2Components}
  LinearInterpolate(
   A=LinearInterpolate(A=@n398, B=ScalarParameter() {group=TopSide_V02_Layer-03_Uvs, parameterName=TopS_03_Side_Rotation_Backpack}, Alpha=@n399.R)
   B=ScalarParameter() {group=TopSide_V02_Layer-03_Uvs, parameterName=TopS_03_Side_Rotation_Pickaxe}
   Alpha=@n399.G
  )
  n400 = VectorParameter() {group=TopSide_V02_Layer-01_Uvs, parameterName="TopS_01_Rotation_Weapon\\Pickaxe", defaultValue=<rgba 0 0 -0.25 1>}
  n401 = VectorParameter() {group=TopSide_V02_Layer-01_Uvs, parameterName=TopS_01_Rotation_Vehicle, defaultValue=<rgba 0 0 0 1>}
  n402 = VectorParameter() {group=TopSide_V02_Layer-01_Uvs, parameterName=TopS_01_Rotation_Vehicle_Side, defaultValue=<rgba 0 0 0 1>}
  n403 = MaterialFunctionCall(Float4=NamedRerouteUsage()) {materialFunction=&BreakOutFloat4Components}
  n404 = VectorParameter() {group=TopSide_V02_Layer-01_Uvs, parameterName="TopS_01_Rotation_Mech\\Character\\Backpack", defaultValue=<rgba 0.25 0.25 0 1>}
  n405 = VectorParameter() {group=TopSide_V02_Layer-01_Uvs, parameterName="TopS_01_Rotation_Mech A 1=On", defaultValue=<rgba 0.25 0.25 0 0>}
  n406 = LinearInterpolate(
   A=LinearInterpolate(A=@n400.RGB, B=LinearInterpolate(A=@n401.RGB, B=@n402.RGB, Alpha=@n403.G), Alpha=@n403.R)
   B=LinearInterpolate(A=LinearInterpolate(A=@n404.RGB, B=@n405.RGB, Alpha=@n405.RGB), B=@n404.RGB, Alpha=@n403.A)
   Alpha=@n403.B
  )
  n407 = MaterialFunctionCall(Float2=NamedRerouteUsage()) {materialFunction=&BreakOutFloat2Components}
  LinearInterpolate(
   A=LinearInterpolate(A=@n406, B=@n404.RGB, Alpha=@n407.R)
   B=VectorParameter() {group=TopSide_V02_Layer-01_Uvs, parameterName=TopS_01_Rotation_Pickaxe, defaultValue=<rgba 0 0 0 1>}.RGB
   Alpha=@n407.G
  )
  n408 = ScalarParameter() {group=TopSide_V02_Layer-03_Uvs, parameterName="TopS_03_Side_Rotation_Mech\\Character\\Backpack"}
  n409 = MaterialFunctionCall(Float4=NamedRerouteUsage()) {materialFunction=&BreakOutFloat4Components}
  LinearInterpolate(A=@n408, B=ScalarParameter() {group=TopSide_V02_Layer-03_Uvs, parameterName=TopS_03_Side_Rotation_Character}, Alpha=@n409.A)
  n410 = ScalarParameter() {group=TopSide_V02_Layer-03_Uvs, parameterName=TopS_03_Side_Rotation_Vehicle}
  n411 = ScalarParameter() {group=TopSide_V02_Layer-03_Uvs, parameterName=TopS_03_Side_Rotation_Vehicle_Side}
  n412 = LinearInterpolate(
   A=LinearInterpolate(
    A=ScalarParameter() {group=TopSide_V02_Layer-03_Uvs, parameterName="TopS_03_Side_Rotation_Weapon\\Pickaxe"}
    B=LinearInterpolate(A=@n410, B=@n411, Alpha=@n409.G)
    Alpha=@n409.R
   )
   B=@n408
   Alpha=@n409.B
  )
  n413 = MaterialFunctionCall(Float2=NamedRerouteUsage()) {materialFunction=&BreakOutFloat2Components}
  LinearInterpolate(
   A=LinearInterpolate(A=@n412, B=ScalarParameter() {group=TopSide_V02_Layer-03_Uvs, parameterName=TopS_03_Side_Rotation_Backpack}, Alpha=@n413.R)
   B=ScalarParameter() {group=TopSide_V02_Layer-03_Uvs, parameterName=TopS_03_Side_Rotation_Pickaxe}
   Alpha=@n413.G
  )
  n414 = ScalarParameter() {group=TopSide_V02_Layer-01_Uvs, parameterName="TopS_01_Side_Rotation_Mech\\Character\\Backpack"}
  n415 = MaterialFunctionCall(Float4=NamedRerouteUsage()) {materialFunction=&BreakOutFloat4Components}
  LinearInterpolate(A=@n414, B=ScalarParameter() {group=TopSide_V02_Layer-01_Uvs, parameterName=TopS_01_Side_Rotation_Character}, Alpha=@n415.A)
  n416 = ScalarParameter() {defaultValue=0.5, group=TopSide_V02_Layer Emissive_03, parameterName="TopS_Em03_Rotation_Mech\\Character\\Backpack"}
  n417 = MaterialFunctionCall(Float4=NamedRerouteUsage()) {materialFunction=&BreakOutFloat4Components}
  LinearInterpolate(
   A=@n416
   B=ScalarParameter() {defaultValue=0.5, group=TopSide_V02_Layer Emissive_03, parameterName=TopS_Em03_Rotation_Character}
   Alpha=@n417.A
  )
  n418 = ScalarParameter() {defaultValue=0.5, group=TopSide_V02_Layer Emissive_03, parameterName=TopS_Em03_Rotation_Vehicle}
  n419 = ScalarParameter() {defaultValue=0.5, group=TopSide_V02_Layer Emissive_03, parameterName=TopS_Em03_Rotation_Vehicle_Side}
  n420 = LinearInterpolate(
   A=LinearInterpolate(
    A=ScalarParameter() {defaultValue=0.5, group=TopSide_V02_Layer Emissive_03, parameterName="TopS_Em03_Rotation_Weapon\\Pickaxe"}
    B=LinearInterpolate(A=@n418, B=@n419, Alpha=@n417.G)
    Alpha=@n417.R
   )
   B=@n416
   Alpha=@n417.B
  )
  n421 = MaterialFunctionCall(Float2=NamedRerouteUsage()) {materialFunction=&BreakOutFloat2Components}
  LinearInterpolate(
   A=LinearInterpolate(
    A=@n420
    B=ScalarParameter() {defaultValue=0.5, group=TopSide_V02_Layer Emissive_03, parameterName=TopS_Em03_Rotation_Backpack}
    Alpha=@n421.R
   )
   B=ScalarParameter() {defaultValue=0.5, group=TopSide_V02_Layer Emissive_03, parameterName=TopS_Em03_Rotation_Pickaxe}
   Alpha=@n421.G
  )
  n422 = ScalarParameter() {group=TopSide_V02_Layer-01_Uvs, parameterName=TopS_01_Side_Rotation_Vehicle}
  n423 = ScalarParameter() {group=TopSide_V02_Layer-01_Uvs, parameterName=TopS_01_Side_Rotation_Vehicle_Side}
  n424 = LinearInterpolate(
   A=LinearInterpolate(
    A=ScalarParameter() {group=TopSide_V02_Layer-01_Uvs, parameterName="TopS_01_Side_Rotation_Weapon\\Pickaxe"}
    B=LinearInterpolate(A=@n422, B=@n423, Alpha=@n415.G)
    Alpha=@n415.R
   )
   B=@n414
   Alpha=@n415.B
  )
  n425 = MaterialFunctionCall(Float2=NamedRerouteUsage()) {materialFunction=&BreakOutFloat2Components}
  LinearInterpolate(
   A=LinearInterpolate(A=@n424, B=ScalarParameter() {group=TopSide_V02_Layer-01_Uvs, parameterName=TopS_01_Side_Rotation_Backpack}, Alpha=@n425.R)
   B=ScalarParameter() {group=TopSide_V02_Layer-01_Uvs, parameterName=TopS_01_Side_Rotation_Pickaxe}
   Alpha=@n425.G
  )
  n426 = VectorParameter() {group=TopSide_V02_Layer-04_Uvs, parameterName="TopS_04_Rotation_Mech\\Character\\Backpack", defaultValue=<rgba 0.25 0.25 0 1>}
  n427 = VectorParameter() {group=TopSide_V02_Layer-04_Uvs, parameterName=TopS_04_Rotation_Character, defaultValue=<rgba 0.25 0.25 0 1>}
  n428 = MaterialFunctionCall(Float4=NamedRerouteUsage()) {materialFunction=&BreakOutFloat4Components}
  LinearInterpolate(A=@n426.RGB, B=@n427.RGB, Alpha=@n428.A)
  n429 = VectorParameter() {group=TopSide_V02_Layer-04_Uvs, parameterName="TopS_04_Rotation_Weapon\\Pickaxe", defaultValue=<rgba 0 0 -0.25 1>}
  n430 = VectorParameter() {group=TopSide_V02_Layer-04_Uvs, parameterName=TopS_04_Rotation_Vehicle, defaultValue=<rgba 0 0 0 1>}
  n431 = VectorParameter() {group=TopSide_V02_Layer-04_Uvs, parameterName=TopS_04_Rotation_Vehicle_Side, defaultValue=<rgba 0 0 0 1>}
  n432 = LinearInterpolate(A=LinearInterpolate(A=@n429.RGB, B=LinearInterpolate(A=@n430.RGB, B=@n431.RGB, Alpha=@n428.G), Alpha=@n428.R), B=@n426.RGB, Alpha=@n428.B)
  n433 = MaterialFunctionCall(Float2=NamedRerouteUsage()) {materialFunction=&BreakOutFloat2Components}
  LinearInterpolate(
   A=LinearInterpolate(
    A=@n432
    B=VectorParameter() {group=TopSide_V02_Layer-04_Uvs, parameterName=TopS_04_Rotation_Backpack, defaultValue=<rgba 0 0 0 1>}.RGB
    Alpha=@n433.R
   )
   B=VectorParameter() {group=TopSide_V02_Layer-04_Uvs, parameterName=TopS_04_Rotation_Pickaxe, defaultValue=<rgba 0 0 0 1>}.RGB
   Alpha=@n433.G
  )
  n434 = ScalarParameter() {group=TopSide_V02_Layer-04_Uvs | Warp, parameterName="TopS_04_Warp_Rotation_Mech\\Character\\Backpack", sortPriority=300}
  n435 = MaterialFunctionCall(Float4=NamedRerouteUsage()) {materialFunction=&BreakOutFloat4Components}
  LinearInterpolate(
   A=@n434
   B=ScalarParameter() {group=TopSide_V02_Layer-04_Uvs | Warp, parameterName=TopS_04_Warp_Rotation_Character, sortPriority=300}
   Alpha=@n435.A
  )
  n436 = ScalarParameter() {group=TopSide_V02_Layer-04_Uvs | Warp, parameterName=TopS_04_Warp_Rotation_Vehicle, sortPriority=300}
  n437 = ScalarParameter() {group=TopSide_V02_Layer-04_Uvs | Warp, parameterName=TopS_04_Warp_Rotation_Vehicle_Side, sortPriority=300}
  n438 = LinearInterpolate(
   A=LinearInterpolate(
    A=ScalarParameter() {group=TopSide_V02_Layer-04_Uvs | Warp, parameterName="TopS_04_Warp_Rotation_Weapon\\Pickaxe", sortPriority=300}
    B=LinearInterpolate(A=@n436, B=@n437, Alpha=@n435.G)
    Alpha=@n435.R
   )
   B=@n434
   Alpha=@n435.B
  )
  n439 = MaterialFunctionCall(Float2=NamedRerouteUsage()) {materialFunction=&BreakOutFloat2Components}
  LinearInterpolate(
   A=LinearInterpolate(
    A=@n438
    B=ScalarParameter() {group=TopSide_V02_Layer-04_Uvs | Warp, parameterName=TopS_04_Warp_Rotation_Backpack, sortPriority=300}
    Alpha=@n439.R
   )
   B=ScalarParameter() {group=TopSide_V02_Layer-04_Uvs | Warp, parameterName=TopS_04_Warp_Rotation_Pickaxe, sortPriority=300}
   Alpha=@n439.G
  )
  n440 = ScalarParameter() {group=TopSide_V02_Layer-04_Uvs, parameterName="TopS_04_Side_Rotation_Mech\\Character\\Backpack"}
  n441 = MaterialFunctionCall(Float4=NamedRerouteUsage()) {materialFunction=&BreakOutFloat4Components}
  LinearInterpolate(A=@n440, B=ScalarParameter() {group=TopSide_V02_Layer-04_Uvs, parameterName=TopS_04_Side_Rotation_Character}, Alpha=@n441.A)
  n442 = ScalarParameter() {group=TopSide_V02_Layer-04_Uvs, parameterName=TopS_04_Side_Rotation_Vehicle}
  n443 = ScalarParameter() {group=TopSide_V02_Layer-04_Uvs, parameterName=TopS_04_Side_Rotation_Vehicle_Side}
  n444 = LinearInterpolate(
   A=LinearInterpolate(
    A=ScalarParameter() {group=TopSide_V02_Layer-04_Uvs, parameterName="TopS_04_Side_Rotation_Weapon\\Pickaxe"}
    B=LinearInterpolate(A=@n442, B=@n443, Alpha=@n441.G)
    Alpha=@n441.R
   )
   B=@n440
   Alpha=@n441.B
  )
  n445 = MaterialFunctionCall(Float2=NamedRerouteUsage()) {materialFunction=&BreakOutFloat2Components}
  LinearInterpolate(
   A=LinearInterpolate(A=@n444, B=ScalarParameter() {group=TopSide_V02_Layer-04_Uvs, parameterName=TopS_04_Side_Rotation_Backpack}, Alpha=@n445.R)
   B=ScalarParameter() {group=TopSide_V02_Layer-04_Uvs, parameterName=TopS_04_Side_Rotation_Pickaxe}
   Alpha=@n445.G
  )
  n446 = ScalarParameter() {group=TopSide_V02_Layer-04_Uvs, parameterName="TopS_04_Side_Rotation_Mech\\Character\\Backpack"}
  n447 = MaterialFunctionCall(Float4=NamedRerouteUsage()) {materialFunction=&BreakOutFloat4Components}
  LinearInterpolate(A=@n446, B=ScalarParameter() {group=TopSide_V02_Layer-04_Uvs, parameterName=TopS_04_Side_Rotation_Character}, Alpha=@n447.A)
  n448 = ScalarParameter() {group=TopSide_V02_Layer-04_Uvs, parameterName=TopS_04_Side_Rotation_Vehicle}
  n449 = ScalarParameter() {group=TopSide_V02_Layer-04_Uvs, parameterName=TopS_04_Side_Rotation_Vehicle_Side}
  n450 = LinearInterpolate(
   A=LinearInterpolate(
    A=ScalarParameter() {group=TopSide_V02_Layer-04_Uvs, parameterName="TopS_04_Side_Rotation_Weapon\\Pickaxe"}
    B=LinearInterpolate(A=@n448, B=@n449, Alpha=@n447.G)
    Alpha=@n447.R
   )
   B=@n446
   Alpha=@n447.B
  )
  n451 = MaterialFunctionCall(Float2=NamedRerouteUsage()) {materialFunction=&BreakOutFloat2Components}
  LinearInterpolate(
   A=LinearInterpolate(A=@n450, B=ScalarParameter() {group=TopSide_V02_Layer-04_Uvs, parameterName=TopS_04_Side_Rotation_Backpack}, Alpha=@n451.R)
   B=ScalarParameter() {group=TopSide_V02_Layer-04_Uvs, parameterName=TopS_04_Side_Rotation_Pickaxe}
   Alpha=@n451.G
  )
  n452 = ScalarParameter() {defaultValue=0.5, group="TopSide_V02_Layer 01 Blend  02", parameterName="TopS_Blend_Rotation_Mech\\Character\\Backpack"}
  n453 = MaterialFunctionCall(Float4=NamedRerouteUsage()) {materialFunction=&BreakOutFloat4Components}
  LinearInterpolate(
   A=@n452
   B=ScalarParameter() {defaultValue=0.5, group="TopSide_V02_Layer 01 Blend  02", parameterName=TopS_Blend_Rotation_Character}
   Alpha=@n453.A
  )
  n454 = ScalarParameter() {defaultValue=0.5, group="TopSide_V02_Layer 01 Blend  02", parameterName=TopS_Blend_Rotation_Vehicle}
  n455 = ScalarParameter() {defaultValue=0.5, group="TopSide_V02_Layer 01 Blend  02", parameterName=TopS_Blend_Rotation_Vehicle_Side}
  n456 = LinearInterpolate(
   A=LinearInterpolate(
    A=ScalarParameter() {defaultValue=0.5, group="TopSide_V02_Layer 01 Blend  02", parameterName="TopS_Blend_Rotation_Weapon\\Pickaxe"}
    B=LinearInterpolate(A=@n454, B=@n455, Alpha=@n453.G)
    Alpha=@n453.R
   )
   B=@n452
   Alpha=@n453.B
  )
  n457 = MaterialFunctionCall(Float2=NamedRerouteUsage()) {materialFunction=&BreakOutFloat2Components}
  LinearInterpolate(
   A=LinearInterpolate(
    A=@n456
    B=ScalarParameter() {defaultValue=0.5, group="TopSide_V02_Layer 01 Blend  02", parameterName=TopS_Blend_Rotation_Backpack}
    Alpha=@n457.R
   )
   B=ScalarParameter() {defaultValue=0.5, group="TopSide_V02_Layer 01 Blend  02", parameterName=TopS_Blend_Rotation_Pickaxe}
   Alpha=@n457.G
  )
  n458 = ScalarParameter() {defaultValue=1, group="TopSide_V02_Layer 01 Blend  02", parameterName="TopS Grad_Falloff_Mech\\Character\\Backpack"}
  n459 = MaterialFunctionCall(Float4=NamedRerouteUsage()) {materialFunction=&BreakOutFloat4Components}
  LinearInterpolate(
   A=@n458
   B=ScalarParameter() {defaultValue=1, group="TopSide_V02_Layer 01 Blend  02", parameterName=TopS Grad_Falloff_Character}
   Alpha=@n459.A
  )
  n460 = ScalarParameter() {defaultValue=1, group="TopSide_V02_Layer 01 Blend  02", parameterName=TopS Grad_Falloff_Vehicle}
  n461 = ScalarParameter() {defaultValue=1, group="TopSide_V02_Layer 01 Blend  02", parameterName=TopS Grad_Falloff_Vehicle_Side}
  n462 = LinearInterpolate(
   A=LinearInterpolate(
    A=ScalarParameter() {defaultValue=1, group="TopSide_V02_Layer 01 Blend  02", parameterName="TopS Grad_Falloff_Weapon\\Pickaxe"}
    B=LinearInterpolate(A=@n460, B=@n461, Alpha=@n459.G)
    Alpha=@n459.R
   )
   B=@n458
   Alpha=@n459.B
  )
  n463 = MaterialFunctionCall(Float2=NamedRerouteUsage()) {materialFunction=&BreakOutFloat2Components}
  LinearInterpolate(
   A=LinearInterpolate(
    A=@n462
    B=ScalarParameter() {defaultValue=1, group="TopSide_V02_Layer 01 Blend  02", parameterName=TopS Grad_Falloff_Backpack}
    Alpha=@n463.R
   )
   B=ScalarParameter() {defaultValue=1, group="TopSide_V02_Layer 01 Blend  02", parameterName=TopS Grad_Falloff_Pickaxe}
   Alpha=@n463.G
  )
  n464 = ScalarParameter() {group=TopSide_V02_Layer-01_Uvs, parameterName="TopS_01_Side_Rotation_Mech\\Character\\Backpack"}
  n465 = MaterialFunctionCall(Float4=NamedRerouteUsage()) {materialFunction=&BreakOutFloat4Components}
  LinearInterpolate(A=@n464, B=ScalarParameter() {group=TopSide_V02_Layer-01_Uvs, parameterName=TopS_01_Side_Rotation_Character}, Alpha=@n465.A)
  n466 = ScalarParameter() {group=TopSide_V02_Layer-01_Uvs, parameterName=TopS_01_Side_Rotation_Vehicle}
  n467 = ScalarParameter() {group=TopSide_V02_Layer-01_Uvs, parameterName=TopS_01_Side_Rotation_Vehicle_Side}
  n468 = LinearInterpolate(
   A=LinearInterpolate(
    A=ScalarParameter() {group=TopSide_V02_Layer-01_Uvs, parameterName="TopS_01_Side_Rotation_Weapon\\Pickaxe"}
    B=LinearInterpolate(A=@n466, B=@n467, Alpha=@n465.G)
    Alpha=@n465.R
   )
   B=@n464
   Alpha=@n465.B
  )
  n469 = MaterialFunctionCall(Float2=NamedRerouteUsage()) {materialFunction=&BreakOutFloat2Components}
  LinearInterpolate(
   A=LinearInterpolate(A=@n468, B=ScalarParameter() {group=TopSide_V02_Layer-01_Uvs, parameterName=TopS_01_Side_Rotation_Backpack}, Alpha=@n469.R)
   B=ScalarParameter() {group=TopSide_V02_Layer-01_Uvs, parameterName=TopS_01_Side_Rotation_Pickaxe}
   Alpha=@n469.G
  )
  LinearInterpolate(B=@n178, Alpha=ScalarParameter() {group=TopSide_V02_Layer Emissive_01, parameterName="TopS Emissive_01_ColorCuve--->Use-UVS   1=On"})
  MaterialFunctionCall(In=@n264.Lerp Alpha No Contrast, Contrast=@n263) {materialFunction=&CheapContrast}
  n470 = ScalarParameter() {group=00-FrontEnd-Overrides, parameterName=00-FE-TODCC-Override}
  n471 = Add(
   A=CollectionParameter() {collection=&FortniteMaterialParameters, parameterName=TimeOfDay_0-1}
   B=@n470
  )
  n472 = CurveAtlasRowParameter(CurveTime=@n471) {
   atlas=/Game/Weapons/WeaponSkins/Wraps/Textures/CruveAtlas_Wraps_01
   curve=/Game/Weapons/WeaponSkins/Wraps/Textures/0_ColorCurves/Curve_TimeOfDay
   defaultValue=182
   group=TopSide_V02_Layer Emissive_00_TOD
   parameterName=TopS Emissivs_TOD_CC
  }
  MaterialFunctionCall!(
   Index=ScalarParameter() {defaultValue=182, group=TopSide_V02_Layer Emissive_00_TOD, parameterName="TopS Emissivs_TOD_CC  Index"}
   CCinput=@n472
   CCinput_Alpha=@n472
   UV=@n471
  ) {materialFunction=&/MF_Global_CC_Tex}
  n473 = VectorParameter() {group="TopSide_V02_Layer 01 Blend  02", parameterName=TopS_Blend_Offset__Master, defaultValue=<rgba 0 0 0 1>}
  Multiply(A=Time(), B=@n473.RGB)
  Multiply(A=@n81.R)
  Multiply(A=MaterialFunctionCall(Value_Morning=Constant() {r=0.5}, Value_Day=Constant() {r=1}, Value_Evening=Constant() {r=0.5}, Value_Night=Constant() {r=0.25}) {materialFunction=&Scalar_Time_of_Day}.Result)
  Multiply()
  n474 = Constant()
  n475 = NamedRerouteUsage()
  n476 = StaticBoolParameter() {group=TopSide_V02_Layer Emissive_02, parameterName="TopS Emissive_03_Grad-Offset------>SafeZoneBack"}
  n477 = StaticBoolParameter() {group=TopSide_V02_Layer Emissive_02, parameterName="TopS Emissive_03_Grad-Offset------>SafeZoneMid"}
  n478 = StaticSwitch(True=NamedRerouteUsage(), False=StaticSwitch(True=NamedRerouteUsage(), False=@n26.RGBA, Value=@n476), Value=@n477)
  n479 = StaticBoolParameter() {group=TopSide_V02_Layer Emissive_02, parameterName="TopS Emissive_03_Grad-Offset------>SafeZoneFront"}
  n480 = MaterialFunctionCall(
   Normal_Map=NamedRerouteUsage()
   Caculated_Bounds=NamedRerouteUsage()
   Pre_Skinned_Normals=NamedRerouteUsage()
   is_State=NamedRerouteUsage()
   UV_Rotation=@n474
   Uvs_Offset=ComponentMask(None=StaticSwitch(True=@n475, False=@n478, Value=@n479)) {b=true, g=true, r=true}
   Uvs_Offset_Pivot=@n474
   For_Safe_Zones=StaticBool() {value=true}
  ) {materialFunction=&MF_Projection_Side_Normal_V2}
  n481 = Constant() {r=2}
  n482 = Multiply(
   A=MaterialFunctionCall(UV Coord=@n480.UVs, LenghtScale=@n481, RadialScale=@n481) {materialFunction=&MF_PolarCoordinates}.Result
   B=Constant() {r=0.9}
  )
  Multiply(A=TextureSample!(UVs=Saturate(None=@n482), Tex, Apply View MipBias) {samplerType=SAMPLERTYPE_Masks, texture=&T_BlackMask}.R)
  n483 = FunctionInput(Preview=@n27.G) {bUsePreviewValueAsDefault=true, inputName=Mask Crevise, inputType=FunctionInput_Scalar, sortPriority=24, previewValue=<1 0 0 1>}
  NamedRerouteDeclaration(None=@n483) {name=Msk_Curvature, nodeColor=<rgba 0 1 0.7422 1>}
  n484 = SetMaterialAttributes!(
   MaterialAttributes
   Normal=TextureSample() {samplerType=SAMPLERTYPE_Normal, texture=/Game/Weapons/FORT_Rifles/T_SCAR_N}.RGB
   BaseColor=Constant3Vector() {constant=<rgba 1 0 0 0>}
  ) {attributeSetTypes [#1, #2]}
  n485 = GetMaterialAttributes(None=FunctionInput(Preview=@n484) {inputName=Root_Material, inputType=FunctionInput_MaterialAttributes}) {attributeGetTypes [
    #2
    #3
    #1
    #4
   ]}
  NamedRerouteDeclaration(None=@n485.MaterialAttributes) {name=Root_Material, nodeColor=<rgba 0.7812 1 0 1>}
  n486 = StaticBoolParameter() {group=TopSide_V02_Layer 02, parameterName="TopS_02 TopSide->Side projection"}
  NamedRerouteDeclaration(None=@n486) {name=02_Side_Projection_state, nodeColor=<rgba 1 0 0.8437 1>}
  n487 = NamedRerouteUsage()
  n488 = NamedRerouteUsage()
  n489 = TextureObjectParameter() {
   group=TopSide_V02_Layer 03
   parameterName=TopS_03__Tex_Mask
   samplerType=SAMPLERTYPE_Masks
   texture=&T_BlackMask
  }
  NamedRerouteDeclaration(None=StaticSwitchParameter(
    True=StaticSwitchParameter(True=@n487, False=@n488) {group=TopSide_V02_Layer 03, parameterName=TopS_03__Tex_Mask Share_00B_Msk_Tex}
    False=@n489
   ) {group=TopSide_V02_Layer 03, parameterName=TopS_03__Tex_Mask Share_00A_Msk_Tex}) {name=Tex_TS03_Mask, nodeColor=<rgba 1 0.8906 0 1>}
  NamedRerouteDeclaration(None=StaticBoolParameter() {group=TopSide_V02_Layer 01, parameterName="TopS_01 TopSide->Side projection->Left_Right"}) {name="01_Left\\Right Stat", nodeColor=<rgba 0 0.2734 1 1>}
  n490 = StaticSwitchParameter(True=@n487, False=@n488) {group=TopSide_V02_Layer-03_Uvs | Warp, parameterName=TopS_03__Tex_Mask Warp Share_00B_Msk_Tex, sortPriority=249}
  n491 = TextureObjectParameter() {
   group=TopSide_V02_Layer-03_Uvs | Warp
   parameterName=TopS_03_Tex_Warp
   samplerType=SAMPLERTYPE_Masks
   sortPriority=249
   texture=&T_BlackMask
  }
  NamedRerouteDeclaration(None=StaticSwitchParameter(
    True=StaticSwitchParameter(True=@n490, False=@n489) {group=TopSide_V02_Layer-03_Uvs | Warp, parameterName=TopS_03__Tex_Mask Warp Share_00A_Msk_Tex, sortPriority=249}
    False=@n491
   ) {group=TopSide_V02_Layer-03_Uvs | Warp, parameterName=TopS_03__Share_Mask_03_Tex, sortPriority=249}) {name=Tex_03_Warp, nodeColor=<rgba 1 0.8906 0 1>}
  n492 = NamedRerouteUsage()
  n493 = NamedRerouteUsage()
  n494 = NamedRerouteUsage()
  n495 = Multiply(A=TextureCoordinate(), B=ComponentMask(None=@n494) {g=true, r=true})
  n496 = VectorParameter() {group=TopSide_V02_Layer-03_Uvs | Warp, parameterName=TopS_03_Warp_UV_Panning, sortPriority=250, defaultValue=<rgba 0 0 0 1>}
  n497 = Multiply(A=@n496.RGB, B=Time())
  n498 = ComponentMask(None=@n497) {g=true, r=true}
  n499 = VectorParameter() {group=TopSide_V02_Layer-03_Uvs | Warp, parameterName=TopS_03_Warp_UV_Offset_Pos, sortPriority=250, defaultValue=<rgba 0 0 0 1>}
  n500 = MaterialFunctionCall(
   Caculated_Bounds=NamedRerouteUsage()
   Pre_Skinned_Normals=NamedRerouteUsage()
   is_State=NamedRerouteUsage()
   UV_Rotation=NamedRerouteUsage()
   Uvs_Offset=@n497
   Uvs_TileRate=@n494
   Uvs_Offset_Pivot=@n499.RGB
  ) {materialFunction=&MF_Projection_Side_Normal_V2}
  n501 = TextureSample(
   UVs=StaticSwitchParameter(
    True=Add(A=@n495, B=@n498)
    False=StaticSwitchParameter(True=@n500.UVs No Rotation, False=@n500.UVs No Rotation) {group=TopSide_V02_Layer-03_Uvs | Warp, parameterName=TopS_03 Warp_No_Rotation, sortPriority=249}
   ) {group=TopSide_V02_Layer-03_Uvs | Warp, parameterName=TopS_03 Warp_ObjectUVs, sortPriority=249}
   Tex=NamedRerouteUsage()
  )
  n502 = StaticSwitchParameter(
   True=NamedRerouteUsage()
   False=StaticSwitchParameter(
    True=NamedRerouteUsage()
    False=StaticSwitchParameter(True=@n493, False=@n501.RGBA) {group=TopSide_V02_Layer-03_Uvs | Warp, parameterName="TopS_03 Warp--->Layer 04 MSk", sortPriority=250}
   ) {group=TopSide_V02_Layer-03_Uvs | Warp, parameterName="TopS_03 Warp--->Layer 02 MSk", sortPriority=250}
  ) {group=TopSide_V02_Layer-03_Uvs | Warp, parameterName="TopS_03 Warp--->Layer 01 MSk", sortPriority=250}
  n503 = Add(
   A=ChannelMaskParameter(None=@n502) {group=TopSide_V02_Layer-03_Uvs | Warp, parameterName=TopS_03_Warp_UV Mask_Channel, sortPriority=250}.R
   B=ScalarParameter() {group=TopSide_V02_Layer-03_Uvs | Warp, parameterName=TopS_03_Warp_UV_Offset, sortPriority=250}
  )
  n504 = Multiply(A=NamedRerouteUsage(), B=ScalarParameter() {group=TopSide_V02_Layer-03_Uvs | Warp, parameterName=TopS_03_Warp_UV_Amount, sortPriority=250})
  n505 = NamedRerouteUsage()
  n506 = VectorParameter() {
   group=TopSide_V02_Layer-03_Uvs | Warp
   parameterName=TopS_03_Warp_Curvature_Amount
   sortPriority=250
   channelNames {a="", b="", g=Pits, r=Peaks}
  }
  n507 = NamedRerouteUsage()
  n508 = Add(A=Add(A=@n492, B=Multiply(A=@n503, B=@n504)), B=Add(A=Multiply(A=@n505, B=@n506.Peaks), B=Multiply(A=@n507, B=@n506.Pits)))
  n509 = StaticSwitchParameter(True=NamedRerouteUsage(), False=@n508) {group=TopSide_V02_Layer-03_Uvs | Warp, parameterName="TopS_03 Warp_Uvs------>Tops_02_Warp", sortPriority=249}
  n510 = NamedRerouteUsage()
  NamedRerouteDeclaration(None=StaticSwitchParameter(True=StaticSwitch(True=@n509, False=@n508, Value=@n510), False=@n492) {group=TopSide_V02_Layer-03_Uvs | Warp, parameterName=TopS_03 Warp_Uvs-----------, sortPriority=0}) {name=Caclcated_Bounds_03, nodeColor=<rgba 0.3594 0 1 1>}
  n511 = MaterialFunctionCall(Color=Abs(None=NamedRerouteUsage())) {materialFunction=&SplitComponents}
  n512 = MaterialFunctionCall(Float4=NamedRerouteUsage()) {materialFunction=&BreakOutFloat4Components}
  n513 = LinearInterpolate(A=@n511.B, B=@n511.B, Alpha=@n512.G)
  n514 = NamedRerouteUsage()
  n515 = VectorParameter() {group=TopSide_V02, parameterName=TopS Blend_Curvature_Amount_, channelNames {a="", b="", g=Pits, r=Peaks}, defaultValue=<rgba 2 2 0 1>}
  n516 = NamedRerouteUsage()
  n517 = Add(
   A=LinearInterpolate(A=LinearInterpolate(A=@n511.R, B=@n513, Alpha=@n512.R), B=@n511.G, Alpha=@n512.B)
   B=Add(A=Multiply(A=@n514, B=@n515.Peaks), B=Multiply(A=@n516, B=@n515.Pits))
  )
  NamedRerouteDeclaration(None=MaterialFunctionCall(In=@n517, Contrast=NamedRerouteUsage()) {materialFunction=&CheapContrast}.Result) {name=Tri_Mask 03, nodeColor=<rgba 1 0.4688 0 1>}
  n518 = NamedRerouteUsage()
  n519 = Max(A=@n259, B=@n261)
  n520 = ScalarParameter() {
   defaultValue=1
   group=TopSide_V02_Layer Emissive_02 Reactive
   parameterName="TopS Emissive_02-Brightness-Reactive--->Gradent BloomFix  0=Active"
  }
  n521 = Saturate(None=Ceil(None=Min(A=@n519, B=@n520)))
  n522 = ScalarParameter() {group=TopSide_V02_Layer Emissive_02, parameterName=TopS Emissive_02_ColorCuve_Offset}
  n523 = Multiply(A=ScalarParameter() {group=TopSide_V02_Layer Emissive_02, parameterName=TopS Emissive_02_ColorCuve_Speed}, B=Time())
  n524 = StaticSwitch(
   True=NamedRerouteUsage()
   False=@n269
   Value=StaticBoolParameter() {desc=setting at bottom, group=TopSide_V02_Layer Emissive_02, parameterName="TopS Emissive_02_ColorCuve--->Use-UVS"}
  )
  n525 = ScalarParameter() {defaultValue=1, group=TopSide_V02_Layer Emissive_02, parameterName=TopS Emissive_02_ColorCuve_TileRate}
  n526 = Frac(None=Add(A=Add(A=@n522, B=@n523), B=Multiply(A=@n524, B=@n525)))
  n527 = CurveAtlasRowParameter(CurveTime=@n526) {
   atlas=&CruveAtlas_Wraps_01
   curve=/Game/Weapons/WeaponSkins/Wraps/Textures/0_ColorCurves/CC_Spectum
   defaultValue=98
   group=TopSide_V02_Layer Emissive_02
   parameterName=TopS Emissive_02_ColorCuve B
  }
  n528 = MaterialFunctionCall(
   Index=ScalarParameter() {defaultValue=1, group=TopSide_V02_Layer Emissive_02, parameterName=TopS Emissive_02_ColorCuve B Index}
   CCinput=@n527
   CCinput_Alpha=@n527
   UV=@n526
  ) {materialFunction=&MF_Global_CC_Tex}
  n529 = CurveAtlasRowParameter(CurveTime=@n526) {
   atlas=&CruveAtlas_Wraps_01
   curve=&CC_Spectum
   defaultValue=98
   group=TopSide_V02_Layer Emissive_02
   parameterName=TopS Emissive_02_ColorCuve
  }
  n530 = MaterialFunctionCall(
   Index=ScalarParameter() {defaultValue=1, group=TopSide_V02_Layer Emissive_02, parameterName="TopS Emissive_02_ColorCuve  Index"}
   CCinput=@n529
   CCinput_Alpha=@n529
   UV=@n526
  ) {materialFunction=&MF_Global_CC_Tex}
  n531 = VectorParameter() {group=TopSide_V02_Layer Emissive_02, parameterName=TopS Emissive_02_ColorCuve Tint, defaultValue=<rgba 1 1 1 1>}
  n532 = Multiply(A=@n530.3V, B=@n531.RGB)
  n533 = NamedRerouteUsage()
  n534 = StaticBoolParameter() {group=TopSide_V02_Layer Emissive_02, parameterName=TopS Emissive_02_Color TOD}
  n535 = StaticSwitch(True=LinearInterpolate(A=@n528.3V, B=@n532, Alpha=@n533), False=@n532, Value=@n534)
  n536 = ScalarParameter() {defaultValue=1, group=TopSide_V02_Layer Emissive_02, parameterName=TopS Emissive_02_ColorCuve-Brightness}
  n537 = StaticSwitch(True=LinearInterpolate(A=@n528.Alpha, B=@n530.Alpha, Alpha=@n533), False=@n530.Alpha, Value=@n534)
  n538 = LinearInterpolate(A=@n536, B=Multiply(A=@n537, B=@n536), Alpha=NamedRerouteUsage())
  n539 = Add(A=ScalarParameter() {group=TopSide_V02_Layer Emissive_02, parameterName="TopS Emissive_02_ColorCuve-CurveMask   0=On"}, B=@n269)
  n540 = VectorParameter() {group=TopSide_V02_Layer Emissive_02, parameterName=TopS Emissive_02_Color-B, defaultValue=<rgba 0 0 0 1>}
  n541 = VectorParameter() {group=TopSide_V02_Layer Emissive_02, parameterName=TopS Emissive_02_Color-, defaultValue=<rgba 0 0 0 1>}
  n542 = LinearInterpolate(A=@n540.RGB, B=@n541.RGB, Alpha=NamedRerouteUsage())
  n543 = StaticSwitchParameter(
   True=Multiply(A=Multiply(A=@n535, B=@n538), B=Saturate(None=@n539))
   False=Multiply(A=@n269, B=StaticSwitch(True=@n542, False=@n541.RGB, Value=@n534))
  ) {group=TopSide_V02_Layer Emissive_02, parameterName="TopS Emissive_02_Color ----->ColorCurve"}
  n544 = StaticSwitch(True=Multiply(A=@n543, B=NamedRerouteUsage()), False=@n543, Value=NamedRerouteUsage())
  n545 = StaticSwitch(
   True=Multiply(A=LinearInterpolate(A=@n518, Alpha=@n521), B=@n544)
   False=@n544
   Value=NamedRerouteDeclaration(None=StaticBoolParameter() {group=TopSide_V02_Layer Emissive_02 Reactive, parameterName=TopS Emissive_02-Brightness-Reactive--------------}) {name=Emissive_02_Reactive_bool_Brightness, nodeColor=<rgba 0.3594 0 1 1>}
  )
  n546 = NamedRerouteUsage()
  n547 = LinearInterpolate(
   A=@n546
   B=OneMinus(None=@n546)
   Alpha=ScalarParameter() {group=TopSide_V02_Layer Emissive_02, parameterName="TopS Emissive_02_Blend 01---> 02 Invert Msk  1=inverted", sortPriority=1}
  )
  n548 = StaticBoolParameter() {group=TopSide_V02_Layer Emissive_02, parameterName="TopS Emissive_02_Blend 01---> 02 Affects Mask", sortPriority=1}
  n549 = StaticSwitch(True=StaticSwitch(True=Multiply(A=@n545, B=@n547), False=@n545, Value=@n548), False=@n545, Value=NamedRerouteUsage())
  n550 = StaticSwitch(
   True=NamedRerouteUsage()
   False=NamedRerouteUsage()
   Value=StaticBoolParameter() {group=TopSide_V02_Layer Emissive_02 Reactive, parameterName="TopS Emissive_02-TOD---->TODB", sortPriority=0}
  )
  n551 = StaticBoolParameter() {group=TopSide_V02_Layer Emissive_02 Reactive, parameterName=TopS Emissive_02-TOD, sortPriority=0}
  n552 = StaticSwitch(True=LinearInterpolate(B=@n549, Alpha=@n550) {constB=0}, False=@n549, Value=@n551)
  n553 = VectorParameter() {
   group=TopSide_V02_Layer Emissive_02
   parameterName=TopS Emissive_02 LiquidCCMask
   channelNames {a="", b="", g=invert, r="0=Use"}
   defaultValue=<rgba 1 0 0 1>
  }
  n554 = ChannelMaskParameter(None=NamedRerouteUsage()) {group=TopSide_V02_Layer Emissive_02, maskChannel=Alpha, parameterName=TopS Emissive_02 LiquidCCMask Channel, defaultValue=<rgba 0 0 0 1>}
  n555 = Max(A=@n553."0=Use", B=LinearInterpolate(A=@n554.R, B=OneMinus(None=@n554.R), Alpha=@n553.invert))
  n556 = NamedRerouteUsage()
  NamedRerouteDeclaration(None=StaticSwitch(True=LinearInterpolate(B=@n552, Alpha=@n555), False=@n552, Value=@n556)) {name=Emissive_02, nodeColor=<rgba 1 0.8203 0 1>}
  n557 = OneMinus(None=@n555)
  n558 = ScalarParameter() {
   defaultValue=1
   desc=CC brightnes gets disabled
   group=TopSide_V02_Layer Emissive_02
   parameterName="TopS Emissive_02_ColorCuve=A-Effect_DarkenBase  0=ON"
  }
  n559 = NamedRerouteDeclaration(None=Saturate(None=Ceil(None=@n558))) {name=EmiS 02 CC A State, nodeColor=<rgba 0 0.8828 1 1>}
  n560 = ScalarParameter() {group=TopSide_V02_Layer Emissive_02 Reactive, parameterName=TopS Audio 02 Emissive_Brightness Darken_Base_Max, sortPriority=211}
  n561 = LinearInterpolate(
   A=ScalarParameter() {group=TopSide_V02_Layer Emissive_02 Reactive, parameterName=TopS Audio 02 Emissive_Brightness Darken_Bas-Min, sortPriority=211}
   B=@n560
   Alpha=NamedRerouteUsage()
  )
  n562 = LinearInterpolate(
   A=ScalarParameter() {group=TopSide_V02_Layer Emissive_02 Reactive, parameterName=TopS Emissive_02_Brightness_Darken_Base Start}
   B=ScalarParameter() {group=TopSide_V02_Layer Emissive_02 Reactive, parameterName=TopS Emissive_02_Brightness_Darken_Base End}
   Alpha=NamedRerouteUsage()
  )
  n563 = StaticSwitch(
   True=StaticSwitch(True=Max(A=@n561, B=@n562), False=@n562, Value=NamedRerouteUsage())
   False=ScalarParameter() {group=TopSide_V02_Layer Emissive_02, parameterName="TopS Emissive_02-Darken_Base    1=On"}
   Value=NamedRerouteUsage()
  )
  n564 = StaticSwitchParameter(True=Multiply(A=LinearInterpolate(A=@n537, B=@n269, Alpha=@n559), B=@n563), False=Multiply(A=@n269, B=@n563)) {group=TopSide_V02_Layer Emissive_02, parameterName="TopS Emissive_02_Color ----->ColorCurve"}
  n565 = StaticSwitch(True=Multiply(A=@n547, B=@n564), False=@n564, Value=@n548)
  n566 = NamedRerouteUsage()
  n567 = OneMinus(None=Saturate(None=StaticSwitch(True=@n565, False=@n564, Value=@n566)))
  n568 = StaticSwitch(True=StaticSwitch(True=Max(A=@n557, B=@n567), False=@n567, Value=@n556), False=Constant() {r=1}, Value=NamedRerouteUsage())
  n569 = LinearInterpolate(B=@n568, Alpha=@n550) {constA=1, constB=0}
  n570 = StaticBoolParameter() {group=TopSide_V02_Layer Emissive_02 Reactive, parameterName=TopS Emissive_02-TOD DoNotAffectDarkenBase, sortPriority=0}
  NamedRerouteDeclaration(None=StaticSwitch(True=StaticSwitch(True=@n568, False=@n569, Value=@n570), False=@n568, Value=@n551)) {name=Emissive_02_DarkenBAse, nodeColor=<rgba 1 0.8203 0 1>}
  n571 = VectorParameter() {group=TopSide_V02_Layer Emissive_02, parameterName=TopS_Em02_Scale_Weapon, defaultValue=<rgba 1 1 1 1>}
  n572 = Multiply(A=@n571.RGB, B=@n571.RGB)
  n573 = VectorParameter() {group=TopSide_V02_Layer Emissive_02, parameterName=TopS_Em02_Scale_Vehicle, defaultValue=<rgba 1 1 1 1>}
  n574 = VectorParameter() {group=TopSide_V02_Layer Emissive_02, parameterName=TopS_Em02_Scale_VehicleSide, defaultValue=<rgba 1 1 1 1>}
  n575 = MaterialFunctionCall(Float4=NamedRerouteUsage()) {materialFunction=&BreakOutFloat4Components}
  n576 = LinearInterpolate(A=Multiply(A=@n573.RGB, B=@n573.RGB), B=Multiply(A=@n574.RGB, B=@n574.RGB), Alpha=@n575.G)
  n577 = VectorParameter() {group=TopSide_V02_Layer Emissive_02, parameterName=TopS_Em02_Scale_Mech, defaultValue=<rgba 1 1 1 1>}
  n578 = Multiply(A=@n577.RGB, B=@n577.RGB)
  n579 = VectorParameter() {group=TopSide_V02_Layer Emissive_02, parameterName=TopS_Em02_Scale_Character, defaultValue=<rgba 1 1 1 1>}
  n580 = Multiply(A=@n579.RGB, B=@n579.RGB)
  n581 = VectorParameter() {group=TopSide_V02_Layer Emissive_02, parameterName=TopS_Em02_Scale_Backpack, defaultValue=<rgba 1 1 1 1>}
  n582 = MaterialFunctionCall(Float2=NamedRerouteUsage()) {materialFunction=&BreakOutFloat2Components}
  n583 = LinearInterpolate(
   A=LinearInterpolate(A=LinearInterpolate(A=@n572, B=@n576, Alpha=@n575.R), B=LinearInterpolate(A=@n578, B=@n580, Alpha=@n575.A), Alpha=@n575.B)
   B=Multiply(A=@n581.RGB, B=@n581.RGB)
   Alpha=@n582.R
  )
  n584 = VectorParameter() {group=TopSide_V02_Layer Emissive_02, parameterName=TopS_Em02_Scale_Pickaxe, defaultValue=<rgba 1 1 1 1>}
  n585 = Multiply(A=@n584.RGB, B=@n584.RGB)
  n586 = VectorParameter() {group=TopSide_V02_Layer Emissive_02, parameterName=TopS_Em02_Scale__Master, defaultValue=<rgba 1 1 1 1>}
  NamedRerouteDeclaration(None=Multiply(A=LinearInterpolate(A=@n583, B=@n585, Alpha=@n582.G), B=Multiply(A=@n586.RGB, B=@n586.RGB))) {name=Uvs_TS_Emis02_Scale, nodeColor=<rgba 0 0.625 1 1>}
  n587 = VectorParameter() {group=TopSide_V02_Layer Emissive_02, parameterName=TopS_Em02_Offset__Master, defaultValue=<rgba 0 0 0 1>}
  n588 = VectorParameter() {group=TopSide_V02_Layer Emissive_02, parameterName=TopS_Em02_Offset_Weapon, defaultValue=<rgba 0 0 0 1>}
  n589 = VectorParameter() {group=TopSide_V02_Layer Emissive_02, parameterName=TopS_Em02_Offset_Vehicle, defaultValue=<rgba 0 0 0 1>}
  n590 = VectorParameter() {group=TopSide_V02_Layer Emissive_02, parameterName=TopS_Em02_Offset_VehicleSide, defaultValue=<rgba 0 0 0 1>}
  n591 = MaterialFunctionCall(Float4=NamedRerouteUsage()) {materialFunction=&BreakOutFloat4Components}
  n592 = LinearInterpolate(A=@n589.RGB, B=@n590.RGB, Alpha=@n591.G)
  n593 = VectorParameter() {group=TopSide_V02_Layer Emissive_02, parameterName=TopS_Em02_Offset_Mech, defaultValue=<rgba 0 0 0 1>}
  n594 = VectorParameter() {group=TopSide_V02_Layer Emissive_02, parameterName=TopS_Em02_Offset_Character, defaultValue=<rgba 0 0 0 1>}
  n595 = VectorParameter() {group=TopSide_V02_Layer Emissive_02, parameterName=TopS_Em02_Offset_Backpack, defaultValue=<rgba 0 0 0 1>}
  n596 = MaterialFunctionCall(Float2=NamedRerouteUsage()) {materialFunction=&BreakOutFloat2Components}
  n597 = LinearInterpolate(
   A=LinearInterpolate(A=LinearInterpolate(A=@n588.RGB, B=@n592, Alpha=@n591.R), B=LinearInterpolate(A=@n593.RGB, B=@n594.RGB, Alpha=@n591.A), Alpha=@n591.B)
   B=@n595.RGB
   Alpha=@n596.R
  )
  n598 = VectorParameter() {group=TopSide_V02_Layer Emissive_02, parameterName=TopS_Em02_Offset_Pickaxe, defaultValue=<rgba 0 0 0 1>}
  NamedRerouteDeclaration(None=Add(A=@n587.RGB, B=LinearInterpolate(A=@n597, B=@n598.RGB, Alpha=@n596.G))) {name=Uvs_TS_Emis02_Offset, nodeColor=<rgba 0 0.625 1 1>}
  NamedRerouteDeclaration(None=MaterialFunctionCall(In=@n517, Contrast=NamedRerouteUsage()) {materialFunction=&CheapContrast}.Result) {name=Tri_Mask 02, nodeColor=<rgba 1 0.4688 0 1>}
  n599 = CollectionParameter() {collection=&FortniteMaterialParameters, parameterName=EventAudio_Spectrum}
  n600 = ChannelMaskParameter(None=@n599) {
   group="TopSide_V02_Layer 01 Blend  02 Reactive"
   maskChannel=Alpha
   parameterName=TopS_01_Blend_Audio Channel
   sortPriority=150
   defaultValue=<rgba 0 0 0 1>
  }
  n601 = CollectionParameter() {collection=&FortniteMaterialParameters, parameterName=Emote_AudioLow}
  n602 = ScalarParameter() {defaultValue=0.25, group="TopSide_V02_Layer 01 Blend  02 Reactive", parameterName=TopS_01_Blend_Audio Low_Remap, sortPriority=151}
  n603 = ScalarParameter() {defaultValue=1, group="TopSide_V02_Layer 01 Blend  02 Reactive", parameterName=TopS_01_Blend_Audio High_Remap, sortPriority=151}
  n604 = Saturate(None=MaterialFunctionCall(Input=Max(A=@n600.R, B=@n601), Input Low=@n602, Input High=@n603, Target Low=Constant(), Target High=Constant() {r=1}) {materialFunction=&RemapValueRange}.Result)
  n605 = CollectionParameter() {collection=&FortniteMaterialParameters, parameterName=Emote_AudioHigh}
  n606 = ScalarParameter() {defaultValue=1, group="TopSide_V02_Layer 01 Blend  02 Reactive", parameterName="TopS_01_Blend_Audio  Falloff", sortPriority=151}
  n607 = LinearInterpolate(
   A=ScalarParameter() {group="TopSide_V02_Layer 01 Blend  02 Reactive", parameterName="TopS_01_Blend_Audio  Min", sortPriority=149}
   B=ScalarParameter() {group="TopSide_V02_Layer 01 Blend  02 Reactive", parameterName=TopS_01_Blend_Audio Max, sortPriority=149}
   Alpha=Power(Base=Max(A=@n604, B=@n605), Exp=@n606)
  )
  n608 = Subtract(
   A=Time()
   B=Add(
    A=ScalarParameter() {group="TopSide_V02_Layer 01 Blend  02 Reactive", parameterName=TopS_01_Blend_Opacity---Delay}
    B=ScalarParameter() {defaultValue=-9999999, group="TopSide_V02_Layer 01 Blend  02 Reactive", parameterName=TopS_01_Blend_Opacity-TBP}
   )
  )
  n609 = ScalarParameter() {defaultValue=0.25, group="TopSide_V02_Layer 01 Blend  02 Reactive", parameterName=TopS_01_Blend_Opacity---Duration}
  n610 = Power(
   Base=Saturate(None=Divide(A=@n608, B=@n609))
   Exp=ScalarParameter() {defaultValue=0.25, group="TopSide_V02_Layer 01 Blend  02 Reactive", parameterName=TopS_01_Blend_Opacity---DFalloff}
  )
  n611 = LinearInterpolate(
   A=ScalarParameter() {defaultValue=1, group="TopSide_V02_Layer 01 Blend  02 Reactive", parameterName=TopS_01_Blend_Opacity---Start}
   B=ScalarParameter() {group="TopSide_V02_Layer 01 Blend  02 Reactive", parameterName=TopS_01_Blend_Opacity---End}
   Alpha=@n610
  )
  NamedRerouteDeclaration(None=StaticSwitchParameter(True=Max(A=@n607, B=@n611), False=@n611) {group="TopSide_V02_Layer 01 Blend  02 Reactive", parameterName=TopS_01_Blend_Audio, sortPriority=149}) {name=Reactive_Blend_Opacity, nodeColor=<rgba 1 0.1641 0 1>}
  n612 = ScalarParameter() {defaultValue=1, group=TopSide_V02_Layer Emissive_02, parameterName=TopS_Em02_Falloff_Vehicle}
  n613 = ScalarParameter() {defaultValue=1, group=TopSide_V02_Layer Emissive_02, parameterName=TopS_Em02_Falloff_Vehicle_Side}
  n614 = MaterialFunctionCall(Float4=NamedRerouteUsage()) {materialFunction=&BreakOutFloat4Components}
  n615 = LinearInterpolate(
   A=LinearInterpolate(
    A=ScalarParameter() {defaultValue=1, group=TopSide_V02_Layer Emissive_02, parameterName=TopS_Em02_Falloff_Weapon}
    B=LinearInterpolate(A=@n612, B=@n613, Alpha=@n614.G)
    Alpha=@n614.R
   )
   B=LinearInterpolate(
    A=ScalarParameter() {defaultValue=1, group=TopSide_V02_Layer Emissive_02, parameterName=TopS_Em02_Falloff_Mech}
    B=ScalarParameter() {defaultValue=1, group=TopSide_V02_Layer Emissive_02, parameterName=TopS_Em02_Falloff_Character}
    Alpha=@n614.A
   )
   Alpha=@n614.B
  )
  n616 = ScalarParameter() {defaultValue=1, group=TopSide_V02_Layer Emissive_02, parameterName=TopS_Em02_Falloff_Backpack}
  n617 = MaterialFunctionCall(Float2=NamedRerouteUsage()) {materialFunction=&BreakOutFloat2Components}
  NamedRerouteDeclaration(None=LinearInterpolate(
    A=LinearInterpolate(A=@n615, B=@n616, Alpha=@n617.R)
    B=ScalarParameter() {defaultValue=1, group=TopSide_V02_Layer Emissive_02, parameterName=TopS_Em02_Falloff_Pickaxe}
    Alpha=@n617.G
   )) {name=Uvs_TS_Emis_02_Grad_Falloff, nodeColor=<rgba 0 0.7188 1 1>}
  n618 = Time()
  n619 = VectorParameter() {group=TopSide_V02_Layer Emissive_02, parameterName=TopS_Em02_Speed_Vehicle, defaultValue=<rgba 0 0 0 1>}
  n620 = VectorParameter() {group=TopSide_V02_Layer Emissive_02, parameterName=TopS_Em02_Speed_VehicleSide, defaultValue=<rgba 0 0 0 1>}
  n621 = MaterialFunctionCall(Float4=NamedRerouteUsage()) {materialFunction=&BreakOutFloat4Components}
  n622 = VectorParameter() {group=TopSide_V02_Layer Emissive_02, parameterName=TopS_Em02_Speed_Character, defaultValue=<rgba 0 0 0 1>}
  n623 = LinearInterpolate(
   A=LinearInterpolate(
    A=VectorParameter() {group=TopSide_V02_Layer Emissive_02, parameterName=TopS_Em02_Speed_Weapon, defaultValue=<rgba 0 0 0 1>}.RGB
    B=LinearInterpolate(A=@n619.RGB, B=@n620.RGB, Alpha=@n621.G)
    Alpha=@n621.R
   )
   B=LinearInterpolate(
    A=VectorParameter() {group=TopSide_V02_Layer Emissive_02, parameterName=TopS_Em02_Speed_Mech, defaultValue=<rgba 0 0 0 1>}.RGB
    B=@n622.RGB
    Alpha=@n621.A
   )
   Alpha=@n621.B
  )
  n624 = VectorParameter() {group=TopSide_V02_Layer Emissive_02, parameterName=TopS_Em02_Speed_Backpack, defaultValue=<rgba 0 0 0 1>}
  n625 = MaterialFunctionCall(Float2=NamedRerouteUsage()) {materialFunction=&BreakOutFloat2Components}
  n626 = VectorParameter() {group=TopSide_V02_Layer Emissive_02, parameterName=TopS_Em02_Speed_Pickaxe, defaultValue=<rgba 0 0 0 1>}
  n627 = VectorParameter() {group=TopSide_V02_Layer Emissive_02, parameterName=TopS_Em02_Speed__Master, defaultValue=<rgba 0 0 0 1>}
  n628 = Add(A=LinearInterpolate(A=LinearInterpolate(A=@n623, B=@n624.RGB, Alpha=@n625.R), B=@n626.RGB, Alpha=@n625.G), B=@n627.RGB)
  NamedRerouteDeclaration(None=Add(A=Multiply(A=@n618, B=@n628), B=NamedRerouteUsage())) {name=Uvs_TS_Emis02_Speed_B, nodeColor=<rgba 0 0.625 1 1>}
  n629 = Time()
  n630 = ScalarParameter() {group=TopSide_V02_Layer Emissive_02, parameterName=TopS_Em02_Rotation Speed}
  NamedRerouteDeclaration(None=Add(A=Multiply(A=@n629, B=@n630), B=@n378)) {name=Uvs_TS_Emis_02_Rotation, nodeColor=<rgba 0 0.7188 1 1>}
  NamedRerouteDeclaration(None=@n198) {name=Emissive_01_bool, nodeColor=<rgba 0.3594 0 1 1>}
  NamedRerouteDeclaration(None=@n270) {name=Emissive_02_bool, nodeColor=<rgba 0.3594 0 1 1>}
  n631 = LinearInterpolate(
   A=ScalarParameter() {group=TopSide_V02_Layer Emissive_01 Reactive, parameterName=TopS Audio 01 Emissive_Brightness-Min, sortPriority=150}
   B=ScalarParameter() {group=TopSide_V02_Layer Emissive_01 Reactive, parameterName=TopS Audio 01 Emissive_Brightness_Max, sortPriority=150}
   Alpha=NamedRerouteUsage()
  )
  n632 = Subtract(
   A=Time()
   B=Add(
    A=ScalarParameter() {group=TopSide_V02_Layer Emissive_01 Reactive, parameterName=TopS Emissive_01_Brightness---Delay}
    B=ScalarParameter() {defaultValue=-9999999, group=TopSide_V02_Layer Emissive_01 Reactive, parameterName=TopS Emissive_01-Brightness-TBP}
   )
  )
  n633 = ScalarParameter() {defaultValue=0.25, group=TopSide_V02_Layer Emissive_01 Reactive, parameterName=TopS Emissive_01_Brightness---Duration}
  n634 = Power(
   Base=Saturate(None=Divide(A=@n632, B=@n633))
   Exp=ScalarParameter() {defaultValue=0.25, group=TopSide_V02_Layer Emissive_01 Reactive, parameterName=TopS Emissive_01_Brightness---DFalloff}
  )
  n635 = LinearInterpolate(
   A=ScalarParameter() {defaultValue=1, group=TopSide_V02_Layer Emissive_01 Reactive, parameterName=TopS Emissive_01_Brightness---Start}
   B=ScalarParameter() {group=TopSide_V02_Layer Emissive_01 Reactive, parameterName=TopS Emissive_01_Brightness---End}
   Alpha=@n634
  )
  NamedRerouteDeclaration(None=StaticSwitch(True=Add(A=@n631, B=@n635), False=@n635, Value=NamedRerouteUsage())) {name=Reactive_Brightness_01, nodeColor=<rgba 1 0.2812 0 1>}
  NamedRerouteDeclaration(None=@n634) {name=Reactive_Brightness_01_Time, nodeColor=<rgba 1 0.1641 0 1>}
  n636 = LinearInterpolate(
   A=ScalarParameter() {group=TopSide_V02_Layer Emissive_02 Reactive, parameterName=TopS Audio 02 Emissive_Brightness-Min, sortPriority=210}
   B=ScalarParameter() {group=TopSide_V02_Layer Emissive_02 Reactive, parameterName=TopS Audio 02 Emissive_Brightness_Max, sortPriority=210}
   Alpha=NamedRerouteUsage()
  )
  n637 = Subtract(
   A=Time()
   B=Add(
    A=ScalarParameter() {group=TopSide_V02_Layer Emissive_02 Reactive, parameterName=TopS Emissive_02_Brightness---Delay}
    B=ScalarParameter() {defaultValue=-9999999, group=TopSide_V02_Layer Emissive_02 Reactive, parameterName=TopS Emissive_02-Brightness-TBP}
   )
  )
  n638 = ScalarParameter() {defaultValue=0.25, group=TopSide_V02_Layer Emissive_02 Reactive, parameterName=TopS Emissive_02_Brightness---Duration}
  n639 = Power(
   Base=Saturate(None=Divide(A=@n637, B=@n638))
   Exp=ScalarParameter() {defaultValue=0.25, group=TopSide_V02_Layer Emissive_02 Reactive, parameterName=TopS Emissive_02_Brightness---DFalloff}
  )
  n640 = LinearInterpolate(
   A=ScalarParameter() {defaultValue=1, group=TopSide_V02_Layer Emissive_02 Reactive, parameterName=TopS Emissive_02_Brightness---Start}
   B=ScalarParameter() {group=TopSide_V02_Layer Emissive_02 Reactive, parameterName=TopS Emissive_02_Brightness---End}
   Alpha=@n639
  )
  NamedRerouteDeclaration(None=StaticSwitch(True=Add(A=@n636, B=@n640), False=@n640, Value=NamedRerouteUsage())) {name=Reactive_Brightness_02, nodeColor=<rgba 1 0.1641 0 1>}
  NamedRerouteDeclaration(None=FunctionInput(Preview=@n484) {inputName=Material, inputType=FunctionInput_MaterialAttributes}) {name=Incoming_Material, nodeColor=<rgba 0.7812 1 0 1>}
  NamedRerouteDeclaration(None=@n639) {name=Reactive_Brightness_02_Time, nodeColor=<rgba 1 0.1641 0 1>}
  NamedRerouteDeclaration(None=@n610) {name=Reactive_Blend_Opacity_Time, nodeColor=<rgba 1 0.1641 0 1>}
  n641 = ScalarParameter() {defaultValue=1, group="TopSide_V02_Layer 01 Blend  02 Reactive", parameterName=TopS_01_Blend_Offset---Start}
  n642 = ScalarParameter() {defaultValue=1, group="TopSide_V02_Layer 01 Blend  02 Reactive", parameterName=TopS_01_Blend_Offset---End}
  n643 = Subtract(
   A=Time()
   B=Add(
    A=ScalarParameter() {group="TopSide_V02_Layer 01 Blend  02 Reactive", parameterName=TopS_01_Blend_Offset---Delay}
    B=ScalarParameter() {defaultValue=-9999999, group="TopSide_V02_Layer 01 Blend  02 Reactive", parameterName=TopS_01_Blend_Offset-TBP}
   )
  )
  n644 = ScalarParameter() {defaultValue=0.25, group="TopSide_V02_Layer 01 Blend  02 Reactive", parameterName=TopS_01_Blend_Offset---Duration}
  n645 = Power(
   Base=Saturate(None=Divide(A=@n643, B=@n644))
   Exp=ScalarParameter() {defaultValue=0.25, group="TopSide_V02_Layer 01 Blend  02 Reactive", parameterName=TopS_01_Blend_Offset---DFalloff}
  )
  NamedRerouteDeclaration(None=AppendVector(A=LinearInterpolate(A=@n641, B=@n642, Alpha=@n645), B=Constant())) {name=Reactive_Blend_Offset, nodeColor=<rgba 1 0.1641 0 1>}
  n646 = Subtract(
   A=Time()
   B=Add(
    A=ScalarParameter() {group=TopSide_V02_Layer Emissive_01 Reactive, parameterName=TopS Emissive_01_Offset---Delay}
    B=ScalarParameter() {defaultValue=-9999999, group=TopSide_V02_Layer Emissive_01 Reactive, parameterName=TopS Emissive_01_Offset----TBP}
   )
  )
  n647 = ScalarParameter() {defaultValue=0.25, group=TopSide_V02_Layer Emissive_01 Reactive, parameterName=TopS Emissive_01_Offset---Duration}
  n648 = Power(
   Base=Saturate(None=Divide(A=@n646, B=@n647))
   Exp=ScalarParameter() {defaultValue=0.25, group=TopSide_V02_Layer Emissive_01 Reactive, parameterName=TopS Emissive_01_Offset---DFalloff}
  )
  n649 = LinearInterpolate(
   A=ScalarParameter() {group=TopSide_V02_Layer Emissive_01 Reactive, parameterName=TopS Emissive_01_Offset---Start}
   B=ScalarParameter() {group=TopSide_V02_Layer Emissive_01 Reactive, parameterName=TopS Emissive_01_Offset---End}
   Alpha=@n648
  )
  n650 = Add(
   A=LinearInterpolate(
    A=ScalarParameter() {group=TopSide_V02_Layer Emissive_01 Reactive, parameterName=TopS Audio 01 Emissive_Offset---Start, sortPriority=150}
    B=ScalarParameter() {group=TopSide_V02_Layer Emissive_01 Reactive, parameterName=TopS Audio 01 Emissive_Offset---End, sortPriority=150}
    Alpha=NamedRerouteUsage()
   )
   B=@n649
  )
  n651 = NamedRerouteUsage()
  NamedRerouteDeclaration(None=AppendVector(A=StaticSwitch(True=@n650, False=@n649, Value=@n651), B=Constant())) {name=Reactive_Offset_01, nodeColor=<rgba 1 0.1641 0 1>}
  NamedRerouteDeclaration(None=@n648) {name=Reactive_Offset_01_Time, nodeColor=<rgba 1 0.1641 0 1>}
  n652 = Subtract(
   A=Time()
   B=Add(
    A=ScalarParameter() {group=TopSide_V02_Layer Emissive_02 Reactive, parameterName=TopS Emissive_02_Offset---Delay}
    B=ScalarParameter() {defaultValue=-9999999, group=TopSide_V02_Layer Emissive_02 Reactive, parameterName=TopS Emissive_02_Offset----TBP}
   )
  )
  n653 = ScalarParameter() {defaultValue=0.25, group=TopSide_V02_Layer Emissive_02 Reactive, parameterName=TopS Emissive_02_Offset---Duration}
  n654 = Power(
   Base=Saturate(None=Divide(A=@n652, B=@n653))
   Exp=ScalarParameter() {defaultValue=0.25, group=TopSide_V02_Layer Emissive_02 Reactive, parameterName=TopS Emissive_02_Offset---DFalloff}
  )
  n655 = LinearInterpolate(
   A=ScalarParameter() {group=TopSide_V02_Layer Emissive_02 Reactive, parameterName=TopS Emissive_02_Offset---Start}
   B=ScalarParameter() {group=TopSide_V02_Layer Emissive_02 Reactive, parameterName=TopS Emissive_02_Offset---End}
   Alpha=@n654
  )
  n656 = Add(
   A=LinearInterpolate(
    A=ScalarParameter() {group=TopSide_V02_Layer Emissive_02 Reactive, parameterName=TopS Audio 02 Emissive_Offset---Start, sortPriority=210}
    B=ScalarParameter() {group=TopSide_V02_Layer Emissive_02 Reactive, parameterName=TopS Audio 02 Emissive_Offset---End, sortPriority=210}
    Alpha=NamedRerouteUsage()
   )
   B=@n655
  )
  n657 = NamedRerouteUsage()
  NamedRerouteDeclaration(None=AppendVector(A=StaticSwitch(True=@n656, False=@n655, Value=@n657), B=Constant())) {name=Reactive_Offset_02, nodeColor=<rgba 1 0.1641 0 1>}
  NamedRerouteDeclaration(None=@n654) {name=Reactive_Offset_02_Time, nodeColor=<rgba 1 0.1641 0 1>}
  n658 = MaterialFunctionCall(Color=Abs(None=NamedRerouteUsage())) {materialFunction=&SplitComponents}
  n659 = Saturate(None=Add(A=@n658.R, B=@n658.R))
  n660 = Saturate(None=Add(A=@n658.G, B=@n658.G))
  n661 = MaterialFunctionCall(Float4=NamedRerouteUsage()) {materialFunction=&BreakOutFloat4Components}
  n662 = Add(A=@n658.R, B=@n658.R)
  NamedRerouteDeclaration(None=LinearInterpolate(A=LinearInterpolate(A=@n659, B=@n660, Alpha=@n661.R), B=Saturate(None=@n662), Alpha=@n661.B)) {name=Tri_Mask 01A, nodeColor=<rgba 1 0.4688 0 1>}
  NamedRerouteDeclaration(None=@n645) {name=Reactive_Brightness_01_Offset, nodeColor=<rgba 1 0.1641 0 1>}
  n663 = Constant4Vector()
  n664 = FunctionInput(Preview=@n663.RGBA) {bUsePreviewValueAsDefault=true, inputName=IsState_B, inputType=FunctionInput_Vector4, sortPriority=4}
  NamedRerouteDeclaration(None=ComponentMask(None=NamedRerouteDeclaration(None=@n664) {name=IsState_B, nodeColor=<rgba 0.1497 1 0.1633 1>}) {a=true}) {name=isCharAccent, nodeColor=<rgba 0 0.9253 1 1>}
  NamedRerouteDeclaration(None=StaticBoolParameter() {group=TopSide_V02_Layer Emissive_03 Reactive, parameterName=TopS_Em03-Animation-CC, sortPriority=300}) {name=Animted_CC_State_03, nodeColor=<rgba 0.0312 0 1 1>}
  n665 = MaterialFunctionCall(Float4=NamedRerouteUsage()) {materialFunction=&BreakOutFloat4Components}
  n666 = ScalarParameter() {group=TopSide_V02, parameterName="TopS_IS_Roll   1=ON"}
  n667 = Constant()
  n668 = MaterialFunctionCall(
   Caculated_Bounds=NamedRerouteUsage()
   is_State=MaterialFunctionCall(R=LinearInterpolate(A=@n665.R, Alpha=@n666) {constB=0}, G=@n665.G, B=@n665.G, A=@n665.G) {materialFunction=&AppendMany}.RGBA
   Uvs_Offset=@n667
   Uvs_TileRate=Constant() {r=1}
   Uvs_Offset_Pivot=@n667
  ) {materialFunction=&MF_Projection_Side_Normal_V2}
  n669 = ScalarParameter() {group=TopSide_V02_Layer-02_Uvs_Split, parameterName=TopS_02 Split_UV_Offset_Weapon}
  n670 = ScalarParameter() {group=TopSide_V02_Layer-02_Uvs_Split, parameterName=TopS_02 Split_UV_Offset_Vehicle}
  n671 = MaterialFunctionCall(Float4=NamedRerouteUsage()) {materialFunction=&BreakOutFloat4Components}
  n672 = Add(
   A=ComponentMask(None=@n668.UVs No Rotation) {r=true}
   B=LinearInterpolate(
    A=LinearInterpolate(A=@n669, B=@n670, Alpha=@n671.R)
    B=ScalarParameter() {group=TopSide_V02_Layer-02_Uvs_Split, parameterName=TopS_02 Split_UV_Offset_Mech}
    Alpha=@n671.B
   )
  )
  n673 = ScalarParameter() {defaultValue=1, desc="negative value inverts ", group=TopSide_V02_Layer-02_Uvs_Split, parameterName=TopS_02 Split_UV_Falloff_Weapon}
  n674 = ScalarParameter() {defaultValue=1, desc="negative value inverts ", group=TopSide_V02_Layer-02_Uvs_Split, parameterName=TopS_02 Split_UV_Falloff_Vehicle}
  n675 = ScalarParameter() {defaultValue=1, desc="negative value inverts ", group=TopSide_V02_Layer-02_Uvs_Split, parameterName=TopS_02 Split_UV_Falloff_Mech}
  n676 = LinearInterpolate(A=LinearInterpolate(A=@n673, B=@n674, Alpha=@n671.R), B=@n675, Alpha=@n671.B)
  NamedRerouteDeclaration(None=Saturate(None=Divide(A=@n672, B=@n676))) {name=Split Uvs_02, nodeColor=<rgba 0 1 0.5078 1>}
  n677 = ScalarParameter() {group="TopSide_V02_Layer 01 Blend  02", parameterName=TopS_Split_Uv_Blend_Trans_Offset_Weapon}
  n678 = ScalarParameter() {group="TopSide_V02_Layer 01 Blend  02", parameterName=TopS_Split_Uv_Blend_Trans_Offset_Vehicle}
  n679 = MaterialFunctionCall(Float4=NamedRerouteUsage()) {materialFunction=&BreakOutFloat4Components}
  n680 = Add(
   A=ComponentMask(None=@n668.UVs No Rotation) {r=true}
   B=LinearInterpolate(
    A=LinearInterpolate(A=@n677, B=@n678, Alpha=@n679.R)
    B=ScalarParameter() {group="TopSide_V02_Layer 01 Blend  02", parameterName=TopS_Split_Uv_Blend_Trans_Offset_Mech}
    Alpha=@n679.B
   )
  )
  n681 = ScalarParameter() {
   defaultValue=0.001
   desc="negative value inverts "
   group="TopSide_V02_Layer 01 Blend  02"
   parameterName=TopS_Split_Uv_Blend_Trans_Falloff_Weapon
  }
  n682 = ScalarParameter() {
   defaultValue=0.001
   desc="negative value inverts "
   group="TopSide_V02_Layer 01 Blend  02"
   parameterName=TopS_Split_Uv_Blend_Trans_Falloff_Vehicle
  }
  n683 = ScalarParameter() {
   defaultValue=0.001
   desc="negative value inverts "
   group="TopSide_V02_Layer 01 Blend  02"
   parameterName=TopS_Split_Uv_Blend_Trans_Falloff_Mech
  }
  n684 = LinearInterpolate(A=LinearInterpolate(A=@n681, B=@n682, Alpha=@n679.R), B=@n683, Alpha=@n679.B)
  NamedRerouteDeclaration(None=Saturate(None=Divide(A=@n680, B=@n684))) {name=Split UvsTrans_02, nodeColor=<rgba 0 1 0.5078 1>}
  n685 = VectorParameter() {group=TopSide_V02_Layer-02_Uvs_Split, parameterName=TopS_02_Split_Scale_Weapon, sortPriority=100, defaultValue=<rgba 1 1 1 1>}
  n686 = VectorParameter() {group=TopSide_V02_Layer-02_Uvs_Split, parameterName=TopS_02_Split_Scale_Vehicle, sortPriority=100, defaultValue=<rgba 1 1 1 1>}
  n687 = Multiply(A=@n686.RGB, B=@n686.RGB)
  n688 = VectorParameter() {group=TopSide_V02_Layer-02_Uvs_Split, parameterName=TopS_02_Split_Scale_VehicleSide, sortPriority=100, defaultValue=<rgba 1 1 1 1>}
  n689 = Multiply(A=@n688.RGB, B=@n688.RGB)
  n690 = MaterialFunctionCall(Float4=NamedRerouteUsage()) {materialFunction=&BreakOutFloat4Components}
  n691 = VectorParameter() {group=TopSide_V02_Layer-02_Uvs_Split, parameterName=TopS_02_Split_Scale_Mech, sortPriority=100, defaultValue=<rgba 1 1 1 1>}
  n692 = VectorParameter() {group=TopSide_V02_Layer-02_Uvs_Split, parameterName=TopS_02_Split_Scale_Character, sortPriority=100, defaultValue=<rgba 1 1 1 1>}
  n693 = LinearInterpolate(
   A=LinearInterpolate(A=Multiply(A=@n685.RGB, B=@n685.RGB), B=LinearInterpolate(A=@n687, B=@n689, Alpha=@n690.G), Alpha=@n690.R)
   B=LinearInterpolate(A=Multiply(A=@n691.RGB, B=@n691.RGB), B=Multiply(A=@n692.RGB, B=@n692.RGB), Alpha=@n690.A)
   Alpha=@n690.B
  )
  n694 = VectorParameter() {group=TopSide_V02_Layer-02_Uvs_Split, parameterName=TopS_02_Split_Scale_Backpack, sortPriority=100, defaultValue=<rgba 1 1 1 1>}
  n695 = Multiply(A=@n694.RGB, B=@n694.RGB)
  n696 = MaterialFunctionCall(Float2=NamedRerouteUsage()) {materialFunction=&BreakOutFloat2Components}
  n697 = VectorParameter() {group=TopSide_V02_Layer-02_Uvs_Split, parameterName=TopS_02_Split_Scale_Pickaxe, sortPriority=100, defaultValue=<rgba 1 1 1 1>}
  NamedRerouteDeclaration(None=LinearInterpolate(A=LinearInterpolate(A=@n693, B=@n695, Alpha=@n696.R), B=Multiply(A=@n697.RGB, B=@n697.RGB), Alpha=@n696.G)) {name=Split Uvs_02_Scale, nodeColor=<rgba 0 0.625 1 1>}
  NamedRerouteDeclaration(None=@n386) {name=Uvs_TS_Layer_03_Rotation, nodeColor=<rgba 1 0 0.3281 1>}
  n698 = VectorParameter() {group=TopSide_V02_Layer-03_Uvs, parameterName=TopS_03_Scale_Weapon, defaultValue=<rgba 1 1 1 1>}
  n699 = VectorParameter() {group=TopSide_V02_Layer-03_Uvs, parameterName=TopS_03_Scale_Vehicle, defaultValue=<rgba 1 1 1 1>}
  n700 = Multiply(A=@n699.RGB, B=@n699.RGB)
  n701 = VectorParameter() {group=TopSide_V02_Layer-03_Uvs, parameterName=TopS_03_Scale_VehicleSide, defaultValue=<rgba 1 1 1 1>}
  n702 = Multiply(A=@n701.RGB, B=@n701.RGB)
  n703 = MaterialFunctionCall(Float4=NamedRerouteUsage()) {materialFunction=&BreakOutFloat4Components}
  n704 = VectorParameter() {group=TopSide_V02_Layer-03_Uvs, parameterName=TopS_03_Scale_Mech, defaultValue=<rgba 1 1 1 1>}
  n705 = VectorParameter() {group=TopSide_V02_Layer-03_Uvs, parameterName=TopS_03_Scale_Character, defaultValue=<rgba 1 1 1 1>}
  n706 = LinearInterpolate(
   A=LinearInterpolate(A=Multiply(A=@n698.RGB, B=@n698.RGB), B=LinearInterpolate(A=@n700, B=@n702, Alpha=@n703.G), Alpha=@n703.R)
   B=LinearInterpolate(A=Multiply(A=@n704.RGB, B=@n704.RGB), B=Multiply(A=@n705.RGB, B=@n705.RGB), Alpha=@n703.A)
   Alpha=@n703.B
  )
  n707 = VectorParameter() {group=TopSide_V02_Layer-03_Uvs, parameterName=TopS_03_Scale_Backpack, defaultValue=<rgba 1 1 1 1>}
  n708 = Multiply(A=@n707.RGB, B=@n707.RGB)
  n709 = MaterialFunctionCall(Float2=NamedRerouteUsage()) {materialFunction=&BreakOutFloat2Components}
  n710 = VectorParameter() {group=TopSide_V02_Layer-03_Uvs, parameterName=TopS_03_Scale_Pickaxe, defaultValue=<rgba 1 1 1 1>}
  NamedRerouteDeclaration(None=LinearInterpolate(A=LinearInterpolate(A=@n706, B=@n708, Alpha=@n709.R), B=Multiply(A=@n710.RGB, B=@n710.RGB), Alpha=@n709.G)) {name=Uvs_TS_Layer_03_Scale, nodeColor=<rgba 0 0.625 1 1>}
  n711 = StaticBoolParameter() {group="TopSide_V02_Layer 01 Blend  02", parameterName="TopS Blend 01<--->02"}
  NamedRerouteDeclaration(None=@n711) {name=Blend_State, nodeColor=<rgba 0 1 0.7656 1>}
  NamedRerouteDeclaration(None=@n19) {name=Uvs_TS_Layer_03_Offset, nodeColor=<rgba 0 0.625 1 1>}
  NamedRerouteDeclaration(None=@n392) {name=Uvs_TS_Layer_03_Warp_Rotation, nodeColor=<rgba 0 0.7188 1 1>}
  n712 = ScalarParameter() {defaultValue=1, group=TopSide_V02_Layer-03_Uvs | Warp, parameterName=TopS_03_Warp_Scale_Vehicle, sortPriority=300}
  n713 = ScalarParameter() {defaultValue=1, group=TopSide_V02_Layer-03_Uvs | Warp, parameterName=TopS_03_Warp_Scale_Vehicle_Side, sortPriority=300}
  n714 = MaterialFunctionCall(Float4=NamedRerouteUsage()) {materialFunction=&BreakOutFloat4Components}
  n715 = LinearInterpolate(
   A=LinearInterpolate(
    A=ScalarParameter() {defaultValue=1, group=TopSide_V02_Layer-03_Uvs | Warp, parameterName=TopS_03_Warp_Scale_Weapon, sortPriority=300}
    B=LinearInterpolate(A=@n712, B=@n713, Alpha=@n714.G)
    Alpha=@n714.R
   )
   B=LinearInterpolate(
    A=ScalarParameter() {defaultValue=1, group=TopSide_V02_Layer-03_Uvs | Warp, parameterName=TopS_03_Warp_Scale_Mech, sortPriority=300}
    B=ScalarParameter() {defaultValue=1, group=TopSide_V02_Layer-03_Uvs | Warp, parameterName=TopS_03_Warp_Scale_Character, sortPriority=300}
    Alpha=@n714.A
   )
   Alpha=@n714.B
  )
  n716 = ScalarParameter() {defaultValue=1, group=TopSide_V02_Layer-03_Uvs | Warp, parameterName=TopS_03_Warp_Scale_Backpack, sortPriority=300}
  n717 = MaterialFunctionCall(Float2=NamedRerouteUsage()) {materialFunction=&BreakOutFloat2Components}
  NamedRerouteDeclaration(None=LinearInterpolate(
    A=LinearInterpolate(A=@n715, B=@n716, Alpha=@n717.R)
    B=ScalarParameter() {defaultValue=1, group=TopSide_V02_Layer-03_Uvs | Warp, parameterName=TopS_03_Warp_Scale_Pickaxe, sortPriority=300}
    Alpha=@n717.G
   )) {name=Uvs_TS_Layer_03_Warp_Scale, nodeColor=<rgba 0 0.7188 1 1>}
  n718 = Frac(None=@n21)
  n719 = ScalarParameter() {defaultValue=10, group=TopSide_V02_Layer-03_Uvs, parameterName=TopS_03_PanningS_Speed_Step_Amount}
  n720 = Divide(A=Ceil(None=Multiply(A=@n718, B=@n719) {constB=30}), B=@n719)
  NamedRerouteDeclaration(None=Add(A=LinearInterpolate(A=@n720, B=@n21, Alpha=@n20."0=step animation"), B=NamedRerouteUsage())) {name=Uvs_TS_Layer_03_Speed, nodeColor=<rgba 0 0.625 1 1>}
  NamedRerouteDeclaration(None=@n398) {name=Uvs_TS_Layer_03_Rotation_Side_1, nodeColor=<rgba 0 0.7188 1 1>}
  n721 = VectorParameter() {group=TopSide_V02_Layer-03_Uvs, parameterName=TopS_03_OffsetS_Vehicle, defaultValue=<rgba 0 0 0 1>}
  n722 = VectorParameter() {group=TopSide_V02_Layer-03_Uvs, parameterName=TopS_03_OffsetS_VehicleSide, defaultValue=<rgba 0 0 0 1>}
  n723 = MaterialFunctionCall(Float4=NamedRerouteUsage()) {materialFunction=&BreakOutFloat4Components}
  n724 = LinearInterpolate(
   A=LinearInterpolate(
    A=VectorParameter() {group=TopSide_V02_Layer-03_Uvs, parameterName=TopS_03_OffsetS_Weapon, defaultValue=<rgba 0 0 0 1>}.RGB
    B=LinearInterpolate(A=@n721.RGB, B=@n722.RGB, Alpha=@n723.G)
    Alpha=@n723.R
   )
   B=LinearInterpolate(
    A=VectorParameter() {group=TopSide_V02_Layer-03_Uvs, parameterName=TopS_03_OffsetS_Mech, defaultValue=<rgba 0 0 0 1>}.RGB
    B=VectorParameter() {group=TopSide_V02_Layer-03_Uvs, parameterName=TopS_03_OffsetS_Character, defaultValue=<rgba 0 0 0 1>}.RGB
    Alpha=@n723.A
   )
   Alpha=@n723.B
  )
  n725 = VectorParameter() {group=TopSide_V02_Layer-03_Uvs, parameterName=TopS_03_OffsetS_Backpack, defaultValue=<rgba 0 0 0 1>}
  n726 = MaterialFunctionCall(Float2=NamedRerouteUsage()) {materialFunction=&BreakOutFloat2Components}
  NamedRerouteDeclaration(None=LinearInterpolate(
    A=LinearInterpolate(A=@n724, B=@n725.RGB, Alpha=@n726.R)
    B=VectorParameter() {group=TopSide_V02_Layer-03_Uvs, parameterName=TopS_03_OffsetS_Pickaxe, defaultValue=<rgba 0 0 0 1>}.RGB
    Alpha=@n726.G
   )) {name=Uvs_TS_Layer_03_Offset_Side, nodeColor=<rgba 0 0.625 1 1>}
  n727 = VectorParameter() {group=TopSide_V02_Layer-03_Uvs, parameterName=TopS_03_ScaleS_Weapon, defaultValue=<rgba 1 1 1 1>}
  n728 = VectorParameter() {group=TopSide_V02_Layer-03_Uvs, parameterName=TopS_03_ScaleS_Vehicle, defaultValue=<rgba 1 1 1 1>}
  n729 = Multiply(A=@n728.RGB, B=@n728.RGB)
  n730 = VectorParameter() {group=TopSide_V02_Layer-03_Uvs, parameterName=TopS_03_ScaleS_VehicleSide, defaultValue=<rgba 1 1 1 1>}
  n731 = Multiply(A=@n730.RGB, B=@n730.RGB)
  n732 = MaterialFunctionCall(Float4=NamedRerouteUsage()) {materialFunction=&BreakOutFloat4Components}
  n733 = VectorParameter() {group=TopSide_V02_Layer-03_Uvs, parameterName=TopS_03_ScaleS_Mech, defaultValue=<rgba 1 1 1 1>}
  n734 = VectorParameter() {group=TopSide_V02_Layer-03_Uvs, parameterName=TopS_03_ScaleS_Character, defaultValue=<rgba 1 1 1 1>}
  n735 = LinearInterpolate(
   A=LinearInterpolate(A=Multiply(A=@n727.RGB, B=@n727.RGB), B=LinearInterpolate(A=@n729, B=@n731, Alpha=@n732.G), Alpha=@n732.R)
   B=LinearInterpolate(A=Multiply(A=@n733.RGB, B=@n733.RGB), B=Multiply(A=@n734.RGB, B=@n734.RGB), Alpha=@n732.A)
   Alpha=@n732.B
  )
  n736 = VectorParameter() {group=TopSide_V02_Layer-03_Uvs, parameterName=TopS_03_ScaleS_Backpack, defaultValue=<rgba 1 1 1 1>}
  n737 = Multiply(A=@n736.RGB, B=@n736.RGB)
  n738 = MaterialFunctionCall(Float2=NamedRerouteUsage()) {materialFunction=&BreakOutFloat2Components}
  n739 = VectorParameter() {group=TopSide_V02_Layer-03_Uvs, parameterName=TopS_03_ScaleS_Pickaxe, defaultValue=<rgba 1 1 1 1>}
  NamedRerouteDeclaration(None=LinearInterpolate(A=LinearInterpolate(A=@n735, B=@n737, Alpha=@n738.R), B=Multiply(A=@n739.RGB, B=@n739.RGB), Alpha=@n738.G)) {name=Uvs_TS_Layer_03_Scale_Side, nodeColor=<rgba 0 0.625 1 1>}
  n740 = Time()
  n741 = VectorParameter() {
   group=TopSide_V02_Layer-03_Uvs
   parameterName=TopS_03_Pannings_Speed_B
   channelNames {a="0=step Animation", b="", g="", r=""}
   defaultValue=<rgba 0 0 0 1>
  }
  n742 = Multiply(A=@n740, B=@n741)
  n743 = Frac(None=@n742)
  n744 = ScalarParameter() {defaultValue=10, group=TopSide_V02_Layer-03_Uvs, parameterName=TopS_03_PanningS_Speed_B_Step_Amount}
  n745 = Divide(A=Ceil(None=Multiply(A=@n743, B=@n744) {constB=30}), B=@n744)
  n746 = VectorParameter() {
   group=TopSide_V02_Layer-03_Uvs
   parameterName=TopS_03_PanningS_Speed------------
   channelNames {a="0=Step animation", b="", g="", r=""}
   defaultValue=<rgba 0 0 0 1>
  }
  n747 = Multiply(A=@n740, B=@n746)
  n748 = ScalarParameter() {defaultValue=10, group=TopSide_V02_Layer-03_Uvs, parameterName=TopS_03_PanningS_Speed--Step_Amount}
  n749 = Multiply(A=Frac(None=@n747), B=@n748) {constB=30}
  n750 = LinearInterpolate(A=Divide(A=Ceil(None=@n749), B=@n748), B=@n747, Alpha=@n746."0=Step animation")
  n751 = StaticSwitchParameter(
   True=LinearInterpolate(A=LinearInterpolate(A=@n745, B=@n742, Alpha=@n741."0=step Animation"), B=@n750, Alpha=NamedRerouteUsage())
   False=@n750
  ) {group=TopSide_V02_Layer-03_Uvs, parameterName=TopS_03_PanningS_Speed--Varation-Msk}
  n752 = VectorParameter() {group=TopSide_V02_Layer-03_Uvs_Reactive, parameterName=TopS_03_Speed_offset_Amount, sortPriority=100, defaultValue=<rgba 0 0 0 1>}
  n753 = Time()
  n754 = Add(
   A=ScalarParameter() {group=TopSide_V02_Layer-03_Uvs_Reactive, parameterName=TopS_03_Speed---Delay, sortPriority=100}
   B=ScalarParameter() {defaultValue=-9999999, group=TopSide_V02_Layer-03_Uvs_Reactive, parameterName=TopS_03_Speed-TBP, sortPriority=100}
  )
  n755 = Saturate(None=Divide(
    A=Subtract(A=@n753, B=@n754)
    B=ScalarParameter() {defaultValue=1, group=TopSide_V02_Layer-03_Uvs_Reactive, parameterName=TopS_03_Speed---Duration, sortPriority=100}
   ))
  n756 = ScalarParameter() {defaultValue=1, group=TopSide_V02_Layer-03_Uvs_Reactive, parameterName=TopS_03_Speed---DFalloff, sortPriority=100}
  n757 = LinearInterpolate(B=@n752.RGB, Alpha=OneMinus(None=Power(Base=@n755, Exp=@n756))) {constB=0}
  n758 = Add(A=@n751, B=@n757)
  NamedRerouteDeclaration(None=@n758) {name=Uvs_TS_Layer_03_Speed_Side, nodeColor=<rgba 0 0.625 1 1>}
  n759 = Time()
  n760 = ScalarParameter() {group=TopSide_V02_Layer-03_Uvs, parameterName=TopS_03_Side_Rotation__Speed}
  NamedRerouteDeclaration(None=Add(A=Multiply(A=@n759, B=@n760), B=@n412)) {name=Uvs_TS_Layer_03_Rotation_Side, nodeColor=<rgba 0 0.7188 1 1>}
  n761 = StaticBoolParameter() {group=TopSide_V02_Layer-02_Uvs_Split, parameterName=TopS_02_Split_Uvs, sortPriority=0}
  NamedRerouteDeclaration(None=@n761) {name=Split_uv02_State, nodeColor=<rgba 0.125 1 0 1>}
  n762 = StaticBoolParameter() {group=TopSide_V02_Layer 01, parameterName=TopS_01-Miplevels-VS-Alternative_Axies}
  NamedRerouteDeclaration(None=@n762) {name=TopS_01_MipLevels, nodeColor=<rgba 1 0.7031 0 1>}
  n763 = ScalarParameter() {group=TopSide_V02_Layer Emissive_01 Reactive, parameterName=TopS Emissive_01_Gradent---Delay}
  n764 = ScalarParameter() {defaultValue=-9999999, group=TopSide_V02_Layer Emissive_01 Reactive, parameterName=TopS Emissive_01_Gradent----TBP}
  n765 = Divide(
   A=Subtract(A=Time(), B=Add(A=@n763, B=@n764))
   B=ScalarParameter() {defaultValue=0.25, group=TopSide_V02_Layer Emissive_01 Reactive, parameterName=TopS Emissive_01_Gradent---Duration}
  )
  NamedRerouteDeclaration(None=Power(
    Base=Saturate(None=@n765)
    Exp=ScalarParameter() {defaultValue=0.25, group=TopSide_V02_Layer Emissive_01 Reactive, parameterName=TopS Emissive_01_Gradent---DFalloff}
   )) {name=TopS Emissive_01_Gradent, nodeColor=<rgba 1 0.1641 0 1>}
  n766 = NamedRerouteUsage()
  n767 = Add(A=Add(A=NamedRerouteUsage(), B=Constant3Vector() {constant=<rgba 0 0 -1 0>}), B=NamedRerouteUsage())
  n768 = NamedRerouteUsage()
  n769 = MaterialFunctionCall(Float4=NamedRerouteUsage()) {materialFunction=&BreakOutFloat4Components}
  n770 = LinearInterpolate(
   A=ScalarParameter() {group=TopSide_V02_Layer 01, parameterName=TopS_01 Blend_Contrast}
   B=ScalarParameter() {group=TopSide_V02_Layer 01, parameterName="TopS_01 Blend_Contrast \\Mech"}
   Alpha=@n769.B
  )
  n771 = NamedRerouteUsage()
  n772 = NamedRerouteUsage()
  n773 = LinearInterpolate(
   A=ScalarParameter() {group=TopSide_V02_Layer 01, parameterName=TopS_01_Alternate Axies Opacity}
   B=ScalarParameter() {group=TopSide_V02_Layer 01, parameterName="TopS_01_Alternate Axies Opacity \\Mech"}
   Alpha=@n769.B
  )
  n774 = VectorParameter() {group=TopSide_V02_Layer 01 -Color_Tint, parameterName=TopS_01_Alternate_Axies MaskColorTint}
  n775 = AppendVector(A=@n774.RGB, B=@n774.RGB)
  n776 = NamedRerouteUsage()
  n777 = Reroute()
  n778 = NamedRerouteUsage()
  n779 = NamedRerouteUsage()
  n780 = StaticBoolParameter() {group=TopSide_V02_Layer 01, parameterName="TopS_01 TopSide->Side projection"}
  NamedRerouteDeclaration(None=StaticSwitch(
    True=NamedRerouteUsage()
    False=MaterialFunctionCall!(
     Texture=@n766
     Caculated_Bounds=@n767
     Pre_Skinned_Normals=@n768
     Blend_Contrast=@n770
     IsStatsB=@n771
     IsState=@n772
     Fill_alternate_Axis_Opacity=@n773
     "Fill_alternate_Axis_Color_A=Roughness"=@n775
     Offset=@n779
     Scale=@n776
     Offset_Pivot=@n777
     Rotation_Axis=@n778
    ) {materialFunction=&/MF_TopsSideProjection_Base_V2}.Combined
    Value=@n780
   )) {name=TS_Layer01 MskColorTint, nodeColor=<rgba 1 0 0.8203 1>}
  n781 = ScalarParameter() {group=TopSide_V02_Layer Emissive_02 Reactive, parameterName=TopS Emissive_02_Gradent---Delay}
  n782 = ScalarParameter() {defaultValue=-9999999, group=TopSide_V02_Layer Emissive_02 Reactive, parameterName=TopS Emissive_02_Gradent----TBP}
  n783 = Divide(
   A=Subtract(A=Time(), B=Add(A=@n781, B=@n782))
   B=ScalarParameter() {defaultValue=0.25, group=TopSide_V02_Layer Emissive_02 Reactive, parameterName=TopS Emissive_02_Gradent---Duration}
  )
  NamedRerouteDeclaration(None=Power(
    Base=Saturate(None=@n783)
    Exp=ScalarParameter() {defaultValue=0.25, group=TopSide_V02_Layer Emissive_02 Reactive, parameterName=TopS Emissive_02_Gradent---DFalloff}
   )) {name=TopS Emissive_02_Gradent, nodeColor=<rgba 1 0.1641 0 1>}
  n784 = TextureObjectParameter() {
   group=TopSide_V02_Layer 01 -Color_Tint
   parameterName=TopS_01__Tex_MaskColorTint
   samplerType=SAMPLERTYPE_Masks
   texture=&T_BlackMask
  }
  NamedRerouteDeclaration(None=@n784) {name=Tex_01_ColorTint, nodeColor=<rgba 1 0.8906 0 1>}
  NamedRerouteDeclaration(None=@n775) {name=Tops 01 Alternate Mask_ColorTint, nodeColor=<rgba 1 0.3984 0 1>}
  n785 = StaticBoolParameter() {group=TopSide_V02_Layer-01_Uvs, parameterName=TopS_01 Clamp Uvs}
  n786 = ScalarParameter() {defaultValue=0.001, group=TopSide_V02_Layer-01_Uvs, parameterName=TopS_01-Clamp Uvs padding Comp}
  n787 = MaterialFunctionCall!(
   Normal_Map=NamedRerouteUsage()
   Offset RT
   Pivot=NamedRerouteUsage()
   "Pivot_L\\R"=StaticBool() {value=true}
   Caculated_Bounds=NamedRerouteUsage()
   Pre_Skinned_Normals=NamedRerouteUsage()
   Is State=NamedRerouteUsage()
   Uvs_Rotation_Mirroring
   Uvs_Rotation=NamedRerouteUsage()
   Uvs_TileRate=NamedRerouteUsage()
   Uvs_Offset=NamedRerouteUsage()
   Clamp Y=@n785
   Clamp X=@n785
   Clamp Padding_Comp=@n786
  ) {materialFunction=&/MF_Projection_Side_Left_Right_Decal_NormalV2}
  n788 = StaticBoolParameter() {desc=use this when animating rotation, group=TopSide_V02_Layer-01_Uvs, parameterName=TopS_01-Rotation-Comp}
  n789 = StaticBoolParameter() {group=TopSide_V02_Layer-01_Uvs, parameterName=TopS_01-No_Rotation}
  n790 = MaterialFunctionCall(
   Normal_Map=NamedRerouteUsage()
   Caculated_Bounds=NamedRerouteUsage()
   Pre_Skinned_Normals=NamedRerouteUsage()
   is_State=NamedRerouteUsage()
   UV_Rotation=NamedRerouteUsage()
   Uvs_Offset=@n23
   Uvs_TileRate=NamedRerouteUsage()
   Uvs_Offset_Pivot=@n22
   Rotation_Comp=@n788
   Clamp Y=@n785
   Clamp X=@n785
   Clamp Padding_Comp=@n786
   No_Rotation=@n789
  ) {materialFunction=&MF_Projection_Side_Normal_V2}
  n791 = NamedRerouteUsage()
  n792 = StaticSwitch(True=@n787.UVs, False=@n790.UVs, Value=@n791)
  n793 = NamedRerouteUsage()
  n794 = ScalarParameter() {defaultValue=3, group=TopSide_V02_Layer 01, parameterName=TopS_01-Miplevels_}
  n795 = TextureSample!(UVs=@n792, Tex=@n793, Level=@n794, Apply View MipBias) {mipValueMode=TMVM_MipLevel}
  n796 = StaticSwitch(True=@n787.UVs No Offsets/Tiling, False=@n790.UVs No Offsets/Tiling, Value=@n791)
  n797 = DDX(Value=@n796)
  n798 = DDY(Value=@n796)
  n799 = TextureSample!(UVs=@n792, Tex=@n793, "DDX(UVs)"=@n797, "DDY(UVs)"=@n798, Apply View MipBias) {mipValueMode=TMVM_Derivative}
  n800 = NamedRerouteUsage()
  n801 = NamedRerouteUsage()
  n802 = OneMinus(None=NamedRerouteUsage())
  n803 = Saturate(None=Add(A=NamedRerouteUsage(), B=Saturate(None=@n802)))
  NamedRerouteDeclaration(None=StaticSwitch(True=LinearInterpolate(A=@n795.RGBA, B=@n799.RGBA, Alpha=@n800), False=LinearInterpolate(A=@n801, B=@n799.RGBA, Alpha=@n803), Value=@n762)) {name=TS 01 Side_Mask_ColorTint, nodeColor=<rgba 1 0.5625 0 1>}
  n804 = CollectionParameter() {collection=&FortniteMaterialParameters, parameterName=EventAudio_Spectrum}
  n805 = ChannelMaskParameter(None=@n804) {
   group=TopSide_V02_Layer Emissive_01 Reactive
   maskChannel=Alpha
   parameterName=TopS Audio 01 Channel
   sortPriority=149
   defaultValue=<rgba 0 0 0 1>
  }
  n806 = CollectionParameter() {collection=&FortniteMaterialParameters, parameterName=Emote_AudioLow}
  n807 = MaterialFunctionCall(
   Input=Max(A=@n805.R, B=@n806)
   Input Low=ScalarParameter() {defaultValue=0.25, group=TopSide_V02_Layer Emissive_01 Reactive, parameterName=TopS Audio 01_Low_Remap, sortPriority=149}
   Input High=ScalarParameter() {defaultValue=1, group=TopSide_V02_Layer Emissive_01 Reactive, parameterName=TopS Audio 01_High_Remap, sortPriority=149}
   Target Low=Constant()
   Target High=Constant() {r=1}
  ) {materialFunction=&RemapValueRange}
  n808 = CollectionParameter() {collection=&FortniteMaterialParameters, parameterName=Emote_AudioHigh}
  n809 = Power(
   Base=Max(A=Saturate(None=@n807.Result), B=@n808)
   Exp=ScalarParameter() {defaultValue=1, group=TopSide_V02_Layer Emissive_01 Reactive, parameterName=TopS Audio 01 Falloff, sortPriority=149}
  )
  NamedRerouteDeclaration(None=@n809) {name=AudioSpectrum 01, nodeColor=<rgba 1 0.8203 0 1>}
  n810 = VectorParameter() {group=TopSide_V02_Layer 01 -Color_Tint, parameterName="TopS_01_ColorTint_CH_Red  A=0=off", defaultValue=<rgba 1 0 0 1>}
  n811 = MaterialFunctionCall(Float4=NamedRerouteUsage()) {materialFunction=&BreakOutFloat4Components}
  n812 = VectorParameter() {group=TopSide_V02_Layer 01 -Color_Tint, parameterName="TopS_01_ColorTint_CH_Green A=0=off", defaultValue=<rgba 0 1 0 1>}
  n813 = LinearInterpolate(A=LinearInterpolate(B=@n810.RGB, Alpha=Multiply(A=@n810.A, B=@n811.R)) {constA=1}, B=@n812.RGB, Alpha=Multiply(A=@n811.G, B=@n812.A)) {constA=1}
  n814 = VectorParameter() {group=TopSide_V02_Layer 01 -Color_Tint, parameterName="TopS_01_ColorTint_CH_Blue  A=0=off", defaultValue=<rgba 0 0 1 1>}
  n815 = Multiply(A=@n811.B, B=@n814.A)
  n816 = VectorParameter() {group=TopSide_V02_Layer 01 -Color_Tint, parameterName="TopS_01_ColorTint_CH_Alpha   A=0=off", defaultValue=<rgba 1 0 1 0>}
  NamedRerouteDeclaration(None=LinearInterpolate(A=LinearInterpolate(A=@n813, B=@n814.RGB, Alpha=@n815), B=@n816.RGB, Alpha=Multiply(A=@n811.A, B=@n816.A))) {name=TS 01 Side_ColorTint, nodeColor=<rgba 1 0.5625 0 1>}
  n817 = CollectionParameter() {collection=&FortniteMaterialParameters, parameterName=EventAudio_Spectrum}
  n818 = ChannelMaskParameter(None=@n817) {
   group=TopSide_V02_Layer Emissive_02 Reactive
   maskChannel=Alpha
   parameterName=TopS Audio 02 Channel
   sortPriority=200
   defaultValue=<rgba 0 0 0 1>
  }
  n819 = CollectionParameter() {collection=&FortniteMaterialParameters, parameterName=Emote_AudioLow}
  n820 = MaterialFunctionCall(
   Input=Max(A=@n818.R, B=@n819)
   Input Low=ScalarParameter() {defaultValue=0.25, group=TopSide_V02_Layer Emissive_02 Reactive, parameterName=TopS Audio 02_Low_Remap, sortPriority=200}
   Input High=ScalarParameter() {defaultValue=1, group=TopSide_V02_Layer Emissive_02 Reactive, parameterName=TopS Audio 02_High_Remap, sortPriority=200}
   Target Low=Constant()
   Target High=Constant() {r=1}
  ) {materialFunction=&RemapValueRange}
  n821 = CollectionParameter() {collection=&FortniteMaterialParameters, parameterName=Emote_AudioHigh}
  n822 = Power(
   Base=Max(A=Saturate(None=@n820.Result), B=@n821)
   Exp=ScalarParameter() {defaultValue=1, group=TopSide_V02_Layer Emissive_02 Reactive, parameterName=TopS Audio 02 Falloff, sortPriority=200}
  )
  NamedRerouteDeclaration(None=@n822) {name=AudioSpectrum 02, nodeColor=<rgba 1 0.8203 0 1>}
  NamedRerouteDeclaration(None=StaticBoolParameter() {group=TopSide_V02_Layer Emissive_01 Reactive, parameterName=TopS Audio 01 Reactive, sortPriority=148}) {name=Audio_01, nodeColor=<rgba 0 0.4141 1 1>}
  n823 = TextureObjectParameter() {group=TopSide_V02_Layer 01, parameterName=TopS_01__Tex_BaseColor, texture=/Game/Weapons/WeaponSkins/Wraps/Textures/UVCheaker_Clamped}
  NamedRerouteDeclaration(None=@n823) {name=Tex_TS01_BaseColor, nodeColor=<rgba 0.0781 0 1 1>}
  NamedRerouteDeclaration(None=StaticBoolParameter() {group=TopSide_V02_Layer Emissive_02 Reactive, parameterName=TopS Audio 02 Reactive, sortPriority=199}) {name=Audio_02, nodeColor=<rgba 0 0.4141 1 1>}
  n824 = Time()
  n825 = ScalarParameter() {defaultValue=2, group=TopSide_V02_Layer Emissive_01 Reactive, parameterName=TopS Audio 01 Static_Speed, sortPriority=149}
  n826 = ConstantBiasScale(None=Sine(None=Multiply(A=@n824, B=@n825)))
  n827 = Multiply(
   A=ScalarParameter() {group=TopSide_V02_Layer Emissive_01 Reactive, parameterName=TopS Audio 01 Static_Strength, sortPriority=149}
   B=NamedRerouteUsage()
  )
  NamedRerouteDeclaration(None=Max(A=@n809, B=Multiply(A=@n826, B=@n827))) {name=AudioSpectrum 01_Static_Speed, nodeColor=<rgba 1 0.8203 0 1>}
  n828 = Time()
  n829 = ScalarParameter() {defaultValue=2, group=TopSide_V02_Layer Emissive_02 Reactive, parameterName=TopS Audio 02 Static_Speed, sortPriority=200}
  n830 = ConstantBiasScale(None=Sine(None=Multiply(A=@n828, B=@n829)))
  n831 = Multiply(
   A=ScalarParameter() {group=TopSide_V02_Layer Emissive_02 Reactive, parameterName=TopS Audio 02 Static_Strength, sortPriority=200}
   B=NamedRerouteUsage()
  )
  NamedRerouteDeclaration(None=Max(A=@n822, B=Multiply(A=@n830, B=@n831))) {name=AudioSpectrum 02_Static_Speed, nodeColor=<rgba 1 0.8203 0 1>}
  n832 = Add(A=@n483) {constB=-0.5}
  NamedRerouteDeclaration(None=Saturate(None=Multiply(A=@n832) {constB=-1})) {name=Curve_Pits, nodeColor=<rgba 1 0 0.8906 1>}
  NamedRerouteDeclaration(None=Saturate(None=@n832)) {name=Curve_Peaks, nodeColor=<rgba 1 0 0.8906 1>}
  NamedRerouteDeclaration(None=@n789) {name=TopS_01-No_Rotation_State, nodeColor=<rgba 0 1 0.2734 1>}
  n833 = TextureObjectParameter() {
   group=TopSide_V02
   parameterName=TopS_00B_Sharerd_Tex_Msk
   samplerType=SAMPLERTYPE_Masks
   sortPriority=249
   texture=&T_BlackMask
  }
  NamedRerouteDeclaration(None=@n833) {name=Tex_00B_Shared_Tex, nodeColor=<rgba 1 0.8906 0 1>}
  NamedRerouteDeclaration(None=@n788) {name=TopS_01-Rotation-Comp_State, nodeColor=<rgba 0 1 0.2734 1>}
  NamedRerouteDeclaration(None=@n38) {name=Trans_State, nodeColor=<rgba 1 0.5391 0 1>}
  n834 = VectorParameter() {group=TopSide_V02_Layer Emissive_03, parameterName=TopS_Em03_Scale_Weapon, defaultValue=<rgba 1 1 1 1>}
  n835 = Multiply(A=@n834.RGB, B=@n834.RGB)
  n836 = VectorParameter() {group=TopSide_V02_Layer Emissive_03, parameterName=TopS_Em03_Scale_Vehicle, defaultValue=<rgba 1 1 1 1>}
  n837 = VectorParameter() {group=TopSide_V02_Layer Emissive_03, parameterName=TopS_Em03_Scale_VehicleSide, defaultValue=<rgba 1 1 1 1>}
  n838 = MaterialFunctionCall(Float4=NamedRerouteUsage()) {materialFunction=&BreakOutFloat4Components}
  n839 = LinearInterpolate(A=Multiply(A=@n836.RGB, B=@n836.RGB), B=Multiply(A=@n837.RGB, B=@n837.RGB), Alpha=@n838.G)
  n840 = VectorParameter() {group=TopSide_V02_Layer Emissive_03, parameterName=TopS_Em03_Scale_Mech, defaultValue=<rgba 1 1 1 1>}
  n841 = Multiply(A=@n840.RGB, B=@n840.RGB)
  n842 = VectorParameter() {group=TopSide_V02_Layer Emissive_03, parameterName=TopS_Em03_Scale_Character, defaultValue=<rgba 1 1 1 1>}
  n843 = Multiply(A=@n842.RGB, B=@n842.RGB)
  n844 = VectorParameter() {group=TopSide_V02_Layer Emissive_03, parameterName=TopS_Em03_Scale_Backpack, defaultValue=<rgba 1 1 1 1>}
  n845 = MaterialFunctionCall(Float2=NamedRerouteUsage()) {materialFunction=&BreakOutFloat2Components}
  n846 = LinearInterpolate(
   A=LinearInterpolate(A=LinearInterpolate(A=@n835, B=@n839, Alpha=@n838.R), B=LinearInterpolate(A=@n841, B=@n843, Alpha=@n838.A), Alpha=@n838.B)
   B=Multiply(A=@n844.RGB, B=@n844.RGB)
   Alpha=@n845.R
  )
  n847 = VectorParameter() {group=TopSide_V02_Layer Emissive_03, parameterName=TopS_Em03_Scale_Pickaxe, defaultValue=<rgba 1 1 1 1>}
  n848 = Multiply(A=@n847.RGB, B=@n847.RGB)
  n849 = VectorParameter() {group=TopSide_V02_Layer Emissive_03, parameterName=TopS_Em03_Scale__Master, defaultValue=<rgba 1 1 1 1>}
  NamedRerouteDeclaration(None=Multiply(A=LinearInterpolate(A=@n846, B=@n848, Alpha=@n845.G), B=Multiply(A=@n849.RGB, B=@n849.RGB))) {name=Uvs_TS_Emis03_Scale, nodeColor=<rgba 0 0.625 1 1>}
  n850 = NamedRerouteUsage()
  n851 = NamedRerouteUsage()
  n852 = TextureObjectParameter() {
   group=TopSide_V02_Layer 01
   parameterName=TopS_01__Tex_Mask
   samplerType=SAMPLERTYPE_Masks
   texture=&T_BlackMask
  }
  NamedRerouteDeclaration(None=StaticSwitchParameter(
    True=StaticSwitchParameter(True=@n850, False=@n851) {group=TopSide_V02_Layer 01, parameterName=TopS_01__Tex_Mask Share 00B_Msk_Tex}
    False=@n852
   ) {group=TopSide_V02_Layer 01, parameterName=TopS_01__Tex_Mask Share 00A_Msk_Tex}) {name=Tex_TS01_Mask, nodeColor=<rgba 1 0.8906 0 1>}
  n853 = VectorParameter() {group=TopSide_V02_Layer Emissive_03, parameterName=TopS_Em03_Offset__Master, defaultValue=<rgba 0 0 0 1>}
  n854 = VectorParameter() {group=TopSide_V02_Layer Emissive_03, parameterName=TopS_Em03_Offset_Weapon, defaultValue=<rgba 0 0 0 1>}
  n855 = VectorParameter() {group=TopSide_V02_Layer Emissive_03, parameterName=TopS_Em03_Offset_Vehicle, defaultValue=<rgba 0 0 0 1>}
  n856 = VectorParameter() {group=TopSide_V02_Layer Emissive_03, parameterName=TopS_Em03_Offset_VehicleSide, defaultValue=<rgba 0 0 0 1>}
  n857 = MaterialFunctionCall(Float4=NamedRerouteUsage()) {materialFunction=&BreakOutFloat4Components}
  n858 = LinearInterpolate(A=@n855.RGB, B=@n856.RGB, Alpha=@n857.G)
  n859 = VectorParameter() {group=TopSide_V02_Layer Emissive_03, parameterName=TopS_Em03_Offset_Mech, defaultValue=<rgba 0 0 0 1>}
  n860 = VectorParameter() {group=TopSide_V02_Layer Emissive_03, parameterName=TopS_Em03_Offset_Character, defaultValue=<rgba 0 0 0 1>}
  n861 = VectorParameter() {group=TopSide_V02_Layer Emissive_03, parameterName=TopS_Em03_Offset_Backpack, defaultValue=<rgba 0 0 0 1>}
  n862 = MaterialFunctionCall(Float2=NamedRerouteUsage()) {materialFunction=&BreakOutFloat2Components}
  n863 = LinearInterpolate(
   A=LinearInterpolate(A=LinearInterpolate(A=@n854.RGB, B=@n858, Alpha=@n857.R), B=LinearInterpolate(A=@n859.RGB, B=@n860.RGB, Alpha=@n857.A), Alpha=@n857.B)
   B=@n861.RGB
   Alpha=@n862.R
  )
  n864 = VectorParameter() {group=TopSide_V02_Layer Emissive_03, parameterName=TopS_Em03_Offset_Pickaxe, defaultValue=<rgba 0 0 0 1>}
  NamedRerouteDeclaration(None=Add(A=@n853.RGB, B=LinearInterpolate(A=@n863, B=@n864.RGB, Alpha=@n862.G))) {name=Uvs_TS_Emis03_Offset, nodeColor=<rgba 0 0.625 1 1>}
  n865 = ScalarParameter() {defaultValue=1, group=TopSide_V02_Layer Emissive_03, parameterName=TopS_Em03_Falloff_Vehicle}
  n866 = ScalarParameter() {defaultValue=1, group=TopSide_V02_Layer Emissive_03, parameterName=TopS_Em03_Falloff_Vehicle_Side}
  n867 = MaterialFunctionCall(Float4=NamedRerouteUsage()) {materialFunction=&BreakOutFloat4Components}
  n868 = LinearInterpolate(
   A=LinearInterpolate(
    A=ScalarParameter() {defaultValue=1, group=TopSide_V02_Layer Emissive_03, parameterName=TopS_Em03_Falloff_Weapon}
    B=LinearInterpolate(A=@n865, B=@n866, Alpha=@n867.G)
    Alpha=@n867.R
   )
   B=LinearInterpolate(
    A=ScalarParameter() {defaultValue=1, group=TopSide_V02_Layer Emissive_03, parameterName=TopS_Em03_Falloff_Mech}
    B=ScalarParameter() {defaultValue=1, group=TopSide_V02_Layer Emissive_03, parameterName=TopS_Em03_Falloff_Character}
    Alpha=@n867.A
   )
   Alpha=@n867.B
  )
  n869 = ScalarParameter() {defaultValue=1, group=TopSide_V02_Layer Emissive_03, parameterName=TopS_Em03_Falloff_Backpack}
  n870 = MaterialFunctionCall(Float2=NamedRerouteUsage()) {materialFunction=&BreakOutFloat2Components}
  NamedRerouteDeclaration(None=LinearInterpolate(
    A=LinearInterpolate(A=@n868, B=@n869, Alpha=@n870.R)
    B=ScalarParameter() {defaultValue=1, group=TopSide_V02_Layer Emissive_03, parameterName=TopS_Em03_Falloff_Pickaxe}
    Alpha=@n870.G
   )) {name=Uvs_TS_Emis_03_Grad_Falloff, nodeColor=<rgba 0 0.7188 1 1>}
  n871 = Time()
  n872 = VectorParameter() {group=TopSide_V02_Layer Emissive_03, parameterName=TopS_Em03_Speed_Vehicle, defaultValue=<rgba 0 0 0 1>}
  n873 = VectorParameter() {group=TopSide_V02_Layer Emissive_03, parameterName=TopS_Em03_Speed_VehicleSide, defaultValue=<rgba 0 0 0 1>}
  n874 = MaterialFunctionCall(Float4=NamedRerouteUsage()) {materialFunction=&BreakOutFloat4Components}
  n875 = VectorParameter() {group=TopSide_V02_Layer Emissive_03, parameterName=TopS_Em03_Speed_Character, defaultValue=<rgba 0 0 0 1>}
  n876 = LinearInterpolate(
   A=LinearInterpolate(
    A=VectorParameter() {group=TopSide_V02_Layer Emissive_03, parameterName=TopS_Em03_Speed_Weapon, defaultValue=<rgba 0 0 0 1>}.RGB
    B=LinearInterpolate(A=@n872.RGB, B=@n873.RGB, Alpha=@n874.G)
    Alpha=@n874.R
   )
   B=LinearInterpolate(
    A=VectorParameter() {group=TopSide_V02_Layer Emissive_03, parameterName=TopS_Em03_Speed_Mech, defaultValue=<rgba 0 0 0 1>}.RGB
    B=@n875.RGB
    Alpha=@n874.A
   )
   Alpha=@n874.B
  )
  n877 = VectorParameter() {group=TopSide_V02_Layer Emissive_03, parameterName=TopS_Em03_Speed_Backpack, defaultValue=<rgba 0 0 0 1>}
  n878 = MaterialFunctionCall(Float2=NamedRerouteUsage()) {materialFunction=&BreakOutFloat2Components}
  n879 = VectorParameter() {group=TopSide_V02_Layer Emissive_03, parameterName=TopS_Em03_Speed_Pickaxe, defaultValue=<rgba 0 0 0 1>}
  n880 = VectorParameter() {group=TopSide_V02_Layer Emissive_03, parameterName=TopS_Em03_Speed__Master, defaultValue=<rgba 0 0 0 1>}
  n881 = Add(A=LinearInterpolate(A=LinearInterpolate(A=@n876, B=@n877.RGB, Alpha=@n878.R), B=@n879.RGB, Alpha=@n878.G), B=@n880.RGB)
  NamedRerouteDeclaration(None=Add(A=Multiply(A=@n871, B=@n881), B=NamedRerouteUsage())) {name=Uvs_TS_Emis03_Speed_B, nodeColor=<rgba 0 0.625 1 1>}
  n882 = Time()
  n883 = ScalarParameter() {group=TopSide_V02_Layer Emissive_03, parameterName=TopS_Em03_Rotation Speed}
  NamedRerouteDeclaration(None=Add(A=Multiply(A=@n882, B=@n883), B=@n420)) {name=Uvs_TS_Emis_03_Rotation, nodeColor=<rgba 0 0.7188 1 1>}
  n884 = Subtract(
   A=Time()
   B=Add(
    A=ScalarParameter() {group=TopSide_V02_Layer Emissive_03 Reactive, parameterName=TopS Emissive_03_Brightness---Delay}
    B=ScalarParameter() {defaultValue=-9999999, group=TopSide_V02_Layer Emissive_03 Reactive, parameterName=TopS Emissive_03-Brightness-TBP}
   )
  )
  n885 = ScalarParameter() {defaultValue=0.25, group=TopSide_V02_Layer Emissive_03 Reactive, parameterName=TopS Emissive_03_Brightness---Duration}
  n886 = Power(
   Base=Saturate(None=Divide(A=@n884, B=@n885))
   Exp=ScalarParameter() {defaultValue=0.25, group=TopSide_V02_Layer Emissive_03 Reactive, parameterName=TopS Emissive_03_Brightness---DFalloff}
  )
  n887 = LinearInterpolate(
   A=ScalarParameter() {defaultValue=1, group=TopSide_V02_Layer Emissive_03 Reactive, parameterName=TopS Emissive_03_Brightness---Start}
   B=ScalarParameter() {group=TopSide_V02_Layer Emissive_03 Reactive, parameterName=TopS Emissive_03_Brightness---End}
   Alpha=@n886
  )
  n888 = NamedRerouteUsage()
  n889 = LinearInterpolate(A=@n887, Alpha=Ceil(None=@n888)) {constB=0}
  n890 = Add(
   A=LinearInterpolate(
    A=ScalarParameter() {group=TopSide_V02_Layer Emissive_03 Reactive, parameterName=TopS Audio 03 Emissive_Brightness-Min, sortPriority=210}
    B=ScalarParameter() {group=TopSide_V02_Layer Emissive_03 Reactive, parameterName=TopS Audio 03 Emissive_Brightness_Max, sortPriority=210}
    Alpha=@n888
   )
   B=@n887
  )
  n891 = StaticBoolParameter() {group=TopSide_V02_Layer Emissive_03 Reactive, parameterName=TopS Audio 03 Reactive Ramp Brightness Off, sortPriority=199}
  NamedRerouteDeclaration(None=StaticSwitch(True=StaticSwitch(True=@n889, False=@n890, Value=@n891), False=@n887, Value=NamedRerouteUsage())) {name=Reactive_Brightness_03, nodeColor=<rgba 1 0.1641 0 1>}
  NamedRerouteDeclaration(None=@n886) {name=Reactive_Brightness_03_Time, nodeColor=<rgba 1 0.1641 0 1>}
  n892 = Subtract(
   A=Time()
   B=Add(
    A=ScalarParameter() {group=TopSide_V02_Layer Emissive_03 Reactive, parameterName=TopS Emissive_03_Offset---Delay}
    B=ScalarParameter() {defaultValue=-9999999, group=TopSide_V02_Layer Emissive_03 Reactive, parameterName=TopS Emissive_03_Offset----TBP}
   )
  )
  n893 = ScalarParameter() {defaultValue=0.25, group=TopSide_V02_Layer Emissive_03 Reactive, parameterName=TopS Emissive_03_Offset---Duration}
  n894 = Power(
   Base=Saturate(None=Divide(A=@n892, B=@n893))
   Exp=ScalarParameter() {defaultValue=0.25, group=TopSide_V02_Layer Emissive_03 Reactive, parameterName=TopS Emissive_03_Offset---DFalloff}
  )
  n895 = LinearInterpolate(
   A=ScalarParameter() {group=TopSide_V02_Layer Emissive_03 Reactive, parameterName=TopS Emissive_03_Offset---Start}
   B=ScalarParameter() {group=TopSide_V02_Layer Emissive_03 Reactive, parameterName=TopS Emissive_03_Offset---End}
   Alpha=@n894
  )
  n896 = Add(
   A=LinearInterpolate(
    A=ScalarParameter() {group=TopSide_V02_Layer Emissive_03 Reactive, parameterName=TopS Audio 03 Emissive_Offset---Start, sortPriority=210}
    B=ScalarParameter() {group=TopSide_V02_Layer Emissive_03 Reactive, parameterName=TopS Audio 03 Emissive_Offset---End, sortPriority=210}
    Alpha=NamedRerouteUsage()
   )
   B=@n895
  )
  n897 = NamedRerouteUsage()
  NamedRerouteDeclaration(None=AppendVector(A=StaticSwitch(True=@n896, False=@n895, Value=@n897), B=Constant())) {name=Reactive_Offset_03, nodeColor=<rgba 1 0.1641 0 1>}
  NamedRerouteDeclaration(None=@n894) {name=Reactive_Offset_03_Time, nodeColor=<rgba 1 0.1641 0 1>}
  n898 = ScalarParameter() {group=TopSide_V02_Layer Emissive_03 Reactive, parameterName=TopS Emissive_03_Gradent---Delay}
  n899 = ScalarParameter() {defaultValue=-9999999, group=TopSide_V02_Layer Emissive_03 Reactive, parameterName=TopS Emissive_03_Gradent----TBP}
  n900 = Divide(
   A=Subtract(A=Time(), B=Add(A=@n898, B=@n899))
   B=ScalarParameter() {defaultValue=0.25, group=TopSide_V02_Layer Emissive_03 Reactive, parameterName=TopS Emissive_03_Gradent---Duration}
  )
  NamedRerouteDeclaration(None=Power(
    Base=Saturate(None=@n900)
    Exp=ScalarParameter() {defaultValue=0.25, group=TopSide_V02_Layer Emissive_03 Reactive, parameterName=TopS Emissive_03_Gradent---DFalloff}
   )) {name=TopS Emissive_03_Gradent, nodeColor=<rgba 1 0.1641 0 1>}
  n901 = NamedRerouteUsage()
  n902 = ScalarParameter() {group=TopSide_V02_Layer Emissive_03, parameterName=TopS_Em03_CC_UVMap_Warp_Amount, sortPriority=499}
  n903 = OneMinus(None=@n120)
  n904 = Add(A=NamedRerouteUsage(), B=Add(A=Multiply(A=@n901, B=@n902), B=Saturate(None=@n903)))
  NamedRerouteDeclaration(None=LinearInterpolate(
    A=@n904
    B=Saturate(None=@n904)
    Alpha=ScalarParameter() {group=TopSide_V02_Layer Emissive_03, parameterName="TopS_Em03_CC_UVMap_Warp_Clamp   1=on", sortPriority=499}
   )) {name=Emissive_03_CC_uvs, nodeColor=<rgba 0.9219 1 0 1>}
  n905 = NamedRerouteUsage()
  n906 = NamedRerouteUsage()
  n907 = GetMaterialAttributes(None=NamedRerouteUsage()) {attributeGetTypes [
    #2
    #3
    #5
    #4
    #1
   ]}
  n908 = VectorParameter() {group=TopSide_V02_Layer 01, parameterName=TopS_01_Base Color_Simple, defaultValue=<rgba 0 0 0 1>}
  n909 = NamedRerouteUsage()
  n910 = StaticSwitchParameter(True=ChannelMaskParameter(None=@n909) {group=TopSide_V02_Layer 01, parameterName=TopS_01_Base_Roughness Msk_Channel}.R, False=@n908.A) {desc=Color tint based of maks texture, group=TopSide_V02_Layer 01, parameterName=TopS_01_Base_Roughness Msk-------}
  n911 = NamedRerouteUsage()
  n912 = ComponentMask(None=@n911) {b=true, g=true, r=true}
  n913 = NamedRerouteUsage()
  n914 = StaticSwitchParameter(True=AppendVector(A=Multiply(A=@n912, B=@n913), B=ComponentMask(None=@n911) {a=true}), False=@n911) {desc=Color tint based of maks texture, group=TopSide_V02_Layer 01 -Color_Tint, parameterName=TopS_01_Base Color Tint, sortPriority=0}
  n915 = StaticSwitchParameter(
   True=StaticSwitchParameter(True=NamedRerouteUsage(), False=AppendVector(A=@n908.RGB, B=@n910)) {desc=uses TopS_01_Alternate Axies_Colors, group=TopSide_V02_Layer 01, parameterName="TopS_01_Base Color Simple---->3Colors"}
   False=@n914
  ) {desc=enable to use to allow root pass thru, group=TopSide_V02_Layer 01, parameterName="TopS_01_Base Color Simple      Enable Root Passthrue"}
  n916 = ComponentMask(None=@n915) {b=true, g=true, r=true}
  n917 = StaticSwitchParameter(True=LinearInterpolate(A=@n916, B=@n907.BaseColor, Alpha=@n38), False=@n916) {desc=enable to use to allow root pass thru, group=TopSide_V02_Layer 01, parameterName="TopS_01_Base Color Simple      Enable Root Passthrue"}
  n918 = Multiply(A=StaticSwitch(True=LinearInterpolate(A=@n917, B=@n905, Alpha=@n906), False=@n917, Value=@n711), B=NamedRerouteUsage())
  n919 = Add(
   A=NamedRerouteUsage()
   B=ScalarParameter() {group=TopSide_V02_Layer Emissive_02, parameterName="TopS Emissive_02-Darken_Emissive01 only 1=On"}
  )
  n920 = StaticSwitch(True=Multiply(A=@n918, B=Saturate(None=@n919)), False=@n918, Value=NamedRerouteUsage())
  n921 = Saturate(None=Add(
    A=NamedRerouteUsage()
    B=ScalarParameter() {group=TopSide_V02_Layer Emissive_03, parameterName="TopS Emissive_03-Darken_Emissive01 only 1=On"}
   ))
  n922 = NamedRerouteUsage()
  n923 = NamedRerouteUsage()
  n924 = NamedRerouteUsage()
  n925 = ComponentMask(None=@n915) {a=true}
  n926 = LinearInterpolate(
   A=ScalarParameter() {group=TopSide_V02_Layer 01, parameterName=TopS_01_Base_Roughness_High}
   B=ScalarParameter() {defaultValue=1, group=TopSide_V02_Layer 01, parameterName=TopS_01_Base_Roughness_Low}
   Alpha=@n925
  )
  n927 = StaticSwitchParameter(
   True=LinearInterpolate(
    A=@n924
    B=Multiply(A=@n925, B=@n924)
    Alpha=ScalarParameter() {group=TopSide_V02_Layer 01 -Color_Tint, parameterName="TopS_01_Roughtness------Mutlplyer BC-Alpha  1=on"}
   )
   False=@n926
  ) {
   desc=Color tint based of maks texture
   group=TopSide_V02_Layer 01 -Color_Tint
   parameterName="TopS_01_Roughness\\Metilic\\Spec Tint"
   sortPriority=0
  }
  n928 = StaticSwitchParameter(True=@n927, False=@n926) {desc=Color tint based of maks texture, group=TopSide_V02_Layer 01 -Color_Tint, parameterName=TopS_01_Base Color Tint, sortPriority=0}
  n929 = StaticSwitchParameter(True=LinearInterpolate(A=@n928, B=@n907.Roughness, Alpha=@n38), False=@n928) {desc=enable to use to allow root pass thru, group=TopSide_V02_Layer 01, parameterName="TopS_01_Base Color Simple      Enable Root Passthrue"}
  n930 = NamedRerouteUsage()
  n931 = NamedRerouteUsage()
  n932 = ScalarParameter() {defaultValue=0.5, group=TopSide_V02_Layer 01, parameterName=TopS_01_Base_Spec_}
  n933 = StaticSwitchParameter(True=NamedRerouteUsage(), False=@n932) {
   desc=Color tint based of maks texture
   group=TopSide_V02_Layer 01 -Color_Tint
   parameterName="TopS_01_Roughness\\Metilic\\Spec Tint"
   sortPriority=0
  }
  n934 = StaticSwitchParameter(True=@n933, False=@n932) {desc=Color tint based of maks texture, group=TopSide_V02_Layer 01 -Color_Tint, parameterName=TopS_01_Base Color Tint, sortPriority=0}
  n935 = StaticSwitchParameter(
   True=LinearInterpolate(
    A=ScalarParameter() {group=TopSide_V02_Layer 01, parameterName=TopS_01_Base_Spec_Min}
    B=ScalarParameter() {defaultValue=1, group=TopSide_V02_Layer 01, parameterName=TopS_01_Base_Spec_Max}
    Alpha=ChannelMaskParameter(None=@n909) {group=TopSide_V02_Layer 01, parameterName=TopS_01_Base_Spec_Mask_Channel}.R
   )
   False=@n934
  ) {group=TopSide_V02_Layer 01, parameterName=TopS_01_Base_Spec Mask----------}
  n936 = StaticSwitchParameter(True=LinearInterpolate(A=@n935, B=@n907.Specular, Alpha=@n38), False=@n935) {desc=enable to use to allow root pass thru, group=TopSide_V02_Layer 01, parameterName="TopS_01_Base Color Simple      Enable Root Passthrue"}
  n937 = NamedRerouteUsage()
  n938 = NamedRerouteUsage()
  n939 = ScalarParameter() {group=TopSide_V02_Layer 01, parameterName=TopS_01_Base_Metalic_}
  n940 = StaticSwitchParameter(True=NamedRerouteUsage(), False=@n939) {
   desc=Color tint based of maks texture
   group=TopSide_V02_Layer 01 -Color_Tint
   parameterName="TopS_01_Roughness\\Metilic\\Spec Tint"
   sortPriority=0
  }
  n941 = StaticSwitchParameter(True=@n940, False=@n939) {desc=Color tint based of maks texture, group=TopSide_V02_Layer 01 -Color_Tint, parameterName=TopS_01_Base Color Tint, sortPriority=0}
  n942 = StaticSwitchParameter(
   True=LinearInterpolate(
    A=ScalarParameter() {group=TopSide_V02_Layer 01, parameterName=TopS_01_Base_Metalic_Min}
    B=ScalarParameter() {defaultValue=1, group=TopSide_V02_Layer 01, parameterName=TopS_01_Base_Metalic_Max}
    Alpha=ChannelMaskParameter(None=@n909) {group=TopSide_V02_Layer 01, parameterName=TopS_01_Base_Metalic_Mask_Channel}.R
   )
   False=@n941
  ) {group=TopSide_V02_Layer 01, parameterName=TopS_01_Base_Metalic Mask----------}
  n943 = StaticSwitchParameter(True=LinearInterpolate(A=@n942, B=@n907.Metallic, Alpha=@n38), False=@n942) {desc=enable to use to allow root pass thru, group=TopSide_V02_Layer 01, parameterName="TopS_01_Base Color Simple      Enable Root Passthrue"}
  n944 = NamedRerouteUsage()
  n945 = Add(A=@n944, B=NamedRerouteUsage())
  n946 = ShadingPathSwitch(Default=@n945, Deferred=@n945, Mobile=@n944) {desc=Migrated from FeatureLevelSwitch}
  n947 = StaticSwitchParameter(
   True=@n944
   False=LinearInterpolate(
    A=@n944
    B=StaticSwitchParameter(True=@n946, False=@n945) {group=TopSide_V02_Layer 01, parameterName=TopS_01_Base Normal Android-Disable}
    Alpha=ScalarParameter() {group=TopSide_V02_Layer 02, parameterName="TopS_02_Base Norma--->TopS_01_Normal_Passthrue"}
   )
  ) {defaultValue=true, group=TopSide_V02_Layer 01, parameterName=TopS_01_Base Normal Simple}
  n948 = NamedRerouteUsage()
  n949 = NamedRerouteUsage()
  n950 = LinearInterpolate(
   A=@n949
   B=Add(A=@n949, B=NamedRerouteUsage())
   Alpha=ScalarParameter() {defaultValue=1, group=TopSide_V02_Layer 01, parameterName=TopS_01_Base Normal_Intensty}
  )
  n951 = StaticSwitchParameter(
   True=@n949
   False=StaticSwitchParameter(True=ShadingPathSwitch(Default=@n950, Deferred=@n950, Mobile=@n949) {desc=Migrated from FeatureLevelSwitch}, False=@n950) {group=TopSide_V02_Layer 01, parameterName=TopS_01_Base Normal Android-Disable}
  ) {defaultValue=true, group=TopSide_V02_Layer 01, parameterName=TopS_01_Base Normal Simple}
  n952 = StaticSwitchParameter(True=LinearInterpolate(A=@n951, B=@n907.Normal, Alpha=@n38), False=@n951) {desc=enable to use to allow root pass thru, group=TopSide_V02_Layer 01, parameterName="TopS_01_Base Color Simple      Enable Root Passthrue"}
  n953 = SetMaterialAttributes!(
   MaterialAttributes=NamedRerouteUsage()
   BaseColor=StaticSwitch(True=Multiply(A=@n920, B=@n921), False=@n920, Value=NamedRerouteUsage())
   Roughness=StaticSwitch(True=LinearInterpolate(A=@n929, B=@n922, Alpha=@n923), False=@n929, Value=@n711)
   Specular=StaticSwitch(True=LinearInterpolate(A=@n936, B=@n930, Alpha=@n931), False=@n936, Value=@n711)
   Metallic=StaticSwitch(True=LinearInterpolate(A=@n943, B=@n937, Alpha=@n938), False=@n943, Value=@n711)
   Normal=StaticSwitch(True=LinearInterpolate(A=@n952, B=@n947, Alpha=@n948), False=@n952, Value=@n711)
   EmissiveColor
  ) {attributeSetTypes [
    #2
    #3
    #5
    #4
    #1
    #6
   ]}
  n954 = GetMaterialAttributes(None=@n953) {attributeGetTypes [#2, #6]}
  n955 = ScalarParameter() {group=TopSide_V02, parameterName="TopS BaseColor----->Emissive", sortPriority=-10}
  n956 = LinearInterpolate(A=@n954.BaseColor, Alpha=@n955) {constB=0}
  n957 = Multiply(A=Multiply(A=NamedRerouteUsage(), B=NamedRerouteUsage()), B=NamedRerouteUsage())
  n958 = ScalarParameter() {group=TopSide_V02_Layer Emissive_01, parameterName="TopS Emissive_01 --->BaseColor"}
  n959 = LinearInterpolate(B=@n957, Alpha=@n958)
  n960 = Constant()
  n961 = NamedRerouteUsage()
  n962 = Multiply(A=NamedRerouteUsage(), B=NamedRerouteUsage())
  n963 = ScalarParameter() {group=TopSide_V02_Layer Emissive_02, parameterName="TopS Emissive_02 --->BaseColor"}
  n964 = LinearInterpolate(B=@n962, Alpha=@n963)
  n965 = Constant()
  n966 = NamedRerouteUsage()
  n967 = NamedRerouteUsage()
  n968 = ScalarParameter() {group=TopSide_V02_Layer Emissive_03, parameterName="TopS Emissive_03 --->BaseColor"}
  n969 = Constant()
  n970 = NamedRerouteUsage()
  n971 = Add(
   A=Add(A=StaticSwitch(True=@n959, False=@n960, Value=@n961), B=StaticSwitch(True=@n964, False=@n965, Value=@n966))
   B=StaticSwitch(True=LinearInterpolate(B=@n967, Alpha=@n968), False=@n969, Value=@n970)
  )
  n972 = Min(A=Min(A=NamedRerouteUsage(), B=NamedRerouteUsage()), B=NamedRerouteUsage())
  n973 = LinearInterpolate(
   B=GetMaterialAttributes(None=NamedRerouteUsage()) {attributeGetTypes [#2, #6]}.EmissiveColor
   Alpha=NamedRerouteUsage()
  )
  n974 = LinearInterpolate(A=@n973, Alpha=NamedRerouteUsage()) {constB=0}
  n975 = NamedRerouteUsage()
  n976 = StaticSwitchParameter(True=Multiply(A=@n972, B=StaticSwitch(True=@n974, False=@n973, Value=@n975)), False=Constant()) {desc=enable to use to allow root pass thru, group=TopSide_V02_Layer 01, parameterName="TopS_01_Base Color Simple      Enable Root Passthrue"}
  n977 = LinearInterpolate(B=@n954.BaseColor, Alpha=@n955)
  n978 = Add(A=@n976, B=@n977)
  n979 = Multiply(
   A=NamedRerouteUsage()
   B=ScalarParameter() {defaultValue=1, group=TopSide_V02_Layer 02, parameterName=TopS_02 Emissive Brightness, sortPriority=0}
  )
  n980 = VectorParameter() {
   group=TopSide_V02_Layer 02
   parameterName=TopS_02 Emissive to Root Emissive
   channelNames {a="", b="", g=Root Emis Brightness, r="Add---->.Multpy"}
   defaultValue=<rgba 0 1 0 1>
  }
  n981 = Multiply(A=@n976, B=@n980.Root Emis Brightness)
  n982 = Add(A=@n977, B=LinearInterpolate(A=Add(A=@n976, B=@n979), B=Multiply(A=@n979, B=@n981), Alpha=@n980."Add---->.Multpy"))
  n983 = Multiply(
   A=@n972
   B=StaticSwitch(
    True=@n978
    False=StaticSwitchParameter(True=@n982, False=@n978) {group=TopSide_V02_Layer 02, parameterName="TopS_02_Base Color--->For Emissive", sortPriority=-1}
    Value=NamedRerouteUsage()
   )
  )
  n984 = LinearInterpolate(A=@n957, Alpha=@n958) {constB=0}
  n985 = LinearInterpolate(A=@n962, Alpha=@n963) {constB=0}
  n986 = Add(
   A=Add(A=StaticSwitch(True=@n984, False=@n960, Value=@n961), B=StaticSwitch(True=@n985, False=@n965, Value=@n966))
   B=StaticSwitch(True=LinearInterpolate(A=@n967, Alpha=@n968) {constB=0}, False=@n969, Value=@n970)
  )
  NamedRerouteDeclaration(None=SetMaterialAttributes!(MaterialAttributes=@n954.MaterialAttributes, BaseColor=Add(A=@n956, B=@n971), EmissiveColor=Add(A=@n983, B=@n986)) {attributeSetTypes [#2, #6]}) {name=Mat_Base, nodeColor=<rgba 0 0.3906 1 1>}
  n987 = NamedRerouteUsage()
  n988 = Max(A=@n328, B=@n330)
  n989 = ScalarParameter() {
   defaultValue=1
   group=TopSide_V02_Layer Emissive_03 Reactive
   parameterName="TopS Emissive_03-Brightness-Reactive--->Gradent BloomFix  0=Active"
  }
  n990 = Saturate(None=Ceil(None=Min(A=@n988, B=@n989)))
  n991 = ScalarParameter() {group=TopSide_V02_Layer Emissive_03, parameterName=TopS Emissive_03_ColorCuve_Offset}
  n992 = Multiply(A=ScalarParameter() {group=TopSide_V02_Layer Emissive_03, parameterName=TopS Emissive_03_ColorCuve_Speed}, B=Time())
  n993 = StaticSwitch(
   True=NamedRerouteUsage()
   False=@n339
   Value=StaticBoolParameter() {desc=setting at bottom, group=TopSide_V02_Layer Emissive_03, parameterName="TopS Emissive_03_ColorCuve--->Use-UVS"}
  )
  n994 = ScalarParameter() {defaultValue=1, group=TopSide_V02_Layer Emissive_03, parameterName=TopS Emissive_03_ColorCuve_TileRate}
  n995 = Frac(None=Add(A=Add(A=@n991, B=@n992), B=Multiply(A=@n993, B=@n994)))
  n996 = CurveAtlasRowParameter(CurveTime=@n995) {
   atlas=&CruveAtlas_Wraps_01
   curve=&CC_Spectum
   defaultValue=98
   group=TopSide_V02_Layer Emissive_03
   parameterName=TopS Emissive_03_ColorCuve B
  }
  n997 = MaterialFunctionCall(
   Index=ScalarParameter() {defaultValue=98, group=TopSide_V02_Layer Emissive_03, parameterName=TopS Emissive_03_ColorCuve B Index}
   CCinput=@n996
   CCinput_Alpha=@n996
   UV=@n995
  ) {materialFunction=&MF_Global_CC_Tex}
  n998 = CurveAtlasRowParameter(CurveTime=@n995) {
   atlas=&CruveAtlas_Wraps_01
   curve=&CC_Spectum
   defaultValue=98
   group=TopSide_V02_Layer Emissive_03
   parameterName=TopS Emissive_03_ColorCuve
  }
  n999 = MaterialFunctionCall(
   Index=ScalarParameter() {defaultValue=98, group=TopSide_V02_Layer Emissive_03, parameterName="TopS Emissive_03_ColorCuve  Index"}
   CCinput=@n998
   CCinput_Alpha=@n998
   UV=@n995
  ) {materialFunction=&MF_Global_CC_Tex}
  n1000 = VectorParameter() {group=TopSide_V02_Layer Emissive_03, parameterName=TopS Emissive_03_ColorCuve Tint, defaultValue=<rgba 1 1 1 1>}
  n1001 = Multiply(A=@n999.3V, B=@n1000.RGB)
  n1002 = NamedRerouteUsage()
  n1003 = StaticBoolParameter() {group=TopSide_V02_Layer Emissive_03, parameterName=TopS Emissive_03_Color TOD}
  n1004 = StaticSwitch(True=LinearInterpolate(A=@n997.3V, B=@n1001, Alpha=@n1002), False=@n1001, Value=@n1003)
  n1005 = ScalarParameter() {defaultValue=1, group=TopSide_V02_Layer Emissive_03, parameterName=TopS Emissive_03_ColorCuve-Brightness}
  n1006 = StaticSwitch(True=LinearInterpolate(A=@n997.Alpha, B=@n999.Alpha, Alpha=@n1002), False=@n999.Alpha, Value=@n1003)
  n1007 = LinearInterpolate(A=@n1005, B=Multiply(A=@n1006, B=@n1005), Alpha=NamedRerouteUsage())
  n1008 = Add(A=ScalarParameter() {group=TopSide_V02_Layer Emissive_03, parameterName="TopS Emissive_03_ColorCuve-CurveMask   0=On"}, B=@n339)
  n1009 = VectorParameter() {group=TopSide_V02_Layer Emissive_03, parameterName=TopS Emissive_03_Color-B, defaultValue=<rgba 0 0 0 1>}
  n1010 = VectorParameter() {group=TopSide_V02_Layer Emissive_03, parameterName=TopS Emissive_03_Color-, defaultValue=<rgba 0 0 0 1>}
  n1011 = LinearInterpolate(A=@n1009.RGB, B=@n1010.RGB, Alpha=NamedRerouteUsage())
  n1012 = StaticSwitchParameter(
   True=Multiply(A=Multiply(A=@n1004, B=@n1007), B=Saturate(None=@n1008))
   False=Multiply(A=@n339, B=StaticSwitch(True=@n1011, False=@n1010.RGB, Value=@n1003))
  ) {group=TopSide_V02_Layer Emissive_03, parameterName="TopS Emissive_03_Color ----->ColorCurve"}
  n1013 = StaticSwitch(True=Multiply(A=@n1012, B=NamedRerouteUsage()), False=@n1012, Value=NamedRerouteUsage())
  n1014 = StaticSwitch(
   True=Multiply(A=LinearInterpolate(A=@n987, Alpha=@n990), B=@n1013)
   False=@n1013
   Value=NamedRerouteDeclaration(None=StaticBoolParameter() {group=TopSide_V02_Layer Emissive_03 Reactive, parameterName=TopS Emissive_03-Brightness-Reactive--------------}) {name=Emissive_03_Reactive_bool_Brightness, nodeColor=<rgba 0.3594 0 1 1>}
  )
  n1015 = NamedRerouteUsage()
  n1016 = LinearInterpolate(
   A=@n1015
   B=OneMinus(None=@n1015)
   Alpha=ScalarParameter() {group=TopSide_V02_Layer Emissive_03, parameterName="TopS Emissive_03_Blend 01---> 02 Invert Msk  1=inverted", sortPriority=1}
  )
  n1017 = StaticBoolParameter() {group=TopSide_V02_Layer Emissive_03, parameterName="TopS Emissive_03_Blend 01---> 02 Affects Mask", sortPriority=1}
  n1018 = StaticSwitch(True=StaticSwitch(True=Multiply(A=@n1014, B=@n1016), False=@n1014, Value=@n1017), False=@n1014, Value=NamedRerouteUsage())
  n1019 = StaticSwitch(
   True=NamedRerouteUsage()
   False=NamedRerouteUsage()
   Value=StaticBoolParameter() {group=TopSide_V02_Layer Emissive_03 Reactive, parameterName="TopS Emissive_03-TOD---->TODB", sortPriority=0}
  )
  n1020 = StaticBoolParameter() {group=TopSide_V02_Layer Emissive_03 Reactive, parameterName=TopS Emissive_03-TOD, sortPriority=0}
  n1021 = StaticSwitch(True=LinearInterpolate(B=@n1018, Alpha=@n1019) {constB=0}, False=@n1018, Value=@n1020)
  n1022 = VectorParameter() {
   group=TopSide_V02_Layer Emissive_03
   parameterName=TopS Emissive_03 LiquidCCMask
   channelNames {a="", b="", g=invert, r="0=Use"}
   defaultValue=<rgba 1 0 0 1>
  }
  n1023 = ChannelMaskParameter(None=NamedRerouteUsage()) {group=TopSide_V02_Layer Emissive_03, maskChannel=Alpha, parameterName=TopS Emissive_03 LiquidCCMask Channel, defaultValue=<rgba 0 0 0 1>}
  n1024 = Max(A=@n1022."0=Use", B=LinearInterpolate(A=@n1023.R, B=OneMinus(None=@n1023.R), Alpha=@n1022.invert))
  n1025 = NamedRerouteUsage()
  NamedRerouteDeclaration(None=StaticSwitch(True=LinearInterpolate(B=@n1021, Alpha=@n1024), False=@n1021, Value=@n1025)) {name=Emissive_03, nodeColor=<rgba 1 0.8203 0 1>}
  n1026 = OneMinus(None=@n1024)
  n1027 = ScalarParameter() {
   defaultValue=1
   desc=CC brightnes gets disabled
   group=TopSide_V02_Layer Emissive_03
   parameterName="TopS Emissive_03_ColorCuve=A-Effect_DarkenBase  0=ON"
  }
  n1028 = NamedRerouteDeclaration(None=Saturate(None=Ceil(None=@n1027))) {name=EmiS 03 CC A State, nodeColor=<rgba 0 0.8828 1 1>}
  n1029 = LinearInterpolate(
   A=ScalarParameter() {group=TopSide_V02_Layer Emissive_03 Reactive, parameterName=TopS Emissive_03_Brightness_Darken_Base Start}
   B=ScalarParameter() {group=TopSide_V02_Layer Emissive_03 Reactive, parameterName=TopS Emissive_03_Brightness_Darken_Base End}
   Alpha=NamedRerouteUsage()
  )
  n1030 = NamedRerouteUsage()
  n1031 = LinearInterpolate(A=@n1029, Alpha=Ceil(None=@n1030)) {constB=0}
  n1032 = ScalarParameter() {group=TopSide_V02_Layer Emissive_03 Reactive, parameterName=TopS Audio 03 Emissive_Brightness Darken_Base_Max, sortPriority=211}
  n1033 = Max(
   A=LinearInterpolate(
    A=ScalarParameter() {group=TopSide_V02_Layer Emissive_03 Reactive, parameterName=TopS Audio 03 Emissive_Brightness Darken_Bas-Min, sortPriority=211}
    B=@n1032
    Alpha=@n1030
   )
   B=@n1029
  )
  n1034 = NamedRerouteUsage()
  n1035 = StaticSwitch(
   True=StaticSwitch(True=StaticSwitch(True=@n1031, False=@n1033, Value=@n1034), False=@n1029, Value=NamedRerouteUsage())
   False=ScalarParameter() {group=TopSide_V02_Layer Emissive_03, parameterName="TopS Emissive_03-Darken_Base    1=On"}
   Value=NamedRerouteUsage()
  )
  n1036 = StaticSwitchParameter(True=Multiply(A=LinearInterpolate(A=@n1006, B=@n339, Alpha=@n1028), B=@n1035), False=Multiply(A=@n339, B=@n1035)) {group=TopSide_V02_Layer Emissive_03, parameterName="TopS Emissive_03_Color ----->ColorCurve"}
  n1037 = StaticSwitch(True=Multiply(A=@n1016, B=@n1036), False=@n1036, Value=@n1017)
  n1038 = NamedRerouteUsage()
  n1039 = OneMinus(None=Saturate(None=StaticSwitch(True=@n1037, False=@n1036, Value=@n1038)))
  n1040 = StaticSwitch(True=StaticSwitch(True=Max(A=@n1026, B=@n1039), False=@n1039, Value=@n1025), False=Constant() {r=1}, Value=NamedRerouteUsage())
  n1041 = LinearInterpolate(B=@n1040, Alpha=@n1019) {constA=1, constB=0}
  n1042 = StaticBoolParameter() {group=TopSide_V02_Layer Emissive_03 Reactive, parameterName=TopS Emissive_03-TOD DoNotAffectDarkenBase, sortPriority=0}
  NamedRerouteDeclaration(None=StaticSwitch(True=StaticSwitch(True=@n1040, False=@n1041, Value=@n1042), False=@n1040, Value=@n1020)) {name=Emissive_03_DarkenBAse, nodeColor=<rgba 1 0.8203 0 1>}
  NamedRerouteDeclaration(None=@n340) {name=Emissive_03_bool, nodeColor=<rgba 0.3594 0 1 1>}
  n1043 = NamedRerouteUsage()
  n1044 = StaticSwitchParameter(True=NamedRerouteUsage(), False=NamedRerouteUsage()) {group=TopSide_V02_Layer Emissive_03, parameterName="TopS_01_Em03_Warp-TS 01---->TS 04", sortPriority=200}
  n1045 = ChannelMaskParameter(None=StaticSwitchParameter(
    True=NamedRerouteUsage()
    False=StaticSwitchParameter(True=@n1043, False=@n1044) {group=TopSide_V02_Layer Emissive_03, parameterName="TopS_01_Em03_Warp-TS 01---->TS 03", sortPriority=200}
   ) {group=TopSide_V02_Layer Emissive_03, parameterName="TopS_01_Em03_Warp-TS 01---->TS 02", sortPriority=200}) {group=TopSide_V02_Layer Emissive_03, parameterName=TopS_Em03_CC_UVMap_Warp_Channel, sortPriority=201}
  n1046 = Constant()
  NamedRerouteDeclaration(None=StaticSwitchParameter(
    True=StaticSwitchParameter(True=@n1045.R, False=@n307) {group=TopSide_V02_Layer Emissive_03, parameterName=TopS_Em03_CC_UVMap_Warp-Channel--Alt, sortPriority=201}
    False=@n1046
   ) {group=TopSide_V02_Layer Emissive_03, parameterName=TopS Emissive_03_Warp, sortPriority=200}) {name=Emissive03_warp, nodeColor=<rgba 0.5 1 0 1>}
  n1047 = CollectionParameter() {collection=&FortniteMaterialParameters, parameterName=EventAudio_Spectrum}
  n1048 = ChannelMaskParameter(None=@n1047) {
   group=TopSide_V02_Layer Emissive_03 Reactive
   maskChannel=Alpha
   parameterName=TopS Audio 03 Channel
   sortPriority=300
   defaultValue=<rgba 0 0 0 1>
  }
  n1049 = CollectionParameter() {collection=&FortniteMaterialParameters, parameterName=Emote_AudioLow}
  n1050 = MaterialFunctionCall(
   Input=Max(A=@n1048.R, B=@n1049)
   Input Low=ScalarParameter() {defaultValue=0.25, group=TopSide_V02_Layer Emissive_03 Reactive, parameterName=TopS Audio 03_Low_Remap, sortPriority=300}
   Input High=ScalarParameter() {defaultValue=1, group=TopSide_V02_Layer Emissive_03 Reactive, parameterName=TopS Audio 03_High_Remap, sortPriority=300}
   Target Low=Constant()
   Target High=Constant() {r=1}
  ) {materialFunction=&RemapValueRange}
  n1051 = CollectionParameter() {collection=&FortniteMaterialParameters, parameterName=Emote_AudioHigh}
  n1052 = Power(
   Base=Max(A=Saturate(None=@n1050.Result), B=@n1051)
   Exp=ScalarParameter() {defaultValue=1, group=TopSide_V02_Layer Emissive_03 Reactive, parameterName=TopS Audio 03 Falloff, sortPriority=300}
  )
  NamedRerouteDeclaration(None=@n1052) {name=AudioSpectrum 03, nodeColor=<rgba 1 0.8203 0 1>}
  NamedRerouteDeclaration(None=StaticBoolParameter() {group=TopSide_V02_Layer Emissive_03 Reactive, parameterName=TopS Audio 03 Reactive, sortPriority=199}) {name=Audio_03, nodeColor=<rgba 0 0.4141 1 1>}
  n1053 = Time()
  n1054 = ScalarParameter() {defaultValue=2, group=TopSide_V02_Layer Emissive_03 Reactive, parameterName=TopS Audio 03 Static_Speed, sortPriority=300}
  n1055 = ConstantBiasScale(None=Sine(None=Multiply(A=@n1053, B=@n1054)))
  n1056 = Multiply(
   A=ScalarParameter() {group=TopSide_V02_Layer Emissive_03 Reactive, parameterName=TopS Audio 03 Static_Strength, sortPriority=300}
   B=NamedRerouteUsage()
  )
  NamedRerouteDeclaration(None=Max(A=@n1052, B=Multiply(A=@n1055, B=@n1056))) {name=AudioSpectrum 03_Static_Speed_1, nodeColor=<rgba 1 0.8203 0 1>}
  n1057 = TextureObjectParameter() {
   group=TopSide_V02_Layer 01
   parameterName=TopS_01__Tex_Normal
   samplerType=SAMPLERTYPE_Normal
   texture=&FlatNormalBlue
  }
  NamedRerouteDeclaration(None=@n1057) {name=Tex_TS01_Normal, nodeColor=<rgba 1 0.8906 0 1>}
  NamedRerouteDeclaration(None=@n432) {name=Uvs_TS_Layer_04_Rotation, nodeColor=<rgba 1 0 0.3281 1>}
  n1058 = VectorParameter() {group=TopSide_V02_Layer-04_Uvs, parameterName=TopS_04_Scale_Weapon, defaultValue=<rgba 1 1 1 1>}
  n1059 = VectorParameter() {group=TopSide_V02_Layer-04_Uvs, parameterName=TopS_04_Scale_Vehicle, defaultValue=<rgba 1 1 1 1>}
  n1060 = Multiply(A=@n1059.RGB, B=@n1059.RGB)
  n1061 = VectorParameter() {group=TopSide_V02_Layer-04_Uvs, parameterName=TopS_04_Scale_VehicleSide, defaultValue=<rgba 1 1 1 1>}
  n1062 = Multiply(A=@n1061.RGB, B=@n1061.RGB)
  n1063 = MaterialFunctionCall(Float4=NamedRerouteUsage()) {materialFunction=&BreakOutFloat4Components}
  n1064 = VectorParameter() {group=TopSide_V02_Layer-04_Uvs, parameterName=TopS_04_Scale_Mech, defaultValue=<rgba 1 1 1 1>}
  n1065 = VectorParameter() {group=TopSide_V02_Layer-04_Uvs, parameterName=TopS_04_Scale_Character, defaultValue=<rgba 1 1 1 1>}
  n1066 = LinearInterpolate(
   A=LinearInterpolate(A=Multiply(A=@n1058.RGB, B=@n1058.RGB), B=LinearInterpolate(A=@n1060, B=@n1062, Alpha=@n1063.G), Alpha=@n1063.R)
   B=LinearInterpolate(A=Multiply(A=@n1064.RGB, B=@n1064.RGB), B=Multiply(A=@n1065.RGB, B=@n1065.RGB), Alpha=@n1063.A)
   Alpha=@n1063.B
  )
  n1067 = VectorParameter() {group=TopSide_V02_Layer-04_Uvs, parameterName=TopS_04_Scale_Backpack, defaultValue=<rgba 1 1 1 1>}
  n1068 = Multiply(A=@n1067.RGB, B=@n1067.RGB)
  n1069 = MaterialFunctionCall(Float2=NamedRerouteUsage()) {materialFunction=&BreakOutFloat2Components}
  n1070 = VectorParameter() {group=TopSide_V02_Layer-04_Uvs, parameterName=TopS_04_Scale_Pickaxe, defaultValue=<rgba 1 1 1 1>}
  NamedRerouteDeclaration(None=LinearInterpolate(A=LinearInterpolate(A=@n1066, B=@n1068, Alpha=@n1069.R), B=Multiply(A=@n1070.RGB, B=@n1070.RGB), Alpha=@n1069.G)) {name=Uvs_TS_Layer_04_Scale, nodeColor=<rgba 0 0.625 1 1>}
  NamedRerouteDeclaration(None=@n12) {name=Uvs_TS_Layer_04_Offset, nodeColor=<rgba 0 0.625 1 1>}
  NamedRerouteDeclaration(None=@n438) {name=Uvs_TS_Layer_04_Warp_Rotation, nodeColor=<rgba 0 0.7188 1 1>}
  n1071 = ScalarParameter() {defaultValue=1, group=TopSide_V02_Layer-04_Uvs | Warp, parameterName=TopS_04_Warp_Scale_Vehicle, sortPriority=300}
  n1072 = ScalarParameter() {defaultValue=1, group=TopSide_V02_Layer-04_Uvs | Warp, parameterName=TopS_04_Warp_Scale_Vehicle_Side, sortPriority=300}
  n1073 = MaterialFunctionCall(Float4=NamedRerouteUsage()) {materialFunction=&BreakOutFloat4Components}
  n1074 = LinearInterpolate(
   A=LinearInterpolate(
    A=ScalarParameter() {defaultValue=1, group=TopSide_V02_Layer-04_Uvs | Warp, parameterName=TopS_04_Warp_Scale_Weapon, sortPriority=300}
    B=LinearInterpolate(A=@n1071, B=@n1072, Alpha=@n1073.G)
    Alpha=@n1073.R
   )
   B=LinearInterpolate(
    A=ScalarParameter() {defaultValue=1, group=TopSide_V02_Layer-04_Uvs | Warp, parameterName=TopS_04_Warp_Scale_Mech, sortPriority=300}
    B=ScalarParameter() {defaultValue=1, group=TopSide_V02_Layer-04_Uvs | Warp, parameterName=TopS_04_Warp_Scale_Character, sortPriority=300}
    Alpha=@n1073.A
   )
   Alpha=@n1073.B
  )
  n1075 = ScalarParameter() {defaultValue=1, group=TopSide_V02_Layer-04_Uvs | Warp, parameterName=TopS_04_Warp_Scale_Backpack, sortPriority=300}
  n1076 = MaterialFunctionCall(Float2=NamedRerouteUsage()) {materialFunction=&BreakOutFloat2Components}
  NamedRerouteDeclaration(None=LinearInterpolate(
    A=LinearInterpolate(A=@n1074, B=@n1075, Alpha=@n1076.R)
    B=ScalarParameter() {defaultValue=1, group=TopSide_V02_Layer-04_Uvs | Warp, parameterName=TopS_04_Warp_Scale_Pickaxe, sortPriority=300}
    Alpha=@n1076.G
   )) {name=Uvs_TS_Layer_04_Warp_Scale, nodeColor=<rgba 0 0.7188 1 1>}
  n1077 = Frac(None=@n14)
  n1078 = ScalarParameter() {defaultValue=10, group=TopSide_V02_Layer-04_Uvs, parameterName=TopS_04_PanningS_Speed_Step_Amount}
  n1079 = Divide(A=Ceil(None=Multiply(A=@n1077, B=@n1078) {constB=30}), B=@n1078)
  NamedRerouteDeclaration(None=Add(A=NamedRerouteUsage(), B=LinearInterpolate(A=@n1079, B=@n14, Alpha=@n13."0=step animation"))) {name=Uvs_TS_Layer_04_Speed, nodeColor=<rgba 0 0.625 1 1>}
  NamedRerouteDeclaration(None=@n444) {name=Uvs_TS_Layer_04_Rotation_Side_1, nodeColor=<rgba 0 0.7188 1 1>}
  n1080 = VectorParameter() {group=TopSide_V02_Layer-04_Uvs, parameterName=TopS_04_OffsetS_Vehicle, defaultValue=<rgba 0 0 0 1>}
  n1081 = VectorParameter() {group=TopSide_V02_Layer-04_Uvs, parameterName=TopS_04_OffsetS_VehicleSide, defaultValue=<rgba 0 0 0 1>}
  n1082 = MaterialFunctionCall(Float4=NamedRerouteUsage()) {materialFunction=&BreakOutFloat4Components}
  n1083 = LinearInterpolate(
   A=LinearInterpolate(
    A=VectorParameter() {group=TopSide_V02_Layer-04_Uvs, parameterName=TopS_04_OffsetS_Weapon, defaultValue=<rgba 0 0 0 1>}.RGB
    B=LinearInterpolate(A=@n1080.RGB, B=@n1081.RGB, Alpha=@n1082.G)
    Alpha=@n1082.R
   )
   B=LinearInterpolate(
    A=VectorParameter() {group=TopSide_V02_Layer-04_Uvs, parameterName=TopS_04_OffsetS_Mech, defaultValue=<rgba 0 0 0 1>}.RGB
    B=VectorParameter() {group=TopSide_V02_Layer-04_Uvs, parameterName=TopS_04_OffsetS_Character, defaultValue=<rgba 0 0 0 1>}.RGB
    Alpha=@n1082.A
   )
   Alpha=@n1082.B
  )
  n1084 = VectorParameter() {group=TopSide_V02_Layer-04_Uvs, parameterName=TopS_04_OffsetS_Backpack, defaultValue=<rgba 0 0 0 1>}
  n1085 = MaterialFunctionCall(Float2=NamedRerouteUsage()) {materialFunction=&BreakOutFloat2Components}
  NamedRerouteDeclaration(None=LinearInterpolate(
    A=LinearInterpolate(A=@n1083, B=@n1084.RGB, Alpha=@n1085.R)
    B=VectorParameter() {group=TopSide_V02_Layer-04_Uvs, parameterName=TopS_04_OffsetS_Pickaxe, defaultValue=<rgba 0 0 0 1>}.RGB
    Alpha=@n1085.G
   )) {name=Uvs_TS_Layer_04_Offset_Side, nodeColor=<rgba 0 0.625 1 1>}
  n1086 = VectorParameter() {group=TopSide_V02_Layer-04_Uvs, parameterName=TopS_04_ScaleS_Weapon, defaultValue=<rgba 1 1 1 1>}
  n1087 = VectorParameter() {group=TopSide_V02_Layer-04_Uvs, parameterName=TopS_04_ScaleS_Vehicle, defaultValue=<rgba 1 1 1 1>}
  n1088 = Multiply(A=@n1087.RGB, B=@n1087.RGB)
  n1089 = VectorParameter() {group=TopSide_V02_Layer-04_Uvs, parameterName=TopS_04_ScaleS_VehicleSide, defaultValue=<rgba 1 1 1 1>}
  n1090 = Multiply(A=@n1089.RGB, B=@n1089.RGB)
  n1091 = MaterialFunctionCall(Float4=NamedRerouteUsage()) {materialFunction=&BreakOutFloat4Components}
  n1092 = VectorParameter() {group=TopSide_V02_Layer-04_Uvs, parameterName=TopS_04_ScaleS_Mech, defaultValue=<rgba 1 1 1 1>}
  n1093 = VectorParameter() {group=TopSide_V02_Layer-04_Uvs, parameterName=TopS_04_ScaleS_Character, defaultValue=<rgba 1 1 1 1>}
  n1094 = LinearInterpolate(
   A=LinearInterpolate(A=Multiply(A=@n1086.RGB, B=@n1086.RGB), B=LinearInterpolate(A=@n1088, B=@n1090, Alpha=@n1091.G), Alpha=@n1091.R)
   B=LinearInterpolate(A=Multiply(A=@n1092.RGB, B=@n1092.RGB), B=Multiply(A=@n1093.RGB, B=@n1093.RGB), Alpha=@n1091.A)
   Alpha=@n1091.B
  )
  n1095 = VectorParameter() {group=TopSide_V02_Layer-04_Uvs, parameterName=TopS_04_ScaleS_Backpack, defaultValue=<rgba 1 1 1 1>}
  n1096 = Multiply(A=@n1095.RGB, B=@n1095.RGB)
  n1097 = MaterialFunctionCall(Float2=NamedRerouteUsage()) {materialFunction=&BreakOutFloat2Components}
  n1098 = VectorParameter() {group=TopSide_V02_Layer-04_Uvs, parameterName=TopS_04_ScaleS_Pickaxe, defaultValue=<rgba 1 1 1 1>}
  NamedRerouteDeclaration(None=LinearInterpolate(A=LinearInterpolate(A=@n1094, B=@n1096, Alpha=@n1097.R), B=Multiply(A=@n1098.RGB, B=@n1098.RGB), Alpha=@n1097.G)) {name=Uvs_TS_Layer_04_Scale_Side, nodeColor=<rgba 0 0.625 1 1>}
  n1099 = Time()
  n1100 = VectorParameter() {
   group=TopSide_V02_Layer-04_Uvs
   parameterName=TopS_04_Pannings_Speed_B
   channelNames {a="0=step Animation", b="", g="", r=""}
   defaultValue=<rgba 0 0 0 1>
  }
  n1101 = Multiply(A=@n1099, B=@n1100)
  n1102 = Frac(None=@n1101)
  n1103 = ScalarParameter() {defaultValue=10, group=TopSide_V02_Layer-04_Uvs, parameterName=TopS_04_PanningS_Speed_B_Step_Amount}
  n1104 = Divide(A=Ceil(None=Multiply(A=@n1102, B=@n1103) {constB=30}), B=@n1103)
  n1105 = VectorParameter() {
   group=TopSide_V02_Layer-04_Uvs
   parameterName=TopS_04_PanningS_Speed------------
   channelNames {a="0=Step Animation", b="", g="", r=""}
   defaultValue=<rgba 0 0 0 1>
  }
  n1106 = Multiply(A=@n1099, B=@n1105)
  n1107 = ScalarParameter() {defaultValue=10, group=TopSide_V02_Layer-04_Uvs, parameterName=TopS_04_PanningS_Speed--Step_Amount}
  n1108 = Multiply(A=Frac(None=@n1106), B=@n1107) {constB=30}
  n1109 = LinearInterpolate(A=Divide(A=Ceil(None=@n1108), B=@n1107), B=@n1106, Alpha=@n1105."0=Step Animation")
  n1110 = StaticSwitchParameter(
   True=LinearInterpolate(A=LinearInterpolate(A=@n1104, B=@n1101, Alpha=@n1100."0=step Animation"), B=@n1109, Alpha=NamedRerouteUsage())
   False=@n1109
  ) {group=TopSide_V02_Layer-04_Uvs, parameterName=TopS_04_PanningS_Speed--Varation-Msk}
  n1111 = VectorParameter() {group=TopSide_V02_Layer-04_Uvs_Reactive, parameterName=TopS_04_Speed_offset_Amount, sortPriority=100, defaultValue=<rgba 0 0 0 1>}
  n1112 = Time()
  n1113 = Add(
   A=ScalarParameter() {group=TopSide_V02_Layer-04_Uvs_Reactive, parameterName=TopS_04_Speed---Delay, sortPriority=100}
   B=ScalarParameter() {defaultValue=-9999999, group=TopSide_V02_Layer-04_Uvs_Reactive, parameterName=TopS_04_Speed-TBP, sortPriority=100}
  )
  n1114 = Saturate(None=Divide(
    A=Subtract(A=@n1112, B=@n1113)
    B=ScalarParameter() {defaultValue=1, group=TopSide_V02_Layer-04_Uvs_Reactive, parameterName=TopS_04_Speed---Duration, sortPriority=100}
   ))
  n1115 = ScalarParameter() {defaultValue=1, group=TopSide_V02_Layer-04_Uvs_Reactive, parameterName=TopS_04_Speed---DFalloff, sortPriority=100}
  n1116 = LinearInterpolate(B=@n1111.RGB, Alpha=OneMinus(None=Power(Base=@n1114, Exp=@n1115)))
  n1117 = Add(A=@n1110, B=@n1116)
  NamedRerouteDeclaration(None=@n1117) {name=Uvs_TS_Layer_04_Speed_Side, nodeColor=<rgba 0 0.625 1 1>}
  NamedRerouteDeclaration(None=StaticSwitchParameter(
    True=NamedRerouteUsage()
    False=TextureObjectParameter() {group=TopSide_V02_Layer 02, parameterName=TopS_02__Tex_BaseColor, texture=/Game/Weapons/WeaponSkins/Wraps/Textures/UVCheaker}
   ) {group=TopSide_V02_Layer 02, parameterName=TopS_02_Share_BC_01_Tex, sortPriority=0}) {name=Tex_TS02_BaseColor, nodeColor=<rgba 1 0.8906 0 1>}
  NamedRerouteDeclaration(None=@n780) {name=TopS_01 ProjectionState, nodeColor=<rgba 0 1 0.8359 1>}
  n1118 = Time()
  n1119 = ScalarParameter() {group=TopSide_V02_Layer-04_Uvs, parameterName=TopS_04_Side_Rotation__Speed}
  NamedRerouteDeclaration(None=Add(A=Multiply(A=@n1118, B=@n1119), B=@n450)) {name=Uvs_TS_Layer_04_Rotation_Side, nodeColor=<rgba 0 0.7188 1 1>}
  n1120 = NamedRerouteUsage()
  n1121 = StaticSwitch(
   True=StaticBool()
   False=StaticBoolParameter() {group=TopSide_V02_Layer-04_Uvs, parameterName=TopS_04_Share_UVs_TopS_01, sortPriority=0}
   Value=NamedRerouteUsage()
  )
  n1122 = Add(
   A=Add(A=NamedRerouteUsage(), B=Constant3Vector() {constant=<rgba 0 0 -1 0>})
   B=StaticSwitch(True=NamedRerouteUsage(), False=NamedRerouteUsage(), Value=@n1121)
  )
  n1123 = NamedRerouteUsage()
  n1124 = MaterialFunctionCall(Float4=NamedRerouteUsage()) {materialFunction=&BreakOutFloat4Components}
  n1125 = LinearInterpolate(
   A=ScalarParameter() {group=TopSide_V02_Layer 04, parameterName=TopS_04 Blend_Contrast}
   B=ScalarParameter() {group=TopSide_V02_Layer 04, parameterName="TopS_04 Blend_Contrast \\Mech"}
   Alpha=@n1124.B
  )
  n1126 = NamedRerouteUsage()
  n1127 = NamedRerouteUsage()
  n1128 = LinearInterpolate(
   A=ScalarParameter() {group=TopSide_V02_Layer 04, parameterName=TopS_04_Alternate Axies Opacity}
   B=ScalarParameter() {group=TopSide_V02_Layer 04, parameterName="TopS_04_Alternate Axies Opacity \\Mech"}
   Alpha=@n1124.B
  )
  n1129 = VectorParameter() {group=TopSide_V02_Layer 04, parameterName=TopS_04_Alternate_Axies Mask, defaultValue=<rgba 0 0 0 1>}
  n1130 = AppendVector(A=@n1129.RGB, B=@n1129.RGB)
  n1131 = StaticSwitch(True=NamedRerouteUsage(), False=NamedRerouteUsage(), Value=@n1121)
  n1132 = StaticSwitch(True=NamedRerouteUsage(), False=NamedRerouteUsage(), Value=@n1121)
  n1133 = NamedRerouteUsage()
  n1134 = StaticSwitch(
   True=NamedRerouteUsage()
   False=StaticBoolParameter() {group=TopSide_V02_Layer 04, parameterName="TopS_04 TopSide->Side projection"}
   Value=@n1121
  )
  NamedRerouteDeclaration(None=StaticSwitch(
    True=NamedRerouteUsage()
    False=MaterialFunctionCall(
     Texture=@n1120
     Caculated_Bounds=@n1122
     Pre_Skinned_Normals=@n1123
     Blend_Contrast=@n1125
     IsStatsB=@n1126
     IsState=@n1127
     Fill_alternate_Axis_Opacity=@n1128
     "Fill_alternate_Axis_Color_A=Roughness"=@n1130
     Offset=@n1133
     Scale=@n1131
     Rotation_Axis=@n1132
    ) {materialFunction=&MF_TopsSideProjection_Base_V2}.Combined
    Value=@n1134
   )) {name=TS_Layer04 Mask, nodeColor=<rgba 0 1 0.9766 1>}
  NamedRerouteDeclaration(None=@n1125) {name=Tops 04 Blend Contrats, nodeColor=<rgba 0 1 0.6953 1>}
  NamedRerouteDeclaration(None=@n1128) {name=Tops 04 Alternate Axies, nodeColor=<rgba 1 0.3984 0 1>}
  NamedRerouteDeclaration(None=@n1130) {name=Tops 04 Alternate Mask, nodeColor=<rgba 1 0.3984 0 1>}
  NamedRerouteDeclaration(None=@n1134) {name=TopS_01 ProjectionState_2, nodeColor=<rgba 0 1 0.8359 1>}
  n1135 = NamedRerouteUsage()
  n1136 = NamedRerouteUsage()
  n1137 = NamedRerouteUsage()
  n1138 = NamedRerouteUsage()
  n1139 = NamedRerouteUsage()
  n1140 = NamedRerouteUsage()
  n1141 = NamedRerouteUsage()
  n1142 = StaticBoolParameter() {group=TopSide_V02_Layer-04_Uvs, parameterName=TopS_04 Clamp Uvs}
  n1143 = ScalarParameter() {defaultValue=0.001, group=TopSide_V02_Layer-04_Uvs, parameterName=TopS_04-Clamp Uvs padding Comp}
  n1144 = MaterialFunctionCall(
   Pivot=@n1135
   "Pivot_L\\R"=StaticBool() {value=true}
   Caculated_Bounds=@n1136
   Pre_Skinned_Normals=@n1137
   Is State=@n1138
   Uvs_Rotation=@n1139
   Uvs_TileRate=@n1140
   Uvs_Offset=@n1141
   Clamp Y=@n1142
   Clamp X=@n1142
   Clamp Padding_Comp=@n1143
  ) {materialFunction=&MF_Projection_Side_Left_Right_Decal_NormalV2}
  n1145 = NamedRerouteUsage()
  n1146 = StaticSwitch(True=NamedRerouteUsage(), False=@n1139, Value=@n1145)
  n1147 = StaticSwitch(True=NamedRerouteUsage(), False=@n1140, Value=@n1145)
  n1148 = StaticSwitch(True=NamedRerouteUsage(), False=@n1141, Value=@n1145)
  n1149 = StaticSwitch(
   True=NamedRerouteUsage()
   False=StaticBoolParameter() {desc=use thiis when animation Rotation, group=TopSide_V02_Layer-04_Uvs, parameterName=TopS_04-Rotation-Comp}
   Value=@n1145
  )
  n1150 = StaticSwitch(True=NamedRerouteUsage(), False=StaticBoolParameter() {group=TopSide_V02_Layer-04_Uvs, parameterName=TopS_04-No_Rotation}, Value=@n1145)
  n1151 = MaterialFunctionCall(
   Caculated_Bounds=@n1136
   Pre_Skinned_Normals=@n1137
   is_State=@n1138
   UV_Rotation=@n1146
   Uvs_Offset=@n1135
   Uvs_TileRate=@n1147
   Uvs_Offset_Pivot=@n1148
   Rotation_Comp=@n1149
   Clamp Y=@n1142
   Clamp X=@n1142
   Clamp Padding_Comp=@n1143
   No_Rotation=@n1150
  ) {materialFunction=&MF_Projection_Side_Normal_V2}
  n1152 = StaticBoolParameter() {group=TopSide_V02_Layer 04, parameterName="TopS_04 TopSide->Side projection->Left_Right"}
  n1153 = StaticSwitch(True=@n1144.UVs, False=@n1151.UVs, Value=@n1152)
  n1154 = NamedRerouteUsage()
  n1155 = TextureSample!(
   UVs=@n1153
   Tex=@n1154
   Level=ScalarParameter() {defaultValue=3, group=TopSide_V02_Layer 04, parameterName=TopS_04-Miplevels_}
   Apply View MipBias
  ) {mipValueMode=TMVM_MipLevel}
  n1156 = StaticSwitch(True=@n1144.UVs No Offsets/Tiling, False=@n1151.UVs No Offsets/Tiling, Value=@n1152)
  n1157 = TextureSample!(UVs=@n1153, Tex=@n1154, "DDX(UVs)"=DDX(Value=@n1156), "DDY(UVs)"=DDY(Value=@n1156), Apply View MipBias) {mipValueMode=TMVM_Derivative}
  n1158 = NamedRerouteUsage()
  n1159 = NamedRerouteUsage()
  n1160 = OneMinus(None=NamedRerouteUsage())
  n1161 = Saturate(None=Add(A=@n1158, B=Saturate(None=@n1160)))
  NamedRerouteDeclaration(None=StaticSwitch(
    True=LinearInterpolate(A=@n1155.RGBA, B=@n1157.RGBA, Alpha=@n1158)
    False=LinearInterpolate(A=@n1159, B=@n1157.RGBA, Alpha=@n1161)
    Value=StaticBoolParameter() {group=TopSide_V02_Layer 04, parameterName=TopS_04-Miplevels-VS-Alternative_Axies}
   )) {name=TS 04 Side_Mask, nodeColor=<rgba 1 0.4922 0 1>}
  n1162 = NamedRerouteUsage()
  n1163 = NamedRerouteUsage()
  n1164 = TextureObjectParameter() {
   group=TopSide_V02_Layer 04
   parameterName=TopS_04__Tex_Mask
   samplerType=SAMPLERTYPE_Masks
   texture=&T_BlackMask
  }
  NamedRerouteDeclaration(None=StaticSwitchParameter(
    True=StaticSwitchParameter(True=@n1162, False=@n1163) {group=TopSide_V02_Layer 04, parameterName=TopS_04__Tex_Mask Share_00B_Msk_Tex}
    False=@n1164
   ) {group=TopSide_V02_Layer 04, parameterName=TopS_04__Tex_Mask Share_00A_Msk_Tex}) {name=Tex_TS04_Mask, nodeColor=<rgba 1 0.8906 0 1>}
  n1165 = StaticSwitchParameter(True=@n1162, False=@n1163) {group=TopSide_V02_Layer-04_Uvs | Warp, parameterName=TopS_04__Tex_Mask Warp Share_00B_Msk_Tex, sortPriority=249}
  n1166 = TextureObjectParameter() {
   group=TopSide_V02_Layer-04_Uvs | Warp
   parameterName=TopS_04_Tex_Warp
   samplerType=SAMPLERTYPE_Masks
   sortPriority=249
   texture=&T_BlackMask
  }
  NamedRerouteDeclaration(None=StaticSwitchParameter(
    True=StaticSwitchParameter(True=@n1165, False=@n1164) {group=TopSide_V02_Layer-04_Uvs | Warp, parameterName=TopS_04__Tex_Mask Warp Share_00A_Msk_Tex, sortPriority=249}
    False=@n1166
   ) {group=TopSide_V02_Layer-04_Uvs | Warp, parameterName=TopS_04__Share_Mask_03_Tex, sortPriority=249}) {name=Tex_04_Warp, nodeColor=<rgba 1 0.8906 0 1>}
  n1167 = NamedRerouteUsage()
  n1168 = NamedRerouteUsage()
  n1169 = NamedRerouteUsage()
  n1170 = Multiply(A=TextureCoordinate(), B=ComponentMask(None=@n1169) {g=true, r=true})
  n1171 = VectorParameter() {group=TopSide_V02_Layer-04_Uvs | Warp, parameterName=TopS_04_Warp_UV_Panning, sortPriority=250, defaultValue=<rgba 0 0 0 1>}
  n1172 = Multiply(A=@n1171.RGB, B=Time())
  n1173 = ComponentMask(None=@n1172) {g=true, r=true}
  n1174 = VectorParameter() {group=TopSide_V02_Layer-04_Uvs | Warp, parameterName=TopS_04_Warp_UV_Offset_Pos, sortPriority=250, defaultValue=<rgba 0 0 0 1>}
  n1175 = MaterialFunctionCall(
   Caculated_Bounds=NamedRerouteUsage()
   Pre_Skinned_Normals=NamedRerouteUsage()
   is_State=NamedRerouteUsage()
   UV_Rotation=NamedRerouteUsage()
   Uvs_Offset=@n1172
   Uvs_TileRate=@n1169
   Uvs_Offset_Pivot=@n1174.RGB
  ) {materialFunction=&MF_Projection_Side_Normal_V2}
  n1176 = TextureSample!(
   UVs=StaticSwitchParameter(
    True=Add(A=@n1170, B=@n1173)
    False=StaticSwitchParameter(True=@n1175.UVs No Rotation, False=@n1175.UVs No Rotation) {group=TopSide_V02_Layer-04_Uvs | Warp, parameterName=TopS_04 Warp_No_Rotation, sortPriority=249}
   ) {group=TopSide_V02_Layer-04_Uvs | Warp, parameterName=TopS_04 Warp_ObjectUVs, sortPriority=249}
   Tex=NamedRerouteUsage()
   Apply View MipBias
  )
  n1177 = StaticSwitchParameter(
   True=NamedRerouteUsage()
   False=StaticSwitchParameter(
    True=NamedRerouteUsage()
    False=StaticSwitchParameter(True=@n1168, False=@n1176.RGBA) {group=TopSide_V02_Layer-04_Uvs | Warp, parameterName="TopS_04 Warp--->Layer 01 MSk", sortPriority=250}
   ) {group=TopSide_V02_Layer-04_Uvs | Warp, parameterName="TopS_04 Warp--->Layer 03 MSk", sortPriority=250}
  ) {group=TopSide_V02_Layer-04_Uvs | Warp, parameterName="TopS_04 Warp--->Layer 02 MSk", sortPriority=250}
  n1178 = ChannelMaskParameter(None=@n1177) {group=TopSide_V02_Layer-04_Uvs | Warp, parameterName=TopS_04_Warp_UV Mask_Channel, sortPriority=250}
  n1179 = ScalarParameter() {group=TopSide_V02_Layer-04_Uvs | Warp, parameterName=TopS_04_Warp_UV_Offset, sortPriority=250}
  n1180 = ScalarParameter() {group=TopSide_V02_Layer-04_Uvs | Warp, parameterName=TopS_04_Warp_UV_Amount, sortPriority=250}
  n1181 = NamedRerouteUsage()
  n1182 = NamedRerouteUsage()
  n1183 = VectorParameter() {
   group=TopSide_V02_Layer-04_Uvs | Warp
   parameterName=TopS_04_Warp_Curvature_Amount
   sortPriority=250
   channelNames {a="", b="", g=Pits, r=Peaks}
  }
  n1184 = NamedRerouteUsage()
  n1185 = Add(
   A=Multiply(A=Add(A=@n1178.R, B=@n1179), B=Multiply(A=@n1180, B=@n1181))
   B=Add(A=Multiply(A=@n1182, B=@n1183.Peaks), B=Multiply(A=@n1184, B=@n1183.Pits))
  )
  NamedRerouteDeclaration(None=StaticSwitchParameter(True=Add(A=@n1167, B=@n1185), False=@n1167) {group=TopSide_V02_Layer-04_Uvs | Warp, parameterName=TopS_04 Warp_Uvs-----------, sortPriority=249}) {name=Caclcated_Bounds_04, nodeColor=<rgba 0.3594 0 1 1>}
  n1186 = TextureObjectParameter() {
   group=TopSide_V02
   parameterName=TopS_00A_Sharerd_Tex_Msk
   samplerType=SAMPLERTYPE_Masks
   sortPriority=249
   texture=&T_BlackMask
  }
  NamedRerouteDeclaration(None=@n1186) {name=Tex_00A_Shared_Tex, nodeColor=<rgba 1 0.8906 0 1>}
  NamedRerouteDeclaration(None=MaterialFunctionCall(In=@n517, Contrast=NamedRerouteUsage()) {materialFunction=&CheapContrast}.Result) {name=Tri_Mask 04, nodeColor=<rgba 1 0.4688 0 1>}
  NamedRerouteDeclaration(None=StaticBoolParameter() {group=TopSide_V02_Layer Emissive_02 Reactive, parameterName=TopS_Em02-Animation-CC, sortPriority=300}) {name=Animted_CC_State_02, nodeColor=<rgba 0.0312 0 1 1>}
  NamedRerouteDeclaration(None=StaticBoolParameter() {group=TopSide_V02_Layer Emissive_01 Reactive, parameterName=TopS_Em01-Animation-CC, sortPriority=300}) {name=Animted_CC_State_01, nodeColor=<rgba 0.0312 0 1 1>}
  n1187 = ScalarParameter() {defaultValue=39, group=TopSide_V02_Layer Emissive_03 Reactive, parameterName="TopS_Em03_Animation_CC  Index", sortPriority=300}
  n1188 = Time()
  n1189 = ScalarParameter() {group=TopSide_V02_Layer Emissive_03 Reactive, parameterName=TopS_Em03_Animation_CC_Offset, sortPriority=300}
  n1190 = Frac(None=Multiply(
    A=Add(A=@n1188, B=@n1189)
    B=ScalarParameter() {group=TopSide_V02_Layer Emissive_03 Reactive, parameterName=TopS_Em03_Animation_CC_Speed, sortPriority=300}
   ))
  n1191 = CurveAtlasRowParameter(CurveTime=@n1190) {
   atlas=&CruveAtlas_Wraps_01
   curve=/Game/Weapons/WeaponSkins/Wraps/Textures/C_Immiscible
   defaultValue=39
   group=TopSide_V02_Layer Emissive_03 Reactive
   parameterName=TopS_Em03_Animation_CC
   sortPriority=300
  }
  n1192 = MaterialFunctionCall(Index=@n1187, CCinput=@n1191, CCinput_Alpha=@n1191, UV=@n1190) {materialFunction=&MF_Global_CC_Tex}
  n1193 = ScalarParameter() {defaultValue=1, group=TopSide_V02_Layer Emissive_03 Reactive, parameterName=TopS_Em03_Animation_CC_BP_Trans, sortPriority=300}
  NamedRerouteDeclaration(None=LinearInterpolate(
    B=ChannelMaskParameter(None=@n1192.4v) {group=TopSide_V02_Layer Emissive_03 Reactive, parameterName=TopS_Em03_Animation_CC_Channel, sortPriority=300}.R
    Alpha=@n1193
   ) {constA=1}) {name=Animted_CC_03, nodeColor=<rgba 0.0312 0 1 1>}
  n1194 = ScalarParameter() {defaultValue=39, group=TopSide_V02_Layer Emissive_02 Reactive, parameterName="TopS_Em02_Animation_CC  Index", sortPriority=300}
  n1195 = Time()
  n1196 = ScalarParameter() {group=TopSide_V02_Layer Emissive_02 Reactive, parameterName=TopS_Em02_Animation_CC_Offset, sortPriority=300}
  n1197 = Frac(None=Multiply(
    A=Add(A=@n1195, B=@n1196)
    B=ScalarParameter() {group=TopSide_V02_Layer Emissive_02 Reactive, parameterName=TopS_Em02_Animation_CC_Speed, sortPriority=300}
   ))
  n1198 = CurveAtlasRowParameter(CurveTime=@n1197) {
   atlas=&CruveAtlas_Wraps_01
   curve=&C_Immiscible
   defaultValue=39
   group=TopSide_V02_Layer Emissive_02 Reactive
   parameterName=TopS_Em02_Animation_CC
   sortPriority=300
  }
  n1199 = MaterialFunctionCall(Index=@n1194, CCinput=@n1198, CCinput_Alpha=@n1198, UV=@n1197) {materialFunction=&MF_Global_CC_Tex}
  n1200 = ScalarParameter() {defaultValue=1, group=TopSide_V02_Layer Emissive_02 Reactive, parameterName=TopS_Em02_Animation_CC_BP_Trans, sortPriority=300}
  NamedRerouteDeclaration(None=LinearInterpolate(
    B=ChannelMaskParameter(None=@n1199.4v) {group=TopSide_V02_Layer Emissive_02 Reactive, parameterName=TopS_Em02_Animation_CC_Channel, sortPriority=300}.R
    Alpha=@n1200
   ) {constA=1}) {name=Animted_CC_02, nodeColor=<rgba 0.0312 0 1 1>}
  n1201 = ScalarParameter() {defaultValue=0.5, group=TopSide_V02_Layer 01 -Color_Tint, parameterName=TopS_01_Roughtness_CH_Red}
  n1202 = MaterialFunctionCall(Float4=NamedRerouteUsage()) {materialFunction=&BreakOutFloat4Components}
  n1203 = Min(A=@n810.A, B=@n1202.R)
  n1204 = LinearInterpolate(
   A=LinearInterpolate(
    A=LinearInterpolate(B=@n1201, Alpha=@n1203) {constA=1}
    B=ScalarParameter() {defaultValue=0.5, group=TopSide_V02_Layer 01 -Color_Tint, parameterName=TopS_01_Roughtness_CH_Green}
    Alpha=Min(A=@n812.A, B=@n1202.G)
   ) {constA=1}
   B=ScalarParameter() {defaultValue=0.5, group=TopSide_V02_Layer 01 -Color_Tint, parameterName=TopS_01_Roughtness_CH_Blue}
   Alpha=Min(A=@n814.A, B=@n1202.B)
  )
  n1205 = ScalarParameter() {defaultValue=0.5, group=TopSide_V02_Layer 01 -Color_Tint, parameterName=TopS_01_Roughtness_CH_Alpha}
  n1206 = Min(A=@n816.A, B=@n1202.A)
  NamedRerouteDeclaration(None=Multiply(
    A=ScalarParameter() {defaultValue=0.5, group=TopSide_V02_Layer 01 -Color_Tint, parameterName=TopS_01_Roughtness------}
    B=LinearInterpolate(A=@n1204, B=@n1205, Alpha=@n1206)
   )) {name=TS 01 Side_ColorTint_Roughness, nodeColor=<rgba 1 0.5625 0 1>}
  n1207 = ScalarParameter() {defaultValue=39, group=TopSide_V02_Layer Emissive_01 Reactive, parameterName="TopS_Em01_Animation_CC  Index", sortPriority=300}
  n1208 = Time()
  n1209 = ScalarParameter() {group=TopSide_V02_Layer Emissive_01 Reactive, parameterName=TopS_Em01_Animation_CC_Offset, sortPriority=300}
  n1210 = Frac(None=Multiply(
    A=Add(A=@n1208, B=@n1209)
    B=ScalarParameter() {group=TopSide_V02_Layer Emissive_01 Reactive, parameterName=TopS_Em01_Animation_CC_Speed, sortPriority=300}
   ))
  n1211 = CurveAtlasRowParameter(CurveTime=@n1210) {
   atlas=&CruveAtlas_Wraps_01
   curve=&C_Immiscible
   defaultValue=39
   group=TopSide_V02_Layer Emissive_01 Reactive
   parameterName=TopS_Em01_Animation_CC
   sortPriority=300
  }
  n1212 = MaterialFunctionCall(Index=@n1207, CCinput=@n1211, CCinput_Alpha=@n1211, UV=@n1210) {materialFunction=&MF_Global_CC_Tex}
  n1213 = ScalarParameter() {defaultValue=1, group=TopSide_V02_Layer Emissive_01 Reactive, parameterName=TopS_Em01_Animation_CC_BP_Trans, sortPriority=300}
  NamedRerouteDeclaration(None=LinearInterpolate(
    B=ChannelMaskParameter(None=@n1212.4v) {group=TopSide_V02_Layer Emissive_01 Reactive, parameterName=TopS_Em01_Animation_CC_Channel, sortPriority=300}.R
    Alpha=@n1213
   ) {constA=1}) {name=Animted_CC_01, nodeColor=<rgba 0.0312 0 1 1>}
  n1214 = MaterialFunctionCall(Float4=NamedRerouteUsage()) {materialFunction=&BreakOutFloat4Components}
  n1215 = LinearInterpolate(
   A=LinearInterpolate(
    A=ScalarParameter() {group=TopSide_V02_Layer 01, parameterName=TopS_01_Base_Metalic_}
    B=ScalarParameter() {group=TopSide_V02_Layer 01 -Color_Tint, parameterName=TopS_01_Metalic_CH_Red}
    Alpha=Min(A=@n810.A, B=@n1214.R)
   ) {constA=1}
   B=ScalarParameter() {group=TopSide_V02_Layer 01 -Color_Tint, parameterName=TopS_01_Metalic_CH_Green}
   Alpha=Min(A=@n812.A, B=@n1214.G)
  ) {constA=1}
  n1216 = ScalarParameter() {group=TopSide_V02_Layer 01 -Color_Tint, parameterName=TopS_01_Metalic_CH_Blue}
  n1217 = Min(A=@n814.A, B=@n1214.B)
  NamedRerouteDeclaration(None=LinearInterpolate(
    A=LinearInterpolate(A=@n1215, B=@n1216, Alpha=@n1217)
    B=ScalarParameter() {group=TopSide_V02_Layer 01 -Color_Tint, parameterName=TopS_01_Metalic_CH_Alpha}
    Alpha=Min(A=@n816.A, B=@n1214.A)
   )) {name=TS 01 Side_ColorTint_Metalic, nodeColor=<rgba 1 0.5625 0 1>}
  n1218 = ScalarParameter() {defaultValue=0.5, group=TopSide_V02_Layer 01 -Color_Tint, parameterName=TopS_01_Spec_CH_Red}
  n1219 = MaterialFunctionCall(Float4=NamedRerouteUsage()) {materialFunction=&BreakOutFloat4Components}
  n1220 = Min(A=@n810.A, B=@n1219.R)
  n1221 = LinearInterpolate(
   A=LinearInterpolate(
    A=LinearInterpolate(B=@n1218, Alpha=@n1220) {constA=1}
    B=ScalarParameter() {defaultValue=0.5, group=TopSide_V02_Layer 01 -Color_Tint, parameterName=TopS_01_Spec_CH_Green}
    Alpha=Min(A=@n812.A, B=@n1219.G)
   ) {constA=1}
   B=ScalarParameter() {defaultValue=0.5, group=TopSide_V02_Layer 01 -Color_Tint, parameterName=TopS_01_Spec_CH_Blue}
   Alpha=Min(A=@n814.A, B=@n1219.B)
  )
  n1222 = ScalarParameter() {defaultValue=0.5, group=TopSide_V02_Layer 01 -Color_Tint, parameterName=TopS_01_Spec_CH_Alpha}
  n1223 = Min(A=@n816.A, B=@n1219.A)
  NamedRerouteDeclaration(None=Multiply(
    A=ScalarParameter() {defaultValue=0.5, group=TopSide_V02_Layer 01 -Color_Tint, parameterName=TopS_01_Spec------}
    B=LinearInterpolate(A=@n1221, B=@n1222, Alpha=@n1223)
   )) {name=TS 01 Side_ColorTint_Spec, nodeColor=<rgba 1 0.5625 0 1>}
  NamedRerouteDeclaration(None=@n891) {name=Auidio03_Ramp_State, nodeColor=<rgba 0.6172 1 0 1>}
  NamedRerouteDeclaration(None=@n406) {name=Uvs_TS_Layer_01_Rotation, nodeColor=<rgba 1 0 0.3281 1>}
  n1224 = ScalarParameter() {group=00-FrontEnd-Overrides, parameterName=00-FE-TODCC-Override}
  n1225 = Add(
   A=CollectionParameter() {collection=&FortniteMaterialParameters, parameterName=TimeOfDay_0-1}
   B=@n1224
  )
  n1226 = CurveAtlasRowParameter(CurveTime=@n1225) {
   atlas=&CruveAtlas_Wraps_01
   curve=&Curve_TimeOfDay
   defaultValue=182
   group="TopSide_V02_Layer 01 Blend  02  TOD"
   parameterName=TopS_01_Blend_TOD_CC
  }
  n1227 = MaterialFunctionCall(
   Index=ScalarParameter() {defaultValue=182, group="TopSide_V02_Layer 01 Blend  02  TOD", parameterName="TopS_01_Blend_TOD_CC  Index"}
   CCinput=@n1226
   CCinput_Alpha=@n1226
   UV=@n1225
  ) {materialFunction=&MF_Global_CC_Tex}
  n1228 = ScalarParameter() {defaultValue=0.5, group="TopSide_V02_Layer 01 Blend  02  TOD", parameterName=TopS_01_Blend_TOD_Morning}
  n1229 = ScalarParameter() {group="TopSide_V02_Layer 01 Blend  02  TOD", parameterName=TopS_01_Blend_TOD_Day}
  n1230 = Add(A=LinearInterpolate(A=@n1224, B=Multiply(A=@n1224) {constB=-1}, Alpha=@n1229), B=@n1229)
  n1231 = ScalarParameter() {defaultValue=0.5, group="TopSide_V02_Layer 01 Blend  02  TOD", parameterName=TopS_01_Blend_TOD_Evening}
  n1232 = ScalarParameter() {defaultValue=1, group="TopSide_V02_Layer 01 Blend  02  TOD", parameterName=TopS_01_Blend_TOD_Night}
  NamedRerouteDeclaration(None=StaticSwitchParameter(
    True=ChannelMaskParameter(None=@n1227.4v) {group="TopSide_V02_Layer 01 Blend  02  TOD", parameterName=TopS_01_Blend_TOD_CC_Channel}.R
    False=MaterialFunctionCall(Value_Morning=@n1228, Value_Day=@n1230, Value_Evening=@n1231, Value_Night=@n1232) {materialFunction=&Scalar_Time_of_Day}.Result
   ) {group="TopSide_V02_Layer 01 Blend  02  TOD", parameterName="TopS Blend TOD --->CCTOD"}) {name=Blend 01-02 TOD, nodeColor=<rgba 0 0.8594 1 1>}
  n1233 = ScalarParameter() {defaultValue=0.5, group=TopSide_V02_Layer Emissive_00_TOD, parameterName=TopS Emissivs_TOD_Morning}
  n1234 = ScalarParameter() {defaultValue=1, group=TopSide_V02_Layer Emissive_00_TOD, parameterName=TopS Emissivs_TOD_Day}
  n1235 = Add(A=LinearInterpolate(A=@n470, B=Multiply(A=@n470) {constB=-1}, Alpha=@n1234), B=@n1234)
  n1236 = ScalarParameter() {defaultValue=0.5, group=TopSide_V02_Layer Emissive_00_TOD, parameterName=TopS Emissivs_TOD_Evening}
  n1237 = ScalarParameter() {defaultValue=1, group=TopSide_V02_Layer Emissive_00_TOD, parameterName=TopS Emissivs_TOD_Night}
  NamedRerouteDeclaration(None=StaticSwitchParameter(
    True=ChannelMaskParameter(None=@n472) {group=TopSide_V02_Layer Emissive_00_TOD, parameterName=TopS Emissivs_TOD_CC_Channel}.R
    False=MaterialFunctionCall(Value_Morning=@n1233, Value_Day=@n1235, Value_Evening=@n1236, Value_Night=@n1237) {materialFunction=&Scalar_Time_of_Day}.Result
   ) {group=TopSide_V02_Layer Emissive_00_TOD, parameterName="TopS Emissivs-TOD --->CCTOD"}) {name=Emissives TOD, nodeColor=<rgba 0 0.8594 1 1>}
  n1238 = PreSkinnedPosition()
  n1239 = PreSkinnedLocalBounds()
  n1240 = MaterialFunctionCall(Color=Subtract(A=@n1239.Max, B=@n1239.Max)) {materialFunction=&SplitComponents}
  n1241 = Divide(A=Subtract(A=VertexInterpolator(VS=@n1238).PS, B=@n1239.Min), B=Max(A=Max(A=@n1240.R, B=@n1240.R), B=@n1240.B))
  NamedRerouteDeclaration(None=@n1241) {name=Bound Full, nodeColor=<rgba 1 0 0.375 1>}
  n1242 = MaterialFunctionCall(Float4=NamedRerouteUsage()) {materialFunction=&BreakOutFloat4Components}
  n1243 = Max(A=@n1242.R, B=@n1242.R)
  n1244 = Max(
   A=@n1242.B
   B=MaterialFunctionCall(Float4=NamedRerouteUsage()) {materialFunction=&BreakOutFloat4Components}.G
  )
  NamedRerouteDeclaration(None=LinearInterpolate(A=NamedRerouteUsage(), B=NamedRerouteUsage(), Alpha=Max(A=@n1243, B=@n1244))) {name=Bounds_Full_Weapon, nodeColor=<rgba 0 0.6953 1 1>}
  n1245 = Multiply(A=NamedRerouteUsage(), B=@n315.PeaksUVs Grad)
  n1246 = Multiply(A=NamedRerouteUsage(), B=@n315.PitsUVS Grad)
  NamedRerouteDeclaration(None=StaticSwitchParameter(True=Add(A=@n1245, B=@n1246), False=@n1046) {group=TopSide_V02_Layer Emissive_03, parameterName=TopS Emissive_03_Warp, sortPriority=200}) {name=Emissive03_warp_Curv, nodeColor=<rgba 0.5 1 0 1>}
  n1247 = Multiply(A=NamedRerouteUsage(), B=@n245.PeaksUVs Grad)
  n1248 = Multiply(A=NamedRerouteUsage(), B=@n245.PitsUVS Grad)
  n1249 = Constant()
  NamedRerouteDeclaration(None=StaticSwitchParameter(True=Add(A=@n1247, B=@n1248), False=@n1249) {group=TopSide_V02_Layer Emissive_02, parameterName=TopS Emissive_02_Warp, sortPriority=200}) {name=Emissive02_warp_Curv, nodeColor=<rgba 0.5 1 0 1>}
  n1250 = StaticBoolParameter() {group=TopSide_V02_Layer-02_Uvs, parameterName=TopS_02-No_Rotation}
  NamedRerouteDeclaration(None=@n1250) {name=02-No_Rotation-State, nodeColor=<rgba 1 0.4688 0 1>}
  n1251 = Multiply(A=NamedRerouteUsage(), B=@n170.PeaksUVs Grad)
  n1252 = Multiply(A=NamedRerouteUsage(), B=@n170.PitsUVS Grad)
  NamedRerouteDeclaration(None=StaticSwitchParameter(True=Add(A=@n1251, B=@n1252), False=Constant()) {group=TopSide_V02_Layer Emissive_01, parameterName=TopS Emissive_01_Warp, sortPriority=200}) {name=Emissive01_warp_Cur, nodeColor=<rgba 0.5 1 0 1>}
  n1253 = StaticBoolParameter() {desc=use when animating rotation, group=TopSide_V02_Layer-02_Uvs, parameterName=TopS_02-Rotation_Comp}
  NamedRerouteDeclaration(None=@n1253) {name=TopS_02-Rotation_Comp_State, nodeColor=<rgba 0.1484 0 1 1>}
  n1254 = NamedRerouteUsage()
  n1255 = StaticSwitch(
   True=StaticBool()
   False=StaticBoolParameter() {group=TopSide_V02_Layer-02_Uvs, parameterName=TopS_02_Share_UVs_TopS_01, sortPriority=0}
   Value=StaticSwitch(True=NamedRerouteUsage(), False=StaticBool(), Value=NamedRerouteUsage())
  )
  n1256 = StaticSwitch(True=NamedRerouteUsage(), False=@n1254, Value=@n1255)
  n1257 = NamedRerouteUsage()
  n1258 = StaticSwitch(True=LinearInterpolate(A=NamedRerouteUsage(), B=@n1257, Alpha=NamedRerouteUsage()), False=@n1257, Value=NamedRerouteUsage())
  n1259 = StaticSwitch(True=NamedRerouteUsage(), False=@n1258, Value=@n1255)
  n1260 = NamedRerouteUsage()
  n1261 = StaticSwitch(True=NamedRerouteUsage(), False=@n1260, Value=@n1255)
  n1262 = StaticBoolParameter() {group=TopSide_V02_Layer-02_Uvs, parameterName=TopS_02 Clamp Uvs}
  n1263 = ScalarParameter() {defaultValue=0.001, group=TopSide_V02_Layer-02_Uvs, parameterName=TopS_02-Clamp Uvs padding Comp}
  NamedRerouteDeclaration(None=MaterialFunctionCall(
    Normal_Map=NamedRerouteUsage()
    Caculated_Bounds=NamedRerouteUsage()
    Pre_Skinned_Normals=NamedRerouteUsage()
    is_State=NamedRerouteUsage()
    UV_Rotation=@n1256
    Uvs_Offset=NamedRerouteUsage()
    Uvs_TileRate=@n1259
    Uvs_Offset_Pivot=@n1261
    Rotation_Comp=@n1253
    Clamp Y=@n1262
    Clamp X=@n1262
    Clamp Padding_Comp=@n1263
    No_Rotation=@n1250
   ) {materialFunction=&MF_Projection_Side_Normal_V2}.UVs) {name=TopS_02- CleanUVS, nodeColor=<rgba 0.8516 0 1 1>}
  n1264 = VectorParameter() {group=TopSide_V02_Layer-01_Uvs, parameterName=TopS_01_Scale_Weapon, defaultValue=<rgba 1 1 1 1>}
  n1265 = VectorParameter() {group=TopSide_V02_Layer-01_Uvs, parameterName=TopS_01_Scale_Vehicle, defaultValue=<rgba 1 1 1 1>}
  n1266 = Multiply(A=@n1265.RGB, B=@n1265.RGB)
  n1267 = VectorParameter() {group=TopSide_V02_Layer-01_Uvs, parameterName=TopS_01_Scale_VehicleSide, defaultValue=<rgba 1 1 1 1>}
  n1268 = Multiply(A=@n1267.RGB, B=@n1267.RGB)
  n1269 = MaterialFunctionCall(Float4=NamedRerouteUsage()) {materialFunction=&BreakOutFloat4Components}
  n1270 = VectorParameter() {group=TopSide_V02_Layer-01_Uvs, parameterName=TopS_01_Scale_Mech, defaultValue=<rgba 1 1 1 1>}
  n1271 = VectorParameter() {group=TopSide_V02_Layer-01_Uvs, parameterName=TopS_01_Scale_Character, defaultValue=<rgba 1 1 1 1>}
  n1272 = LinearInterpolate(
   A=LinearInterpolate(A=Multiply(A=@n1264.RGB, B=@n1264.RGB), B=LinearInterpolate(A=@n1266, B=@n1268, Alpha=@n1269.G), Alpha=@n1269.R)
   B=LinearInterpolate(A=Multiply(A=@n1270.RGB, B=@n1270.RGB), B=Multiply(A=@n1271.RGB, B=@n1271.RGB), Alpha=@n1269.A)
   Alpha=@n1269.B
  )
  n1273 = VectorParameter() {group=TopSide_V02_Layer-01_Uvs, parameterName=TopS_01_Scale_Backpack, defaultValue=<rgba 1 1 1 1>}
  n1274 = Multiply(A=@n1273.RGB, B=@n1273.RGB)
  n1275 = MaterialFunctionCall(Float2=NamedRerouteUsage()) {materialFunction=&BreakOutFloat2Components}
  n1276 = VectorParameter() {group=TopSide_V02_Layer-01_Uvs, parameterName=TopS_01_Scale_Pickaxe, defaultValue=<rgba 1 1 1 1>}
  NamedRerouteDeclaration(None=LinearInterpolate(A=LinearInterpolate(A=@n1272, B=@n1274, Alpha=@n1275.R), B=Multiply(A=@n1276.RGB, B=@n1276.RGB), Alpha=@n1275.G)) {name=Uvs_TS_Layer_01_Scale, nodeColor=<rgba 0 0.625 1 1>}
  n1277 = NamedRerouteUsage()
  n1278 = NamedRerouteUsage()
  n1279 = NamedRerouteUsage()
  n1280 = Multiply(A=TextureCoordinate(), B=ComponentMask(None=@n1279) {g=true, r=true})
  n1281 = VectorParameter() {group=TopSide_V02_Layer-02_Uvs | Warp, parameterName=TopS_02_Warp_UV_Panning, sortPriority=250, defaultValue=<rgba 0 0 0 1>}
  n1282 = Multiply(A=@n1281.RGB, B=Time())
  n1283 = ComponentMask(None=@n1282) {g=true, r=true}
  n1284 = VectorParameter() {group=TopSide_V02_Layer-02_Uvs | Warp, parameterName=TopS_02_Warp_UV_Offset_Pos, sortPriority=250, defaultValue=<rgba 0 0 0 1>}
  n1285 = MaterialFunctionCall(
   Caculated_Bounds=NamedRerouteUsage()
   Pre_Skinned_Normals=NamedRerouteUsage()
   is_State=NamedRerouteUsage()
   UV_Rotation=NamedRerouteUsage()
   Uvs_Offset=@n1282
   Uvs_TileRate=@n1279
   Uvs_Offset_Pivot=@n1284.RGB
  ) {materialFunction=&MF_Projection_Side_Normal_V2}
  n1286 = TextureSample(
   UVs=StaticSwitchParameter(
    True=Add(A=@n1280, B=@n1283)
    False=StaticSwitchParameter(True=@n1285.UVs No Rotation, False=@n1285.UVs No Rotation) {group=TopSide_V02_Layer-02_Uvs | Warp, parameterName=TopS_02 Warp No_Rotation, sortPriority=249}
   ) {group=TopSide_V02_Layer-02_Uvs | Warp, parameterName=TopS_02 Warp_ObjectUVs, sortPriority=249}
   Tex=NamedRerouteUsage()
  )
  n1287 = StaticSwitchParameter(
   True=NamedRerouteUsage()
   False=StaticSwitchParameter(
    True=NamedRerouteUsage()
    False=StaticSwitchParameter(True=@n1278, False=@n1286.RGBA) {group=TopSide_V02_Layer-02_Uvs | Warp, parameterName="TopS_02 Warp--->Layer 04 MSk", sortPriority=250}
   ) {group=TopSide_V02_Layer-02_Uvs | Warp, parameterName="TopS_02 Warp--->Layer 03 MSk", sortPriority=250}
  ) {group=TopSide_V02_Layer-02_Uvs | Warp, parameterName="TopS_02 Warp--->Layer 01 MSk", sortPriority=250}
  n1288 = Add(
   A=ChannelMaskParameter(None=@n1287) {group=TopSide_V02_Layer-02_Uvs | Warp, parameterName=TopS_02_Warp_UV Mask_Channel, sortPriority=250}.R
   B=ScalarParameter() {group=TopSide_V02_Layer-02_Uvs | Warp, parameterName=TopS_02_Warp_UV_Offset, sortPriority=250}
  )
  n1289 = Multiply(A=ScalarParameter() {group=TopSide_V02_Layer-02_Uvs | Warp, parameterName=TopS_02_Warp_UV_Amount, sortPriority=250}, B=NamedRerouteUsage())
  n1290 = NamedRerouteUsage()
  n1291 = VectorParameter() {
   group=TopSide_V02_Layer-02_Uvs | Warp
   parameterName=TopS_02_Warp_Curvature_Amount
   sortPriority=250
   channelNames {a="", b="", g=Pits, r=Peaks}
  }
  n1292 = NamedRerouteUsage()
  n1293 = Add(A=Add(A=@n1277, B=Multiply(A=@n1288, B=@n1289)), B=Add(A=Multiply(A=@n1290, B=@n1291.Peaks), B=Multiply(A=@n1292, B=@n1291.Pits)))
  n1294 = VectorParameter() {
   group=TopSide_V02_Layer-02_Uvs | Warp
   parameterName=TopS_02_Warp_Grad_Msk
   sortPriority=250
   channelNames {a="Falloff    -invert", b=Black falloof, g=Offset, r="0= Active"}
   defaultValue=<rgba 1 0 1 1>
  }
  n1295 = Add(A=ComponentMask(None=NamedRerouteUsage()) {r=true}, B=@n1294.Offset)
  n1296 = Saturate(None=Power(Base=Divide(A=@n1295, B=@n1294."Falloff    -invert"), Exp=@n1294.Black falloof))
  n1297 = StaticBool()
  n1298 = NamedRerouteUsage()
  n1299 = NamedRerouteUsage()
  n1300 = StaticSwitch(
   True=LinearInterpolate(A=@n1277, B=@n1293, Alpha=Max(A=@n1296, B=@n1294."0= Active"))
   False=@n1293
   Value=StaticSwitch(True=@n1297, False=@n1298, Value=StaticSwitch(True=@n1299, False=@n1297, Value=@n1298))
  )
  NamedRerouteDeclaration(None=@n1300) {name=TS_02_Warp, nodeColor=<rgba 0.1719 1 0 1>}
  n1301 = Constant4Vector() {constant=<rgba 0 0 0 0.5>}
  NamedRerouteDeclaration(None=FunctionInput(Preview=@n1301.RGBA) {bUsePreviewValueAsDefault=true, inputName=Object-SafeZone-Back, inputType=FunctionInput_Vector4, sortPriority=100}) {name=SafeZone_Back, nodeColor=<rgba 0.1016 1 0 1>}
  NamedRerouteDeclaration(None=FunctionInput(Preview=@n1301.RGBA) {bUsePreviewValueAsDefault=true, inputName=Object-SafeZone-Middle, inputType=FunctionInput_Vector4, sortPriority=101}) {name=SafeZone_Middle, nodeColor=<rgba 0.1016 1 0 1>}
  NamedRerouteDeclaration(None=FunctionInput(Preview=@n1301.RGBA) {bUsePreviewValueAsDefault=true, inputName=Object-SafeZone-Front, inputType=FunctionInput_Vector4, sortPriority=102}) {name=SafeZone_Front, nodeColor=<rgba 0.1016 1 0 1>}
  n1302 = StaticBool() {value=true}
  n1303 = StaticBoolParameter() {group=TopSide_V02_Layer Emissive_01, parameterName="TopS Emissive_01_Grad-Offset------>SafeZoneBack"}
  n1304 = StaticSwitch(True=@n1302, False=StaticBool(), Value=@n1303)
  n1305 = StaticBoolParameter() {group=TopSide_V02_Layer Emissive_01, parameterName="TopS Emissive_01_Grad-Offset------>SafeZoneMid"}
  n1306 = StaticBoolParameter() {group=TopSide_V02_Layer Emissive_01, parameterName="TopS Emissive_01_Grad-Offset------>SafeZoneFront"}
  NamedRerouteDeclaration(None=StaticSwitch(True=@n1302, False=StaticSwitch(True=@n1302, False=@n1304, Value=@n1305), Value=@n1306)) {name=Emis-01-SZ-State, nodeColor=<rgba 0 0.0156 1 1>}
  n1307 = ScalarParameter() {group=00-FrontEnd-Overrides, parameterName=00-FE-TODCC-Override}
  n1308 = Add(
   A=CollectionParameter() {collection=&FortniteMaterialParameters, parameterName=TimeOfDay_0-1}
   B=@n1307
  )
  n1309 = CurveAtlasRowParameter(CurveTime=@n1308) {
   atlas=&CruveAtlas_Wraps_01
   curve=&Curve_TimeOfDay
   defaultValue=182
   group=TopSide_V02_Layer Emissive_00_TOD
   parameterName=TopS Emissivs_TOD_CC
  }
  n1310 = MaterialFunctionCall(
   Index=ScalarParameter() {defaultValue=182, group=TopSide_V02_Layer Emissive_00_TOD, parameterName="TopS Emissivs_TOD_CC  Index"}
   CCinput=@n1309
   CCinput_Alpha=@n1309
   UV=@n1308
  ) {materialFunction=&MF_Global_CC_Tex}
  n1311 = ScalarParameter() {defaultValue=0.5, group=TopSide_V02_Layer Emissive_00_TOD, parameterName=TopS Emissivs_TODB_Morning}
  n1312 = ScalarParameter() {group=TopSide_V02_Layer Emissive_00_TOD, parameterName=TopS Emissivs_TODB_Day}
  n1313 = Add(A=LinearInterpolate(A=@n1307, B=Multiply(A=@n1307) {constB=-1}, Alpha=@n1312), B=@n1312)
  n1314 = ScalarParameter() {defaultValue=0.5, group=TopSide_V02_Layer Emissive_00_TOD, parameterName=TopS Emissivs_TODB_Evening}
  n1315 = ScalarParameter() {defaultValue=1, group=TopSide_V02_Layer Emissive_00_TOD, parameterName=TopS Emissivs_TODB_Night}
  NamedRerouteDeclaration(None=StaticSwitchParameter(
    True=ChannelMaskParameter(None=@n1310.4v) {group=TopSide_V02_Layer Emissive_00_TOD, parameterName=TopS Emissivs_TODB_CC_Channel}.R
    False=MaterialFunctionCall(Value_Morning=@n1311, Value_Day=@n1313, Value_Evening=@n1314, Value_Night=@n1315) {materialFunction=&Scalar_Time_of_Day}.Result
   ) {group=TopSide_V02_Layer Emissive_00_TOD, parameterName="TopS Emissivs-TOD --->CCTOD"}) {name=Emissives TODB, nodeColor=<rgba 0 0.8594 1 1>}
  n1316 = Constant()
  n1317 = NamedRerouteUsage()
  n1318 = StaticSwitch(
   True=NamedRerouteUsage()
   False=StaticSwitch(True=NamedRerouteUsage(), False=StaticSwitch(True=@n1317, False=@n24.RGBA, Value=@n1303), Value=@n1305)
   Value=@n1306
  )
  NamedRerouteDeclaration(None=MaterialFunctionCall(
    Normal_Map=NamedRerouteUsage()
    Caculated_Bounds=NamedRerouteUsage()
    Pre_Skinned_Normals=NamedRerouteUsage()
    is_State=NamedRerouteUsage()
    UV_Rotation=@n1316
    Uvs_Offset=ComponentMask(None=@n1318) {b=true, g=true, r=true}
    Uvs_Offset_Pivot=@n1316
    For_Safe_Zones=StaticBool() {value=true}
   ) {materialFunction=&MF_Projection_Side_Normal_V2}.UVs) {name=Emis01_SafeZ_Uv, nodeColor=<rgba 0.3359 1 0 1>}
  n1319 = StaticBool() {value=true}
  n1320 = StaticBoolParameter() {group=TopSide_V02_Layer Emissive_02, parameterName="TopS Emissive_02_Grad-Offset------>SafeZoneBack"}
  n1321 = StaticSwitch(True=@n1319, False=StaticBool(), Value=@n1320)
  n1322 = StaticBoolParameter() {group=TopSide_V02_Layer Emissive_02, parameterName="TopS Emissive_02_Grad-Offset------>SafeZoneMid"}
  n1323 = StaticBoolParameter() {group=TopSide_V02_Layer Emissive_02, parameterName="TopS Emissive_02_Grad-Offset------>SafeZoneFront"}
  NamedRerouteDeclaration(None=StaticSwitch(True=@n1319, False=StaticSwitch(True=@n1319, False=@n1321, Value=@n1322), Value=@n1323)) {name=Emis-02-SZ-State, nodeColor=<rgba 0 0.0156 1 1>}
  n1324 = Constant()
  n1325 = NamedRerouteUsage()
  n1326 = StaticSwitch(
   True=NamedRerouteUsage()
   False=StaticSwitch(True=NamedRerouteUsage(), False=StaticSwitch(True=@n1325, False=@n25.RGBA, Value=@n1320), Value=@n1322)
   Value=@n1323
  )
  NamedRerouteDeclaration(None=MaterialFunctionCall(
    Normal_Map=NamedRerouteUsage()
    Caculated_Bounds=NamedRerouteUsage()
    Pre_Skinned_Normals=NamedRerouteUsage()
    is_State=NamedRerouteUsage()
    UV_Rotation=@n1324
    Uvs_Offset=ComponentMask(None=@n1326) {b=true, g=true, r=true}
    Uvs_Offset_Pivot=@n1324
    For_Safe_Zones=StaticBool() {value=true}
   ) {materialFunction=&MF_Projection_Side_Normal_V2}.UVs) {name=Emis02_SafeZ_Uv, nodeColor=<rgba 0.3359 1 0 1>}
  n1327 = StaticBool() {value=true}
  n1328 = StaticSwitch(True=@n1327, False=StaticBool(), Value=@n476)
  NamedRerouteDeclaration(None=StaticSwitch(True=@n1327, False=StaticSwitch(True=@n1327, False=@n1328, Value=@n477), Value=@n479)) {name=Emis-03-SZ-State, nodeColor=<rgba 0 0.0156 1 1>}
  NamedRerouteDeclaration(None=@n5) {name=Uvs_TS_Layer_01_Offset, nodeColor=<rgba 0 0.625 1 1>}
  NamedRerouteDeclaration(None=@n480.UVs) {name=Emis03_SafeZ_Uv, nodeColor=<rgba 0.3359 1 0 1>}
  NamedRerouteDeclaration(None=@n517) {name=Tri_Mask 01B, nodeColor=<rgba 1 0.4688 0 1>}
  NamedRerouteDeclaration(None=@n792) {name=Uv_side_01, nodeColor=<rgba 0.5 0 1 1>}
  n1329 = VectorParameter() {group=TopSide_V02_Layer 01, parameterName="TopS_01_Alternate Axies_ColorA A=Rougness", defaultValue=<rgba 0 0 0 1>}
  n1330 = VectorParameter() {
   group=TopSide_V02_Layer 01
   parameterName=TopS_01_Alternate Axies_Color
   channelNames {a=B Max, b=B Min, g=A Max, r=A Min}
   defaultValue=<rgba 0 0.5 0 0.5>
  }
  n1331 = VectorParameter() {group=TopSide_V02_Layer 01, parameterName="TopS_01_Alternate Axies_ColorC A=Rougness", defaultValue=<rgba 0 0 0 1>}
  n1332 = VectorParameter() {group=TopSide_V02_Layer 01, parameterName="TopS_01_Alternate Axies_ColorB A=Rougness", defaultValue=<rgba 0 0 0 1>}
  n1333 = StaticSwitch(True=NamedRerouteUsage(), False=NamedRerouteUsage(), Value=NamedRerouteUsage())
  n1334 = MaterialFunctionCall!(
   A=AppendVector(A=@n1329.RGB, B=@n1329.RGB)
   A_B_Min_Max=AppendVector(A=@n1330, B=@n1330)
   C=AppendVector(A=@n1331.RGB, B=@n1331.RGB)
   B=AppendVector(A=@n1332.RGB, B=@n1332.RGB)
   Alpha=ComponentMask(None=Frac(None=@n1333)) {r=true}
  ) {materialFunction=&/MF_3ColorBlend_Adjustment}
  n1335 = VectorParameter() {group=TopSide_V02_Layer 01, parameterName="TopS_01_Alternate Axies_Color  A=Rougness", defaultValue=<rgba 0 0 0 1>}
  n1336 = AppendVector(A=@n1335.RGB, B=@n1335.RGB)
  n1337 = MaterialFunctionCall(
   Texture=NamedRerouteUsage()
   Caculated_Bounds=@n767
   Pre_Skinned_Normals=NamedRerouteUsage()
   Blend_Contrast=@n770
   IsStatsB=NamedRerouteUsage()
   IsState=NamedRerouteUsage()
   Fill_alternate_Axis_Opacity=@n773
   "Fill_alternate_Axis_Color_A=Roughness"=NamedRerouteDeclaration(None=StaticSwitchParameter(True=@n1334.Result, False=@n1336) {group=TopSide_V02_Layer 01, parameterName="TopS_01_Alternate Axies-Color------->3Colors"}) {name=Tops 01 Alternate Color, nodeColor=<rgba 1 0.3984 0 1>}
   Offset=@n779
   Scale=NamedRerouteUsage()
   Offset_Pivot=@n777
   Rotation_Axis=NamedRerouteUsage()
  ) {materialFunction=&MF_TopsSideProjection_Base_V2}
  NamedRerouteDeclaration(None=@n1337.Uv direction) {name=Tops_01 Uv_direction, nodeColor=<rgba 0 1 0.1797 1>}
  n1338 = Add(
   A=Add(A=NamedRerouteUsage(), B=Constant3Vector() {constant=<rgba 0 0 -1 0>})
   B=StaticSwitch(True=NamedRerouteUsage(), False=NamedRerouteUsage(), Value=@n1255)
  )
  n1339 = MaterialFunctionCall(Float4=NamedRerouteUsage()) {materialFunction=&BreakOutFloat4Components}
  n1340 = LinearInterpolate(
   A=ScalarParameter() {group=TopSide_V02_Layer 02, parameterName=TopS_02 Blend_Contrast}
   B=ScalarParameter() {group=TopSide_V02_Layer 02, parameterName="TopS_02 Blend_Contrast \\Mech"}
   Alpha=@n1339.B
  )
  n1341 = LinearInterpolate(
   A=ScalarParameter() {group=TopSide_V02_Layer 02, parameterName=TopS_02_Alternate Axies Opacity}
   B=ScalarParameter() {group=TopSide_V02_Layer 02, parameterName="TopS_02_Alternate Axies Opacity \\Mech"}
   Alpha=@n1339.B
  )
  n1342 = VectorParameter() {group=TopSide_V02_Layer 02, parameterName="TopS_02_Alternate Axies_ColorA A=Rougness", defaultValue=<rgba 0 0 0 1>}
  n1343 = VectorParameter() {
   group=TopSide_V02_Layer 02
   parameterName=TopS_02_Alternate Axies_Color
   channelNames {a=B Max, b=B Min, g=A Max, r=A Min}
   defaultValue=<rgba 0 0.5 0 0.5>
  }
  n1344 = VectorParameter() {group=TopSide_V02_Layer 02, parameterName="TopS_02_Alternate Axies_ColorC A=Rougness", defaultValue=<rgba 0 0 0 1>}
  n1345 = VectorParameter() {group=TopSide_V02_Layer 02, parameterName="TopS_02_Alternate Axies_ColorB A=Rougness", defaultValue=<rgba 0 0 0 1>}
  n1346 = StaticSwitch(True=NamedRerouteUsage(), False=NamedRerouteUsage(), Value=NamedRerouteUsage())
  n1347 = MaterialFunctionCall(
   A=AppendVector(A=@n1342.RGB, B=@n1342.RGB)
   A_B_Min_Max=AppendVector(A=@n1343, B=@n1343)
   C=AppendVector(A=@n1344.RGB, B=@n1344.RGB)
   B=AppendVector(A=@n1345.RGB, B=@n1345.RGB)
   Alpha=ComponentMask(None=Frac(None=@n1346)) {r=true}
  ) {materialFunction=&MF_3ColorBlend_Adjustment}
  n1348 = VectorParameter() {group=TopSide_V02_Layer 02, parameterName="TopS_02_Alternate Axies_Color  A=Rougness", defaultValue=<rgba 0 0 0 1>}
  n1349 = AppendVector(A=@n1348.RGB, B=@n1348.RGB)
  n1350 = NamedRerouteUsage()
  n1351 = NamedRerouteUsage()
  n1352 = NamedRerouteUsage()
  n1353 = StaticSwitch(
   True=NamedRerouteUsage()
   False=StaticSwitch(True=LinearInterpolate(A=@n1350, B=@n1351, Alpha=@n1352), False=@n1351, Value=@n761)
   Value=@n1255
  )
  n1354 = StaticSwitch(True=NamedRerouteUsage(), False=NamedRerouteUsage(), Value=@n1255)
  n1355 = NamedRerouteUsage()
  n1356 = MaterialFunctionCall(
   Texture=NamedRerouteUsage()
   Caculated_Bounds=@n1338
   Pre_Skinned_Normals=NamedRerouteUsage()
   Blend_Contrast=@n1340
   IsStatsB=NamedRerouteUsage()
   IsState=NamedRerouteUsage()
   Fill_alternate_Axis_Opacity=@n1341
   "Fill_alternate_Axis_Color_A=Roughness"=NamedRerouteDeclaration(None=StaticSwitchParameter(True=@n1347.Result, False=@n1349) {group=TopSide_V02_Layer 02, parameterName="TopS_02_Alternate Axies-Color------->3Colors"}) {name=Tops 02 Alternate Color, nodeColor=<rgba 1 0.3984 0 1>}
   Offset=@n1355
   Scale=@n1353
   Rotation_Axis=@n1354
  ) {materialFunction=&MF_TopsSideProjection_Base_V2}
  NamedRerouteDeclaration(None=@n1356.Uv direction) {name=Ts02 Uv direction, nodeColor=<rgba 1 0.2109 0 1>}
  n1357 = NamedRerouteUsage()
  n1358 = MaterialFunctionCall(
   Normal_Map=NamedRerouteUsage()
   Pivot=@n1357
   "Pivot_L\\R"=StaticBool() {value=true}
   Caculated_Bounds=NamedRerouteUsage()
   Pre_Skinned_Normals=NamedRerouteUsage()
   Is State=NamedRerouteUsage()
   Uvs_Rotation=@n1254
   Uvs_TileRate=@n1258
   Uvs_Offset=@n1260
   Clamp Y=@n1262
   Clamp X=@n1262
   Clamp Padding_Comp=@n1263
  ) {materialFunction=&MF_Projection_Side_Left_Right_Decal_NormalV2}
  n1359 = MaterialFunctionCall(
   Normal_Map=NamedRerouteUsage()
   Caculated_Bounds=NamedRerouteUsage()
   Pre_Skinned_Normals=NamedRerouteUsage()
   is_State=NamedRerouteUsage()
   UV_Rotation=@n1256
   Uvs_Offset=@n1357
   Uvs_TileRate=@n1259
   Uvs_Offset_Pivot=@n1261
   Rotation_Comp=@n1253
   Clamp Y=@n1262
   Clamp X=@n1262
   Clamp Padding_Comp=@n1263
   No_Rotation=@n1250
  ) {materialFunction=&MF_Projection_Side_Normal_V2}
  n1360 = NamedRerouteUsage()
  n1361 = StaticSwitch(True=@n1358.UVs, False=@n1359.UVs, Value=@n1360)
  NamedRerouteDeclaration(None=@n1361) {name=UV_Side_02, nodeColor=<rgba 0 0.1094 1 1>}
  NamedRerouteDeclaration(None=@n1347.Result) {name=TopS 02 3Color, nodeColor=<rgba 0 1 0.2734 1>}
  NamedRerouteDeclaration(None=@n1334.Result) {name=TopS 01 3Colors, nodeColor=<rgba 1 0 0.0469 1>}
  n1362 = VectorParameter() {group=TopSide_V02_Layer Emissive_01 Reactive, parameterName=TopS Emissive_01_Speed-Offset_End, sortPriority=500, defaultValue=<rgba 0 0 0 1>}
  n1363 = Time()
  n1364 = ScalarParameter() {defaultValue=-9999999, group=TopSide_V02_Layer Emissive_01 Reactive, parameterName=TopS Emissive_01_Speed-TBP, sortPriority=500}
  n1365 = Add(A=ScalarParameter() {group=TopSide_V02_Layer Emissive_01 Reactive, parameterName=TopS Emissive_01_Speed---Delay, sortPriority=500}, B=@n1364)
  n1366 = ScalarParameter() {defaultValue=0.25, group=TopSide_V02_Layer Emissive_01 Reactive, parameterName=TopS Emissive_01_Speed---Duration, sortPriority=500}
  n1367 = Saturate(None=Divide(A=Subtract(A=@n1363, B=@n1365), B=@n1366))
  n1368 = ScalarParameter() {defaultValue=0.25, group=TopSide_V02_Layer Emissive_01 Reactive, parameterName=TopS Emissive_01_Speed---DFalloff, sortPriority=500}
  NamedRerouteDeclaration(None=LinearInterpolate(A=Constant3Vector(), B=@n1362.RGB, Alpha=Power(Base=@n1367, Exp=@n1368))) {name=Reactive_Speed_Emis_01, nodeColor=<rgba 1 0.2344 0 1>}
  n1369 = VectorParameter() {group=TopSide_V02_Layer Emissive_02 Reactive, parameterName=TopS Emissive_02_Speed-Offset_End, sortPriority=500, defaultValue=<rgba 0 0 0 1>}
  n1370 = Time()
  n1371 = ScalarParameter() {defaultValue=-9999999, group=TopSide_V02_Layer Emissive_02 Reactive, parameterName=TopS Emissive_02_Speed-TBP, sortPriority=500}
  n1372 = Add(A=ScalarParameter() {group=TopSide_V02_Layer Emissive_02 Reactive, parameterName=TopS Emissive_02_Speed---Delay, sortPriority=500}, B=@n1371)
  n1373 = ScalarParameter() {defaultValue=0.25, group=TopSide_V02_Layer Emissive_02 Reactive, parameterName=TopS Emissive_02_Speed---Duration, sortPriority=500}
  n1374 = Saturate(None=Divide(A=Subtract(A=@n1370, B=@n1372), B=@n1373))
  n1375 = ScalarParameter() {defaultValue=0.25, group=TopSide_V02_Layer Emissive_02 Reactive, parameterName=TopS Emissive_02_Speed---DFalloff, sortPriority=500}
  NamedRerouteDeclaration(None=LinearInterpolate(A=Constant3Vector(), B=@n1369.RGB, Alpha=Power(Base=@n1374, Exp=@n1375))) {name=Reactive_Speed_Emis_02, nodeColor=<rgba 1 0.2344 0 1>}
  NamedRerouteDeclaration(None=StaticSwitch(True=NamedRerouteUsage(), False=@n1337.Combined, Value=@n780)) {name=TS_Layer01 BC, nodeColor=<rgba 0 0.0625 1 1>}
  n1376 = VectorParameter() {group=TopSide_V02_Layer Emissive_03 Reactive, parameterName=TopS Emissive_03_Speed-Offset_End, sortPriority=500, defaultValue=<rgba 0 0 0 1>}
  n1377 = Time()
  n1378 = ScalarParameter() {defaultValue=-9999999, group=TopSide_V02_Layer Emissive_03 Reactive, parameterName=TopS Emissive_03_Speed-TBP, sortPriority=500}
  n1379 = Add(A=ScalarParameter() {group=TopSide_V02_Layer Emissive_03 Reactive, parameterName=TopS Emissive_03_Speed---Delay, sortPriority=500}, B=@n1378)
  n1380 = ScalarParameter() {defaultValue=0.25, group=TopSide_V02_Layer Emissive_03 Reactive, parameterName=TopS Emissive_03_Speed---Duration, sortPriority=500}
  n1381 = Saturate(None=Divide(A=Subtract(A=@n1377, B=@n1379), B=@n1380))
  n1382 = ScalarParameter() {defaultValue=0.25, group=TopSide_V02_Layer Emissive_03 Reactive, parameterName=TopS Emissive_03_Speed---DFalloff, sortPriority=500}
  NamedRerouteDeclaration(None=LinearInterpolate(A=Constant3Vector(), B=@n1376.RGB, Alpha=Power(Base=@n1381, Exp=@n1382))) {name=Reactive_Speed_Emis_03, nodeColor=<rgba 1 0.2344 0 1>}
  n1383 = ScalarParameter() {defaultValue=1, group=TopSide_V02_Layer Emissive_01 Reactive, parameterName=TopS Audio 01 Static_Strength_Multi_Vehicle, sortPriority=149}
  n1384 = ScalarParameter() {
   defaultValue=1
   group=TopSide_V02_Layer Emissive_01 Reactive
   parameterName=TopS Audio 01 Static_Strength_Multi_Vehicle_Side
   sortPriority=149
  }
  n1385 = MaterialFunctionCall(Float4=NamedRerouteUsage()) {materialFunction=&BreakOutFloat4Components}
  n1386 = ScalarParameter() {defaultValue=1, group=TopSide_V02_Layer Emissive_01 Reactive, parameterName=TopS Audio 01 Static_Strength_Multi_Mech, sortPriority=149}
  n1387 = ScalarParameter() {defaultValue=1, group=TopSide_V02_Layer Emissive_01 Reactive, parameterName=TopS Audio 01 Static_Strength_Multi_Character, sortPriority=149}
  n1388 = LinearInterpolate(
   A=LinearInterpolate(A=Constant() {r=1}, B=LinearInterpolate(A=@n1383, B=@n1384, Alpha=@n1385.G), Alpha=@n1385.R)
   B=LinearInterpolate(A=@n1386, B=@n1387, Alpha=@n1385.A)
   Alpha=@n1385.B
  )
  n1389 = ScalarParameter() {defaultValue=1, group=TopSide_V02_Layer Emissive_01 Reactive, parameterName=TopS Audio 01 Static_Strength_Multi_Backpack, sortPriority=149}
  n1390 = MaterialFunctionCall(Float2=NamedRerouteUsage()) {materialFunction=&BreakOutFloat2Components}
  n1391 = ScalarParameter() {defaultValue=1, group=TopSide_V02_Layer Emissive_01 Reactive, parameterName=TopS Audio 01 Static_Strength_Multi_Pickaxe, sortPriority=149}
  NamedRerouteDeclaration(None=LinearInterpolate(A=LinearInterpolate(A=@n1388, B=@n1389, Alpha=@n1390.R), B=@n1391, Alpha=@n1390.G)) {name=Uvs_Audio 01 Static_Strength_Multi, nodeColor=<rgba 0 0.7188 1 1>}
  n1392 = ScalarParameter() {defaultValue=1, group=TopSide_V02_Layer Emissive_02 Reactive, parameterName=TopS Audio 02 Static_Strength_Multi_Vehicle, sortPriority=200}
  n1393 = ScalarParameter() {
   defaultValue=1
   group=TopSide_V02_Layer Emissive_02 Reactive
   parameterName=TopS Audio 02 Static_Strength_Multi_Vehicle_Side
   sortPriority=200
  }
  n1394 = MaterialFunctionCall(Float4=NamedRerouteUsage()) {materialFunction=&BreakOutFloat4Components}
  n1395 = ScalarParameter() {defaultValue=1, group=TopSide_V02_Layer Emissive_02 Reactive, parameterName=TopS Audio 02 Static_Strength_Multi_Mech, sortPriority=200}
  n1396 = ScalarParameter() {defaultValue=1, group=TopSide_V02_Layer Emissive_02 Reactive, parameterName=TopS Audio 02 Static_Strength_Multi_Character, sortPriority=200}
  n1397 = LinearInterpolate(
   A=LinearInterpolate(A=Constant() {r=1}, B=LinearInterpolate(A=@n1392, B=@n1393, Alpha=@n1394.G), Alpha=@n1394.R)
   B=LinearInterpolate(A=@n1395, B=@n1396, Alpha=@n1394.A)
   Alpha=@n1394.B
  )
  n1398 = ScalarParameter() {defaultValue=1, group=TopSide_V02_Layer Emissive_02 Reactive, parameterName=TopS Audio 02 Static_Strength_Multi_Backpack, sortPriority=200}
  n1399 = MaterialFunctionCall(Float2=NamedRerouteUsage()) {materialFunction=&BreakOutFloat2Components}
  n1400 = ScalarParameter() {defaultValue=1, group=TopSide_V02_Layer Emissive_02 Reactive, parameterName=TopS Audio 02 Static_Strength_Multi_Pickaxe, sortPriority=200}
  NamedRerouteDeclaration(None=LinearInterpolate(A=LinearInterpolate(A=@n1397, B=@n1398, Alpha=@n1399.R), B=@n1400, Alpha=@n1399.G)) {name=Uvs_Audio 02 Static_Strength_Multi, nodeColor=<rgba 0 0.7188 1 1>}
  n1401 = ScalarParameter() {defaultValue=1, group=TopSide_V02_Layer Emissive_03 Reactive, parameterName=TopS Audio 03 Static_Strength_Multi_Vehicle, sortPriority=300}
  n1402 = ScalarParameter() {
   defaultValue=1
   group=TopSide_V02_Layer Emissive_03 Reactive
   parameterName=TopS Audio 03 Static_Strength_Multi_Vehicle_Side
   sortPriority=300
  }
  n1403 = MaterialFunctionCall(Float4=NamedRerouteUsage()) {materialFunction=&BreakOutFloat4Components}
  n1404 = ScalarParameter() {defaultValue=1, group=TopSide_V02_Layer Emissive_03 Reactive, parameterName=TopS Audio 03 Static_Strength_Multi_Mech, sortPriority=300}
  n1405 = ScalarParameter() {defaultValue=1, group=TopSide_V02_Layer Emissive_03 Reactive, parameterName=TopS Audio 03 Static_Strength_Multi_Character, sortPriority=300}
  n1406 = LinearInterpolate(
   A=LinearInterpolate(A=Constant() {r=1}, B=LinearInterpolate(A=@n1401, B=@n1402, Alpha=@n1403.G), Alpha=@n1403.R)
   B=LinearInterpolate(A=@n1404, B=@n1405, Alpha=@n1403.A)
   Alpha=@n1403.B
  )
  n1407 = ScalarParameter() {defaultValue=1, group=TopSide_V02_Layer Emissive_03 Reactive, parameterName=TopS Audio 03 Static_Strength_Multi_Backpack, sortPriority=300}
  n1408 = MaterialFunctionCall(Float2=NamedRerouteUsage()) {materialFunction=&BreakOutFloat2Components}
  n1409 = ScalarParameter() {defaultValue=1, group=TopSide_V02_Layer Emissive_03 Reactive, parameterName=TopS Audio 03 Static_Strength_Multi_Pickaxe, sortPriority=300}
  NamedRerouteDeclaration(None=LinearInterpolate(A=LinearInterpolate(A=@n1406, B=@n1407, Alpha=@n1408.R), B=@n1409, Alpha=@n1408.G)) {name=Uvs_Audio 03 Static_Strength_Multi, nodeColor=<rgba 0 0.7188 1 1>}
  n1410 = NamedRerouteUsage()
  n1411 = NamedRerouteUsage()
  n1412 = Multiply(A=TextureCoordinate(), B=ComponentMask(None=@n1411) {g=true, r=true})
  n1413 = VectorParameter() {group=TopSide_V02_Layer-01_Uvs | Warp, parameterName=TopS_01_Warp_UV_Panning, sortPriority=250, defaultValue=<rgba 0 0 0 1>}
  n1414 = Multiply(A=@n1413.RGB, B=Time())
  n1415 = ComponentMask(None=@n1414) {g=true, r=true}
  n1416 = VectorParameter() {group=TopSide_V02_Layer-01_Uvs | Warp, parameterName=TopS_01_Warp_UV_Offset_Pos, sortPriority=250, defaultValue=<rgba 0 0 0 1>}
  n1417 = MaterialFunctionCall(
   Caculated_Bounds=NamedRerouteUsage()
   Pre_Skinned_Normals=NamedRerouteUsage()
   is_State=NamedRerouteUsage()
   UV_Rotation=NamedRerouteUsage()
   Uvs_Offset=@n1414
   Uvs_TileRate=@n1411
   Uvs_Offset_Pivot=@n1416.RGB
  ) {materialFunction=&MF_Projection_Side_Normal_V2}
  n1418 = TextureSample(
   UVs=StaticSwitchParameter(
    True=Add(A=@n1412, B=@n1415)
    False=StaticSwitchParameter(True=@n1417.UVs No Rotation, False=@n1417.UVs No Rotation) {group=TopSide_V02_Layer-01_Uvs | Warp, parameterName=TopS_01 Warp_No_Rotation, sortPriority=249}
   ) {group=TopSide_V02_Layer-01_Uvs | Warp, parameterName=TopS_01 Warp_ObjectUV, sortPriority=249}
   Tex=NamedRerouteUsage()
  )
  n1419 = StaticSwitchParameter(
   True=NamedRerouteUsage()
   False=StaticSwitchParameter(
    True=NamedRerouteUsage()
    False=StaticSwitchParameter(True=@n1410, False=@n1418.RGBA) {group=TopSide_V02_Layer-01_Uvs | Warp, parameterName="TopS_01 Warp--->Layer 04 MSk", sortPriority=250}
   ) {group=TopSide_V02_Layer-01_Uvs | Warp, parameterName="TopS_01 Warp--->Layer 03 MSk", sortPriority=250}
  ) {group=TopSide_V02_Layer-01_Uvs | Warp, parameterName="TopS_01 Warp--->Layer 02 MSk", sortPriority=250}
  n1420 = TextureSample!(UVs=NamedRerouteUsage(), Tex=NamedRerouteUsage(), Level, Apply View MipBias) {mipValueMode=TMVM_MipLevel}
  NamedRerouteDeclaration(None=StaticSwitchParameter(
    True=ChannelMaskParameter(None=@n1419) {group=TopSide_V02_Layer-01_Uvs | Warp, parameterName=TopS_01_PanningS_Speed__Warp Mask_Channel}.R
    False=ChannelMaskParameter(None=@n1420.RGBA) {group=TopSide_V02_Layer-01_Uvs | Warp, parameterName=TopS_01_PanningS_Speed__Mask_Channel}.R
   ) {group=TopSide_V02_Layer-01_Uvs | Warp, parameterName="TopS_01_PanningS_Speed__01-Msk -> WarpMsk"}) {name=TopSide_01_Warp, nodeColor=<rgba 0.3594 0 1 1>}
  n1421 = TextureSample(UVs=NamedRerouteUsage(), Tex=NamedRerouteUsage()) {mipValueMode=TMVM_MipLevel}
  NamedRerouteDeclaration(None=StaticSwitchParameter(
    True=ChannelMaskParameter(None=@n1287) {group=TopSide_V02_Layer-02_Uvs | Warp, parameterName=TopS_02_PanningS_Speed__Warp_Mask_Channel}.R
    False=ChannelMaskParameter(None=@n1421.RGBA) {group=TopSide_V02_Layer-02_Uvs | Warp, parameterName=TopS_02_PanningS_Speed__Mask_Channel}.R
   ) {group=TopSide_V02_Layer-02_Uvs | Warp, parameterName="TopS_02_PanningS_Speed__02-Msk -> WarpMsk"}) {name=TopSide 02_WarpSpeed, nodeColor=<rgba 0.3594 0 1 1>}
  NamedRerouteDeclaration(None=@n1121) {name=04_Share_uvs, nodeColor=<rgba 0 0.6953 1 1>}
  n1422 = StaticBool()
  n1423 = StaticSwitch(
   True=@n1422
   False=StaticBoolParameter() {group=TopSide_V02_Layer-03_Uvs, parameterName=TopS_03_Share_UVs_TopS_02, sortPriority=0}
   Value=StaticSwitch(True=NamedRerouteUsage(), False=@n1422, Value=NamedRerouteUsage())
  )
  NamedRerouteDeclaration(None=@n1423) {name=03_Share_State 02, nodeColor=<rgba 0.6641 0 1 1>}
  n1424 = StaticSwitch(
   True=@n1422
   False=StaticBoolParameter() {group=TopSide_V02_Layer-03_Uvs, parameterName=TopS_03_Share_UVs_TopS_01, sortPriority=0}
   Value=NamedRerouteUsage()
  )
  NamedRerouteDeclaration(None=@n1424) {name=03_Share_State 01, nodeColor=<rgba 0.6641 0 1 1>}
  NamedRerouteDeclaration(None=StaticSwitch(True=NamedRerouteUsage(), False=StaticBool(), Value=NamedRerouteUsage())) {name=02_uv_State, nodeColor=<rgba 0 0.0625 1 1>}
  n1425 = NamedRerouteUsage()
  n1426 = NamedRerouteUsage()
  n1427 = NamedRerouteUsage()
  n1428 = NamedRerouteUsage()
  n1429 = VectorParameter() {group=TopSide_V02_Layer 01, parameterName=TopS_01_Alternate_Axies Mask, defaultValue=<rgba 0 0 0 1>}
  n1430 = AppendVector(A=@n1429.RGB, B=@n1429.RGB)
  n1431 = NamedRerouteUsage()
  n1432 = NamedRerouteUsage()
  NamedRerouteDeclaration(None=StaticSwitch(
    True=NamedRerouteUsage()
    False=MaterialFunctionCall(
     Texture=@n1425
     Caculated_Bounds=@n767
     Pre_Skinned_Normals=@n1426
     Blend_Contrast=@n770
     IsStatsB=@n1427
     IsState=@n1428
     Fill_alternate_Axis_Opacity=@n773
     "Fill_alternate_Axis_Color_A=Roughness"=@n1430
     Offset=@n779
     Scale=@n1431
     Offset_Pivot=@n777
     Rotation_Axis=@n1432
    ) {materialFunction=&MF_TopsSideProjection_Base_V2}.Combined
    Value=@n780
   )) {name=TS_Layer01 Mask, nodeColor=<rgba 1 0 0.0469 1>}
  NamedRerouteDeclaration(None=StaticSwitch(True=NamedRerouteUsage(), False=StaticBool(), Value=NamedRerouteUsage())) {name=01_uv_State_1, nodeColor=<rgba 0 0.0625 1 1>}
  n1433 = TextureSample(UVs=NamedRerouteUsage(), Tex=NamedRerouteUsage()) {mipValueMode=TMVM_MipLevel}
  NamedRerouteDeclaration(None=StaticSwitchParameter(
    True=ChannelMaskParameter(None=@n1177) {group=TopSide_V02_Layer-04_Uvs | Warp, parameterName=TopS_04_PanningS_Speed__Warp Mask_Channel}.R
    False=ChannelMaskParameter(None=@n1433.RGBA) {group=TopSide_V02_Layer-04_Uvs | Warp, parameterName=TopS_04_PanningS_Speed__Mask_Channel}.R
   ) {group=TopSide_V02_Layer-04_Uvs | Warp, parameterName="TopS_04_PanningS_Speed__02-Msk -> WarpMsk"}) {name=04_Speed_Warp, nodeColor=<rgba 0 1 0.4141 1>}
  n1434 = TextureSample!(UVs=NamedRerouteUsage(), Tex=NamedRerouteUsage(), Apply View MipBias)
  NamedRerouteDeclaration(None=StaticSwitchParameter(
    True=ChannelMaskParameter(None=@n502) {group=TopSide_V02_Layer-03_Uvs | Warp, parameterName=TopS_03_PanningS_Speed__Warp Mask_Channel}.R
    False=ChannelMaskParameter(None=@n1434.RGBA) {group=TopSide_V02_Layer-03_Uvs | Warp, parameterName=TopS_03_PanningS_Speed__Mask_Channel}.R
   ) {group=TopSide_V02_Layer-03_Uvs | Warp, parameterName="TopS_03_PanningS_Speed__02-Msk -> WarpMsk"}) {name=03_Speed_Warp, nodeColor=<rgba 1 0.0938 0 1>}
  n1435 = StaticBool() {value=true}
  n1436 = NamedRerouteUsage()
  n1437 = NamedRerouteUsage()
  n1438 = NamedRerouteUsage()
  n1439 = NamedRerouteUsage()
  n1440 = NamedRerouteUsage()
  n1441 = NamedRerouteUsage()
  NamedRerouteDeclaration(None=StaticSwitch(
    True=MaterialFunctionCall(
     "Pivot_L\\R"=@n1435
     Caculated_Bounds=@n1436
     Pre_Skinned_Normals=@n1437
     Is State=@n1438
     Uvs_Rotation=@n1439
     Uvs_TileRate=@n1440
     Uvs_Offset=@n1441
    ) {materialFunction=&MF_Projection_Side_Left_Right_Decal_NormalV2}.UVs
    False=MaterialFunctionCall(
     Caculated_Bounds=@n1436
     Pre_Skinned_Normals=@n1437
     is_State=@n1438
     UV_Rotation=@n1439
     Uvs_TileRate=@n1440
     Uvs_Offset_Pivot=@n1441
     Rotation_Comp=@n788
     No_Rotation=@n789
    ) {materialFunction=&MF_Projection_Side_Normal_V2}.UVs
    Value=NamedRerouteUsage()
   )) {name=01_Clean_Uvs, nodeColor=<rgba 1 0 0.1875 1>}
  n1442 = StaticBool() {value=true}
  n1443 = NamedRerouteUsage()
  n1444 = NamedRerouteUsage()
  n1445 = NamedRerouteUsage()
  NamedRerouteDeclaration(None=StaticSwitch(
    True=MaterialFunctionCall(
     "Pivot_L\\R"=@n1442
     Caculated_Bounds=@n1443
     Pre_Skinned_Normals=@n1444
     Is State=@n1445
     Uvs_Rotation=@n1256
     Uvs_TileRate=@n1259
     Uvs_Offset=@n1261
    ) {materialFunction=&MF_Projection_Side_Left_Right_Decal_NormalV2}.UVs
    False=MaterialFunctionCall(
     Caculated_Bounds=@n1443
     Pre_Skinned_Normals=@n1444
     is_State=@n1445
     UV_Rotation=@n1256
     Uvs_TileRate=@n1259
     Uvs_Offset_Pivot=@n1261
     Rotation_Comp=@n1253
     No_Rotation=@n1250
    ) {materialFunction=&MF_Projection_Side_Normal_V2}.UVs
    Value=NamedRerouteUsage()
   )) {name=TopS_02- CleanUVS_2, nodeColor=<rgba 0.8516 0 1 1>}
  n1446 = NamedRerouteUsage()
  n1447 = NamedRerouteUsage()
  n1448 = NamedRerouteUsage()
  n1449 = NamedRerouteUsage()
  n1450 = StaticSwitch(True=NamedRerouteUsage(), False=StaticSwitch(True=NamedRerouteUsage(), False=@n1447, Value=@n1448), Value=@n1449)
  n1451 = NamedRerouteUsage()
  n1452 = StaticSwitch(True=NamedRerouteUsage(), False=StaticSwitch(True=NamedRerouteUsage(), False=@n1451, Value=@n1448), Value=@n1449)
  n1453 = NamedRerouteUsage()
  n1454 = StaticSwitch(True=NamedRerouteUsage(), False=StaticSwitch(True=NamedRerouteUsage(), False=@n1453, Value=@n1448), Value=@n1449)
  n1455 = StaticSwitch(
   True=NamedRerouteUsage()
   False=StaticSwitch(
    True=NamedRerouteUsage()
    False=StaticBoolParameter() {desc=use thiis when animation Rotation, group=TopSide_V02_Layer-03_Uvs, parameterName=TopS_03-Rotation-Comp}
    Value=@n1448
   )
   Value=@n1449
  )
  n1456 = StaticSwitch(
   True=NamedRerouteUsage()
   False=StaticSwitch(True=NamedRerouteUsage(), False=StaticBoolParameter() {group=TopSide_V02_Layer-03_Uvs, parameterName=TopS_03-No_Rotation}, Value=@n1448)
   Value=@n1449
  )
  NamedRerouteDeclaration(None=MaterialFunctionCall(
    Caculated_Bounds=NamedRerouteUsage()
    Pre_Skinned_Normals=NamedRerouteUsage()
    is_State=@n1446
    UV_Rotation=@n1450
    Uvs_TileRate=@n1452
    Uvs_Offset_Pivot=@n1454
    Rotation_Comp=NamedRerouteDeclaration(None=@n1455) {name=tempB, nodeColor=<rgba 0.5234 0 1 1>}
    No_Rotation=NamedRerouteDeclaration(None=@n1456) {name=tempA, nodeColor=<rgba 0.5234 0 1 1>}
   ) {materialFunction=&MF_Projection_Side_Normal_V2}.UVs) {name=03_Clean_uvs, nodeColor=<rgba 1 0 0.1172 1>}
  n1457 = NamedRerouteUsage()
  n1458 = NamedRerouteUsage()
  n1459 = NamedRerouteUsage()
  n1460 = NamedRerouteUsage()
  n1461 = NamedRerouteUsage()
  NamedRerouteDeclaration(None=StaticSwitch(
    True=NamedRerouteUsage()
    False=MaterialFunctionCall!(
     Texture_Normals=@n1457
     Rotation_Axis=@n1432
     Caculated_Bounds=@n767
     Pre_Skinned_Normals=@n1458
     IsStatsB=@n1459
     IsState=@n1460
     Blend_Contrast=@n770
     Fill_alternate_Axis_Opacity=@n773
     Scale=@n1461
     Offset_Pivot=@n777
     Offset=@n779
    ) {materialFunction=&/MF_TopsSideProjection_Normals_V2}.Add_Surface_Normal
    Value=@n780
   )) {name=TS_Layer01 Normal, nodeColor=<rgba 1 0 0.8203 1>}
  NamedRerouteDeclaration(None=MaterialFunctionCall(
    Caculated_Bounds=NamedRerouteUsage()
    Pre_Skinned_Normals=NamedRerouteUsage()
    is_State=NamedRerouteUsage()
    UV_Rotation=@n1146
    Uvs_TileRate=@n1147
    Uvs_Offset_Pivot=@n1148
    Rotation_Comp=NamedRerouteDeclaration(None=@n1149) {name=tempC, nodeColor=<rgba 0.5234 0 1 1>}
    No_Rotation=NamedRerouteDeclaration(None=@n1150) {name=tempD, nodeColor=<rgba 0.5234 0 1 1>}
   ) {materialFunction=&MF_Projection_Side_Normal_V2}.UVs) {name=04_Clean_uvs, nodeColor=<rgba 1 0 0.1172 1>}
  n1462 = VectorParameter() {group=TopSide_V02_Layer-02_Uvs_Reactive, parameterName=TopS_02_Speed_offset_Amount, sortPriority=100, defaultValue=<rgba 0 0 0 1>}
  n1463 = Time()
  n1464 = Add(
   A=ScalarParameter() {group=TopSide_V02_Layer-02_Uvs_Reactive, parameterName=TopS_02_Speed---Delay, sortPriority=100}
   B=ScalarParameter() {defaultValue=-9999999, group=TopSide_V02_Layer-02_Uvs_Reactive, parameterName=TopS_02_Speed-TBP, sortPriority=100}
  )
  n1465 = Saturate(None=Divide(
    A=Subtract(A=@n1463, B=@n1464)
    B=ScalarParameter() {defaultValue=1, group=TopSide_V02_Layer-02_Uvs_Reactive, parameterName=TopS_02_Speed---Duration, sortPriority=100}
   ))
  n1466 = ScalarParameter() {defaultValue=1, group=TopSide_V02_Layer-02_Uvs_Reactive, parameterName=TopS_02_Speed---DFalloff, sortPriority=100}
  n1467 = LinearInterpolate(B=@n1462.RGB, Alpha=OneMinus(None=Power(Base=@n1465, Exp=@n1466)))
  NamedRerouteDeclaration(None=@n1467) {name=Reactive 02 speed, nodeColor=<rgba 0 0.0625 1 1>}
  NamedRerouteDeclaration(None=@n757) {name=Reactive 03 Speed, nodeColor=<rgba 0.3828 1 0 1>}
  NamedRerouteDeclaration(None=@n1116) {name=reactive 04 Speed, nodeColor=<rgba 1 0 0.7031 1>}
  n1468 = VectorParameter() {group=TopSide_V02_Layer-01_Uvs_Reactive, parameterName=TopS_01_Speed_offset_Amount, sortPriority=100, defaultValue=<rgba 0 0 0 1>}
  n1469 = Time()
  n1470 = Add(
   A=ScalarParameter() {group=TopSide_V02_Layer-01_Uvs_Reactive, parameterName=TopS_01_Speed---Delay, sortPriority=100}
   B=ScalarParameter() {defaultValue=-9999999, group=TopSide_V02_Layer-01_Uvs_Reactive, parameterName=TopS_01_Speed-TBP, sortPriority=100}
  )
  n1471 = Saturate(None=Divide(
    A=Subtract(A=@n1469, B=@n1470)
    B=ScalarParameter() {defaultValue=1, group=TopSide_V02_Layer-01_Uvs_Reactive, parameterName=TopS_01_Speed---Duration, sortPriority=100}
   ))
  n1472 = ScalarParameter() {defaultValue=1, group=TopSide_V02_Layer-01_Uvs_Reactive, parameterName=TopS_01_Speed---DFalloff, sortPriority=100}
  n1473 = LinearInterpolate(B=@n1468.RGB, Alpha=OneMinus(None=Power(Base=@n1471, Exp=@n1472)))
  NamedRerouteDeclaration(None=@n1473) {name=Reactive 01 Speed, nodeColor=<rgba 0 1 0.9766 1>}
  n1474 = VectorParameter() {
   group=TopSide_V02_Layer-01_Uvs | Warp
   parameterName=TopS_01_Warp_UV_Mulplyer
   sortPriority=250
   channelNames {a=Mech, b=Character, g=Vehicle_Side, r=Vehicle}
   defaultValue=<rgba 1 1 1 1>
  }
  n1475 = MaterialFunctionCall(Float4=NamedRerouteUsage()) {materialFunction=&BreakOutFloat4Components}
  n1476 = LinearInterpolate(A=@n1474.Vehicle, B=@n1474.Vehicle, Alpha=@n1475.G)
  NamedRerouteDeclaration(None=LinearInterpolate(A=LinearInterpolate(B=@n1476, Alpha=@n1475.R) {constA=1}, B=LinearInterpolate(A=@n1474.Mech, B=@n1474.Mech, Alpha=@n1475.A), Alpha=@n1475.B)) {name=Uvs_TS_01_warp_mult, nodeColor=<rgba 0 0.7188 1 1>}
  n1477 = VectorParameter() {
   group=TopSide_V02_Layer-02_Uvs | Warp
   parameterName=TopS_02_Warp_UV_Mulplyer
   sortPriority=250
   channelNames {a=Mech, b=Character, g=Vehicle_Side, r=Vehicle}
   defaultValue=<rgba 1 1 1 1>
  }
  n1478 = MaterialFunctionCall(Float4=NamedRerouteUsage()) {materialFunction=&BreakOutFloat4Components}
  n1479 = LinearInterpolate(A=@n1477.Vehicle, B=@n1477.Vehicle, Alpha=@n1478.G)
  NamedRerouteDeclaration(None=LinearInterpolate(A=LinearInterpolate(B=@n1479, Alpha=@n1478.R) {constA=1}, B=LinearInterpolate(A=@n1477.Mech, B=@n1477.Mech, Alpha=@n1478.A), Alpha=@n1478.B)) {name=Uvs_TS_02_warp_mult, nodeColor=<rgba 0 0.7188 1 1>}
  n1480 = VectorParameter() {
   group=TopSide_V02_Layer-03_Uvs | Warp
   parameterName=TopS_03_Warp_UV_Mulplyer
   sortPriority=250
   channelNames {a=Mech, b=Character, g=Vehicle_Side, r=Vehicle}
   defaultValue=<rgba 1 1 1 1>
  }
  n1481 = MaterialFunctionCall(Float4=NamedRerouteUsage()) {materialFunction=&BreakOutFloat4Components}
  n1482 = LinearInterpolate(A=@n1480.Vehicle, B=@n1480.Vehicle, Alpha=@n1481.G)
  NamedRerouteDeclaration(None=LinearInterpolate(A=LinearInterpolate(B=@n1482, Alpha=@n1481.R) {constA=1}, B=LinearInterpolate(A=@n1480.Mech, B=@n1480.Mech, Alpha=@n1481.A), Alpha=@n1481.B)) {name=Uvs_TS_03_warp_mult, nodeColor=<rgba 0 0.7188 1 1>}
  n1483 = VectorParameter() {
   group=TopSide_V02_Layer-04_Uvs | Warp
   parameterName=TopS_04_Warp_UV_Mulplyer
   sortPriority=250
   channelNames {a=Mech, b=Character, g=Vehicle_Side, r=Vehicle}
   defaultValue=<rgba 1 1 1 1>
  }
  n1484 = MaterialFunctionCall(Float4=NamedRerouteUsage()) {materialFunction=&BreakOutFloat4Components}
  n1485 = LinearInterpolate(A=@n1483.Vehicle, B=@n1483.Vehicle, Alpha=@n1484.G)
  NamedRerouteDeclaration(None=LinearInterpolate(A=LinearInterpolate(B=@n1485, Alpha=@n1484.R) {constA=1}, B=LinearInterpolate(A=@n1483.Mech, B=@n1483.Mech, Alpha=@n1484.A), Alpha=@n1484.B)) {name=Uvs_TS_04_warp_mult, nodeColor=<rgba 0 0.7188 1 1>}
  NamedRerouteDeclaration(None=FunctionInput(Preview=Constant4Vector().RGBA) {bUsePreviewValueAsDefault=true, inputName=Liquid Mask Input, inputType=FunctionInput_Vector4, sortPriority=102}) {name=LiquidMasks, nodeColor=<rgba 1 0 0.7969 1>}
  n1486 = StaticSwitchParameter(True=NamedRerouteUsage(), False=NamedRerouteUsage()) {group=TopSide_V02_Layer-01_Uvs | Warp, parameterName=TopS_01_Share_Mask_00B_Msk_Tex, sortPriority=249}
  n1487 = TextureObjectParameter() {
   group=TopSide_V02_Layer-01_Uvs | Warp
   parameterName=TopS_01_Tex_Warp
   samplerType=SAMPLERTYPE_Masks
   sortPriority=249
   texture=&T_BlackMask
  }
  NamedRerouteDeclaration(None=StaticSwitchParameter(
    True=@n852
    False=StaticSwitchParameter(True=@n1486, False=@n1487) {group=TopSide_V02_Layer-01_Uvs | Warp, parameterName=TopS_01_Share_Mask_00A_Msk_Tex, sortPriority=249}
   ) {group=TopSide_V02_Layer-01_Uvs | Warp, parameterName=TopS_01_Share_Mask_01_Tex, sortPriority=249}) {name=Tex_01_Warp, nodeColor=<rgba 1 0.8906 0 1>}
  NamedRerouteDeclaration(None=StaticBoolParameter() {group=Customization, parameterName=Effect_Liquid}) {name=LiquidState, nodeColor=<rgba 1 0 0.7969 1>}
  NamedRerouteDeclaration(None=@n354) {name=Uvs_TS_Layer_01_Warp_Rotation, nodeColor=<rgba 0 0.7188 1 1>}
  n1488 = ScalarParameter() {defaultValue=1, group=TopSide_V02_Layer-01_Uvs | Warp, parameterName=TopS_01_Warp_Scale_Vehicle, sortPriority=300}
  n1489 = ScalarParameter() {defaultValue=1, group=TopSide_V02_Layer-01_Uvs | Warp, parameterName=TopS_01_Warp_Scale_Vehicle_Side, sortPriority=300}
  n1490 = MaterialFunctionCall(Float4=NamedRerouteUsage()) {materialFunction=&BreakOutFloat4Components}
  n1491 = LinearInterpolate(
   A=LinearInterpolate(
    A=ScalarParameter() {defaultValue=1, group=TopSide_V02_Layer-01_Uvs | Warp, parameterName=TopS_01_Warp_Scale_Weapon, sortPriority=300}
    B=LinearInterpolate(A=@n1488, B=@n1489, Alpha=@n1490.G)
    Alpha=@n1490.R
   )
   B=LinearInterpolate(
    A=ScalarParameter() {defaultValue=1, group=TopSide_V02_Layer-01_Uvs | Warp, parameterName=TopS_01_Warp_Scale_Mech, sortPriority=300}
    B=ScalarParameter() {defaultValue=1, group=TopSide_V02_Layer-01_Uvs | Warp, parameterName=TopS_01_Warp_Scale_Character, sortPriority=300}
    Alpha=@n1490.A
   )
   Alpha=@n1490.B
  )
  n1492 = ScalarParameter() {defaultValue=1, group=TopSide_V02_Layer-01_Uvs | Warp, parameterName=TopS_01_Warp_Scale_Backpack, sortPriority=300}
  n1493 = MaterialFunctionCall(Float2=NamedRerouteUsage()) {materialFunction=&BreakOutFloat2Components}
  NamedRerouteDeclaration(None=LinearInterpolate(
    A=LinearInterpolate(A=@n1491, B=@n1492, Alpha=@n1493.R)
    B=ScalarParameter() {defaultValue=1, group=TopSide_V02_Layer-01_Uvs | Warp, parameterName=TopS_01_Warp_Scale_Pickaxe, sortPriority=300}
    Alpha=@n1493.G
   )) {name=Uvs_TS_Layer_01_Warp_Scale, nodeColor=<rgba 0 0.7188 1 1>}
  n1494 = TextureObjectParameter() {
   group=TopSide_V02_Layer 02
   parameterName=TopS_02__Tex_Normal
   samplerType=SAMPLERTYPE_Normal
   texture=&FlatNormalBlue
  }
  NamedRerouteDeclaration(None=StaticSwitchParameter(True=NamedRerouteUsage(), False=@n1494) {group=TopSide_V02_Layer 02, parameterName=TopS_02_Share_Normal_01_Tex, sortPriority=0}) {name=Tex_TS02_Normal, nodeColor=<rgba 1 0.8906 0 1>}
  n1495 = NamedRerouteUsage()
  n1496 = ChannelMaskParameter(None=@n1419) {group=TopSide_V02_Layer-01_Uvs | Warp, parameterName=TopS_01_Warp_UV Mask_Channel, sortPriority=250}
  n1497 = ScalarParameter() {group=TopSide_V02_Layer-01_Uvs | Warp, parameterName=TopS_01_Warp_UV_Offset, sortPriority=250}
  n1498 = NamedRerouteUsage()
  n1499 = ScalarParameter() {group=TopSide_V02_Layer-01_Uvs | Warp, parameterName=TopS_01_Warp_UV_Amount, sortPriority=250}
  n1500 = Add(A=@n1495, B=Multiply(A=Add(A=@n1496.R, B=@n1497), B=Multiply(A=@n1498, B=@n1499)))
  n1501 = VectorParameter() {
   group=TopSide_V02_Layer-01_Uvs | Warp
   parameterName=TopS_01_Warp_Curvature_Amount
   sortPriority=250
   channelNames {a="", b="", g=Pits, r=Peaks}
  }
  n1502 = Add(A=Multiply(A=NamedRerouteUsage(), B=@n1501.Peaks), B=Multiply(A=NamedRerouteUsage(), B=@n1501.Pits))
  NamedRerouteDeclaration(None=StaticSwitchParameter(True=Add(A=@n1500, B=@n1502), False=@n1495) {group=TopSide_V02_Layer-01_Uvs | Warp, parameterName=TopS_01 Warp_Uvs-----------, sortPriority=0}) {name=Caclcated_Bounds_01, nodeColor=<rgba 0.3594 0 1 1>}
  n1503 = FunctionInput(Preview=@n663.RGBA) {bUsePreviewValueAsDefault=true, inputName=IsState, inputType=FunctionInput_Vector4, sortPriority=4}
  NamedRerouteDeclaration(None=MaterialFunctionCall(Float4=NamedRerouteDeclaration(None=@n1503) {name=IsState, nodeColor=<rgba 0.1497 1 0.1633 1>}) {materialFunction=&BreakOutFloat4Components}.A) {name=IsState_Character, nodeColor=<rgba 0.1497 1 0.1633 1>}
  n1504 = NamedRerouteUsage()
  n1505 = NamedRerouteUsage()
  n1506 = NamedRerouteUsage()
  n1507 = Add(
   A=Add(A=NamedRerouteUsage(), B=Constant3Vector() {constant=<rgba 0 0 -1 0>})
   B=StaticSwitch(True=NamedRerouteUsage(), False=StaticSwitch(True=@n1505, False=@n1506, Value=@n1424), Value=@n1423)
  )
  n1508 = NamedRerouteUsage()
  n1509 = MaterialFunctionCall(Float4=NamedRerouteUsage()) {materialFunction=&BreakOutFloat4Components}
  n1510 = LinearInterpolate(
   A=ScalarParameter() {group=TopSide_V02_Layer 03, parameterName=TopS_03 Blend_Contrast}
   B=ScalarParameter() {group=TopSide_V02_Layer 03, parameterName="TopS_03 Blend_Contrast \\Mech"}
   Alpha=@n1509.B
  )
  n1511 = NamedRerouteUsage()
  n1512 = NamedRerouteUsage()
  n1513 = LinearInterpolate(
   A=ScalarParameter() {group=TopSide_V02_Layer 03, parameterName=TopS_03_Alternate Axies Opacity}
   B=ScalarParameter() {group=TopSide_V02_Layer 03, parameterName="TopS_03_Alternate Axies Opacity \\Mech"}
   Alpha=@n1509.B
  )
  n1514 = VectorParameter() {group=TopSide_V02_Layer 03, parameterName=TopS_03_Alternate_Axies Mask, defaultValue=<rgba 0 0 0 1>}
  n1515 = AppendVector(A=@n1514.RGB, B=@n1514.RGB)
  n1516 = StaticSwitch(True=NamedRerouteUsage(), False=StaticSwitch(True=NamedRerouteUsage(), False=NamedRerouteUsage(), Value=@n1424), Value=@n1423)
  n1517 = StaticSwitch(True=NamedRerouteUsage(), False=StaticSwitch(True=NamedRerouteUsage(), False=NamedRerouteUsage(), Value=@n1424), Value=@n1423)
  n1518 = NamedRerouteUsage()
  n1519 = StaticSwitch(
   True=NamedRerouteUsage()
   False=StaticBoolParameter() {group=TopSide_V02_Layer 03, parameterName="TopS_03 TopSide->Side projection"}
   Value=@n1424
  )
  NamedRerouteDeclaration(None=StaticSwitch(
    True=NamedRerouteUsage()
    False=MaterialFunctionCall(
     Texture=@n1504
     Caculated_Bounds=@n1507
     Pre_Skinned_Normals=@n1508
     Blend_Contrast=@n1510
     IsStatsB=@n1511
     IsState=@n1512
     Fill_alternate_Axis_Opacity=@n1513
     "Fill_alternate_Axis_Color_A=Roughness"=@n1515
     Offset=@n1518
     Scale=@n1516
     Rotation_Axis=@n1517
    ) {materialFunction=&MF_TopsSideProjection_Base_V2}.Combined
    Value=@n1519
   )) {name=TS_Layer03 Mask, nodeColor=<rgba 0 1 0.9766 1>}
  n1520 = NamedRerouteUsage()
  n1521 = TextureSample!(UVs=@n792, Tex=@n1520, Level=@n794, Apply View MipBias) {mipValueMode=TMVM_MipLevel}
  n1522 = TextureSample!(UVs=@n792, Tex=@n1520, "DDX(UVs)"=@n797, "DDY(UVs)"=@n798, Apply View MipBias) {mipValueMode=TMVM_Derivative}
  n1523 = NamedRerouteUsage()
  n1524 = NamedRerouteUsage()
  NamedRerouteDeclaration(None=StaticSwitch(
    True=LinearInterpolate(A=@n1521.RGBA, B=@n1522.RGBA, Alpha=@n1523)
    False=LinearInterpolate(A=@n1524, B=@n1522.RGBA, Alpha=@n803)
    Value=@n762
   )) {name=TS 01 Side_BaseColor, nodeColor=<rgba 1 0 0.2812 1>}
  n1525 = NamedRerouteUsage()
  n1526 = TextureSample!(UVs=@n792, Tex=@n1525, Level=@n794, Apply View MipBias) {mipValueMode=TMVM_MipLevel}
  n1527 = TextureSample!(UVs=@n792, Tex=@n1525, "DDX(UVs)"=@n797, "DDY(UVs)"=@n798, Apply View MipBias) {mipValueMode=TMVM_Derivative}
  n1528 = NamedRerouteUsage()
  NamedRerouteDeclaration(None=StaticSwitch(
    True=LinearInterpolate(A=@n1526.RGBA, B=@n1527.RGBA, Alpha=@n1523)
    False=LinearInterpolate(A=@n1528, B=@n1527.RGBA, Alpha=@n803)
    Value=@n762
   )) {name=TS 01 Side_Mask, nodeColor=<rgba 1 0.5625 0 1>}
  n1529 = Constant()
  n1530 = StaticSwitch(True=@n787.Raw_Normal, False=@n790.Raw_Normal, Value=@n791)
  n1531 = StaticSwitch(True=@n787.Clamp_Mask, False=Constant() {r=1}, Value=@n791)
  n1532 = Constant() {r=1}
  n1533 = StaticBoolParameter() {group=TopSide_V02_Layer-01_Uvs, parameterName=TopS_01 Clamp Uvs}
  NamedRerouteDeclaration(None=Multiply(A=LinearInterpolate(A=@n1529, B=@n1530, Alpha=@n803), B=StaticSwitch(True=@n1531, False=@n1532, Value=@n1533))) {name=TS 01 Side_Normal, nodeColor=<rgba 1 0.1172 0 1>}
  NamedRerouteDeclaration(None=MaterialFunctionCall(In=@n517, Contrast=NamedRerouteUsage()) {materialFunction=&CheapContrast}.Result) {name=Tri_Mask 01, nodeColor=<rgba 1 0.4688 0 1>}
  NamedRerouteDeclaration(None=@n770) {name=Tops 01 Blend Contrats, nodeColor=<rgba 0 1 0.6953 1>}
  NamedRerouteDeclaration(None=@n773) {name=Tops 01 Alternate Axies, nodeColor=<rgba 1 0.3984 0 1>}
  NamedRerouteDeclaration(None=@n1430) {name=Tops 01 Alternate Mask, nodeColor=<rgba 1 0.3984 0 1>}
  NamedRerouteDeclaration(None=@n98) {name=Blend TopS_01 to TopS_02, nodeColor=<rgba 1 0 0.4922 1>}
  NamedRerouteDeclaration(None=Add(A=NamedRerouteUsage(), B=@n7)) {name=Uvs_TS_Layer_01_Speed, nodeColor=<rgba 0 0.625 1 1>}
  n1534 = Time()
  n1535 = ScalarParameter() {group="TopSide_V02_Layer 01 Blend  02", parameterName=TopS_Blend_Rotation Speed}
  NamedRerouteDeclaration(None=Add(A=Multiply(A=@n1534, B=@n1535), B=@n456)) {name=Uvs_TS_Blend_Rotation, nodeColor=<rgba 0 0.7188 1 1>}
  n1536 = VectorParameter() {group="TopSide_V02_Layer 01 Blend  02", parameterName=TopS_Blend_Offset_Weapon, defaultValue=<rgba 0 0 0 1>}
  n1537 = VectorParameter() {group="TopSide_V02_Layer 01 Blend  02", parameterName=TopS_Blend_Offset_Vehicle, defaultValue=<rgba 0 0 0 1>}
  n1538 = VectorParameter() {group="TopSide_V02_Layer 01 Blend  02", parameterName=TopS_Blend_Offset_VehicleSide, defaultValue=<rgba 0 0 0 1>}
  n1539 = MaterialFunctionCall(Float4=NamedRerouteUsage()) {materialFunction=&BreakOutFloat4Components}
  n1540 = LinearInterpolate(A=@n1537.RGB, B=@n1538.RGB, Alpha=@n1539.G)
  n1541 = VectorParameter() {group="TopSide_V02_Layer 01 Blend  02", parameterName=TopS_Blend_Offset_Mech, defaultValue=<rgba 0 0 0 1>}
  n1542 = VectorParameter() {group="TopSide_V02_Layer 01 Blend  02", parameterName=TopS_Blend_Offset_Character, defaultValue=<rgba 0 0 0 1>}
  n1543 = VectorParameter() {group="TopSide_V02_Layer 01 Blend  02", parameterName=TopS_Blend_Offset_Backpack, defaultValue=<rgba 0 0 0 1>}
  n1544 = MaterialFunctionCall(Float2=NamedRerouteUsage()) {materialFunction=&BreakOutFloat2Components}
  n1545 = LinearInterpolate(
   A=LinearInterpolate(A=LinearInterpolate(A=@n1536.RGB, B=@n1540, Alpha=@n1539.R), B=LinearInterpolate(A=@n1541.RGB, B=@n1542.RGB, Alpha=@n1539.A), Alpha=@n1539.B)
   B=@n1543.RGB
   Alpha=@n1544.R
  )
  n1546 = VectorParameter() {group="TopSide_V02_Layer 01 Blend  02", parameterName=TopS_Blend_Offset_Pickaxe, defaultValue=<rgba 0 0 0 1>}
  NamedRerouteDeclaration(None=Add(A=LinearInterpolate(A=@n1545, B=@n1546.RGB, Alpha=@n1544.G), B=@n473.RGB)) {name=Uvs_TS_Blend_Offset, nodeColor=<rgba 0 0.625 1 1>}
  n1547 = VectorParameter() {group="TopSide_V02_Layer 01 Blend  02", parameterName=TopS_Blend_Speed__Master, defaultValue=<rgba 0 0 0 1>}
  n1548 = VectorParameter() {group="TopSide_V02_Layer 01 Blend  02", parameterName=TopS_Blend_Speed_Weapon, defaultValue=<rgba 0 0 0 1>}
  n1549 = VectorParameter() {group="TopSide_V02_Layer 01 Blend  02", parameterName=TopS_Blend_Speed_Vehicle, defaultValue=<rgba 0 0 0 1>}
  n1550 = VectorParameter() {group="TopSide_V02_Layer 01 Blend  02", parameterName=TopS_Blend_Speed_VehicleSide, defaultValue=<rgba 0 0 0 1>}
  n1551 = MaterialFunctionCall(Float4=NamedRerouteUsage()) {materialFunction=&BreakOutFloat4Components}
  n1552 = LinearInterpolate(A=@n1548.RGB, B=LinearInterpolate(A=@n1549.RGB, B=@n1550.RGB, Alpha=@n1551.G), Alpha=@n1551.R)
  n1553 = VectorParameter() {group="TopSide_V02_Layer 01 Blend  02", parameterName=TopS_Blend_Speed_Character, defaultValue=<rgba 0 0 0 1>}
  n1554 = LinearInterpolate(
   A=VectorParameter() {group="TopSide_V02_Layer 01 Blend  02", parameterName=TopS_Blend_Speed_Mech, defaultValue=<rgba 0 0 0 1>}.RGB
   B=@n1553.RGB
   Alpha=@n1551.A
  )
  n1555 = VectorParameter() {group="TopSide_V02_Layer 01 Blend  02", parameterName=TopS_Blend_Speed_Backpack, defaultValue=<rgba 0 0 0 1>}
  n1556 = MaterialFunctionCall(Float2=NamedRerouteUsage()) {materialFunction=&BreakOutFloat2Components}
  n1557 = VectorParameter() {group="TopSide_V02_Layer 01 Blend  02", parameterName=TopS_Blend_Speed_Pickaxe, defaultValue=<rgba 0 0 0 1>}
  n1558 = LinearInterpolate(A=LinearInterpolate(A=LinearInterpolate(A=@n1552, B=@n1554, Alpha=@n1551.B), B=@n1555.RGB, Alpha=@n1556.R), B=@n1557.RGB, Alpha=@n1556.G)
  NamedRerouteDeclaration(None=Multiply(A=Add(A=@n1547.RGB, B=@n1558), B=Time())) {name=Uvs_TS_Blend_Speed, nodeColor=<rgba 0 0.625 1 1>}
  n1559 = VectorParameter() {group="TopSide_V02_Layer 01 Blend  02", parameterName=TopS_Blend_Scale_Weapon, defaultValue=<rgba -1 1 1 1>}
  n1560 = Multiply(A=@n1559.RGB, B=@n1559.RGB)
  n1561 = VectorParameter() {group="TopSide_V02_Layer 01 Blend  02", parameterName=TopS_Blend_Scale_Vehicle, defaultValue=<rgba -1 1 1 1>}
  n1562 = VectorParameter() {group="TopSide_V02_Layer 01 Blend  02", parameterName=TopS_Blend_Scale_VehicleSide, defaultValue=<rgba -1 1 1 1>}
  n1563 = MaterialFunctionCall(Float4=NamedRerouteUsage()) {materialFunction=&BreakOutFloat4Components}
  n1564 = LinearInterpolate(A=Multiply(A=@n1561.RGB, B=@n1561.RGB), B=Multiply(A=@n1562.RGB, B=@n1562.RGB), Alpha=@n1563.G)
  n1565 = VectorParameter() {group="TopSide_V02_Layer 01 Blend  02", parameterName=TopS_Blend_Scale_Mech, defaultValue=<rgba -1 1 1 1>}
  n1566 = Multiply(A=@n1565.RGB, B=@n1565.RGB)
  n1567 = VectorParameter() {group="TopSide_V02_Layer 01 Blend  02", parameterName=TopS_Blend_Scale_Character, defaultValue=<rgba -1 1 1 1>}
  n1568 = Multiply(A=@n1567.RGB, B=@n1567.RGB)
  n1569 = VectorParameter() {group="TopSide_V02_Layer 01 Blend  02", parameterName=TopS_Blend_Scale_Backpack, defaultValue=<rgba -1 1 1 1>}
  n1570 = MaterialFunctionCall(Float2=NamedRerouteUsage()) {materialFunction=&BreakOutFloat2Components}
  n1571 = LinearInterpolate(
   A=LinearInterpolate(A=LinearInterpolate(A=@n1560, B=@n1564, Alpha=@n1563.R), B=LinearInterpolate(A=@n1566, B=@n1568, Alpha=@n1563.A), Alpha=@n1563.B)
   B=Multiply(A=@n1569.RGB, B=@n1569.RGB)
   Alpha=@n1570.R
  )
  n1572 = VectorParameter() {group="TopSide_V02_Layer 01 Blend  02", parameterName=TopS_Blend_Scale_Pickaxe, defaultValue=<rgba -1 1 1 1>}
  n1573 = Multiply(A=@n1572.RGB, B=@n1572.RGB)
  n1574 = VectorParameter() {group="TopSide_V02_Layer 01 Blend  02", parameterName=TopS_Blend_Scale__Master, defaultValue=<rgba 1 1 1 1>}
  NamedRerouteDeclaration(None=Multiply(A=LinearInterpolate(A=@n1571, B=@n1573, Alpha=@n1570.G), B=Multiply(A=@n1574.RGB, B=@n1574.RGB))) {name=Uvs_TS_Blend_Scale, nodeColor=<rgba 0 0.625 1 1>}
  n1575 = NamedRerouteUsage()
  n1576 = NamedRerouteUsage()
  n1577 = TextureObjectParameter() {
   group=TopSide_V02_Layer 02
   parameterName=TopS_02__Tex_Mask
   samplerType=SAMPLERTYPE_Masks
   texture=&T_BlackMask
  }
  n1578 = StaticSwitchParameter(
   True=@n852
   False=StaticSwitchParameter(
    True=StaticSwitchParameter(True=@n1575, False=@n1576) {group=TopSide_V02_Layer 02, parameterName=TopS_02__Tex_Mask Share 00B_Msk_Tex}
    False=@n1577
   ) {group=TopSide_V02_Layer 02, parameterName=TopS_02__Tex_Mask Share 00A_Msk_Tex}
  ) {group=TopSide_V02_Layer 02, parameterName=TopS_02_Share_Mask_01_Tex, sortPriority=0}
  n1579 = StaticSwitchParameter(True=NamedRerouteUsage(), False=NamedRerouteUsage()) {group=TopSide_V02_Layer-02_Uvs | Warp, parameterName=TopS_02_Share_Mask_00B_MskTex, sortPriority=249}
  n1580 = TextureObjectParameter() {
   group=TopSide_V02_Layer-02_Uvs | Warp
   parameterName=TopS_02_Tex_Warp
   samplerType=SAMPLERTYPE_Masks
   sortPriority=249
   texture=&T_BlackMask
  }
  NamedRerouteDeclaration(None=StaticSwitchParameter(
    True=NamedRerouteDeclaration(None=@n1578) {name=Tex_TS02_Mask, nodeColor=<rgba 1 0.8906 0 1>}
    False=StaticSwitchParameter(True=@n1579, False=@n1580) {group=TopSide_V02_Layer-02_Uvs | Warp, parameterName=TopS_02_Share_Mask_00A_MskTex, sortPriority=249}
   ) {group=TopSide_V02_Layer-02_Uvs | Warp, parameterName=TopS_02_Share_Mask_02_Tex, sortPriority=249}) {name=Tex_02_Warp, nodeColor=<rgba 1 0.8906 0 1>}
  NamedRerouteDeclaration(None=@n462) {name=Uvs_TS_Grad_Falloff, nodeColor=<rgba 0 0.7188 1 1>}
  n1581 = StaticSwitch(True=NamedRerouteUsage(), False=@n486, Value=@n1255)
  NamedRerouteDeclaration(None=StaticSwitch(True=NamedRerouteUsage(), False=@n1356.Combined, Value=@n1581)) {name=TS_Layer02 BC, nodeColor=<rgba 0 0.625 1 1>}
  n1582 = NamedRerouteUsage()
  n1583 = NamedRerouteUsage()
  n1584 = NamedRerouteUsage()
  n1585 = NamedRerouteUsage()
  n1586 = VectorParameter() {group=TopSide_V02_Layer 02, parameterName=TopS_02_Alternate_Axies Mask, defaultValue=<rgba 0 0 0 1>}
  n1587 = AppendVector(A=@n1586.RGB, B=@n1586.RGB)
  NamedRerouteDeclaration(None=StaticSwitch(
    True=NamedRerouteUsage()
    False=MaterialFunctionCall(
     Texture=@n1582
     Caculated_Bounds=@n1338
     Pre_Skinned_Normals=@n1583
     Blend_Contrast=@n1340
     IsStatsB=@n1584
     IsState=@n1585
     Fill_alternate_Axis_Opacity=@n1341
     "Fill_alternate_Axis_Color_A=Roughness"=@n1587
     Offset=@n1355
     Scale=@n1353
     Rotation_Axis=@n1354
    ) {materialFunction=&MF_TopsSideProjection_Base_V2}.Combined
    Value=@n1581
   )) {name=TS_Layer02 Mask, nodeColor=<rgba 1 0 0.1641 1>}
  n1588 = NamedRerouteUsage()
  n1589 = NamedRerouteUsage()
  n1590 = NamedRerouteUsage()
  n1591 = NamedRerouteUsage()
  NamedRerouteDeclaration(None=StaticSwitch(
    True=NamedRerouteUsage()
    False=MaterialFunctionCall(
     Texture_Normals=@n1588
     Rotation_Axis=@n1354
     Caculated_Bounds=@n1338
     Pre_Skinned_Normals=@n1589
     IsStatsB=@n1590
     IsState=@n1591
     Blend_Contrast=@n1340
     Fill_alternate_Axis_Opacity=@n1341
     Scale=@n1353
     Offset=@n1355
    ) {materialFunction=&MF_TopsSideProjection_Normals_V2}.Add_Surface_Normal
    Value=@n1581
   )) {name=TS_Layer02 Normal, nodeColor=<rgba 0 1 0.1094 1>}
  NamedRerouteDeclaration(None=@n1340) {name=Tops 02 Blend Contrats, nodeColor=<rgba 1 0.3984 0 1>}
  NamedRerouteDeclaration(None=@n1341) {name=Tops 02 Alternate Axies, nodeColor=<rgba 1 0.3984 0 1>}
  NamedRerouteDeclaration(None=@n1587) {name=Tops 02 Alternate Mask, nodeColor=<rgba 1 0.3984 0 1>}
  NamedRerouteDeclaration(None=@n468) {name=Uvs_TS_Layer_01_Rotation_Side_1, nodeColor=<rgba 0 0.7188 1 1>}
  NamedRerouteDeclaration(None=StaticSwitchParameter(True=@n1300, False=@n1277) {group=TopSide_V02_Layer-02_Uvs | Warp, parameterName=TopS_02 Warp_Uvs-----------, sortPriority=0}) {name=Caclcated_Bounds_02, nodeColor=<rgba 0.3594 0 1 1>}
  NamedRerouteDeclaration(None=FunctionInput(Preview=@n1241) {bUsePreviewValueAsDefault=true, inputName=Calcated_Bounds, sortPriority=11}) {name=Caclcated_Bounds, nodeColor=<rgba 0 0.9297 1 1>}
  n1592 = NamedRerouteUsage()
  n1593 = ScalarParameter() {defaultValue=3, group=TopSide_V02_Layer 02, parameterName=TopS_02-Miplevels_}
  n1594 = TextureSample!(UVs=@n1361, Tex=@n1592, Level=@n1593, Apply View MipBias) {mipValueMode=TMVM_MipLevel}
  n1595 = StaticSwitch(True=@n1358.UVs No Offsets/Tiling, False=@n1359.UVs No Offsets/Tiling, Value=@n1360)
  n1596 = DDX(Value=@n1595)
  n1597 = DDY(Value=@n1595)
  n1598 = TextureSample!(UVs=@n1361, Tex=@n1592, "DDX(UVs)"=@n1596, "DDY(UVs)"=@n1597, Apply View MipBias) {mipValueMode=TMVM_Derivative}
  n1599 = NamedRerouteUsage()
  n1600 = NamedRerouteUsage()
  n1601 = OneMinus(None=NamedRerouteUsage())
  n1602 = Saturate(None=Add(A=NamedRerouteUsage(), B=Saturate(None=@n1601)))
  n1603 = StaticSwitch(
   True=NamedRerouteUsage()
   False=StaticBoolParameter() {group=TopSide_V02_Layer 02, parameterName=TopS_02-Miplevels-VS-Alternative_Axies}
   Value=@n1255
  )
  NamedRerouteDeclaration(None=StaticSwitch(
    True=LinearInterpolate(A=@n1594.RGBA, B=@n1598.RGBA, Alpha=@n1599)
    False=LinearInterpolate(A=@n1600, B=@n1598.RGBA, Alpha=@n1602)
    Value=@n1603
   )) {name=TS 021 Side_BaseColor, nodeColor=<rgba 0.8984 0 1 1>}
  n1604 = NamedRerouteUsage()
  n1605 = TextureSample(UVs=@n1361, Tex=@n1604, "DDX(UVs)"=@n1596, "DDY(UVs)"=@n1597) {mipValueMode=TMVM_Derivative}
  n1606 = LinearInterpolate(A=TextureSample!(UVs=@n1361, Tex=@n1604, Level=@n1593, Apply View MipBias) {mipValueMode=TMVM_MipLevel}.RGBA, B=@n1605.RGBA, Alpha=@n1599)
  n1607 = LinearInterpolate(A=NamedRerouteUsage(), B=@n1605.RGBA, Alpha=@n1602)
  n1608 = StaticSwitch(True=StaticSwitch(True=@n1358.Clamp_Mask, False=Constant() {r=1}, Value=@n1360), False=Constant() {r=1}, Value=@n1262)
  NamedRerouteDeclaration(None=Multiply(A=StaticSwitch(True=@n1606, False=@n1607, Value=@n1603), B=@n1608)) {name=TS 02 Side_Mask, nodeColor=<rgba 0 1 0.1562 1>}
  n1609 = Constant()
  n1610 = StaticSwitch(True=@n1358.Raw_Normal, False=@n1359.Raw_Normal, Value=@n1360)
  NamedRerouteDeclaration(None=Multiply(A=LinearInterpolate(A=@n1609, B=@n1610, Alpha=@n1602), B=@n1608)) {name=TS 02 Side_Normal, nodeColor=<rgba 1 0 0.9141 1>}
  n1611 = VectorParameter() {group=TopSide_V02_Layer Emissive_01, parameterName=TopS_Em01_Scale_Weapon, defaultValue=<rgba 1 1 1 1>}
  n1612 = Multiply(A=@n1611.RGB, B=@n1611.RGB)
  n1613 = VectorParameter() {group=TopSide_V02_Layer Emissive_01, parameterName=TopS_Em01_Scale_Vehicle, defaultValue=<rgba 1 1 1 1>}
  n1614 = VectorParameter() {group=TopSide_V02_Layer Emissive_01, parameterName=TopS_Em01_Scale_VehicleSide, defaultValue=<rgba 1 1 1 1>}
  n1615 = MaterialFunctionCall(Float4=NamedRerouteUsage()) {materialFunction=&BreakOutFloat4Components}
  n1616 = LinearInterpolate(A=Multiply(A=@n1613.RGB, B=@n1613.RGB), B=Multiply(A=@n1614.RGB, B=@n1614.RGB), Alpha=@n1615.G)
  n1617 = VectorParameter() {group=TopSide_V02_Layer Emissive_01, parameterName=TopS_Em01_Scale_Mech, defaultValue=<rgba 1 1 1 1>}
  n1618 = Multiply(A=@n1617.RGB, B=@n1617.RGB)
  n1619 = VectorParameter() {group=TopSide_V02_Layer Emissive_01, parameterName=TopS_Em01_Scale_Character, defaultValue=<rgba 1 1 1 1>}
  n1620 = Multiply(A=@n1619.RGB, B=@n1619.RGB)
  n1621 = VectorParameter() {group=TopSide_V02_Layer Emissive_01, parameterName=TopS_Em01_Scale_Backpack, defaultValue=<rgba 1 1 1 1>}
  n1622 = MaterialFunctionCall(Float2=NamedRerouteUsage()) {materialFunction=&BreakOutFloat2Components}
  n1623 = LinearInterpolate(
   A=LinearInterpolate(A=LinearInterpolate(A=@n1612, B=@n1616, Alpha=@n1615.R), B=LinearInterpolate(A=@n1618, B=@n1620, Alpha=@n1615.A), Alpha=@n1615.B)
   B=Multiply(A=@n1621.RGB, B=@n1621.RGB)
   Alpha=@n1622.R
  )
  n1624 = VectorParameter() {group=TopSide_V02_Layer Emissive_01, parameterName=TopS_Em01_Scale_Pickaxe, defaultValue=<rgba 1 1 1 1>}
  n1625 = Multiply(A=@n1624.RGB, B=@n1624.RGB)
  n1626 = VectorParameter() {group=TopSide_V02_Layer Emissive_01, parameterName=TopS_Em01_Scale__Master, defaultValue=<rgba 1 1 1 1>}
  NamedRerouteDeclaration(None=Multiply(A=LinearInterpolate(A=@n1623, B=@n1625, Alpha=@n1622.G), B=Multiply(A=@n1626.RGB, B=@n1626.RGB))) {name=Uvs_TS_Emis01_Scale, nodeColor=<rgba 0 0.625 1 1>}
  n1627 = VectorParameter() {group=TopSide_V02_Layer Emissive_01, parameterName=TopS_Em01_Offset__Master, defaultValue=<rgba 0 0 0 1>}
  n1628 = VectorParameter() {group=TopSide_V02_Layer Emissive_01, parameterName=TopS_Em01_Offset_Weapon, defaultValue=<rgba 0 0 0 1>}
  n1629 = VectorParameter() {group=TopSide_V02_Layer Emissive_01, parameterName=TopS_Em01_Offset_Vehicle, defaultValue=<rgba 0 0 0 1>}
  n1630 = VectorParameter() {group=TopSide_V02_Layer Emissive_01, parameterName=TopS_Em01_Offset_VehicleSide, defaultValue=<rgba 0 0 0 1>}
  n1631 = MaterialFunctionCall(Float4=NamedRerouteUsage()) {materialFunction=&BreakOutFloat4Components}
  n1632 = LinearInterpolate(A=@n1629.RGB, B=@n1630.RGB, Alpha=@n1631.G)
  n1633 = VectorParameter() {group=TopSide_V02_Layer Emissive_01, parameterName=TopS_Em01_Offset_Mech, defaultValue=<rgba 0 0 0 1>}
  n1634 = VectorParameter() {group=TopSide_V02_Layer Emissive_01, parameterName=TopS_Em01_Offset_Character, defaultValue=<rgba 0 0 0 1>}
  n1635 = VectorParameter() {group=TopSide_V02_Layer Emissive_01, parameterName=TopS_Em01_Offset_Backpack, defaultValue=<rgba 0 0 0 1>}
  n1636 = MaterialFunctionCall(Float2=NamedRerouteUsage()) {materialFunction=&BreakOutFloat2Components}
  n1637 = LinearInterpolate(
   A=LinearInterpolate(A=LinearInterpolate(A=@n1628.RGB, B=@n1632, Alpha=@n1631.R), B=LinearInterpolate(A=@n1633.RGB, B=@n1634.RGB, Alpha=@n1631.A), Alpha=@n1631.B)
   B=@n1635.RGB
   Alpha=@n1636.R
  )
  n1638 = VectorParameter() {group=TopSide_V02_Layer Emissive_01, parameterName=TopS_Em01_Offset_Pickaxe, defaultValue=<rgba 0 0 0 1>}
  NamedRerouteDeclaration(None=Add(A=@n1627.RGB, B=LinearInterpolate(A=@n1637, B=@n1638.RGB, Alpha=@n1636.G))) {name=Uvs_TS_Emis01_Offset, nodeColor=<rgba 0 0.625 1 1>}
  n1639 = VectorParameter() {group=TopSide_V02_Layer Emissive_01, parameterName=TopS_Em01_Panning_Speed------------, defaultValue=<rgba 0 0 0 1>}
  NamedRerouteDeclaration(None=Multiply(A=Time(), B=@n1639.RGB)) {name=Uvs_TS_Emis01_Speed, nodeColor=<rgba 0 0.625 1 1>}
  n1640 = ScalarParameter() {defaultValue=1, group=TopSide_V02_Layer Emissive_01, parameterName=TopS_Em01_Falloff_Vehicle}
  n1641 = ScalarParameter() {defaultValue=1, group=TopSide_V02_Layer Emissive_01, parameterName=TopS_Em01_Falloff_Vehicle_Side}
  n1642 = MaterialFunctionCall(Float4=NamedRerouteUsage()) {materialFunction=&BreakOutFloat4Components}
  n1643 = LinearInterpolate(
   A=LinearInterpolate(
    A=ScalarParameter() {defaultValue=1, group=TopSide_V02_Layer Emissive_01, parameterName=TopS_Em01_Falloff_Weapon}
    B=LinearInterpolate(A=@n1640, B=@n1641, Alpha=@n1642.G)
    Alpha=@n1642.R
   )
   B=LinearInterpolate(
    A=ScalarParameter() {defaultValue=1, group=TopSide_V02_Layer Emissive_01, parameterName=TopS_Em01_Falloff_Mech}
    B=ScalarParameter() {defaultValue=1, group=TopSide_V02_Layer Emissive_01, parameterName=TopS_Em01_Falloff_Character}
    Alpha=@n1642.A
   )
   Alpha=@n1642.B
  )
  n1644 = ScalarParameter() {defaultValue=1, group=TopSide_V02_Layer Emissive_01, parameterName=TopS_Em01_Falloff_Backpack}
  n1645 = MaterialFunctionCall(Float2=NamedRerouteUsage()) {materialFunction=&BreakOutFloat2Components}
  NamedRerouteDeclaration(None=LinearInterpolate(
    A=LinearInterpolate(A=@n1643, B=@n1644, Alpha=@n1645.R)
    B=ScalarParameter() {defaultValue=1, group=TopSide_V02_Layer Emissive_01, parameterName=TopS_Em01_Falloff_Pickaxe}
    Alpha=@n1645.G
   )) {name=Uvs_TS_Emis_01_Grad_Falloff, nodeColor=<rgba 0 0.7188 1 1>}
  n1646 = Time()
  n1647 = VectorParameter() {group=TopSide_V02_Layer Emissive_01, parameterName=TopS_Em01_Speed_Vehicle, defaultValue=<rgba 0 0 0 1>}
  n1648 = VectorParameter() {group=TopSide_V02_Layer Emissive_01, parameterName=TopS_Em01_Speed_VehicleSide, defaultValue=<rgba 0 0 0 1>}
  n1649 = MaterialFunctionCall(Float4=NamedRerouteUsage()) {materialFunction=&BreakOutFloat4Components}
  n1650 = VectorParameter() {group=TopSide_V02_Layer Emissive_01, parameterName=TopS_Em01_Speed_Character, defaultValue=<rgba 0 0 0 1>}
  n1651 = LinearInterpolate(
   A=LinearInterpolate(
    A=VectorParameter() {group=TopSide_V02_Layer Emissive_01, parameterName=TopS_Em01_Speed_Weapon, defaultValue=<rgba 0 0 0 1>}.RGB
    B=LinearInterpolate(A=@n1647.RGB, B=@n1648.RGB, Alpha=@n1649.G)
    Alpha=@n1649.R
   )
   B=LinearInterpolate(
    A=VectorParameter() {group=TopSide_V02_Layer Emissive_01, parameterName=TopS_Em01_Speed_Mech, defaultValue=<rgba 0 0 0 1>}.RGB
    B=@n1650.RGB
    Alpha=@n1649.A
   )
   Alpha=@n1649.B
  )
  n1652 = VectorParameter() {group=TopSide_V02_Layer Emissive_01, parameterName=TopS_Em01_Speed_Backpack, defaultValue=<rgba 0 0 0 1>}
  n1653 = MaterialFunctionCall(Float2=NamedRerouteUsage()) {materialFunction=&BreakOutFloat2Components}
  n1654 = VectorParameter() {group=TopSide_V02_Layer Emissive_01, parameterName=TopS_Em01_Speed_Pickaxe, defaultValue=<rgba 0 0 0 1>}
  n1655 = VectorParameter() {group=TopSide_V02_Layer Emissive_01, parameterName=TopS_Em01_Speed__Master, defaultValue=<rgba 0 0 0 1>}
  n1656 = Add(A=LinearInterpolate(A=LinearInterpolate(A=@n1651, B=@n1652.RGB, Alpha=@n1653.R), B=@n1654.RGB, Alpha=@n1653.G), B=@n1655.RGB)
  NamedRerouteDeclaration(None=Add(A=Multiply(A=@n1646, B=@n1656), B=NamedRerouteUsage())) {name=Uvs_TS_Emis01_Speed_B, nodeColor=<rgba 0 0.625 1 1>}
  n1657 = NamedRerouteUsage()
  n1658 = LinearInterpolate(
   A=@n1657
   B=OneMinus(None=@n1657)
   Alpha=ScalarParameter() {group=TopSide_V02_Layer Emissive_01, parameterName="TopS Emissive_01_Blend 01---> 02 Invert Msk  1=inverted", sortPriority=1}
  )
  n1659 = NamedRerouteUsage()
  n1660 = Max(A=@n184, B=@n186)
  n1661 = ScalarParameter() {
   defaultValue=1
   group=TopSide_V02_Layer Emissive_01 Reactive
   parameterName="TopS Emissive_01-Brightness-Reactive--->Gradent BloomFix  0=Active"
  }
  n1662 = Saturate(None=Ceil(None=Min(A=@n1660, B=@n1661)))
  n1663 = ScalarParameter() {group=TopSide_V02_Layer Emissive_01, parameterName=TopS Emissive_01_ColorCuve_Offset}
  n1664 = Multiply(A=ScalarParameter() {group=TopSide_V02_Layer Emissive_01, parameterName=TopS Emissive_01_ColorCuve_Speed}, B=Time())
  n1665 = ScalarParameter() {defaultValue=1, group=TopSide_V02_Layer Emissive_01, parameterName=TopS Emissive_01_ColorCuve_TileRate}
  n1666 = StaticSwitch(
   True=NamedRerouteUsage()
   False=@n197
   Value=StaticBoolParameter() {desc=setting at bottom, group=TopSide_V02_Layer Emissive_01, parameterName="TopS Emissive_01_ColorCuve--->Use-UVS"}
  )
  n1667 = Frac(None=Add(A=Add(A=@n1663, B=@n1664), B=Multiply(A=@n1665, B=@n1666)))
  n1668 = CurveAtlasRowParameter(CurveTime=@n1667) {
   atlas=&CruveAtlas_Wraps_01
   curve=&CC_Spectum
   defaultValue=98
   group=TopSide_V02_Layer Emissive_01
   parameterName=TopS Emissive_01_ColorCuve B
  }
  n1669 = MaterialFunctionCall(
   Index=ScalarParameter() {defaultValue=98, group=TopSide_V02_Layer Emissive_01, parameterName=TopS Emissive_01_ColorCuve B Index}
   CCinput=@n1668
   CCinput_Alpha=@n1668
   UV=@n1667
  ) {materialFunction=&MF_Global_CC_Tex}
  n1670 = CurveAtlasRowParameter(CurveTime=@n1667) {
   atlas=&CruveAtlas_Wraps_01
   curve=&CC_Spectum
   defaultValue=98
   group=TopSide_V02_Layer Emissive_01
   parameterName=TopS Emissive_01_ColorCuve
  }
  n1671 = MaterialFunctionCall(
   Index=ScalarParameter() {defaultValue=98, group=TopSide_V02_Layer Emissive_01, parameterName="TopS Emissive_01_ColorCuve  Index"}
   CCinput=@n1670
   CCinput_Alpha=@n1670
   UV=@n1667
  ) {materialFunction=&MF_Global_CC_Tex}
  n1672 = VectorParameter() {group=TopSide_V02_Layer Emissive_01, parameterName=TopS Emissive_01_ColorCuve Tint, defaultValue=<rgba 1 1 1 1>}
  n1673 = Multiply(A=@n1671.3V, B=@n1672.RGB)
  n1674 = NamedRerouteUsage()
  n1675 = StaticBoolParameter() {group=TopSide_V02_Layer Emissive_01, parameterName=TopS Emissive_01_Color TOD}
  n1676 = StaticSwitch(True=LinearInterpolate(A=@n1669.3V, B=@n1673, Alpha=@n1674), False=@n1673, Value=@n1675)
  n1677 = ScalarParameter() {defaultValue=1, group=TopSide_V02_Layer Emissive_01, parameterName=TopS Emissive_01_ColorCuve-Brightness}
  n1678 = StaticSwitch(True=LinearInterpolate(A=@n1669.Alpha, B=@n1671.Alpha, Alpha=@n1674), False=@n1671.Alpha, Value=@n1675)
  n1679 = LinearInterpolate(A=@n1677, B=Multiply(A=@n1678, B=@n1677), Alpha=NamedRerouteUsage())
  n1680 = Add(A=@n197, B=ScalarParameter() {group=TopSide_V02_Layer Emissive_01, parameterName="TopS Emissive_01_ColorCuve-CurveMask   0=On"})
  n1681 = VectorParameter() {group=TopSide_V02_Layer Emissive_01, parameterName=TopS Emissive_01_Color-B, defaultValue=<rgba 0 0 0 1>}
  n1682 = VectorParameter() {group=TopSide_V02_Layer Emissive_01, parameterName=TopS Emissive_01_Color-, defaultValue=<rgba 0 0 0 1>}
  n1683 = LinearInterpolate(A=@n1681.RGB, B=@n1682.RGB, Alpha=NamedRerouteUsage())
  n1684 = StaticSwitchParameter(
   True=Multiply(A=Multiply(A=@n1676, B=@n1679), B=Saturate(None=@n1680))
   False=Multiply(A=StaticSwitch(True=@n1683, False=@n1682.RGB, Value=@n1675), B=@n197)
  ) {group=TopSide_V02_Layer Emissive_01, parameterName="TopS Emissive_01_Color ----->ColorCurve"}
  n1685 = StaticSwitch(True=Multiply(A=@n1684, B=NamedRerouteUsage()), False=@n1684, Value=NamedRerouteUsage())
  n1686 = StaticSwitch(
   True=Multiply(A=LinearInterpolate(A=@n1659, Alpha=@n1662) {constA=1}, B=@n1685)
   False=@n1685
   Value=NamedRerouteDeclaration(None=StaticBoolParameter() {group=TopSide_V02_Layer Emissive_01 Reactive, parameterName=TopS Emissive_01-Brightness-Reactive--------------}) {name=Emissive_01_Reactive_bool, nodeColor=<rgba 0.3594 0 1 1>}
  )
  n1687 = StaticBoolParameter() {group=TopSide_V02_Layer Emissive_01, parameterName="TopS Emissive_01_Blend 01---> 02 Affects Mask", sortPriority=1}
  n1688 = StaticSwitch(True=StaticSwitch(True=Multiply(A=@n1658, B=@n1686), False=@n1686, Value=@n1687), False=@n1686, Value=NamedRerouteUsage())
  n1689 = StaticSwitch(
   True=NamedRerouteUsage()
   False=NamedRerouteUsage()
   Value=StaticBoolParameter() {group=TopSide_V02_Layer Emissive_01 Reactive, parameterName="TopS Emissive_01-TOD---->TODB", sortPriority=0}
  )
  n1690 = StaticBoolParameter() {group=TopSide_V02_Layer Emissive_01 Reactive, parameterName=TopS Emissive_01-TOD, sortPriority=0}
  n1691 = StaticSwitch(True=LinearInterpolate(B=@n1688, Alpha=@n1689) {constB=0}, False=@n1688, Value=@n1690)
  n1692 = VectorParameter() {
   group=TopSide_V02_Layer Emissive_01
   parameterName=TopS Emissive_01 LiquidCCMask
   channelNames {a="", b="", g=invert, r="0=Use"}
   defaultValue=<rgba 1 0 0 1>
  }
  n1693 = ChannelMaskParameter(None=NamedRerouteUsage()) {group=TopSide_V02_Layer Emissive_01, maskChannel=Alpha, parameterName=TopS Emissive_01 LiquidCCMask Channel, defaultValue=<rgba 0 0 0 1>}
  n1694 = Max(A=@n1692."0=Use", B=LinearInterpolate(A=@n1693.R, B=OneMinus(None=@n1693.R), Alpha=@n1692.invert))
  n1695 = NamedRerouteUsage()
  NamedRerouteDeclaration(None=StaticSwitch(True=LinearInterpolate(B=@n1691, Alpha=@n1694), False=@n1691, Value=@n1695)) {name=Emissive_01, nodeColor=<rgba 0.2891 1 0 1>}
  NamedRerouteDeclaration(None=@n351) {name=Uvs_TS_Layer_02_Rotation, nodeColor=<rgba 1 0 0.3281 1>}
  n1696 = PreSkinnedPosition()
  NamedRerouteDeclaration(None=FunctionInput(Preview=VertexInterpolator(VS=@n1696).PS) {bUsePreviewValueAsDefault=true, inputName=Local Position, sortPriority=11}) {name=PreSkin_Position, nodeColor=<rgba 0.8281 1 0 1>}
  n1697 = VectorParameter() {group=TopSide_V02_Layer-02_Uvs, parameterName=TopS_02_Scale_Weapon, defaultValue=<rgba 1 1 1 1>}
  n1698 = VectorParameter() {group=TopSide_V02_Layer-02_Uvs, parameterName=TopS_02_Scale_Vehicle, defaultValue=<rgba 1 1 1 1>}
  n1699 = Multiply(A=@n1698.RGB, B=@n1698.RGB)
  n1700 = VectorParameter() {group=TopSide_V02_Layer-02_Uvs, parameterName=TopS_02_Scale_VehicleSide, defaultValue=<rgba 1 1 1 1>}
  n1701 = Multiply(A=@n1700.RGB, B=@n1700.RGB)
  n1702 = MaterialFunctionCall(Float4=NamedRerouteUsage()) {materialFunction=&BreakOutFloat4Components}
  n1703 = VectorParameter() {group=TopSide_V02_Layer-02_Uvs, parameterName=TopS_02_Scale_Mech, defaultValue=<rgba 1 1 1 1>}
  n1704 = VectorParameter() {group=TopSide_V02_Layer-02_Uvs, parameterName=TopS_02_Scale_Character, defaultValue=<rgba 1 1 1 1>}
  n1705 = LinearInterpolate(
   A=LinearInterpolate(A=Multiply(A=@n1697.RGB, B=@n1697.RGB), B=LinearInterpolate(A=@n1699, B=@n1701, Alpha=@n1702.G), Alpha=@n1702.R)
   B=LinearInterpolate(A=Multiply(A=@n1703.RGB, B=@n1703.RGB), B=Multiply(A=@n1704.RGB, B=@n1704.RGB), Alpha=@n1702.A)
   Alpha=@n1702.B
  )
  n1706 = VectorParameter() {group=TopSide_V02_Layer-02_Uvs, parameterName=TopS_02_Scale_Backpack, defaultValue=<rgba 1 1 1 1>}
  n1707 = Multiply(A=@n1706.RGB, B=@n1706.RGB)
  n1708 = MaterialFunctionCall(Float2=NamedRerouteUsage()) {materialFunction=&BreakOutFloat2Components}
  n1709 = VectorParameter() {group=TopSide_V02_Layer-02_Uvs, parameterName=TopS_02_Scale_Pickaxe, defaultValue=<rgba 1 1 1 1>}
  NamedRerouteDeclaration(None=LinearInterpolate(A=LinearInterpolate(A=@n1705, B=@n1707, Alpha=@n1708.R), B=Multiply(A=@n1709.RGB, B=@n1709.RGB), Alpha=@n1708.G)) {name=Uvs_TS_Layer_02_Scale, nodeColor=<rgba 0 0.625 1 1>}
  n1710 = VectorParameter() {group=TopSide_V02_Layer-02_Uvs, parameterName=TopS_02_Offset_Vehicle, defaultValue=<rgba 0 0 0 1>}
  n1711 = VectorParameter() {group=TopSide_V02_Layer-02_Uvs, parameterName=TopS_02_Offset_VehicleSide, defaultValue=<rgba 0 0 0 1>}
  n1712 = MaterialFunctionCall(Float4=NamedRerouteUsage()) {materialFunction=&BreakOutFloat4Components}
  n1713 = LinearInterpolate(
   A=LinearInterpolate(
    A=VectorParameter() {group=TopSide_V02_Layer-02_Uvs, parameterName=TopS_02_Offset_Weapon, defaultValue=<rgba 0 0 0 1>}.RGB
    B=LinearInterpolate(A=@n1710.RGB, B=@n1711.RGB, Alpha=@n1712.G)
    Alpha=@n1712.R
   )
   B=LinearInterpolate(
    A=VectorParameter() {group=TopSide_V02_Layer-02_Uvs, parameterName=TopS_02_Offset_Mech, defaultValue=<rgba 0 0 0 1>}.RGB
    B=VectorParameter() {group=TopSide_V02_Layer-02_Uvs, parameterName=TopS_02_Offset_Character, defaultValue=<rgba 0 0 0 1>}.RGB
    Alpha=@n1712.A
   )
   Alpha=@n1712.B
  )
  n1714 = VectorParameter() {group=TopSide_V02_Layer-02_Uvs, parameterName=TopS_02_Offset_Backpack, defaultValue=<rgba 0 0 0 1>}
  n1715 = MaterialFunctionCall(Float2=NamedRerouteUsage()) {materialFunction=&BreakOutFloat2Components}
  NamedRerouteDeclaration(None=LinearInterpolate(
    A=LinearInterpolate(A=@n1713, B=@n1714.RGB, Alpha=@n1715.R)
    B=VectorParameter() {group=TopSide_V02_Layer-02_Uvs, parameterName=TopS_02_Offset_Pickaxe, defaultValue=<rgba 0 0 0 1>}.RGB
    Alpha=@n1715.G
   )) {name=Uvs_TS_Layer_02_Offset, nodeColor=<rgba 0 0.625 1 1>}
  NamedRerouteDeclaration(None=@n360) {name=Uvs_TS_Layer_02_Warp_Rotation, nodeColor=<rgba 0 0.7188 1 1>}
  n1716 = ScalarParameter() {defaultValue=1, group=TopSide_V02_Layer-02_Uvs | Warp, parameterName=TopS_02_Warp_Scale_Vehicle, sortPriority=300}
  n1717 = ScalarParameter() {defaultValue=1, group=TopSide_V02_Layer-02_Uvs | Warp, parameterName=TopS_02_Warp_Scale_Vehicle_Side, sortPriority=300}
  n1718 = MaterialFunctionCall(Float4=NamedRerouteUsage()) {materialFunction=&BreakOutFloat4Components}
  n1719 = LinearInterpolate(
   A=LinearInterpolate(
    A=ScalarParameter() {defaultValue=1, group=TopSide_V02_Layer-02_Uvs | Warp, parameterName=TopS_02_Warp_Scale_Weapon, sortPriority=300}
    B=LinearInterpolate(A=@n1716, B=@n1717, Alpha=@n1718.G)
    Alpha=@n1718.R
   )
   B=LinearInterpolate(
    A=ScalarParameter() {defaultValue=1, group=TopSide_V02_Layer-02_Uvs | Warp, parameterName=TopS_02_Warp_Scale_Mech, sortPriority=300}
    B=ScalarParameter() {defaultValue=1, group=TopSide_V02_Layer-02_Uvs | Warp, parameterName=TopS_02_Warp_Scale_Character, sortPriority=300}
    Alpha=@n1718.A
   )
   Alpha=@n1718.B
  )
  n1720 = ScalarParameter() {defaultValue=1, group=TopSide_V02_Layer-02_Uvs | Warp, parameterName=TopS_02_Warp_Scale_Backpack, sortPriority=300}
  n1721 = MaterialFunctionCall(Float2=NamedRerouteUsage()) {materialFunction=&BreakOutFloat2Components}
  NamedRerouteDeclaration(None=LinearInterpolate(
    A=LinearInterpolate(A=@n1719, B=@n1720, Alpha=@n1721.R)
    B=ScalarParameter() {defaultValue=1, group=TopSide_V02_Layer-02_Uvs | Warp, parameterName=TopS_02_Warp_Scale_Pickaxe, sortPriority=300}
    Alpha=@n1721.G
   )) {name=Uvs_TS_Layer_02_Warp_Scale, nodeColor=<rgba 0 0.7188 1 1>}
  n1722 = VectorParameter() {
   group=TopSide_V02_Layer-02_Uvs
   parameterName=TopS_02_Panning_Speed------------
   channelNames {a="0=step animation", b="", g="", r=""}
   defaultValue=<rgba 0 0 0 1>
  }
  n1723 = Multiply(A=Time(), B=@n1722)
  n1724 = Frac(None=@n1723)
  n1725 = ScalarParameter() {defaultValue=10, group=TopSide_V02_Layer-02_Uvs, parameterName=TopS_02_PanningS_Speed_Step_Amount}
  n1726 = Divide(A=Ceil(None=Multiply(A=@n1724, B=@n1725) {constB=30}), B=@n1725)
  NamedRerouteDeclaration(None=Add(A=NamedRerouteUsage(), B=LinearInterpolate(A=@n1726, B=@n1723, Alpha=@n1722."0=step animation"))) {name=Uvs_TS_Layer_02_Speed, nodeColor=<rgba 0 0.625 1 1>}
  n1727 = Time()
  n1728 = ScalarParameter() {group=TopSide_V02_Layer-01_Uvs, parameterName=TopS_01_Side_Rotation__Speed}
  NamedRerouteDeclaration(None=Add(A=Multiply(A=@n1727, B=@n1728), B=@n424)) {name=Uvs_TS_Layer_01_Rotation_Side, nodeColor=<rgba 0 0.7188 1 1>}
  n1729 = Time()
  n1730 = ScalarParameter() {group=TopSide_V02_Layer-02_Uvs, parameterName=TopS_02_Side_Rotation__Speed}
  NamedRerouteDeclaration(None=Add(A=Multiply(A=@n1729, B=@n1730), B=@n366)) {name=Uvs_TS_Layer_02_Rotation_side, nodeColor=<rgba 0 0.7188 1 1>}
  n1731 = Time()
  n1732 = ScalarParameter() {group=TopSide_V02_Layer Emissive_01, parameterName=TopS_Em01_Rotation Speed}
  NamedRerouteDeclaration(None=Add(A=Multiply(A=@n1731, B=@n1732), B=@n372)) {name=Uvs_TS_Emis_01_Rotation, nodeColor=<rgba 0 0.7188 1 1>}
  n1733 = VectorParameter() {group=TopSide_V02_Layer 02, parameterName=TopS_02_Base Color_Simple, defaultValue=<rgba 0 0 0 1>}
  n1734 = NamedRerouteUsage()
  n1735 = StaticSwitchParameter(True=ChannelMaskParameter(None=@n1734) {group=TopSide_V02_Layer 02, parameterName=TopS_02_Base_Roughness_Mask_Channel}.R, False=@n1733.A) {group=TopSide_V02_Layer 02, parameterName=TopS_02_Base_Roughness Msk------}
  n1736 = StaticSwitchParameter(
   True=StaticSwitchParameter(True=NamedRerouteUsage(), False=AppendVector(A=@n1733.RGB, B=@n1735)) {desc=uses TopS_02_Alternate Axies_Colors, group=TopSide_V02_Layer 02, parameterName="TopS_02_Base Color Simple---->3Colors"}
   False=NamedRerouteUsage()
  ) {group=TopSide_V02_Layer 02, parameterName=TopS_02_Base Color Simple}
  NamedRerouteDeclaration(None=ComponentMask(None=@n1736) {b=true, g=true, r=true}) {name=Layer 02 BC_Output, nodeColor=<rgba 0 1 0.4844 1>}
  NamedRerouteDeclaration(None=LinearInterpolate(
    A=ScalarParameter() {group=TopSide_V02_Layer 02, parameterName=TopS_02_Base_Roughness_High}
    B=ScalarParameter() {defaultValue=1, group=TopSide_V02_Layer 02, parameterName=TopS_02_Base_Roughness_Low}
    Alpha=ComponentMask(None=@n1736) {a=true}
   )) {name=Layer 02 Roughness_Output, nodeColor=<rgba 0 1 0.4844 1>}
  n1737 = PreSkinnedNormal()
  NamedRerouteDeclaration(None=FunctionInput(Preview=VertexInterpolator(VS=@n1737).PS) {bUsePreviewValueAsDefault=true, inputName=Pre-Skinned Local Normal, sortPriority=10}) {name=PreSkin_Normal, nodeColor=<rgba 0.8281 1 0 1>}
  n1738 = ScalarParameter() {group=TopSide_V02_Layer 02, parameterName=TopS_02_Base_Metalic_Min}
  n1739 = ScalarParameter() {defaultValue=1, group=TopSide_V02_Layer 02, parameterName=TopS_02_Base_Metalic_Max}
  n1740 = ChannelMaskParameter(None=@n1734) {group=TopSide_V02_Layer 02, parameterName=TopS_02_Base_Metalic_Mask_Channel}
  NamedRerouteDeclaration(None=StaticSwitchParameter(
    True=LinearInterpolate(A=@n1738, B=@n1739, Alpha=@n1740.R)
    False=ScalarParameter() {group=TopSide_V02_Layer 02, parameterName=TopS_02_Base_Metalic_}
   ) {group=TopSide_V02_Layer 02, parameterName=TopS_02_Base_Metalic Mask----------}) {name=Layer 02 Metal_Output, nodeColor=<rgba 0 1 0.4844 1>}
  n1741 = ScalarParameter() {group=TopSide_V02_Layer 02, parameterName=TopS_02_Base_Spec_Min}
  n1742 = ScalarParameter() {defaultValue=1, group=TopSide_V02_Layer 02, parameterName=TopS_02_Base_Spec_Max}
  n1743 = ChannelMaskParameter(None=@n1734) {group=TopSide_V02_Layer 02, parameterName=TopS_02_Base_Spec_Mask_Channel}
  NamedRerouteDeclaration(None=StaticSwitchParameter(
    True=LinearInterpolate(A=@n1741, B=@n1742, Alpha=@n1743.R)
    False=ScalarParameter() {defaultValue=0.5, group=TopSide_V02_Layer 02, parameterName=TopS_02_Base_Spec_}
   ) {group=TopSide_V02_Layer 02, parameterName=TopS_02_Base_Spec Mask----------}) {name=Layer 02 Spec_Output, nodeColor=<rgba 0 1 0.4844 1>}
  n1744 = NamedRerouteUsage()
  n1745 = LinearInterpolate(
   A=@n1744
   B=Add(A=@n1744, B=NamedRerouteUsage())
   Alpha=ScalarParameter() {defaultValue=1, group=TopSide_V02_Layer 02, parameterName=TopS_02_Base Normal_Intensty}
  )
  n1746 = ShadingPathSwitch(Default=@n1745, Deferred=@n1745, Mobile=@n1744) {desc=Migrated from FeatureLevelSwitch}
  NamedRerouteDeclaration(None=StaticSwitchParameter(
    True=@n1744
    False=StaticSwitchParameter(True=@n1746, False=@n1745) {group=TopSide_V02_Layer 02, parameterName=TopS_02_Base Normal Android-Disable}
   ) {defaultValue=true, group=TopSide_V02_Layer 02, parameterName=TopS_02_Base Normal Simple}) {name=Layer 02 Normal_Output, nodeColor=<rgba 0 1 0.4844 1>}
  n1747 = VectorParameter() {group=TopSide_V02_Layer-01_Uvs, parameterName=TopS_01_OffsetS_Vehicle, defaultValue=<rgba 0 0 0 1>}
  n1748 = VectorParameter() {group=TopSide_V02_Layer-01_Uvs, parameterName=TopS_01_OffsetS_VehicleSide, defaultValue=<rgba 0 0 0 1>}
  n1749 = MaterialFunctionCall(Float4=NamedRerouteUsage()) {materialFunction=&BreakOutFloat4Components}
  n1750 = LinearInterpolate(
   A=LinearInterpolate(
    A=VectorParameter() {group=TopSide_V02_Layer-01_Uvs, parameterName=TopS_01_OffsetS_Weapon, defaultValue=<rgba 0 0 0 1>}.RGB
    B=LinearInterpolate(A=@n1747.RGB, B=@n1748.RGB, Alpha=@n1749.G)
    Alpha=@n1749.R
   )
   B=LinearInterpolate(
    A=VectorParameter() {group=TopSide_V02_Layer-01_Uvs, parameterName=TopS_01_OffsetS_Mech, defaultValue=<rgba 0 0 0 1>}.RGB
    B=VectorParameter() {group=TopSide_V02_Layer-01_Uvs, parameterName=TopS_01_OffsetS_Character, defaultValue=<rgba 0 0 0 1>}.RGB
    Alpha=@n1749.A
   )
   Alpha=@n1749.B
  )
  n1751 = VectorParameter() {group=TopSide_V02_Layer-01_Uvs, parameterName=TopS_01_OffsetS_Backpack, defaultValue=<rgba 0 0 0 1>}
  n1752 = MaterialFunctionCall(Float2=NamedRerouteUsage()) {materialFunction=&BreakOutFloat2Components}
  NamedRerouteDeclaration(None=LinearInterpolate(
    A=LinearInterpolate(A=@n1750, B=@n1751.RGB, Alpha=@n1752.R)
    B=VectorParameter() {group=TopSide_V02_Layer-01_Uvs, parameterName=TopS_01_OffsetS_Pickaxe, defaultValue=<rgba 0 0 0 1>}.RGB
    Alpha=@n1752.G
   )) {name=Uvs_TS_Layer_01_Offset_Side, nodeColor=<rgba 0 0.625 1 1>}
  n1753 = VectorParameter() {group=TopSide_V02_Layer-02_Uvs, parameterName=TopS_02_OffsetS_Vehicle, defaultValue=<rgba 0 0 0 1>}
  n1754 = VectorParameter() {group=TopSide_V02_Layer-02_Uvs, parameterName=TopS_02_OffsetS_VehicleSide, defaultValue=<rgba 0 0 0 1>}
  n1755 = MaterialFunctionCall(Float4=NamedRerouteUsage()) {materialFunction=&BreakOutFloat4Components}
  n1756 = LinearInterpolate(
   A=LinearInterpolate(
    A=VectorParameter() {group=TopSide_V02_Layer-02_Uvs, parameterName=TopS_02_OffsetS_Weapon, defaultValue=<rgba 0 0 0 1>}.RGB
    B=LinearInterpolate(A=@n1753.RGB, B=@n1754.RGB, Alpha=@n1755.G)
    Alpha=@n1755.R
   )
   B=LinearInterpolate(
    A=VectorParameter() {group=TopSide_V02_Layer-02_Uvs, parameterName=TopS_02_OffsetS_Mech, defaultValue=<rgba 0 0 0 1>}.RGB
    B=VectorParameter() {group=TopSide_V02_Layer-02_Uvs, parameterName=TopS_02_OffsetS_Character, defaultValue=<rgba 0 0 0 1>}.RGB
    Alpha=@n1755.A
   )
   Alpha=@n1755.B
  )
  n1757 = VectorParameter() {group=TopSide_V02_Layer-02_Uvs, parameterName=TopS_02_OffsetS_Backpack, defaultValue=<rgba 0 0 0 1>}
  n1758 = MaterialFunctionCall(Float2=NamedRerouteUsage()) {materialFunction=&BreakOutFloat2Components}
  NamedRerouteDeclaration(None=LinearInterpolate(
    A=LinearInterpolate(A=@n1756, B=@n1757.RGB, Alpha=@n1758.R)
    B=VectorParameter() {group=TopSide_V02_Layer-02_Uvs, parameterName=TopS_02_OffsetS_Pickaxe, defaultValue=<rgba 0 0 0 1>}.RGB
    Alpha=@n1758.G
   )) {name=Uvs_TS_Layer_02_Offset_Side, nodeColor=<rgba 0 0.625 1 1>}
  n1759 = VectorParameter() {group=TopSide_V02_Layer-01_Uvs, parameterName=TopS_01_ScaleS_Weapon, defaultValue=<rgba 1 1 1 1>}
  n1760 = VectorParameter() {group=TopSide_V02_Layer-01_Uvs, parameterName=TopS_01_ScaleS_Vehicle, defaultValue=<rgba 1 1 1 1>}
  n1761 = Multiply(A=@n1760.RGB, B=@n1760.RGB)
  n1762 = VectorParameter() {group=TopSide_V02_Layer-01_Uvs, parameterName=TopS_01_ScaleS_VehicleSide, defaultValue=<rgba 1 1 1 1>}
  n1763 = Multiply(A=@n1762.RGB, B=@n1762.RGB)
  n1764 = MaterialFunctionCall(Float4=NamedRerouteUsage()) {materialFunction=&BreakOutFloat4Components}
  n1765 = VectorParameter() {group=TopSide_V02_Layer-01_Uvs, parameterName=TopS_01_ScaleS_Mech, defaultValue=<rgba 1 1 1 1>}
  n1766 = VectorParameter() {group=TopSide_V02_Layer-01_Uvs, parameterName=TopS_01_ScaleS_Character, defaultValue=<rgba 1 1 1 1>}
  n1767 = LinearInterpolate(
   A=LinearInterpolate(A=Multiply(A=@n1759.RGB, B=@n1759.RGB), B=LinearInterpolate(A=@n1761, B=@n1763, Alpha=@n1764.G), Alpha=@n1764.R)
   B=LinearInterpolate(A=Multiply(A=@n1765.RGB, B=@n1765.RGB), B=Multiply(A=@n1766.RGB, B=@n1766.RGB), Alpha=@n1764.A)
   Alpha=@n1764.B
  )
  n1768 = VectorParameter() {group=TopSide_V02_Layer-01_Uvs, parameterName=TopS_01_ScaleS_Backpack, defaultValue=<rgba 1 1 1 1>}
  n1769 = Multiply(A=@n1768.RGB, B=@n1768.RGB)
  n1770 = MaterialFunctionCall(Float2=NamedRerouteUsage()) {materialFunction=&BreakOutFloat2Components}
  n1771 = VectorParameter() {group=TopSide_V02_Layer-01_Uvs, parameterName=TopS_01_ScaleS_Pickaxe, defaultValue=<rgba 1 1 1 1>}
  NamedRerouteDeclaration(None=LinearInterpolate(A=LinearInterpolate(A=@n1767, B=@n1769, Alpha=@n1770.R), B=Multiply(A=@n1771.RGB, B=@n1771.RGB), Alpha=@n1770.G)) {name=Uvs_TS_Layer_01_Scale_Side, nodeColor=<rgba 0 0.625 1 1>}
  n1772 = VectorParameter() {group=TopSide_V02_Layer-01_Uvs, parameterName=TopS_01_PanningS_Speed------------, defaultValue=<rgba 0 0 0 1>}
  n1773 = Multiply(A=Time(), B=@n1772.RGB)
  n1774 = LinearInterpolate(
   A=Multiply(A=Time(), B=VectorParameter() {group=TopSide_V02_Layer-01_Uvs, parameterName=TopS_01_PanningS_Speed_B, defaultValue=<rgba 0 0 0 1>}.RGB)
   B=@n1773
   Alpha=NamedRerouteUsage()
  )
  NamedRerouteDeclaration(None=Add(
    A=StaticSwitchParameter(True=@n1774, False=@n1773) {group=TopSide_V02_Layer-01_Uvs, parameterName=TopS_01_PanningS_Speed--Varation-Msk}
    B=@n1473
   )) {name=Uvs_TS_Layer_01_Speed_Side, nodeColor=<rgba 0 0.625 1 1>}
  n1775 = VectorParameter() {group=TopSide_V02_Layer-02_Uvs, parameterName=TopS_02_ScaleS_Weapon, defaultValue=<rgba 1 1 1 1>}
  n1776 = VectorParameter() {group=TopSide_V02_Layer-02_Uvs, parameterName=TopS_02_ScaleS_Vehicle, defaultValue=<rgba 1 1 1 1>}
  n1777 = Multiply(A=@n1776.RGB, B=@n1776.RGB)
  n1778 = VectorParameter() {group=TopSide_V02_Layer-02_Uvs, parameterName=TopS_02_ScaleS_VehicleSide, defaultValue=<rgba 1 1 1 1>}
  n1779 = Multiply(A=@n1778.RGB, B=@n1778.RGB)
  n1780 = MaterialFunctionCall(Float4=NamedRerouteUsage()) {materialFunction=&BreakOutFloat4Components}
  n1781 = VectorParameter() {group=TopSide_V02_Layer-02_Uvs, parameterName=TopS_02_ScaleS_Mech, defaultValue=<rgba 1 1 1 1>}
  n1782 = VectorParameter() {group=TopSide_V02_Layer-02_Uvs, parameterName=TopS_02_ScaleS_Character, defaultValue=<rgba 1 1 1 1>}
  n1783 = LinearInterpolate(
   A=LinearInterpolate(A=Multiply(A=@n1775.RGB, B=@n1775.RGB), B=LinearInterpolate(A=@n1777, B=@n1779, Alpha=@n1780.G), Alpha=@n1780.R)
   B=LinearInterpolate(A=Multiply(A=@n1781.RGB, B=@n1781.RGB), B=Multiply(A=@n1782.RGB, B=@n1782.RGB), Alpha=@n1780.A)
   Alpha=@n1780.B
  )
  n1784 = VectorParameter() {group=TopSide_V02_Layer-02_Uvs, parameterName=TopS_02_ScaleS_Backpack, defaultValue=<rgba 1 1 1 1>}
  n1785 = Multiply(A=@n1784.RGB, B=@n1784.RGB)
  n1786 = MaterialFunctionCall(Float2=NamedRerouteUsage()) {materialFunction=&BreakOutFloat2Components}
  n1787 = VectorParameter() {group=TopSide_V02_Layer-02_Uvs, parameterName=TopS_02_ScaleS_Pickaxe, defaultValue=<rgba 1 1 1 1>}
  NamedRerouteDeclaration(None=LinearInterpolate(A=LinearInterpolate(A=@n1783, B=@n1785, Alpha=@n1786.R), B=Multiply(A=@n1787.RGB, B=@n1787.RGB), Alpha=@n1786.G)) {name=Uvs_TS_Layer_02_Scale_Side, nodeColor=<rgba 0 0.625 1 1>}
  n1788 = Time()
  n1789 = VectorParameter() {
   group=TopSide_V02_Layer-02_Uvs
   parameterName=TopS_02_Pannings_Speed_B
   channelNames {a="0=step Animation", b="", g="", r=""}
   defaultValue=<rgba 0 0 0 1>
  }
  n1790 = Multiply(A=@n1788, B=@n1789)
  n1791 = Frac(None=@n1790)
  n1792 = ScalarParameter() {defaultValue=10, group=TopSide_V02_Layer-02_Uvs, parameterName=TopS_02_PanningS_Speed_B_Step_Amount}
  n1793 = Divide(A=Ceil(None=Multiply(A=@n1791, B=@n1792) {constB=30}), B=@n1792)
  n1794 = VectorParameter() {
   group=TopSide_V02_Layer-02_Uvs
   parameterName=TopS_02_Pannings_Speed------------
   channelNames {a="0=step Animation", b="", g="", r=""}
   defaultValue=<rgba 0 0 0 1>
  }
  n1795 = Multiply(A=@n1788, B=@n1794)
  n1796 = ScalarParameter() {defaultValue=10, group=TopSide_V02_Layer-02_Uvs, parameterName=TopS_02_PanningS_Speed--Step_Amount}
  n1797 = Multiply(A=Frac(None=@n1795), B=@n1796) {constB=30}
  n1798 = LinearInterpolate(A=Divide(A=Ceil(None=@n1797), B=@n1796), B=@n1795, Alpha=@n1794."0=step Animation")
  n1799 = StaticSwitchParameter(
   True=LinearInterpolate(A=LinearInterpolate(A=@n1793, B=@n1790, Alpha=@n1789."0=step Animation"), B=@n1798, Alpha=NamedRerouteUsage())
   False=@n1798
  ) {group=TopSide_V02_Layer-02_Uvs, parameterName=TopS_02_PanningS_Speed--Varation-Msk}
  n1800 = Add(A=@n1799, B=@n1467)
  NamedRerouteDeclaration(None=@n1800) {name=Uvs_TS_Layer_02_Speed_Side, nodeColor=<rgba 0 0.625 1 1>}
  n1801 = NamedRerouteUsage()
  n1802 = StaticSwitchParameter(True=NamedRerouteUsage(), False=NamedRerouteUsage()) {group=TopSide_V02_Layer Emissive_02, parameterName="TopS_01_Em02_Warp-TS 01---->TS 04", sortPriority=200}
  n1803 = ChannelMaskParameter(None=StaticSwitchParameter(
    True=NamedRerouteUsage()
    False=StaticSwitchParameter(True=@n1801, False=@n1802) {group=TopSide_V02_Layer Emissive_02, parameterName="TopS_01_Em02_Warp-TS 01---->TS 03", sortPriority=200}
   ) {group=TopSide_V02_Layer Emissive_02, parameterName="TopS_01_Em02_Warp-TS 01---->TS 02", sortPriority=200}) {group=TopSide_V02_Layer Emissive_02, parameterName=TopS_Em02_CC_UVMap_Warp_Channel, sortPriority=201}
  NamedRerouteDeclaration(None=StaticSwitchParameter(
    True=StaticSwitchParameter(True=@n1803.R, False=@n237) {group=TopSide_V02_Layer Emissive_02, parameterName=TopS_Em02_CC_UVMap_Warp-Channel--Alt, sortPriority=201}
    False=@n1249
   ) {group=TopSide_V02_Layer Emissive_02, parameterName=TopS Emissive_02_Warp, sortPriority=200}) {name=Emissive02_warp, nodeColor=<rgba 0.5 1 0 1>}
  NamedRerouteDeclaration(None=@n485.Normal) {name=Root_Normal, nodeColor=<rgba 0.4062 0 1 1>}
  NamedRerouteDeclaration(None=@n1510) {name=Tops 03 Blend Contrats, nodeColor=<rgba 0 1 0.6953 1>}
  n1804 = OneMinus(None=@n1694)
  n1805 = ScalarParameter() {
   defaultValue=1
   desc=CC Brightness gets disable
   group=TopSide_V02_Layer Emissive_01
   parameterName="TopS Emissive_01_ColorCuve=A-Effect_DarkenBase  0=ON"
   sliderMax=1
  }
  n1806 = NamedRerouteDeclaration(None=Saturate(None=Ceil(None=@n1805))) {name=EmiS 01 CC A State, nodeColor=<rgba 0 0.8828 1 1>}
  n1807 = ScalarParameter() {group=TopSide_V02_Layer Emissive_01 Reactive, parameterName=TopS Audio 01 Emissive_Brightness Darken_Base_Max, sortPriority=151}
  n1808 = LinearInterpolate(
   A=ScalarParameter() {group=TopSide_V02_Layer Emissive_01 Reactive, parameterName=TopS Audio 01 Emissive_Brightness Darken_Bas-Min, sortPriority=151}
   B=@n1807
   Alpha=NamedRerouteUsage()
  )
  n1809 = LinearInterpolate(
   A=ScalarParameter() {group=TopSide_V02_Layer Emissive_01 Reactive, parameterName=TopS Emissive_01_Brightness_Darken_Base Start}
   B=ScalarParameter() {group=TopSide_V02_Layer Emissive_01 Reactive, parameterName=TopS Emissive_01_Brightness_Darken_Base End}
   Alpha=NamedRerouteUsage()
  )
  n1810 = StaticSwitch(
   True=StaticSwitch(True=Max(A=@n1808, B=@n1809), False=@n1809, Value=NamedRerouteUsage())
   False=ScalarParameter() {group=TopSide_V02_Layer Emissive_01, parameterName="TopS Emissive_01-Darken_Base    1=On"}
   Value=NamedRerouteUsage()
  )
  n1811 = StaticSwitchParameter(True=Multiply(A=LinearInterpolate(A=@n1678, B=@n197, Alpha=@n1806), B=@n1810), False=Multiply(A=@n197, B=@n1810)) {group=TopSide_V02_Layer Emissive_01, parameterName="TopS Emissive_01_Color ----->ColorCurve"}
  n1812 = StaticSwitch(True=StaticSwitch(True=Multiply(A=@n1658, B=@n1811), False=@n1811, Value=@n1687), False=@n1811, Value=NamedRerouteUsage())
  n1813 = StaticSwitch(True=OneMinus(None=Saturate(None=@n1812)), False=Constant() {r=1}, Value=NamedRerouteUsage())
  n1814 = StaticSwitch(
   True=StaticSwitch(
    True=@n1813
    False=LinearInterpolate(B=@n1813, Alpha=@n1689) {constA=1, constB=0}
    Value=StaticBoolParameter() {group=TopSide_V02_Layer Emissive_01 Reactive, parameterName="TopS Emissive_01-TOD  DoNotAffectDarkenBase", sortPriority=0}
   )
   False=@n1813
   Value=@n1690
  )
  NamedRerouteDeclaration(None=StaticSwitch(True=Max(A=@n1804, B=@n1814), False=@n1814, Value=@n1695)) {name=Emissive_01_DarkenBAse, nodeColor=<rgba 1 0.8203 0 1>}
  NamedRerouteDeclaration(None=@n1513) {name=Tops 03 Alternate Axies, nodeColor=<rgba 1 0.3984 0 1>}
  n1815 = NamedRerouteUsage()
  n1816 = ScalarParameter() {group=TopSide_V02_Layer Emissive_02, parameterName=TopS_Em02_CC_UVMap_Warp_Amount, sortPriority=499}
  n1817 = OneMinus(None=@n109)
  n1818 = Add(A=NamedRerouteUsage(), B=Add(A=Multiply(A=@n1815, B=@n1816), B=Saturate(None=@n1817)))
  NamedRerouteDeclaration(None=LinearInterpolate(
    A=@n1818
    B=Saturate(None=@n1818)
    Alpha=ScalarParameter() {group=TopSide_V02_Layer Emissive_02, parameterName="TopS_Em02_CC_UVMap_Warp_Clamp   1=on", sortPriority=499}
   )) {name=Emissive_02_CC_uvs, nodeColor=<rgba 0.9219 1 0 1>}
  NamedRerouteDeclaration(None=@n1515) {name=Tops 03 Alternate Mask, nodeColor=<rgba 1 0.3984 0 1>}
  NamedRerouteDeclaration(None=@n1519) {name=TopS_03 ProjectionState, nodeColor=<rgba 0 1 0.8359 1>}
  n1819 = NamedRerouteUsage()
  n1820 = ScalarParameter() {group=TopSide_V02_Layer Emissive_01, parameterName=TopS_Em01_CC_UVMap_Warp_Amount, sortPriority=499}
  n1821 = Constant()
  n1822 = ComponentMask(None=MaterialFunctionCall(Caculated_Bounds=NamedRerouteUsage(), is_State=@n100.RGBA, Uvs_Offset=@n1821, Uvs_TileRate=Constant() {r=1}, Uvs_Offset_Pivot=@n1821) {materialFunction=&MF_Projection_Side_Normal_V2}.UVs No Rotation) {r=true}
  n1823 = MaterialFunctionCall(Float4=NamedRerouteUsage()) {materialFunction=&BreakOutFloat4Components}
  n1824 = LinearInterpolate(
   A=LinearInterpolate(
    A=ScalarParameter() {defaultValue=-1, group=TopSide_V02_Layer Emissive_01, parameterName=TopS_Em01_CC_UVMap_Offset_Weapon, sortPriority=500}
    B=ScalarParameter() {defaultValue=-1, group=TopSide_V02_Layer Emissive_01, parameterName=TopS_Em01_CC_UVMap_Offset_Vehicle, sortPriority=500}
    Alpha=@n1823.R
   )
   B=ScalarParameter() {defaultValue=-1, group=TopSide_V02_Layer Emissive_01, parameterName=TopS_Em01_CC_UVMap_Offset_Mech, sortPriority=500}
   Alpha=@n1823.B
  )
  n1825 = ScalarParameter() {
   defaultValue=-1
   desc="negative value inverts "
   group=TopSide_V02_Layer Emissive_01
   parameterName=TopS_Em01_CC_UVMap_Falloff_Weapon
   sortPriority=500
  }
  n1826 = ScalarParameter() {
   defaultValue=-1
   desc="negative value inverts "
   group=TopSide_V02_Layer Emissive_01
   parameterName=TopS_Em01_CC_UVMap_Falloff_Vehicle
   sortPriority=500
  }
  n1827 = LinearInterpolate(A=@n1825, B=@n1826, Alpha=@n1823.R)
  n1828 = ScalarParameter() {
   defaultValue=-1
   desc="negative value inverts "
   group=TopSide_V02_Layer Emissive_01
   parameterName=TopS_Em01_CC_UVMap_Falloff_Mech
   sortPriority=500
  }
  n1829 = OneMinus(None=Divide(A=Add(A=@n1822, B=@n1824), B=LinearInterpolate(A=@n1827, B=@n1828, Alpha=@n1823.B)))
  n1830 = Add(A=Add(A=Multiply(A=@n1819, B=@n1820), B=Saturate(None=@n1829)), B=NamedRerouteUsage())
  NamedRerouteDeclaration(None=LinearInterpolate(
    A=@n1830
    B=Saturate(None=@n1830)
    Alpha=ScalarParameter() {group=TopSide_V02_Layer Emissive_01, parameterName="TopS_Em01_CC_UVMap_Warp_Clamp   1=on", sortPriority=499}
   )) {name=Emissive_01_CC_uvs, nodeColor=<rgba 0.1016 1 0 1>}
  n1831 = NamedRerouteUsage()
  n1832 = NamedRerouteUsage()
  n1833 = NamedRerouteUsage()
  n1834 = StaticBoolParameter() {group=TopSide_V02_Layer-03_Uvs, parameterName=TopS_03 Clamp Uvs}
  n1835 = ScalarParameter() {defaultValue=0.001, group=TopSide_V02_Layer-03_Uvs, parameterName=TopS_03-Clamp Uvs padding Comp}
  n1836 = MaterialFunctionCall(
   Pivot=@n1831
   "Pivot_L\\R"=StaticBool() {value=true}
   Caculated_Bounds=@n1832
   Pre_Skinned_Normals=@n1833
   Is State=@n1446
   Uvs_Rotation=@n1447
   Uvs_TileRate=@n1451
   Uvs_Offset=@n1453
   Clamp Y=@n1834
   Clamp X=@n1834
   Clamp Padding_Comp=@n1835
  ) {materialFunction=&MF_Projection_Side_Left_Right_Decal_NormalV2}
  n1837 = MaterialFunctionCall(
   Caculated_Bounds=@n1832
   Pre_Skinned_Normals=@n1833
   is_State=NamedRerouteUsage()
   UV_Rotation=@n1450
   Uvs_Offset=@n1831
   Uvs_TileRate=@n1452
   Uvs_Offset_Pivot=@n1454
   Rotation_Comp=@n1455
   Clamp Y=@n1834
   Clamp X=@n1834
   Clamp Padding_Comp=@n1835
   No_Rotation=@n1456
  ) {materialFunction=&MF_Projection_Side_Normal_V2}
  n1838 = StaticBoolParameter() {group=TopSide_V02_Layer 03, parameterName="TopS_03 TopSide->Side projection->Left_Right"}
  n1839 = StaticSwitch(True=@n1836.UVs, False=@n1837.UVs, Value=@n1838)
  n1840 = NamedRerouteUsage()
  n1841 = TextureSample(UVs=@n1839, Tex=@n1840, Level=ScalarParameter() {defaultValue=3, group=TopSide_V02_Layer 03, parameterName=TopS_03-Miplevels_}) {mipValueMode=TMVM_MipLevel}
  n1842 = StaticSwitch(True=@n1836.UVs No Offsets/Tiling, False=@n1837.UVs No Offsets/Tiling, Value=@n1838)
  n1843 = TextureSample!(UVs=@n1839, Tex=@n1840, "DDX(UVs)"=DDX(Value=@n1842), "DDY(UVs)"=DDY(Value=@n1842), Apply View MipBias) {mipValueMode=TMVM_Derivative}
  n1844 = NamedRerouteUsage()
  n1845 = NamedRerouteUsage()
  n1846 = OneMinus(None=NamedRerouteUsage())
  n1847 = Saturate(None=Add(A=NamedRerouteUsage(), B=Saturate(None=@n1846)))
  NamedRerouteDeclaration(None=StaticSwitch(
    True=LinearInterpolate(A=@n1841.RGBA, B=@n1843.RGBA, Alpha=@n1844)
    False=LinearInterpolate(A=@n1845, B=@n1843.RGBA, Alpha=@n1847)
    Value=StaticBoolParameter() {group=TopSide_V02_Layer 03, parameterName=TopS_03-Miplevels-VS-Alternative_Axies}
   )) {name=TS 03 Side_Mask, nodeColor=<rgba 1 0.4922 0 1>}
  n1848 = NamedRerouteUsage()
  n1849 = StaticSwitchParameter(True=NamedRerouteUsage(), False=NamedRerouteUsage()) {group=TopSide_V02_Layer Emissive_01, parameterName="TopS_01_Em01_Warp-TS 01---->TS 04", sortPriority=200}
  n1850 = ChannelMaskParameter(None=StaticSwitchParameter(
    True=NamedRerouteUsage()
    False=StaticSwitchParameter(True=@n1848, False=@n1849) {group=TopSide_V02_Layer Emissive_01, parameterName="TopS_01_Em01_Warp-TS 01---->TS 03", sortPriority=200}
   ) {group=TopSide_V02_Layer Emissive_01, parameterName="TopS_01_Em01_Warp-TS 01---->TS 02", sortPriority=200}) {group=TopSide_V02_Layer Emissive_01, parameterName=TopS_Em01_CC_UVMap_Warp_Channel, sortPriority=201}
  NamedRerouteDeclaration(None=StaticSwitchParameter(
    True=StaticSwitchParameter(True=@n1850.R, False=@n162) {group=TopSide_V02_Layer Emissive_01, parameterName=TopS_Em01_CC_UVMap_Warp-Channel--Alt, sortPriority=201}
    False=Constant()
   ) {group=TopSide_V02_Layer Emissive_01, parameterName=TopS Emissive_01_Warp, sortPriority=200}) {name=Emissive01_warp, nodeColor=<rgba 0.5 1 0 1>}
  NamedRerouteDeclaration(None=StaticBoolParameter() {group=TopSide_V02_Layer 02, parameterName="TopS_02 TopSide->Side projection->Left_Right"}) {name="02_Left\\Right Stat", nodeColor=<rgba 0 0.2734 1 1>}
  NamedRerouteUsage()
  NamedRerouteUsage()
  NamedRerouteUsage()
  NamedRerouteUsage()
  NamedRerouteUsage()
  NamedRerouteUsage()
  NamedRerouteUsage()
  NamedRerouteUsage()
  NamedRerouteUsage()
  NamedRerouteUsage()
  NamedRerouteUsage()
  NamedRerouteUsage()
  NamedRerouteUsage()
  NamedRerouteUsage()
  NamedRerouteUsage()
  NamedRerouteUsage()
  NamedRerouteUsage()
  NamedRerouteUsage()
  NamedRerouteUsage()
  NamedRerouteUsage()
  NamedRerouteUsage()
  NamedRerouteUsage()
  NamedRerouteUsage()
  NamedRerouteUsage()
  NamedRerouteUsage()
  NamedRerouteUsage()
  NamedRerouteUsage()
  NamedRerouteUsage()
  NamedRerouteUsage()
  OneMinus()
  Reroute(None=ScalarParameter() {defaultValue=1, group=TopSide_V02, parameterName=TopS Blend_Contrast})
  Reroute(None=@n253)
  Saturate(None=Ceil(None=NamedRerouteUsage()))
  ScalarParameter() {group=TopSide_V02_Layer Emissive_02 Reactive, parameterName="TopS Emissive_02-Brightness-Reactive--->Gradent"}
  ScalarParameter() {group=TopSide_V02_Layer Emissive_01 Reactive, parameterName=TopS Emissive_01_Gradent---Start}
  ScalarParameter() {group=TopSide_V02_Layer Emissive_02 Reactive, parameterName=TopS Emissive_02_Gradent---Start}
  ScalarParameter() {defaultValue=2, group=TopSide_V02, parameterName=TopS Blend_Curvature_Amount}
  ScalarParameter() {group=TopSide_V02_Layer Emissive_03 Reactive, parameterName=TopS Emissive_03_Gradent---Start}
  StaticBool() {value=true}
  StaticSwitchParameter(True=@n1800, False=@n1799) {group=TopSide_V02_Layer-02_Uvs_Reactive, parameterName=TopS-02-Speed---Reactive--------, sortPriority=100}
  StaticSwitchParameter(True=@n758, False=@n751) {group=TopSide_V02_Layer-03_Uvs_Reactive, parameterName=TopS-03-Speed---Reactive--------, sortPriority=100}
  StaticSwitchParameter(True=@n1117, False=@n1110) {group=TopSide_V02_Layer-04_Uvs_Reactive, parameterName=TopS-04-Speed---Reactive--------, sortPriority=100}
  StaticSwitchParameter(True=@n36, False=@n121) {group=Customization, parameterName=TopSideProjection_V02}
  StaticSwitch(True=Constant() {r=1}, False=@n219, Value=@n218)
  n1851 = Sine(None=Time())
  n1852 = LinearInterpolate(A=Constant4Vector() {constant=<rgba 1 1 1 1>}.RGBA, B=Constant4Vector().RGBA, Alpha=Round(None=ConstantBiasScale(None=@n1851)))
  n1853 = StaticSwitch(
   True=StaticBool() {value=true}
   False=StaticBoolParameter() {group=TopSideProjection, parameterName=Use_TopSideProjection_V02_Maskes, sortPriority=-100}
   Value=@n122
  )
  StaticSwitch(True=@n123, False=@n1852, Value=@n1853)
  StaticSwitch(True=@n124, False=@n1852, Value=@n1853)
  TextureSample() {mipValueMode=TMVM_Derivative}
  TextureSample!(UVs=NamedRerouteUsage(), Tex=NamedRerouteUsage(), Apply View MipBias)
  VectorParameter() {group=TopSide_V02_Layer-01_Uvs, parameterName=TopS_01_Rotation_Backpack, defaultValue=<rgba 0 0 0 1>}
  VectorParameter() {
   group=TopSide_V02_Layer Emissive_01 Reactive
   parameterName=TopS Emissive_01_Speed-Offset_Start
   sortPriority=500
   defaultValue=<rgba 0 0 0 1>
  }
  VectorParameter() {
   group=TopSide_V02_Layer Emissive_02 Reactive
   parameterName=TopS Emissive_02_Speed-Offset_Start
   sortPriority=500
   defaultValue=<rgba 0 0 0 1>
  }
  VectorParameter() {
   group=TopSide_V02_Layer Emissive_03 Reactive
   parameterName=TopS Emissive_03_Speed-Offset_Start
   sortPriority=500
   defaultValue=<rgba 0 0 0 1>
  }
  VectorParameter() {group=TopSide_V02_Layer-02_Uvs, parameterName=TopS_02_Rotation_Backpack, defaultValue=<rgba 0 0 0 1>}
 ]}
