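# Material function that projects a weapon wrap texture from top and side axes onto weapons, vehicles and mechs.
# Samples the texture on three rotatable planes from pre-skinned bounds coordinates, blending them with contrast-adjusted pre-skinned normal masks.
# Is_Vehicle and Is_Mech select different axis pairings, with the remaining axis filled by an alternate color and weapon Z optionally filled.
# A Pliant shoulder override parameter masks the mech fill; outputs are combined, RGB and alpha.
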MaterialFunction /Game/Weapons/WeaponSkins/Wraps/MaterialFunctions/MF_TopsSideProjection_Base
graph {nodes [
  n1 = PreSkinnedLocalBounds()
  n2 = Subtract(A=VertexInterpolator(VS=PreSkinnedPosition()).PS, B=@n1.Min)
  n3 = Subtract(A=@n1.Max, B=@n1.Max)
  n4 = Multiply(A=FunctionInput(Preview=Divide(A=@n2, B=@n3)) {inputName=Caculated_Bounds, sortPriority=3}, B=Constant3Vector() {constant=<rgba 1 1 -1 0>})
  n5 = MaterialFunctionCall(Color=FunctionInput() {inputName=Rotation_Axis}) {materialFunction=&SplitComponents}
  n6 = FunctionInput(Preview=TextureObject() {texture=/Game/Weapons/WeaponSkins/Wraps/Textures/UVCheaker}) {inputName=Texture, inputType=FunctionInput_Texture2D}
  n7 = TextureSample(
   UVs=MaterialFunctionCall(UVs=ComponentMask(None=@n4) {g=true, r=true}, "Rotation Angle (0-1)"=@n5.B) {materialFunction=&CustomRotator}.Rotated Values
   Tex=@n6
  )
  n8 = AppendVector(A=@n7.RGB, B=@n7.RGB)
  n9 = FunctionInput(Preview=Constant4Vector().RGBA) {bUsePreviewValueAsDefault=true, inputName=Fill_Z_Weapon_Color, inputType=FunctionInput_Vector4, sortPriority=11, previewValue=<0 0 0 0>}
  n10 = FunctionInput(Preview=Constant()) {bUsePreviewValueAsDefault=true, inputName=Fill_Z_Weapon_opacity, inputType=FunctionInput_Scalar, sortPriority=10}
  n11 = StaticBool()
  n12 = TextureSample(
   UVs=MaterialFunctionCall(UVs=ComponentMask(None=@n4) {b=true, g=true}, "Rotation Angle (0-1)"=@n5.R) {materialFunction=&CustomRotator}.Rotated Values
   Tex=@n6
  )
  n13 = AppendVector(A=@n12.RGB, B=@n12.RGB)
  n14 = FunctionInput(Preview=VertexInterpolator(VS=PreSkinnedNormal()).PS) {inputName=Pre_Skinned_Normals, sortPriority=4}
  n15 = FunctionInput(Preview=Constant() {r=10}) {inputName=Blend_Contrast, inputType=FunctionInput_Scalar, sortPriority=5}
  n16 = MaterialFunctionCall(In=Abs(None=ComponentMask(None=@n14) {r=true}), Contrast=@n15) {materialFunction=&CheapContrast}
  n17 = MaterialFunctionCall(In=Abs(None=ComponentMask(None=@n14) {g=true}), Contrast=@n15) {materialFunction=&CheapContrast}
  n18 = Saturate(None=Add(A=@n16.Result, B=@n17.Result))
  n19 = LinearInterpolate(
   A=StaticSwitch(
    True=LinearInterpolate(A=@n8, B=@n9, Alpha=@n10)
    False=@n8
    Value=FunctionInput(Preview=@n11) {bUsePreviewValueAsDefault=true, inputName=Enable_Fill_Z_Weapon, inputType=FunctionInput_StaticBool, sortPriority=10}
   )
   B=@n13
   Alpha=@n18
  )
  n20 = FunctionInput(Preview=Constant() {r=1}) {inputName=Fill_alternate_Axis_Opacity, inputType=FunctionInput_Scalar, sortPriority=6}
  n21 = MaterialFunctionCall(In=Abs(None=ComponentMask(None=@n14) {b=true}), Contrast=@n15) {materialFunction=&CheapContrast}
  n22 = Add(A=@n16.Result, B=Saturate(None=Subtract(A=@n21.Result, B=@n17.Result)))
  n23 = Multiply(A=@n20, B=OneMinus(None=Saturate(None=@n22)))
  n24 = FunctionInput() {inputName="Fill_alternate_Axis_Color_A=Roughness", inputType=FunctionInput_Vector4, sortPriority=7}
  n25 = TextureSample(
   UVs=MaterialFunctionCall(UVs=ComponentMask(None=@n4) {b=true, r=true}, "Rotation Angle (0-1)"=@n5.G) {materialFunction=&CustomRotator}.Rotated Values
   Tex=@n6
  )
  n26 = AppendVector(A=@n25.RGB, B=@n25.RGB)
  n27 = LinearInterpolate(A=@n26, B=@n8, Alpha=Saturate(None=Add(A=@n21.Result, B=@n16.Result))) {desc=Float3}
  n28 = Saturate(None=Add(A=@n21.Result, B=@n17.Result))
  n29 = Multiply(A=@n20, B=OneMinus(None=@n28))
  n30 = Constant()
  n31 = Multiply(A=@n20, B=OneMinus(None=@n18))
  n32 = Constant()
  n33 = NamedRerouteUsage()
  n34 = Multiply(
   A=Subtract(A=NamedRerouteUsage(), B=ComponentMask(None=@n33) {desc=IsolateBackpack, r=true})
   B=ScalarParameter() {group=Pliant, parameterName=TopSideProjection_UsePliantShoulderOverride}
  )
  n35 = LinearInterpolate(
   A=LinearInterpolate(
    A=LinearInterpolate(A=@n19, B=@n24, Alpha=@n23)
    B=LinearInterpolate(A=@n27, B=@n24, Alpha=@n29)
    Alpha=FunctionInput(Preview=@n30) {inputName=Is_Vehicle, inputType=FunctionInput_Scalar, sortPriority=8}
   )
   B=LinearInterpolate(A=LinearInterpolate(A=@n13, B=@n26, Alpha=@n28), B=@n24, Alpha=LinearInterpolate(A=@n31, B=@n32, Alpha=@n34) {desc=Pliant Specific})
   Alpha=FunctionInput(Preview=@n30) {inputName=Is_Mech, inputType=FunctionInput_Scalar, sortPriority=9}
  )
  FunctionOutput(None=@n35) {outputName=Combined}
  FunctionOutput(None=ComponentMask(None=@n35) {b=true, g=true, r=true}) {outputName=RGB}
  FunctionOutput(None=ComponentMask(None=@n35) {a=true}) {outputName=Alpha}
  NamedRerouteDeclaration(None=FunctionInput() {bUsePreviewValueAsDefault=true, inputName=IsCharacter, inputType=FunctionInput_Scalar}) {name=IsCharacter, nodeColor=<rgba 0.0078 1 0 1>}
  NamedRerouteDeclaration(None=FunctionInput() {bUsePreviewValueAsDefault=true, inputName=IsStatsB, inputType=FunctionInput_Vector2}) {name=IsStatsB, nodeColor=<rgba 0 0.4844 1 1>}
  PreSkinnedNormal()
 ]}
