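# Weapon wrap material function that outputs animated coordinates, offsets and scales for two triplanar layers.
# Lerps pick parameters by item type (weapon, vehicle, mech, character, backpack, pickaxe).
# Static switches map pre-skinned local position to object bounds, mirror and scale it, then add time panning and noise-texture warp.
# Layer one also supports stepped animation and panning Y flipped by local Y sign; reroutes supply jacket and backpack speed and warp overrides.
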MaterialFunction /Game/Weapons/WeaponSkins/Wraps/MaterialFunctions/MF_WrapPanningWorldSpace
graph {nodes [
  n1 = VectorParameter() {group=Primary Triplaner 0UVs, parameterName=Triplaner_01_DirectionWeapon, sortPriority=53, defaultValue=<rgba 0 0 0 1>}
  n2 = VectorParameter() {group=Primary Triplaner 0UVs, parameterName=Triplaner_01_DirectionVehicle, sortPriority=54, defaultValue=<rgba 0 0 0 1>}
  n3 = FunctionInput(Preview=Constant()) {inputName=isVehicle, inputType=FunctionInput_Scalar}
  n4 = VectorParameter() {group=Primary Triplaner 0UVs, parameterName=Triplaner_01_DirectionMech, sortPriority=54, defaultValue=<rgba 0 0 0 1>}
  n5 = FunctionInput() {bUsePreviewValueAsDefault=true, inputName=isMech, inputType=FunctionInput_Scalar}
  n6 = FunctionInput(Preview=Constant3Vector() {constant=<rgba 1 0 0.2 0>}) {bUsePreviewValueAsDefault=true, inputName=ForwardAxis}
  n7 = Add(A=@n6, B=LinearInterpolate(A=LinearInterpolate(A=@n1.RGB, B=@n2.RGB, Alpha=@n3), B=@n4.RGB, Alpha=@n5))
  n8 = ScalarParameter() {defaultValue=2, group=Primary Triplaner 0UVs, parameterName=Triplaner_01_PanningVehicleSpeedModulation, sortPriority=52}
  n9 = Multiply(A=Time() {period=1}, B=Multiply(A=NamedRerouteUsage(), B=LinearInterpolate(B=@n8, Alpha=@n3) {constA=1}))
  n10 = Multiply(A=@n7, B=@n9)
  n11 = VertexInterpolator(VS=PreSkinnedPosition())
  n12 = ComponentMask(None=Add(A=@n11.PS, B=ScalarParameter() {defaultValue=1, parameterName=Pliant_Kurohomura_Panning_Y_Shift})) {g=true}
  n13 = MaterialFunctionCall(X=@n9, Y=Multiply(A=@n9, B=Sign(None=@n12)), Z=@n9) {materialFunction=&MakeFloat3}
  n14 = StaticSwitchParameter(True=LinearInterpolate(A=@n10, B=Multiply(A=@n13.Result, B=@n7), Alpha=NamedRerouteUsage()), False=@n10) {parameterName=Pliant_ModifyPanningDirection_BaseOnLocalYDirection}
  n15 = VectorParameter() {group=Primary Triplaner 0UVs, parameterName=Triplaner_01_StepWeapon, sortPriority=56, defaultValue=<rgba 1 1 1 1>}
  n16 = VectorParameter() {group=Primary Triplaner 0UVs, parameterName=Triplaner_01_StepVehicle, sortPriority=56, defaultValue=<rgba 1 1 1 1>}
  n17 = LinearInterpolate(A=@n15.RGB, B=@n16.RGB, Alpha=@n3) {constA=1}
  n18 = Ceil(None=Multiply(A=Frac(None=@n14), B=@n17) {constB=30})
  n19 = VectorParameter() {group=Primary Triplaner 0UVs, parameterName=Triplaner_01 MirrorScaleWeapon, sortPriority=4, defaultValue=<rgba -1 -1 -1 1>}
  n20 = VectorParameter() {group=Primary Triplaner 0UVs, parameterName=Triplaner_01 MirrorScaleVehicle, sortPriority=6, defaultValue=<rgba -1 -1 -1 1>}
  n21 = VectorParameter() {group=Primary Triplaner 0UVs, parameterName=Triplaner_01 MirrorScaleMech, sortPriority=6, defaultValue=<rgba -1 -1 -1 1>}
  n22 = LinearInterpolate(A=LinearInterpolate(A=@n19.RGB, B=@n20.RGB, Alpha=@n3), B=@n21.RGB, Alpha=@n5)
  n23 = LinearInterpolate(A=LinearInterpolate(A=@n19.A, B=@n20.A, Alpha=@n3), B=@n21.A, Alpha=@n5)
  n24 = VectorParameter() {group=Primary Triplaner 0UVs, parameterName=Triplaner_01_OffsetWeaponA, sortPriority=3, defaultValue=<rgba 0 0 0 1>}
  n25 = VectorParameter() {group=Primary Triplaner 0UVs, parameterName=Triplaner_01_OffsetVehicleA, sortPriority=5, defaultValue=<rgba 0 0 0 1>}
  n26 = LinearInterpolate(A=@n24.RGB, B=@n25.RGB, Alpha=@n3)
  n27 = VectorParameter() {group=Primary Triplaner 0UVs, parameterName=Triplaner_01_OffsetMechA, sortPriority=5, defaultValue=<rgba 0 0 0 1>}
  n28 = PreSkinnedLocalBounds()
  n29 = MaterialFunctionCall(Color=@n28.Min) {desc=Make sure Vehechles uvs center aling the X, materialFunction=&SplitComponents}
  n30 = AppendVector(A=@n29.R, B=Constant())
  n31 = FunctionInput(Preview=@n11.PS) {bUsePreviewValueAsDefault=true, inputName=Pre-Skinned Local Position, sortPriority=41}
  n32 = Subtract(A=@n31, B=LinearInterpolate(A=@n28.Min, B=AppendVector(A=@n30, B=@n29.B), Alpha=@n3))
  n33 = Constant3Vector() {constant=<rgba -0.5 -0.5 -0.5 1>}
  n34 = Constant3Vector() {constant=<rgba -0.5 0 -0.5 1>}
  n35 = VectorParameter() {group=Primary Triplaner 0UVs, parameterName=Triplaner_01_OffsetWeapon, sortPriority=3, defaultValue=<rgba -0.5 -0.5 -0.5 1>}
  n36 = VectorParameter() {group=Primary Triplaner 0UVs, parameterName=Triplaner_01_OffsetVehicle, sortPriority=5, defaultValue=<rgba -0.5 0 -0.5 1>}
  n37 = LinearInterpolate(A=@n35.RGB, B=@n36.RGB, Alpha=@n3)
  n38 = VectorParameter() {group=Primary Triplaner 0UVs, parameterName=Triplaner_01_OffsetMech, sortPriority=5, defaultValue=<rgba 0 0 0 1>}
  n39 = Subtract(A=@n28.Max, B=@n28.Max)
  n40 = Multiply(
   A=StaticSwitchParameter(True=LinearInterpolate(A=@n33, B=@n34, Alpha=@n3), False=LinearInterpolate(A=@n37, B=@n38.RGB, Alpha=@n5)) {group=Primary Triplaner 0UVs, parameterName=Triplaner_01_MakseSquare_Position, sortPriority=2}
   B=@n39
  )
  n41 = MaterialFunctionCall(Color=@n39) {desc="Finding the maxium length ", materialFunction=&SplitComponents}
  n42 = StaticSwitchParameter(True=Max(A=Max(A=@n41.R, B=@n41.R), B=@n41.B), False=@n39) {desc=Use Object max bounds to make the texture Sqaure, group=Primary Triplaner 0UVs, parameterName=Triplaner_01_MakseSquare, sortPriority=2}
  n43 = StaticSwitchParameter(True=Divide(A=Add(A=@n32, B=@n40), B=@n42), False=@n31) {
   desc="Use the object bound to map the texture to the object "
   group=Primary Triplaner 0UVs
   parameterName=UseObjectBoundsForUV_Triplaner_01
   sortPriority=1
  }
  n44 = StaticSwitchParameter(True=Add(A=LinearInterpolate(A=@n26, B=@n27.RGB, Alpha=@n5), B=@n43), False=@n43) {group=Primary Triplaner 0UVs, parameterName=Triplaner_01_MakseSquare_Position, sortPriority=2}
  n45 = StaticSwitchParameter(True=Multiply(A=@n44, B=Multiply(A=@n22, B=@n23)), False=@n44) {desc=Abilty to scale and mirror the uvs, group=Primary Triplaner 0UVs, parameterName=UVMirrorScale_Triplaner_01, sortPriority=1}
  n46 = Add(
   A=@n45
   B=StaticSwitchParameter(True=Divide(A=@n18, B=@n17), False=@n14) {group=Primary Triplaner 0UVs, parameterName=Triplaner_01_Step_Animation, sortPriority=55}
  )
  n47 = TextureCoordinate()
  n48 = ScalarParameter() {defaultValue=1, group=Primary Triplaner 0UVs, parameterName=Triplaner_01_WarpUvTileRate, sortPriority=58}
  n49 = Time() {period=1}
  n50 = AppendVector(
   A=ScalarParameter() {group=Primary Triplaner 0UVs, parameterName=Triplaner_01_WarpUvSpeedX, sortPriority=61}
   B=ScalarParameter() {group=Primary Triplaner 0UVs, parameterName=Triplaner_01_WarpUvSpeedY, sortPriority=62}
  )
  n51 = TextureSampleParameter2D(UVs=Add(A=Multiply(A=@n47, B=@n48), B=Multiply(A=@n49, B=@n50))) {
   group=Primary Triplaner 0UVs
   parameterName=Triplaner_01_WarpUvTexture
   sortPriority=58
   texture=&AtmosphericCloudNoise01
  }
  n52 = ScalarParameter() {group=Primary Triplaner 0UVs, parameterName=Triplaner_01_WarpUvAmount Weapon, sortPriority=59}
  n53 = NamedRerouteUsage()
  n54 = Add(A=@n46, B=Multiply(A=@n51.R, B=LinearInterpolate(A=@n52, B=@n53, Alpha=@n3)))
  FunctionOutput(None=StaticSwitchParameter(
    True=StaticSwitchParameter(True=@n54, False=@n46) {group=Primary Triplaner 0UVs, parameterName=-Triplaner-01-WarpUvs-------, sortPriority=57}
    False=@n45
   ) {group=Primary Triplaner 0UVs, parameterName=Triplaner_01_Panning, sortPriority=50}) {outputName=Triplaner_01_Panning}
  n55 = VectorParameter() {group=Primary Triplaner 0UVs, parameterName=Triplaner_01_OffsetWeapon, sortPriority=3, defaultValue=<rgba -0.5 -0.5 -0.5 1>}
  n56 = VectorParameter() {group=Primary Triplaner 0UVs, parameterName=Triplaner_01_OffsetVehicle, sortPriority=5, defaultValue=<rgba -0.5 0 -0.5 1>}
  n57 = LinearInterpolate(A=@n55.RGB, B=@n56.RGB, Alpha=@n3)
  n58 = FunctionInput() {bUsePreviewValueAsDefault=true, inputName=Pre-Skinned Bounds, sortPriority=50}
  n59 = Constant3Vector()
  n60 = StaticSwitchParameter(
   True=@n59
   False=StaticSwitch(
    True=Multiply(A=@n57, B=@n58)
    False=@n59
    Value=StaticBoolParameter() {desc="Offset uv for triplanner 01 ", group=Primary Triplaner 0UVs, parameterName=Triplaner_01_Offset, sortPriority=2}
   )
  ) {
   desc="Use the object bound to map the texture to the object "
   group=Primary Triplaner 0UVs
   parameterName=UseObjectBoundsForUV_Triplaner_01
   sortPriority=1
  }
  FunctionOutput(None=@n60) {outputName=Triplaner_01_Offset, sortPriority=3}
  n61 = VectorParameter() {group=Primary Triplaner 0UVs, parameterName=Triplaner_02_OffsetWeapon, defaultValue=<rgba -0.5 -0.5 -0.5 1>}
  n62 = VectorParameter() {group=Primary Triplaner 0UVs, parameterName=Triplaner_02_OffsetVehicle, sortPriority=34, defaultValue=<rgba -0.5 0 -0.5 1>}
  n63 = LinearInterpolate(A=@n61.RGB, B=@n62.RGB, Alpha=@n3)
  n64 = StaticSwitchParameter(
   True=@n59
   False=StaticSwitch(
    True=Multiply(A=@n58, B=@n63)
    False=@n59
    Value=StaticBoolParameter() {desc="Offset uv for triplanner 01 ", group=Primary Triplaner 0UVs, parameterName=Triplaner_02_Offset, sortPriority=30}
   )
  ) {
   desc="Use the object bound to map the texture to the object "
   group=Primary Triplaner 0UVs
   parameterName=UseObjectBoundsForUV_Triplaner_02
   sortPriority=29
  }
  FunctionOutput(None=@n64) {outputName=Triplaner_02_Offset, sortPriority=4}
  n65 = VectorParameter() {group=Primary Triplaner 0UVs, parameterName=Triplaner_02 ScaleWeapon, sortPriority=33, defaultValue=<rgba 1 1 1 1>}
  n66 = VectorParameter() {group=Primary Triplaner 0UVs, parameterName=Triplaner_02 ScalePickaxe, sortPriority=33, defaultValue=<rgba 1 1 1 1>}
  n67 = LinearInterpolate(
   A=MaterialFunctionCall(X=@n65.R, Y=@n65.R, Z=@n65.R, A=@n65.R) {materialFunction=&MakeFloat4}.Result
   B=MaterialFunctionCall(X=@n66.R, Y=@n66.R, Z=@n66.R, A=@n66.R) {materialFunction=&MakeFloat4}.Result
   Alpha=NamedRerouteUsage()
  )
  n68 = VectorParameter() {group=Primary Triplaner 0UVs, parameterName=Triplaner_02 ScaleVehicle, sortPriority=35, defaultValue=<rgba 1 1 1 1>}
  n69 = VectorParameter() {group=Primary Triplaner 0UVs, parameterName=Triplaner_02 ScaleMech0, sortPriority=35, defaultValue=<rgba 1 1 1 1>}
  n70 = VectorParameter() {group=Primary Triplaner 0UVs, parameterName=Triplaner_02 ScaleCharacter, sortPriority=35, defaultValue=<rgba 1 1 1 1>}
  n71 = VectorParameter() {group=Primary Triplaner 0UVs, parameterName=Triplaner_02 ScaleBackpack, sortPriority=35, defaultValue=<rgba 1 1 1 1>}
  n72 = LinearInterpolate(
   A=MaterialFunctionCall(X=@n69.R, Y=@n69.R, Z=@n69.R, A=@n69.R) {materialFunction=&MakeFloat4}.Result
   B=LinearInterpolate(
    A=MaterialFunctionCall(X=@n70.R, Y=@n70.R, Z=@n70.R, A=@n70.R) {materialFunction=&MakeFloat4}.Result
    B=MaterialFunctionCall(X=@n71.R, Y=@n71.R, Z=@n71.R, A=@n71.R) {materialFunction=&MakeFloat4}.Result
    Alpha=NamedRerouteUsage()
   )
   Alpha=NamedRerouteUsage()
  )
  n73 = LinearInterpolate(
   A=LinearInterpolate(A=ComponentMask(None=@n67) {b=true, g=true, r=true}, B=@n68.RGB, Alpha=@n3)
   B=ComponentMask(None=@n72) {b=true, g=true, r=true}
   Alpha=@n5
  )
  n74 = LinearInterpolate(A=LinearInterpolate(A=ComponentMask(None=@n67) {a=true}, B=@n68.A, Alpha=@n3), B=ComponentMask(None=@n72) {a=true}, Alpha=@n5)
  n75 = StaticSwitchParameter(True=Constant3Vector() {constant=<rgba 1 1 1 0>}, False=Multiply(A=@n73, B=@n74)) {desc=Abilty to scale and mirror the uvs, group=Primary Triplaner 0UVs, parameterName=UVMirrorScale_Triplaner_02, sortPriority=29}
  n76 = MaterialFunctionCall(Float3=@n58) {materialFunction=&BreakOutFloat3Components}
  n77 = Max(A=Max(A=@n76.R, B=@n76.R), B=@n76.B)
  n78 = StaticSwitchParameter(True=Multiply(A=Multiply(A=@n77, B=@n73), B=@n74), False=Multiply(A=Multiply(A=@n58, B=@n73), B=@n74)) {
   desc="Use UseObjectBoundsForUV_Triplaner_02 instead "
   group=Primary Triplaner 0UVs
   parameterName=Legace DoNotUse_Triplaner_02 MakeSquare
   sortPriority=28
  }
  n79 = FunctionInput() {
   bUsePreviewValueAsDefault=true
   description=Size of texture in worldspace units
   inputName=TextureScale
   sortPriority=-18
   previewValue=<64 64 64 6>
  }
  n80 = StaticSwitchParameter(True=@n78, False=@n79) {
   desc="Use UseObjectBoundsForUV_Triplaner_02 instead "
   group=Primary Triplaner 0UVs
   parameterName="Legace DoNotUse_Triplaner_02 UseBounds "
   sortPriority=27
  }
  n81 = StaticSwitchParameter(True=@n75, False=@n80) {
   desc="Use the object bound to map the texture to the object "
   group=Primary Triplaner 0UVs
   parameterName=UseObjectBoundsForUV_Triplaner_02
   sortPriority=29
  }
  FunctionOutput(None=@n81) {outputName=Triplaner_02_Scale, sortPriority=2}
  n82 = VectorParameter() {group=Primary Triplaner 0UVs, parameterName=Triplaner_01 ScaleVehicle, sortPriority=6, defaultValue=<rgba 1 1 1 1>}
  n83 = VectorParameter() {group=Primary Triplaner 0UVs, parameterName=Triplaner_01 ScaleWeapon, sortPriority=4, defaultValue=<rgba 1 1 1 1>}
  n84 = VectorParameter() {group=Primary Triplaner 0UVs, parameterName=Triplaner_01 ScalePickaxe, sortPriority=4, defaultValue=<rgba 1 1 1 1>}
  n85 = LinearInterpolate(A=AppendVector(A=@n83.RGB, B=@n83.RGB), B=AppendVector(A=@n84.RGB, B=@n84.RGB), Alpha=NamedRerouteUsage()) {constAlpha=1}
  n86 = VectorParameter() {group=Primary Triplaner 0UVs, parameterName=Triplaner_01 ScaleMech, sortPriority=6, defaultValue=<rgba 1 1 1 1>}
  n87 = VectorParameter() {group=Primary Triplaner 0UVs, parameterName=Triplaner_01 ScaleCharacter, sortPriority=6, defaultValue=<rgba 1 1 1 1>}
  n88 = VectorParameter() {group=Primary Triplaner 0UVs, parameterName=Triplaner_01 ScaleBackpack, sortPriority=6, defaultValue=<rgba 1 1 1 1>}
  n89 = LinearInterpolate(
   A=LinearInterpolate(A=AppendVector(A=@n86.RGB, B=@n86.RGB), B=AppendVector(A=@n87.RGB, B=@n87.RGB), Alpha=NamedRerouteUsage())
   B=AppendVector(A=@n88.RGB, B=@n88.RGB)
   Alpha=NamedRerouteUsage()
  ) {constAlpha=1}
  n90 = LinearInterpolate(
   A=LinearInterpolate(A=ComponentMask(None=@n85) {b=true, g=true, r=true}, B=@n82.RGB, Alpha=@n3)
   B=ComponentMask(None=@n89) {b=true, g=true, r=true}
   Alpha=@n5
  )
  n91 = LinearInterpolate(A=LinearInterpolate(A=ComponentMask(None=@n85) {a=true}, B=@n82.A, Alpha=@n3), B=ComponentMask(None=@n89) {a=true}, Alpha=@n5)
  n92 = StaticSwitchParameter(True=Constant3Vector() {constant=<rgba 1 1 1 0>}, False=Multiply(A=@n90, B=@n91)) {desc=Abilty to scale and mirror the uvs, group=Primary Triplaner 0UVs, parameterName=UVMirrorScale_Triplaner_01, sortPriority=1}
  n93 = MaterialFunctionCall(Float3=@n58) {materialFunction=&BreakOutFloat3Components}
  n94 = Max(A=Max(A=@n93.R, B=@n93.R), B=@n93.B)
  n95 = StaticSwitchParameter(True=Multiply(A=Multiply(A=@n94, B=@n90), B=@n91), False=Multiply(A=Multiply(A=@n58, B=@n90), B=@n91)) {
   desc="Use UseObjectBoundsForUV_Triplaner_01 instead "
   group=Primary Triplaner 0UVs
   parameterName=Legace DoNotUse_Triplaner_01 MakeSquare
   sortPriority=-8
  }
  n96 = StaticSwitchParameter(True=@n95, False=@n79) {
   desc="Use UseObjectBoundsForUV_Triplaner_01 instead "
   group=Primary Triplaner 0UVs
   parameterName="Legace DoNotUse_Triplaner_01 UseBounds "
   sortPriority=-9
  }
  n97 = StaticSwitchParameter(True=@n92, False=@n96) {
   desc="Use the object bound to map the texture to the object "
   group=Primary Triplaner 0UVs
   parameterName=UseObjectBoundsForUV_Triplaner_01
   sortPriority=1
  }
  FunctionOutput(None=@n97) {outputName=Triplaner_01_Scale, sortPriority=1}
  n98 = VectorParameter() {group=Primary Triplaner 0UVs, parameterName=Triplaner_02_DirectionWeapon, sortPriority=83, defaultValue=<rgba 0 0 0 1>}
  n99 = VectorParameter() {group=Primary Triplaner 0UVs, parameterName=Triplaner_02_DirectionVehicle, sortPriority=84, defaultValue=<rgba 0 0 0 1>}
  n100 = VectorParameter() {group=Primary Triplaner 0UVs, parameterName=Triplaner_02_DirectionMech, sortPriority=84, defaultValue=<rgba 0 0 0 1>}
  n101 = LinearInterpolate(A=LinearInterpolate(A=@n98.RGB, B=@n99.RGB, Alpha=@n3), B=@n100.RGB, Alpha=@n5)
  n102 = Time() {period=1}
  n103 = Multiply(
   A=NamedRerouteUsage()
   B=LinearInterpolate(
    B=ScalarParameter() {defaultValue=2, group=Primary Triplaner 0UVs, parameterName=Triplaner_02_PanningVehicleSpeedModulation, sortPriority=82}
    Alpha=@n3
   ) {constA=1}
  )
  n104 = VectorParameter() {group=Primary Triplaner 0UVs, parameterName=Triplaner_02 ScaleWeapon, sortPriority=33, defaultValue=<rgba -1 -1 -1 1>}
  n105 = VectorParameter() {group=Primary Triplaner 0UVs, parameterName=Triplaner_02 ScaleVehicle, sortPriority=35, defaultValue=<rgba -1 -1 -1 1>}
  n106 = VectorParameter() {group=Primary Triplaner 0UVs, parameterName=Triplaner_02 ScaleMech, sortPriority=33, defaultValue=<rgba -1 -1 -1 1>}
  n107 = LinearInterpolate(A=LinearInterpolate(A=@n104.RGB, B=@n105.RGB, Alpha=@n3), B=@n106.RGB, Alpha=@n5)
  n108 = LinearInterpolate(A=LinearInterpolate(A=@n104.A, B=@n105.A, Alpha=@n3), B=@n106.A, Alpha=@n5)
  n109 = VectorParameter() {group=Primary Triplaner 0UVs, parameterName=Triplaner_02_OffsetVehicleA, sortPriority=34, defaultValue=<rgba 0 0 0 1>}
  n110 = LinearInterpolate(
   A=VectorParameter() {group=Primary Triplaner 0UVs, parameterName=Triplaner_02_OffsetWeaponA, defaultValue=<rgba 0 0 0 1>}.RGB
   B=@n109.RGB
   Alpha=@n3
  )
  n111 = VectorParameter() {group=Primary Triplaner 0UVs, parameterName=Triplaner_02_OffsetMechA, sortPriority=34, defaultValue=<rgba 0 0 0 1>}
  n112 = PreSkinnedLocalBounds()
  n113 = MaterialFunctionCall(Color=@n112.Min) {desc=Make sure Vehechles uvs center aling the X, materialFunction=&SplitComponents}
  n114 = AppendVector(A=AppendVector(A=@n113.R, B=Constant()), B=@n113.B)
  n115 = Subtract(A=@n31, B=LinearInterpolate(A=LinearInterpolate(A=@n112.Min, B=@n114, Alpha=@n3), B=@n112.Min, Alpha=@n5))
  n116 = Constant3Vector() {constant=<rgba -0.5 -0.5 -0.5 1>}
  n117 = Constant3Vector() {constant=<rgba -0.5 0 -0.5 1>}
  n118 = VectorParameter() {group=Primary Triplaner 0UVs, parameterName=Triplaner_02_OffsetWeapon, defaultValue=<rgba -0.5 -0.5 -0.5 1>}
  n119 = VectorParameter() {group=Primary Triplaner 0UVs, parameterName=Triplaner_02_OffsetVehicle, sortPriority=34, defaultValue=<rgba -0.5 0 -0.5 1>}
  n120 = LinearInterpolate(A=@n118.RGB, B=@n119.RGB, Alpha=@n3)
  n121 = VectorParameter() {group=Primary Triplaner 0UVs, parameterName=Triplaner_02_OffsetMech, sortPriority=34, defaultValue=<rgba 0 0 0 1>}
  n122 = Subtract(A=@n112.Max, B=@n112.Max)
  n123 = Multiply(
   A=StaticSwitchParameter(True=LinearInterpolate(A=@n116, B=@n117, Alpha=@n3), False=LinearInterpolate(A=@n120, B=@n121.RGB, Alpha=@n5)) {group=Primary Triplaner 0UVs, parameterName=Triplaner_02_MakseSquare_Position, sortPriority=31}
   B=@n122
  )
  n124 = MaterialFunctionCall(Color=@n122) {desc="Finding the maxium length ", materialFunction=&SplitComponents}
  n125 = StaticSwitchParameter(True=Max(A=Max(A=@n124.R, B=@n124.R), B=@n124.B), False=@n122) {desc=Use Object max bounds to make the texture Sqaure, group=Primary Triplaner 0UVs, parameterName=Triplaner_02_MakseSquare, sortPriority=31}
  n126 = StaticSwitchParameter(True=Divide(A=Add(A=@n115, B=@n123), B=@n125), False=@n31) {
   desc="Use the object bound to map the texture to the object "
   group=Primary Triplaner 0UVs
   parameterName=UseObjectBoundsForUV_Triplaner_02
   sortPriority=29
  }
  n127 = StaticSwitchParameter(True=Add(A=LinearInterpolate(A=@n110, B=@n111.RGB, Alpha=@n5), B=@n126), False=@n126) {group=Primary Triplaner 0UVs, parameterName=Triplaner_02_MakseSquare_Position, sortPriority=31}
  n128 = StaticSwitchParameter(True=Multiply(A=@n127, B=Multiply(A=@n107, B=@n108)), False=@n127) {desc=Abilty to scale and mirror the uvs, group=Primary Triplaner 0UVs, parameterName=UVMirrorScale_Triplaner_02, sortPriority=29}
  n129 = Add(A=@n128, B=Multiply(A=Add(A=@n6, B=@n101), B=Multiply(A=@n102, B=@n103)))
  n130 = TextureCoordinate()
  n131 = ScalarParameter() {defaultValue=1, group=Primary Triplaner 0UVs, parameterName=Triplaner_02_WarpUvTileRate, sortPriority=86}
  n132 = Time() {period=1}
  n133 = AppendVector(
   A=ScalarParameter() {group=Primary Triplaner 0UVs, parameterName=Triplaner_02_WarpUvSpeedX, sortPriority=89}
   B=ScalarParameter() {group=Primary Triplaner 0UVs, parameterName=Triplaner_02_WarpUvSpeedY, sortPriority=89}
  )
  n134 = TextureSampleParameter2D(UVs=Add(A=Multiply(A=@n130, B=@n131), B=Multiply(A=@n132, B=@n133))) {
   group=Primary Triplaner 0UVs
   parameterName=Triplaner_02_WarpUvTexture
   sortPriority=86
   texture=&AtmosphericCloudNoise01
  }
  n135 = Multiply(
   A=@n134.R
   B=LinearInterpolate(
    A=ScalarParameter() {group=Primary Triplaner 0UVs, parameterName=Triplaner_02_WarpUvAmount Weapon, sortPriority=87}
    B=NamedRerouteUsage()
    Alpha=@n3
   )
  )
  n136 = StaticSwitchParameter(
   True=StaticSwitchParameter(True=Add(A=@n129, B=@n135), False=@n129) {group=Primary Triplaner 0UVs, parameterName=-Triplaner-02-WarpUvs------, sortPriority=85}
   False=@n128
  ) {desc="use for the second tri planner mapping ", group=Primary Triplaner 0UVs, parameterName=Triplaner_02_Panning, sortPriority=80}
  FunctionOutput(None=@n136) {outputName=Triplaner_02_Panning}
  MaterialFunctionCall!(TimeSpeed, Jump Amount Per Step, Number of Steps Per Unit of Time, Time Offset) {materialFunction=/Engine/Functions/Engine_MaterialFunctions02/SteppingPannerTime}
  NamedRerouteDeclaration(None=FunctionInput() {bUsePreviewValueAsDefault=true, inputName=IsCharacter, inputType=FunctionInput_Scalar}) {name=IsCharacter, nodeColor=<rgba 0 0.9253 1 1>}
  n137 = FunctionInput() {bUsePreviewValueAsDefault=true, inputName=IsStatsB, inputType=FunctionInput_Vector2}
  NamedRerouteDeclaration(None=ComponentMask(None=@n137) {r=true}) {name=IsBackpack, nodeColor=<rgba 0 0.9253 1 1>}
  NamedRerouteDeclaration(None=ComponentMask(None=@n137) {g=true}) {name=IsPickaxe, nodeColor=<rgba 0 0.9253 1 1>}
  n138 = ScalarParameter() {defaultValue=1, group=Primary Triplaner 0UVs, parameterName=Triplaner_01_PanningSpeed_Jacket, sortPriority=51}
  n139 = ScalarParameter() {defaultValue=1, group=Primary Triplaner 0UVs, parameterName=Triplaner_01_PanningSpeed_Backpack, sortPriority=51}
  n140 = NamedRerouteUsage()
  NamedRerouteDeclaration(None=LinearInterpolate(
    A=ScalarParameter() {defaultValue=1, group=Primary Triplaner 0UVs, parameterName=Triplaner_01_PanningSpeed, sortPriority=51}
    B=LinearInterpolate(A=@n138, B=@n139, Alpha=@n140)
    Alpha=NamedRerouteUsage()
   )) {name=Triplaner_PanningSpeed_PliantOverrides, nodeColor=<rgba 0 0.9253 1 1>}
  n141 = ScalarParameter() {defaultValue=1, group=Primary Triplaner 0UVs, parameterName=Triplaner_01_WarpUvAmount_Jacket, sortPriority=51}
  n142 = ScalarParameter() {defaultValue=1, group=Primary Triplaner 0UVs, parameterName=Triplaner_01_WarpUvAmount_Backpack, sortPriority=51}
  n143 = NamedRerouteUsage()
  NamedRerouteDeclaration(None=LinearInterpolate(
    A=ScalarParameter() {group=Primary Triplaner 0UVs, parameterName=Triplaner_01_WarpUvAmount_Vehicle, sortPriority=60}
    B=LinearInterpolate(A=@n141, B=@n142, Alpha=@n143)
    Alpha=NamedRerouteUsage()
   )) {name=Triplaner_01_WarpUvAmount_PliantOverrides, nodeColor=<rgba 0 0.9253 1 1>}
  n144 = ScalarParameter() {defaultValue=1, group=Primary Triplaner 0UVs, parameterName=Triplaner_02_PanningSpeed_Jacket, sortPriority=51}
  n145 = ScalarParameter() {defaultValue=1, group=Primary Triplaner 0UVs, parameterName=Triplaner_02_PanningSpeed_Backpack, sortPriority=51}
  n146 = NamedRerouteUsage()
  NamedRerouteDeclaration(None=LinearInterpolate(
    A=ScalarParameter() {defaultValue=1, group=Primary Triplaner 0UVs, parameterName=Triplaner_02_PanningSpeed, sortPriority=51}
    B=LinearInterpolate(A=@n144, B=@n145, Alpha=@n146)
    Alpha=NamedRerouteUsage()
   )) {name=Triplaner_02_PanningSpeed_PliantOverrides, nodeColor=<rgba 0 0.9253 1 1>}
  n147 = ScalarParameter() {defaultValue=1, group=Primary Triplaner 0UVs, parameterName=Triplaner_02_WarpUvAmount_Jacket, sortPriority=51}
  n148 = ScalarParameter() {defaultValue=1, group=Primary Triplaner 0UVs, parameterName=Triplaner_02_WarpUvAmount_Backpack, sortPriority=51}
  n149 = NamedRerouteUsage()
  NamedRerouteDeclaration(None=LinearInterpolate(
    A=ScalarParameter() {group=Primary Triplaner 0UVs, parameterName=Triplaner_02_WarpUvAmount_Vehicle, sortPriority=88}
    B=LinearInterpolate(A=@n147, B=@n148, Alpha=@n149)
    Alpha=NamedRerouteUsage()
   )) {name=Triplaner_02_WarpUvAmount_PliantOverrides, nodeColor=<rgba 0 0.9253 1 1>}
 ]}
