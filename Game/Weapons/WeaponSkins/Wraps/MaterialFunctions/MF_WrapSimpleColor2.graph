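# Material function for weapon wrap skins that lets the second triplanar layer use simple values instead of textures for diffuse, roughness and normals.
# Three independent static switches choose between the incoming Diffuse, Roughness and Normal inputs and a color parameter, a scalar parameter or an optional Material Normals input.
# The switches resolve at compile time, which reduces texture calls when the simple path is chosen.
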MaterialFunction /Game/Weapons/WeaponSkins/Wraps/MaterialFunctions/MF_WrapSimpleColor2
graph {nodes [
  n1 = VectorParameter() {group=Primary Triplaner, parameterName=Triplaner_02_DiffuseSimpleColor, sortPriority=-18, defaultValue=<rgba 0 0 0 1>}
  n2 = StaticSwitchParameter(True=@n1.RGB, False=FunctionInput() {inputName=Diffuse}) {
   desc=use simple Color vs texture help reduce texture calls
   group=Primary Triplaner
   parameterName=Triplaner_02_DiffuseSimple
   sortPriority=-18
  }
  FunctionOutput(None=@n2) {outputName=Diffuse, sortPriority=1}
  n3 = StaticSwitchParameter(
   True=ScalarParameter() {defaultValue=1, group=Primary Triplaner, parameterName=Triplaner_02_RoughnessSimpleValue, sortPriority=-18}
   False=FunctionInput() {inputName=Roughness, inputType=FunctionInput_Scalar, sortPriority=2}
  ) {
   desc=use simple Value vs texture help reduce texture calls
   group=Primary Triplaner
   parameterName=Triplaner_02_RoughnessSimple
   sortPriority=-18
  }
  FunctionOutput(None=@n3) {outputName=Roughness, sortPriority=2}
  n4 = StaticSwitchParameter(
   True=FunctionInput() {bUsePreviewValueAsDefault=true, inputName=Material Normals, sortPriority=-21}
   False=FunctionInput() {inputName=Normal, sortPriority=3}
  ) {
   desc=use simple Normal vs texture help reduce texture calls
   group=Primary Triplaner
   parameterName=Triplaner_02_NormalsSimple
   sortPriority=-17
  }
  FunctionOutput(None=@n4) {outputName=Normal, sortPriority=3}
 ]}
