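# Material function for weapon wrap skins that builds a simple scalar gradient along a selectable axis from a Bounds input.
# Bounds plus Offset is divided by FallOff, multiplied by ForwardAxis, then collapsed by taking the maximum component.
# Clamped_Data saturates before masking and after the max, while Raw_Data repeats the math unclamped.
# Both max chains read R twice and then B, so the G component never contributes.
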MaterialFunction /Game/Weapons/WeaponSkins/Wraps/MaterialFunctions/MF_Wrap_Gradent_Simple
graph {nodes [
  n1 = FunctionInput() {inputName=Bounds}
  n2 = FunctionInput() {inputName=Offset, inputType=FunctionInput_Scalar, sortPriority=2}
  n3 = FunctionInput() {inputName=FallOff, inputType=FunctionInput_Scalar, sortPriority=3}
  n4 = Divide(A=Add(A=@n1, B=@n2), B=@n3)
  n5 = FunctionInput() {bUsePreviewValueAsDefault=true, inputName=ForwardAxis, sortPriority=1}
  n6 = MaterialFunctionCall(Float3=Multiply(A=Saturate(None=@n4), B=@n5)) {materialFunction=&BreakOutFloat3Components}
  n7 = Max(A=@n6.R, B=@n6.R)
  FunctionOutput(None=Saturate(None=Max(A=@n7, B=@n6.B))) {outputName=Clamped_Data}
  n8 = Add(A=@n1, B=@n2)
  n9 = MaterialFunctionCall(Float3=Multiply(A=Divide(A=@n8, B=@n3), B=@n5)) {materialFunction=&BreakOutFloat3Components}
  FunctionOutput(None=Max(A=Max(A=@n9.R, B=@n9.R), B=@n9.B)) {outputName=Raw_Data}
 ]}
