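# Material function for weapon wrap skins that outputs an animated color-curve tint and a hex-cell mask.
# MF_SafeZone_VirtualPlane supplies projected UVs, offset by a Warp input.
# Time-scrolled U samples a curve atlas row, rainbow by default, through MF_Global_CC_Tex for color.
# Hex_Msk multiplies tiled hex texture, rounded noise indexed by warp-texture alpha plus Speed, warp-texture red, center sphere-mask falloff and plane masking.
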MaterialFunction /Game/Weapons/WeaponSkins/Wraps/MaterialFunctions/MF_Wrap_HexColor_Layers
graph {nodes [
  n1 = Multiply(
   A=FunctionInput(Preview=Constant()) {inputName=Height_Offset, inputType=FunctionInput_Scalar}
   B=FunctionInput() {inputName=Constrant, sortPriority=1}
  )
  n2 = FunctionInput() {inputName=SafeZone_Pivot, inputType=FunctionInput_Vector4, sortPriority=2}
  Add(A=AppendVector(A=@n1, B=Constant()), B=@n2)
  FunctionInput() {bUsePreviewValueAsDefault=true, desc=0.0.0, description=0.0.0, inputName=PlaneCenter_distortion, sortPriority=60, previewValue=<0 0 1 1>}
  n3 = Add(A=OneMinus(None=MaterialFunctionCall() {materialFunction=/Engine/Functions/Engine_MaterialFunctions02/WorldPositionOffset/ObjectScale}.Scale Z)) {constB=-1}
  n4 = PreSkinnedLocalBounds()
  n5 = MaterialFunctionCall(Float3=Subtract(A=@n4.Max, B=@n4.Max)) {materialFunction=&BreakOutFloat3Components}
  n6 = Multiply(A=Max(A=Max(A=@n5.R, B=@n5.R), B=@n5.B), B=ComponentMask(None=Constant() {r=1}) {a=true})
  n7 = Constant3Vector() {constant=<rgba 1 0 0 1>}
  n8 = Constant3Vector() {constant=<rgba 0 1 0 1>}
  n9 = MaterialFunctionCall!(
   Camera _pos
   Axis
   Scale=FunctionInput(Preview=Multiply(A=@n3, B=@n6)) {inputName=Scale, inputType=FunctionInput_Vector2, sortPriority=1}
   Scale_Multplyer
   Object-SafeZone-Virtual-Pivot=@n2
   Object-SafeZone-Virtual-CH-Switch=FunctionInput() {inputName=SafeZone_Chr_Switch, inputType=FunctionInput_Scalar, sortPriority=3}
   SinglePlane Normal Axis=FunctionInput(Preview=Transform(None=@n7) {transformSourceType=TRANSFORMSOURCE_Local}) {inputName=SinglePlane_Normal, sortPriority=4}
   SinglePlane X Axis=FunctionInput(Preview=Transform(None=@n8) {transformSourceType=TRANSFORMSOURCE_Local}) {inputName=SinglePlane_X_Axis, sortPriority=5}
   PlaneCenter_distortion=@n1
  ) {materialFunction=&/MF_SafeZone_VirtualPlane}
  n10 = Add(
   A=FunctionInput(Preview=Constant2Vector()) {bUsePreviewValueAsDefault=true, inputName=Warp, inputType=FunctionInput_Vector2}
   B=@n9.Single_plane_UVs
  )
  n11 = Time()
  n12 = ScalarParameter() {defaultValue=0.1, group=Wrap_HexColor_Layers, parameterName=HexC_ColoCurve_Speed}
  n13 = Frac(None=Add(A=ComponentMask(None=@n10) {r=true}, B=Multiply(A=@n11, B=@n12)))
  n14 = CurveAtlasRowParameter(CurveTime=@n13) {
   atlas=/Game/Weapons/WeaponSkins/Wraps/Textures/CruveAtlas_Wraps_01
   curve=/Game/Weapons/WeaponSkins/Wraps/Textures/HunterD/CC_Rainbow
   defaultValue=42
   group=Wrap_HexColor_Layers
   parameterName=HexC_ColorCurve
  }
  FunctionOutput(None=MaterialFunctionCall!(
    Index=ScalarParameter() {defaultValue=42, group=Wrap_HexColor_Layers, parameterName="HexC_ColorCurve  Index"}
    CCinput=@n14
    CCinput_Alpha
    UV=@n13
   ) {materialFunction=&/MF_Global_CC_Tex}.3V)
  n15 = Constant() {r=0.5}
  n16 = MaterialFunctionCall(Float2=@n10) {materialFunction=&BreakOutFloat2Components}
  n17 = Multiply(A=SphereMask(A=@n15, B=@n16.R) {attenuationRadius=0.5}, B=SphereMask(A=@n16.G, B=@n15) {attenuationRadius=0.5})
  n18 = FunctionInput() {inputName=Tex_Warp, inputType=FunctionInput_Texture2D, sortPriority=11}
  n19 = TextureSample(UVs=Saturate(None=@n10), Tex=@n18) {samplerSource=SSM_Clamp_WorldGroupSettings}
  n20 = Multiply(A=@n10, B=FunctionInput(Preview=Constant()) {inputName=Tile_Rate_Master, inputType=FunctionInput_Scalar, sortPriority=6})
  n21 = Constant3Vector()
  n22 = Add(A=TextureSample(UVs=@n20, Tex=@n18).A, B=ComponentMask(None=FunctionInput(Preview=@n21) {inputName=Speed, sortPriority=50}) {g=true, r=true})
  n23 = FunctionInput() {inputName=Tex_Noise, inputType=FunctionInput_Texture2D, sortPriority=12}
  n24 = Multiply(
   A=TextureSample(UVs=@n20, Tex=FunctionInput() {inputName=Tex_HexMsk, inputType=FunctionInput_Texture2D, sortPriority=10}).R
   B=Round(None=TextureSample(UVs=@n22, Tex=@n23).R)
  )
  FunctionOutput(None=Multiply(A=Multiply(A=@n17, B=@n19.R), B=Multiply(A=@n9.Single_plane_Masking, B=@n24))) {outputName=Hex_Msk}
  MaterialFunctionCall() {materialFunction=&ScaleUVsByCenter}
 ]}
