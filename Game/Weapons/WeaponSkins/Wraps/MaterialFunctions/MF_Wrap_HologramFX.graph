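# Material function for weapon skin wraps that adds a hologram effect to incoming emissive, opacity and world position offset.
# Scrolling scanlines with camera-distance density blending, sine-based flicker with a custom override, and noise-driven glitching produce opacity masks, hue shifts and stepped camera-axis displacement.
# A switchable travelling stroke band adds color and vertex displacement, with the displacement zeroed on forward and mobile shading paths.
# Scanline, flicker and glitch masks are also exposed as outputs.
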MaterialFunction /Game/Weapons/WeaponSkins/Wraps/MaterialFunctions/MF_Wrap_HologramFX
graph {nodes [
  Add()
  n1 = ScalarParameter() {defaultValue=0.75, group=Base_HoloFX_Glitch, parameterName=Glitch_FlickerFrequency}
  n2 = NamedRerouteUsage()
  n3 = NamedRerouteUsage()
  n4 = Multiply(A=ScalarParameter() {defaultValue=1, group=Base_HoloFX_Glitch, parameterName=Glitch_FlickerSpeed}, B=@n3)
  n5 = Add(
   A=Multiply(
    A=MaterialFunctionCall(Float2=@n2) {materialFunction=&BreakOutFloat2Components}.G
    B=ScalarParameter() {defaultValue=0.25, group=Base_HoloFX_Glitch, parameterName=Glitch_FlickerUVScale}
   ) {constB=0.25}
   B=@n4
  )
  n6 = Add(A=Multiply(B=@n1) {constA=0.9, constB=0.9}, B=Sine(None=@n5) {period=1.25}) {constB=0.9}
  Ceil(None=@n6)
  n7 = NamedRerouteUsage()
  n8 = NamedRerouteUsage()
  n9 = FunctionInput() {bUsePreviewValueAsDefault=true, inputName=OpacityIn, inputType=FunctionInput_Scalar, sortPriority=11, previewValue=<1 1 1 1>}
  FunctionOutput(None=Multiply(A=Multiply(A=@n7, B=@n8), B=@n9)) {outputName=OpacityOut}
  n10 = NamedRerouteUsage()
  n11 = StaticSwitchParameter(True=@n10, False=Add(A=@n10, B=NamedRerouteUsage())) {group=Base_HoloFX_Stroke, parameterName=DisableStrokeFX}
  n12 = FunctionInput() {bUsePreviewValueAsDefault=true, inputName=MeshScaleIn, inputType=FunctionInput_Scalar, sortPriority=-1, previewValue=<1 0 0 1>}
  FunctionOutput(None=Add(A=Multiply(A=@n11, B=@n12), B=FunctionInput() {bUsePreviewValueAsDefault=true, inputName=WPOIn, sortPriority=12})) {outputName=WpoOut}
  n13 = Multiply(
   A=NamedRerouteUsage()
   B=FunctionInput(Preview=Constant3Vector() {constant=<rgba 0.25 0 0.25 0>}) {bUsePreviewValueAsDefault=true, inputName=EmissiveIn, sortPriority=10}
  )
  n14 = VectorParameter() {group=Base_HoloFX_Glitch, parameterName=Glitch_Color_Overlay, channelNames {a=1 Activate, b="", g="", r=""}, defaultValue=<rgba 1 0 0 0>}
  n15 = NamedRerouteUsage()
  n16 = VectorParameter() {
   group=Base_HoloFX_Glitch
   parameterName=Glitch_Color_Overlay_Strength_
   channelNames {a=0, b=overall, g=Negative, r=Positive}
   defaultValue=<rgba 1 0 1 0>
  }
  n17 = Multiply(A=Saturate(None=Multiply(A=@n15) {constB=-1}), B=OneMinus(None=@n16.Negative))
  n18 = Add(A=Multiply(A=Saturate(None=@n15), B=@n16.Positive), B=@n17)
  n19 = LinearInterpolate(
   A=MaterialFunctionCall(Hue Shift Percentage=NamedRerouteUsage(), Texture=@n13) {materialFunction=&HueShift}.Result
   B=LinearInterpolate(A=@n13, B=Multiply(A=@n13, B=@n14), Alpha=Multiply(A=@n18, B=@n16.overall) {constB=10})
   Alpha=@n14.1 Activate
  )
  n20 = NamedRerouteUsage()
  FunctionOutput(None=StaticSwitchParameter(True=@n19, False=Add(A=@n20, B=@n19)) {group=Base_HoloFX_Stroke, parameterName=DisableStrokeFX}) {outputName=EmissiveOut}
  n21 = NamedRerouteUsage()
  n22 = AppendVector(A=Constant(), B=ScalarParameter() {defaultValue=0.01, group=Base_HoloFX_Scanline, parameterName=Scanline_FlowSpeed})
  n23 = NamedRerouteUsage()
  n24 = MaterialFunctionCall(Float2=Add(A=@n21, B=Multiply(A=@n22, B=@n23))) {materialFunction=&BreakOutFloat2Components}
  n25 = Frac(None=Multiply(A=@n24.G, B=ScalarParameter() {defaultValue=94.8956, group=Base_HoloFX_Scanline, parameterName=Scanline_Density}))
  n26 = Min(A=OneMinus(None=@n25), B=@n25)
  n27 = ScalarParameter() {defaultValue=-0.3968, group=Base_HoloFX_Scanline, parameterName=Scanline_ThicknessOffset}
  n28 = Multiply(A=Add(A=Multiply(A=@n26) {constB=2}, B=@n27), B=Divide(B=Add(A=@n27)))
  n29 = Clamp(None=Add(A=ScalarParameter() {defaultValue=0.704, group=Base_HoloFX_Scanline, parameterName=Scanline_Sharpness})) {clampMode=CMODE_ClampMin, minDefault=1}
  n30 = Frac(None=Multiply(A=@n24.G, B=ScalarParameter() {defaultValue=94.8956, group=Base_HoloFX_Scanline, parameterName=Scanline_Density_Far}))
  n31 = Min(A=OneMinus(None=@n30), B=@n30)
  n32 = Multiply(A=Add(A=Multiply(A=@n31) {constB=2}, B=@n27))
  n33 = WorldPosition()
  n34 = CameraPositionWS()
  n35 = VectorParameter() {
   group=Base_HoloFX_Scanline
   parameterName="Scanline Glitch\\Scan Density Camera"
   channelNames {a="1=Use", b=Falloff, g=Far Distance, r=Near Distance}
   defaultValue=<rgba 50 100 1 0>
  }
  n36 = Divide(A=Subtract(A=Distance(A=@n33.XYZ, B=@n34), B=@n35.Near Distance), B=Subtract(A=@n35.Far Distance, B=@n35.Far Distance))
  n37 = Min(A=Saturate(None=Power(Base=@n36, Exp=@n35.Falloff) {constExponent=3}), B=@n35."1=Use")
  n38 = LinearInterpolate(A=Saturate(None=Multiply(A=@n28, B=@n29)), B=Saturate(None=Multiply(A=@n32, B=@n29)), Alpha=@n37)
  FunctionOutput(None=@n38) {outputName=UseForFX_ScanlineOut, sortPriority=1}
  n39 = Constant()
  n40 = ScalarParameter() {defaultValue=0.15, group=Base_HoloFX_Glitch, parameterName=Glitch_Speed}
  n41 = Add(A=NamedRerouteUsage(), B=Multiply(A=AppendVector(A=@n39, B=@n40), B=NamedRerouteUsage()))
  n42 = ScalarParameter() {defaultValue=28, group=Base_HoloFX_Glitch, parameterName=Glitch_Density}
  n43 = TextureObjectParameter() {
   group=Base_HoloFX_Glitch
   parameterName=Glitch_Tex
   samplerType=SAMPLERTYPE_LinearColor
   texture=&T_NoisePack_LinearColor
  }
  n44 = TextureSample(UVs=Multiply(A=Constant2Vector() {g=1}, B=Multiply(A=@n41, B=@n42)), Tex=@n43) {samplerType=SAMPLERTYPE_LinearColor}
  n45 = Power(Base=ChannelMaskParameter(None=@n44.RGBA) {group=Base_HoloFX_Glitch, parameterName=Glitch_Trex_ToOpacity_Channel}.R, Exp=Constant() {r=4})
  n46 = Saturate(None=Add(A=@n45, B=ScalarParameter() {defaultValue=0.2, group=Base_HoloFX_Glitch, parameterName=Glitch_ToOpacity_MinOpacity}))
  n47 = NamedRerouteUsage()
  FunctionOutput(None=NamedRerouteDeclaration(None=LinearInterpolate(B=@n46, Alpha=@n47) {constA=1}) {name=Glitch_ToOpacity, nodeColor=<rgba 0.9298 0 1 1>}) {outputName=UseForFX_GlitchToOpacityOut, sortPriority=1}
  n48 = Sine(None=@n5) {period=2.1}
  n49 = Multiply(B=@n1) {constA=0.25}
  n50 = Add(A=Sine(None=@n5) {period=0.72}, B=Multiply(B=@n1) {constA=0.75}) {constB=0.75}
  n51 = Multiply(A=Multiply(A=Add(A=@n48, B=@n49), B=@n50), B=Multiply(A=@n50, B=@n6))
  n52 = OneMinus(None=ScalarParameter() {defaultValue=0.4, group=Base_HoloFX_Glitch, parameterName=Glitch_FlickerDetailShakeStrength})
  n53 = ConstantBiasScale(None=Sine(None=@n3) {period=0.075})
  n54 = Multiply(A=Ceil(None=Sine(None=@n4) {period=1.6}), B=Multiply(A=Saturate(None=@n51), B=LinearInterpolate(A=@n52, Alpha=@n53) {constA=0.7}))
  n55 = FunctionInput() {bUsePreviewValueAsDefault=true, inputName=CustomFlicker, inputType=FunctionInput_Scalar}
  n56 = FunctionInput() {bUsePreviewValueAsDefault=true, inputName=UseCustomFlicker, inputType=FunctionInput_Scalar}
  FunctionOutput(None=NamedRerouteDeclaration(None=LinearInterpolate(A=@n54, B=@n55, Alpha=@n56)) {name=Flicker, nodeColor=<rgba 0 1 0.5781 1>}) {outputName=UseForFX_FlickerOut, sortPriority=1}
  n57 = Multiply(A=Subtract(A=ChannelMaskParameter(None=@n44.RGBA) {group=Base_HoloFX_Glitch, parameterName=Glitch_Trex_Subdivision_Channel}.R, B=Constant() {r=0.5}) {constB=0.5}) {constB=2}
  n58 = ScalarParameter() {defaultValue=1, group=Base_HoloFX_Glitch, parameterName=Glitch_Subdivision}
  n59 = Ceil(None=Multiply(A=Abs(None=@n57), B=@n58))
  n60 = Add(A=Floor(None=Clamp(None=@n57) {maxDefault=0, minDefault=-1})) {constB=0.5}
  n61 = Multiply(A=Multiply(A=Divide(A=@n59, B=@n58), B=Multiply(A=@n60) {constB=2}), B=NamedRerouteUsage())
  n62 = Transform(None=Constant3Vector() {constant=<rgba 1 0 0 0>}) {transformSourceType=TRANSFORMSOURCE_Camera}
  n63 = ScalarParameter() {defaultValue=20, group=Base_HoloFX_Glitch, parameterName=Glitch_WPOStrength}
  NamedRerouteDeclaration(None=Multiply(A=@n61, B=Multiply(A=@n62, B=@n63))) {name=Glitch_WpoOutput, nodeColor=<rgba 0 0.1217 1 1>}
  n64 = ScalarParameter() {defaultValue=0.5, group=Base_HoloFX_Scanline, parameterName=Scanline_ToEmissiveScale}
  NamedRerouteDeclaration(None=Add(A=Multiply(A=@n38, B=@n64))) {name=Scanline_ToEmissive, nodeColor=<rgba 0 0.7656 1 1>}
  NamedRerouteDeclaration(None=FunctionInput(Preview=ScreenPosition().ViewportUV) {bUsePreviewValueAsDefault=true, inputName=UVForScanline, inputType=FunctionInput_Vector2}) {name=ScanlineUV, nodeColor=<rgba 1 0.9141 0 1>}
  NamedRerouteDeclaration(None=FunctionInput(Preview=TextureCoordinate()) {bUsePreviewValueAsDefault=true, inputName=UVForStroke, inputType=FunctionInput_Vector2, sortPriority=1}) {name=StrokeUV, nodeColor=<rgba 0 0.3906 1 1>}
  NamedRerouteDeclaration(None=FunctionInput(Preview=Time() {bOverride_Period=true, period=6000}) {bUsePreviewValueAsDefault=true, inputName=Time, inputType=FunctionInput_Scalar, sortPriority=-100}) {name=Time, nodeColor=<rgba 1 0 0.8493 1>}
  NamedRerouteDeclaration(None=FunctionInput(Preview=MaterialFunctionCall() {materialFunction=/Game/Tools/3DHB/Shader/Base/MF_3DHB_Base_ScreenUVWithDistanceScale}.Result) {bUsePreviewValueAsDefault=true, inputName=UVForGlitch, inputType=FunctionInput_Vector2, sortPriority=1}) {name=GlitchUV, nodeColor=<rgba 1 0.6094 0 1>}
  NamedRerouteDeclaration(None=@n61) {name=Glitch_HueShiftOuput_1, nodeColor=<rgba 0.6875 0 1 1>}
  NamedRerouteDeclaration(None=LinearInterpolate(B=ScalarParameter() {defaultValue=0.3, group=Base_HoloFX_Glitch, parameterName=Glitch_HueshiftStrength}, Alpha=@n61)) {name=Glitch_HueShiftOuput, nodeColor=<rgba 0.6875 0 1 1>}
  NamedRerouteDeclaration(None=FunctionInput(Preview=TextureCoordinate()) {bUsePreviewValueAsDefault=true, inputName=BaseUVIn, inputType=FunctionInput_Vector2, sortPriority=1}) {name=BaseUV, nodeColor=<rgba 0.5234 0 1 1>}
  n65 = NamedRerouteUsage()
  n66 = NamedRerouteUsage()
  n67 = Multiply(A=ScalarParameter() {defaultValue=7, group=Base_HoloFX_Stroke, parameterName=Stroke_FlowSpeed}) {constB=-1}
  n68 = Add(
   A=Multiply(A=MaterialFunctionCall(Float2=@n65) {materialFunction=&BreakOutFloat2Components}.G) {constB=-10}
   B=Add(A=Multiply(A=@n66, B=@n67)) {constB=0}
  )
  n69 = Multiply(A=ScalarParameter() {defaultValue=3, group=Base_HoloFX_Stroke, parameterName=Stroke_Frequency}) {constB=0.01}
  n70 = ScalarParameter() {defaultValue=1.3, group=Base_HoloFX_Stroke, parameterName=Stroke_Thickness}
  n71 = OneMinus(None=Saturate(None=Multiply(A=@n69, B=@n70)))
  n72 = Subtract(A=Frac(None=Multiply(A=@n68, B=@n69)), B=@n71)
  n73 = Power(
   Base=MaterialFunctionCall(Input=Saturate(None=@n72), Input Low=Constant(), Input High=Subtract(B=@n71)) {materialFunction=&RemapValue_0-1}.Result
   Exp=ScalarParameter() {defaultValue=2, group=Base_HoloFX_Stroke, parameterName=Stroke_GradientTweak}
  )
  n74 = Multiply(A=@n73, B=ScalarParameter() {defaultValue=20, group=Base_HoloFX_Stroke, parameterName=Stroke_WPOStrength})
  n75 = Constant3Vector() {constant=<rgba -1 0 0 0>}
  n76 = Constant3Vector() {constant=<rgba 1 0 0 0>}
  n77 = MaterialFunctionCall(Float2=NamedRerouteUsage()) {materialFunction=&BreakOutFloat2Components}
  n78 = Transform(None=Clamp(None=LinearInterpolate(A=@n75, B=@n76, Alpha=@n77.R)) {minDefault=-1})
  n79 = Constant()
  NamedRerouteDeclaration(None=ShadingPathSwitch(Default=Multiply(A=@n74, B=@n78), Forward=@n79, Mobile=@n79)) {name=Stroke_WpoOutput, nodeColor=<rgba 1 0 0.5625 1>}
  NamedRerouteDeclaration(None=Multiply(A=@n73, B=VectorParameter() {group=Base_HoloFX_Stroke, parameterName=Stroke_Color, defaultValue=<rgba 1.37 1.04 5 1>}.RGB)) {name=Stroke_ColorOutput, nodeColor=<rgba 0 1 0.6484 1>}
  NamedRerouteDeclaration(None=@n73) {name=Stroke_RangeOutput, nodeColor=<rgba 0.0141 1 0 1>}
  NamedRerouteDeclaration(None=LinearInterpolate(B=@n45, Alpha=@n47) {constA=1}) {name=Glitch_ToScanline, nodeColor=<rgba 0.125 1 0 1>}
  n80 = MaterialFunctionCall(Float2=@n21) {materialFunction=&BreakOutFloat2Components}
  n81 = Multiply(A=@n80.G, B=ScalarParameter() {defaultValue=108.2719, group=Base_HoloFX_Scanline, parameterName=Scanline_GlitchDensity_})
  n82 = Multiply(A=Add(A=Sine(None=@n81), B=Constant() {r=-0.3}), B=Constant() {r=0.5})
  n83 = Multiply(A=@n80.G, B=ScalarParameter() {defaultValue=108.2719, group=Base_HoloFX_Scanline, parameterName=Scanline_GlitchDensity_Far})
  n84 = Multiply(A=Add(A=Sine(None=@n83), B=Constant() {r=-0.3}), B=Constant() {r=0.5})
  n85 = Constant() {r=1}
  n86 = OneMinus(None=NamedRerouteUsage())
  n87 = NamedRerouteUsage()
  n88 = Constant() {r=1}
  n89 = Add(
   A=LinearInterpolate(A=Saturate(None=@n82), B=Saturate(None=@n84), Alpha=@n37)
   B=Multiply(A=StaticSwitchParameter(True=@n85, False=@n86) {group=Base_HoloFX_Stroke, parameterName=DisableStrokeFX}, B=Multiply(A=@n87, B=@n88))
  )
  n90 = NamedRerouteUsage()
  n91 = ScalarParameter() {defaultValue=1, group=Base_HoloFX_Scanline, parameterName=Scanline_Breath_Frequency}
  n92 = ConstantBiasScale(None=Sine(None=Multiply(A=@n90, B=@n91)))
  n93 = Constant()
  n94 = Constant() {r=1}
  n95 = ScalarParameter() {defaultValue=0.8, group=Base_HoloFX_Scanline, parameterName=Scanline_Breath_DarkPercent}
  n96 = ScalarParameter() {defaultValue=1, group=Base_HoloFX_Scanline, parameterName=Scanline_Breath_BrightPercent}
  n97 = LinearInterpolate(
   A=Multiply(
    A=ScalarParameter() {defaultValue=0.2, group=Base_HoloFX_Scanline, parameterName=Scanline_BaseOpacity}
    B=MaterialFunctionCall(Input=@n92, Input Low=@n93, Input High=@n94, Target Low=@n95, Target High=@n96) {materialFunction=&RemapValueRange}.Result
   )
   Alpha=@n38
  )
  NamedRerouteDeclaration(None=Multiply(A=Saturate(None=@n89), B=Saturate(None=@n97))) {name=Scanline_ToOpacity, nodeColor=<rgba 1 0.75 0 1>}
  OneMinus(None=@n17)
  TextureCoordinate()
  TextureCoordinate()
 ]}
