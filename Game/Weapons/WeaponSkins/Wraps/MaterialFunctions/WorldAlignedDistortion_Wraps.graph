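# Material function for weapon wraps that outputs an animated distortion offset, tuned separately for weapons, vehicles and mechs.
# Despite the WorldAligned name, it triplanar-projects a noise texture from pre-skinned position, blended by pre-skinned normals, so the pattern sticks to the animated mesh.
# Scalar isVehicle and isMech inputs select tiling, pan and strength sets, with mech applied last.
# Time pans the coordinates; noise is remapped to signed values and scaled by strength.
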MaterialFunction /Game/Weapons/WeaponSkins/Wraps/MaterialFunctions/WorldAlignedDistortion_Wraps
graph {nodes [
  n1 = PreSkinnedPosition()
  n2 = Constant()
  n3 = FunctionInput(Preview=@n2) {bUsePreviewValueAsDefault=true, inputName=isVehicle, inputType=FunctionInput_Scalar}
  n4 = LinearInterpolate(
   A=FunctionInput(Preview=VectorParameter() {group=SmokeyFX, parameterName=DS_WeaponDistortionTiling, defaultValue=<rgba 6 48 16 1>}.RGB) {bUsePreviewValueAsDefault=true, inputName=WeaponDistortionTiling}
   B=FunctionInput(Preview=VectorParameter() {group=SmokeyFX, parameterName=DS_VehicleDistortionTiling, defaultValue=<rgba 32 32 32 1>}.RGB) {bUsePreviewValueAsDefault=true, inputName=VehicleDistortionTiling}
   Alpha=@n3
  )
  n5 = FunctionInput(Preview=VectorParameter() {group=SmokeyFX, parameterName=SN_MechDistortionTiling, defaultValue=<rgba 32 32 32 1>}.RGB) {bUsePreviewValueAsDefault=true, inputName=MechDistortionTiling}
  n6 = FunctionInput(Preview=@n2) {bUsePreviewValueAsDefault=true, inputName=isMech, inputType=FunctionInput_Scalar}
  n7 = LinearInterpolate(
   A=FunctionInput(Preview=VectorParameter() {group=SmokeyFX, parameterName=SN_WeaponDistortionPanSpeedAndDirection, defaultValue=<rgba 0.5 0.5 0 1>}.RGB) {bUsePreviewValueAsDefault=true, inputName=WeaponDistortionPanSpeedAndDirection}
   B=FunctionInput(Preview=VectorParameter() {group=SmokeyFX, parameterName=SN_VehicleDistortionPanSpeedAndDirection, defaultValue=<rgba 0 -0.5 0 1>}.RGB) {bUsePreviewValueAsDefault=true, inputName=VehicleDistortionPanSpeedAndDirection}
   Alpha=@n3
  )
  n8 = FunctionInput(Preview=VectorParameter() {group=SmokeyFX, parameterName=SN_MechDistortionPanSpeedAndDirection, defaultValue=<rgba 0 0 -0.5 1>}.RGB) {bUsePreviewValueAsDefault=true, inputName=MechDistortionPanSpeedAndDirection}
  n9 = Add(
   A=Divide(A=VertexInterpolator(VS=@n1).PS, B=LinearInterpolate(A=@n4, B=@n5, Alpha=@n6))
   B=Multiply(A=LinearInterpolate(A=@n7, B=@n8, Alpha=@n6), B=Time())
  )
  n10 = FunctionInput(Preview=TextureObjectParameter() {parameterName=DistortionNoise, texture=/Game/Characters/Player/Male/Medium/Bodies/M_MED_Viper/Textures/Glitch/T_SoftNoise}) {bUsePreviewValueAsDefault=true, inputName=DistortionNoise, inputType=FunctionInput_Texture2D}
  n11 = TextureSample(UVs=ComponentMask(None=@n9) {g=true, r=true}, Tex=@n10)
  n12 = TextureSample(UVs=ComponentMask(None=@n9) {b=true, r=true}, Tex=@n10)
  n13 = VertexInterpolator(VS=PreSkinnedNormal())
  n14 = Abs(None=ComponentMask(None=@n13.PS) {g=true})
  n15 = ComponentMask(None=@n9) {b=true, g=true}
  n16 = ComponentMask(None=@n13.PS) {r=true}
  n17 = Multiply(A=LinearInterpolate(A=LinearInterpolate(A=@n11.RGB, B=@n12.RGB, Alpha=@n14), B=TextureSample(UVs=@n15, Tex=@n10).RGB, Alpha=Abs(None=@n16))) {constB=2}
  n18 = LinearInterpolate(
   A=FunctionInput(Preview=VectorParameter() {group=SmokeyFX, parameterName=SN_WeaponDistortionStrength, defaultValue=<rgba 0.25 0.25 0.25 1>}.RGB) {bUsePreviewValueAsDefault=true, inputName=WeaponDistortionStrength}
   B=FunctionInput(Preview=VectorParameter() {group=SmokeyFX, parameterName=SN_VehicleDistortionStrength, defaultValue=<rgba 0.25 0.25 0.25 1>}.RGB) {bUsePreviewValueAsDefault=true, inputName=VehicleDistortionStrength}
   Alpha=@n3
  )
  n19 = FunctionInput(Preview=VectorParameter() {group=SmokeyFX, parameterName=SN_MechDistortionStrength, defaultValue=<rgba 0.25 0.25 0.25 1>}.RGB) {bUsePreviewValueAsDefault=true, inputName=MechDistortionStrength}
  FunctionOutput(None=Multiply(A=Subtract(A=@n17), B=LinearInterpolate(A=@n18, B=@n19, Alpha=@n6)))
 ]}
